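# Sound cue for one-shot interior metal distress creaks in the airborne city ambience, using stereo blimp metal distress recordings.
# A Random node picks one of six equally weighted waves and feeds a Modulator that varies pitch.
# It plays through the indoor ambient one-shot sound class with a distant-creak attenuation, with bus and submix sends disabled.
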SoundCue /AmbientAudioContent/SoundCues/OneShots/Spot_AirborneCity_Metal_Distress_Int_Cue
root=Modulator_0
nodes [
 {id=Modulator_0, children [Random_0], properties {pitchMin=0.9}}
 {id=Random_0, children [WavePlayer_0,WavePlayer_1,WavePlayer_2,WavePlayer_3,WavePlayer_4,WavePlayer_5], properties {weights [1,1,1,1,1,1]}}
 {id=WavePlayer_0, properties {soundWaveAssetPtr=/AmbientAudioContent/Waves/OneShots/Spot_Blimp_Metal_Distress_01_Stereo}}
 {id=WavePlayer_1, properties {soundWaveAssetPtr=/AmbientAudioContent/Waves/OneShots/Spot_Blimp_Metal_Distress_02_Stereo}}
 {id=WavePlayer_2, properties {soundWaveAssetPtr=/AmbientAudioContent/Waves/OneShots/Spot_Blimp_Metal_Distress_03_Stereo}}
 {id=WavePlayer_3, properties {soundWaveAssetPtr=/AmbientAudioContent/Waves/OneShots/Spot_Blimp_Metal_Distress_04_Stereo}}
 {id=WavePlayer_4, properties {soundWaveAssetPtr=/AmbientAudioContent/Waves/OneShots/Spot_Blimp_Metal_Distress_05_Stereo}}
 {id=WavePlayer_5, properties {soundWaveAssetPtr=/AmbientAudioContent/Waves/OneShots/Spot_Blimp_Metal_Distress_06_Stereo}}
]
properties {
 attenuationSettings=/Game/Sounds/Attenuation/Fort_TileAmbient_OneShot_Distant_Creak_Att
 bEnableBusSends=false
 bEnableSubmixSends=false
 maxDistance=8200
 soundClassObject=/Game/Sounds/FortSoundClassesAndModes/Fort_Ambient_OneShots_Indoors
 volumeMultiplier=1.002
 attenuationOverrides {falloffDistance=6000, stereoSpread=10000, attenuationShapeExtents {x=2000}}
}
