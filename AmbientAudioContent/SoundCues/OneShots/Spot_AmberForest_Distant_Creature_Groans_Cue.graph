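# Sound cue for Amber Forest ambient one-shot spots, playing distant creature groans.
# An equally weighted Random node picks one of seven groan waves, which are named for Cattus Cave, and feeds a Modulator at the root.
# The cue plays in the ambient one-shots sound class at reduced volume, with a shared closer-range tile-ambient attenuation asset assigned.
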SoundCue /AmbientAudioContent/SoundCues/OneShots/Spot_AmberForest_Distant_Creature_Groans_Cue
root=Modulator_0
nodes [
 {id=Modulator_0, children [Random_1]}
 {
  id=Random_1
  children [WavePlayer_7, WavePlayer_8, WavePlayer_9, WavePlayer_10, WavePlayer_11, WavePlayer_12, WavePlayer_13]
  properties {weights [1,1,1,1,1,1,1]}
 }
 {id=WavePlayer_10, properties {soundWaveAssetPtr=/AmbientAudioContent/Waves/OneShots/Spot_Cattus_Cave_Distant_Creature_Groan_04}}
 {id=WavePlayer_11, properties {soundWaveAssetPtr=/AmbientAudioContent/Waves/OneShots/Spot_Cattus_Cave_Distant_Creature_Groan_05}}
 {id=WavePlayer_12, properties {soundWaveAssetPtr=/AmbientAudioContent/Waves/OneShots/Spot_Cattus_Cave_Distant_Creature_Groan_06}}
 {id=WavePlayer_13, properties {soundWaveAssetPtr=/AmbientAudioContent/Waves/OneShots/Spot_Cattus_Cave_Distant_Creature_Groan_07}}
 {id=WavePlayer_7, properties {soundWaveAssetPtr=/AmbientAudioContent/Waves/OneShots/Spot_Cattus_Cave_Distant_Creature_Groan_01}}
 {id=WavePlayer_8, properties {soundWaveAssetPtr=/AmbientAudioContent/Waves/OneShots/Spot_Cattus_Cave_Distant_Creature_Groan_02}}
 {id=WavePlayer_9, properties {soundWaveAssetPtr=/AmbientAudioContent/Waves/OneShots/Spot_Cattus_Cave_Distant_Creature_Groan_03}}
]
properties {
 attenuationSettings=/Game/Sounds/Attenuation/Fort_TileAmbient_OneShot_Att_Closer
 maxDistance=8500
 soundClassObject=&Fort_Ambient_OneShots
 volumeMultiplier=0.66
 attenuationOverrides {bAttenuate=false}
}
