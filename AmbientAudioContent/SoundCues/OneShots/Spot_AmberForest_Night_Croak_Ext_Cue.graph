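# Sound cue for a one-shot outdoor ambient alien croak, named for the Amber Forest at night.
# A quality-level node feeds the first two levels from nine equally weighted random croaks and the last two from only the first four.
# Each pool passes through a modulator for pitch and volume variation.
# The cue lowers overall pitch and volume, adds wide one-shot attenuation, a Chrome Animal effect chain and a distant-reverb send.
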SoundCue /AmbientAudioContent/SoundCues/OneShots/Spot_AmberForest_Night_Croak_Ext_Cue
root=QualityLevel_0
nodes [
 {id=Modulator_1, children [Random_0], properties {pitchMax=1.1, volumeMax=1.1, volumeMin=0.9}}
 {id=Modulator_2, children [Random_3], properties {pitchMax=1.1, volumeMax=1.1, volumeMin=0.9}}
 {id=QualityLevel_0, children [Modulator_1,Modulator_1,Modulator_2,Modulator_2]}
 {
  id=Random_0
  children [WavePlayer_7, WavePlayer_14, WavePlayer_15, WavePlayer_16, WavePlayer_17, WavePlayer_18, WavePlayer_19, WavePlayer_20, WavePlayer_21]
  properties {weights [1,1,1,1,1,1,1,1,1]}
 }
 {id=Random_3, children [WavePlayer_10, WavePlayer_11, WavePlayer_12, WavePlayer_13], properties {weights [1,1,1,1]}}
 {id=WavePlayer_10, properties {soundWaveAssetPtr=/AmbientAudioContent/Waves/OneShots/Spot_Biome_Reality_Forest_SS_AlienCroak_01}}
 {id=WavePlayer_11, properties {soundWaveAssetPtr=/AmbientAudioContent/Waves/OneShots/Spot_Biome_Reality_Forest_SS_AlienCroak_02}}
 {id=WavePlayer_12, properties {soundWaveAssetPtr=/AmbientAudioContent/Waves/OneShots/Spot_Biome_Reality_Forest_SS_AlienCroak_03}}
 {id=WavePlayer_13, properties {soundWaveAssetPtr=/AmbientAudioContent/Waves/OneShots/Spot_Biome_Reality_Forest_SS_AlienCroak_04}}
 {id=WavePlayer_14, properties {soundWaveAssetPtr=&Spot_Biome_Reality_Forest_SS_AlienCroak_02}}
 {id=WavePlayer_15, properties {soundWaveAssetPtr=&Spot_Biome_Reality_Forest_SS_AlienCroak_03}}
 {id=WavePlayer_16, properties {soundWaveAssetPtr=&Spot_Biome_Reality_Forest_SS_AlienCroak_04}}
 {id=WavePlayer_17, properties {soundWaveAssetPtr=/AmbientAudioContent/Waves/OneShots/Spot_Biome_Reality_Forest_SS_AlienCroak_05}}
 {id=WavePlayer_18, properties {soundWaveAssetPtr=/AmbientAudioContent/Waves/OneShots/Spot_Biome_Reality_Forest_SS_AlienCroak_06}}
 {id=WavePlayer_19, properties {soundWaveAssetPtr=/AmbientAudioContent/Waves/OneShots/Spot_Biome_Reality_Forest_SS_AlienCroak_07}}
 {id=WavePlayer_20, properties {soundWaveAssetPtr=/AmbientAudioContent/Waves/OneShots/Spot_Biome_Reality_Forest_SS_AlienCroak_08}}
 {id=WavePlayer_21, properties {soundWaveAssetPtr=/AmbientAudioContent/Waves/OneShots/Spot_Biome_Reality_Forest_SS_AlienCroak_09}}
 {id=WavePlayer_7, properties {soundWaveAssetPtr=&Spot_Biome_Reality_Forest_SS_AlienCroak_01}}
]
properties {
 attenuationSettings=/Game/Sounds/Attenuation/Fort_TileAmbient_OneShot_Wider_Att
 maxDistance=8000
 pitchMultiplier=0.8
 soundClassObject=&Fort_Ambient_OneShots_Outdoors
 sourceEffectChain=/Game/Sounds/Artemis/POI/HerraldsFortress/Spots/Spot_Chrome_Animal_FXChain
 volumeMultiplier=0.501
 preEffectBusSends [{}]
 soundSubmixSends [{maxSendLevel=0.5, minSendLevel=0.2, sendLevel=0, soundSubmix=&Fortnite_Reverb_Distant}]
}
