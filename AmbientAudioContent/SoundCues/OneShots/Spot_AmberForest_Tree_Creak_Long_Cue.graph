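# Sound cue for long tree creak one-shots in the Amber Forest outdoor ambience, using a distant-creak attenuation and reduced volume.
# A Random node picks among sixteen equally weighted creak waves and feeds a Modulator that varies volume widely and pitch mostly downward.
# A Quality Level node routes all four inputs to that same branch, so every level sounds alike.
# A root Modulator adds a narrower pitch and volume variation on top.
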SoundCue /AmbientAudioContent/SoundCues/OneShots/Spot_AmberForest_Tree_Creak_Long_Cue
root=Modulator_13
nodes [
 {id=Modulator_1, children [Random_2], properties {pitchMax=1, pitchMin=0.7, volumeMax=1.1, volumeMin=0.5}}
 {id=Modulator_13, children [QualityLevel_0], properties {pitchMax=1.1, pitchMin=0.9, volumeMax=1.1, volumeMin=0.9}}
 {id=QualityLevel_0, children [Modulator_1,Modulator_1,Modulator_1,Modulator_1]}
 {
  id=Random_2
  children [
   WavePlayer_20
   WavePlayer_21
   WavePlayer_22
   WavePlayer_23
   WavePlayer_24
   WavePlayer_25
   WavePlayer_26
   WavePlayer_27
   WavePlayer_28
   WavePlayer_29
   WavePlayer_30
   WavePlayer_31
   WavePlayer_32
   WavePlayer_33
   WavePlayer_34
   WavePlayer_35
  ]
  properties {weights [1,1,1,1,1,1,1,1,1,1,1,1,1,1,1,1]}
 }
 {id=WavePlayer_20, properties {soundWaveAssetPtr=/AmbientAudioContent/Waves/OneShots/Spot_TreeCreakLong001}}
 {id=WavePlayer_21, properties {soundWaveAssetPtr=/AmbientAudioContent/Waves/OneShots/Spot_TreeCreakLong002}}
 {id=WavePlayer_22, properties {soundWaveAssetPtr=/AmbientAudioContent/Waves/OneShots/Spot_TreeCreakLong003}}
 {id=WavePlayer_23, properties {soundWaveAssetPtr=/AmbientAudioContent/Waves/OneShots/Spot_TreeCreakLong004}}
 {id=WavePlayer_24, properties {soundWaveAssetPtr=/AmbientAudioContent/Waves/OneShots/Spot_TreeCreakLong005}}
 {id=WavePlayer_25, properties {soundWaveAssetPtr=/AmbientAudioContent/Waves/OneShots/Spot_TreeCreakLong006}}
 {id=WavePlayer_26, properties {soundWaveAssetPtr=/AmbientAudioContent/Waves/OneShots/Spot_TreeCreakLong007}}
 {id=WavePlayer_27, properties {soundWaveAssetPtr=/AmbientAudioContent/Waves/OneShots/Spot_TreeCreakLong008}}
 {id=WavePlayer_28, properties {soundWaveAssetPtr=/AmbientAudioContent/Waves/OneShots/Spot_TreeCreakLong009}}
 {id=WavePlayer_29, properties {soundWaveAssetPtr=/AmbientAudioContent/Waves/OneShots/Spot_TreeCreakLong010}}
 {id=WavePlayer_30, properties {soundWaveAssetPtr=/AmbientAudioContent/Waves/OneShots/Spot_TreeCreakLong011}}
 {id=WavePlayer_31, properties {soundWaveAssetPtr=/AmbientAudioContent/Waves/OneShots/Spot_TreeCreakLong012}}
 {id=WavePlayer_32, properties {soundWaveAssetPtr=/AmbientAudioContent/Waves/OneShots/Spot_TreeCreakLong013}}
 {id=WavePlayer_33, properties {soundWaveAssetPtr=/AmbientAudioContent/Waves/OneShots/Spot_TreeCreakLong014}}
 {id=WavePlayer_34, properties {soundWaveAssetPtr=/AmbientAudioContent/Waves/OneShots/Spot_TreeCreakLong015}}
 {id=WavePlayer_35, properties {soundWaveAssetPtr=/AmbientAudioContent/Waves/OneShots/Spot_TreeCreakLong016}}
]
properties {
 attenuationSettings=/Game/Sounds/Attenuation/Fort_TileAmbient_OneShot_Distant_Creak_Att
 bEnableBusSends=false
 maxDistance=8200
 soundClassObject=&Fort_Ambient_OneShots_Outdoors
 volumeMultiplier=0.35
 preEffectBusSends [{}]
}
