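# One-shot ambient spot Sound Cue named for bamboo forest insects, though every wave it references is a woods bird sound.
# A Random node picks among fourteen equally weighted Woods_Bird waves and feeds a root Modulator that varies pitch and volume on each play.
# The cue uses shared tile-ambient one-shot attenuation, the outdoor ambient one-shots sound class and a reduced overall volume.
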SoundCue /AmbientAudioContent/SoundCues/OneShots/Spot_BambooForest_Insects_Cue
root=Modulator_0
nodes [
 {id=Modulator_0, children [Random_0], properties {pitchMax=1.2, pitchMin=0.8, volumeMax=1.2, volumeMin=0.8}}
 {
  id=Random_0
  children [
   WavePlayer_0
   WavePlayer_1
   WavePlayer_2
   WavePlayer_3
   WavePlayer_4
   WavePlayer_5
   WavePlayer_7
   WavePlayer_8
   WavePlayer_10
   WavePlayer_15
   WavePlayer_16
   WavePlayer_17
   WavePlayer_18
   WavePlayer_19
  ]
  properties {weights [1,1,1,1,1,1,1,1,1,1,1,1,1,1]}
 }
 {id=WavePlayer_0, properties {soundWaveAssetPtr=/AmbientAudioContent/Waves/OneShots/Spot_Biome_Woods_Bird_14}}
 {id=WavePlayer_1, properties {soundWaveAssetPtr=/AmbientAudioContent/Waves/OneShots/Spot_Biome_Woods_Bird_13}}
 {id=WavePlayer_10, properties {soundWaveAssetPtr=/AmbientAudioContent/Waves/OneShots/Spot_Biome_Woods_Bird_06}}
 {id=WavePlayer_15, properties {soundWaveAssetPtr=/AmbientAudioContent/Waves/OneShots/Spot_Biome_Woods_Bird_05}}
 {id=WavePlayer_16, properties {soundWaveAssetPtr=/AmbientAudioContent/Waves/OneShots/Spot_Biome_Woods_Bird_04}}
 {id=WavePlayer_17, properties {soundWaveAssetPtr=/AmbientAudioContent/Waves/OneShots/Spot_Biome_Woods_Bird_03}}
 {id=WavePlayer_18, properties {soundWaveAssetPtr=/AmbientAudioContent/Waves/OneShots/Spot_Biome_Woods_Bird_02}}
 {id=WavePlayer_19, properties {soundWaveAssetPtr=/AmbientAudioContent/Waves/OneShots/Spot_Biome_Woods_Bird_01}}
 {id=WavePlayer_2, properties {soundWaveAssetPtr=/AmbientAudioContent/Waves/OneShots/Spot_Biome_Woods_Bird_12}}
 {id=WavePlayer_3, properties {soundWaveAssetPtr=/AmbientAudioContent/Waves/OneShots/Spot_Biome_Woods_Bird_11}}
 {id=WavePlayer_4, properties {soundWaveAssetPtr=/AmbientAudioContent/Waves/OneShots/Spot_Biome_Woods_Bird_10}}
 {id=WavePlayer_5, properties {soundWaveAssetPtr=/AmbientAudioContent/Waves/OneShots/Spot_Biome_Woods_Bird_09}}
 {id=WavePlayer_7, properties {soundWaveAssetPtr=/AmbientAudioContent/Waves/OneShots/Spot_Biome_Woods_Bird_08}}
 {id=WavePlayer_8, properties {soundWaveAssetPtr=/AmbientAudioContent/Waves/OneShots/Spot_Biome_Woods_Bird_07}}
]
properties {
 attenuationSettings=/Game/Sounds/Attenuation/Fort_TileAmbient_OneShot_Att
 maxDistance=8000
 soundClassObject=&Fort_Ambient_OneShots_Outdoors
 volumeMultiplier=0.551
 preEffectBusSends [{}]
}
