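# Sound cue for daytime wind gust one-shots in the bamboo forest ambience, reusing Woods biome wind gust recordings.
# A Random node picks one of eight equally weighted wave players, and a Modulator varies pitch upward and volume around nominal on each play.
# Output is turned well down, routed to the outdoor ambient one-shots sound class, and spatialized by a shared tile-ambient one-shot attenuation asset.
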SoundCue /AmbientAudioContent/SoundCues/OneShots/Spot_BambooForest_WindGustDay_Cue
root=Modulator_0
nodes [
 {id=Modulator_0, children [Random_1], properties {pitchMax=1.2, pitchMin=1, volumeMax=1.2, volumeMin=0.8}}
 {
  id=Random_1
  children [WavePlayer_0, WavePlayer_6, WavePlayer_7, WavePlayer_8, WavePlayer_9, WavePlayer_10, WavePlayer_11, WavePlayer_12]
  properties {weights [1,1,1,1,1,1,1,1]}
 }
 {id=WavePlayer_0, properties {soundWaveAssetPtr=/AmbientAudioContent/Waves/OneShots/Spot_Biome_Woods_WindGustDay_01}}
 {id=WavePlayer_10, properties {soundWaveAssetPtr=/AmbientAudioContent/Waves/OneShots/Spot_Biome_Woods_WindGustDay_06}}
 {id=WavePlayer_11, properties {soundWaveAssetPtr=/AmbientAudioContent/Waves/OneShots/Spot_Biome_Woods_WindGustDay_07}}
 {id=WavePlayer_12, properties {soundWaveAssetPtr=/AmbientAudioContent/Waves/OneShots/Spot_Biome_Woods_WindGustDay_08}}
 {id=WavePlayer_6, properties {soundWaveAssetPtr=/AmbientAudioContent/Waves/OneShots/Spot_Biome_Woods_WindGustDay_02}}
 {id=WavePlayer_7, properties {soundWaveAssetPtr=/AmbientAudioContent/Waves/OneShots/Spot_Biome_Woods_WindGustDay_03}}
 {id=WavePlayer_8, properties {soundWaveAssetPtr=/AmbientAudioContent/Waves/OneShots/Spot_Biome_Woods_WindGustDay_04}}
 {id=WavePlayer_9, properties {soundWaveAssetPtr=/AmbientAudioContent/Waves/OneShots/Spot_Biome_Woods_WindGustDay_05}}
]
properties {
 attenuationSettings=/Game/Sounds/Attenuation/Fort_TileAmbient_OneShot_Att
 maxDistance=8000
 soundClassObject=&Fort_Ambient_OneShots_Outdoors
 volumeMultiplier=0.251
 preEffectBusSends [{}]
}
