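# One-shot ambient SoundCue for a bamboo forest wind rustle, built from Grassland wind rustle waves and played as an outdoor ambient sound with wind attenuation.
# A quality-level root sends its first two levels to a random pick of ten waves and its last two to a six-wave subset.
# Each branch passes through a modulator that varies pitch and volume slightly.
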SoundCue /AmbientAudioContent/SoundCues/OneShots/Spot_BambooForest_Wind_Rustle_03_Cue
root=QualityLevel_0
nodes [
 {id=Modulator_1, children [Random_3], properties {pitchMax=1.1, pitchMin=1, volumeMax=1.1, volumeMin=0.9}}
 {id=Modulator_2, children [Random_4], properties {pitchMax=1.1, volumeMax=1.1, volumeMin=0.9}}
 {id=QualityLevel_0, children [Modulator_1,Modulator_1,Modulator_2,Modulator_2]}
 {
  id=Random_3
  children [
   WavePlayer_36
   WavePlayer_37
   WavePlayer_38
   WavePlayer_39
   WavePlayer_40
   WavePlayer_41
   WavePlayer_42
   WavePlayer_43
   WavePlayer_44
   WavePlayer_45
  ]
  properties {weights [1,1,1,1,1,1,1,1,1,1]}
 }
 {
  id=Random_4
  children [WavePlayer_46, WavePlayer_47, WavePlayer_48, WavePlayer_49, WavePlayer_50, WavePlayer_51]
  properties {weights [1,1,1,1,1,1]}
 }
 {id=WavePlayer_36, properties {soundWaveAssetPtr=/AmbientAudioContent/Waves/OneShots/Spot_Grassland_Wind_Rustle_01}}
 {id=WavePlayer_37, properties {soundWaveAssetPtr=/AmbientAudioContent/Waves/OneShots/Spot_Grassland_Wind_Rustle_02}}
 {id=WavePlayer_38, properties {soundWaveAssetPtr=/AmbientAudioContent/Waves/OneShots/Spot_Grassland_Wind_Rustle_03}}
 {id=WavePlayer_39, properties {soundWaveAssetPtr=/AmbientAudioContent/Waves/OneShots/Spot_Grassland_Wind_Rustle_04}}
 {id=WavePlayer_40, properties {soundWaveAssetPtr=/AmbientAudioContent/Waves/OneShots/Spot_Grassland_Wind_Rustle_05}}
 {id=WavePlayer_41, properties {soundWaveAssetPtr=/AmbientAudioContent/Waves/OneShots/Spot_Grassland_Wind_Rustle_06}}
 {id=WavePlayer_42, properties {soundWaveAssetPtr=/AmbientAudioContent/Waves/OneShots/Spot_Grassland_Wind_Rustle_07}}
 {id=WavePlayer_43, properties {soundWaveAssetPtr=/AmbientAudioContent/Waves/OneShots/Spot_Grassland_Wind_Rustle_08}}
 {id=WavePlayer_44, properties {soundWaveAssetPtr=/AmbientAudioContent/Waves/OneShots/Spot_Grassland_Wind_Rustle_10}}
 {id=WavePlayer_45, properties {soundWaveAssetPtr=/AmbientAudioContent/Waves/OneShots/Spot_Grassland_Wind_Rustle_09}}
 {id=WavePlayer_46, properties {soundWaveAssetPtr=&Spot_Grassland_Wind_Rustle_03}}
 {id=WavePlayer_47, properties {soundWaveAssetPtr=&Spot_Grassland_Wind_Rustle_04}}
 {id=WavePlayer_48, properties {soundWaveAssetPtr=&Spot_Grassland_Wind_Rustle_05}}
 {id=WavePlayer_49, properties {soundWaveAssetPtr=&Spot_Grassland_Wind_Rustle_06}}
 {id=WavePlayer_50, properties {soundWaveAssetPtr=&Spot_Grassland_Wind_Rustle_07}}
 {id=WavePlayer_51, properties {soundWaveAssetPtr=&Spot_Grassland_Wind_Rustle_08}}
]
properties {
 attenuationSettings=/Game/Sounds/Attenuation/Fort_TileAmbient_OneShot_Wind_Att
 bEnableBusSends=false
 maxDistance=8000
 soundClassObject=&Fort_Ambient_OneShots_Outdoors
 volumeMultiplier=0.376
 preEffectBusSends [{}]
}
