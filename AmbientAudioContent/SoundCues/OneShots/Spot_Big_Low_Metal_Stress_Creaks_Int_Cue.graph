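# Sound cue for an ambient one-shot of big, low metal stress creaks, named as interior but assigned to the outdoor ambient sound class.
# A Random node picks among twelve equally weighted knock and stretch groan waves, feeding a Modulator for pitch variation.
# A very low cue pitch multiplier deepens the recordings.
# Overridden attenuation adds long falloff and muffled filtering, and the exterior source bus send is set to zero.
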SoundCue /AmbientAudioContent/SoundCues/OneShots/Spot_Big_Low_Metal_Stress_Creaks_Int_Cue
root=Modulator_0
nodes [
 {id=Modulator_0, children [Random_1], properties {pitchMin=0.9}}
 {
  id=Random_1
  children [
   WavePlayer_8
   WavePlayer_9
   WavePlayer_10
   WavePlayer_11
   WavePlayer_12
   WavePlayer_17
   WavePlayer_18
   WavePlayer_19
   WavePlayer_20
   WavePlayer_21
   WavePlayer_22
   WavePlayer_23
  ]
  properties {weights [1,1,1,1,1,1,1,1,1,1,1,1]}
 }
 {id=WavePlayer_10, properties {soundWaveAssetPtr=/AmbientAudioContent/Waves/OneShots/Spot_Metal_Stress_Wronk_Groan_Knock_03}}
 {id=WavePlayer_11, properties {soundWaveAssetPtr=/AmbientAudioContent/Waves/OneShots/Spot_Metal_Stress_Wronk_Groan_Knock_04}}
 {id=WavePlayer_12, properties {soundWaveAssetPtr=/AmbientAudioContent/Waves/OneShots/Spot_Metal_Stress_Wronk_Groan_Knock_05}}
 {id=WavePlayer_17, properties {soundWaveAssetPtr=/AmbientAudioContent/Waves/OneShots/Spot_Metal_Stress_Wronk_Groan_Knock_06}}
 {id=WavePlayer_18, properties {soundWaveAssetPtr=/AmbientAudioContent/Waves/OneShots/Spot_Metal_Stress_Wronk_Groan_Knock_07}}
 {id=WavePlayer_19, properties {soundWaveAssetPtr=/AmbientAudioContent/Waves/OneShots/Spot_Metal_Stress_Wronk_Groan_Knock_08}}
 {id=WavePlayer_20, properties {soundWaveAssetPtr=/AmbientAudioContent/Waves/OneShots/Spot_Metal_Stress_Wronk_Groan_Knock_09}}
 {id=WavePlayer_21, properties {soundWaveAssetPtr=/AmbientAudioContent/Waves/OneShots/Spot_Metal_Stress_Wronk_Groan_Stretch_02}}
 {id=WavePlayer_22, properties {soundWaveAssetPtr=/AmbientAudioContent/Waves/OneShots/Spot_Metal_Stress_Wronk_Groan_Stretch_03}}
 {id=WavePlayer_23, properties {soundWaveAssetPtr=/AmbientAudioContent/Waves/OneShots/Spot_Metal_Stress_Wronk_Groan_Stretch_04}}
 {id=WavePlayer_8, properties {soundWaveAssetPtr=/AmbientAudioContent/Waves/OneShots/Spot_Metal_Stress_Wronk_Groan_Knock_01}}
 {id=WavePlayer_9, properties {soundWaveAssetPtr=/AmbientAudioContent/Waves/OneShots/Spot_Metal_Stress_Wronk_Groan_Knock_02}}
]
properties {
 attenuationSettings=/Game/Sounds/Attenuation/Fort_TileAmbient_OneShot_Att
 bOverrideAttenuation=true
 maxDistance=8400
 pitchMultiplier=0.2
 soundClassObject=&Fort_Ambient_Outdoors_Athena
 volumeMultiplier=0.501
 attenuationOverrides {
  falloffDistance=8000
  hPFFrequencyAtMax=100
  hPFFrequencyAtMin=20
  lPFFrequencyAtMax=500
  lPFFrequencyAtMin=400
  lPFRadiusMax=500
  lPFRadiusMin=400
  occlusionLowPassFilterFrequency=500
 }
 busSends [{audioBus=null, minSendLevel=0, sendLevel=0, soundSourceBus=&AmbientAudio_Exterior_SourceBus}]
}
