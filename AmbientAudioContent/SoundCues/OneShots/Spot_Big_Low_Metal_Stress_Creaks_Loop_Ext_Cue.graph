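# Looping exterior ambient SoundCue for big, low metal stress creaks, in the Athena outdoor ambience sound class and sending to the exterior source bus.
# The first quality slot, the only one filled, loops a 6–8 second random delay into a modulated, equally weighted pick of twelve knock and stretch groans.
# A very low pitch multiplier deepens the waves.
# Overridden attenuation adds low-pass filtering and listener focus, with concurrency capped at two.
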SoundCue /AmbientAudioContent/SoundCues/OneShots/Spot_Big_Low_Metal_Stress_Creaks_Loop_Ext_Cue
root=QualityLevel_0
nodes [
 {id=Delay_0, children [Modulator_0], properties {delayMax=8, delayMin=6}}
 {id=Looping_0, children [Delay_0]}
 {id=Modulator_0, children [Random_1], properties {pitchMin=0.9}}
 {id=QualityLevel_0, children [Looping_0,null,null,null]}
 {
  id=Random_1
  children [
   WavePlayer_8
   WavePlayer_9
   WavePlayer_10
   WavePlayer_11
   WavePlayer_12
   WavePlayer_17
   WavePlayer_18
   WavePlayer_19
   WavePlayer_20
   WavePlayer_21
   WavePlayer_22
   WavePlayer_23
  ]
  properties {weights [1,1,1,1,1,1,1,1,1,1,1,1]}
 }
 {id=WavePlayer_10, properties {soundWaveAssetPtr=/AmbientAudioContent/Waves/OneShots/Spot_Metal_Stress_Wronk_Groan_Knock_03}}
 {id=WavePlayer_11, properties {soundWaveAssetPtr=/AmbientAudioContent/Waves/OneShots/Spot_Metal_Stress_Wronk_Groan_Knock_04}}
 {id=WavePlayer_12, properties {soundWaveAssetPtr=/AmbientAudioContent/Waves/OneShots/Spot_Metal_Stress_Wronk_Groan_Knock_05}}
 {id=WavePlayer_17, properties {soundWaveAssetPtr=/AmbientAudioContent/Waves/OneShots/Spot_Metal_Stress_Wronk_Groan_Knock_06}}
 {id=WavePlayer_18, properties {soundWaveAssetPtr=/AmbientAudioContent/Waves/OneShots/Spot_Metal_Stress_Wronk_Groan_Knock_07}}
 {id=WavePlayer_19, properties {soundWaveAssetPtr=/AmbientAudioContent/Waves/OneShots/Spot_Metal_Stress_Wronk_Groan_Knock_08}}
 {id=WavePlayer_20, properties {soundWaveAssetPtr=/AmbientAudioContent/Waves/OneShots/Spot_Metal_Stress_Wronk_Groan_Knock_09}}
 {id=WavePlayer_21, properties {soundWaveAssetPtr=/AmbientAudioContent/Waves/OneShots/Spot_Metal_Stress_Wronk_Groan_Stretch_02}}
 {id=WavePlayer_22, properties {soundWaveAssetPtr=/AmbientAudioContent/Waves/OneShots/Spot_Metal_Stress_Wronk_Groan_Stretch_03}}
 {id=WavePlayer_23, properties {soundWaveAssetPtr=/AmbientAudioContent/Waves/OneShots/Spot_Metal_Stress_Wronk_Groan_Stretch_04}}
 {id=WavePlayer_8, properties {soundWaveAssetPtr=/AmbientAudioContent/Waves/OneShots/Spot_Metal_Stress_Wronk_Groan_Knock_01}}
 {id=WavePlayer_9, properties {soundWaveAssetPtr=/AmbientAudioContent/Waves/OneShots/Spot_Metal_Stress_Wronk_Groan_Knock_02}}
]
properties {
 bOverrideAttenuation=true
 bOverrideConcurrency=true
 maxDistance=4000
 pitchMultiplier=0.2
 soundClassObject=&Fort_Ambient_Outdoors_Athena
 volumeMultiplier=0.4
 attenuationOverrides {
  bEnableListenerFocus=true
  falloffDistance=3000
  hPFFrequencyAtMax=100
  hPFFrequencyAtMin=20
  lPFFrequencyAtMax=500
  lPFFrequencyAtMin=400
  lPFRadiusMax=500
  lPFRadiusMin=400
  nonFocusVolumeAttenuation=0.5
  nonSpatializedRadiusEnd=1000
  nonSpatializedRadiusStart=1500
  occlusionLowPassFilterFrequency=500
  stereoSpread=800
  attenuationShapeExtents {x=1000}
 }
 busSends [{audioBus=null, minSendLevel=0, sendLevel=0, soundSourceBus=&AmbientAudio_Exterior_SourceBus}]
 concurrencyOverrides {maxCount=2}
}
