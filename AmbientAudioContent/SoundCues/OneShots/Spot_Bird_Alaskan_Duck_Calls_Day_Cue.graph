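# Sound cue for daytime Alaskan duck call one-shots, an outdoor ambient bird spot sound with reduced volume and bird attenuation.
# A quality-level root sends its first two tiers and last two tiers to separate Modulator branches, each feeding an equal-weight Random over the same five waves.
# Output uses the outdoor ambient one-shot sound class and also sends to an exterior ambient source bus.
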SoundCue /AmbientAudioContent/SoundCues/OneShots/Spot_Bird_Alaskan_Duck_Calls_Day_Cue
root=QualityLevel_0
nodes [
 {id=Modulator_0, children [Random_2]}
 {id=Modulator_3, children [Random_3]}
 {id=QualityLevel_0, children [Modulator_0,Modulator_0,Modulator_3,Modulator_3]}
 {id=Random_2, children [WavePlayer_0,WavePlayer_1,WavePlayer_2,WavePlayer_3,WavePlayer_4], properties {weights [1,1,1,1,1]}}
 {id=Random_3, children [WavePlayer_5, WavePlayer_6, WavePlayer_7, WavePlayer_18, WavePlayer_19], properties {weights [1,1,1,1,1]}}
 {id=WavePlayer_0, properties {soundWaveAssetPtr=/AmbientAudioContent/Waves/OneShots/Spot_Bird_Alaskan_Duck_Calls_Day_01}}
 {id=WavePlayer_1, properties {soundWaveAssetPtr=/AmbientAudioContent/Waves/OneShots/Spot_Bird_Alaskan_Duck_Calls_Day_02}}
 {id=WavePlayer_18, properties {soundWaveAssetPtr=/AmbientAudioContent/Waves/OneShots/Spot_Bird_Alaskan_Duck_Calls_Day_04}}
 {id=WavePlayer_19, properties {soundWaveAssetPtr=/AmbientAudioContent/Waves/OneShots/Spot_Bird_Alaskan_Duck_Calls_Day_05}}
 {id=WavePlayer_2, properties {soundWaveAssetPtr=/AmbientAudioContent/Waves/OneShots/Spot_Bird_Alaskan_Duck_Calls_Day_03}}
 {id=WavePlayer_3, properties {soundWaveAssetPtr=&Spot_Bird_Alaskan_Duck_Calls_Day_04}}
 {id=WavePlayer_4, properties {soundWaveAssetPtr=&Spot_Bird_Alaskan_Duck_Calls_Day_05}}
 {id=WavePlayer_5, properties {soundWaveAssetPtr=&Spot_Bird_Alaskan_Duck_Calls_Day_01}}
 {id=WavePlayer_6, properties {soundWaveAssetPtr=&Spot_Bird_Alaskan_Duck_Calls_Day_02}}
 {id=WavePlayer_7, properties {soundWaveAssetPtr=&Spot_Bird_Alaskan_Duck_Calls_Day_03}}
]
properties {
 attenuationSettings=&Fort_TileAmbient_OneShot_Bird_Att
 maxDistance=8005
 soundClassObject=&Fort_Ambient_OneShots_Outdoors
 volumeMultiplier=0.2
 busSends [{audioBus=null, minSendLevel=0.6, sendLevel=1, soundSourceBus=&AmbientAudio_Exterior_SourceBus}]
}
