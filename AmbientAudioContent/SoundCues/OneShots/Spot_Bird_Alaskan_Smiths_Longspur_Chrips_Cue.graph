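# Sound cue for one-shot Alaskan Smith's Longspur bird chirps in outdoor ambient audio.
# A Random node picks among nine equally weighted morning chirp waves and feeds a Modulator.
# A quality-level root routes all four of its inputs to that same Modulator chain.
# The cue plays quietly with bird one-shot attenuation, an outdoor ambient sound class and a distance-based send to the exterior ambient source bus.
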SoundCue /AmbientAudioContent/SoundCues/OneShots/Spot_Bird_Alaskan_Smiths_Longspur_Chrips_Cue
root=QualityLevel_0
nodes [
 {id=Modulator_0, children [Random_0]}
 {id=QualityLevel_0, children [Modulator_0,Modulator_0,Modulator_0,Modulator_0]}
 {
  id=Random_0
  children [WavePlayer_8, WavePlayer_9, WavePlayer_10, WavePlayer_11, WavePlayer_12, WavePlayer_13, WavePlayer_14, WavePlayer_15, WavePlayer_16]
  properties {weights [1,1,1,1,1,1,1,1,1]}
 }
 {id=WavePlayer_10, properties {soundWaveAssetPtr=/AmbientAudioContent/Waves/OneShots/Spot_Bird_Alaskan_Smiths_Longspur_Chrips_Morning_03}}
 {id=WavePlayer_11, properties {soundWaveAssetPtr=/AmbientAudioContent/Waves/OneShots/Spot_Bird_Alaskan_Smiths_Longspur_Chrips_Morning_04}}
 {id=WavePlayer_12, properties {soundWaveAssetPtr=/AmbientAudioContent/Waves/OneShots/Spot_Bird_Alaskan_Smiths_Longspur_Chrips_Morning_05}}
 {id=WavePlayer_13, properties {soundWaveAssetPtr=/AmbientAudioContent/Waves/OneShots/Spot_Bird_Alaskan_Smiths_Longspur_Chrips_Morning_06}}
 {id=WavePlayer_14, properties {soundWaveAssetPtr=/AmbientAudioContent/Waves/OneShots/Spot_Bird_Alaskan_Smiths_Longspur_Chrips_Morning_07}}
 {id=WavePlayer_15, properties {soundWaveAssetPtr=/AmbientAudioContent/Waves/OneShots/Spot_Bird_Alaskan_Smiths_Longspur_Chrips_Morning_08}}
 {id=WavePlayer_16, properties {soundWaveAssetPtr=/AmbientAudioContent/Waves/OneShots/Spot_Bird_Alaskan_Smiths_Longspur_Chrips_Morning_09}}
 {id=WavePlayer_8, properties {soundWaveAssetPtr=/AmbientAudioContent/Waves/OneShots/Spot_Bird_Alaskan_Smiths_Longspur_Chrips_Morning_01}}
 {id=WavePlayer_9, properties {soundWaveAssetPtr=/AmbientAudioContent/Waves/OneShots/Spot_Bird_Alaskan_Smiths_Longspur_Chrips_Morning_02}}
]
properties {
 attenuationSettings=&Fort_TileAmbient_OneShot_Bird_Att
 maxDistance=8005
 soundClassObject=&Fort_Ambient_OneShots_Outdoors
 volumeMultiplier=0.2
 busSends [{
   audioBus=null
   maxSendDistance=6000
   minSendDistance=3000
   minSendLevel=0.6
   sendLevel=1
   soundSourceBus=&AmbientAudio_Exterior_SourceBus
  }]
}
