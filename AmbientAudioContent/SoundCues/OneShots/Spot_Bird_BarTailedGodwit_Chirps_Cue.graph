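# Sound cue for one-shot bar-tailed godwit chirps in outdoor ambience, using bird attenuation and the outdoor one-shot sound class.
# A quality switch sends its first two levels to a random pick of ten waves and its last two to a five-wave subset, each through a default modulator.
# The smaller random node is excluded from branch culling.
# A distance-ranged send also feeds the exterior ambient source bus.
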SoundCue /AmbientAudioContent/SoundCues/OneShots/Spot_Bird_BarTailedGodwit_Chirps_Cue
root=QualityLevel_0
nodes [
 {id=Modulator_0, children [Random_2]}
 {id=Modulator_3, children [Random_3]}
 {id=QualityLevel_0, children [Modulator_0,Modulator_0,Modulator_3,Modulator_3]}
 {
  id=Random_2
  children [WavePlayer_0,WavePlayer_1,WavePlayer_2,WavePlayer_3,WavePlayer_4,WavePlayer_5,WavePlayer_6,WavePlayer_7,WavePlayer_8,WavePlayer_9]
  properties {weights [1,1,1,1,1,1,1,1,1,1]}
 }
 {
  id=Random_3
  children [WavePlayer_10, WavePlayer_11, WavePlayer_12, WavePlayer_13, WavePlayer_14]
  properties {bShouldExcludeFromBranchCulling=true, weights [1,1,1,1,1]}
 }
 {id=WavePlayer_0, properties {soundWaveAssetPtr=/AmbientAudioContent/Waves/OneShots/Spot_Bird_BarTailedGodwit_Chirps_01}}
 {id=WavePlayer_1, properties {soundWaveAssetPtr=/AmbientAudioContent/Waves/OneShots/Spot_Bird_BarTailedGodwit_Chirps_02}}
 {id=WavePlayer_10, properties {soundWaveAssetPtr=/AmbientAudioContent/Waves/OneShots/Spot_Bird_BarTailedGodwit_Chirps_06}}
 {id=WavePlayer_11, properties {soundWaveAssetPtr=/AmbientAudioContent/Waves/OneShots/Spot_Bird_BarTailedGodwit_Chirps_07}}
 {id=WavePlayer_12, properties {soundWaveAssetPtr=/AmbientAudioContent/Waves/OneShots/Spot_Bird_BarTailedGodwit_Chirps_08}}
 {id=WavePlayer_13, properties {soundWaveAssetPtr=/AmbientAudioContent/Waves/OneShots/Spot_Bird_BarTailedGodwit_Chirps_09}}
 {id=WavePlayer_14, properties {soundWaveAssetPtr=/AmbientAudioContent/Waves/OneShots/Spot_Bird_BarTailedGodwit_Chirps_10}}
 {id=WavePlayer_2, properties {soundWaveAssetPtr=/AmbientAudioContent/Waves/OneShots/Spot_Bird_BarTailedGodwit_Chirps_03}}
 {id=WavePlayer_3, properties {soundWaveAssetPtr=/AmbientAudioContent/Waves/OneShots/Spot_Bird_BarTailedGodwit_Chirps_04}}
 {id=WavePlayer_4, properties {soundWaveAssetPtr=/AmbientAudioContent/Waves/OneShots/Spot_Bird_BarTailedGodwit_Chirps_05}}
 {id=WavePlayer_5, properties {soundWaveAssetPtr=&Spot_Bird_BarTailedGodwit_Chirps_06}}
 {id=WavePlayer_6, properties {soundWaveAssetPtr=&Spot_Bird_BarTailedGodwit_Chirps_07}}
 {id=WavePlayer_7, properties {soundWaveAssetPtr=&Spot_Bird_BarTailedGodwit_Chirps_08}}
 {id=WavePlayer_8, properties {soundWaveAssetPtr=&Spot_Bird_BarTailedGodwit_Chirps_09}}
 {id=WavePlayer_9, properties {soundWaveAssetPtr=&Spot_Bird_BarTailedGodwit_Chirps_10}}
]
properties {
 attenuationSettings=&Fort_TileAmbient_OneShot_Bird_Att
 maxDistance=8005
 soundClassObject=&Fort_Ambient_OneShots_Outdoors
 volumeMultiplier=0.251
 busSends [{
   audioBus=null
   maxSendDistance=6000
   minSendDistance=3000
   minSendLevel=0.6
   sendLevel=1
   soundSourceBus=&AmbientAudio_Exterior_SourceBus
  }]
}
