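# Sound cue for outdoor ambient one-shot bird song (Barncaster), played quietly with bird-specific distance attenuation.
# A quality-level switch gives the first two levels a random pick from seven song waves and the last two a reduced pool of three.
# Each pool passes through a modulator at default settings for slight variation.
# The cue also sends to the exterior ambient source bus, with send level set over distance.
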SoundCue /AmbientAudioContent/SoundCues/OneShots/Spot_Bird_Barncaster_Song_Cue
root=QualityLevel_0
nodes [
 {id=Modulator_0, children [Random_0]}
 {id=Modulator_1, children [Random_3]}
 {id=QualityLevel_0, children [Modulator_0,Modulator_0,Modulator_1,Modulator_1]}
 {
  id=Random_0
  children [WavePlayer_2, WavePlayer_3, WavePlayer_4, WavePlayer_5, WavePlayer_14, WavePlayer_15, WavePlayer_16]
  properties {weights [1,1,1,1,1,1,1]}
 }
 {id=Random_3, children [WavePlayer_17, WavePlayer_18, WavePlayer_19], properties {weights [1,1,1]}}
 {id=WavePlayer_14, properties {soundWaveAssetPtr=/AmbientAudioContent/Waves/OneShots/Spot_Bird_Barncaster_Song_05}}
 {id=WavePlayer_15, properties {soundWaveAssetPtr=/AmbientAudioContent/Waves/OneShots/Spot_Bird_Barncaster_Song_06}}
 {id=WavePlayer_16, properties {soundWaveAssetPtr=/AmbientAudioContent/Waves/OneShots/Spot_Bird_Barncaster_Song_07}}
 {id=WavePlayer_17, properties {soundWaveAssetPtr=/AmbientAudioContent/Waves/OneShots/Spot_Bird_Barncaster_Song_01}}
 {id=WavePlayer_18, properties {soundWaveAssetPtr=/AmbientAudioContent/Waves/OneShots/Spot_Bird_Barncaster_Song_02}}
 {id=WavePlayer_19, properties {soundWaveAssetPtr=/AmbientAudioContent/Waves/OneShots/Spot_Bird_Barncaster_Song_03}}
 {id=WavePlayer_2, properties {soundWaveAssetPtr=&Spot_Bird_Barncaster_Song_01}}
 {id=WavePlayer_3, properties {soundWaveAssetPtr=&Spot_Bird_Barncaster_Song_02}}
 {id=WavePlayer_4, properties {soundWaveAssetPtr=&Spot_Bird_Barncaster_Song_03}}
 {id=WavePlayer_5, properties {soundWaveAssetPtr=/AmbientAudioContent/Waves/OneShots/Spot_Bird_Barncaster_Song_04}}
]
properties {
 attenuationSettings=&Fort_TileAmbient_OneShot_Bird_Att
 maxDistance=8005
 soundClassObject=&Fort_Ambient_OneShots_Outdoors
 volumeMultiplier=0.25
 busSends [{
   audioBus=null
   maxSendDistance=6000
   minSendDistance=3000
   minSendLevel=0.6
   sendLevel=1
   soundSourceBus=&AmbientAudio_Exterior_SourceBus
  }]
}
