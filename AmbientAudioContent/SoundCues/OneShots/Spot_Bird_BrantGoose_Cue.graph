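# Sound cue for outdoor ambient Brant goose one-shot calls, using bird one-shot attenuation and the outdoor one-shots sound class.
# A quality-level node feeds the first two levels from a random pick of seven waves and the last two from a three-wave subset.
# Each random branch passes through a modulator with default settings for variation.
# The cue also sends to the exterior ambient source bus with a distance-scaled level.
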SoundCue /AmbientAudioContent/SoundCues/OneShots/Spot_Bird_BrantGoose_Cue
root=QualityLevel_0
nodes [
 {id=Modulator_0, children [Random_2]}
 {id=Modulator_3, children [Random_3]}
 {id=QualityLevel_0, children [Modulator_0,Modulator_0,Modulator_3,Modulator_3]}
 {
  id=Random_2
  children [WavePlayer_6, WavePlayer_7, WavePlayer_8, WavePlayer_9, WavePlayer_10, WavePlayer_11, WavePlayer_12]
  properties {weights [1,1,1,1,1,1,1]}
 }
 {id=Random_3, children [WavePlayer_13, WavePlayer_14, WavePlayer_15], properties {weights [1,1,1]}}
 {id=WavePlayer_10, properties {soundWaveAssetPtr=/AmbientAudioContent/Waves/OneShots/Spot_Bird_BrantGoose_05}}
 {id=WavePlayer_11, properties {soundWaveAssetPtr=/AmbientAudioContent/Waves/OneShots/Spot_Bird_BrantGoose_06}}
 {id=WavePlayer_12, properties {soundWaveAssetPtr=/AmbientAudioContent/Waves/OneShots/Spot_Bird_BrantGoose_07}}
 {id=WavePlayer_13, properties {soundWaveAssetPtr=/AmbientAudioContent/Waves/OneShots/Spot_Bird_BrantGoose_01}}
 {id=WavePlayer_14, properties {soundWaveAssetPtr=/AmbientAudioContent/Waves/OneShots/Spot_Bird_BrantGoose_02}}
 {id=WavePlayer_15, properties {soundWaveAssetPtr=&Spot_Bird_BrantGoose_06}}
 {id=WavePlayer_6, properties {soundWaveAssetPtr=&Spot_Bird_BrantGoose_01}}
 {id=WavePlayer_7, properties {soundWaveAssetPtr=&Spot_Bird_BrantGoose_02}}
 {id=WavePlayer_8, properties {soundWaveAssetPtr=/AmbientAudioContent/Waves/OneShots/Spot_Bird_BrantGoose_03}}
 {id=WavePlayer_9, properties {soundWaveAssetPtr=/AmbientAudioContent/Waves/OneShots/Spot_Bird_BrantGoose_04}}
]
properties {
 attenuationSettings=&Fort_TileAmbient_OneShot_Bird_Att
 maxDistance=8005
 soundClassObject=&Fort_Ambient_OneShots_Outdoors
 volumeMultiplier=0.501
 busSends [{
   audioBus=null
   maxSendDistance=6000
   minSendDistance=3000
   minSendLevel=0.6
   sendLevel=1
   soundSourceBus=&AmbientAudio_Exterior_SourceBus
  }]
}
