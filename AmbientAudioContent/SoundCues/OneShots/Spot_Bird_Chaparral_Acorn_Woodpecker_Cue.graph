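# Sound Cue for an outdoor ambient one-shot of an acorn woodpecker call, named for the chaparral setting.
# An equally weighted Random node picks one of nine wave variations and feeds a Modulator.
# All inputs of the root Quality Level node share that Modulator, so every quality level sounds the same.
# It plays quietly with bird one-shot attenuation and an outdoor ambient sound class, plus a distance-based send to the exterior ambient source bus.
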SoundCue /AmbientAudioContent/SoundCues/OneShots/Spot_Bird_Chaparral_Acorn_Woodpecker_Cue
root=QualityLevel_0
nodes [
 {id=Modulator_0, children [Random_1]}
 {id=QualityLevel_0, children [Modulator_0,Modulator_0,Modulator_0,Modulator_0]}
 {
  id=Random_1
  children [WavePlayer_0,WavePlayer_1,WavePlayer_2,WavePlayer_3,WavePlayer_4,WavePlayer_5,WavePlayer_6,WavePlayer_7,WavePlayer_8]
  properties {weights [1,1,1,1,1,1,1,1,1]}
 }
 {id=WavePlayer_0, properties {soundWaveAssetPtr=/AmbientAudioContent/Waves/OneShots/Spot_Bird_Acorn_Woodpecker_01}}
 {id=WavePlayer_1, properties {soundWaveAssetPtr=/AmbientAudioContent/Waves/OneShots/Spot_Bird_Acorn_Woodpecker_02}}
 {id=WavePlayer_2, properties {soundWaveAssetPtr=/AmbientAudioContent/Waves/OneShots/Spot_Bird_Acorn_Woodpecker_03}}
 {id=WavePlayer_3, properties {soundWaveAssetPtr=/AmbientAudioContent/Waves/OneShots/Spot_Bird_Acorn_Woodpecker_04}}
 {id=WavePlayer_4, properties {soundWaveAssetPtr=/AmbientAudioContent/Waves/OneShots/Spot_Bird_Acorn_Woodpecker_05}}
 {id=WavePlayer_5, properties {soundWaveAssetPtr=/AmbientAudioContent/Waves/OneShots/Spot_Bird_Acorn_Woodpecker_06}}
 {id=WavePlayer_6, properties {soundWaveAssetPtr=/AmbientAudioContent/Waves/OneShots/Spot_Bird_Acorn_Woodpecker_07}}
 {id=WavePlayer_7, properties {soundWaveAssetPtr=/AmbientAudioContent/Waves/OneShots/Spot_Bird_Acorn_Woodpecker_08}}
 {id=WavePlayer_8, properties {soundWaveAssetPtr=/AmbientAudioContent/Waves/OneShots/Spot_Bird_Acorn_Woodpecker_09}}
]
properties {
 attenuationSettings=&Fort_TileAmbient_OneShot_Bird_Att
 maxDistance=8005
 soundClassObject=&Fort_Ambient_OneShots_Outdoors
 volumeMultiplier=0.2
 busSends [{
   audioBus=null
   maxSendDistance=6000
   minSendDistance=3000
   minSendLevel=0.6
   sendLevel=1
   soundSourceBus=&AmbientAudio_Exterior_SourceBus
  }]
}
