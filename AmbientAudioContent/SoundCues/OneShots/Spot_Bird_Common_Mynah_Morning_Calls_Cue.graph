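# Sound cue for one-shot common mynah morning bird calls, used as outdoor ambient spot sound.
# A quality-level node feeds two modulated random branches: the first two levels pick from eight call waves, the last two from a four-wave subset to save resources.
# Bird one-shot attenuation spatializes it at reduced volume, with a linear distance-scaled send to a distant reverb submix.
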SoundCue /AmbientAudioContent/SoundCues/OneShots/Spot_Bird_Common_Mynah_Morning_Calls_Cue
root=QualityLevel_0
nodes [
 {id=Modulator_0, children [Random_0]}
 {id=Modulator_1, children [Random_1]}
 {id=QualityLevel_0, children [Modulator_0,Modulator_0,Modulator_1,Modulator_1]}
 {
  id=Random_0
  children [WavePlayer_0, WavePlayer_1, WavePlayer_2, WavePlayer_3, WavePlayer_4, WavePlayer_5, WavePlayer_16, WavePlayer_17]
  properties {weights [1,1,1,1,1,1,1,1]}
 }
 {id=Random_1, children [WavePlayer_18, WavePlayer_21, WavePlayer_22, WavePlayer_23], properties {weights [1,1,1,1]}}
 {id=WavePlayer_0, properties {soundWaveAssetPtr=/AmbientAudioContent/Waves/OneShots/Spot_Bird_Common_Mynah_Morning_Calls_08}}
 {id=WavePlayer_1, properties {soundWaveAssetPtr=/AmbientAudioContent/Waves/OneShots/Spot_Bird_Common_Mynah_Morning_Calls_07}}
 {id=WavePlayer_16, properties {soundWaveAssetPtr=/AmbientAudioContent/Waves/OneShots/Spot_Bird_Common_Mynah_Morning_Calls_02}}
 {id=WavePlayer_17, properties {soundWaveAssetPtr=/AmbientAudioContent/Waves/OneShots/Spot_Bird_Common_Mynah_Morning_Calls_01}}
 {id=WavePlayer_18, properties {soundWaveAssetPtr=/AmbientAudioContent/Waves/OneShots/Spot_Bird_Common_Mynah_Morning_Calls_05}}
 {id=WavePlayer_2, properties {soundWaveAssetPtr=/AmbientAudioContent/Waves/OneShots/Spot_Bird_Common_Mynah_Morning_Calls_06}}
 {id=WavePlayer_21, properties {soundWaveAssetPtr=&Spot_Bird_Common_Mynah_Morning_Calls_06}}
 {id=WavePlayer_22, properties {soundWaveAssetPtr=/AmbientAudioContent/Waves/OneShots/Spot_Bird_Common_Mynah_Morning_Calls_04}}
 {id=WavePlayer_23, properties {soundWaveAssetPtr=/AmbientAudioContent/Waves/OneShots/Spot_Bird_Common_Mynah_Morning_Calls_03}}
 {id=WavePlayer_3, properties {soundWaveAssetPtr=&Spot_Bird_Common_Mynah_Morning_Calls_05}}
 {id=WavePlayer_4, properties {soundWaveAssetPtr=&Spot_Bird_Common_Mynah_Morning_Calls_04}}
 {id=WavePlayer_5, properties {soundWaveAssetPtr=&Spot_Bird_Common_Mynah_Morning_Calls_03}}
]
properties {
 attenuationSettings=&Fort_TileAmbient_OneShot_Bird_Att
 maxDistance=8005
 soundClassObject=&Fort_Ambient_OneShots_Outdoors
 volumeMultiplier=0.251
 busSends [{audioBus=null, minSendLevel=0, sendLevel=1, soundSourceBus=null}]
 soundSubmixSends [{
   maxSendDistance=8000
   maxSendLevel=0.4
   minSendDistance=0
   minSendLevel=0.1
   sendLevelControlMethod=Linear
   soundSubmix=&Fortnite_Reverb_Distant
  }]
}
