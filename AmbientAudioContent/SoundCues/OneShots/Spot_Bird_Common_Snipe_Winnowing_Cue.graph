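# Ambient one-shot sound cue for the winnowing of a common snipe, classed as outdoor ambience with bird attenuation.
# A quality-level root sends the first two levels to a random pick of six waves and the last two to a reduced set of three of those waves.
# Each random set passes through its own modulator for variation, and the cue feeds an exterior ambient source bus with a distance-ranged send.
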SoundCue /AmbientAudioContent/SoundCues/OneShots/Spot_Bird_Common_Snipe_Winnowing_Cue
root=QualityLevel_0
nodes [
 {id=Modulator_0, children [Random_2]}
 {id=Modulator_3, children [Random_3]}
 {id=QualityLevel_0, children [Modulator_0,Modulator_0,Modulator_3,Modulator_3]}
 {
  id=Random_2
  children [WavePlayer_3, WavePlayer_6, WavePlayer_7, WavePlayer_8, WavePlayer_9, WavePlayer_10]
  properties {weights [1,1,1,1,1,1]}
 }
 {id=Random_3, children [WavePlayer_11, WavePlayer_12, WavePlayer_13], properties {weights [1,1,1]}}
 {id=WavePlayer_10, properties {soundWaveAssetPtr=/AmbientAudioContent/Waves/OneShots/Spot_Bird_Common_Snipe_Winnowing_06}}
 {id=WavePlayer_11, properties {soundWaveAssetPtr=/AmbientAudioContent/Waves/OneShots/Spot_Bird_Common_Snipe_Winnowing_01}}
 {id=WavePlayer_12, properties {soundWaveAssetPtr=/AmbientAudioContent/Waves/OneShots/Spot_Bird_Common_Snipe_Winnowing_02}}
 {id=WavePlayer_13, properties {soundWaveAssetPtr=/AmbientAudioContent/Waves/OneShots/Spot_Bird_Common_Snipe_Winnowing_05}}
 {id=WavePlayer_3, properties {soundWaveAssetPtr=&Spot_Bird_Common_Snipe_Winnowing_01}}
 {id=WavePlayer_6, properties {soundWaveAssetPtr=&Spot_Bird_Common_Snipe_Winnowing_02}}
 {id=WavePlayer_7, properties {soundWaveAssetPtr=/AmbientAudioContent/Waves/OneShots/Spot_Bird_Common_Snipe_Winnowing_03}}
 {id=WavePlayer_8, properties {soundWaveAssetPtr=/AmbientAudioContent/Waves/OneShots/Spot_Bird_Common_Snipe_Winnowing_04}}
 {id=WavePlayer_9, properties {soundWaveAssetPtr=&Spot_Bird_Common_Snipe_Winnowing_05}}
]
properties {
 attenuationSettings=&Fort_TileAmbient_OneShot_Bird_Att
 maxDistance=8005
 soundClassObject=&Fort_Ambient_OneShots_Outdoors
 volumeMultiplier=0.15
 busSends [{
   audioBus=null
   maxSendDistance=6000
   minSendDistance=3000
   minSendLevel=0.6
   sendLevel=1
   soundSourceBus=&AmbientAudio_Exterior_SourceBus
  }]
}
