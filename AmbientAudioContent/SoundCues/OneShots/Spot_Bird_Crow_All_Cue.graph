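# Outdoor ambient Sound Cue that plays a one-shot crow call from a varied pool of recordings.
# All quality levels share one random node that picks evenly among four groups: rare mountain, closer, Eastern European day/evening and Japanese crows.
# Each group randomly picks a wave, then passes through its own pitch/volume modulator and a volume-reducing mixer.
# Bird one-shot attenuation, a concurrency override capping simultaneous voices and a master reverb send shape playback.
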SoundCue /AmbientAudioContent/SoundCues/OneShots/Spot_Bird_Crow_All_Cue
root=QualityLevel_0
nodes [
 {id=Mixer_0, children [Modulator_0], properties {inputVolume [0.4]}}
 {id=Mixer_1, children [Modulator_1], properties {inputVolume [0.4]}}
 {id=Mixer_2, children [Modulator_2], properties {inputVolume [0.4]}}
 {id=Mixer_4, children [Modulator_4], properties {inputVolume [0.4]}}
 {id=Modulator_0, children [Random_0], properties {pitchMin=0.8, volumeMax=1, volumeMin=0.8}}
 {id=Modulator_1, children [Random_6]}
 {id=Modulator_2, children [Random_1]}
 {id=Modulator_4, children [Random_2]}
 {id=QualityLevel_0, children [Random_5,Random_5,Random_5,Random_5]}
 {id=Random_0, children [WavePlayer_1,WavePlayer_6,WavePlayer_7], properties {weights [1,1,1]}}
 {id=Random_1, children [WavePlayer_0,WavePlayer_2,WavePlayer_3,WavePlayer_4], properties {weights [1,1,1,1]}}
 {
  id=Random_2
  children [WavePlayer_5, WavePlayer_9, WavePlayer_11, WavePlayer_12, WavePlayer_13, WavePlayer_14, WavePlayer_15, WavePlayer_16, WavePlayer_17]
  properties {weights [1,1,1,1,1,1,1,1,1]}
 }
 {id=Random_5, children [Mixer_0,Mixer_1,Mixer_2,Mixer_4], properties {weights [1,1,1,1]}}
 {id=Random_6, children [WavePlayer_8, WavePlayer_10, WavePlayer_23, WavePlayer_24], properties {weights [1,1,1,1]}}
 {id=WavePlayer_0, properties {soundWaveAssetPtr=/AmbientAudioContent/Waves/OneShots/Spot_Bird_Crow_Closer_01}}
 {id=WavePlayer_1, properties {soundWaveAssetPtr=/AmbientAudioContent/Waves/OneShots/Spot_Bird_Rare_Mountain_Crow_01}}
 {id=WavePlayer_10, properties {soundWaveAssetPtr=/AmbientAudioContent/Waves/OneShots/Spot_Bird_Crow_DayEvening_Eastern_Europe_02}}
 {id=WavePlayer_11, properties {soundWaveAssetPtr=/AmbientAudioContent/Waves/OneShots/Spot_Bird_Japanese_Inawashiro_Crow_01}}
 {id=WavePlayer_12, properties {soundWaveAssetPtr=/AmbientAudioContent/Waves/OneShots/Spot_Bird_Japanese_Inawashiro_Crow_02}}
 {id=WavePlayer_13, properties {soundWaveAssetPtr=/AmbientAudioContent/Waves/OneShots/Spot_Bird_Japanese_Mizumoto_Crow_Caw_03}}
 {id=WavePlayer_14, properties {soundWaveAssetPtr=/AmbientAudioContent/Waves/OneShots/Spot_Bird_Japanese_Inawashiro_Crow_03}}
 {id=WavePlayer_15, properties {soundWaveAssetPtr=/AmbientAudioContent/Waves/OneShots/Spot_Bird_Japanese_Inawashiro_Crow_04}}
 {id=WavePlayer_16, properties {soundWaveAssetPtr=/AmbientAudioContent/Waves/OneShots/Spot_Bird_Japanese_Mizumoto_Crow_Caw_01}}
 {id=WavePlayer_17, properties {soundWaveAssetPtr=/AmbientAudioContent/Waves/OneShots/Spot_Bird_Japanese_Mizumoto_Crow_Caw_02}}
 {id=WavePlayer_2, properties {soundWaveAssetPtr=/AmbientAudioContent/Waves/OneShots/Spot_Bird_Crow_Closer_02}}
 {id=WavePlayer_23, properties {soundWaveAssetPtr=/AmbientAudioContent/Waves/OneShots/Spot_Bird_Crow_DayEvening_Eastern_Europe_03}}
 {id=WavePlayer_24, properties {soundWaveAssetPtr=/AmbientAudioContent/Waves/OneShots/Spot_Bird_Crow_DayEvening_Eastern_Europe_04}}
 {id=WavePlayer_3, properties {soundWaveAssetPtr=/AmbientAudioContent/Waves/OneShots/Spot_Bird_Crow_Closer_03}}
 {id=WavePlayer_4, properties {soundWaveAssetPtr=/AmbientAudioContent/Waves/OneShots/Spot_Bird_Crow_Closer_04}}
 {id=WavePlayer_5, properties {soundWaveAssetPtr=/AmbientAudioContent/Waves/OneShots/Spot_Bird_Japanese_Crow_Parkland_01}}
 {id=WavePlayer_6, properties {soundWaveAssetPtr=/AmbientAudioContent/Waves/OneShots/Spot_Bird_Rare_Mountain_Crow_02}}
 {id=WavePlayer_7, properties {soundWaveAssetPtr=/AmbientAudioContent/Waves/OneShots/Spot_Bird_Rare_Mountain_Crow_03}}
 {id=WavePlayer_8, properties {soundWaveAssetPtr=/AmbientAudioContent/Waves/OneShots/Spot_Bird_Crow_DayEvening_Eastern_Europe_01}}
 {id=WavePlayer_9, properties {soundWaveAssetPtr=/AmbientAudioContent/Waves/OneShots/Spot_Bird_Japanese_Crow_Parkland_02}}
]
properties {
 attenuationSettings=/AmbientAudioBase/Attenuations/Fort_TileAmbient_OneShot_Bird_Att
 bEnableBusSends=false
 bOverrideConcurrency=true
 maxDistance=8005
 soundClassObject=&Fort_Ambient_OneShots_Outdoors
 volumeMultiplier=1
 concurrencyOverrides {maxCount=3, voiceStealReleaseTime=0.25}
 preEffectBusSends [{}]
 soundSubmixSends [{maxSendLevel=1.2, minSendLevel=0.4, sendLevel=0, soundSubmix=/Engine/EngineSounds/Submixes/MasterReverbSubmixDefault}]
}
