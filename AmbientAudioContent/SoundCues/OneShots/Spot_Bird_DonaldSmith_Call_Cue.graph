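# Sound cue for a one-shot outdoor ambient bird call (DonaldSmith), played quietly with the tile-ambient bird attenuation and outdoor one-shots sound class.
# Six call waves feed an equally weighted Random node, then a Modulator.
# A Quality Level root routes that one Modulator into its first two slots and leaves the other two empty.
# The cue also sends to the exterior ambient source bus.
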SoundCue /AmbientAudioContent/SoundCues/OneShots/Spot_Bird_DonaldSmith_Call_Cue
root=QualityLevel_0
nodes [
 {id=Modulator_0, children [Random_1]}
 {id=QualityLevel_0, children [Modulator_0,Modulator_0,null,null]}
 {id=Random_1, children [WavePlayer_0,WavePlayer_1,WavePlayer_4,WavePlayer_6,WavePlayer_7,WavePlayer_8], properties {weights [1,1,1,1,1,1]}}
 {id=WavePlayer_0, properties {soundWaveAssetPtr=/AmbientAudioContent/Waves/OneShots/Spot_Bird_DonaldSmith_Call_01}}
 {id=WavePlayer_1, properties {soundWaveAssetPtr=/AmbientAudioContent/Waves/OneShots/Spot_Bird_DonaldSmith_Call_02}}
 {id=WavePlayer_4, properties {soundWaveAssetPtr=/AmbientAudioContent/Waves/OneShots/Spot_Bird_DonaldSmith_Call_03}}
 {id=WavePlayer_6, properties {soundWaveAssetPtr=/AmbientAudioContent/Waves/OneShots/Spot_Bird_DonaldSmith_Call_04}}
 {id=WavePlayer_7, properties {soundWaveAssetPtr=/AmbientAudioContent/Waves/OneShots/Spot_Bird_DonaldSmith_Call_05}}
 {id=WavePlayer_8, properties {soundWaveAssetPtr=/AmbientAudioContent/Waves/OneShots/Spot_Bird_DonaldSmith_Call_06}}
]
properties {
 attenuationSettings=&Fort_TileAmbient_OneShot_Bird_Att
 maxDistance=8005
 soundClassObject=&Fort_Ambient_OneShots_Outdoors
 volumeMultiplier=0.125
 busSends [{
   audioBus=null
   maxSendDistance=6000
   minSendDistance=3000
   minSendLevel=0.6
   sendLevel=1
   soundSourceBus=&AmbientAudio_Exterior_SourceBus
   customSendLevelCurve {editorCurveData {keys [<at 0 0.6>,<at 1 1>]}}
  }]
}
