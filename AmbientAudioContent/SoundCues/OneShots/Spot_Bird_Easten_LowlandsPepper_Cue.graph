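# Spot ambient one-shot cue for swamp bird and peeper calls, using outdoor ambient class and one-shot attenuation.
# The first two quality slots randomly choose a peeper or bird mix, each pairing a random call with a quieter, delayed, pitched-down call from the same pool, like a distant reply.
# The other two slots use a simpler modulated random pick from a mixed set of peeper and bird waves, without layering.
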SoundCue /AmbientAudioContent/SoundCues/OneShots/Spot_Bird_Easten_LowlandsPepper_Cue
root=QualityLevel_0
nodes [
 {id=Delay_0, children [Modulator_2], properties {delayMax=2, delayMin=1}}
 {id=Delay_1, children [Modulator_3], properties {delayMax=2, delayMin=1}}
 {id=Mixer_0, children [Modulator_0,Delay_0], properties {inputVolume [1,0.25]}}
 {id=Mixer_1, children [Random_4,Delay_1], properties {inputVolume [1,0.25]}}
 {id=Modulator_0, children [Random_3]}
 {id=Modulator_1, children [Random_1]}
 {id=Modulator_2, children [Random_3], properties {pitchMax=0.9, pitchMin=0.7, volumeMin=0.8}}
 {id=Modulator_3, children [Random_5], properties {pitchMax=0.9, pitchMin=0.7}}
 {id=QualityLevel_0, children [Random_6,Random_6,Modulator_1,Modulator_1]}
 {id=Random_1, children [WavePlayer_1, WavePlayer_2, WavePlayer_9, WavePlayer_13], properties {weights [1,1,1,1]}}
 {id=Random_3, children [WavePlayer_16, WavePlayer_17, WavePlayer_18, WavePlayer_19], properties {weights [1,1,1,1]}}
 {id=Random_4, children [WavePlayer_5,WavePlayer_6], properties {weights [1,1]}}
 {id=Random_5, children [WavePlayer_3,WavePlayer_4], properties {weights [1,1]}}
 {id=Random_6, children [Mixer_0,Mixer_1], properties {weights [1,1]}}
 {id=WavePlayer_1, properties {soundWaveAssetPtr=/AmbientAudioContent/Waves/OneShots/Spot_Biome_Swamp_01_Peepers_01}}
 {id=WavePlayer_13, properties {soundWaveAssetPtr=/AmbientAudioContent/Waves/OneShots/Spot_Biome_Swamp_01_Birds_06}}
 {id=WavePlayer_16, properties {soundWaveAssetPtr=&Spot_Biome_Swamp_01_Peepers_01}}
 {id=WavePlayer_17, properties {soundWaveAssetPtr=/AmbientAudioContent/Waves/OneShots/Spot_Biome_Swamp_01_Peepers_02}}
 {id=WavePlayer_18, properties {soundWaveAssetPtr=/AmbientAudioContent/Waves/OneShots/Spot_Biome_Swamp_01_Peepers_03}}
 {id=WavePlayer_19, properties {soundWaveAssetPtr=/AmbientAudioContent/Waves/OneShots/Spot_Biome_Swamp_01_Peepers_04}}
 {id=WavePlayer_2, properties {soundWaveAssetPtr=&Spot_Biome_Swamp_01_Peepers_02}}
 {id=WavePlayer_3, properties {soundWaveAssetPtr=/AmbientAudioContent/Waves/OneShots/Spot_Biome_Swamp_01_Birds_05}}
 {id=WavePlayer_4, properties {soundWaveAssetPtr=&Spot_Biome_Swamp_01_Birds_06}}
 {id=WavePlayer_5, properties {soundWaveAssetPtr=&Spot_Biome_Swamp_01_Birds_05}}
 {id=WavePlayer_6, properties {soundWaveAssetPtr=&Spot_Biome_Swamp_01_Birds_06}}
 {id=WavePlayer_9, properties {soundWaveAssetPtr=&Spot_Biome_Swamp_01_Birds_05}}
]
properties {
 attenuationSettings=/Game/Sounds/Attenuation/Fort_TileAmbient_OneShot_Att
 maxDistance=8000
 soundClassObject=&Fort_Ambient_OneShots_Outdoors
 volumeMultiplier=0.451
 preEffectBusSends [{}]
}
