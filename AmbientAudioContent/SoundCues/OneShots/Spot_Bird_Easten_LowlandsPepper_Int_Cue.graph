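# Indoor ambient one-shot cue for swamp bird and peeper spot sounds, played quietly at lowered pitch with tile-ambient attenuation.
# A quality-level switch sends its first two slots to a random choice of peeper or bird mixers.
# Each mixer layers a random wave with a quieter, delayed, pitch-lowered pick from the same set, like an echo.
# The last two slots use a simpler modulated random pick from two peeper and two bird waves.
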SoundCue /AmbientAudioContent/SoundCues/OneShots/Spot_Bird_Easten_LowlandsPepper_Int_Cue
root=QualityLevel_1
nodes [
 {id=Delay_2, children [Modulator_6], properties {delayMax=2, delayMin=1}}
 {id=Delay_3, children [Modulator_7], properties {delayMax=2, delayMin=1}}
 {id=Mixer_2, children [Modulator_4,Delay_2], properties {inputVolume [1,0.25]}}
 {id=Mixer_3, children [Random_7,Delay_3], properties {inputVolume [1,0.25]}}
 {id=Modulator_4, children [Random_10]}
 {id=Modulator_5, children [Random_3]}
 {id=Modulator_6, children [Random_10], properties {pitchMax=0.9, pitchMin=0.7, volumeMin=0.8}}
 {id=Modulator_7, children [Random_8], properties {pitchMax=0.9, pitchMin=0.7}}
 {id=QualityLevel_1, children [Random_9,Random_9,Modulator_5,Modulator_5]}
 {id=Random_10, children [WavePlayer_24, WavePlayer_25, WavePlayer_26, WavePlayer_27], properties {weights [1,1,1,1]}}
 {id=Random_3, children [WavePlayer_18, WavePlayer_19, WavePlayer_20, WavePlayer_21], properties {weights [1,1,1,1]}}
 {id=Random_7, children [WavePlayer_16, WavePlayer_17], properties {weights [1,1]}}
 {id=Random_8, children [WavePlayer_22, WavePlayer_23], properties {weights [1,1]}}
 {id=Random_9, children [Mixer_2,Mixer_3], properties {weights [1,1]}}
 {id=WavePlayer_16, properties {soundWaveAssetPtr=/AmbientAudioContent/Waves/OneShots/Spot_Biome_Swamp_01_Birds_05}}
 {id=WavePlayer_17, properties {soundWaveAssetPtr=/AmbientAudioContent/Waves/OneShots/Spot_Biome_Swamp_01_Birds_06}}
 {id=WavePlayer_18, properties {soundWaveAssetPtr=/AmbientAudioContent/Waves/OneShots/Spot_Biome_Swamp_01_Peepers_01}}
 {id=WavePlayer_19, properties {soundWaveAssetPtr=/AmbientAudioContent/Waves/OneShots/Spot_Biome_Swamp_01_Peepers_02}}
 {id=WavePlayer_20, properties {soundWaveAssetPtr=&Spot_Biome_Swamp_01_Birds_05}}
 {id=WavePlayer_21, properties {soundWaveAssetPtr=&Spot_Biome_Swamp_01_Birds_06}}
 {id=WavePlayer_22, properties {soundWaveAssetPtr=&Spot_Biome_Swamp_01_Birds_05}}
 {id=WavePlayer_23, properties {soundWaveAssetPtr=&Spot_Biome_Swamp_01_Birds_06}}
 {id=WavePlayer_24, properties {soundWaveAssetPtr=&Spot_Biome_Swamp_01_Peepers_01}}
 {id=WavePlayer_25, properties {soundWaveAssetPtr=&Spot_Biome_Swamp_01_Peepers_02}}
 {id=WavePlayer_26, properties {soundWaveAssetPtr=/AmbientAudioContent/Waves/OneShots/Spot_Biome_Swamp_01_Peepers_03}}
 {id=WavePlayer_27, properties {soundWaveAssetPtr=/AmbientAudioContent/Waves/OneShots/Spot_Biome_Swamp_01_Peepers_04}}
]
properties {
 attenuationSettings=/Game/Sounds/Attenuation/Fort_TileAmbient_OneShot_Att
 maxDistance=8000
 pitchMultiplier=0.8
 soundClassObject=/Game/Sounds/FortSoundClassesAndModes/Fort_Ambient_OneShots_Indoors
 volumeMultiplier=0.15
 preEffectBusSends [{}]
}
