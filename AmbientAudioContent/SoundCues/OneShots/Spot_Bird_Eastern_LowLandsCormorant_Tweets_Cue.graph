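# One-shot ambient Sound Cue for cormorant tweet calls, named as an Eastern lowlands bird spot sound.
# A Random node picks among six equally weighted tropical-day cormorant tweet waves, and a Modulator varies pitch and volume on each play.
# The root Quality Level node routes all four of its inputs to that same Modulator.
# It uses the outdoor ambient one-shot sound class and bird attenuation, with a reduced volume multiplier.
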SoundCue /AmbientAudioContent/SoundCues/OneShots/Spot_Bird_Eastern_LowLandsCormorant_Tweets_Cue
root=QualityLevel_0
nodes [
 {id=Modulator_1, children [Random_0], properties {pitchMax=1.1, volumeMax=1.1, volumeMin=0.9}}
 {id=QualityLevel_0, children [Modulator_1,Modulator_1,Modulator_1,Modulator_1]}
 {
  id=Random_0
  children [WavePlayer_6, WavePlayer_7, WavePlayer_8, WavePlayer_9, WavePlayer_10, WavePlayer_11]
  properties {weights [1,1,1,1,1,1]}
 }
 {id=WavePlayer_10, properties {soundWaveAssetPtr=/AmbientAudioContent/Waves/OneShots/Spot_Tropical_Day_Bird_Cormorant_Tweets_02}}
 {id=WavePlayer_11, properties {soundWaveAssetPtr=/AmbientAudioContent/Waves/OneShots/Spot_Tropical_Day_Bird_Cormorant_Tweets_01}}
 {id=WavePlayer_6, properties {soundWaveAssetPtr=/AmbientAudioContent/Waves/OneShots/Spot_Tropical_Day_Bird_Cormorant_Tweets_06}}
 {id=WavePlayer_7, properties {soundWaveAssetPtr=/AmbientAudioContent/Waves/OneShots/Spot_Tropical_Day_Bird_Cormorant_Tweets_05}}
 {id=WavePlayer_8, properties {soundWaveAssetPtr=/AmbientAudioContent/Waves/OneShots/Spot_Tropical_Day_Bird_Cormorant_Tweets_04}}
 {id=WavePlayer_9, properties {soundWaveAssetPtr=/AmbientAudioContent/Waves/OneShots/Spot_Tropical_Day_Bird_Cormorant_Tweets_03}}
]
properties {
 attenuationSettings=&Fort_TileAmbient_OneShot_Bird_Att
 bEnableBusSends=false
 maxDistance=8005
 soundClassObject=&Fort_Ambient_OneShots_Outdoors
 volumeMultiplier=0.251
 preEffectBusSends [{}]
}
