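# One-shot ambient sound cue for an Eastern lowlands owl call, played as outdoor bird ambience.
# A quality node's first level randomly picks between two pitch-modulated pools: mixed distant calls and hoots, or volume-varied night hoots.
# The other three quality levels share a smaller distant-call pool with wider pitch variation.
# Bird one-shot attenuation, a distant reverb send and a prevent-new concurrency override shape playback.
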SoundCue /AmbientAudioContent/SoundCues/OneShots/Spot_Bird_Eastern_LowLandsOwl_Cue
root=QualityLevel_0
nodes [
 {id=Modulator_0, children [Random_4], properties {pitchMin=0.9, volumeMax=0.9, volumeMin=0.8}}
 {id=Modulator_1, children [Random_3], properties {pitchMin=0.8}}
 {id=Modulator_2, children [Random_5], properties {pitchMin=0.5}}
 {id=QualityLevel_0, children [Random_2,Modulator_2,Modulator_2,Modulator_2]}
 {id=Random_2, children [Modulator_1,Modulator_0], properties {weights [1,1]}}
 {
  id=Random_3
  children [WavePlayer_9, WavePlayer_10, WavePlayer_11, WavePlayer_12, WavePlayer_13, WavePlayer_14]
  properties {weights [1,1,1,1,1,1]}
 }
 {id=Random_4, children [WavePlayer_15, WavePlayer_16, WavePlayer_17], properties {weights [1,1,1]}}
 {id=Random_5, children [WavePlayer_18, WavePlayer_19, WavePlayer_20], properties {weights [1,1,1]}}
 {id=WavePlayer_10, properties {soundWaveAssetPtr=/AmbientAudioContent/Waves/OneShots/Spot_Bird_OwlDistant08}}
 {id=WavePlayer_11, properties {soundWaveAssetPtr=/AmbientAudioContent/Waves/OneShots/Spot_Bird_OwlHoot01}}
 {id=WavePlayer_12, properties {soundWaveAssetPtr=/AmbientAudioContent/Waves/OneShots/Spot_Bird_OwlHoot02}}
 {id=WavePlayer_13, properties {soundWaveAssetPtr=/AmbientAudioContent/Waves/OneShots/Spot_Bird_OwlDistant04}}
 {id=WavePlayer_14, properties {soundWaveAssetPtr=/AmbientAudioContent/Waves/OneShots/Spot_Bird_OwlDistant05}}
 {id=WavePlayer_15, properties {soundWaveAssetPtr=/AmbientAudioContent/Waves/OneShots/Spot_Bird_Night_Owl_Hoots_01}}
 {id=WavePlayer_16, properties {soundWaveAssetPtr=/AmbientAudioContent/Waves/OneShots/Spot_Bird_Night_Owl_Hoots_02}}
 {id=WavePlayer_17, properties {soundWaveAssetPtr=/AmbientAudioContent/Waves/OneShots/Spot_Bird_Night_Owl_Hoots_03}}
 {id=WavePlayer_18, properties {soundWaveAssetPtr=/AmbientAudioContent/Waves/OneShots/Spot_Bird_OwlDistant01}}
 {id=WavePlayer_19, properties {soundWaveAssetPtr=&Spot_Bird_OwlDistant04}}
 {id=WavePlayer_20, properties {soundWaveAssetPtr=&Spot_Bird_OwlDistant05}}
 {id=WavePlayer_9, properties {soundWaveAssetPtr=&Spot_Bird_OwlDistant01}}
]
properties {
 attenuationSettings=&Fort_TileAmbient_OneShot_Bird_Att
 bOverrideConcurrency=true
 maxDistance=8005
 soundClassObject=&Fort_Ambient_OneShots_Outdoors
 volumeMultiplier=0.3
 concurrencyOverrides {resolutionRule=PreventNew}
 soundSubmixSends [{minSendLevel=0.6, sendLevel=0.6, soundSubmix=&Fortnite_Reverb_Distant}]
}
