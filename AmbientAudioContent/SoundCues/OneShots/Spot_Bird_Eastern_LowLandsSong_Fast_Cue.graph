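# Sound cue for one-shot Eastern lowlands birdsong, played as outdoor ambient audio through a bird attenuation asset at reduced volume.
# A Random node picks among nine equally weighted Eastern biome daytime bird waves.
# A Modulator adds slight pitch and volume variation to each pick.
# All four branches of the root quality-level node share that Modulator, so every quality level gets the same chain.
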SoundCue /AmbientAudioContent/SoundCues/OneShots/Spot_Bird_Eastern_LowLandsSong_Fast_Cue
root=QualityLevel_0
nodes [
 {id=Modulator_1, children [Random_0], properties {pitchMax=1.1, volumeMax=1.1, volumeMin=0.9}}
 {id=QualityLevel_0, children [Modulator_1,Modulator_1,Modulator_1,Modulator_1]}
 {
  id=Random_0
  children [WavePlayer_6, WavePlayer_8, WavePlayer_9, WavePlayer_10, WavePlayer_11, WavePlayer_12, WavePlayer_13, WavePlayer_21, WavePlayer_22]
  properties {weights [1,1,1,1,1,1,1,1,1]}
 }
 {id=WavePlayer_10, properties {soundWaveAssetPtr=/AmbientAudioContent/Waves/OneShots/Spot_Bird_Eastern_Biome_Day_06}}
 {id=WavePlayer_11, properties {soundWaveAssetPtr=/AmbientAudioContent/Waves/OneShots/Spot_Bird_Eastern_Biome_Day_07}}
 {id=WavePlayer_12, properties {soundWaveAssetPtr=/AmbientAudioContent/Waves/OneShots/Spot_Bird_Eastern_Biome_Day_08}}
 {id=WavePlayer_13, properties {soundWaveAssetPtr=/AmbientAudioContent/Waves/OneShots/Spot_Bird_Eastern_Biome_Day_09}}
 {id=WavePlayer_21, properties {soundWaveAssetPtr=/AmbientAudioContent/Waves/OneShots/Spot_Bird_Eastern_Biome_Day_10}}
 {id=WavePlayer_22, properties {soundWaveAssetPtr=/AmbientAudioContent/Waves/OneShots/Spot_Bird_Eastern_Biome_Day_11}}
 {id=WavePlayer_6, properties {soundWaveAssetPtr=/AmbientAudioContent/Waves/OneShots/Spot_Bird_Eastern_Biome_Day_01}}
 {id=WavePlayer_8, properties {soundWaveAssetPtr=/AmbientAudioContent/Waves/OneShots/Spot_Bird_Eastern_Biome_Day_03}}
 {id=WavePlayer_9, properties {soundWaveAssetPtr=/AmbientAudioContent/Waves/OneShots/Spot_Bird_Eastern_Biome_Day_04}}
]
properties {
 attenuationSettings=&Fort_TileAmbient_OneShot_Bird_Att
 bEnableBusSends=false
 maxDistance=8005
 soundClassObject=&Fort_Ambient_OneShots_Outdoors
 volumeMultiplier=0.4
 preEffectBusSends [{}]
}
