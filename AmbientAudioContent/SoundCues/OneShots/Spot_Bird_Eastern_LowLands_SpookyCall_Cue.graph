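# Sound cue for a spooky eastern-lowlands bird call, played as an outdoor ambient one-shot spot sound.
# A root Random node, excluded from branch culling, picks with equal weight among five nightmare-jungle bird chatter and short call waves.
# The cue plays at reduced volume through a bird one-shot attenuation asset and the outdoor ambient one-shots sound class.
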SoundCue /AmbientAudioContent/SoundCues/OneShots/Spot_Bird_Eastern_LowLands_SpookyCall_Cue
root=Random_1
nodes [
 {
  id=Random_1
  children [WavePlayer_0,WavePlayer_6,WavePlayer_7,WavePlayer_8,WavePlayer_9]
  properties {bShouldExcludeFromBranchCulling=true, weights [1,1,1,1,1]}
 }
 {id=WavePlayer_0, properties {soundWaveAssetPtr=/AmbientAudioContent/Waves/OneShots/Spot_Nightmare_Jungle_OneShot_BirdChat_01}}
 {id=WavePlayer_6, properties {soundWaveAssetPtr=/AmbientAudioContent/Waves/OneShots/Spot_Nightmare_Jungle_OneShot_BirdChat_03}}
 {id=WavePlayer_7, properties {soundWaveAssetPtr=/AmbientAudioContent/Waves/OneShots/Spot_Nightmare_Jungle_OneShot_BirdChat_04}}
 {id=WavePlayer_8, properties {soundWaveAssetPtr=/AmbientAudioContent/Waves/OneShots/Spot_Nightmare_Jungle_OneShot_BirdCallShort_01}}
 {id=WavePlayer_9, properties {soundWaveAssetPtr=/AmbientAudioContent/Waves/OneShots/Spot_Nightmare_Jungle_OneShot_BirdCallShort_02}}
]
properties {
 attenuationSettings=&Fort_TileAmbient_OneShot_Bird_Att
 maxDistance=8005
 soundClassObject=&Fort_Ambient_OneShots_Outdoors
 volumeMultiplier=0.401
}
