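# Sound cue for an outdoor ambient one-shot: a long, spooky Eastern lowlands bird call.
# A random node picks one of four equally weighted nightmare-jungle long bird call waves, then a modulator varies its pitch for variety.
# Playback is quietened by a reduced volume multiplier and spatialized through a shared bird one-shot attenuation asset under the outdoor ambient one-shot sound class.
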SoundCue /AmbientAudioContent/SoundCues/OneShots/Spot_Bird_Eastern_LowLands_SpookyCall_Long_Cue
root=Modulator_0
nodes [
 {id=Modulator_0, children [Random_0], properties {pitchMax=1.1, pitchMin=0.8}}
 {id=Random_0, children [WavePlayer_0,WavePlayer_1,WavePlayer_2,WavePlayer_3], properties {weights [1,1,1,1]}}
 {id=WavePlayer_0, properties {soundWaveAssetPtr=/AmbientAudioContent/Waves/OneShots/Spot_Nightmare_Jungle_OneShot_BirdCallLong_02}}
 {id=WavePlayer_1, properties {soundWaveAssetPtr=/AmbientAudioContent/Waves/OneShots/Spot_Nightmare_Jungle_OneShot_BirdCallLong_01}}
 {id=WavePlayer_2, properties {soundWaveAssetPtr=/AmbientAudioContent/Waves/OneShots/Spot_Nightmare_Jungle_OneShot_BirdCallLong_03}}
 {id=WavePlayer_3, properties {soundWaveAssetPtr=/AmbientAudioContent/Waves/OneShots/Spot_Nightmare_Jungle_OneShot_BirdCallLong_04}}
]
properties {
 attenuationSettings=&Fort_TileAmbient_OneShot_Bird_Att
 maxDistance=8005
 soundClassObject=&Fort_Ambient_OneShots_Outdoors
 volumeMultiplier=0.251
}
