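# Sound cue for a one-shot ambient mountain crow call, classed as an outdoor ambient one-shot with bird attenuation.
# A Random node picks among three crow waves with equal weight, and every Quality Level branch shares that same node.
# A root Modulator varies each call, with its pitch minimum lowered, and the cue plays at reduced volume with a master reverb submix send configured.
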SoundCue /AmbientAudioContent/SoundCues/OneShots/Spot_Bird_Eastern_MountainsCrow_Call_Cue
root=Modulator_0
nodes [
 {id=Modulator_0, children [QualityLevel_0], properties {pitchMin=0.8}}
 {id=QualityLevel_0, children [Random_0,Random_0,Random_0,Random_0]}
 {id=Random_0, children [WavePlayer_0,WavePlayer_1,WavePlayer_2], properties {weights [1,1,1]}}
 {id=WavePlayer_0, properties {soundWaveAssetPtr=/AmbientAudioContent/Waves/OneShots/Spot_Bird_Rare_Mountain_Crow_01}}
 {id=WavePlayer_1, properties {soundWaveAssetPtr=/AmbientAudioContent/Waves/OneShots/Spot_Bird_Rare_Mountain_Crow_02}}
 {id=WavePlayer_2, properties {soundWaveAssetPtr=/AmbientAudioContent/Waves/OneShots/Spot_Bird_Rare_Mountain_Crow_03}}
]
properties {
 attenuationSettings=&Fort_TileAmbient_OneShot_Bird_Att
 bEnableBusSends=false
 maxDistance=8005
 soundClassObject=&Fort_Ambient_OneShots_Outdoors
 volumeMultiplier=0.2
 preEffectBusSends [{}]
 soundSubmixSends [{maxSendLevel=1.2, minSendLevel=0.4, sendLevel=0, soundSubmix=/Engine/EngineSounds/Submixes/MasterReverbSubmixDefault}]
}
