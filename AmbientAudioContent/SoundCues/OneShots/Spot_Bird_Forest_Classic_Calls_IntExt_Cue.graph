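# Sound Cue for randomized one-shot forest bird calls, played as quiet ambient spot sounds with bird-specific attenuation and the ambient one-shot sound class.
# A quality-level switch sends the first two levels to an equally weighted pool of 32 calls and the last two to a pool of four.
# Each pool passes through a modulator for slight pitch and volume variation.
# The small pool is excluded from branch culling.
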SoundCue /AmbientAudioContent/SoundCues/OneShots/Spot_Bird_Forest_Classic_Calls_IntExt_Cue
root=QualityLevel_0
nodes [
 {id=Modulator_1, children [Random_0], properties {pitchMax=1.1, volumeMax=1.1, volumeMin=0.9}}
 {id=Modulator_2, children [Random_3], properties {pitchMax=1.1, volumeMax=1.1, volumeMin=0.9}}
 {id=QualityLevel_0, children [Modulator_1,Modulator_1,Modulator_2,Modulator_2]}
 {
  id=Random_0
  children [
   WavePlayer_0
   WavePlayer_1
   WavePlayer_2
   WavePlayer_3
   WavePlayer_4
   WavePlayer_5
   WavePlayer_6
   WavePlayer_7
   WavePlayer_8
   WavePlayer_46
   WavePlayer_47
   WavePlayer_48
   WavePlayer_49
   WavePlayer_50
   WavePlayer_51
   WavePlayer_52
   WavePlayer_53
   WavePlayer_54
   WavePlayer_55
   WavePlayer_56
   WavePlayer_57
   WavePlayer_58
   WavePlayer_59
   WavePlayer_60
   WavePlayer_61
   WavePlayer_62
   WavePlayer_63
   WavePlayer_64
   WavePlayer_65
   WavePlayer_66
   WavePlayer_67
   WavePlayer_68
  ]
  properties {weights [1,1,1,1,1,1,1,1,1,1,1,1,1,1,1,1,1,1,1,1,1,1,1,1,1,1,1,1,1,1,1,1]}
 }
 {
  id=Random_3
  children [WavePlayer_42, WavePlayer_43, WavePlayer_44, WavePlayer_45]
  properties {bShouldExcludeFromBranchCulling=true, weights [1,1,1,1]}
 }
 {id=WavePlayer_0, properties {soundWaveAssetPtr=/AmbientAudioContent/Waves/OneShots/Spot_Bird_Forest_Classic_Call_33}}
 {id=WavePlayer_1, properties {soundWaveAssetPtr=/AmbientAudioContent/Waves/OneShots/Spot_Bird_Forest_Classic_Call_32}}
 {id=WavePlayer_2, properties {soundWaveAssetPtr=/AmbientAudioContent/Waves/OneShots/Spot_Bird_Forest_Classic_Call_31}}
 {id=WavePlayer_3, properties {soundWaveAssetPtr=/AmbientAudioContent/Waves/OneShots/Spot_Bird_Forest_Classic_Call_30}}
 {id=WavePlayer_4, properties {soundWaveAssetPtr=/AmbientAudioContent/Waves/OneShots/Spot_Bird_Forest_Classic_Call_29}}
 {id=WavePlayer_42, properties {soundWaveAssetPtr=/AmbientAudioContent/Waves/OneShots/Spot_Bird_Forest_Classic_Call_01}}
 {id=WavePlayer_43, properties {soundWaveAssetPtr=/AmbientAudioContent/Waves/OneShots/Spot_Bird_Forest_Classic_Call_11}}
 {id=WavePlayer_44, properties {soundWaveAssetPtr=/AmbientAudioContent/Waves/OneShots/Spot_Bird_Forest_Classic_Call_12}}
 {id=WavePlayer_45, properties {soundWaveAssetPtr=/AmbientAudioContent/Waves/OneShots/Spot_Bird_Forest_Classic_Call_26}}
 {id=WavePlayer_46, properties {soundWaveAssetPtr=/AmbientAudioContent/Waves/OneShots/Spot_Bird_Forest_Classic_Call_24}}
 {id=WavePlayer_47, properties {soundWaveAssetPtr=/AmbientAudioContent/Waves/OneShots/Spot_Bird_Forest_Classic_Call_23}}
 {id=WavePlayer_48, properties {soundWaveAssetPtr=/AmbientAudioContent/Waves/OneShots/Spot_Bird_Forest_Classic_Call_22}}
 {id=WavePlayer_49, properties {soundWaveAssetPtr=/AmbientAudioContent/Waves/OneShots/Spot_Bird_Forest_Classic_Call_21}}
 {id=WavePlayer_5, properties {soundWaveAssetPtr=/AmbientAudioContent/Waves/OneShots/Spot_Bird_Forest_Classic_Call_28}}
 {id=WavePlayer_50, properties {soundWaveAssetPtr=/AmbientAudioContent/Waves/OneShots/Spot_Bird_Forest_Classic_Call_20}}
 {id=WavePlayer_51, properties {soundWaveAssetPtr=/AmbientAudioContent/Waves/OneShots/Spot_Bird_Forest_Classic_Call_19}}
 {id=WavePlayer_52, properties {soundWaveAssetPtr=/AmbientAudioContent/Waves/OneShots/Spot_Bird_Forest_Classic_Call_18}}
 {id=WavePlayer_53, properties {soundWaveAssetPtr=/AmbientAudioContent/Waves/OneShots/Spot_Bird_Forest_Classic_Call_17}}
 {id=WavePlayer_54, properties {soundWaveAssetPtr=/AmbientAudioContent/Waves/OneShots/Spot_Bird_Forest_Classic_Call_16}}
 {id=WavePlayer_55, properties {soundWaveAssetPtr=/AmbientAudioContent/Waves/OneShots/Spot_Bird_Forest_Classic_Call_15}}
 {id=WavePlayer_56, properties {soundWaveAssetPtr=/AmbientAudioContent/Waves/OneShots/Spot_Bird_Forest_Classic_Call_14}}
 {id=WavePlayer_57, properties {soundWaveAssetPtr=/AmbientAudioContent/Waves/OneShots/Spot_Bird_Forest_Classic_Call_13}}
 {id=WavePlayer_58, properties {soundWaveAssetPtr=&Spot_Bird_Forest_Classic_Call_12}}
 {id=WavePlayer_59, properties {soundWaveAssetPtr=&Spot_Bird_Forest_Classic_Call_11}}
 {id=WavePlayer_6, properties {soundWaveAssetPtr=/AmbientAudioContent/Waves/OneShots/Spot_Bird_Forest_Classic_Call_27}}
 {id=WavePlayer_60, properties {soundWaveAssetPtr=/AmbientAudioContent/Waves/OneShots/Spot_Bird_Forest_Classic_Call_10}}
 {id=WavePlayer_61, properties {soundWaveAssetPtr=/AmbientAudioContent/Waves/OneShots/Spot_Bird_Forest_Classic_Call_09}}
 {id=WavePlayer_62, properties {soundWaveAssetPtr=/AmbientAudioContent/Waves/OneShots/Spot_Bird_Forest_Classic_Call_08}}
 {id=WavePlayer_63, properties {soundWaveAssetPtr=/AmbientAudioContent/Waves/OneShots/Spot_Bird_Forest_Classic_Call_07}}
 {id=WavePlayer_64, properties {soundWaveAssetPtr=/AmbientAudioContent/Waves/OneShots/Spot_Bird_Forest_Classic_Call_06}}
 {id=WavePlayer_65, properties {soundWaveAssetPtr=/AmbientAudioContent/Waves/OneShots/Spot_Bird_Forest_Classic_Call_05}}
 {id=WavePlayer_66, properties {soundWaveAssetPtr=/AmbientAudioContent/Waves/OneShots/Spot_Bird_Forest_Classic_Call_04}}
 {id=WavePlayer_67, properties {soundWaveAssetPtr=/AmbientAudioContent/Waves/OneShots/Spot_Bird_Forest_Classic_Call_03}}
 {id=WavePlayer_68, properties {soundWaveAssetPtr=/AmbientAudioContent/Waves/OneShots/Spot_Bird_Forest_Classic_Call_02}}
 {id=WavePlayer_7, properties {soundWaveAssetPtr=&Spot_Bird_Forest_Classic_Call_26}}
 {id=WavePlayer_8, properties {soundWaveAssetPtr=/AmbientAudioContent/Waves/OneShots/Spot_Bird_Forest_Classic_Call_25}}
]
properties {
 attenuationSettings=&Fort_TileAmbient_OneShot_Bird_Att
 bEnableBusSends=false
 maxDistance=8005
 soundClassObject=&Fort_Ambient_OneShots
 volumeMultiplier=0.15
 preEffectBusSends [{}]
}
