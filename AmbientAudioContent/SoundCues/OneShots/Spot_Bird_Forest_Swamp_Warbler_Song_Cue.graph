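# Sound cue for swamp warbler bird song one-shots, played as outdoor forest ambience with bird-specific attenuation.
# A quality level node routes its four inputs in pairs to two modulators, each fed by a random node over the same seven song waves with equal weights.
# It sends to the exterior ambient source bus, with the send level set by distance through a custom curve.
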SoundCue /AmbientAudioContent/SoundCues/OneShots/Spot_Bird_Forest_Swamp_Warbler_Song_Cue
root=QualityLevel_0
nodes [
 {id=Modulator_0, children [Random_0]}
 {id=Modulator_1, children [Random_2]}
 {id=QualityLevel_0, children [Modulator_0,Modulator_0,Modulator_1,Modulator_1]}
 {
  id=Random_0
  children [WavePlayer_2, WavePlayer_3, WavePlayer_5, WavePlayer_9, WavePlayer_10, WavePlayer_11, WavePlayer_12]
  properties {weights [1,1,1,1,1,1,1]}
 }
 {
  id=Random_2
  children [WavePlayer_13, WavePlayer_14, WavePlayer_15, WavePlayer_16, WavePlayer_17, WavePlayer_18, WavePlayer_19]
  properties {weights [1,1,1,1,1,1,1]}
 }
 {id=WavePlayer_10, properties {soundWaveAssetPtr=/Game/Sounds/Helios/Biomes/ForgeForest/Spots/Spot_Bird_Forest_Swamp_Warbler_Song_05}}
 {id=WavePlayer_11, properties {soundWaveAssetPtr=/Game/Sounds/Helios/Biomes/ForgeForest/Spots/Spot_Bird_Forest_Swamp_Warbler_Song_06}}
 {id=WavePlayer_12, properties {soundWaveAssetPtr=/Game/Sounds/Helios/Biomes/ForgeForest/Spots/Spot_Bird_Forest_Swamp_Warbler_Song_07}}
 {id=WavePlayer_13, properties {soundWaveAssetPtr=/Game/Sounds/Helios/Biomes/ForgeForest/Spots/Spot_Bird_Forest_Swamp_Warbler_Song_01}}
 {id=WavePlayer_14, properties {soundWaveAssetPtr=/Game/Sounds/Helios/Biomes/ForgeForest/Spots/Spot_Bird_Forest_Swamp_Warbler_Song_02}}
 {id=WavePlayer_15, properties {soundWaveAssetPtr=/Game/Sounds/Helios/Biomes/ForgeForest/Spots/Spot_Bird_Forest_Swamp_Warbler_Song_03}}
 {id=WavePlayer_16, properties {soundWaveAssetPtr=/Game/Sounds/Helios/Biomes/ForgeForest/Spots/Spot_Bird_Forest_Swamp_Warbler_Song_04}}
 {id=WavePlayer_17, properties {soundWaveAssetPtr=&Spot_Bird_Forest_Swamp_Warbler_Song_05}}
 {id=WavePlayer_18, properties {soundWaveAssetPtr=&Spot_Bird_Forest_Swamp_Warbler_Song_06}}
 {id=WavePlayer_19, properties {soundWaveAssetPtr=&Spot_Bird_Forest_Swamp_Warbler_Song_07}}
 {id=WavePlayer_2, properties {soundWaveAssetPtr=&Spot_Bird_Forest_Swamp_Warbler_Song_01}}
 {id=WavePlayer_3, properties {soundWaveAssetPtr=&Spot_Bird_Forest_Swamp_Warbler_Song_02}}
 {id=WavePlayer_5, properties {soundWaveAssetPtr=&Spot_Bird_Forest_Swamp_Warbler_Song_03}}
 {id=WavePlayer_9, properties {soundWaveAssetPtr=&Spot_Bird_Forest_Swamp_Warbler_Song_04}}
]
properties {
 attenuationSettings=&Fort_TileAmbient_OneShot_Bird_Att
 maxDistance=8005
 soundClassObject=&Fort_Ambient_OneShots_Outdoors
 volumeMultiplier=0.125
 busSends [{
   audioBus=null
   maxSendDistance=6000
   minSendDistance=3000
   minSendLevel=0.6
   sendLevel=1
   soundSourceBus=&AmbientAudio_Exterior_SourceBus
   customSendLevelCurve {editorCurveData {keys [<at 0 0.6>,<at 1 1>]}}
  }]
}
