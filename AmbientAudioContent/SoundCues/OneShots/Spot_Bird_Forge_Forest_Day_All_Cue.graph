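# Sound cue for spot one-shot bird calls in the daytime Forge forest ambience, using Eastern European bird and Redwing waves.
# A random node picks one of ten bird groups, each randomizing its takes with slight pitch variation and, for some, reduced volume.
# The quality node gives the first two levels this full pool and the last two only the Bird 07 group.
# Outdoor one-shot class, bird attenuation, two-voice concurrency and an exterior bus send shape playback.
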SoundCue /AmbientAudioContent/SoundCues/OneShots/Spot_Bird_Forge_Forest_Day_All_Cue
root=QualityLevel_0
nodes [
 {id=Modulator_0, children [Random_1], properties {pitchMax=1.1, pitchMin=0.9}}
 {id=Modulator_10, children [Random_23], properties {pitchMax=1.1, pitchMin=0.9, volumeMax=0.7, volumeMin=0.4}}
 {id=Modulator_11, children [Random_24], properties {pitchMax=1.1, pitchMin=0.9}}
 {id=Modulator_2, children [Random_2], properties {pitchMax=1.1, pitchMin=0.9, volumeMax=0.7, volumeMin=0.4}}
 {id=Modulator_3, children [Random_4], properties {pitchMax=1.1, pitchMin=0.9}}
 {id=Modulator_5, children [Random_10], properties {pitchMax=1.1, pitchMin=0.9}}
 {id=Modulator_6, children [Random_17], properties {pitchMax=1.1, pitchMin=0.9, volumeMax=0.3, volumeMin=0.2}}
 {id=Modulator_7, children [Random_20], properties {pitchMax=1.1, pitchMin=0.9}}
 {id=Modulator_8, children [Random_21], properties {pitchMax=1.1, pitchMin=0.9, volumeMax=0.3, volumeMin=0.2}}
 {id=Modulator_9, children [Random_22], properties {pitchMax=1.1, pitchMin=0.9}}
 {id=QualityLevel_0, children [Random_18,Random_18,Modulator_7,Modulator_7]}
 {id=Random_1, children [WavePlayer_138, WavePlayer_139, WavePlayer_140], properties {weights [1,1,1]}}
 {id=Random_10, children [WavePlayer_150, WavePlayer_151, WavePlayer_152], properties {weights [1,1,1]}}
 {id=Random_17, children [WavePlayer_153, WavePlayer_154, WavePlayer_155], properties {weights [1,1,1]}}
 {
  id=Random_18
  children [Modulator_0,Modulator_2,Modulator_3,Modulator_5,Modulator_6,Modulator_7,Modulator_9,Modulator_10,Modulator_11,Modulator_8]
  properties {weights [1,1,1,1,1,1,1,1,1,1]}
 }
 {id=Random_2, children [WavePlayer_141, WavePlayer_142, WavePlayer_143], properties {weights [1,1,1]}}
 {id=Random_20, children [WavePlayer_156, WavePlayer_157, WavePlayer_158, WavePlayer_159, WavePlayer_160], properties {weights [1,1,1,1,1]}}
 {id=Random_21, children [WavePlayer_161, WavePlayer_162, WavePlayer_163], properties {weights [1,1,1]}}
 {id=Random_22, children [WavePlayer_164, WavePlayer_165, WavePlayer_166], properties {weights [1,1,1]}}
 {id=Random_23, children [WavePlayer_167, WavePlayer_168], properties {weights [1,1]}}
 {
  id=Random_24
  children [WavePlayer_169, WavePlayer_170, WavePlayer_171, WavePlayer_172, WavePlayer_173, WavePlayer_174]
  properties {weights [1,1,1,1,1,1]}
 }
 {id=Random_4, children [WavePlayer_144, WavePlayer_145, WavePlayer_146], properties {weights [1,1,1]}}
 {id=WavePlayer_138, properties {soundWaveAssetPtr=/AmbientAudioContent/Waves/OneShots/Spot_Bird_01_Day_Eastern_Europe_01}}
 {id=WavePlayer_139, properties {soundWaveAssetPtr=/AmbientAudioContent/Waves/OneShots/Spot_Bird_01_Day_Eastern_Europe_02}}
 {id=WavePlayer_140, properties {soundWaveAssetPtr=/AmbientAudioContent/Waves/OneShots/Spot_Bird_01_Day_Eastern_Europe_03}}
 {id=WavePlayer_141, properties {soundWaveAssetPtr=/AmbientAudioContent/Waves/OneShots/Spot_Bird_02_Day_Eastern_Europe_01}}
 {id=WavePlayer_142, properties {soundWaveAssetPtr=/AmbientAudioContent/Waves/OneShots/Spot_Bird_02_Day_Eastern_Europe_02}}
 {id=WavePlayer_143, properties {soundWaveAssetPtr=/AmbientAudioContent/Waves/OneShots/Spot_Bird_02_Day_Eastern_Europe_03}}
 {id=WavePlayer_144, properties {soundWaveAssetPtr=/AmbientAudioContent/Waves/OneShots/Spot_Bird_03_Day_Eastern_Europe_01}}
 {id=WavePlayer_145, properties {soundWaveAssetPtr=/AmbientAudioContent/Waves/OneShots/Spot_Bird_03_Day_Eastern_Europe_02}}
 {id=WavePlayer_146, properties {soundWaveAssetPtr=/AmbientAudioContent/Waves/OneShots/Spot_Bird_03_Day_Eastern_Europe_03}}
 {id=WavePlayer_150, properties {soundWaveAssetPtr=/AmbientAudioContent/Waves/OneShots/Spot_Bird_05_Day_Eastern_Europe_01}}
 {id=WavePlayer_151, properties {soundWaveAssetPtr=/AmbientAudioContent/Waves/OneShots/Spot_Bird_05_Day_Eastern_Europe_02}}
 {id=WavePlayer_152, properties {soundWaveAssetPtr=/AmbientAudioContent/Waves/OneShots/Spot_Bird_05_Day_Eastern_Europe_03}}
 {id=WavePlayer_153, properties {soundWaveAssetPtr=/AmbientAudioContent/Waves/OneShots/Spot_Bird_06_Day_Eastern_Europe_01}}
 {id=WavePlayer_154, properties {soundWaveAssetPtr=/AmbientAudioContent/Waves/OneShots/Spot_Bird_06_Day_Eastern_Europe_02}}
 {id=WavePlayer_155, properties {soundWaveAssetPtr=/AmbientAudioContent/Waves/OneShots/Spot_Bird_06_Day_Eastern_Europe_03}}
 {id=WavePlayer_156, properties {soundWaveAssetPtr=/AmbientAudioContent/Waves/OneShots/Spot_Bird_07_Day_Eastern_Europe_01}}
 {id=WavePlayer_157, properties {soundWaveAssetPtr=/AmbientAudioContent/Waves/OneShots/Spot_Bird_07_Day_Eastern_Europe_02}}
 {id=WavePlayer_158, properties {soundWaveAssetPtr=/AmbientAudioContent/Waves/OneShots/Spot_Bird_07_Day_Eastern_Europe_03}}
 {id=WavePlayer_159, properties {soundWaveAssetPtr=/AmbientAudioContent/Waves/OneShots/Spot_Bird_07_Day_Eastern_Europe_04}}
 {id=WavePlayer_160, properties {soundWaveAssetPtr=/AmbientAudioContent/Waves/OneShots/Spot_Bird_07_Day_Eastern_Europe_05}}
 {id=WavePlayer_161, properties {soundWaveAssetPtr=/AmbientAudioContent/Waves/OneShots/Spot_Bird_08_Day_Eastern_Europe_01}}
 {id=WavePlayer_162, properties {soundWaveAssetPtr=/AmbientAudioContent/Waves/OneShots/Spot_Bird_08_Day_Eastern_Europe_02}}
 {id=WavePlayer_163, properties {soundWaveAssetPtr=/AmbientAudioContent/Waves/OneShots/Spot_Bird_08_Day_Eastern_Europe_03}}
 {id=WavePlayer_164, properties {soundWaveAssetPtr=/AmbientAudioContent/Waves/OneShots/Spot_Bird_09_Day_Eastern_Europe_01}}
 {id=WavePlayer_165, properties {soundWaveAssetPtr=/AmbientAudioContent/Waves/OneShots/Spot_Bird_09_Day_Eastern_Europe_02}}
 {id=WavePlayer_166, properties {soundWaveAssetPtr=/AmbientAudioContent/Waves/OneShots/Spot_Bird_09_Day_Eastern_Europe_03}}
 {id=WavePlayer_167, properties {soundWaveAssetPtr=/AmbientAudioContent/Waves/OneShots/Spot_Bird_13_Day_Eastern_Europe_01}}
 {id=WavePlayer_168, properties {soundWaveAssetPtr=/AmbientAudioContent/Waves/OneShots/Spot_Bird_13_Day_Eastern_Europe_02}}
 {id=WavePlayer_169, properties {soundWaveAssetPtr=/AmbientAudioContent/Waves/OneShots/Spot_Bird_Redwing_Chirps_01}}
 {id=WavePlayer_170, properties {soundWaveAssetPtr=/AmbientAudioContent/Waves/OneShots/Spot_Bird_Redwing_Chirps_02}}
 {id=WavePlayer_171, properties {soundWaveAssetPtr=/AmbientAudioContent/Waves/OneShots/Spot_Bird_Redwing_Chirps_03}}
 {id=WavePlayer_172, properties {soundWaveAssetPtr=/AmbientAudioContent/Waves/OneShots/Spot_Bird_Redwing_Chirps_04}}
 {id=WavePlayer_173, properties {soundWaveAssetPtr=/AmbientAudioContent/Waves/OneShots/Spot_Bird_Redwing_Chirps_05}}
 {id=WavePlayer_174, properties {soundWaveAssetPtr=/AmbientAudioContent/Waves/OneShots/Spot_Bird_Redwing_Chirps_06}}
]
properties {
 attenuationSettings=&Fort_TileAmbient_OneShot_Bird_Att
 bOverrideConcurrency=true
 maxDistance=8005
 soundClassObject=&Fort_Ambient_OneShots_Outdoors
 volumeMultiplier=0.175
 busSends [{
   audioBus=null
   maxSendDistance=6000
   minSendDistance=3000
   minSendLevel=0.6
   sendLevel=1
   soundSourceBus=&AmbientAudio_Exterior_SourceBus
   customSendLevelCurve {editorCurveData {keys [<at 0 0.4>,<at 1 1>]}}
  }]
 concurrencyOverrides {maxCount=2}
}
