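# Sound cue for one-shot evening forest bird calls in outdoor ambient audio, attenuated as a spot sound.
# A top random node picks one of seven bird groups, with DonaldSmith calls at half weight, and each group randomly picks a wave variant.
# Per-group modulators vary pitch, and some vary volume.
# All four quality levels share that tree, and a distance-scaled send feeds the exterior ambient source bus.
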SoundCue /AmbientAudioContent/SoundCues/OneShots/Spot_Bird_Forge_Forest_Evening_All_Cue
root=QualityLevel_0
nodes [
 {id=Modulator_0, children [Random_0], properties {pitchMax=1.1, pitchMin=0.9}}
 {id=Modulator_2, children [Random_1], properties {pitchMax=1.1, pitchMin=0.9}}
 {id=Modulator_3, children [Random_2], properties {pitchMax=1.1, pitchMin=0.9, volumeMax=1.1, volumeMin=1}}
 {id=Modulator_4, children [Random_3], properties {pitchMax=1.1, pitchMin=0.9}}
 {id=Modulator_5, children [Random_4], properties {pitchMax=1.1, pitchMin=0.9}}
 {id=Modulator_6, children [Random_13], properties {pitchMax=1.1, pitchMin=0.9, volumeMax=0.7, volumeMin=0.5}}
 {id=Modulator_7, children [Random_14], properties {pitchMax=1.1, pitchMin=0.9, volumeMax=0.9, volumeMin=0.8}}
 {id=QualityLevel_0, children [Random_12,Random_12,Random_12,Random_12]}
 {id=Random_0, children [WavePlayer_2, WavePlayer_7, WavePlayer_8, WavePlayer_9, WavePlayer_10], properties {weights [1,1,1,1,1]}}
 {id=Random_1, children [WavePlayer_11, WavePlayer_26, WavePlayer_27], properties {weights [1,1,1]}}
 {
  id=Random_12
  children [Modulator_0,Modulator_2,Modulator_3,Modulator_4,Modulator_5,Modulator_6,Modulator_7]
  properties {weights [1,1,1,1,1,0.5,1]}
 }
 {
  id=Random_13
  children [WavePlayer_41, WavePlayer_42, WavePlayer_43, WavePlayer_44, WavePlayer_45, WavePlayer_46]
  properties {weights [1,1,1,1,1,1]}
 }
 {id=Random_14, children [WavePlayer_47, WavePlayer_48, WavePlayer_49], properties {weights [1,1,1]}}
 {id=Random_2, children [WavePlayer_28, WavePlayer_29, WavePlayer_30, WavePlayer_31, WavePlayer_32], properties {weights [1,1,1,1,1]}}
 {id=Random_3, children [WavePlayer_33, WavePlayer_34, WavePlayer_35, WavePlayer_36], properties {weights [1,1,1,1]}}
 {id=Random_4, children [WavePlayer_37, WavePlayer_38, WavePlayer_39, WavePlayer_40], properties {weights [1,1,1,1]}}
 {id=WavePlayer_10, properties {soundWaveAssetPtr=/AmbientAudioContent/Waves/OneShots/Spot_Bird_12_Evening_Eastern_Europe_05}}
 {id=WavePlayer_11, properties {soundWaveAssetPtr=/AmbientAudioContent/Waves/OneShots/Spot_Bird_16_Evening_Eastern_Europe_01}}
 {id=WavePlayer_2, properties {soundWaveAssetPtr=/AmbientAudioContent/Waves/OneShots/Spot_Bird_12_Evening_Eastern_Europe_01}}
 {id=WavePlayer_26, properties {soundWaveAssetPtr=/AmbientAudioContent/Waves/OneShots/Spot_Bird_16_Evening_Eastern_Europe_02}}
 {id=WavePlayer_27, properties {soundWaveAssetPtr=/AmbientAudioContent/Waves/OneShots/Spot_Bird_16_Evening_Eastern_Europe_03}}
 {id=WavePlayer_28, properties {soundWaveAssetPtr=/AmbientAudioContent/Waves/OneShots/Spot_Bird_18_EveningNight_Eastern_Europe_01}}
 {id=WavePlayer_29, properties {soundWaveAssetPtr=/AmbientAudioContent/Waves/OneShots/Spot_Bird_18_EveningNight_Eastern_Europe_02}}
 {id=WavePlayer_30, properties {soundWaveAssetPtr=/AmbientAudioContent/Waves/OneShots/Spot_Bird_18_EveningNight_Eastern_Europe_03}}
 {id=WavePlayer_31, properties {soundWaveAssetPtr=/AmbientAudioContent/Waves/OneShots/Spot_Bird_18_EveningNight_Eastern_Europe_04}}
 {id=WavePlayer_32, properties {soundWaveAssetPtr=/AmbientAudioContent/Waves/OneShots/Spot_Bird_18_EveningNight_Eastern_Europe_05}}
 {id=WavePlayer_33, properties {soundWaveAssetPtr=/AmbientAudioContent/Waves/OneShots/Spot_Bird_19_EveningNight_Eastern_Europe_01}}
 {id=WavePlayer_34, properties {soundWaveAssetPtr=/AmbientAudioContent/Waves/OneShots/Spot_Bird_19_EveningNight_Eastern_Europe_02}}
 {id=WavePlayer_35, properties {soundWaveAssetPtr=/AmbientAudioContent/Waves/OneShots/Spot_Bird_19_EveningNight_Eastern_Europe_03}}
 {id=WavePlayer_36, properties {soundWaveAssetPtr=/AmbientAudioContent/Waves/OneShots/Spot_Bird_19_EveningNight_Eastern_Europe_04}}
 {id=WavePlayer_37, properties {soundWaveAssetPtr=/AmbientAudioContent/Waves/OneShots/Spot_Bird_Arctic_Snow_Fowl_Calls_Evening_01}}
 {id=WavePlayer_38, properties {soundWaveAssetPtr=/AmbientAudioContent/Waves/OneShots/Spot_Bird_Arctic_Snow_Fowl_Calls_Evening_02}}
 {id=WavePlayer_39, properties {soundWaveAssetPtr=/AmbientAudioContent/Waves/OneShots/Spot_Bird_Arctic_Snow_Fowl_Calls_Evening_03}}
 {id=WavePlayer_40, properties {soundWaveAssetPtr=/AmbientAudioContent/Waves/OneShots/Spot_Bird_Arctic_Snow_Fowl_Calls_Evening_04}}
 {id=WavePlayer_41, properties {soundWaveAssetPtr=/AmbientAudioContent/Waves/OneShots/Spot_Bird_DonaldSmith_Call_01}}
 {id=WavePlayer_42, properties {soundWaveAssetPtr=/AmbientAudioContent/Waves/OneShots/Spot_Bird_DonaldSmith_Call_02}}
 {id=WavePlayer_43, properties {soundWaveAssetPtr=/AmbientAudioContent/Waves/OneShots/Spot_Bird_DonaldSmith_Call_03}}
 {id=WavePlayer_44, properties {soundWaveAssetPtr=/AmbientAudioContent/Waves/OneShots/Spot_Bird_DonaldSmith_Call_04}}
 {id=WavePlayer_45, properties {soundWaveAssetPtr=/AmbientAudioContent/Waves/OneShots/Spot_Bird_DonaldSmith_Call_05}}
 {id=WavePlayer_46, properties {soundWaveAssetPtr=/AmbientAudioContent/Waves/OneShots/Spot_Bird_DonaldSmith_Call_06}}
 {id=WavePlayer_47, properties {soundWaveAssetPtr=/AmbientAudioContent/Waves/OneShots/Spot_Bird_Rare_Mountain_Crow_01}}
 {id=WavePlayer_48, properties {soundWaveAssetPtr=/AmbientAudioContent/Waves/OneShots/Spot_Bird_Rare_Mountain_Crow_02}}
 {id=WavePlayer_49, properties {soundWaveAssetPtr=/AmbientAudioContent/Waves/OneShots/Spot_Bird_Rare_Mountain_Crow_03}}
 {id=WavePlayer_7, properties {soundWaveAssetPtr=/AmbientAudioContent/Waves/OneShots/Spot_Bird_12_Evening_Eastern_Europe_02}}
 {id=WavePlayer_8, properties {soundWaveAssetPtr=/AmbientAudioContent/Waves/OneShots/Spot_Bird_12_Evening_Eastern_Europe_03}}
 {id=WavePlayer_9, properties {soundWaveAssetPtr=/AmbientAudioContent/Waves/OneShots/Spot_Bird_12_Evening_Eastern_Europe_04}}
]
properties {
 attenuationSettings=&Fort_TileAmbient_OneShot_Bird_Att
 maxDistance=8005
 soundClassObject=&Fort_Ambient_OneShots_Outdoors
 volumeMultiplier=0.2
 busSends [{
   audioBus=null
   maxSendDistance=6000
   minSendDistance=3000
   minSendLevel=0.6
   sendLevel=1
   soundSourceBus=&AmbientAudio_Exterior_SourceBus
   customSendLevelCurve {editorCurveData {keys [<at 0 0.4>,<at 1 1>]}}
  }]
}
