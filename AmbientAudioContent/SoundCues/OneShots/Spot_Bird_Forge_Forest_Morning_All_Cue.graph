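# Quiet one-shot ambient spot cue for morning forest birdsong, using bird attenuation, the outdoor one-shot sound class and an exterior source-bus send.
# A quality switch routes its first two levels to a seven-way random picker and the last two straight to rubythroat songs.
# The picker's modulated branches hold two duplicate rubythroat sets, one pitch-varied meadowlark set and four empty branches without waves.
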SoundCue /AmbientAudioContent/SoundCues/OneShots/Spot_Bird_Forge_Forest_Morning_All_Cue
root=QualityLevel_0
nodes [
 {id=Modulator_1, children [Random_14], properties {pitchMax=1.1, volumeMax=1, volumeMin=0.5}}
 {id=Modulator_11, children [null], properties {volumeMax=0.7, volumeMin=0.5}}
 {id=Modulator_12, children [null], properties {volumeMax=0.7, volumeMin=0.5}}
 {id=Modulator_13, children [Random_7], properties {volumeMax=0.9, volumeMin=0.7}}
 {id=Modulator_14, children [Random_6], properties {volumeMax=0.6, volumeMin=0.5}}
 {id=Modulator_15, children [Random_10], properties {volumeMax=0.6, volumeMin=0.5}}
 {id=Modulator_9, children [Random_5], properties {volumeMax=0.6, volumeMin=0.5}}
 {id=QualityLevel_0, children [Random_12,Random_12,Modulator_9,Modulator_9]}
 {
  id=Random_10
  children [WavePlayer_54, WavePlayer_55, WavePlayer_65, WavePlayer_66, WavePlayer_67, WavePlayer_68, WavePlayer_69]
  properties {weights [1,1,1,1,1,1,1]}
 }
 {
  id=Random_12
  children [Modulator_14,Modulator_13,Modulator_12,Modulator_11,Modulator_9,Modulator_15,Modulator_1]
  properties {weights [1,1,1,1,1,1,1]}
 }
 {id=Random_14, children [WavePlayer_70, WavePlayer_71, WavePlayer_72, WavePlayer_73, WavePlayer_74], properties {weights [1,1,1,1,1]}}
 {
  id=Random_5
  children [WavePlayer_47, WavePlayer_48, WavePlayer_49, WavePlayer_50, WavePlayer_51, WavePlayer_52, WavePlayer_53]
  properties {weights [1,1,1,1,1,1,1]}
 }
 {id=Random_6, children [null,null,null,null,null,null], properties {weights [1,1,1,1,1,1]}}
 {id=Random_7, children [null,null,null,null,null], properties {weights [1,1,1,1,1]}}
 {id=WavePlayer_47, properties {soundWaveAssetPtr=/AmbientAudioContent/Waves/OneShots/Spot_Bird_Mountain_Rubythroat_Singing_Morning_Day_01}}
 {id=WavePlayer_48, properties {soundWaveAssetPtr=/AmbientAudioContent/Waves/OneShots/Spot_Bird_Mountain_Rubythroat_Singing_Morning_Day_02}}
 {id=WavePlayer_49, properties {soundWaveAssetPtr=/AmbientAudioContent/Waves/OneShots/Spot_Bird_Mountain_Rubythroat_Singing_Morning_Day_03}}
 {id=WavePlayer_50, properties {soundWaveAssetPtr=/AmbientAudioContent/Waves/OneShots/Spot_Bird_Mountain_Rubythroat_Singing_Morning_Day_04}}
 {id=WavePlayer_51, properties {soundWaveAssetPtr=/AmbientAudioContent/Waves/OneShots/Spot_Bird_Mountain_Rubythroat_Singing_Morning_Day_05}}
 {id=WavePlayer_52, properties {soundWaveAssetPtr=/AmbientAudioContent/Waves/OneShots/Spot_Bird_Mountain_Rubythroat_Singing_Morning_Day_06}}
 {id=WavePlayer_53, properties {soundWaveAssetPtr=/AmbientAudioContent/Waves/OneShots/Spot_Bird_Mountain_Rubythroat_Singing_Morning_Day_07}}
 {id=WavePlayer_54, properties {soundWaveAssetPtr=&Spot_Bird_Mountain_Rubythroat_Singing_Morning_Day_01}}
 {id=WavePlayer_55, properties {soundWaveAssetPtr=&Spot_Bird_Mountain_Rubythroat_Singing_Morning_Day_02}}
 {id=WavePlayer_65, properties {soundWaveAssetPtr=&Spot_Bird_Mountain_Rubythroat_Singing_Morning_Day_03}}
 {id=WavePlayer_66, properties {soundWaveAssetPtr=&Spot_Bird_Mountain_Rubythroat_Singing_Morning_Day_04}}
 {id=WavePlayer_67, properties {soundWaveAssetPtr=&Spot_Bird_Mountain_Rubythroat_Singing_Morning_Day_05}}
 {id=WavePlayer_68, properties {soundWaveAssetPtr=&Spot_Bird_Mountain_Rubythroat_Singing_Morning_Day_06}}
 {id=WavePlayer_69, properties {soundWaveAssetPtr=&Spot_Bird_Mountain_Rubythroat_Singing_Morning_Day_07}}
 {id=WavePlayer_70, properties {soundWaveAssetPtr=/AmbientAudioContent/Waves/OneShots/Spot_Grassland_Morning_Meadowlark_Call_01}}
 {id=WavePlayer_71, properties {soundWaveAssetPtr=/AmbientAudioContent/Waves/OneShots/Spot_Grassland_Morning_Meadowlark_Call_02}}
 {id=WavePlayer_72, properties {soundWaveAssetPtr=/AmbientAudioContent/Waves/OneShots/Spot_Grassland_Morning_Meadowlark_Call_03}}
 {id=WavePlayer_73, properties {soundWaveAssetPtr=/AmbientAudioContent/Waves/OneShots/Spot_Grassland_Morning_Meadowlark_Call_04}}
 {id=WavePlayer_74, properties {soundWaveAssetPtr=/AmbientAudioContent/Waves/OneShots/Spot_Grassland_Morning_Meadowlark_Call_05}}
]
properties {
 attenuationSettings=&Fort_TileAmbient_OneShot_Bird_Att
 maxDistance=8005
 soundClassObject=&Fort_Ambient_OneShots_Outdoors
 volumeMultiplier=0.125
 busSends [{
   audioBus=null
   maxSendDistance=6000
   minSendDistance=3000
   minSendLevel=0.6
   sendLevel=1
   soundSourceBus=&AmbientAudio_Exterior_SourceBus
   customSendLevelCurve {editorCurveData {keys [<at 0 0.4>,<at 1 1>]}}
  }]
}
