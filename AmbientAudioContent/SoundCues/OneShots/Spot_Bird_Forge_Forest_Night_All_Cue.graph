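# Sound cue for one-shot night forest bird ambience, playing evening/night bird calls plus crow, raven and magpie.
# A top-level random picks one of five modulated groups, with ravens weighted rarer, then each group randomly picks a wave variation.
# All four quality levels share that same random tree.
# It uses bird attenuation, the outdoor one-shot sound class, and distance-based sends to an exterior source bus and reverb submix.
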SoundCue /AmbientAudioContent/SoundCues/OneShots/Spot_Bird_Forge_Forest_Night_All_Cue
root=QualityLevel_0
nodes [
 {id=Modulator_0, children [Random_1]}
 {id=Modulator_2, children [Random_3], properties {volumeMax=1, volumeMin=0.5}}
 {id=Modulator_3, children [Random_8], properties {volumeMax=0.5, volumeMin=0.2}}
 {id=Modulator_4, children [Random_9], properties {volumeMax=0.6, volumeMin=0.4}}
 {id=Modulator_5, children [Random_10], properties {pitchMin=0.9, volumeMax=0.5, volumeMin=0.3}}
 {id=QualityLevel_0, children [Random_7,Random_7,Random_7,Random_7]}
 {id=Random_1, children [WavePlayer_0, WavePlayer_1, WavePlayer_5, WavePlayer_6, WavePlayer_12], properties {weights [1,1,1,1,1]}}
 {id=Random_10, children [WavePlayer_37, WavePlayer_38, WavePlayer_39, WavePlayer_40], properties {weights [1,1,1,1]}}
 {id=Random_3, children [WavePlayer_13, WavePlayer_14, WavePlayer_15, WavePlayer_28], properties {weights [1,1,1,1]}}
 {id=Random_7, children [Modulator_0,Modulator_2,Modulator_3,Modulator_4,Modulator_5], properties {weights [1,1,1,0.2,1]}}
 {id=Random_8, children [WavePlayer_29, WavePlayer_30, WavePlayer_31, WavePlayer_32], properties {weights [1,1,1,1]}}
 {id=Random_9, children [WavePlayer_33, WavePlayer_34, WavePlayer_35, WavePlayer_36], properties {weights [1,1,1,1]}}
 {id=WavePlayer_0, properties {soundWaveAssetPtr=/AmbientAudioContent/Waves/OneShots/Spot_Bird_18_EveningNight_Eastern_Europe_01}}
 {id=WavePlayer_1, properties {soundWaveAssetPtr=/AmbientAudioContent/Waves/OneShots/Spot_Bird_18_EveningNight_Eastern_Europe_02}}
 {id=WavePlayer_12, properties {soundWaveAssetPtr=/AmbientAudioContent/Waves/OneShots/Spot_Bird_18_EveningNight_Eastern_Europe_05}}
 {id=WavePlayer_13, properties {soundWaveAssetPtr=/AmbientAudioContent/Waves/OneShots/Spot_Bird_19_EveningNight_Eastern_Europe_01}}
 {id=WavePlayer_14, properties {soundWaveAssetPtr=/AmbientAudioContent/Waves/OneShots/Spot_Bird_19_EveningNight_Eastern_Europe_02}}
 {id=WavePlayer_15, properties {soundWaveAssetPtr=/AmbientAudioContent/Waves/OneShots/Spot_Bird_19_EveningNight_Eastern_Europe_03}}
 {id=WavePlayer_28, properties {soundWaveAssetPtr=/AmbientAudioContent/Waves/OneShots/Spot_Bird_19_EveningNight_Eastern_Europe_04}}
 {id=WavePlayer_29, properties {soundWaveAssetPtr=/AmbientAudioContent/Waves/OneShots/Spot_Bird_Crow_DayEvening_Eastern_Europe_01}}
 {id=WavePlayer_30, properties {soundWaveAssetPtr=/AmbientAudioContent/Waves/OneShots/Spot_Bird_Crow_DayEvening_Eastern_Europe_02}}
 {id=WavePlayer_31, properties {soundWaveAssetPtr=/AmbientAudioContent/Waves/OneShots/Spot_Bird_Crow_DayEvening_Eastern_Europe_03}}
 {id=WavePlayer_32, properties {soundWaveAssetPtr=/AmbientAudioContent/Waves/OneShots/Spot_Bird_Crow_DayEvening_Eastern_Europe_04}}
 {id=WavePlayer_33, properties {soundWaveAssetPtr=/AmbientAudioContent/Waves/OneShots/Spot_Grassland_Evening_Raven_Call_01}}
 {id=WavePlayer_34, properties {soundWaveAssetPtr=/AmbientAudioContent/Waves/OneShots/Spot_Grassland_Evening_Raven_Call_02}}
 {id=WavePlayer_35, properties {soundWaveAssetPtr=/AmbientAudioContent/Waves/OneShots/Spot_Grassland_Evening_Raven_Call_03}}
 {id=WavePlayer_36, properties {soundWaveAssetPtr=/AmbientAudioContent/Waves/OneShots/Spot_Grassland_Evening_Raven_Call_04}}
 {id=WavePlayer_37, properties {soundWaveAssetPtr=/AmbientAudioContent/Waves/OneShots/Spot_Bird_BlackBilled_Magpie_Calls_01}}
 {id=WavePlayer_38, properties {soundWaveAssetPtr=/AmbientAudioContent/Waves/OneShots/Spot_Bird_BlackBilled_Magpie_Calls_02}}
 {id=WavePlayer_39, properties {soundWaveAssetPtr=/AmbientAudioContent/Waves/OneShots/Spot_Bird_BlackBilled_Magpie_Calls_03}}
 {id=WavePlayer_40, properties {soundWaveAssetPtr=/AmbientAudioContent/Waves/OneShots/Spot_Bird_BlackBilled_Magpie_Calls_04}}
 {id=WavePlayer_5, properties {soundWaveAssetPtr=/AmbientAudioContent/Waves/OneShots/Spot_Bird_18_EveningNight_Eastern_Europe_03}}
 {id=WavePlayer_6, properties {soundWaveAssetPtr=/AmbientAudioContent/Waves/OneShots/Spot_Bird_18_EveningNight_Eastern_Europe_04}}
]
properties {
 attenuationSettings=&Fort_TileAmbient_OneShot_Bird_Att
 maxDistance=8005
 soundClassObject=&Fort_Ambient_OneShots_Outdoors
 volumeMultiplier=0.251
 busSends [{
   audioBus=null
   maxSendDistance=6000
   minSendDistance=3000
   minSendLevel=0.6
   sendLevel=1
   soundSourceBus=&AmbientAudio_Exterior_SourceBus
   customSendLevelCurve {editorCurveData {keys [<at 0 0.6>,<at 1 1>]}}
  }]
 soundSubmixSends [{
   maxSendDistance=8000
   minSendDistance=3000
   minSendLevel=0
   soundSubmix=&Fortnite_Reverb_Default
   customSendLevelCurve {editorCurveData {keys [<at 0 0.25>,<at 0.6 0.3>,<at 1 0.66>]}}
  }]
}
