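# Sound cue for randomized grassland bird-call one-shots in outdoor ambience, with bird attenuation, two-voice concurrency, and exterior-bus and reverb sends.
# The first two quality-level slots randomly play either one pitch/volume-modulated call or that call mixed with a quieter, delayed call from a second pool.
# The other two slots play a modulated pick from just five waves, referenced from a separate Artemis grassland folder and excluded from branch culling.
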SoundCue /AmbientAudioContent/SoundCues/OneShots/Spot_Bird_Grassland_Classic_Calls_01_Cue
root=QualityLevel_0
nodes [
 {id=Delay_0, children [Random_3], properties {delayMax=1.2, delayMin=0.5}}
 {id=Mixer_1, children [Modulator_1,Delay_0], properties {inputVolume [1,0.33]}}
 {id=Modulator_1, children [Random_2], properties {pitchMax=1.1, volumeMax=1.2, volumeMin=0.9}}
 {id=Modulator_2, children [Random_1], properties {pitchMax=1.1, volumeMax=1.1, volumeMin=0.9}}
 {id=QualityLevel_0, children [Random_4,Random_4,Modulator_2,Modulator_2]}
 {
  id=Random_1
  children [WavePlayer_28, WavePlayer_29, WavePlayer_30, WavePlayer_31, WavePlayer_32]
  properties {bShouldExcludeFromBranchCulling=true, weights [1,1,1,1,1]}
 }
 {
  id=Random_2
  children [
   WavePlayer_33
   WavePlayer_34
   WavePlayer_35
   WavePlayer_36
   WavePlayer_37
   WavePlayer_38
   WavePlayer_39
   WavePlayer_40
   WavePlayer_41
   WavePlayer_42
   WavePlayer_43
   WavePlayer_44
   WavePlayer_45
   WavePlayer_46
   WavePlayer_47
   WavePlayer_48
   WavePlayer_49
   WavePlayer_50
   WavePlayer_51
   WavePlayer_52
   WavePlayer_53
   WavePlayer_54
   WavePlayer_55
   WavePlayer_56
   WavePlayer_57
   WavePlayer_58
   WavePlayer_59
   WavePlayer_60
   WavePlayer_61
   WavePlayer_62
   WavePlayer_63
  ]
  properties {weights [1,1,1,1,1,1,1,1,1,1,1,1,1,1,1,1,1,1,1,1,1,1,1,1,1,1,1,1,1,1,1]}
 }
 {
  id=Random_3
  children [
   WavePlayer_64
   WavePlayer_65
   WavePlayer_66
   WavePlayer_67
   WavePlayer_68
   WavePlayer_69
   WavePlayer_70
   WavePlayer_71
   WavePlayer_72
   WavePlayer_73
   WavePlayer_74
   WavePlayer_75
   WavePlayer_76
   WavePlayer_77
   WavePlayer_78
   WavePlayer_79
   WavePlayer_80
   WavePlayer_81
   WavePlayer_82
   WavePlayer_83
   WavePlayer_84
   WavePlayer_85
   WavePlayer_86
   WavePlayer_87
   WavePlayer_88
   WavePlayer_89
  ]
  properties {weights [1,1,1,1,1,1,1,1,1,1,1,1,1,1,1,1,1,1,1,1,1,1,1,1,1,1]}
 }
 {id=Random_4, children [Modulator_1,Mixer_1], properties {weights [1,1]}}
 {id=WavePlayer_28, properties {soundWaveAssetPtr=/Game/Sounds/Artemis/Biomes/Grasslands/Grassland_Spots/Spot_Bird_Grassland_Classic_Call_31}}
 {id=WavePlayer_29, properties {soundWaveAssetPtr=/Game/Sounds/Artemis/Biomes/Grasslands/Grassland_Spots/Spot_Bird_Grassland_Classic_Call_27}}
 {id=WavePlayer_30, properties {soundWaveAssetPtr=/Game/Sounds/Artemis/Biomes/Grasslands/Grassland_Spots/Spot_Bird_Grassland_Classic_Call_28}}
 {id=WavePlayer_31, properties {soundWaveAssetPtr=/Game/Sounds/Artemis/Biomes/Grasslands/Grassland_Spots/Spot_Bird_Grassland_Classic_Call_29}}
 {id=WavePlayer_32, properties {soundWaveAssetPtr=/Game/Sounds/Artemis/Biomes/Grasslands/Grassland_Spots/Spot_Bird_Grassland_Classic_Call_30}}
 {id=WavePlayer_33, properties {soundWaveAssetPtr=/AmbientAudioContent/Waves/OneShots/Spot_Bird_Grassland_Classic_Call_01}}
 {id=WavePlayer_34, properties {soundWaveAssetPtr=/AmbientAudioContent/Waves/OneShots/Spot_Bird_Grassland_Classic_Call_02}}
 {id=WavePlayer_35, properties {soundWaveAssetPtr=/AmbientAudioContent/Waves/OneShots/Spot_Bird_Grassland_Classic_Call_03}}
 {id=WavePlayer_36, properties {soundWaveAssetPtr=/AmbientAudioContent/Waves/OneShots/Spot_Bird_Grassland_Classic_Call_04}}
 {id=WavePlayer_37, properties {soundWaveAssetPtr=/AmbientAudioContent/Waves/OneShots/Spot_Bird_Grassland_Classic_Call_05}}
 {id=WavePlayer_38, properties {soundWaveAssetPtr=/AmbientAudioContent/Waves/OneShots/Spot_Bird_Grassland_Classic_Call_06}}
 {id=WavePlayer_39, properties {soundWaveAssetPtr=/AmbientAudioContent/Waves/OneShots/Spot_Bird_Grassland_Classic_Call_07}}
 {id=WavePlayer_40, properties {soundWaveAssetPtr=/AmbientAudioContent/Waves/OneShots/Spot_Bird_Grassland_Classic_Call_08}}
 {id=WavePlayer_41, properties {soundWaveAssetPtr=/AmbientAudioContent/Waves/OneShots/Spot_Bird_Grassland_Classic_Call_09}}
 {id=WavePlayer_42, properties {soundWaveAssetPtr=/AmbientAudioContent/Waves/OneShots/Spot_Bird_Grassland_Classic_Call_10}}
 {id=WavePlayer_43, properties {soundWaveAssetPtr=/AmbientAudioContent/Waves/OneShots/Spot_Bird_Grassland_Classic_Call_11}}
 {id=WavePlayer_44, properties {soundWaveAssetPtr=/AmbientAudioContent/Waves/OneShots/Spot_Bird_Grassland_Classic_Call_12}}
 {id=WavePlayer_45, properties {soundWaveAssetPtr=/AmbientAudioContent/Waves/OneShots/Spot_Bird_Grassland_Classic_Call_13}}
 {id=WavePlayer_46, properties {soundWaveAssetPtr=/AmbientAudioContent/Waves/OneShots/Spot_Bird_Grassland_Classic_Call_14}}
 {id=WavePlayer_47, properties {soundWaveAssetPtr=/AmbientAudioContent/Waves/OneShots/Spot_Bird_Grassland_Classic_Call_15}}
 {id=WavePlayer_48, properties {soundWaveAssetPtr=/AmbientAudioContent/Waves/OneShots/Spot_Bird_Grassland_Classic_Call_16}}
 {id=WavePlayer_49, properties {soundWaveAssetPtr=/AmbientAudioContent/Waves/OneShots/Spot_Bird_Grassland_Classic_Call_17}}
 {id=WavePlayer_50, properties {soundWaveAssetPtr=/AmbientAudioContent/Waves/OneShots/Spot_Bird_Grassland_Classic_Call_18}}
 {id=WavePlayer_51, properties {soundWaveAssetPtr=/AmbientAudioContent/Waves/OneShots/Spot_Bird_Grassland_Classic_Call_19}}
 {id=WavePlayer_52, properties {soundWaveAssetPtr=/AmbientAudioContent/Waves/OneShots/Spot_Bird_Grassland_Classic_Call_20}}
 {id=WavePlayer_53, properties {soundWaveAssetPtr=/AmbientAudioContent/Waves/OneShots/Spot_Bird_Grassland_Classic_Call_21}}
 {id=WavePlayer_54, properties {soundWaveAssetPtr=/AmbientAudioContent/Waves/OneShots/Spot_Bird_Grassland_Classic_Call_22}}
 {id=WavePlayer_55, properties {soundWaveAssetPtr=/AmbientAudioContent/Waves/OneShots/Spot_Bird_Grassland_Classic_Call_23}}
 {id=WavePlayer_56, properties {soundWaveAssetPtr=/AmbientAudioContent/Waves/OneShots/Spot_Bird_Grassland_Classic_Call_24}}
 {id=WavePlayer_57, properties {soundWaveAssetPtr=/AmbientAudioContent/Waves/OneShots/Spot_Bird_Grassland_Classic_Call_25}}
 {id=WavePlayer_58, properties {soundWaveAssetPtr=/AmbientAudioContent/Waves/OneShots/Spot_Bird_Grassland_Classic_Call_26}}
 {id=WavePlayer_59, properties {soundWaveAssetPtr=/AmbientAudioContent/Waves/OneShots/Spot_Bird_Grassland_Classic_Call_27}}
 {id=WavePlayer_60, properties {soundWaveAssetPtr=/AmbientAudioContent/Waves/OneShots/Spot_Bird_Grassland_Classic_Call_28}}
 {id=WavePlayer_61, properties {soundWaveAssetPtr=/AmbientAudioContent/Waves/OneShots/Spot_Bird_Grassland_Classic_Call_29}}
 {id=WavePlayer_62, properties {soundWaveAssetPtr=/AmbientAudioContent/Waves/OneShots/Spot_Bird_Grassland_Classic_Call_30}}
 {id=WavePlayer_63, properties {soundWaveAssetPtr=/AmbientAudioContent/Waves/OneShots/Spot_Bird_Grassland_Classic_Call_31}}
 {id=WavePlayer_64, properties {soundWaveAssetPtr=/AmbientAudioContent/Waves/OneShots/Spot_Bird_Grassland_Classic_Call_56}}
 {id=WavePlayer_65, properties {soundWaveAssetPtr=/AmbientAudioContent/Waves/OneShots/Spot_Bird_Grassland_Classic_Call_32}}
 {id=WavePlayer_66, properties {soundWaveAssetPtr=&Spot_Bird_Grassland_Classic_Call_31}}
 {id=WavePlayer_67, properties {soundWaveAssetPtr=/AmbientAudioContent/Waves/OneShots/Spot_Bird_Grassland_Classic_Call_33}}
 {id=WavePlayer_68, properties {soundWaveAssetPtr=/AmbientAudioContent/Waves/OneShots/Spot_Bird_Grassland_Classic_Call_34}}
 {id=WavePlayer_69, properties {soundWaveAssetPtr=/AmbientAudioContent/Waves/OneShots/Spot_Bird_Grassland_Classic_Call_35}}
 {id=WavePlayer_70, properties {soundWaveAssetPtr=/AmbientAudioContent/Waves/OneShots/Spot_Bird_Grassland_Classic_Call_36}}
 {id=WavePlayer_71, properties {soundWaveAssetPtr=/AmbientAudioContent/Waves/OneShots/Spot_Bird_Grassland_Classic_Call_37}}
 {id=WavePlayer_72, properties {soundWaveAssetPtr=/AmbientAudioContent/Waves/OneShots/Spot_Bird_Grassland_Classic_Call_38}}
 {id=WavePlayer_73, properties {soundWaveAssetPtr=/AmbientAudioContent/Waves/OneShots/Spot_Bird_Grassland_Classic_Call_39}}
 {id=WavePlayer_74, properties {soundWaveAssetPtr=/AmbientAudioContent/Waves/OneShots/Spot_Bird_Grassland_Classic_Call_40}}
 {id=WavePlayer_75, properties {soundWaveAssetPtr=/AmbientAudioContent/Waves/OneShots/Spot_Bird_Grassland_Classic_Call_41}}
 {id=WavePlayer_76, properties {soundWaveAssetPtr=/AmbientAudioContent/Waves/OneShots/Spot_Bird_Grassland_Classic_Call_42}}
 {id=WavePlayer_77, properties {soundWaveAssetPtr=/AmbientAudioContent/Waves/OneShots/Spot_Bird_Grassland_Classic_Call_43}}
 {id=WavePlayer_78, properties {soundWaveAssetPtr=/AmbientAudioContent/Waves/OneShots/Spot_Bird_Grassland_Classic_Call_44}}
 {id=WavePlayer_79, properties {soundWaveAssetPtr=/AmbientAudioContent/Waves/OneShots/Spot_Bird_Grassland_Classic_Call_45}}
 {id=WavePlayer_80, properties {soundWaveAssetPtr=/AmbientAudioContent/Waves/OneShots/Spot_Bird_Grassland_Classic_Call_46}}
 {id=WavePlayer_81, properties {soundWaveAssetPtr=/AmbientAudioContent/Waves/OneShots/Spot_Bird_Grassland_Classic_Call_47}}
 {id=WavePlayer_82, properties {soundWaveAssetPtr=/AmbientAudioContent/Waves/OneShots/Spot_Bird_Grassland_Classic_Call_48}}
 {id=WavePlayer_83, properties {soundWaveAssetPtr=/AmbientAudioContent/Waves/OneShots/Spot_Bird_Grassland_Classic_Call_49}}
 {id=WavePlayer_84, properties {soundWaveAssetPtr=/AmbientAudioContent/Waves/OneShots/Spot_Bird_Grassland_Classic_Call_50}}
 {id=WavePlayer_85, properties {soundWaveAssetPtr=/AmbientAudioContent/Waves/OneShots/Spot_Bird_Grassland_Classic_Call_51}}
 {id=WavePlayer_86, properties {soundWaveAssetPtr=/AmbientAudioContent/Waves/OneShots/Spot_Bird_Grassland_Classic_Call_52}}
 {id=WavePlayer_87, properties {soundWaveAssetPtr=/AmbientAudioContent/Waves/OneShots/Spot_Bird_Grassland_Classic_Call_53}}
 {id=WavePlayer_88, properties {soundWaveAssetPtr=/AmbientAudioContent/Waves/OneShots/Spot_Bird_Grassland_Classic_Call_54}}
 {id=WavePlayer_89, properties {soundWaveAssetPtr=/AmbientAudioContent/Waves/OneShots/Spot_Bird_Grassland_Classic_Call_55}}
]
properties {
 attenuationSettings=&Fort_TileAmbient_OneShot_Bird_Att
 bOverrideConcurrency=true
 maxDistance=8005
 soundClassObject=&Fort_Ambient_OneShots_Outdoors
 volumeMultiplier=0.15
 busSends [{
   audioBus=null
   maxSendDistance=6000
   minSendDistance=3000
   minSendLevel=0.6
   sendLevel=1
   soundSourceBus=&AmbientAudio_Exterior_SourceBus
  }]
 concurrencyOverrides {maxCount=2}
 soundSubmixSends [{
   minSendLevel=0
   soundSubmix=&Fortnite_Reverb_Default
   customSendLevelCurve {editorCurveData {keys [<at 0 0.4>,<at 1 1>]}}
  }]
}
