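# One-shot Sound Cue for Himalayan snow tern morning bird calls, an outdoor ambient spot sound using bird attenuation.
# A quality-level node at the root sends its first two slots to a random pick of fifteen waves and its last two to a pool of four.
# The smaller pool reuses the last four waves of the full set.
# Each pool passes through its own Modulator at default settings for pitch and volume variation.
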SoundCue /AmbientAudioContent/SoundCues/OneShots/Spot_Bird_Himalayan_Snow_Tern_Morning_Cue
root=QualityLevel_0
nodes [
 {id=Modulator_0, children [Random_1]}
 {id=Modulator_3, children [Random_0]}
 {id=QualityLevel_0, children [Modulator_3,Modulator_3,Modulator_0,Modulator_0]}
 {
  id=Random_0
  children [
   WavePlayer_0
   WavePlayer_1
   WavePlayer_2
   WavePlayer_3
   WavePlayer_4
   WavePlayer_5
   WavePlayer_6
   WavePlayer_7
   WavePlayer_19
   WavePlayer_20
   WavePlayer_21
   WavePlayer_22
   WavePlayer_23
   WavePlayer_32
   WavePlayer_33
  ]
  properties {weights [1,1,1,1,1,1,1,1,1,1,1,1,1,1,1]}
 }
 {id=Random_1, children [WavePlayer_34, WavePlayer_35, WavePlayer_36, WavePlayer_37], properties {weights [1,1,1,1]}}
 {id=WavePlayer_0, properties {soundWaveAssetPtr=/AmbientAudioContent/Waves/OneShots/Spot_Bird_Himalayan_Snow_Tern_Morning_Day_01}}
 {id=WavePlayer_1, properties {soundWaveAssetPtr=/AmbientAudioContent/Waves/OneShots/Spot_Bird_Himalayan_Snow_Tern_Morning_Day_02}}
 {id=WavePlayer_19, properties {soundWaveAssetPtr=/AmbientAudioContent/Waves/OneShots/Spot_Bird_Himalayan_Snow_Tern_Morning_Day_09}}
 {id=WavePlayer_2, properties {soundWaveAssetPtr=/AmbientAudioContent/Waves/OneShots/Spot_Bird_Himalayan_Snow_Tern_Morning_Day_03}}
 {id=WavePlayer_20, properties {soundWaveAssetPtr=/AmbientAudioContent/Waves/OneShots/Spot_Bird_Himalayan_Snow_Tern_Morning_Day_10}}
 {id=WavePlayer_21, properties {soundWaveAssetPtr=/AmbientAudioContent/Waves/OneShots/Spot_Bird_Himalayan_Snow_Tern_Morning_Day_11}}
 {id=WavePlayer_22, properties {soundWaveAssetPtr=/AmbientAudioContent/Waves/OneShots/Spot_Bird_Himalayan_Snow_Tern_Morning_Day_12}}
 {id=WavePlayer_23, properties {soundWaveAssetPtr=/AmbientAudioContent/Waves/OneShots/Spot_Bird_Himalayan_Snow_Tern_Morning_Day_13}}
 {id=WavePlayer_3, properties {soundWaveAssetPtr=/AmbientAudioContent/Waves/OneShots/Spot_Bird_Himalayan_Snow_Tern_Morning_Day_04}}
 {id=WavePlayer_32, properties {soundWaveAssetPtr=/AmbientAudioContent/Waves/OneShots/Spot_Bird_Himalayan_Snow_Tern_Morning_Day_14}}
 {id=WavePlayer_33, properties {soundWaveAssetPtr=/AmbientAudioContent/Waves/OneShots/Spot_Bird_Himalayan_Snow_Tern_Morning_Day_15}}
 {id=WavePlayer_34, properties {soundWaveAssetPtr=&Spot_Bird_Himalayan_Snow_Tern_Morning_Day_15}}
 {id=WavePlayer_35, properties {soundWaveAssetPtr=&Spot_Bird_Himalayan_Snow_Tern_Morning_Day_14}}
 {id=WavePlayer_36, properties {soundWaveAssetPtr=&Spot_Bird_Himalayan_Snow_Tern_Morning_Day_13}}
 {id=WavePlayer_37, properties {soundWaveAssetPtr=&Spot_Bird_Himalayan_Snow_Tern_Morning_Day_12}}
 {id=WavePlayer_4, properties {soundWaveAssetPtr=/AmbientAudioContent/Waves/OneShots/Spot_Bird_Himalayan_Snow_Tern_Morning_Day_05}}
 {id=WavePlayer_5, properties {soundWaveAssetPtr=/AmbientAudioContent/Waves/OneShots/Spot_Bird_Himalayan_Snow_Tern_Morning_Day_06}}
 {id=WavePlayer_6, properties {soundWaveAssetPtr=/AmbientAudioContent/Waves/OneShots/Spot_Bird_Himalayan_Snow_Tern_Morning_Day_07}}
 {id=WavePlayer_7, properties {soundWaveAssetPtr=/AmbientAudioContent/Waves/OneShots/Spot_Bird_Himalayan_Snow_Tern_Morning_Day_08}}
]
properties {
 attenuationSettings=&Fort_TileAmbient_OneShot_Bird_Att
 maxDistance=8005
 soundClassObject=&Fort_Ambient_OneShots_Outdoors
 volumeMultiplier=0.401
 busSends [{audioBus=null, minSendLevel=0, sendLevel=1, soundSourceBus=null}]
}
