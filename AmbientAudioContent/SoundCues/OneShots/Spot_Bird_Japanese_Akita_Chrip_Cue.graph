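# Sound cue for one-shot Japanese Akita bird chirps in outdoor ambience, using a bird attenuation asset.
# A shared random node picks from six chirp waves.
# A weighted random plays either a single modulated chirp or, less often, a mix adding a delayed, lower-pitched, quieter chirp, like an echo or answering call.
# All quality levels use this same branch.
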SoundCue /AmbientAudioContent/SoundCues/OneShots/Spot_Bird_Japanese_Akita_Chrip_Cue
root=QualityLevel_0
nodes [
 {id=Delay_0, children [Modulator_0], properties {delayMax=0.75, delayMin=0.33}}
 {id=Mixer_0, children [Modulator_1,Delay_0], properties {inputVolume [1,0.33]}}
 {id=Modulator_0, children [Random_0], properties {pitchMax=0.9, pitchMin=0.8, volumeMax=0.6, volumeMin=0.5}}
 {id=Modulator_1, children [Random_0], properties {pitchMax=1.1, pitchMin=1, volumeMax=1.1, volumeMin=0.9}}
 {id=QualityLevel_0, children [Random_5,Random_5,Random_5,Random_5]}
 {id=Random_0, children [WavePlayer_0,WavePlayer_1,WavePlayer_2,WavePlayer_3,WavePlayer_4,WavePlayer_5], properties {weights [1,1,1,1,1,1]}}
 {id=Random_5, children [Modulator_1,Mixer_0], properties {weights [1,0.33]}}
 {id=WavePlayer_0, properties {soundWaveAssetPtr=/AmbientAudioContent/Waves/OneShots/Spot_Bird_Japanese_Akita_Chirp_01}}
 {id=WavePlayer_1, properties {soundWaveAssetPtr=/AmbientAudioContent/Waves/OneShots/Spot_Bird_Japanese_Akita_Chirp_03}}
 {id=WavePlayer_2, properties {soundWaveAssetPtr=/AmbientAudioContent/Waves/OneShots/Spot_Bird_Japanese_Akita_Chirp_04}}
 {id=WavePlayer_3, properties {soundWaveAssetPtr=/AmbientAudioContent/Waves/OneShots/Spot_Bird_Japanese_Akita_Chirp_05}}
 {id=WavePlayer_4, properties {soundWaveAssetPtr=/AmbientAudioContent/Waves/OneShots/Spot_Bird_Japanese_Akita_Chirp_06}}
 {id=WavePlayer_5, properties {soundWaveAssetPtr=/AmbientAudioContent/Waves/OneShots/Spot_Bird_Japanese_Akita_Chirp_02}}
]
properties {
 attenuationSettings=/AmbientAudioBase/Attenuations/Fort_TileAmbient_OneShot_Bird_Att
 bEnableBusSends=false
 maxDistance=8005
 soundClassObject=&Fort_Ambient_OneShots_Outdoors
 volumeMultiplier=0.376
 preEffectBusSends [{}]
}
