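# Sound cue for outdoor one-shot Japanese bird calls (Aso Grassland), randomly drawn from twelve parkland waves.
# The first two quality slots play either a modulated call or, about a quarter of the time, layer it with a delayed, quieter, lower-pitched second call.
# The other two slots play only a single pitch- and volume-modulated call.
# Bird attenuation applies, with distance-based sends to exterior bus and reverb submix.
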SoundCue /AmbientAudioContent/SoundCues/OneShots/Spot_Bird_Japanese_Aso_Grassland_Call_Cue
root=QualityLevel_0
nodes [
 {id=Delay_0, children [Modulator_0], properties {delayMax=1.5, delayMin=0.75}}
 {id=Mixer_0, children [Modulator_1,Delay_0], properties {inputVolume [1,0.75]}}
 {id=Modulator_0, children [Random_0], properties {pitchMax=0.9, pitchMin=0.8, volumeMax=0.6, volumeMin=0.5}}
 {id=Modulator_1, children [Random_0], properties {pitchMax=1, pitchMin=0.8, volumeMax=1.1, volumeMin=0.9}}
 {id=Modulator_2, children [Random_0], properties {pitchMax=1.1, volumeMax=1.1, volumeMin=0.9}}
 {id=QualityLevel_0, children [Random_5,Random_5,Modulator_2,Modulator_2]}
 {
  id=Random_0
  children [
   WavePlayer_0
   WavePlayer_1
   WavePlayer_2
   WavePlayer_16
   WavePlayer_17
   WavePlayer_18
   WavePlayer_19
   WavePlayer_20
   WavePlayer_21
   WavePlayer_22
   WavePlayer_23
   WavePlayer_24
  ]
  properties {weights [1,1,1,1,1,1,1,1,1,1,1,1]}
 }
 {id=Random_5, children [Modulator_1,Mixer_0], properties {weights [1,0.33]}}
 {id=WavePlayer_0, properties {soundWaveAssetPtr=/AmbientAudioContent/Waves/OneShots/Spot_Bird_Japanese_Parkland_01}}
 {id=WavePlayer_1, properties {soundWaveAssetPtr=/AmbientAudioContent/Waves/OneShots/Spot_Bird_Japanese_Parkland_12}}
 {id=WavePlayer_16, properties {soundWaveAssetPtr=/AmbientAudioContent/Waves/OneShots/Spot_Bird_Japanese_Parkland_03}}
 {id=WavePlayer_17, properties {soundWaveAssetPtr=/AmbientAudioContent/Waves/OneShots/Spot_Bird_Japanese_Parkland_04}}
 {id=WavePlayer_18, properties {soundWaveAssetPtr=/AmbientAudioContent/Waves/OneShots/Spot_Bird_Japanese_Parkland_05}}
 {id=WavePlayer_19, properties {soundWaveAssetPtr=/AmbientAudioContent/Waves/OneShots/Spot_Bird_Japanese_Parkland_06}}
 {id=WavePlayer_2, properties {soundWaveAssetPtr=/AmbientAudioContent/Waves/OneShots/Spot_Bird_Japanese_Parkland_02}}
 {id=WavePlayer_20, properties {soundWaveAssetPtr=/AmbientAudioContent/Waves/OneShots/Spot_Bird_Japanese_Parkland_07}}
 {id=WavePlayer_21, properties {soundWaveAssetPtr=/AmbientAudioContent/Waves/OneShots/Spot_Bird_Japanese_Parkland_08}}
 {id=WavePlayer_22, properties {soundWaveAssetPtr=/AmbientAudioContent/Waves/OneShots/Spot_Bird_Japanese_Parkland_09}}
 {id=WavePlayer_23, properties {soundWaveAssetPtr=/AmbientAudioContent/Waves/OneShots/Spot_Bird_Japanese_Parkland_10}}
 {id=WavePlayer_24, properties {soundWaveAssetPtr=/AmbientAudioContent/Waves/OneShots/Spot_Bird_Japanese_Parkland_11}}
]
properties {
 attenuationSettings=/AmbientAudioBase/Attenuations/Fort_TileAmbient_OneShot_Bird_Att
 maxDistance=8005
 soundClassObject=&Fort_Ambient_OneShots_Outdoors
 volumeMultiplier=0.2
 busSends [{
   audioBus=null
   maxSendDistance=8000
   minSendDistance=3000
   minSendLevel=0.4
   sendLevel=1
   soundSourceBus=&AmbientAudio_Exterior_SourceBus
   customSendLevelCurve {editorCurveData {keys [<at 0 0.4>,<at 1 1>]}}
  }]
 soundSubmixSends [{
   maxSendDistance=8000
   minSendDistance=3000
   minSendLevel=0.4
   soundSubmix=&Fortnite_Reverb_Default
   customSendLevelCurve {editorCurveData {keys [<at 0 0.4>,<at 1 1>]}}
  }]
}
