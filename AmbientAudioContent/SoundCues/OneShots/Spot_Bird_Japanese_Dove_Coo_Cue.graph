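# Outdoor ambient one-shot Sound Cue that plays a Japanese bird call, either a dove coo or an Okutama hoot.
# An equal-weight Random node picks one of five waves and feeds a Modulator that lowers pitch and varies volume slightly.
# All four quality levels share that branch.
# Output is quietened, uses bird one-shot attenuation, and sends to the exterior ambient source bus and default reverb submix.
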SoundCue /AmbientAudioContent/SoundCues/OneShots/Spot_Bird_Japanese_Dove_Coo_Cue
root=QualityLevel_0
nodes [
 {id=Modulator_1, children [Random_0], properties {pitchMax=1, pitchMin=0.8, volumeMax=1.1, volumeMin=0.9}}
 {id=QualityLevel_0, children [Modulator_1,Modulator_1,Modulator_1,Modulator_1]}
 {id=Random_0, children [WavePlayer_0,WavePlayer_1,WavePlayer_2,WavePlayer_3,WavePlayer_4], properties {weights [1,1,1,1,1]}}
 {id=WavePlayer_0, properties {soundWaveAssetPtr=/AmbientAudioContent/Waves/OneShots/Spot_Bird_Japanese_Dove_Coo_01}}
 {id=WavePlayer_1, properties {soundWaveAssetPtr=/AmbientAudioContent/Waves/OneShots/Spot_Bird_Japanese_Dove_Coo_02}}
 {id=WavePlayer_2, properties {soundWaveAssetPtr=/AmbientAudioContent/Waves/OneShots/Spot_Bird_Japanese_Dove_Coo_03}}
 {id=WavePlayer_3, properties {soundWaveAssetPtr=/AmbientAudioContent/Waves/OneShots/Spot_Bird_Japanese_Okutama_Hoot_Call_02}}
 {id=WavePlayer_4, properties {soundWaveAssetPtr=/AmbientAudioContent/Waves/OneShots/Spot_Bird_Japanese_Okutama_Hoot_Call_01}}
]
properties {
 attenuationSettings=/AmbientAudioBase/Attenuations/Fort_TileAmbient_OneShot_Bird_Att
 maxDistance=8005
 soundClassObject=&Fort_Ambient_OneShots_Outdoors
 volumeMultiplier=0.3
 busSends [{
   audioBus=null
   maxSendDistance=8000
   minSendDistance=3000
   minSendLevel=0.4
   sendLevel=1
   soundSourceBus=&AmbientAudio_Exterior_SourceBus
   customSendLevelCurve {editorCurveData {keys [<at 0 0.4>,<at 1 1>]}}
  }]
 soundSubmixSends [{
   maxSendDistance=8000
   minSendDistance=3000
   minSendLevel=0.4
   soundSubmix=&Fortnite_Reverb_Default
   customSendLevelCurve {editorCurveData {keys [<at 0 0.4>,<at 1 1>]}}
  }]
}
