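# Sound cue for one-shot Japanese Blakiston fish owl hoots, played as outdoor ambient audio with bird attenuation.
# All branches draw from an equally weighted random pick of four owl waves, varied by pitch and volume modulators.
# The first two quality levels usually play one call, occasionally mixing in a delayed, quieter, lower-pitched second call.
# The last two quality levels skip that layering and play a single modulated call.
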SoundCue /AmbientAudioContent/SoundCues/OneShots/Spot_Bird_Japanese_Fish_Owl_Hoot_Cue
root=QualityLevel_0
nodes [
 {id=Delay_0, children [Modulator_0], properties {delayMax=1, delayMin=0.75}}
 {id=Mixer_0, children [Modulator_1,Delay_0], properties {inputVolume [1,0.75]}}
 {id=Modulator_0, children [Random_1], properties {pitchMax=0.9, pitchMin=0.8, volumeMax=0.8, volumeMin=0.5}}
 {id=Modulator_1, children [Random_1], properties {pitchMax=1, pitchMin=0.8, volumeMax=1.1, volumeMin=0.9}}
 {id=Modulator_2, children [Random_1], properties {pitchMax=1.1, volumeMax=1.1, volumeMin=0.9}}
 {id=QualityLevel_0, children [Random_5,Random_5,Modulator_2,Modulator_2]}
 {id=Random_1, children [WavePlayer_5,WavePlayer_6,WavePlayer_7,WavePlayer_8], properties {weights [1,1,1,1]}}
 {id=Random_5, children [Modulator_1,Mixer_0], properties {weights [1,0.33]}}
 {id=WavePlayer_5, properties {soundWaveAssetPtr=/AmbientAudioContent/Waves/OneShots/Spot_Bird_Japanese_Blakiston_Fish_Owl_01}}
 {id=WavePlayer_6, properties {soundWaveAssetPtr=/AmbientAudioContent/Waves/OneShots/Spot_Bird_Japanese_Blakiston_Fish_Owl_02}}
 {id=WavePlayer_7, properties {soundWaveAssetPtr=/AmbientAudioContent/Waves/OneShots/Spot_Bird_Japanese_Blakiston_Fish_Owl_03}}
 {id=WavePlayer_8, properties {soundWaveAssetPtr=/AmbientAudioContent/Waves/OneShots/Spot_Bird_Japanese_Owl_Blakiston_Hoot}}
]
properties {
 attenuationSettings=/AmbientAudioBase/Attenuations/Fort_TileAmbient_OneShot_Bird_Att
 bEnableBusSends=false
 maxDistance=8005
 soundClassObject=&Fort_Ambient_OneShots_Outdoors
 volumeMultiplier=0.4
 preEffectBusSends [{}]
}
