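# Outdoor ambient one-shot cue for spot bird calls, using goshawk and crested serpent eagle waves with attenuation and an exterior source bus send.
# The first two quality slots randomly choose a twice-looped, randomly delayed goshawk variant or a single eagle call.
# The last two slots cut variety to one goshawk wave, played either once or as a delayed double call.
# Both branches pass through Modulators for variation.
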SoundCue /AmbientAudioContent/SoundCues/OneShots/Spot_Bird_Olympus_GosHawk_Cue
root=QualityLevel_0
nodes [
 {id=Delay_0, children [Random_1], properties {delayMax=2}}
 {id=Delay_1, children [WavePlayer_13], properties {delayMax=2}}
 {id=Looping_0, children [Delay_0], properties {bLoopIndefinitely=false, loopCount=2}}
 {id=Looping_1, children [Delay_1], properties {bLoopIndefinitely=false, loopCount=2}}
 {id=Modulator_0, children [Random_6]}
 {id=Modulator_1, children [Random_4]}
 {id=QualityLevel_0, children [Modulator_0,Modulator_0,Modulator_1,Modulator_1]}
 {id=Random_1, children [WavePlayer_7,WavePlayer_8,WavePlayer_9], properties {weights [1,1,1]}}
 {id=Random_3, children [WavePlayer_10, WavePlayer_11, WavePlayer_12], properties {weights [1,1,1]}}
 {id=Random_4, children [Looping_1, WavePlayer_15], properties {weights [1,1]}}
 {id=Random_6, children [Looping_0,Random_3], properties {weights [1,1]}}
 {id=WavePlayer_10, properties {soundWaveAssetPtr=/AmbientAudioContent/Waves/OneShots/Spot_Bird_CrestedSerpentEagle_01}}
 {id=WavePlayer_11, properties {soundWaveAssetPtr=/AmbientAudioContent/Waves/OneShots/Spot_Bird_CrestedSerpentEagle_02}}
 {id=WavePlayer_12, properties {soundWaveAssetPtr=/AmbientAudioContent/Waves/OneShots/Spot_Bird_CrestedSerpentEagle_03}}
 {id=WavePlayer_13, properties {soundWaveAssetPtr=/AmbientAudioContent/Waves/OneShots/Spot_Bird_GosHawk_01}}
 {id=WavePlayer_15, properties {soundWaveAssetPtr=&Spot_Bird_GosHawk_01}}
 {id=WavePlayer_7, properties {soundWaveAssetPtr=&Spot_Bird_GosHawk_01}}
 {id=WavePlayer_8, properties {soundWaveAssetPtr=/AmbientAudioContent/Waves/OneShots/Spot_Bird_GosHawk_03}}
 {id=WavePlayer_9, properties {soundWaveAssetPtr=/AmbientAudioContent/Waves/OneShots/Spot_Bird_GosHawk_02}}
]
properties {
 attenuationSettings=&Fort_TileAmbient_OneShot_Bird_Att
 maxDistance=8005
 soundClassObject=&Fort_Ambient_OneShots_Outdoors
 volumeMultiplier=0.2
 busSends [{
   audioBus=null
   maxSendDistance=6000
   minSendDistance=3000
   minSendLevel=0.6
   sendLevel=1
   soundSourceBus=&AmbientAudio_Exterior_SourceBus
  }]
}
