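# Sound cue for an outdoor ambient one-shot turtle dove call, played quietly with bird attenuation.
# A top-level random node picks a triple loop, a double loop or a single play, each drawing randomly from the same four dove waves to vary call phrases.
# The result passes through a default modulator, which every quality level shares.
# A distance-scaled send also feeds the exterior ambient source bus.
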SoundCue /AmbientAudioContent/SoundCues/OneShots/Spot_Bird_Olympus_Turtle_Dove_Cue
root=QualityLevel_0
nodes [
 {id=Looping_0, children [Random_0], properties {bLoopIndefinitely=false, loopCount=3}}
 {id=Looping_2, children [Random_2], properties {bLoopIndefinitely=false, loopCount=2}}
 {id=Modulator_0, children [Random_6]}
 {id=QualityLevel_0, children [Modulator_0,Modulator_0,Modulator_0,Modulator_0]}
 {id=Random_0, children [WavePlayer_12, WavePlayer_13, WavePlayer_14, WavePlayer_15], properties {weights [1,1,1,1]}}
 {id=Random_2, children [WavePlayer_16, WavePlayer_17, WavePlayer_18, WavePlayer_19], properties {weights [1,1,1,1]}}
 {id=Random_3, children [WavePlayer_20, WavePlayer_21, WavePlayer_22, WavePlayer_23], properties {weights [1,1,1,1]}}
 {id=Random_6, children [Looping_0,Looping_2,Random_3], properties {weights [1,1,1]}}
 {id=WavePlayer_12, properties {soundWaveAssetPtr=/AmbientAudioContent/Waves/OneShots/Spot_Bird_Turtle_Dove_01}}
 {id=WavePlayer_13, properties {soundWaveAssetPtr=/AmbientAudioContent/Waves/OneShots/Spot_Bird_Turtle_Dove_02}}
 {id=WavePlayer_14, properties {soundWaveAssetPtr=/AmbientAudioContent/Waves/OneShots/Spot_Bird_Turtle_Dove_03}}
 {id=WavePlayer_15, properties {soundWaveAssetPtr=/AmbientAudioContent/Waves/OneShots/Spot_Bird_Turtle_Dove_04}}
 {id=WavePlayer_16, properties {soundWaveAssetPtr=&Spot_Bird_Turtle_Dove_01}}
 {id=WavePlayer_17, properties {soundWaveAssetPtr=&Spot_Bird_Turtle_Dove_02}}
 {id=WavePlayer_18, properties {soundWaveAssetPtr=&Spot_Bird_Turtle_Dove_03}}
 {id=WavePlayer_19, properties {soundWaveAssetPtr=&Spot_Bird_Turtle_Dove_04}}
 {id=WavePlayer_20, properties {soundWaveAssetPtr=&Spot_Bird_Turtle_Dove_01}}
 {id=WavePlayer_21, properties {soundWaveAssetPtr=&Spot_Bird_Turtle_Dove_02}}
 {id=WavePlayer_22, properties {soundWaveAssetPtr=&Spot_Bird_Turtle_Dove_03}}
 {id=WavePlayer_23, properties {soundWaveAssetPtr=&Spot_Bird_Turtle_Dove_04}}
]
properties {
 attenuationSettings=&Fort_TileAmbient_OneShot_Bird_Att
 maxDistance=8005
 soundClassObject=&Fort_Ambient_OneShots_Outdoors
 volumeMultiplier=0.2
 busSends [{
   audioBus=null
   maxSendDistance=6000
   minSendDistance=3000
   minSendLevel=0.6
   sendLevel=1
   soundSourceBus=&AmbientAudio_Exterior_SourceBus
  }]
}
