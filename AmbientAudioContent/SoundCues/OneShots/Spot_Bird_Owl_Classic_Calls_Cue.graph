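# Sound cue for one-shot classic owl calls, routed to the ambient one-shots sound class with bird-specific attenuation.
# A Random node picks one of six equally weighted owl waves and is excluded from branch culling.
# A Modulator then randomizes pitch and volume with greatly lowered minimums for wide variation.
# The Quality Level root feeds all four levels from that same Modulator, so every quality setting sounds identical.
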SoundCue /AmbientAudioContent/SoundCues/OneShots/Spot_Bird_Owl_Classic_Calls_Cue
root=QualityLevel_0
nodes [
 {id=Modulator_0, children [Random_0], properties {pitchMin=0.3, volumeMin=0.3}}
 {id=QualityLevel_0, children [Modulator_0,Modulator_0,Modulator_0,Modulator_0]}
 {
  id=Random_0
  children [WavePlayer_1,WavePlayer_2,WavePlayer_3,WavePlayer_4,WavePlayer_5,WavePlayer_6]
  properties {bShouldExcludeFromBranchCulling=true, weights [1,1,1,1,1,1]}
 }
 {id=WavePlayer_1, properties {soundWaveAssetPtr=/AmbientAudioContent/Waves/OneShots/Spot_Bird_Owl_Classic_01}}
 {id=WavePlayer_2, properties {soundWaveAssetPtr=/AmbientAudioContent/Waves/OneShots/Spot_Bird_Owl_Classic_02}}
 {id=WavePlayer_3, properties {soundWaveAssetPtr=/AmbientAudioContent/Waves/OneShots/Spot_Bird_Owl_Classic_03}}
 {id=WavePlayer_4, properties {soundWaveAssetPtr=/AmbientAudioContent/Waves/OneShots/Spot_Bird_Owl_Classic_04}}
 {id=WavePlayer_5, properties {soundWaveAssetPtr=/AmbientAudioContent/Waves/OneShots/Spot_Bird_Owl_Classic_05}}
 {id=WavePlayer_6, properties {soundWaveAssetPtr=/AmbientAudioContent/Waves/OneShots/Spot_Bird_Owl_Classic_06}}
]
properties {
 attenuationSettings=&Fort_TileAmbient_OneShot_Bird_Att
 bEnableBusSends=false
 maxDistance=8005
 soundClassObject=&Fort_Ambient_OneShots
 volumeMultiplier=0.351
 preEffectBusSends [{}]
}
