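# Sound cue for an outdoor ambient tawny owl one-shot, played quietly with bird attenuation and the outdoor one-shots sound class.
# A quality-level root feeds two Modulator branches, each shared by two adjacent quality slots.
# Each Modulator wraps an equal-weight Random over the same seven owl waves, so both branches differ only in node instances.
# A distance-scaled bus send also routes the sound to the exterior ambient source bus.
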SoundCue /AmbientAudioContent/SoundCues/OneShots/Spot_Bird_Owl_Tawny_Cue
root=QualityLevel_0
nodes [
 {id=Modulator_3, children [Random_2]}
 {id=Modulator_4, children [Random_3]}
 {id=QualityLevel_0, children [Modulator_3,Modulator_3,Modulator_4,Modulator_4]}
 {
  id=Random_2
  children [WavePlayer_14, WavePlayer_15, WavePlayer_16, WavePlayer_17, WavePlayer_18, WavePlayer_19, WavePlayer_20]
  properties {weights [1,1,1,1,1,1,1]}
 }
 {
  id=Random_3
  children [WavePlayer_21, WavePlayer_22, WavePlayer_23, WavePlayer_24, WavePlayer_25, WavePlayer_26, WavePlayer_27]
  properties {weights [1,1,1,1,1,1,1]}
 }
 {id=WavePlayer_14, properties {soundWaveAssetPtr=/AmbientAudioContent/Waves/OneShots/Spot_Bird_Owl_Tawny_01}}
 {id=WavePlayer_15, properties {soundWaveAssetPtr=/AmbientAudioContent/Waves/OneShots/Spot_Bird_Owl_Tawny_02}}
 {id=WavePlayer_16, properties {soundWaveAssetPtr=/AmbientAudioContent/Waves/OneShots/Spot_Bird_Owl_Tawny_03}}
 {id=WavePlayer_17, properties {soundWaveAssetPtr=/AmbientAudioContent/Waves/OneShots/Spot_Bird_Owl_Tawny_04}}
 {id=WavePlayer_18, properties {soundWaveAssetPtr=/AmbientAudioContent/Waves/OneShots/Spot_Bird_Owl_Tawny_05}}
 {id=WavePlayer_19, properties {soundWaveAssetPtr=/AmbientAudioContent/Waves/OneShots/Spot_Bird_Owl_Tawny_06}}
 {id=WavePlayer_20, properties {soundWaveAssetPtr=/AmbientAudioContent/Waves/OneShots/Spot_Bird_Owl_Tawny_07}}
 {id=WavePlayer_21, properties {soundWaveAssetPtr=&Spot_Bird_Owl_Tawny_01}}
 {id=WavePlayer_22, properties {soundWaveAssetPtr=&Spot_Bird_Owl_Tawny_02}}
 {id=WavePlayer_23, properties {soundWaveAssetPtr=&Spot_Bird_Owl_Tawny_03}}
 {id=WavePlayer_24, properties {soundWaveAssetPtr=&Spot_Bird_Owl_Tawny_04}}
 {id=WavePlayer_25, properties {soundWaveAssetPtr=&Spot_Bird_Owl_Tawny_05}}
 {id=WavePlayer_26, properties {soundWaveAssetPtr=&Spot_Bird_Owl_Tawny_06}}
 {id=WavePlayer_27, properties {soundWaveAssetPtr=&Spot_Bird_Owl_Tawny_07}}
]
properties {
 attenuationSettings=&Fort_TileAmbient_OneShot_Bird_Att
 maxDistance=8005
 soundClassObject=&Fort_Ambient_OneShots_Outdoors
 volumeMultiplier=0.251
 busSends [{
   audioBus=null
   maxSendDistance=6000
   minSendDistance=3000
   minSendLevel=0.6
   sendLevel=1
   soundSourceBus=&AmbientAudio_Exterior_SourceBus
  }]
}
