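# Sound cue for one-shot ambient pigeon roosting coos, using bird attenuation and the ambient one-shots sound class.
# A random node picks the strong, medium or soft group, with single-input mixers trimming strong and boosting medium.
# Each group randomly picks a wave at equal weight, with empty weighted slots alongside the waves.
# A default modulator feeds the first two quality levels; the other two are empty.
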SoundCue /AmbientAudioContent/SoundCues/OneShots/Spot_Bird_Pigeon_Roosting_Cooing_Cue
root=QualityLevel_0
nodes [
 {id=Mixer_0, children [Random_2], properties {inputVolume [0.5]}}
 {id=Mixer_1, children [Random_3], properties {inputVolume [1.2]}}
 {id=Modulator_0, children [Random_6]}
 {id=QualityLevel_0, children [Modulator_0,Modulator_0,null,null]}
 {
  id=Random_2
  children [WavePlayer_16, WavePlayer_27, WavePlayer_28, WavePlayer_29, WavePlayer_30, WavePlayer_31, WavePlayer_32, null, null, null]
  properties {weights [1,1,1,1,1,1,1,1,1,1]}
 }
 {
  id=Random_3
  children [WavePlayer_33, WavePlayer_34, WavePlayer_35, WavePlayer_36, WavePlayer_37, WavePlayer_38, null, null]
  properties {weights [1,1,1,1,1,1,1,1]}
 }
 {
  id=Random_5
  children [WavePlayer_39, WavePlayer_40, WavePlayer_41, WavePlayer_42, WavePlayer_43, WavePlayer_44, null, null]
  properties {weights [1,1,1,1,1,1,1,1]}
 }
 {id=Random_6, children [Mixer_0,Mixer_1,Random_5], properties {weights [1,1,1]}}
 {id=WavePlayer_16, properties {soundWaveAssetPtr=/AmbientAudioContent/Waves/OneShots/Spot_Bird_Pigeon_Cooing_Strong_01}}
 {id=WavePlayer_27, properties {soundWaveAssetPtr=/AmbientAudioContent/Waves/OneShots/Spot_Bird_Pigeon_Cooing_Strong_02}}
 {id=WavePlayer_28, properties {soundWaveAssetPtr=/AmbientAudioContent/Waves/OneShots/Spot_Bird_Pigeon_Cooing_Strong_03}}
 {id=WavePlayer_29, properties {soundWaveAssetPtr=/AmbientAudioContent/Waves/OneShots/Spot_Bird_Pigeon_Cooing_Strong_04}}
 {id=WavePlayer_30, properties {soundWaveAssetPtr=/AmbientAudioContent/Waves/OneShots/Spot_Bird_Pigeon_Cooing_Strong_05}}
 {id=WavePlayer_31, properties {soundWaveAssetPtr=/AmbientAudioContent/Waves/OneShots/Spot_Bird_Pigeon_Cooing_Strong_06}}
 {id=WavePlayer_32, properties {soundWaveAssetPtr=/AmbientAudioContent/Waves/OneShots/Spot_Bird_Pigeon_Cooing_Strong_07}}
 {id=WavePlayer_33, properties {soundWaveAssetPtr=/AmbientAudioContent/Waves/OneShots/Spot_Bird_Pigeon_Cooing_Medium_06}}
 {id=WavePlayer_34, properties {soundWaveAssetPtr=/AmbientAudioContent/Waves/OneShots/Spot_Bird_Pigeon_Cooing_Medium_05}}
 {id=WavePlayer_35, properties {soundWaveAssetPtr=/AmbientAudioContent/Waves/OneShots/Spot_Bird_Pigeon_Cooing_Medium_04}}
 {id=WavePlayer_36, properties {soundWaveAssetPtr=/AmbientAudioContent/Waves/OneShots/Spot_Bird_Pigeon_Cooing_Medium_03}}
 {id=WavePlayer_37, properties {soundWaveAssetPtr=/AmbientAudioContent/Waves/OneShots/Spot_Bird_Pigeon_Cooing_Medium_02}}
 {id=WavePlayer_38, properties {soundWaveAssetPtr=/AmbientAudioContent/Waves/OneShots/Spot_Bird_Pigeon_Cooing_Medium_01}}
 {id=WavePlayer_39, properties {soundWaveAssetPtr=/AmbientAudioContent/Waves/OneShots/Spot_Bird_Pigeon_Cooing_Soft_01}}
 {id=WavePlayer_40, properties {soundWaveAssetPtr=/AmbientAudioContent/Waves/OneShots/Spot_Bird_Pigeon_Cooing_Soft_02}}
 {id=WavePlayer_41, properties {soundWaveAssetPtr=/AmbientAudioContent/Waves/OneShots/Spot_Bird_Pigeon_Cooing_Soft_03}}
 {id=WavePlayer_42, properties {soundWaveAssetPtr=/AmbientAudioContent/Waves/OneShots/Spot_Bird_Pigeon_Cooing_Soft_04}}
 {id=WavePlayer_43, properties {soundWaveAssetPtr=/AmbientAudioContent/Waves/OneShots/Spot_Bird_Pigeon_Cooing_Soft_05}}
 {id=WavePlayer_44, properties {soundWaveAssetPtr=/AmbientAudioContent/Waves/OneShots/Spot_Bird_Pigeon_Cooing_Soft_06}}
]
properties {
 attenuationSettings=&Fort_TileAmbient_OneShot_Bird_Att
 bEnableBusSends=false
 maxDistance=8005
 soundClassObject=&Fort_Ambient_OneShots
 volumeMultiplier=0.601
 preEffectBusSends [{}]
}
