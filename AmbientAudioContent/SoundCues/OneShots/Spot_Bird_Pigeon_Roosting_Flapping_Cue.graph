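# Sound cue for one-shot ambient spot sounds of pigeons roosting and flapping.
# A Random node picks among seven equally weighted wave variations, and a Quality Level node at the root feeds it to the first two quality slots, leaving the other two empty.
# Output uses the bird one-shot attenuation and ambient one-shots sound class, with slightly reduced volume and a send to the exterior ambient source bus.
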SoundCue /AmbientAudioContent/SoundCues/OneShots/Spot_Bird_Pigeon_Roosting_Flapping_Cue
root=QualityLevel_0
nodes [
 {id=QualityLevel_0, children [Random_0,Random_0,null,null]}
 {
  id=Random_0
  children [WavePlayer_0,WavePlayer_1,WavePlayer_2,WavePlayer_3,WavePlayer_4,WavePlayer_5,WavePlayer_6]
  properties {weights [1,1,1,1,1,1,1]}
 }
 {id=WavePlayer_0, properties {soundWaveAssetPtr=/AmbientAudioContent/Waves/OneShots/Spot_Bird_Pigeon_Roosting_Flapping_01}}
 {id=WavePlayer_1, properties {soundWaveAssetPtr=/AmbientAudioContent/Waves/OneShots/Spot_Bird_Pigeon_Roosting_Flapping_02}}
 {id=WavePlayer_2, properties {soundWaveAssetPtr=/AmbientAudioContent/Waves/OneShots/Spot_Bird_Pigeon_Roosting_Flapping_03}}
 {id=WavePlayer_3, properties {soundWaveAssetPtr=/AmbientAudioContent/Waves/OneShots/Spot_Bird_Pigeon_Roosting_Flapping_04}}
 {id=WavePlayer_4, properties {soundWaveAssetPtr=/AmbientAudioContent/Waves/OneShots/Spot_Bird_Pigeon_Roosting_Flapping_05}}
 {id=WavePlayer_5, properties {soundWaveAssetPtr=/AmbientAudioContent/Waves/OneShots/Spot_Bird_Pigeon_Roosting_Flapping_06}}
 {id=WavePlayer_6, properties {soundWaveAssetPtr=/AmbientAudioContent/Waves/OneShots/Spot_Bird_Pigeon_Roosting_Flapping_07}}
]
properties {
 attenuationSettings=&Fort_TileAmbient_OneShot_Bird_Att
 maxDistance=8005
 soundClassObject=&Fort_Ambient_OneShots
 volumeMultiplier=0.752
 busSends [{audioBus=null, minSendLevel=0.6, sendLevel=1, soundSourceBus=&AmbientAudio_Exterior_SourceBus}]
}
