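# Sound cue for a one-shot rainforest Greenbul bird call, used as outdoor ambient audio.
# A Random node picks one of five equally weighted waves, and a Modulator varies its pitch and volume.
# All four QualityLevel inputs share that Modulator chain, so every quality tier sounds the same.
# It uses shared bird attenuation, a distance-scaled send to the exterior ambient source bus, and a two-instance concurrency cap that rejects new plays.
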SoundCue /AmbientAudioContent/SoundCues/OneShots/Spot_Bird_Rainforest_Greenbul_Cue
root=QualityLevel_0
nodes [
 {id=Modulator_0, children [Random_0], properties {pitchMax=1.1, volumeMax=1, volumeMin=0.7}}
 {id=QualityLevel_0, children [Modulator_0,Modulator_0,Modulator_0,Modulator_0]}
 {id=Random_0, children [WavePlayer_5,WavePlayer_6,WavePlayer_7,WavePlayer_8,WavePlayer_9], properties {weights [1,1,1,1,1]}}
 {id=WavePlayer_5, properties {soundWaveAssetPtr=/AmbientAudioContent/Waves/OneShots/Spot_Bird_Rainforest_Greenbul_01}}
 {id=WavePlayer_6, properties {soundWaveAssetPtr=/AmbientAudioContent/Waves/OneShots/Spot_Bird_Rainforest_Greenbul_03}}
 {id=WavePlayer_7, properties {soundWaveAssetPtr=/AmbientAudioContent/Waves/OneShots/Spot_Bird_Rainforest_Greenbul_04}}
 {id=WavePlayer_8, properties {soundWaveAssetPtr=/AmbientAudioContent/Waves/OneShots/Spot_Bird_Rainforest_Greenbul_05}}
 {id=WavePlayer_9, properties {soundWaveAssetPtr=/AmbientAudioContent/Waves/OneShots/Spot_Bird_Rainforest_Greenbul_06}}
]
properties {
 attenuationSettings=&Fort_TileAmbient_OneShot_Bird_Att
 bOverrideConcurrency=true
 maxDistance=8005
 soundClassObject=&Fort_Ambient_OneShots_Outdoors
 volumeMultiplier=0.501
 busSends [{
   audioBus=null
   maxSendDistance=5000
   maxSendLevel=1.2
   minSendDistance=3000
   minSendLevel=0.2
   sendLevel=0
   soundSourceBus=&AmbientAudio_Exterior_SourceBus
  }]
 concurrencyOverrides {maxCount=2, resolutionRule=PreventNew}
 preEffectBusSends [{}]
}
