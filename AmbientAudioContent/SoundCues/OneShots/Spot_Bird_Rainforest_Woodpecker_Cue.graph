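# Sound Cue for a rainforest woodpecker ambient one-shot, classed under Fort ambient one-shots.
# A Random node picks one of four equally weighted waves, then a Modulator varies its volume.
# All four quality-level inputs share that same chain, so content is identical at every quality.
# Bird one-shot attenuation shapes falloff, and a distance-ranged send feeds the exterior ambient source bus.
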SoundCue /AmbientAudioContent/SoundCues/OneShots/Spot_Bird_Rainforest_Woodpecker_Cue
root=QualityLevel_0
nodes [
 {id=Modulator_0, children [Random_0], properties {volumeMax=1, volumeMin=0.7}}
 {id=QualityLevel_0, children [Modulator_0,Modulator_0,Modulator_0,Modulator_0]}
 {id=Random_0, children [WavePlayer_0,WavePlayer_1,WavePlayer_2,WavePlayer_3], properties {weights [1,1,1,1]}}
 {id=WavePlayer_0, properties {soundWaveAssetPtr=/AmbientAudioContent/Waves/OneShots/Spot_Bird_Rainforest_Woodpecker_01}}
 {id=WavePlayer_1, properties {soundWaveAssetPtr=/AmbientAudioContent/Waves/OneShots/Spot_Bird_Rainforest_Woodpecker_02}}
 {id=WavePlayer_2, properties {soundWaveAssetPtr=/AmbientAudioContent/Waves/OneShots/Spot_Bird_Rainforest_Woodpecker_03}}
 {id=WavePlayer_3, properties {soundWaveAssetPtr=/AmbientAudioContent/Waves/OneShots/Spot_Bird_Rainforest_Woodpecker_04}}
]
properties {
 attenuationSettings=&Fort_TileAmbient_OneShot_Bird_Att
 maxDistance=8005
 soundClassObject=&Fort_Ambient_OneShots
 volumeMultiplier=0.501
 busSends [{
   audioBus=null
   maxSendDistance=5000
   maxSendLevel=1.2
   minSendDistance=3000
   minSendLevel=0.2
   sendLevel=0
   soundSourceBus=&AmbientAudio_Exterior_SourceBus
  }]
 preEffectBusSends [{}]
}
