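# One-shot ambient SoundCue for a singing red-throated loon, played quietly as an outdoor spot bird sound.
# A Quality Level node routes its first two levels and last two levels to separate Modulator and Random branches.
# Both branches pick evenly between the same two daytime waves, differing only in a lower minimum pitch on the first.
# It uses shared bird one-shot attenuation and a distance-scaled send to the exterior ambient source bus.
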SoundCue /AmbientAudioContent/SoundCues/OneShots/Spot_Bird_Red_Throated_Loon_Singing_Cue
root=QualityLevel_0
nodes [
 {id=Modulator_3, children [Random_0], properties {pitchMin=0.9}}
 {id=Modulator_4, children [Random_1]}
 {id=QualityLevel_0, children [Modulator_3,Modulator_3,Modulator_4,Modulator_4]}
 {id=Random_0, children [WavePlayer_0,WavePlayer_1], properties {weights [1,1]}}
 {id=Random_1, children [WavePlayer_2,WavePlayer_3], properties {weights [1,1]}}
 {id=WavePlayer_0, properties {soundWaveAssetPtr=/AmbientAudioContent/Waves/OneShots/Spot_Bird_Red_Throated_Loon_Singing_Day_01}}
 {id=WavePlayer_1, properties {soundWaveAssetPtr=/AmbientAudioContent/Waves/OneShots/Spot_Bird_Red_Throated_Loon_Singing_Day_02}}
 {id=WavePlayer_2, properties {soundWaveAssetPtr=&Spot_Bird_Red_Throated_Loon_Singing_Day_01}}
 {id=WavePlayer_3, properties {soundWaveAssetPtr=&Spot_Bird_Red_Throated_Loon_Singing_Day_02}}
]
properties {
 attenuationSettings=&Fort_TileAmbient_OneShot_Bird_Att
 maxDistance=8005
 soundClassObject=&Fort_Ambient_OneShots_Outdoors
 volumeMultiplier=0.05
 busSends [{
   audioBus=null
   maxSendDistance=6000
   minSendDistance=3000
   minSendLevel=0.6
   sendLevel=1
   soundSourceBus=&AmbientAudio_Exterior_SourceBus
  }]
}
