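# Sound Cue for scary bird and forest creature one-shot calls, used as outdoor ambient spot sounds.
# A Random node picks among sixteen equally weighted wave variations, then a Modulator slightly varies the pitch.
# It plays at reduced volume with bird attenuation, an outdoor one-shot sound class and a distance-ranged send to the exterior ambient source bus.
# Overridden concurrency caps simultaneous voices at two, with distance-based volume scaling.
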SoundCue /AmbientAudioContent/SoundCues/OneShots/Spot_Bird_Scary_Forest_Creature_Calls_Cue
root=Modulator_0
nodes [
 {id=Modulator_0, children [Random_0], properties {pitchMax=1.1, pitchMin=0.9}}
 {
  id=Random_0
  children [
   WavePlayer_1
   WavePlayer_2
   WavePlayer_3
   WavePlayer_4
   WavePlayer_5
   WavePlayer_10
   WavePlayer_11
   WavePlayer_12
   WavePlayer_13
   WavePlayer_14
   WavePlayer_15
   WavePlayer_16
   WavePlayer_17
   WavePlayer_18
   WavePlayer_19
   WavePlayer_20
  ]
  properties {weights [1,1,1,1,1,1,1,1,1,1,1,1,1,1,1,1]}
 }
 {id=WavePlayer_1, properties {soundWaveAssetPtr=/AmbientAudioContent/Waves/OneShots/Spot_Bird_Scary_Forest_Creature_Calls_01}}
 {id=WavePlayer_10, properties {soundWaveAssetPtr=/AmbientAudioContent/Waves/OneShots/Spot_Bird_Scary_Forest_Creature_Calls_06}}
 {id=WavePlayer_11, properties {soundWaveAssetPtr=/AmbientAudioContent/Waves/OneShots/Spot_Bird_Scary_Forest_Creature_Calls_07}}
 {id=WavePlayer_12, properties {soundWaveAssetPtr=/AmbientAudioContent/Waves/OneShots/Spot_Bird_Scary_Forest_Creature_Calls_08}}
 {id=WavePlayer_13, properties {soundWaveAssetPtr=/AmbientAudioContent/Waves/OneShots/Spot_Bird_Scary_Forest_Creature_Calls_09}}
 {id=WavePlayer_14, properties {soundWaveAssetPtr=/AmbientAudioContent/Waves/OneShots/Spot_Bird_Scary_Forest_Creature_Calls_10}}
 {id=WavePlayer_15, properties {soundWaveAssetPtr=/AmbientAudioContent/Waves/OneShots/Spot_Bird_Scary_Forest_Creature_Calls_11}}
 {id=WavePlayer_16, properties {soundWaveAssetPtr=/AmbientAudioContent/Waves/OneShots/Spot_Bird_Scary_Forest_Creature_Calls_12}}
 {id=WavePlayer_17, properties {soundWaveAssetPtr=/AmbientAudioContent/Waves/OneShots/Spot_Bird_Scary_Forest_Creature_Calls_13}}
 {id=WavePlayer_18, properties {soundWaveAssetPtr=/AmbientAudioContent/Waves/OneShots/Spot_Bird_Scary_Forest_Creature_Calls_14}}
 {id=WavePlayer_19, properties {soundWaveAssetPtr=/AmbientAudioContent/Waves/OneShots/Spot_Bird_Scary_Forest_Creature_Calls_15}}
 {id=WavePlayer_2, properties {soundWaveAssetPtr=/AmbientAudioContent/Waves/OneShots/Spot_Bird_Scary_Forest_Creature_Calls_02}}
 {id=WavePlayer_20, properties {soundWaveAssetPtr=/AmbientAudioContent/Waves/OneShots/Spot_Bird_Scary_Forest_Creature_Calls_16}}
 {id=WavePlayer_3, properties {soundWaveAssetPtr=/AmbientAudioContent/Waves/OneShots/Spot_Bird_Scary_Forest_Creature_Calls_03}}
 {id=WavePlayer_4, properties {soundWaveAssetPtr=/AmbientAudioContent/Waves/OneShots/Spot_Bird_Scary_Forest_Creature_Calls_04}}
 {id=WavePlayer_5, properties {soundWaveAssetPtr=/AmbientAudioContent/Waves/OneShots/Spot_Bird_Scary_Forest_Creature_Calls_05}}
]
properties {
 attenuationSettings=&Fort_TileAmbient_OneShot_Bird_Att
 bOverrideConcurrency=true
 maxDistance=8005
 soundClassObject=&Fort_Ambient_OneShots_Outdoors
 volumeMultiplier=0.5
 busSends [{
   audioBus=null
   maxSendDistance=8000
   minSendDistance=1000
   minSendLevel=0.4
   sendLevel=1
   soundSourceBus=&AmbientAudio_Exterior_SourceBus
  }]
 concurrencyOverrides {maxCount=2, voiceStealReleaseTime=0.25, volumeScaleMode=Distance}
}
