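# Sound cue for one-shot seagull calls in the ambient audio, routed to the indoor one-shots sound class.
# All quality levels share one random pick among three equally weighted groups: starter island calls, mountain group squawks and coastal gulls.
# Each group randomly selects a wave through a modulator that varies pitch and volume.
# The cue plays quiet and slightly pitched down, with bird attenuation, limited concurrency and a send to the exterior source bus.
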SoundCue /AmbientAudioContent/SoundCues/OneShots/Spot_Bird_Seagulls_Calls_All_Int_Cue
root=QualityLevel_0
nodes [
 {id=Modulator_0, children [Random_7], properties {pitchMax=1.1, volumeMax=1.1, volumeMin=0.9}}
 {id=Modulator_1, children [Random_3], properties {pitchMax=1.1, volumeMax=1.1, volumeMin=0.9}}
 {id=Modulator_5, children [Random_8], properties {pitchMax=1.1, volumeMax=1.1, volumeMin=0.9}}
 {id=QualityLevel_0, children [Random_5,Random_5,Random_5,Random_5]}
 {
  id=Random_3
  children [WavePlayer_20, WavePlayer_21, WavePlayer_22, WavePlayer_23, WavePlayer_30, WavePlayer_31]
  properties {weights [1,1,1,1,1,1]}
 }
 {id=Random_5, children [Modulator_1,Modulator_0,Modulator_5], properties {weights [1,1,1]}}
 {
  id=Random_7
  children [WavePlayer_32, WavePlayer_33, WavePlayer_34, WavePlayer_35, WavePlayer_36, WavePlayer_37, WavePlayer_38]
  properties {weights [1,1,1,1,1,1,1]}
 }
 {
  id=Random_8
  children [WavePlayer_39, WavePlayer_40, WavePlayer_41, WavePlayer_42, WavePlayer_43, WavePlayer_44]
  properties {weights [1,1,1,1,1,1]}
 }
 {id=WavePlayer_20, properties {soundWaveAssetPtr=/AmbientAudioContent/Waves/OneShots/Spot_Starter_Island_Seagull_Calls_03}}
 {id=WavePlayer_21, properties {soundWaveAssetPtr=/AmbientAudioContent/Waves/OneShots/Spot_Starter_Island_Seagull_Calls_04}}
 {id=WavePlayer_22, properties {soundWaveAssetPtr=/AmbientAudioContent/Waves/OneShots/Spot_Starter_Island_Seagull_Calls_05}}
 {id=WavePlayer_23, properties {soundWaveAssetPtr=/AmbientAudioContent/Waves/OneShots/Spot_Starter_Island_Seagull_Calls_06}}
 {id=WavePlayer_30, properties {soundWaveAssetPtr=/AmbientAudioContent/Waves/OneShots/Spot_Starter_Island_Seagull_Calls_07}}
 {id=WavePlayer_31, properties {soundWaveAssetPtr=/AmbientAudioContent/Waves/OneShots/Spot_Starter_Island_Seagull_Calls_08}}
 {id=WavePlayer_32, properties {soundWaveAssetPtr=/AmbientAudioContent/Waves/OneShots/Spot_Mountain_Seagulls_Squarks_Small_Group_07}}
 {id=WavePlayer_33, properties {soundWaveAssetPtr=/AmbientAudioContent/Waves/OneShots/Spot_Mountain_Seagulls_Squarks_Small_Group_06}}
 {id=WavePlayer_34, properties {soundWaveAssetPtr=/AmbientAudioContent/Waves/OneShots/Spot_Mountain_Seagulls_Squarks_Small_Group_05}}
 {id=WavePlayer_35, properties {soundWaveAssetPtr=/AmbientAudioContent/Waves/OneShots/Spot_Mountain_Seagulls_Squarks_Small_Group_04}}
 {id=WavePlayer_36, properties {soundWaveAssetPtr=/AmbientAudioContent/Waves/OneShots/Spot_Mountain_Seagulls_Squarks_Small_Group_03}}
 {id=WavePlayer_37, properties {soundWaveAssetPtr=/AmbientAudioContent/Waves/OneShots/Spot_Mountain_Seagulls_Squarks_Small_Group_02}}
 {id=WavePlayer_38, properties {soundWaveAssetPtr=/AmbientAudioContent/Waves/OneShots/Spot_Mountain_Seagulls_Squarks_Small_Group_01}}
 {id=WavePlayer_39, properties {soundWaveAssetPtr=/AmbientAudioContent/Waves/OneShots/Spot_Coastal_Gulls_OneShot_06}}
 {id=WavePlayer_40, properties {soundWaveAssetPtr=/AmbientAudioContent/Waves/OneShots/Spot_Coastal_Gulls_OneShot_05}}
 {id=WavePlayer_41, properties {soundWaveAssetPtr=/AmbientAudioContent/Waves/OneShots/Spot_Coastal_Gulls_OneShot_04}}
 {id=WavePlayer_42, properties {soundWaveAssetPtr=/AmbientAudioContent/Waves/OneShots/Spot_Coastal_Gulls_OneShot_03}}
 {id=WavePlayer_43, properties {soundWaveAssetPtr=/AmbientAudioContent/Waves/OneShots/Spot_Coastal_Gulls_OneShot_02}}
 {id=WavePlayer_44, properties {soundWaveAssetPtr=/AmbientAudioContent/Waves/OneShots/Spot_Coastal_Gulls_OneShot_01}}
]
properties {
 attenuationSettings=&Fort_TileAmbient_OneShot_Bird_Att
 bOverrideConcurrency=true
 maxDistance=8005
 pitchMultiplier=0.9
 soundClassObject=/Game/Sounds/FortSoundClassesAndModes/Fort_Ambient_OneShots_Indoors
 volumeMultiplier=0.15
 busSends [{audioBus=null, minSendLevel=0.6, sendLevel=1, soundSourceBus=&AmbientAudio_Exterior_SourceBus}]
 concurrencyOverrides {maxCount=3}
}
