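# Sound Cue for an outdoor ambient one-shot of snowy plover chirps, using bird attenuation and the outdoor one-shots sound class.
# An equally weighted Random node picks one of six waves, with single-input Mixers rebalancing the volume of the third and sixth.
# It feeds a Modulator wired to all four Quality Level inputs, so every quality level plays identically.
# The cue plays at reduced volume and has a distance-scaled send to the exterior ambient source bus.
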SoundCue /AmbientAudioContent/SoundCues/OneShots/Spot_Bird_Snowy_Plover_Chirps_Cue
root=QualityLevel_0
nodes [
 {id=Mixer_0, children [WavePlayer_14], properties {inputVolume [1.6]}}
 {id=Mixer_2, children [WavePlayer_17], properties {inputVolume [0.6]}}
 {id=Modulator_3, children [Random_2]}
 {id=QualityLevel_0, children [Modulator_3,Modulator_3,Modulator_3,Modulator_3]}
 {id=Random_2, children [WavePlayer_12, WavePlayer_13, Mixer_0, WavePlayer_15, WavePlayer_16, Mixer_2], properties {weights [1,1,1,1,1,1]}}
 {id=WavePlayer_12, properties {soundWaveAssetPtr=/AmbientAudioContent/Waves/OneShots/Spot_Bird_Snowy_Plover_Chirps_01}}
 {id=WavePlayer_13, properties {soundWaveAssetPtr=/AmbientAudioContent/Waves/OneShots/Spot_Bird_Snowy_Plover_Chirps_02}}
 {id=WavePlayer_14, properties {soundWaveAssetPtr=/AmbientAudioContent/Waves/OneShots/Spot_Bird_Snowy_Plover_Chirps_03}}
 {id=WavePlayer_15, properties {soundWaveAssetPtr=/AmbientAudioContent/Waves/OneShots/Spot_Bird_Snowy_Plover_Chirps_04}}
 {id=WavePlayer_16, properties {soundWaveAssetPtr=/AmbientAudioContent/Waves/OneShots/Spot_Bird_Snowy_Plover_Chirps_05}}
 {id=WavePlayer_17, properties {soundWaveAssetPtr=/AmbientAudioContent/Waves/OneShots/Spot_Bird_Snowy_Plover_Chirps_06}}
]
properties {
 attenuationSettings=&Fort_TileAmbient_OneShot_Bird_Att
 maxDistance=8005
 soundClassObject=&Fort_Ambient_OneShots_Outdoors
 volumeMultiplier=0.2
 busSends [{
   audioBus=null
   maxSendDistance=6000
   minSendDistance=3000
   minSendLevel=0.6
   sendLevel=1
   soundSourceBus=&AmbientAudio_Exterior_SourceBus
  }]
}
