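# One-shot ambient SoundCue for a tropical night SteelPecker bird croak, classed as outdoor ambience.
# An equal-weight Random node picks one of five croak waves, and a Modulator varies its pitch and volume.
# All four quality-level inputs share that same chain, so the quality setting does not change the sound.
# Playback uses bird one-shot attenuation at reduced volume, with a distance-ranged send to the exterior ambient source bus.
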SoundCue /AmbientAudioContent/SoundCues/OneShots/Spot_Bird_Tropical_Night_SteelPecker_Croak_Cue
root=QualityLevel_0
nodes [
 {id=Modulator_1, children [Random_0], properties {pitchMax=1.1, pitchMin=0.8, volumeMax=1.2, volumeMin=1}}
 {id=QualityLevel_0, children [Modulator_1,Modulator_1,Modulator_1,Modulator_1]}
 {id=Random_0, children [WavePlayer_5,WavePlayer_6,WavePlayer_7,WavePlayer_8,WavePlayer_9], properties {weights [1,1,1,1,1]}}
 {id=WavePlayer_5, properties {soundWaveAssetPtr=/AmbientAudioContent/Waves/OneShots/Spot_Bird_Tropical_Night_SteelPecker_Croak_01}}
 {id=WavePlayer_6, properties {soundWaveAssetPtr=/AmbientAudioContent/Waves/OneShots/Spot_Bird_Tropical_Night_SteelPecker_Croak_02}}
 {id=WavePlayer_7, properties {soundWaveAssetPtr=/AmbientAudioContent/Waves/OneShots/Spot_Bird_Tropical_Night_SteelPecker_Croak_03}}
 {id=WavePlayer_8, properties {soundWaveAssetPtr=/AmbientAudioContent/Waves/OneShots/Spot_Bird_Tropical_Night_SteelPecker_Croak_04}}
 {id=WavePlayer_9, properties {soundWaveAssetPtr=/AmbientAudioContent/Waves/OneShots/Spot_Bird_Tropical_Night_SteelPecker_Croak_05}}
]
properties {
 attenuationSettings=&Fort_TileAmbient_OneShot_Bird_Att
 maxDistance=8005
 soundClassObject=&Fort_Ambient_OneShots_Outdoors
 volumeMultiplier=0.251
 busSends [{
   audioBus=null
   maxSendDistance=5000
   maxSendLevel=1.2
   minSendDistance=3000
   minSendLevel=0.2
   sendLevel=0
   soundSourceBus=&AmbientAudio_Exterior_SourceBus
  }]
 preEffectBusSends [{}]
}
