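# Sound Cue for randomized interior brick ambient one-shot spot sounds.
# A Random node picks among fourteen equally weighted brick wave variations and feeds a root Modulator that varies volume and lowers pitch by random amounts for variety.
# It plays through a shared tile-ambient one-shot attenuation asset under the indoor ambient Athena sound class, with reduced overall volume.
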SoundCue /AmbientAudioContent/SoundCues/OneShots/Spot_Brick_Oneshots_Cue
root=Modulator_0
nodes [
 {id=Modulator_0, children [Random_1], properties {pitchMax=1, pitchMin=0.8, volumeMax=1, volumeMin=0.7}}
 {
  id=Random_1
  children [
   WavePlayer_3
   WavePlayer_4
   WavePlayer_5
   WavePlayer_6
   WavePlayer_7
   WavePlayer_8
   WavePlayer_9
   WavePlayer_10
   WavePlayer_22
   WavePlayer_23
   WavePlayer_24
   WavePlayer_25
   WavePlayer_26
   WavePlayer_27
  ]
  properties {weights [1,1,1,1,1,1,1,1,1,1,1,1,1,1]}
 }
 {id=WavePlayer_10, properties {soundWaveAssetPtr=/AmbientAudioContent/Waves/OneShots/Spot_AMB_SS_Int_Brick_08}}
 {id=WavePlayer_22, properties {soundWaveAssetPtr=/AmbientAudioContent/Waves/OneShots/Spot_AMB_SS_Int_Brick_09}}
 {id=WavePlayer_23, properties {soundWaveAssetPtr=/AmbientAudioContent/Waves/OneShots/Spot_AMB_SS_Int_Brick_10}}
 {id=WavePlayer_24, properties {soundWaveAssetPtr=/AmbientAudioContent/Waves/OneShots/Spot_AMB_SS_Int_Brick_11}}
 {id=WavePlayer_25, properties {soundWaveAssetPtr=/AmbientAudioContent/Waves/OneShots/Spot_AMB_SS_Int_Brick_12}}
 {id=WavePlayer_26, properties {soundWaveAssetPtr=/AmbientAudioContent/Waves/OneShots/Spot_AMB_SS_Int_Brick_13}}
 {id=WavePlayer_27, properties {soundWaveAssetPtr=/AmbientAudioContent/Waves/OneShots/Spot_AMB_SS_Int_Brick_14}}
 {id=WavePlayer_3, properties {soundWaveAssetPtr=/AmbientAudioContent/Waves/OneShots/Spot_AMB_SS_Int_Brick_01}}
 {id=WavePlayer_4, properties {soundWaveAssetPtr=/AmbientAudioContent/Waves/OneShots/Spot_AMB_SS_Int_Brick_02}}
 {id=WavePlayer_5, properties {soundWaveAssetPtr=/AmbientAudioContent/Waves/OneShots/Spot_AMB_SS_Int_Brick_03}}
 {id=WavePlayer_6, properties {soundWaveAssetPtr=/AmbientAudioContent/Waves/OneShots/Spot_AMB_SS_Int_Brick_04}}
 {id=WavePlayer_7, properties {soundWaveAssetPtr=/AmbientAudioContent/Waves/OneShots/Spot_AMB_SS_Int_Brick_05}}
 {id=WavePlayer_8, properties {soundWaveAssetPtr=/AmbientAudioContent/Waves/OneShots/Spot_AMB_SS_Int_Brick_06}}
 {id=WavePlayer_9, properties {soundWaveAssetPtr=/AmbientAudioContent/Waves/OneShots/Spot_AMB_SS_Int_Brick_07}}
]
properties {
 attenuationSettings=/Game/Sounds/Attenuation/Fort_TileAmbient_OneShot_Att
 maxDistance=8000
 soundClassObject=&Fort_Ambient_Indoors_Athena
 volumeMultiplier=0.65
}
