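# Sound Cue for a distant metal "wronk" ambient one-shot at bridges, reusing cave wave assets in the outdoor Athena ambient sound class.
# An equally weighted Random node picks one of eight waves, then feeds a Modulator for pitch variation and an Attenuation node with its own long-falloff override.
# All four quality-level inputs share that same chain.
# A concurrency override caps simultaneous instances at two, and bus and submix sends are off.
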SoundCue /AmbientAudioContent/SoundCues/OneShots/Spot_Bridge_Distant_Metal_Wronk_Cue
root=QualityLevel_0
nodes [
 {
  id=Attenuation_0
  children [Modulator_0]
  properties {bOverrideAttenuation=true, attenuationOverrides {falloffDistance=7000, attenuationShapeExtents {x=2000}}}
 }
 {id=Modulator_0, children [Random_1], properties {pitchMin=0.9}}
 {id=QualityLevel_0, children [Attenuation_0, Attenuation_0, Attenuation_0, Attenuation_0]}
 {
  id=Random_1
  children [WavePlayer_8, WavePlayer_9, WavePlayer_10, WavePlayer_11, WavePlayer_12, WavePlayer_13, WavePlayer_14, WavePlayer_15]
  properties {weights [1,1,1,1,1,1,1,1]}
 }
 {id=WavePlayer_10, properties {soundWaveAssetPtr=/AmbientAudioContent/Waves/OneShots/Spot_Cave_Distant_Metal_Wronk_3}}
 {id=WavePlayer_11, properties {soundWaveAssetPtr=/AmbientAudioContent/Waves/OneShots/Spot_Cave_Distant_Metal_Wronk_4}}
 {id=WavePlayer_12, properties {soundWaveAssetPtr=/AmbientAudioContent/Waves/OneShots/Spot_Cave_Distant_Metal_Wronk_5}}
 {id=WavePlayer_13, properties {soundWaveAssetPtr=/AmbientAudioContent/Waves/OneShots/Spot_Cave_Distant_Metal_Wronk_6}}
 {id=WavePlayer_14, properties {soundWaveAssetPtr=/AmbientAudioContent/Waves/OneShots/Spot_Cave_Distant_Metal_Wronk_7}}
 {id=WavePlayer_15, properties {soundWaveAssetPtr=/AmbientAudioContent/Waves/OneShots/Spot_Cave_Distant_Metal_Wronk_8}}
 {id=WavePlayer_8, properties {soundWaveAssetPtr=/AmbientAudioContent/Waves/OneShots/Spot_Cave_Distant_Metal_Wronk_1}}
 {id=WavePlayer_9, properties {soundWaveAssetPtr=/AmbientAudioContent/Waves/OneShots/Spot_Cave_Distant_Metal_Wronk_2}}
]
properties {
 attenuationSettings=/Game/Sounds/Attenuation/Fort_TileAmbient_OneShot_Distant_Creak_Att
 bEnableBusSends=false
 bEnableSubmixSends=false
 bOverrideConcurrency=true
 maxDistance=9000
 soundClassObject=&Fort_Ambient_Outdoors_Athena
 volumeMultiplier=0.576
 attenuationOverrides {falloffDistance=6000, stereoSpread=10000, attenuationShapeExtents {x=2000}}
 concurrencyOverrides {maxCount=2}
}
