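# Ambient one-shot Sound Cue for distant Brimstone chain rattles, played very quietly in the Fort_Ambient_OneShots sound class.
# A Random node picks among thirteen equally weighted chain-movement waves, feeding a Modulator that varies pitch and volume.
# A custom attenuation curve keeps it near-silent up close, full at mid-range and silent at maximum distance.
# Distance low-pass filtering and occlusion with slow interpolation muffle it further.
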SoundCue /AmbientAudioContent/SoundCues/OneShots/Spot_Brimstone_Chain_Rattles_Distant_Cue
root=Modulator_0
nodes [
 {id=Modulator_0, children [Random_0], properties {pitchMin=0.8, volumeMin=0.8}}
 {
  id=Random_0
  children [
   WavePlayer_0
   WavePlayer_2
   WavePlayer_3
   WavePlayer_4
   WavePlayer_5
   WavePlayer_6
   WavePlayer_7
   WavePlayer_8
   WavePlayer_21
   WavePlayer_22
   WavePlayer_23
   WavePlayer_24
   WavePlayer_25
  ]
  properties {weights [1,1,1,1,1,1,1,1,1,1,1,1,1]}
 }
 {id=WavePlayer_0, properties {soundWaveAssetPtr=/AmbientAudioContent/Waves/OneShots/Spot_Chain_Movement_01}}
 {id=WavePlayer_2, properties {soundWaveAssetPtr=/AmbientAudioContent/Waves/OneShots/Spot_Chain_Movement_02}}
 {id=WavePlayer_21, properties {soundWaveAssetPtr=/AmbientAudioContent/Waves/OneShots/Spot_Chain_Movement_09}}
 {id=WavePlayer_22, properties {soundWaveAssetPtr=/AmbientAudioContent/Waves/OneShots/Spot_Chain_Movement_10}}
 {id=WavePlayer_23, properties {soundWaveAssetPtr=/AmbientAudioContent/Waves/OneShots/Spot_Chain_Movement_11}}
 {id=WavePlayer_24, properties {soundWaveAssetPtr=/AmbientAudioContent/Waves/OneShots/Spot_Chain_Movement_12}}
 {id=WavePlayer_25, properties {soundWaveAssetPtr=/AmbientAudioContent/Waves/OneShots/Spot_Chain_Movement_13}}
 {id=WavePlayer_3, properties {soundWaveAssetPtr=/AmbientAudioContent/Waves/OneShots/Spot_Chain_Movement_03}}
 {id=WavePlayer_4, properties {soundWaveAssetPtr=/AmbientAudioContent/Waves/OneShots/Spot_Chain_Movement_04}}
 {id=WavePlayer_5, properties {soundWaveAssetPtr=/AmbientAudioContent/Waves/OneShots/Spot_Chain_Movement_05}}
 {id=WavePlayer_6, properties {soundWaveAssetPtr=/AmbientAudioContent/Waves/OneShots/Spot_Chain_Movement_06}}
 {id=WavePlayer_7, properties {soundWaveAssetPtr=/AmbientAudioContent/Waves/OneShots/Spot_Chain_Movement_07}}
 {id=WavePlayer_8, properties {soundWaveAssetPtr=/AmbientAudioContent/Waves/OneShots/Spot_Chain_Movement_08}}
]
properties {
 bOverrideAttenuation=true
 maxDistance=2500
 soundClassObject=&Fort_Ambient_OneShots
 volumeMultiplier=0.05
 attenuationOverrides {
  bAttenuateWithLPF=true
  bEnableOcclusion=true
  distanceAlgorithm=Custom
  falloffDistance=2500
  lPFFrequencyAtMax=1000
  lPFRadiusMax=3000
  lPFRadiusMin=1000
  occlusionInterpolationTime=3
  occlusionLowPassFilterFrequency=1000
  occlusionTraceChannel=ECC_GameTraceChannel11
  occlusionVolumeAttenuation=0.5
  stereoSpread=100
  attenuationShapeExtents {x=0}
  customAttenuationCurve {editorCurveData {keys [
     <at 0 0>
     {arriveTangent=2.5, interpMode=RCIM_Cubic, leaveTangent=2.5, time=0.25, value=0.0018}
     <at 0.4 1>
     {arriveTangent=-0.0071, interpMode=RCIM_Cubic, leaveTangent=-0.0071, tangentMode=RCTM_User, time=0.6, value=1}
     <at 1 0>
    ]}}
 }
}
