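# Sound cue for a distant brimstone lava steam release, an ambient one-shot spot sound.
# An equally weighted Random node picks one of five wave variations and feeds a Modulator that varies its playback, with a lowered minimum pitch.
# It uses the ambient one-shots sound class and a shared less-positional logarithmic attenuation asset, with bus and submix sends disabled.
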SoundCue /AmbientAudioContent/SoundCues/OneShots/Spot_Brimstone_Lava_Steam_Release_Distant_Cue
root=Modulator_0
nodes [
 {id=Modulator_0, children [Random_1], properties {pitchMin=0.9}}
 {id=Random_1, children [WavePlayer_5,WavePlayer_6,WavePlayer_7,WavePlayer_8,WavePlayer_9], properties {weights [1,1,1,1,1]}}
 {id=WavePlayer_5, properties {soundWaveAssetPtr=/AmbientAudioContent/Waves/OneShots/Spot_Brimstone_Lava_Steam_Release_Distant_01}}
 {id=WavePlayer_6, properties {soundWaveAssetPtr=/AmbientAudioContent/Waves/OneShots/Spot_Brimstone_Lava_Steam_Release_Distant_02}}
 {id=WavePlayer_7, properties {soundWaveAssetPtr=/AmbientAudioContent/Waves/OneShots/Spot_Brimstone_Lava_Steam_Release_Distant_03}}
 {id=WavePlayer_8, properties {soundWaveAssetPtr=/AmbientAudioContent/Waves/OneShots/Spot_Brimstone_Lava_Steam_Release_Distant_04}}
 {id=WavePlayer_9, properties {soundWaveAssetPtr=/AmbientAudioContent/Waves/OneShots/Spot_Brimstone_Lava_Steam_Release_Distant_05}}
]
properties {
 attenuationSettings=/Game/Sounds/Attenuation/Fort_TileAmbient_OneShot_LessPositional_Log_Att
 bEnableBusSends=false
 bEnableSubmixSends=false
 maxDistance=6200
 soundClassObject=&Fort_Ambient_OneShots
 volumeMultiplier=1.002
 attenuationOverrides {falloffDistance=6000, stereoSpread=10000, attenuationShapeExtents {x=2000}}
}
