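# Sound cue for a castle interior ambient one-shot that plays distant metallic rattles, routed to the indoor ambient sound class with less-positional attenuation.
# An evenly weighted random node picks either one of six rattle variations or an empty mixer, so about half the triggers are silent.
# A root modulator varies pitch and volume.
# A quality-level node feeds this only in its first two slots; the other two are empty.
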SoundCue /AmbientAudioContent/SoundCues/OneShots/Spot_Castle_Interior_Foley_Movements_Small_Cue
root=Modulator_0
nodes [
 {id=Mixer_1, children [null], properties {inputVolume [1]}}
 {id=Mixer_2, children [Random_0], properties {inputVolume [1]}}
 {id=Modulator_0, children [QualityLevel_0], properties {pitchMin=0.7, volumeMin=0.85}}
 {id=QualityLevel_0, children [Random_8,Random_8,null,null]}
 {id=Random_0, children [WavePlayer_0,WavePlayer_1,WavePlayer_2,WavePlayer_3,WavePlayer_4,WavePlayer_5], properties {weights [1,1,1,1,1,1]}}
 {id=Random_8, children [Mixer_2,Mixer_1], properties {weights [1,1]}}
 {id=WavePlayer_0, properties {soundWaveAssetPtr=/AmbientAudioContent/Waves/OneShots/Spot_Castle_Metallic_Rattle_Distant_01}}
 {id=WavePlayer_1, properties {soundWaveAssetPtr=/AmbientAudioContent/Waves/OneShots/Spot_Castle_Metallic_Rattle_Distant_02}}
 {id=WavePlayer_2, properties {soundWaveAssetPtr=/AmbientAudioContent/Waves/OneShots/Spot_Castle_Metallic_Rattle_Distant_03}}
 {id=WavePlayer_3, properties {soundWaveAssetPtr=/AmbientAudioContent/Waves/OneShots/Spot_Castle_Metallic_Rattle_Distant_04}}
 {id=WavePlayer_4, properties {soundWaveAssetPtr=/AmbientAudioContent/Waves/OneShots/Spot_Castle_Metallic_Rattle_Distant_05}}
 {id=WavePlayer_5, properties {soundWaveAssetPtr=/AmbientAudioContent/Waves/OneShots/Spot_Castle_Metallic_Rattle_Distant_06}}
]
properties {
 attenuationSettings=/Game/Sounds/Attenuation/Fort_TileAmbient_OneShot_LessPositional_Log_Att
 maxDistance=6200
 soundClassObject=&Fort_Ambient_Indoors_Athena
 volumeMultiplier=0.501
}
