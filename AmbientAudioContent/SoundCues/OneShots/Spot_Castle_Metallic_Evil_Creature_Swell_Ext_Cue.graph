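# Sound cue for an outdoor ambient one-shot: distant metallic, evil-creature swells for a castle setting.
# A random node picks one of 22 wave variants, feeding the first two quality-level slots (the other two are empty), then a modulator varies pitch and volume.
# Distant-creak attenuation, a tight concurrency override and distance-scaled sends to default and distant reverb submixes shape playback.
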SoundCue /AmbientAudioContent/SoundCues/OneShots/Spot_Castle_Metallic_Evil_Creature_Swell_Ext_Cue
root=Modulator_0
nodes [
 {id=Modulator_0, children [QualityLevel_0], properties {pitchMin=0.8, volumeMax=1.2, volumeMin=0.85}}
 {id=QualityLevel_0, children [Random_1,Random_1,null,null]}
 {
  id=Random_1
  children [
   WavePlayer_15
   WavePlayer_16
   WavePlayer_17
   WavePlayer_18
   WavePlayer_19
   WavePlayer_20
   WavePlayer_21
   WavePlayer_22
   WavePlayer_23
   WavePlayer_24
   WavePlayer_25
   WavePlayer_26
   WavePlayer_27
   WavePlayer_28
   WavePlayer_29
   WavePlayer_30
   WavePlayer_31
   WavePlayer_39
   WavePlayer_40
   WavePlayer_41
   WavePlayer_42
   WavePlayer_43
  ]
  properties {weights [1,1,1,1,1,1,1,1,1,1,1,1,1,1,1,1,1,1,1,1,1,1]}
 }
 {id=WavePlayer_15, properties {soundWaveAssetPtr=/AmbientAudioContent/Waves/OneShots/Spot_Castle_Metallic_Evil_Creature_Swell_01}}
 {id=WavePlayer_16, properties {soundWaveAssetPtr=/AmbientAudioContent/Waves/OneShots/Spot_Castle_Metallic_Evil_Creature_Swell_02}}
 {id=WavePlayer_17, properties {soundWaveAssetPtr=/AmbientAudioContent/Waves/OneShots/Spot_Castle_Metallic_Evil_Creature_Swell_03}}
 {id=WavePlayer_18, properties {soundWaveAssetPtr=/AmbientAudioContent/Waves/OneShots/Spot_Castle_Metallic_Evil_Creature_Swell_04}}
 {id=WavePlayer_19, properties {soundWaveAssetPtr=/AmbientAudioContent/Waves/OneShots/Spot_Castle_Metallic_Evil_Creature_Swell_05}}
 {id=WavePlayer_20, properties {soundWaveAssetPtr=/AmbientAudioContent/Waves/OneShots/Spot_Castle_Metallic_Evil_Creature_Swell_06}}
 {id=WavePlayer_21, properties {soundWaveAssetPtr=/AmbientAudioContent/Waves/OneShots/Spot_Castle_Metallic_Evil_Creature_Swell_07}}
 {id=WavePlayer_22, properties {soundWaveAssetPtr=/AmbientAudioContent/Waves/OneShots/Spot_Castle_Metallic_Evil_Creature_Swell_08}}
 {id=WavePlayer_23, properties {soundWaveAssetPtr=/AmbientAudioContent/Waves/OneShots/Spot_Castle_Metallic_Evil_Creature_Swell_09}}
 {id=WavePlayer_24, properties {soundWaveAssetPtr=/AmbientAudioContent/Waves/OneShots/Spot_Castle_Metallic_Evil_Creature_Swell_10}}
 {id=WavePlayer_25, properties {soundWaveAssetPtr=/AmbientAudioContent/Waves/OneShots/Spot_Castle_Metallic_Evil_Creature_Swell_11}}
 {id=WavePlayer_26, properties {soundWaveAssetPtr=/AmbientAudioContent/Waves/OneShots/Spot_Castle_Metallic_Evil_Creature_Swell_12}}
 {id=WavePlayer_27, properties {soundWaveAssetPtr=/AmbientAudioContent/Waves/OneShots/Spot_Castle_Metallic_Evil_Creature_Swell_13}}
 {id=WavePlayer_28, properties {soundWaveAssetPtr=/AmbientAudioContent/Waves/OneShots/Spot_Castle_Metallic_Evil_Creature_Swell_14}}
 {id=WavePlayer_29, properties {soundWaveAssetPtr=/AmbientAudioContent/Waves/OneShots/Spot_Castle_Metallic_Evil_Creature_Swell_15}}
 {id=WavePlayer_30, properties {soundWaveAssetPtr=/AmbientAudioContent/Waves/OneShots/Spot_Castle_Metallic_Evil_Creature_Swell_16}}
 {id=WavePlayer_31, properties {soundWaveAssetPtr=/AmbientAudioContent/Waves/OneShots/Spot_Castle_Metallic_Evil_Creature_Swell_17}}
 {id=WavePlayer_39, properties {soundWaveAssetPtr=/AmbientAudioContent/Waves/OneShots/Spot_Castle_Metallic_Evil_Creature_Swell_18}}
 {id=WavePlayer_40, properties {soundWaveAssetPtr=/AmbientAudioContent/Waves/OneShots/Spot_Castle_Metallic_Evil_Creature_Swell_19}}
 {id=WavePlayer_41, properties {soundWaveAssetPtr=/AmbientAudioContent/Waves/OneShots/Spot_Castle_Metallic_Evil_Creature_Swell_20}}
 {id=WavePlayer_42, properties {soundWaveAssetPtr=/AmbientAudioContent/Waves/OneShots/Spot_Castle_Metallic_Evil_Creature_Swell_21}}
 {id=WavePlayer_43, properties {soundWaveAssetPtr=/AmbientAudioContent/Waves/OneShots/Spot_Castle_Metallic_Evil_Creature_Swell_22}}
]
properties {
 attenuationSettings=/Game/Sounds/Attenuation/Fort_TileAmbient_OneShot_Distant_Creak_Linear_Att
 bOverrideConcurrency=true
 maxDistance=8700
 soundClassObject=&Fort_Ambient_OneShots_Outdoors
 volumeMultiplier=0.6
 concurrencyOverrides {maxCount=2, voiceStealReleaseTime=2}
 soundSubmixSends [
  {
   maxSendDistance=8500
   minSendDistance=200
   minSendLevel=0.6
   soundSubmix=&Fortnite_Reverb_Default
   customSendLevelCurve {editorCurveData {keys [<at 0 0.25>,<at 1 1>]}}
  }
  {
   maxSendDistance=8500
   maxSendLevel=0.6
   minSendDistance=200
   minSendLevel=0.25
   soundSubmix=&Fortnite_Reverb_Distant
   customSendLevelCurve {editorCurveData {keys [
      <at 0 0>
      {arriveTangent=1, interpMode=RCIM_Cubic, leaveTangent=1, time=0.66, value=0.33}
      {arriveTangent=2.1868, interpMode=RCIM_Cubic, leaveTangent=2.1868, tangentMode=RCTM_User, time=1, value=1}
     ]}}
  }
 ]
}
