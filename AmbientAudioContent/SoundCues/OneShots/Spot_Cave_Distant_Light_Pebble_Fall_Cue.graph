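# Sound cue for a one-shot ambient spot effect of light pebbles falling in a distant cave.
# A Random node picks one of six equally weighted pebble-fall waves and feeds it through a root Modulator left at default settings.
# It routes to the Athena indoor ambient sound class with bus and submix sends disabled, and references a shared, less-positional, logarithmic one-shot attenuation asset.
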SoundCue /AmbientAudioContent/SoundCues/OneShots/Spot_Cave_Distant_Light_Pebble_Fall_Cue
root=Modulator_0
nodes [
 {id=Modulator_0, children [Random_0]}
 {
  id=Random_0
  children [WavePlayer_0, WavePlayer_1, WavePlayer_2, WavePlayer_3, WavePlayer_4, WavePlayer_11]
  properties {weights [1,1,1,1,1,1]}
 }
 {id=WavePlayer_0, properties {soundWaveAssetPtr=/AmbientAudioContent/Waves/OneShots/Spot_Cattus_Cave_Pebble_Fall_06}}
 {id=WavePlayer_1, properties {soundWaveAssetPtr=/AmbientAudioContent/Waves/OneShots/Spot_Cattus_Cave_Pebble_Fall_05}}
 {id=WavePlayer_11, properties {soundWaveAssetPtr=/AmbientAudioContent/Waves/OneShots/Spot_Cattus_Cave_Pebble_Fall_01}}
 {id=WavePlayer_2, properties {soundWaveAssetPtr=/AmbientAudioContent/Waves/OneShots/Spot_Cattus_Cave_Pebble_Fall_04}}
 {id=WavePlayer_3, properties {soundWaveAssetPtr=/AmbientAudioContent/Waves/OneShots/Spot_Cattus_Cave_Pebble_Fall_03}}
 {id=WavePlayer_4, properties {soundWaveAssetPtr=/AmbientAudioContent/Waves/OneShots/Spot_Cattus_Cave_Pebble_Fall_02}}
]
properties {
 attenuationSettings=/Game/Sounds/Attenuation/Fort_TileAmbient_OneShot_LessPositional_Log_Att
 bEnableBusSends=false
 bEnableSubmixSends=false
 maxDistance=6200
 soundClassObject=&Fort_Ambient_Indoors_Athena
 volumeMultiplier=1.002
 attenuationOverrides {bAttenuate=false}
}
