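# Sound cue for one-shot ambient cave water drips on metal, classed as indoor ambience.
# A Random node picks among thirty equally weighted drip waves and feeds a Modulator that varies pitch.
# A Quality Level root routes its first two slots to that chain and leaves the other two empty.
# The cue uses a shared wide one-shot attenuation asset and overrides concurrency to cap it at two simultaneous instances.
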SoundCue /AmbientAudioContent/SoundCues/OneShots/Spot_Cave_Drips_Metal_Cue
root=QualityLevel_0
nodes [
 {id=Modulator_0, children [Random_1], properties {pitchMax=1.1}}
 {id=QualityLevel_0, children [Modulator_0,Modulator_0,null,null]}
 {
  id=Random_1
  children [
   WavePlayer_0
   WavePlayer_1
   WavePlayer_2
   WavePlayer_8
   WavePlayer_9
   WavePlayer_10
   WavePlayer_19
   WavePlayer_20
   WavePlayer_21
   WavePlayer_22
   WavePlayer_23
   WavePlayer_24
   WavePlayer_42
   WavePlayer_43
   WavePlayer_44
   WavePlayer_45
   WavePlayer_46
   WavePlayer_47
   WavePlayer_48
   WavePlayer_49
   WavePlayer_50
   WavePlayer_51
   WavePlayer_52
   WavePlayer_53
   WavePlayer_54
   WavePlayer_55
   WavePlayer_56
   WavePlayer_57
   WavePlayer_58
   WavePlayer_59
  ]
  properties {weights [1,1,1,1,1,1,1,1,1,1,1,1,1,1,1,1,1,1,1,1,1,1,1,1,1,1,1,1,1,1]}
 }
 {id=WavePlayer_0, properties {soundWaveAssetPtr=/AmbientAudioContent/Waves/OneShots/Spot_Cave_Drips_Metal_01}}
 {id=WavePlayer_1, properties {soundWaveAssetPtr=/AmbientAudioContent/Waves/OneShots/Spot_Cave_Drips_Metal_02}}
 {id=WavePlayer_10, properties {soundWaveAssetPtr=/AmbientAudioContent/Waves/OneShots/Spot_Cave_Drips_Metal_06}}
 {id=WavePlayer_19, properties {soundWaveAssetPtr=/AmbientAudioContent/Waves/OneShots/Spot_Cave_Drips_Metal_07}}
 {id=WavePlayer_2, properties {soundWaveAssetPtr=/AmbientAudioContent/Waves/OneShots/Spot_Cave_Drips_Metal_03}}
 {id=WavePlayer_20, properties {soundWaveAssetPtr=/AmbientAudioContent/Waves/OneShots/Spot_Cave_Drips_Metal_08}}
 {id=WavePlayer_21, properties {soundWaveAssetPtr=/AmbientAudioContent/Waves/OneShots/Spot_Cave_Drips_Metal_09}}
 {id=WavePlayer_22, properties {soundWaveAssetPtr=/AmbientAudioContent/Waves/OneShots/Spot_Cave_Drips_Metal_10}}
 {id=WavePlayer_23, properties {soundWaveAssetPtr=/AmbientAudioContent/Waves/OneShots/Spot_Cave_Drips_Metal_11}}
 {id=WavePlayer_24, properties {soundWaveAssetPtr=/AmbientAudioContent/Waves/OneShots/Spot_Cave_Drips_Metal_12}}
 {id=WavePlayer_42, properties {soundWaveAssetPtr=/AmbientAudioContent/Waves/OneShots/Spot_Cave_Drips_Metal_13}}
 {id=WavePlayer_43, properties {soundWaveAssetPtr=/AmbientAudioContent/Waves/OneShots/Spot_Cave_Drips_Metal_14}}
 {id=WavePlayer_44, properties {soundWaveAssetPtr=/AmbientAudioContent/Waves/OneShots/Spot_Cave_Drips_Metal_15}}
 {id=WavePlayer_45, properties {soundWaveAssetPtr=/AmbientAudioContent/Waves/OneShots/Spot_Cave_Drips_Metal_16}}
 {id=WavePlayer_46, properties {soundWaveAssetPtr=/AmbientAudioContent/Waves/OneShots/Spot_Cave_Drips_Metal_17}}
 {id=WavePlayer_47, properties {soundWaveAssetPtr=/AmbientAudioContent/Waves/OneShots/Spot_Cave_Drips_Metal_18}}
 {id=WavePlayer_48, properties {soundWaveAssetPtr=/AmbientAudioContent/Waves/OneShots/Spot_Cave_Drips_Metal_19}}
 {id=WavePlayer_49, properties {soundWaveAssetPtr=/AmbientAudioContent/Waves/OneShots/Spot_Cave_Drips_Metal_20}}
 {id=WavePlayer_50, properties {soundWaveAssetPtr=/AmbientAudioContent/Waves/OneShots/Spot_Cave_Drips_Metal_21}}
 {id=WavePlayer_51, properties {soundWaveAssetPtr=/AmbientAudioContent/Waves/OneShots/Spot_Cave_Drips_Metal_22}}
 {id=WavePlayer_52, properties {soundWaveAssetPtr=/AmbientAudioContent/Waves/OneShots/Spot_Cave_Drips_Metal_23}}
 {id=WavePlayer_53, properties {soundWaveAssetPtr=/AmbientAudioContent/Waves/OneShots/Spot_Cave_Drips_Metal_24}}
 {id=WavePlayer_54, properties {soundWaveAssetPtr=/AmbientAudioContent/Waves/OneShots/Spot_Cave_Drips_Metal_25}}
 {id=WavePlayer_55, properties {soundWaveAssetPtr=/AmbientAudioContent/Waves/OneShots/Spot_Cave_Drips_Metal_26}}
 {id=WavePlayer_56, properties {soundWaveAssetPtr=/AmbientAudioContent/Waves/OneShots/Spot_Cave_Drips_Metal_27}}
 {id=WavePlayer_57, properties {soundWaveAssetPtr=/AmbientAudioContent/Waves/OneShots/Spot_Cave_Drips_Metal_28}}
 {id=WavePlayer_58, properties {soundWaveAssetPtr=/AmbientAudioContent/Waves/OneShots/Spot_Cave_Drips_Metal_29}}
 {id=WavePlayer_59, properties {soundWaveAssetPtr=/AmbientAudioContent/Waves/OneShots/Spot_Cave_Drips_Metal_30}}
 {id=WavePlayer_8, properties {soundWaveAssetPtr=/AmbientAudioContent/Waves/OneShots/Spot_Cave_Drips_Metal_04}}
 {id=WavePlayer_9, properties {soundWaveAssetPtr=/AmbientAudioContent/Waves/OneShots/Spot_Cave_Drips_Metal_05}}
]
properties {
 attenuationSettings=/Game/Sounds/Attenuation/Fort_TileAmbient_OneShot_Wider_Att
 bEnableBusSends=false
 bEnableSubmixSends=false
 bOverrideConcurrency=true
 maxDistance=8000
 soundClassObject=&Fort_Ambient_Indoors_Athena
 volumeMultiplier=0.501
 attenuationOverrides {falloffDistance=6000, stereoSpread=10000, attenuationShapeExtents {x=2000}}
 concurrencyOverrides {maxCount=2}
}
