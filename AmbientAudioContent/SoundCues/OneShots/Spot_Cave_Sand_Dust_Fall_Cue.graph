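# Ambient one-shot Sound Cue for sand and dust falling in a cave.
# A Random node picks one of four equally weighted fall recordings, and a Modulator varies the result, with a lowered volume floor for softer variations.
# It uses a closer tile-ambient one-shot attenuation asset and the indoor ambient one-shots sound class, with a concurrency override capping simultaneous instances at a small count.
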SoundCue /AmbientAudioContent/SoundCues/OneShots/Spot_Cave_Sand_Dust_Fall_Cue
root=Modulator_0
nodes [
 {id=Modulator_0, children [Random_0], properties {volumeMin=0.8}}
 {id=Random_0, children [WavePlayer_4, WavePlayer_9, WavePlayer_10, WavePlayer_11], properties {weights [1,1,1,1]}}
 {id=WavePlayer_10, properties {soundWaveAssetPtr=/AmbientAudioContent/Waves/OneShots/Spot_Cattus_Cave_Sand_Dust_Fall_03}}
 {id=WavePlayer_11, properties {soundWaveAssetPtr=/AmbientAudioContent/Waves/OneShots/Spot_Cattus_Cave_Sand_Dust_Fall_04}}
 {id=WavePlayer_4, properties {soundWaveAssetPtr=/AmbientAudioContent/Waves/OneShots/Spot_Cattus_Cave_Sand_Dust_Fall_01}}
 {id=WavePlayer_9, properties {soundWaveAssetPtr=/AmbientAudioContent/Waves/OneShots/Spot_Cattus_Cave_Sand_Dust_Fall_02}}
]
properties {
 attenuationSettings=/Game/Sounds/Attenuation/Fort_TileAmbient_OneShot_Att_Closer
 bOverrideConcurrency=true
 maxDistance=8500
 soundClassObject=/Game/Sounds/FortSoundClassesAndModes/Fort_Ambient_OneShots_Indoors
 volumeMultiplier=1
 attenuationOverrides {bAttenuate=false}
 concurrencyOverrides {maxCount=2}
}
