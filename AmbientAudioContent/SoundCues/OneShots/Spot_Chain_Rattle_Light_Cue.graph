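# Sound cue for a light chain-rattle ambient one-shot, classed as indoor ambience.
# A random node picks among eight equally weighted slow rattle waves, feeding a quality-level switch that fills only its first two slots, then a default-setting modulator for pitch and volume variation.
# It borrows a bird one-shot attenuation asset, overrides concurrency to two voices, and sends to an exterior ambient source bus and the default reverb submix.
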SoundCue /AmbientAudioContent/SoundCues/OneShots/Spot_Chain_Rattle_Light_Cue
root=Modulator_0
nodes [
 {id=Modulator_0, children [QualityLevel_0]}
 {id=QualityLevel_0, children [Random_1,Random_1,null,null]}
 {
  id=Random_1
  children [WavePlayer_1, WavePlayer_9, WavePlayer_10, WavePlayer_11, WavePlayer_12, WavePlayer_13, WavePlayer_14, WavePlayer_15]
  properties {weights [1,1,1,1,1,1,1,1]}
 }
 {id=WavePlayer_1, properties {soundWaveAssetPtr=/AmbientAudioContent/Waves/OneShots/Spot_ChainRattle_Slow_01}}
 {id=WavePlayer_10, properties {soundWaveAssetPtr=/AmbientAudioContent/Waves/OneShots/Spot_ChainRattle_Slow_03}}
 {id=WavePlayer_11, properties {soundWaveAssetPtr=/AmbientAudioContent/Waves/OneShots/Spot_ChainRattle_Slow_04}}
 {id=WavePlayer_12, properties {soundWaveAssetPtr=/AmbientAudioContent/Waves/OneShots/Spot_ChainRattle_Slow_05}}
 {id=WavePlayer_13, properties {soundWaveAssetPtr=/AmbientAudioContent/Waves/OneShots/Spot_ChainRattle_Slow_06}}
 {id=WavePlayer_14, properties {soundWaveAssetPtr=/AmbientAudioContent/Waves/OneShots/Spot_ChainRattle_Slow_07}}
 {id=WavePlayer_15, properties {soundWaveAssetPtr=/AmbientAudioContent/Waves/OneShots/Spot_ChainRattle_Slow_08}}
 {id=WavePlayer_9, properties {soundWaveAssetPtr=/AmbientAudioContent/Waves/OneShots/Spot_ChainRattle_Slow_02}}
]
properties {
 attenuationSettings=&Fort_TileAmbient_OneShot_Bird_Att
 bOverrideConcurrency=true
 maxDistance=8005
 soundClassObject=&Fort_Ambient_Indoors_Athena
 volumeMultiplier=0.551
 attenuationOverrides {
  bAttenuateWithLPF=true
  distanceAlgorithm=NaturalSound
  falloffDistance=2000
  lPFFrequencyAtMax=1000
  lPFRadiusMax=3000
  lPFRadiusMin=1000
  stereoSpread=2000
  attenuationShapeExtents {x=1500}
 }
 busSends [{audioBus=null, minSendLevel=1, sendLevel=1, soundSourceBus=&AmbientAudio_Exterior_SourceBus}]
 concurrencyOverrides {maxCount=2}
 soundSubmixSends [{maxSendDistance=4000, minSendDistance=3000, minSendLevel=0.6, soundSubmix=&Fortnite_Reverb_Default}]
}
