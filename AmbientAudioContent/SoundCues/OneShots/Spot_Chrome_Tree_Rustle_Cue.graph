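# Sound cue for a spot ambient one-shot of chrome tree rustling, using a wider tile-ambient attenuation and the ambient one-shots sound class.
# A Random node picks one of ten equally weighted rustle waves and feeds a Modulator that varies volume and slightly raises pitch.
# The root quality-level node routes its first two slots to that same Modulator and leaves the other two empty.
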SoundCue /AmbientAudioContent/SoundCues/OneShots/Spot_Chrome_Tree_Rustle_Cue
root=QualityLevel_0
nodes [
 {id=Modulator_1, children [Random_1], properties {pitchMax=1.1, pitchMin=1, volumeMax=1.1, volumeMin=0.5}}
 {id=QualityLevel_0, children [Modulator_1,Modulator_1,null,null]}
 {
  id=Random_1
  children [
   WavePlayer_11
   WavePlayer_12
   WavePlayer_13
   WavePlayer_14
   WavePlayer_15
   WavePlayer_16
   WavePlayer_17
   WavePlayer_18
   WavePlayer_19
   WavePlayer_21
  ]
  properties {weights [1,1,1,1,1,1,1,1,1,1]}
 }
 {id=WavePlayer_11, properties {soundWaveAssetPtr=/AmbientAudioContent/Waves/OneShots/Spot_Chrome_Trees_Rustle_11}}
 {id=WavePlayer_12, properties {soundWaveAssetPtr=/AmbientAudioContent/Waves/OneShots/Spot_Chrome_Trees_Rustle_10}}
 {id=WavePlayer_13, properties {soundWaveAssetPtr=/AmbientAudioContent/Waves/OneShots/Spot_Chrome_Trees_Rustle_09}}
 {id=WavePlayer_14, properties {soundWaveAssetPtr=/AmbientAudioContent/Waves/OneShots/Spot_Chrome_Trees_Rustle_08}}
 {id=WavePlayer_15, properties {soundWaveAssetPtr=/AmbientAudioContent/Waves/OneShots/Spot_Chrome_Trees_Rustle_07}}
 {id=WavePlayer_16, properties {soundWaveAssetPtr=/AmbientAudioContent/Waves/OneShots/Spot_Chrome_Trees_Rustle_06}}
 {id=WavePlayer_17, properties {soundWaveAssetPtr=/AmbientAudioContent/Waves/OneShots/Spot_Chrome_Trees_Rustle_05}}
 {id=WavePlayer_18, properties {soundWaveAssetPtr=/AmbientAudioContent/Waves/OneShots/Spot_Chrome_Trees_Rustle_04}}
 {id=WavePlayer_19, properties {soundWaveAssetPtr=/AmbientAudioContent/Waves/OneShots/Spot_Chrome_Trees_Rustle_03}}
 {id=WavePlayer_21, properties {soundWaveAssetPtr=/AmbientAudioContent/Waves/OneShots/Spot_Chrome_Trees_Rustle_01}}
]
properties {
 attenuationSettings=/Game/Sounds/Attenuation/Fort_TileAmbient_OneShot_Wider_Att
 bEnableBusSends=false
 maxDistance=8000
 soundClassObject=&Fort_Ambient_OneShots
 volumeMultiplier=0.501
 preEffectBusSends [{}]
}
