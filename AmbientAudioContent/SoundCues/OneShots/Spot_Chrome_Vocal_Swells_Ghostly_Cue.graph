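# Sound cue for ghostly chrome vocal swell one-shots, played as outdoor ambient spot sounds with distant-creature attenuation.
# A Random node picks among eleven equally weighted waves, and every quality level routes to that same node.
# A Modulator randomizes pitch and volume, then a root Enveloper applies a pitch curve over time with flat volume.
# A concurrency override caps simultaneous instances at two.
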SoundCue /AmbientAudioContent/SoundCues/OneShots/Spot_Chrome_Vocal_Swells_Ghostly_Cue
root=Enveloper_0
nodes [
 {
  id=Enveloper_0
  children [Modulator_0]
  properties {pitchCurve {editorCurveData {keys [{time=0}, {time=1}, <at 7.5 0.8>]}}, volumeCurve {editorCurveData {keys [<at 0 1>]}}}
 }
 {id=Modulator_0, children [QualityLevel_0], properties {pitchMin=0.8, volumeMax=1, volumeMin=0.8}}
 {id=QualityLevel_0, children [Random_1,Random_1,Random_1,Random_1]}
 {
  id=Random_1
  children [
   WavePlayer_10
   WavePlayer_11
   WavePlayer_12
   WavePlayer_13
   WavePlayer_14
   WavePlayer_15
   WavePlayer_17
   WavePlayer_18
   WavePlayer_19
   WavePlayer_20
   WavePlayer_21
  ]
  properties {weights [1,1,1,1,1,1,1,1,1,1,1]}
 }
 {id=WavePlayer_10, properties {soundWaveAssetPtr=/AmbientAudioContent/Waves/OneShots/Spot_Chrome_Vocal_Swells_Ghostly_01}}
 {id=WavePlayer_11, properties {soundWaveAssetPtr=/AmbientAudioContent/Waves/OneShots/Spot_Chrome_Vocal_Swells_Ghostly_02}}
 {id=WavePlayer_12, properties {soundWaveAssetPtr=/AmbientAudioContent/Waves/OneShots/Spot_Chrome_Vocal_Swells_Ghostly_03}}
 {id=WavePlayer_13, properties {soundWaveAssetPtr=/AmbientAudioContent/Waves/OneShots/Spot_Chrome_Vocal_Swells_Ghostly_04}}
 {id=WavePlayer_14, properties {soundWaveAssetPtr=/AmbientAudioContent/Waves/OneShots/Spot_Chrome_Vocal_Swells_Ghostly_05}}
 {id=WavePlayer_15, properties {soundWaveAssetPtr=/AmbientAudioContent/Waves/OneShots/Spot_Chrome_Vocal_Swells_Ghostly_06}}
 {id=WavePlayer_17, properties {soundWaveAssetPtr=/AmbientAudioContent/Waves/OneShots/Spot_Chrome_Vocal_Swells_Ghostly_07}}
 {id=WavePlayer_18, properties {soundWaveAssetPtr=/AmbientAudioContent/Waves/OneShots/Spot_Chrome_Vocal_Swells_Ghostly_08}}
 {id=WavePlayer_19, properties {soundWaveAssetPtr=/AmbientAudioContent/Waves/OneShots/Spot_Chrome_Vocal_Swells_Ghostly_09}}
 {id=WavePlayer_20, properties {soundWaveAssetPtr=/AmbientAudioContent/Waves/OneShots/Spot_Chrome_Vocal_Swells_Ghostly_10}}
 {id=WavePlayer_21, properties {soundWaveAssetPtr=/AmbientAudioContent/Waves/OneShots/Spot_Chrome_Vocal_Swells_Ghostly_11}}
]
properties {
 attenuationSettings=/Game/Sounds/Attenuation/Fort_TileAmbient_OneShot_Distant_Creature_Att
 bOverrideConcurrency=true
 maxDistance=8500
 soundClassObject=&Fort_Ambient_Outdoors_Athena
 volumeMultiplier=0.351
 concurrencyOverrides {maxCount=2}
}
