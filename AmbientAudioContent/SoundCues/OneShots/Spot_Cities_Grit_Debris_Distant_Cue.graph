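# Sound cue for distant outdoor ambient one-shots of gritty city glass debris drops, using a distant creak attenuation asset and the outdoor one-shot sound class.
# A quality switch sends its first two levels to an equal-weight random pick of eight waves and its last two to a three-wave subset.
# Each branch passes through its own modulator for slight pitch and volume variation.
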SoundCue /AmbientAudioContent/SoundCues/OneShots/Spot_Cities_Grit_Debris_Distant_Cue
root=QualityLevel_0
nodes [
 {id=Modulator_1, children [Random_0], properties {pitchMax=1.1, volumeMax=1.1, volumeMin=0.9}}
 {id=Modulator_2, children [Random_3], properties {pitchMax=1.1, volumeMax=1.1, volumeMin=0.9}}
 {id=QualityLevel_0, children [Modulator_1,Modulator_1,Modulator_2,Modulator_2]}
 {
  id=Random_0
  children [WavePlayer_0,WavePlayer_1,WavePlayer_2,WavePlayer_3,WavePlayer_4,WavePlayer_5,WavePlayer_6,WavePlayer_7]
  properties {weights [1,1,1,1,1,1,1,1]}
 }
 {id=Random_3, children [WavePlayer_17, WavePlayer_18, WavePlayer_19], properties {weights [1,1,1]}}
 {id=WavePlayer_0, properties {soundWaveAssetPtr=/AmbientAudioContent/Waves/OneShots/Spot_City_Grit_Glass_Debris_Drop_1}}
 {id=WavePlayer_1, properties {soundWaveAssetPtr=/AmbientAudioContent/Waves/OneShots/Spot_City_Grit_Glass_Debris_Drop_2}}
 {id=WavePlayer_17, properties {soundWaveAssetPtr=&Spot_City_Grit_Glass_Debris_Drop_1}}
 {id=WavePlayer_18, properties {soundWaveAssetPtr=/AmbientAudioContent/Waves/OneShots/Spot_City_Grit_Glass_Debris_Drop_4}}
 {id=WavePlayer_19, properties {soundWaveAssetPtr=/AmbientAudioContent/Waves/OneShots/Spot_City_Grit_Glass_Debris_Drop_8}}
 {id=WavePlayer_2, properties {soundWaveAssetPtr=/AmbientAudioContent/Waves/OneShots/Spot_City_Grit_Glass_Debris_Drop_3}}
 {id=WavePlayer_3, properties {soundWaveAssetPtr=&Spot_City_Grit_Glass_Debris_Drop_4}}
 {id=WavePlayer_4, properties {soundWaveAssetPtr=/AmbientAudioContent/Waves/OneShots/Spot_City_Grit_Glass_Debris_Drop_5}}
 {id=WavePlayer_5, properties {soundWaveAssetPtr=/AmbientAudioContent/Waves/OneShots/Spot_City_Grit_Glass_Debris_Drop_6}}
 {id=WavePlayer_6, properties {soundWaveAssetPtr=/AmbientAudioContent/Waves/OneShots/Spot_City_Grit_Glass_Debris_Drop_7}}
 {id=WavePlayer_7, properties {soundWaveAssetPtr=&Spot_City_Grit_Glass_Debris_Drop_8}}
]
properties {
 attenuationSettings=/Game/Sounds/Attenuation/Fort_TileAmbient_OneShot_Distant_Creak_Att
 bEnableBusSends=false
 maxDistance=8200
 soundClassObject=&Fort_Ambient_OneShots_Outdoors
 volumeMultiplier=1
 preEffectBusSends [{}]
}
