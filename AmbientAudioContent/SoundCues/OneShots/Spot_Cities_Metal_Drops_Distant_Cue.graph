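# Sound Cue for distant metal-drop one-shots in outdoor city ambience, using a distant one-shot attenuation and the outdoor ambient sound class.
# A quality-level root sends the first two levels to a random pick of fourteen light and medium waves, and the last two to a five-wave subset.
# Each pool passes through its own modulator for slight pitch and volume variation, and the cue plays at reduced volume.
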SoundCue /AmbientAudioContent/SoundCues/OneShots/Spot_Cities_Metal_Drops_Distant_Cue
root=QualityLevel_0
nodes [
 {id=Modulator_1, children [Random_4], properties {pitchMax=1.1, volumeMax=1.1, volumeMin=0.9}}
 {id=Modulator_2, children [Random_5], properties {pitchMax=1.1, volumeMax=1.1, volumeMin=0.9}}
 {id=QualityLevel_0, children [Modulator_1,Modulator_1,Modulator_2,Modulator_2]}
 {
  id=Random_4
  children [
   WavePlayer_43
   WavePlayer_44
   WavePlayer_45
   WavePlayer_46
   WavePlayer_47
   WavePlayer_48
   WavePlayer_49
   WavePlayer_50
   WavePlayer_51
   WavePlayer_52
   WavePlayer_53
   WavePlayer_54
   WavePlayer_55
   WavePlayer_56
  ]
  properties {weights [1,1,1,1,1,1,1,1,1,1,1,1,1,1]}
 }
 {id=Random_5, children [WavePlayer_62, WavePlayer_64, WavePlayer_61, WavePlayer_60, WavePlayer_63], properties {weights [1,1,1,1,1]}}
 {id=WavePlayer_43, properties {soundWaveAssetPtr=/AmbientAudioContent/Waves/OneShots/Spot_City_Distant_Metal_Drops_Med_09}}
 {id=WavePlayer_44, properties {soundWaveAssetPtr=/AmbientAudioContent/Waves/OneShots/Spot_City_Distant_Metal_Drops_Med_08}}
 {id=WavePlayer_45, properties {soundWaveAssetPtr=/AmbientAudioContent/Waves/OneShots/Spot_City_Distant_Metal_Drops_Med_07}}
 {id=WavePlayer_46, properties {soundWaveAssetPtr=/AmbientAudioContent/Waves/OneShots/Spot_City_Distant_Metal_Drops_Med_06}}
 {id=WavePlayer_47, properties {soundWaveAssetPtr=/AmbientAudioContent/Waves/OneShots/Spot_City_Distant_Metal_Drops_Med_05}}
 {id=WavePlayer_48, properties {soundWaveAssetPtr=/AmbientAudioContent/Waves/OneShots/Spot_City_Distant_Metal_Drops_Med_04}}
 {id=WavePlayer_49, properties {soundWaveAssetPtr=/AmbientAudioContent/Waves/OneShots/Spot_City_Distant_Metal_Drops_Med_03}}
 {id=WavePlayer_50, properties {soundWaveAssetPtr=/AmbientAudioContent/Waves/OneShots/Spot_City_Distant_Metal_Drops_Med_02}}
 {id=WavePlayer_51, properties {soundWaveAssetPtr=/AmbientAudioContent/Waves/OneShots/Spot_City_Distant_Metal_Drops_Med_01}}
 {id=WavePlayer_52, properties {soundWaveAssetPtr=/AmbientAudioContent/Waves/OneShots/Spot_City_Distant_Metal_Drops_Light_5}}
 {id=WavePlayer_53, properties {soundWaveAssetPtr=/AmbientAudioContent/Waves/OneShots/Spot_City_Distant_Metal_Drops_Light_4}}
 {id=WavePlayer_54, properties {soundWaveAssetPtr=/AmbientAudioContent/Waves/OneShots/Spot_City_Distant_Metal_Drops_Light_3}}
 {id=WavePlayer_55, properties {soundWaveAssetPtr=/AmbientAudioContent/Waves/OneShots/Spot_City_Distant_Metal_Drops_Light_2}}
 {id=WavePlayer_56, properties {soundWaveAssetPtr=/AmbientAudioContent/Waves/OneShots/Spot_City_Distant_Metal_Drops_Light_1}}
 {id=WavePlayer_60, properties {soundWaveAssetPtr=&Spot_City_Distant_Metal_Drops_Light_1}}
 {id=WavePlayer_61, properties {soundWaveAssetPtr=&Spot_City_Distant_Metal_Drops_Med_01}}
 {id=WavePlayer_62, properties {soundWaveAssetPtr=&Spot_City_Distant_Metal_Drops_Light_5}}
 {id=WavePlayer_63, properties {soundWaveAssetPtr=&Spot_City_Distant_Metal_Drops_Med_05}}
 {id=WavePlayer_64, properties {soundWaveAssetPtr=&Spot_City_Distant_Metal_Drops_Med_09}}
]
properties {
 attenuationSettings=/Game/Sounds/Attenuation/Fort_TileAmbient_OneShot_Distant_Creak_Att
 bEnableBusSends=false
 maxDistance=8200
 soundClassObject=&Fort_Ambient_OneShots_Outdoors
 volumeMultiplier=0.4
 preEffectBusSends [{}]
}
