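# Sound cue for a metal squeak ambient one-shot in city areas, using a distant-creak attenuation and the ambient one-shots sound class.
# A quality-level root sends its first two slots to a seven-wave random pool and its last two to a three-wave subset of those same waves.
# Each pool passes through a modulator that slightly varies volume and lowers pitch, on top of reduced cue-wide pitch and volume.
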SoundCue /AmbientAudioContent/SoundCues/OneShots/Spot_Cities_Metal_Squeak_Cue
root=QualityLevel_0
nodes [
 {id=Modulator_1, children [Random_2], properties {pitchMax=0.9, pitchMin=0.8, volumeMax=1.1, volumeMin=0.9}}
 {id=Modulator_2, children [Random_1], properties {pitchMax=0.9, pitchMin=0.8, volumeMax=1.1, volumeMin=0.9}}
 {id=QualityLevel_0, children [Modulator_1,Modulator_1,Modulator_2,Modulator_2]}
 {id=Random_1, children [WavePlayer_24, WavePlayer_25, WavePlayer_26], properties {weights [1,1,1]}}
 {
  id=Random_2
  children [WavePlayer_0, WavePlayer_2, WavePlayer_3, WavePlayer_8, WavePlayer_9, WavePlayer_10, WavePlayer_11]
  properties {weights [1,1,1,1,1,1,1]}
 }
 {id=WavePlayer_0, properties {soundWaveAssetPtr=/AmbientAudioContent/Waves/OneShots/Spot_Cities_Metal_Squeak_04}}
 {id=WavePlayer_10, properties {soundWaveAssetPtr=/AmbientAudioContent/Waves/OneShots/Spot_Cities_Metal_Squeak_09}}
 {id=WavePlayer_11, properties {soundWaveAssetPtr=/AmbientAudioContent/Waves/OneShots/Spot_Cities_Metal_Squeak_10}}
 {id=WavePlayer_2, properties {soundWaveAssetPtr=/AmbientAudioContent/Waves/OneShots/Spot_Cities_Metal_Squeak_05}}
 {id=WavePlayer_24, properties {soundWaveAssetPtr=/AmbientAudioContent/Waves/OneShots/Spot_Cities_Metal_Squeak_07}}
 {id=WavePlayer_25, properties {soundWaveAssetPtr=/AmbientAudioContent/Waves/OneShots/Spot_Cities_Metal_Squeak_08}}
 {id=WavePlayer_26, properties {soundWaveAssetPtr=&Spot_Cities_Metal_Squeak_09}}
 {id=WavePlayer_3, properties {soundWaveAssetPtr=/AmbientAudioContent/Waves/OneShots/Spot_Cities_Metal_Squeak_06}}
 {id=WavePlayer_8, properties {soundWaveAssetPtr=&Spot_Cities_Metal_Squeak_07}}
 {id=WavePlayer_9, properties {soundWaveAssetPtr=&Spot_Cities_Metal_Squeak_08}}
]
properties {
 attenuationSettings=/Game/Sounds/Attenuation/Fort_TileAmbient_OneShot_Distant_Creak_Att
 bEnableBusSends=false
 maxDistance=8200
 pitchMultiplier=0.9
 soundClassObject=&Fort_Ambient_OneShots
 volumeMultiplier=0.351
 preEffectBusSends [{}]
}
