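# Sound cue for distant city trash one-shot ambience, played at reduced volume through the ambient one-shot sound class with a distant-creak attenuation asset.
# A quality-level root routes its first two levels to a random pool of twelve trash waves and its last two to a three-wave subset.
# Each pool passes through its own modulator, which varies pitch and volume on each play to reduce repetition.
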SoundCue /AmbientAudioContent/SoundCues/OneShots/Spot_Cities_Trash_Distant_Cue
root=QualityLevel_0
nodes [
 {id=Modulator_1, children [Random_1], properties {pitchMax=1.1, volumeMax=1.1, volumeMin=0.9}}
 {id=Modulator_2, children [Random_3], properties {pitchMax=1.1, volumeMax=1.1, volumeMin=0.9}}
 {id=QualityLevel_0, children [Modulator_1,Modulator_1,Modulator_2,Modulator_2]}
 {
  id=Random_1
  children [
   WavePlayer_8
   WavePlayer_9
   WavePlayer_10
   WavePlayer_11
   WavePlayer_12
   WavePlayer_13
   WavePlayer_14
   WavePlayer_15
   WavePlayer_16
   WavePlayer_21
   WavePlayer_22
   WavePlayer_23
  ]
  properties {weights [1,1,1,1,1,1,1,1,1,1,1,1]}
 }
 {id=Random_3, children [WavePlayer_26, WavePlayer_27, WavePlayer_28], properties {weights [1,1,1]}}
 {id=WavePlayer_10, properties {soundWaveAssetPtr=/AmbientAudioContent/Waves/OneShots/Spot_City_Distant_Trash_10}}
 {id=WavePlayer_11, properties {soundWaveAssetPtr=/AmbientAudioContent/Waves/OneShots/Spot_City_Distant_Trash_09}}
 {id=WavePlayer_12, properties {soundWaveAssetPtr=/AmbientAudioContent/Waves/OneShots/Spot_City_Distant_Trash_08}}
 {id=WavePlayer_13, properties {soundWaveAssetPtr=/AmbientAudioContent/Waves/OneShots/Spot_City_Distant_Trash_07}}
 {id=WavePlayer_14, properties {soundWaveAssetPtr=/AmbientAudioContent/Waves/OneShots/Spot_City_Distant_Trash_06}}
 {id=WavePlayer_15, properties {soundWaveAssetPtr=/AmbientAudioContent/Waves/OneShots/Spot_City_Distant_Trash_05}}
 {id=WavePlayer_16, properties {soundWaveAssetPtr=/AmbientAudioContent/Waves/OneShots/Spot_City_Distant_Trash_04}}
 {id=WavePlayer_21, properties {soundWaveAssetPtr=/AmbientAudioContent/Waves/OneShots/Spot_City_Distant_Trash_03}}
 {id=WavePlayer_22, properties {soundWaveAssetPtr=/AmbientAudioContent/Waves/OneShots/Spot_City_Distant_Trash_02}}
 {id=WavePlayer_23, properties {soundWaveAssetPtr=/AmbientAudioContent/Waves/OneShots/Spot_City_Distant_Trash_01}}
 {id=WavePlayer_26, properties {soundWaveAssetPtr=&Spot_City_Distant_Trash_01}}
 {id=WavePlayer_27, properties {soundWaveAssetPtr=&Spot_City_Distant_Trash_06}}
 {id=WavePlayer_28, properties {soundWaveAssetPtr=/AmbientAudioContent/Waves/OneShots/Spot_City_Distant_Trash_12}}
 {id=WavePlayer_8, properties {soundWaveAssetPtr=&Spot_City_Distant_Trash_12}}
 {id=WavePlayer_9, properties {soundWaveAssetPtr=/AmbientAudioContent/Waves/OneShots/Spot_City_Distant_Trash_11}}
]
properties {
 attenuationSettings=/Game/Sounds/Attenuation/Fort_TileAmbient_OneShot_Distant_Creak_Att
 bEnableBusSends=false
 maxDistance=8200
 soundClassObject=&Fort_Ambient_OneShots
 volumeMultiplier=0.5
 preEffectBusSends [{}]
}
