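# Sound cue for a distant city burglar alarm ambient one-shot, quietened and routed to the outdoor ambient one-shots sound class.
# A Random node picks with equal weight among three single alarm waves and a Concatenator that plays the fourth wave twice in a row.
# A concurrency override allows one instance at a time and blocks new ones.
# It uses distant one-shot attenuation and distance-based sends to default and distant reverb submixes with custom curves.
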SoundCue /AmbientAudioContent/SoundCues/OneShots/Spot_City_Burglar_Alarm_Distant_Cue
root=Random_0
nodes [
 {id=Concatenator_0, children [WavePlayer_5,WavePlayer_3], properties {inputVolume [1,1]}}
 {id=Random_0, children [WavePlayer_0, WavePlayer_1, WavePlayer_2, Concatenator_0], properties {weights [1,1,1,1]}}
 {id=WavePlayer_0, properties {soundWaveAssetPtr=/AmbientAudioContent/Waves/OneShots/Spot_City_Distant_Burglar_Alarm_1}}
 {id=WavePlayer_1, properties {soundWaveAssetPtr=/AmbientAudioContent/Waves/OneShots/Spot_City_Distant_Burglar_Alarm_2}}
 {id=WavePlayer_2, properties {soundWaveAssetPtr=/AmbientAudioContent/Waves/OneShots/Spot_City_Distant_Burglar_Alarm_3}}
 {id=WavePlayer_3, properties {soundWaveAssetPtr=/AmbientAudioContent/Waves/OneShots/Spot_City_Distant_Burglar_Alarm_4}}
 {id=WavePlayer_5, properties {soundWaveAssetPtr=&Spot_City_Distant_Burglar_Alarm_4}}
]
properties {
 attenuationSettings=/Game/Sounds/Attenuation/Fort_TileAmbient_OneShot_Distant_Creak_Linear_Att
 bOverrideConcurrency=true
 maxDistance=8700
 soundClassObject=&Fort_Ambient_OneShots_Outdoors
 volumeMultiplier=0.3
 concurrencyOverrides {maxCount=1, resolutionRule=PreventNew, voiceStealReleaseTime=2}
 soundSubmixSends [
  {
   maxSendDistance=8500
   minSendDistance=200
   minSendLevel=0.6
   soundSubmix=&Fortnite_Reverb_Default
   customSendLevelCurve {editorCurveData {keys [<at 0 0.25>,<at 1 1>]}}
  }
  {
   maxSendDistance=8500
   maxSendLevel=0.6
   minSendDistance=200
   minSendLevel=0.25
   soundSubmix=&Fortnite_Reverb_Distant
   customSendLevelCurve {editorCurveData {keys [
      <at 0 0>
      {arriveTangent=1, interpMode=RCIM_Cubic, leaveTangent=1, time=0.66, value=0.33}
      {arriveTangent=2.1868, interpMode=RCIM_Cubic, leaveTangent=2.1868, tangentMode=RCTM_User, time=1, value=1}
     ]}}
  }
 ]
}
