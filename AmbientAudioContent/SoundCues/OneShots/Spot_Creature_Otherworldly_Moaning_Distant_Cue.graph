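# Sound Cue for an outdoor ambient one-shot of distant, otherworldly creature moaning, played quietly through a shared one-shot attenuation asset.
# A Random node picks among four equally weighted waves, and a Modulator varies pitch and volume.
# All four quality-level inputs share that same chain.
# A concurrency override limits simultaneous plays to two, and a submix send targets the distant reverb.
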SoundCue /AmbientAudioContent/SoundCues/OneShots/Spot_Creature_Otherworldly_Moaning_Distant_Cue
root=QualityLevel_0
nodes [
 {id=Modulator_0, children [Random_0], properties {pitchMax=1.1, pitchMin=0.9, volumeMax=1.2, volumeMin=1}}
 {id=QualityLevel_0, children [Modulator_0,Modulator_0,Modulator_0,Modulator_0]}
 {id=Random_0, children [WavePlayer_0,WavePlayer_1,WavePlayer_2,WavePlayer_3], properties {weights [1,1,1,1]}}
 {id=WavePlayer_0, properties {soundWaveAssetPtr=/AmbientAudioContent/Waves/OneShots/Spot_Creature_Otherworldly_Moaning_Distant_01}}
 {id=WavePlayer_1, properties {soundWaveAssetPtr=/AmbientAudioContent/Waves/OneShots/Spot_Creature_Otherworldly_Moaning_Distant_02}}
 {id=WavePlayer_2, properties {soundWaveAssetPtr=/AmbientAudioContent/Waves/OneShots/Spot_Creature_Otherworldly_Moaning_Distant_03}}
 {id=WavePlayer_3, properties {soundWaveAssetPtr=/AmbientAudioContent/Waves/OneShots/Spot_Creature_Otherworldly_Moaning_Distant_04}}
]
properties {
 attenuationSettings=/Game/Sounds/Attenuation/Fort_TileAmbient_OneShot_Att_Closer
 bOverrideConcurrency=true
 maxDistance=8500
 soundClassObject=&Fort_Ambient_OneShots_Outdoors
 volumeMultiplier=0.1
 attenuationOverrides {bAttenuate=false}
 concurrencyOverrides {maxCount=2}
 soundSubmixSends [{
   maxSendDistance=8000
   minSendDistance=3000
   minSendLevel=0
   soundSubmix=&Fortnite_Reverb_Distant
   customSendLevelCurve {editorCurveData {keys [<at 0 0.25>,<at 0.6 0.3>,<at 1 0.66>]}}
  }]
}
