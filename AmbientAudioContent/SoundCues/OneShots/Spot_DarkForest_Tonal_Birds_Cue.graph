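# Sound Cue for tonal bird one-shots in dark forest ambience, using alien bird recordings.
# A Random node picks one of eight equally weighted waves, then a Modulator slightly varies volume and lowers pitch by a random amount.
# All four quality levels share that chain, and the cue adds a very low pitch multiplier and reduced volume.
# It uses wide one-shot attenuation, the outdoor ambient sound class and an exterior ambience source bus send.
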SoundCue /AmbientAudioContent/SoundCues/OneShots/Spot_DarkForest_Tonal_Birds_Cue
root=QualityLevel_0
nodes [
 {id=Modulator_1, children [Random_0], properties {pitchMax=1, pitchMin=0.8, volumeMax=1.1, volumeMin=0.9}}
 {id=QualityLevel_0, children [Modulator_1,Modulator_1,Modulator_1,Modulator_1]}
 {
  id=Random_0
  children [WavePlayer_7, WavePlayer_9, WavePlayer_10, WavePlayer_11, WavePlayer_14, WavePlayer_15, WavePlayer_16, WavePlayer_17]
  properties {weights [1,1,1,1,1,1,1,1]}
 }
 {id=WavePlayer_10, properties {soundWaveAssetPtr=/AmbientAudioContent/Waves/OneShots/Spot_Biome_Reality_Forest_SS_AlienBird_03}}
 {id=WavePlayer_11, properties {soundWaveAssetPtr=/AmbientAudioContent/Waves/OneShots/Spot_Biome_Reality_Forest_SS_AlienBird_04}}
 {id=WavePlayer_14, properties {soundWaveAssetPtr=/AmbientAudioContent/Waves/OneShots/Spot_Biome_Reality_Forest_SS_AlienBird_07}}
 {id=WavePlayer_15, properties {soundWaveAssetPtr=/AmbientAudioContent/Waves/OneShots/Spot_Biome_Reality_Forest_SS_AlienBird_08}}
 {id=WavePlayer_16, properties {soundWaveAssetPtr=/AmbientAudioContent/Waves/OneShots/Spot_Biome_Reality_Forest_SS_AlienBird_09}}
 {id=WavePlayer_17, properties {soundWaveAssetPtr=/AmbientAudioContent/Waves/OneShots/Spot_Biome_Reality_Forest_SS_AlienBird_10}}
 {id=WavePlayer_7, properties {soundWaveAssetPtr=/AmbientAudioContent/Waves/OneShots/Spot_Biome_Reality_Forest_SS_AlienBird_01}}
 {id=WavePlayer_9, properties {soundWaveAssetPtr=/AmbientAudioContent/Waves/OneShots/Spot_Biome_Reality_Forest_SS_AlienBird_02}}
]
properties {
 attenuationSettings=/Game/Sounds/Attenuation/Fort_TileAmbient_OneShot_Wider_Att
 maxDistance=8000
 pitchMultiplier=0.01
 soundClassObject=&Fort_Ambient_OneShots_Outdoors
 volumeMultiplier=0.376
 busSends [{audioBus=null, minSendLevel=0, sendLevel=1, soundSourceBus=&AmbientAudio_Exterior_SourceBus}]
}
