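# Sound cue for a spot one-shot ambient desert cricket that plays short bursts of chirps outdoors, with insect attenuation and an exterior ambient bus send.
# Each burst randomly picks one of three finite loop counts, repeating a short random delay, a modulator and a random single-chirp wave.
# A quality-level node gives the first two levels ten chirp waves and the last two only four, plus an empty slot.
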SoundCue /AmbientAudioContent/SoundCues/OneShots/Spot_Desert_Cricket_Chirps_Integer_Looped_Cue
root=QualityLevel_0
nodes [
 {id=Delay_12, children [Modulator_3], properties {delayMax=0.075, delayMin=0.025}}
 {id=Delay_13, children [Modulator_4], properties {delayMax=0.075, delayMin=0.025}}
 {id=Delay_14, children [Modulator_5], properties {delayMax=0.075, delayMin=0.025}}
 {id=Delay_17, children [Modulator_8], properties {delayMax=0.075, delayMin=0.025}}
 {id=Delay_18, children [Modulator_9], properties {delayMax=0.075, delayMin=0.025}}
 {id=Delay_19, children [Modulator_10], properties {delayMax=0.075, delayMin=0.025}}
 {id=Looping_11, children [Delay_17], properties {bLoopIndefinitely=false, loopCount=8}}
 {id=Looping_12, children [Delay_18], properties {bLoopIndefinitely=false, loopCount=6}}
 {id=Looping_13, children [Delay_19], properties {bLoopIndefinitely=false, loopCount=4}}
 {id=Looping_6, children [Delay_12], properties {bLoopIndefinitely=false, loopCount=8}}
 {id=Looping_7, children [Delay_13], properties {bLoopIndefinitely=false, loopCount=6}}
 {id=Looping_8, children [Delay_14], properties {bLoopIndefinitely=false, loopCount=4}}
 {id=Modulator_10, children [Random_4]}
 {id=Modulator_3, children [Random_0]}
 {id=Modulator_4, children [Random_0]}
 {id=Modulator_5, children [Random_0]}
 {id=Modulator_8, children [Random_4]}
 {id=Modulator_9, children [Random_4]}
 {id=QualityLevel_0, children [Random_9,Random_9,Random_5,Random_5]}
 {
  id=Random_0
  children [WavePlayer_0,WavePlayer_1,WavePlayer_2,WavePlayer_3,WavePlayer_4,WavePlayer_5,WavePlayer_6,WavePlayer_7,WavePlayer_8,WavePlayer_9]
  properties {weights [1,1,1,1,1,1,1,1,1,1]}
 }
 {id=Random_4, children [WavePlayer_29, WavePlayer_30, WavePlayer_31, WavePlayer_32, null], properties {weights [1,1,1,1,1]}}
 {id=Random_5, children [Looping_11,Looping_12,Looping_13], properties {weights [1,1,1]}}
 {id=Random_9, children [Looping_6,Looping_7,Looping_8], properties {weights [1,1,1]}}
 {id=WavePlayer_0, properties {soundWaveAssetPtr=/AmbientAudioContent/Waves/OneShots/Spot_Desert_Cricket_Chirp_Single_10}}
 {id=WavePlayer_1, properties {soundWaveAssetPtr=/AmbientAudioContent/Waves/OneShots/Spot_Desert_Cricket_Chirp_Single_09}}
 {id=WavePlayer_2, properties {soundWaveAssetPtr=/AmbientAudioContent/Waves/OneShots/Spot_Desert_Cricket_Chirp_Single_08}}
 {id=WavePlayer_29, properties {soundWaveAssetPtr=/AmbientAudioContent/Waves/OneShots/Spot_Desert_Cricket_Chirp_Single_01}}
 {id=WavePlayer_3, properties {soundWaveAssetPtr=/AmbientAudioContent/Waves/OneShots/Spot_Desert_Cricket_Chirp_Single_07}}
 {id=WavePlayer_30, properties {soundWaveAssetPtr=/AmbientAudioContent/Waves/OneShots/Spot_Desert_Cricket_Chirp_Single_02}}
 {id=WavePlayer_31, properties {soundWaveAssetPtr=/AmbientAudioContent/Waves/OneShots/Spot_Desert_Cricket_Chirp_Single_03}}
 {id=WavePlayer_32, properties {soundWaveAssetPtr=/AmbientAudioContent/Waves/OneShots/Spot_Desert_Cricket_Chirp_Single_04}}
 {id=WavePlayer_4, properties {soundWaveAssetPtr=/AmbientAudioContent/Waves/OneShots/Spot_Desert_Cricket_Chirp_Single_06}}
 {id=WavePlayer_5, properties {soundWaveAssetPtr=/AmbientAudioContent/Waves/OneShots/Spot_Desert_Cricket_Chirp_Single_05}}
 {id=WavePlayer_6, properties {soundWaveAssetPtr=&Spot_Desert_Cricket_Chirp_Single_04}}
 {id=WavePlayer_7, properties {soundWaveAssetPtr=&Spot_Desert_Cricket_Chirp_Single_03}}
 {id=WavePlayer_8, properties {soundWaveAssetPtr=&Spot_Desert_Cricket_Chirp_Single_02}}
 {id=WavePlayer_9, properties {soundWaveAssetPtr=&Spot_Desert_Cricket_Chirp_Single_01}}
]
properties {
 attenuationSettings=/Game/Sounds/Attenuation/Fort_TileAmbient_OneShot_Insect_By_Att
 maxDistance=6000
 soundClassObject=&Fort_Ambient_OneShots_Outdoors
 volumeMultiplier=0.075
 busSends [{audioBus=null, minSendLevel=1, sendLevel=1, soundSourceBus=&AmbientAudio_Exterior_SourceBus}]
}
