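# Sound cue for one-shot desert cricket multi-chirp ambience, classed as an outdoor ambient one-shot and played quietly with insect attenuation.
# A quality-level node sends its first two slots to an equal-weight random pick of eight chirp waves, and its last two to a pick of only the first four.
# Each random pick passes through its own modulator at default settings.
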SoundCue /AmbientAudioContent/SoundCues/OneShots/Spot_Desert_Cricket_Chirps_Multi_Cue
root=QualityLevel_0
nodes [
 {id=Modulator_12, children [Random_1]}
 {id=Modulator_2, children [Random_2]}
 {id=QualityLevel_0, children [Modulator_2,Modulator_2,Modulator_12,Modulator_12]}
 {id=Random_1, children [WavePlayer_8, WavePlayer_9, WavePlayer_10, WavePlayer_11], properties {weights [1,1,1,1]}}
 {
  id=Random_2
  children [WavePlayer_12, WavePlayer_13, WavePlayer_14, WavePlayer_15, WavePlayer_16, WavePlayer_17, WavePlayer_18, WavePlayer_19]
  properties {weights [1,1,1,1,1,1,1,1]}
 }
 {id=WavePlayer_10, properties {soundWaveAssetPtr=/AmbientAudioContent/Waves/OneShots/Spot_Desert_Cricket_Chirp_Multi_02}}
 {id=WavePlayer_11, properties {soundWaveAssetPtr=/AmbientAudioContent/Waves/OneShots/Spot_Desert_Cricket_Chirp_Multi_04}}
 {id=WavePlayer_12, properties {soundWaveAssetPtr=/AmbientAudioContent/Waves/OneShots/Spot_Desert_Cricket_Chirp_Multi_01}}
 {id=WavePlayer_13, properties {soundWaveAssetPtr=&Spot_Desert_Cricket_Chirp_Multi_02}}
 {id=WavePlayer_14, properties {soundWaveAssetPtr=/AmbientAudioContent/Waves/OneShots/Spot_Desert_Cricket_Chirp_Multi_03}}
 {id=WavePlayer_15, properties {soundWaveAssetPtr=&Spot_Desert_Cricket_Chirp_Multi_04}}
 {id=WavePlayer_16, properties {soundWaveAssetPtr=/AmbientAudioContent/Waves/OneShots/Spot_Desert_Cricket_Chirp_Multi_05}}
 {id=WavePlayer_17, properties {soundWaveAssetPtr=/AmbientAudioContent/Waves/OneShots/Spot_Desert_Cricket_Chirp_Multi_06}}
 {id=WavePlayer_18, properties {soundWaveAssetPtr=/AmbientAudioContent/Waves/OneShots/Spot_Desert_Cricket_Chirp_Multi_07}}
 {id=WavePlayer_19, properties {soundWaveAssetPtr=/AmbientAudioContent/Waves/OneShots/Spot_Desert_Cricket_Chirp_Multi_08}}
 {id=WavePlayer_8, properties {soundWaveAssetPtr=&Spot_Desert_Cricket_Chirp_Multi_01}}
 {id=WavePlayer_9, properties {soundWaveAssetPtr=&Spot_Desert_Cricket_Chirp_Multi_03}}
]
properties {
 attenuationSettings=/Game/Sounds/Attenuation/Fort_TileAmbient_OneShot_Insect_By_Att
 bEnableBusSends=false
 maxDistance=6000
 soundClassObject=&Fort_Ambient_OneShots_Outdoors
 volumeMultiplier=0.15
 preEffectBusSends [{}]
}
