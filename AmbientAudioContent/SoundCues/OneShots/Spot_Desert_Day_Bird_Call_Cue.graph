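# Sound cue for quiet, attenuated one-shot desert daytime bird calls in outdoor ambience.
# The first three quality slots weight a single modulated call over a mix that adds a quieter, lower-pitched, randomly delayed second call, like an answering bird.
# Both paths share a random pool of six waves with pitch and volume variation.
# The fourth quality slot plays one of only three waves with wider pitch variation.
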SoundCue /AmbientAudioContent/SoundCues/OneShots/Spot_Desert_Day_Bird_Call_Cue
root=QualityLevel_0
nodes [
 {id=Delay_0, children [Modulator_0], properties {delayMax=1.25, delayMin=0.6}}
 {id=Mixer_0, children [Modulator_1,Delay_0], properties {inputVolume [1,0.75]}}
 {id=Modulator_0, children [Random_4], properties {pitchMax=1, pitchMin=0.8, volumeMax=1.1, volumeMin=1}}
 {id=Modulator_1, children [Random_4], properties {pitchMax=1.1, volumeMax=1.1, volumeMin=0.9}}
 {id=Modulator_2, children [Random_1], properties {pitchMax=1.1, pitchMin=0.5, volumeMax=1.1, volumeMin=0.9}}
 {id=QualityLevel_0, children [Random_3,Random_3,Random_3,Modulator_2]}
 {id=Random_1, children [WavePlayer_8,WavePlayer_7,WavePlayer_6], properties {weights [1,1,1]}}
 {id=Random_3, children [Modulator_1,Mixer_0], properties {bRandomizeWithoutReplacement=false, weights [1,0.5]}}
 {
  id=Random_4
  children [WavePlayer_10, WavePlayer_15, WavePlayer_16, WavePlayer_17, WavePlayer_18, WavePlayer_19]
  properties {weights [1,1,1,1,1,1]}
 }
 {id=WavePlayer_10, properties {soundWaveAssetPtr=/AmbientAudioContent/Waves/OneShots/Spot_Desert_Day_Bird_Call_01}}
 {id=WavePlayer_15, properties {soundWaveAssetPtr=/AmbientAudioContent/Waves/OneShots/Spot_Desert_Day_Bird_Call_02}}
 {id=WavePlayer_16, properties {soundWaveAssetPtr=/AmbientAudioContent/Waves/OneShots/Spot_Desert_Day_Bird_Call_03}}
 {id=WavePlayer_17, properties {soundWaveAssetPtr=/AmbientAudioContent/Waves/OneShots/Spot_Desert_Day_Bird_Call_04}}
 {id=WavePlayer_18, properties {soundWaveAssetPtr=/AmbientAudioContent/Waves/OneShots/Spot_Desert_Day_Bird_Call_05}}
 {id=WavePlayer_19, properties {soundWaveAssetPtr=/AmbientAudioContent/Waves/OneShots/Spot_Desert_Day_Bird_Call_06}}
 {id=WavePlayer_6, properties {soundWaveAssetPtr=&Spot_Desert_Day_Bird_Call_04}}
 {id=WavePlayer_7, properties {soundWaveAssetPtr=&Spot_Desert_Day_Bird_Call_02}}
 {id=WavePlayer_8, properties {soundWaveAssetPtr=&Spot_Desert_Day_Bird_Call_01}}
]
properties {
 attenuationSettings=&Fort_TileAmbient_OneShot_Bird_Att
 bEnableBusSends=false
 maxDistance=8005
 soundClassObject=&Fort_Ambient_OneShots_Outdoors
 volumeMultiplier=0.15
 preEffectBusSends [{}]
}
