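# One-shot Sound Cue for Gila woodpecker calls in outdoor desert daytime ambience, using bird attenuation and the outdoor one-shot sound class.
# A quality-level node gives the first two levels a random pick of six calls and the last two a reduced set of three.
# Each branch passes through a Modulator for slight pitch and volume variation.
# Output also sends to an exterior ambient source bus and the default reverb submix.
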SoundCue /AmbientAudioContent/SoundCues/OneShots/Spot_Desert_Day_Gila_Woodpecker_Call_Cue
root=QualityLevel_0
nodes [
 {id=Modulator_1, children [Random_2], properties {pitchMax=1.1, volumeMax=1.1, volumeMin=0.9}}
 {id=Modulator_2, children [Random_3], properties {pitchMax=1.1, volumeMax=1.1, volumeMin=0.9}}
 {id=QualityLevel_0, children [Modulator_1,Modulator_1,Modulator_2,Modulator_2]}
 {
  id=Random_2
  children [WavePlayer_9, WavePlayer_10, WavePlayer_11, WavePlayer_12, WavePlayer_13, WavePlayer_14]
  properties {weights [1,1,1,1,1,1]}
 }
 {id=Random_3, children [WavePlayer_15, WavePlayer_16, WavePlayer_17], properties {weights [1,1,1]}}
 {id=WavePlayer_10, properties {soundWaveAssetPtr=/AmbientAudioContent/Waves/OneShots/Spot_Desert_Day_Gila_Woodpecker_Call_06}}
 {id=WavePlayer_11, properties {soundWaveAssetPtr=/AmbientAudioContent/Waves/OneShots/Spot_Desert_Day_Gila_Woodpecker_Call_02}}
 {id=WavePlayer_12, properties {soundWaveAssetPtr=/AmbientAudioContent/Waves/OneShots/Spot_Desert_Day_Gila_Woodpecker_Call_03}}
 {id=WavePlayer_13, properties {soundWaveAssetPtr=/AmbientAudioContent/Waves/OneShots/Spot_Desert_Day_Gila_Woodpecker_Call_04}}
 {id=WavePlayer_14, properties {soundWaveAssetPtr=/AmbientAudioContent/Waves/OneShots/Spot_Desert_Day_Gila_Woodpecker_Call_05}}
 {id=WavePlayer_15, properties {soundWaveAssetPtr=/AmbientAudioContent/Waves/OneShots/Spot_Desert_Day_Gila_Woodpecker_Call_01}}
 {id=WavePlayer_16, properties {soundWaveAssetPtr=&Spot_Desert_Day_Gila_Woodpecker_Call_02}}
 {id=WavePlayer_17, properties {soundWaveAssetPtr=&Spot_Desert_Day_Gila_Woodpecker_Call_03}}
 {id=WavePlayer_9, properties {soundWaveAssetPtr=&Spot_Desert_Day_Gila_Woodpecker_Call_01}}
]
properties {
 attenuationSettings=&Fort_TileAmbient_OneShot_Bird_Att
 maxDistance=8005
 soundClassObject=&Fort_Ambient_OneShots_Outdoors
 volumeMultiplier=0.165
 busSends [{
   audioBus=null
   maxSendDistance=6000
   minSendDistance=3000
   minSendLevel=0.6
   sendLevel=1
   soundSourceBus=&AmbientAudio_Exterior_SourceBus
  }]
 soundSubmixSends [{minSendLevel=0.6, soundSubmix=&Fortnite_Reverb_Default}]
}
