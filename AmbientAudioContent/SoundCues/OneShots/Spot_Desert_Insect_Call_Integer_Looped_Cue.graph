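# Sound cue for an outdoor ambient one-shot desert insect call, played as a short burst of repeats at low volume with insect-specific attenuation.
# A random node picks one of three finite loop counts, and each repeat passes through a brief random delay, pitch modulation and random wave choice.
# The longest burst uses a raised pitch range.
# A quality switch gives the first two levels six waves and the last two only three.
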SoundCue /AmbientAudioContent/SoundCues/OneShots/Spot_Desert_Insect_Call_Integer_Looped_Cue
root=QualityLevel_0
nodes [
 {id=Delay_12, children [Modulator_3], properties {delayMax=0.075, delayMin=0.025}}
 {id=Delay_15, children [Modulator_6], properties {delayMax=0.075, delayMin=0.025}}
 {id=Delay_3, children [Modulator_13], properties {delayMax=0.075, delayMin=0.025}}
 {id=Delay_4, children [Modulator_14], properties {delayMax=0.075, delayMin=0.025}}
 {id=Delay_5, children [Modulator_15], properties {delayMax=0.075, delayMin=0.025}}
 {id=Delay_7, children [Modulator_0], properties {delayMax=0.075, delayMin=0.025}}
 {id=Looping_15, children [Delay_5], properties {bLoopIndefinitely=false, loopCount=2}}
 {id=Looping_3, children [Delay_3], properties {bLoopIndefinitely=false, loopCount=6}}
 {id=Looping_4, children [Delay_4], properties {bLoopIndefinitely=false, loopCount=4}}
 {id=Looping_5, children [Delay_7], properties {bLoopIndefinitely=false, loopCount=6}}
 {id=Looping_6, children [Delay_12], properties {bLoopIndefinitely=false, loopCount=4}}
 {id=Looping_9, children [Delay_15], properties {bLoopIndefinitely=false, loopCount=2}}
 {id=Modulator_0, children [Random_3], properties {pitchMax=1.1, pitchMin=1.05}}
 {id=Modulator_13, children [Random_0], properties {pitchMax=1.1, pitchMin=1.05}}
 {id=Modulator_14, children [Random_0], properties {pitchMin=1}}
 {id=Modulator_15, children [Random_0], properties {pitchMin=1}}
 {id=Modulator_3, children [Random_3], properties {pitchMin=1}}
 {id=Modulator_6, children [Random_3], properties {pitchMin=1}}
 {id=QualityLevel_0, children [Random_9,Random_9,Random_2,Random_2]}
 {id=Random_0, children [WavePlayer_0,WavePlayer_1,WavePlayer_2], properties {weights [1,1,1]}}
 {id=Random_2, children [Looping_3,Looping_4,Looping_15], properties {weights [1,1,1]}}
 {
  id=Random_3
  children [WavePlayer_9, WavePlayer_10, WavePlayer_11, WavePlayer_12, WavePlayer_13, WavePlayer_14]
  properties {weights [1,1,1,1,1,1]}
 }
 {id=Random_9, children [Looping_5,Looping_6,Looping_9], properties {weights [1,1,1]}}
 {id=WavePlayer_0, properties {soundWaveAssetPtr=/AmbientAudioContent/Waves/OneShots/Spot_Desert_Insect_Call_01}}
 {id=WavePlayer_1, properties {soundWaveAssetPtr=/AmbientAudioContent/Waves/OneShots/Spot_Desert_Insect_Call_02}}
 {id=WavePlayer_10, properties {soundWaveAssetPtr=&Spot_Desert_Insect_Call_02}}
 {id=WavePlayer_11, properties {soundWaveAssetPtr=/AmbientAudioContent/Waves/OneShots/Spot_Desert_Insect_Call_03}}
 {id=WavePlayer_12, properties {soundWaveAssetPtr=/AmbientAudioContent/Waves/OneShots/Spot_Desert_Insect_Call_04}}
 {id=WavePlayer_13, properties {soundWaveAssetPtr=/AmbientAudioContent/Waves/OneShots/Spot_Desert_Insect_Call_05}}
 {id=WavePlayer_14, properties {soundWaveAssetPtr=/AmbientAudioContent/Waves/OneShots/Spot_Desert_Insect_Call_06}}
 {id=WavePlayer_2, properties {soundWaveAssetPtr=&Spot_Desert_Insect_Call_03}}
 {id=WavePlayer_9, properties {soundWaveAssetPtr=&Spot_Desert_Insect_Call_01}}
]
properties {
 attenuationSettings=/Game/Sounds/Attenuation/Fort_TileAmbient_OneShot_Insect_By_Att
 bEnableBusSends=false
 maxDistance=6000
 soundClassObject=&Fort_Ambient_OneShots_Outdoors
 volumeMultiplier=0.1
 preEffectBusSends [{}]
}
