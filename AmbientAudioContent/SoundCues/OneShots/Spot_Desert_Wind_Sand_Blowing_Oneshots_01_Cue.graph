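# Sound Cue that plays a desert wind and blowing sand ambient one-shot for outdoor spot emitters.
# A Random node picks one of five equally weighted wave variations, and a Modulator adds slight volume variation and upward-only pitch variation.
# All four Quality Level inputs share that same chain.
# The cue plays at reduced volume with a wind one-shot attenuation and the outdoor ambient one-shots sound class.
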SoundCue /AmbientAudioContent/SoundCues/OneShots/Spot_Desert_Wind_Sand_Blowing_Oneshots_01_Cue
root=QualityLevel_0
nodes [
 {id=Modulator_1, children [Random_1], properties {pitchMax=1.1, pitchMin=1, volumeMax=1.1, volumeMin=0.9}}
 {id=QualityLevel_0, children [Modulator_1,Modulator_1,Modulator_1,Modulator_1]}
 {id=Random_1, children [WavePlayer_0,WavePlayer_1,WavePlayer_6,WavePlayer_7,WavePlayer_9], properties {weights [1,1,1,1,1]}}
 {id=WavePlayer_0, properties {soundWaveAssetPtr=/AmbientAudioContent/Waves/OneShots/Spot_Desert_Wind_Sand_Emitter_Oneshot_01}}
 {id=WavePlayer_1, properties {soundWaveAssetPtr=/AmbientAudioContent/Waves/OneShots/Spot_Desert_Wind_Sand_Emitter_Oneshot_02}}
 {id=WavePlayer_6, properties {soundWaveAssetPtr=/AmbientAudioContent/Waves/OneShots/Spot_Desert_Wind_Sand_Emitter_Oneshot_03}}
 {id=WavePlayer_7, properties {soundWaveAssetPtr=/AmbientAudioContent/Waves/OneShots/Spot_Desert_Wind_Sand_Emitter_Oneshot_04}}
 {id=WavePlayer_9, properties {soundWaveAssetPtr=/AmbientAudioContent/Waves/OneShots/Spot_Desert_Wind_Sand_Emitter_Oneshot_05}}
]
properties {
 attenuationSettings=/Game/Sounds/Attenuation/Fort_TileAmbient_OneShot_Wind_Att
 bEnableBusSends=false
 maxDistance=8000
 soundClassObject=&Fort_Ambient_OneShots_Outdoors
 volumeMultiplier=0.2
 preEffectBusSends [{}]
}
