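# Sound cue for one-shot desert wind and sand gust ambience, using an outdoor ambient sound class and one-shot wind attenuation.
# A quality-level root sends the first two levels to a weighted random choice that favours ten slow sand gusts over five wind gusts.
# The last two levels use a reduced random pool of three of those wind gusts.
# Each pool passes through its own modulator for playback variation.
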SoundCue /AmbientAudioContent/SoundCues/OneShots/Spot_Desert_Wind_Sand_Gust_Cue
root=QualityLevel_0
nodes [
 {id=Modulator_0, children [Random_4], properties {volumeMax=2.1, volumeMin=2}}
 {id=Modulator_1, children [Random_5], properties {pitchMax=1.1, volumeMax=2.2, volumeMin=2}}
 {id=Modulator_2, children [Random_6], properties {pitchMax=1.1, volumeMax=2.2, volumeMin=2}}
 {id=QualityLevel_0, children [Random_3,Random_3,Modulator_2,Modulator_2]}
 {id=Random_3, children [Modulator_0,Modulator_1], properties {weights [1,0.5]}}
 {
  id=Random_4
  children [
   WavePlayer_18
   WavePlayer_19
   WavePlayer_20
   WavePlayer_21
   WavePlayer_22
   WavePlayer_23
   WavePlayer_24
   WavePlayer_25
   WavePlayer_26
   WavePlayer_27
  ]
  properties {weights [1,1,1,1,1,1,1,1,1,1]}
 }
 {id=Random_5, children [WavePlayer_28, WavePlayer_29, WavePlayer_30, WavePlayer_31, WavePlayer_32], properties {weights [1,1,1,1,1]}}
 {id=Random_6, children [WavePlayer_33, WavePlayer_34, WavePlayer_35], properties {weights [1,1,1]}}
 {id=WavePlayer_18, properties {soundWaveAssetPtr=/AmbientAudioContent/Waves/OneShots/Spot_Desert_Slow_Sand_Gust_01}}
 {id=WavePlayer_19, properties {soundWaveAssetPtr=/AmbientAudioContent/Waves/OneShots/Spot_Desert_Slow_Sand_Gust_02}}
 {id=WavePlayer_20, properties {soundWaveAssetPtr=/AmbientAudioContent/Waves/OneShots/Spot_Desert_Slow_Sand_Gust_03}}
 {id=WavePlayer_21, properties {soundWaveAssetPtr=/AmbientAudioContent/Waves/OneShots/Spot_Desert_Slow_Sand_Gust_04}}
 {id=WavePlayer_22, properties {soundWaveAssetPtr=/AmbientAudioContent/Waves/OneShots/Spot_Desert_Slow_Sand_Gust_05}}
 {id=WavePlayer_23, properties {soundWaveAssetPtr=/AmbientAudioContent/Waves/OneShots/Spot_Desert_Slow_Sand_Gust_06}}
 {id=WavePlayer_24, properties {soundWaveAssetPtr=/AmbientAudioContent/Waves/OneShots/Spot_Desert_Slow_Sand_Gust_07}}
 {id=WavePlayer_25, properties {soundWaveAssetPtr=/AmbientAudioContent/Waves/OneShots/Spot_Desert_Slow_Sand_Gust_08}}
 {id=WavePlayer_26, properties {soundWaveAssetPtr=/AmbientAudioContent/Waves/OneShots/Spot_Desert_Slow_Sand_Gust_09}}
 {id=WavePlayer_27, properties {soundWaveAssetPtr=/AmbientAudioContent/Waves/OneShots/Spot_Desert_Slow_Sand_Gust_10}}
 {id=WavePlayer_28, properties {soundWaveAssetPtr=/AmbientAudioContent/Waves/OneShots/Spot_Desert_Wind_Sand_Gust_05}}
 {id=WavePlayer_29, properties {soundWaveAssetPtr=/AmbientAudioContent/Waves/OneShots/Spot_Desert_Wind_Sand_Gust_04}}
 {id=WavePlayer_30, properties {soundWaveAssetPtr=/AmbientAudioContent/Waves/OneShots/Spot_Desert_Wind_Sand_Gust_03}}
 {id=WavePlayer_31, properties {soundWaveAssetPtr=/AmbientAudioContent/Waves/OneShots/Spot_Desert_Wind_Sand_Gust_02}}
 {id=WavePlayer_32, properties {soundWaveAssetPtr=/AmbientAudioContent/Waves/OneShots/Spot_Desert_Wind_Sand_Gust_01}}
 {id=WavePlayer_33, properties {soundWaveAssetPtr=&Spot_Desert_Wind_Sand_Gust_01}}
 {id=WavePlayer_34, properties {soundWaveAssetPtr=&Spot_Desert_Wind_Sand_Gust_02}}
 {id=WavePlayer_35, properties {soundWaveAssetPtr=&Spot_Desert_Wind_Sand_Gust_05}}
]
properties {
 attenuationSettings=/Game/Sounds/Attenuation/Fort_TileAmbient_OneShot_Wind_Att
 bEnableBusSends=false
 maxDistance=8000
 soundClassObject=&Fort_Ambient_OneShots_Outdoors
 volumeMultiplier=0.3
 preEffectBusSends [{}]
}
