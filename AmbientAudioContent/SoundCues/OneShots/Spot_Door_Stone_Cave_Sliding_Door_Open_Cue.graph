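# Sound cue for a one-shot spot sound of a huge stone cave sliding door opening.
# A single wave player feeds the root, with no randomization or modulation.
# It uses the Fort_Doors sound class, a large-door attenuation asset with occlusion naming, and a send to the default Fortnite reverb submix.
# A concurrency override allows one instance at a time and prevents new plays while one is active.
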SoundCue /AmbientAudioContent/SoundCues/OneShots/Spot_Door_Stone_Cave_Sliding_Door_Open_Cue
root=WavePlayer_1
nodes [{id=WavePlayer_1, properties {soundWaveAssetPtr=/AmbientAudioContent/Waves/OneShots/Spot_Door_Huge_Stone_Sliding_Cave_Open}}]
properties {
 attenuationSettings=/AmbientAudioBase/Attenuations/Fort_Doors_Large_Occ_Att
 bOverrideConcurrency=true
 maxDistance=9000
 soundClassObject=&Fort_Doors
 concurrencyOverrides {maxCount=1, resolutionRule=PreventNew}
 soundSubmixSends [{minSendLevel=0.4, soundSubmix=&Fortnite_Reverb_Default}]
}
