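# Sound cue for a one-shot ambient spot sound of elevator cable stress, named for an elevator interior.
# A Random node picks with equal weight among nine elevator shaft cable foley waves.
# A Modulator at the root varies each play with slightly lowered random pitch and volume.
# The cue uses the ambient one-shot sound class, a shared tile-ambient one-shot attenuation asset and a reduced overall volume multiplier.
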SoundCue /AmbientAudioContent/SoundCues/OneShots/Spot_Elevator_Interior_Cable_Stress_Cue
root=Modulator_0
nodes [
 {id=Modulator_0, children [Random_1], properties {pitchMax=1, pitchMin=0.8, volumeMax=1, volumeMin=0.7}}
 {
  id=Random_1
  children [WavePlayer_9, WavePlayer_10, WavePlayer_11, WavePlayer_12, WavePlayer_13, WavePlayer_14, WavePlayer_15, WavePlayer_16, WavePlayer_17]
  properties {weights [1,1,1,1,1,1,1,1,1]}
 }
 {id=WavePlayer_10, properties {soundWaveAssetPtr=/AmbientAudioContent/Waves/OneShots/Spot_Evelator_Shaft_Cable_Foley_02}}
 {id=WavePlayer_11, properties {soundWaveAssetPtr=/AmbientAudioContent/Waves/OneShots/Spot_Evelator_Shaft_Cable_Foley_03}}
 {id=WavePlayer_12, properties {soundWaveAssetPtr=/AmbientAudioContent/Waves/OneShots/Spot_Evelator_Shaft_Cable_Foley_04}}
 {id=WavePlayer_13, properties {soundWaveAssetPtr=/AmbientAudioContent/Waves/OneShots/Spot_Evelator_Shaft_Cable_Foley_05}}
 {id=WavePlayer_14, properties {soundWaveAssetPtr=/AmbientAudioContent/Waves/OneShots/Spot_Evelator_Shaft_Cable_Foley_06}}
 {id=WavePlayer_15, properties {soundWaveAssetPtr=/AmbientAudioContent/Waves/OneShots/Spot_Evelator_Shaft_Cable_Foley_07}}
 {id=WavePlayer_16, properties {soundWaveAssetPtr=/AmbientAudioContent/Waves/OneShots/Spot_Evelator_Shaft_Cable_Foley_08}}
 {id=WavePlayer_17, properties {soundWaveAssetPtr=/AmbientAudioContent/Waves/OneShots/Spot_Evelator_Shaft_Cable_Foley_09}}
 {id=WavePlayer_9, properties {soundWaveAssetPtr=/AmbientAudioContent/Waves/OneShots/Spot_Evelator_Shaft_Cable_Foley_01}}
]
properties {
 attenuationSettings=/Game/Sounds/Attenuation/Fort_TileAmbient_OneShot_Att
 maxDistance=8000
 soundClassObject=&Fort_Ambient_OneShots
 volumeMultiplier=0.451
}
