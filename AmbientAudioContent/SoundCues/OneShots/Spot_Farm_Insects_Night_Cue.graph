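# Sound Cue for night-time insect one-shots in the farm biome, part of the outdoor ambient audio.
# A Random node picks with equal weight among seven farm insect waves from two variant sets.
# A Modulator at the root then varies pitch and volume slightly on each play.
# Playback uses a shared tile-ambient one-shot attenuation and the Athena outdoor ambient sound class, with volume reduced.
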SoundCue /AmbientAudioContent/SoundCues/OneShots/Spot_Farm_Insects_Night_Cue
root=Modulator_0
nodes [
 {id=Modulator_0, children [Random_1], properties {pitchMax=1.2, pitchMin=0.8, volumeMax=1.2, volumeMin=0.8}}
 {
  id=Random_1
  children [WavePlayer_4, WavePlayer_5, WavePlayer_7, WavePlayer_10, WavePlayer_11, WavePlayer_12, WavePlayer_13]
  properties {weights [1,1,1,1,1,1,1]}
 }
 {id=WavePlayer_10, properties {soundWaveAssetPtr=/AmbientAudioContent/Waves/OneShots/Spot_Biome_Farm_InsectNight_04}}
 {id=WavePlayer_11, properties {soundWaveAssetPtr=/AmbientAudioContent/Waves/OneShots/Spot_Biome_Farm_InsectNightB_01}}
 {id=WavePlayer_12, properties {soundWaveAssetPtr=/AmbientAudioContent/Waves/OneShots/Spot_Biome_Farm_InsectNightB_02}}
 {id=WavePlayer_13, properties {soundWaveAssetPtr=/AmbientAudioContent/Waves/OneShots/Spot_Biome_Farm_InsectNightB_03}}
 {id=WavePlayer_4, properties {soundWaveAssetPtr=/AmbientAudioContent/Waves/OneShots/Spot_Biome_Farm_InsectNight_01}}
 {id=WavePlayer_5, properties {soundWaveAssetPtr=/AmbientAudioContent/Waves/OneShots/Spot_Biome_Farm_InsectNight_02}}
 {id=WavePlayer_7, properties {soundWaveAssetPtr=/AmbientAudioContent/Waves/OneShots/Spot_Biome_Farm_InsectNight_03}}
]
properties {
 attenuationSettings=/Game/Sounds/Attenuation/Fort_TileAmbient_OneShot_Att
 maxDistance=8000
 soundClassObject=&Fort_Ambient_Outdoors_Athena
 volumeMultiplier=0.451
 preEffectBusSends [{}]
}
