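# Sound cue for outdoor ambient firefly wing one-shots, built from dragonfly wing and fly-by recordings.
# A random node picks one of three pools, each randomly choosing a wave and varying pitch and volume; the small wing pool repeats waves from the larger one.
# The quality node feeds only its first two levels, leaving the others empty.
# It uses dragonfly attenuation, tight concurrency, exterior bus and reverb sends.
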SoundCue /AmbientAudioContent/SoundCues/OneShots/Spot_Firefly_Wings_Oneshots_Cue
root=QualityLevel_0
nodes [
 {id=Modulator_0, children [Random_5], properties {pitchMax=1, pitchMin=0.8, volumeMax=0.9, volumeMin=0.6}}
 {id=Modulator_2, children [Random_2], properties {pitchMax=1, pitchMin=0.8, volumeMax=0.9, volumeMin=0.6}}
 {id=Modulator_3, children [Random_3], properties {pitchMax=1.25, pitchMin=0.9, volumeMax=1.1, volumeMin=0.9}}
 {id=QualityLevel_0, children [Random_4,Random_4,null,null]}
 {
  id=Random_2
  children [WavePlayer_12, WavePlayer_13, WavePlayer_14, WavePlayer_15, WavePlayer_16, WavePlayer_37, WavePlayer_38, WavePlayer_39, WavePlayer_40]
  properties {weights [1,1,1,1,1,1,1,1,1]}
 }
 {
  id=Random_3
  children [
   WavePlayer_25
   WavePlayer_26
   WavePlayer_27
   WavePlayer_28
   WavePlayer_29
   WavePlayer_30
   WavePlayer_31
   WavePlayer_32
   WavePlayer_33
   WavePlayer_34
   WavePlayer_35
   WavePlayer_36
  ]
  properties {weights [1,1,1,1,1,1,1,1,1,1,1,1]}
 }
 {id=Random_4, children [Modulator_2,Modulator_0,Modulator_3], properties {weights [1,1,1]}}
 {id=Random_5, children [WavePlayer_47, WavePlayer_48, WavePlayer_49, WavePlayer_50], properties {weights [1,1,1,1]}}
 {id=WavePlayer_12, properties {soundWaveAssetPtr=/Game/Sounds/Helios/Environment_Actors/Dragonflies/Spot_DragonFly_Wings_By_A_01}}
 {id=WavePlayer_13, properties {soundWaveAssetPtr=/Game/Sounds/Helios/Environment_Actors/Dragonflies/Spot_DragonFly_Wings_By_A_02}}
 {id=WavePlayer_14, properties {soundWaveAssetPtr=/Game/Sounds/Helios/Environment_Actors/Dragonflies/Spot_DragonFly_Wings_By_A_03}}
 {id=WavePlayer_15, properties {soundWaveAssetPtr=/Game/Sounds/Helios/Environment_Actors/Dragonflies/Spot_DragonFly_Wings_By_B_01}}
 {id=WavePlayer_16, properties {soundWaveAssetPtr=/Game/Sounds/Helios/Environment_Actors/Dragonflies/Spot_DragonFly_Wings_By_B_02}}
 {id=WavePlayer_25, properties {soundWaveAssetPtr=/Game/Sounds/Helios/Environment_Actors/Dragonflies/Spot_DragonFly_By_B_01}}
 {id=WavePlayer_26, properties {soundWaveAssetPtr=/Game/Sounds/Helios/Environment_Actors/Dragonflies/Spot_DragonFly_By_B_02}}
 {id=WavePlayer_27, properties {soundWaveAssetPtr=/Game/Sounds/Helios/Environment_Actors/Dragonflies/Spot_DragonFly_By_B_03}}
 {id=WavePlayer_28, properties {soundWaveAssetPtr=/Game/Sounds/Helios/Environment_Actors/Dragonflies/Spot_DragonFly_By_B_04}}
 {id=WavePlayer_29, properties {soundWaveAssetPtr=/Game/Sounds/Helios/Environment_Actors/Dragonflies/Spot_DragonFly_By_B_05}}
 {id=WavePlayer_30, properties {soundWaveAssetPtr=/Game/Sounds/Helios/Environment_Actors/Dragonflies/Spot_DragonFly_By_B_06}}
 {id=WavePlayer_31, properties {soundWaveAssetPtr=/Game/Sounds/Helios/Environment_Actors/Dragonflies/Spot_DragonFly_By_B_07}}
 {id=WavePlayer_32, properties {soundWaveAssetPtr=/Game/Sounds/Helios/Environment_Actors/Dragonflies/Spot_DragonFly_By_B_08}}
 {id=WavePlayer_33, properties {soundWaveAssetPtr=/Game/Sounds/Helios/Environment_Actors/Dragonflies/Spot_DragonFly_By_B_09}}
 {id=WavePlayer_34, properties {soundWaveAssetPtr=/Game/Sounds/Helios/Environment_Actors/Dragonflies/Spot_DragonFly_By_B_10}}
 {id=WavePlayer_35, properties {soundWaveAssetPtr=/Game/Sounds/Helios/Environment_Actors/Dragonflies/Spot_DragonFly_By_B_11}}
 {id=WavePlayer_36, properties {soundWaveAssetPtr=/Game/Sounds/Helios/Environment_Actors/Dragonflies/Spot_DragonFly_By_B_12}}
 {id=WavePlayer_37, properties {soundWaveAssetPtr=/Game/Sounds/Helios/Environment_Actors/Dragonflies/Spot_DragonFly_Wings_By_B_03}}
 {id=WavePlayer_38, properties {soundWaveAssetPtr=/Game/Sounds/Helios/Environment_Actors/Dragonflies/Spot_DragonFly_Wings_By_C_01}}
 {id=WavePlayer_39, properties {soundWaveAssetPtr=/Game/Sounds/Helios/Environment_Actors/Dragonflies/Spot_DragonFly_Wings_By_C_02}}
 {id=WavePlayer_40, properties {soundWaveAssetPtr=/Game/Sounds/Helios/Environment_Actors/Dragonflies/Spot_DragonFly_Wings_By_C_03}}
 {id=WavePlayer_47, properties {soundWaveAssetPtr=&Spot_DragonFly_Wings_By_B_03}}
 {id=WavePlayer_48, properties {soundWaveAssetPtr=&Spot_DragonFly_Wings_By_C_01}}
 {id=WavePlayer_49, properties {soundWaveAssetPtr=&Spot_DragonFly_Wings_By_C_02}}
 {id=WavePlayer_50, properties {soundWaveAssetPtr=&Spot_DragonFly_Wings_By_C_03}}
]
properties {
 attenuationSettings=/AmbientAudioBase/Attenuations/Spot_Dragonflies_Oneshots_Att
 bOverrideConcurrency=true
 maxDistance=4500
 priority=0
 soundClassObject=&Fort_Ambient_OneShots_Outdoors
 volumeMultiplier=0.3
 busSends [{audioBus=null, minSendLevel=0.6, sendLevel=1, soundSourceBus=&AmbientAudio_Exterior_SourceBus}]
 concurrencyOverrides {maxCount=2, voiceStealReleaseTime=0.5}
 soundSubmixSends [{minSendLevel=1, soundSubmix=&Fortnite_Reverb_Default}]
}
