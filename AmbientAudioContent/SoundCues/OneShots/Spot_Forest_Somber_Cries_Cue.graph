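# One-shot ambient SoundCue for somber, eerie tonal cries heard as forest spot sounds.
# A Random node picks among five equally weighted cry waves and feeds a Modulator that randomizes pitch slightly below to above normal for variety.
# The cue plays at reduced volume through the Fort_Ambient_OneShots sound class and references the shared Fort_Ambient_Att attenuation asset.
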SoundCue /AmbientAudioContent/SoundCues/OneShots/Spot_Forest_Somber_Cries_Cue
root=Modulator_0
nodes [
 {id=Modulator_0, children [Random_1], properties {pitchMax=1.1, pitchMin=0.8}}
 {id=Random_1, children [WavePlayer_6, WavePlayer_7, WavePlayer_8, WavePlayer_9, WavePlayer_10], properties {weights [1,1,1,1,1]}}
 {id=WavePlayer_10, properties {soundWaveAssetPtr=/AmbientAudioContent/Waves/OneShots/Spot_Forest_Erie_Tonal_Cry_01}}
 {id=WavePlayer_6, properties {soundWaveAssetPtr=/AmbientAudioContent/Waves/OneShots/Spot_Forest_Erie_Tonal_Cry_05}}
 {id=WavePlayer_7, properties {soundWaveAssetPtr=/AmbientAudioContent/Waves/OneShots/Spot_Forest_Erie_Tonal_Cry_04}}
 {id=WavePlayer_8, properties {soundWaveAssetPtr=/AmbientAudioContent/Waves/OneShots/Spot_Forest_Erie_Tonal_Cry_03}}
 {id=WavePlayer_9, properties {soundWaveAssetPtr=/AmbientAudioContent/Waves/OneShots/Spot_Forest_Erie_Tonal_Cry_02}}
]
properties {
 attenuationSettings=/AmbientAudioBase/Attenuations/Fort_Ambient_Att
 maxDistance=2097152
 soundClassObject=&Fort_Ambient_OneShots
 volumeMultiplier=0.33
 attenuationOverrides {bAttenuate=false}
}
