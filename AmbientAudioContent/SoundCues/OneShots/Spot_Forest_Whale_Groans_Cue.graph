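# Sound cue for one-shot ambient forest spot sounds, playing eerie tonal "whale groan" waves.
# A Random node picks one of six equally weighted groan waves and feeds a Modulator at the root for pitch and volume variation.
# The cue plays at reduced volume in the Fort ambient one-shots sound class and references a shared "closer" tile-ambient one-shot attenuation asset.
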SoundCue /AmbientAudioContent/SoundCues/OneShots/Spot_Forest_Whale_Groans_Cue
root=Modulator_0
nodes [
 {id=Modulator_0, children [Random_0]}
 {id=Random_0, children [WavePlayer_0,WavePlayer_1,WavePlayer_2,WavePlayer_3,WavePlayer_4,WavePlayer_5], properties {weights [1,1,1,1,1,1]}}
 {id=WavePlayer_0, properties {soundWaveAssetPtr=/AmbientAudioContent/Waves/OneShots/Spot_Forest_Erie_Tonal_Groan_01}}
 {id=WavePlayer_1, properties {soundWaveAssetPtr=/AmbientAudioContent/Waves/OneShots/Spot_Forest_Erie_Tonal_Groan_02}}
 {id=WavePlayer_2, properties {soundWaveAssetPtr=/AmbientAudioContent/Waves/OneShots/Spot_Forest_Erie_Tonal_Groan_03}}
 {id=WavePlayer_3, properties {soundWaveAssetPtr=/AmbientAudioContent/Waves/OneShots/Spot_Forest_Erie_Tonal_Groan_04}}
 {id=WavePlayer_4, properties {soundWaveAssetPtr=/AmbientAudioContent/Waves/OneShots/Spot_Forest_Erie_Tonal_Groan_05}}
 {id=WavePlayer_5, properties {soundWaveAssetPtr=/AmbientAudioContent/Waves/OneShots/Spot_Forest_Erie_Tonal_Groan_06}}
]
properties {
 attenuationSettings=/Game/Sounds/Attenuation/Fort_TileAmbient_OneShot_Att_Closer
 maxDistance=8500
 soundClassObject=&Fort_Ambient_OneShots
 volumeMultiplier=0.5
 attenuationOverrides {bAttenuate=false}
}
