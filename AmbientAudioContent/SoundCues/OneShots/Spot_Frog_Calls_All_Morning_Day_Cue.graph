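# Sound cue for outdoor one-shot ambient frog calls, named for morning and day.
# All quality levels share one random pick among ten branches of frog and toad wave sets, mostly with per-branch volume or pitch modulation.
# The Fukui branch mixes a call with a briefly delayed, lower-pitched second pick from the same set.
# It uses a bird one-shot attenuation, two-voice concurrency and distance-based reverb sends.
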SoundCue /AmbientAudioContent/SoundCues/OneShots/Spot_Frog_Calls_All_Morning_Day_Cue
root=QualityLevel_0
nodes [
 {id=Delay_0, children [Modulator_12], properties {delayMax=0.33, delayMin=0.2}}
 {id=Mixer_0, children [Modulator_9,Delay_0], properties {inputVolume [1,1]}}
 {id=Modulator_0, children [Random_3], properties {volumeMax=1.75, volumeMin=1.5}}
 {id=Modulator_10, children [Random_10], properties {volumeMax=0.6, volumeMin=0.4}}
 {id=Modulator_11, children [Random_11], properties {volumeMax=0.3, volumeMin=0.2}}
 {id=Modulator_12, children [Random_9], properties {pitchMax=0.9, pitchMin=0.7}}
 {id=Modulator_13, children [Random_12], properties {pitchMax=0.9, pitchMin=0.7}}
 {id=Modulator_3, children [Random_1], properties {pitchMax=0.9, pitchMin=0.7, volumeMax=0.8, volumeMin=0.7}}
 {id=Modulator_4, children [Random_2], properties {pitchMax=0.9, pitchMin=0.8, volumeMax=0.8, volumeMin=0.6}}
 {id=Modulator_5, children [Random_5]}
 {id=Modulator_6, children [Random_7], properties {volumeMax=1.2, volumeMin=1}}
 {id=Modulator_7, children [Random_8], properties {pitchMax=0.9, pitchMin=0.7, volumeMax=0.9, volumeMin=0.7}}
 {id=Modulator_9, children [Random_9]}
 {id=QualityLevel_0, children [Random_6,Random_6,Random_6,Random_6]}
 {
  id=Random_1
  children [WavePlayer_7, WavePlayer_8, WavePlayer_9, WavePlayer_10, WavePlayer_11, WavePlayer_12]
  properties {weights [1,1,1,1,1,1]}
 }
 {
  id=Random_10
  children [WavePlayer_48, WavePlayer_49, WavePlayer_50, WavePlayer_51, WavePlayer_52, WavePlayer_53, WavePlayer_54]
  properties {weights [1,1,1,1,1,1,1]}
 }
 {id=Random_11, children [WavePlayer_55, WavePlayer_56, WavePlayer_57, WavePlayer_58], properties {weights [1,1,1,1]}}
 {id=Random_12, children [WavePlayer_59, WavePlayer_60, WavePlayer_61, WavePlayer_62, WavePlayer_63], properties {weights [1,1,1,1,1]}}
 {
  id=Random_2
  children [WavePlayer_13, WavePlayer_14, WavePlayer_15, WavePlayer_16, WavePlayer_17, WavePlayer_18, WavePlayer_19]
  properties {weights [1,1,1,1,1,1,1]}
 }
 {id=Random_3, children [WavePlayer_20, WavePlayer_21, WavePlayer_22], properties {weights [1,1,1]}}
 {id=Random_5, children [WavePlayer_35, WavePlayer_36, WavePlayer_37, WavePlayer_38], properties {weights [1,1,1,1]}}
 {
  id=Random_6
  children [Modulator_3,Modulator_4,Modulator_0,Modulator_5,Modulator_6,Modulator_7,Modulator_13,Mixer_0,Modulator_10,Modulator_11]
  properties {weights [1,1,1,1,1,1,1,1,1,1]}
 }
 {id=Random_7, children [WavePlayer_39, WavePlayer_40], properties {weights [1,1]}}
 {id=Random_8, children [WavePlayer_41, WavePlayer_42, WavePlayer_43, WavePlayer_44, WavePlayer_45], properties {weights [1,1,1,1,1]}}
 {id=Random_9, children [WavePlayer_46, WavePlayer_47], properties {weights [1,1]}}
 {id=WavePlayer_10, properties {soundWaveAssetPtr=/AmbientAudioContent/Waves/OneShots/Spot_Japanese_Ezo_Brown_Frog_Solo_Woodland_4}}
 {id=WavePlayer_11, properties {soundWaveAssetPtr=/AmbientAudioContent/Waves/OneShots/Spot_Japanese_Ezo_Brown_Frog_Solo_Woodland_3}}
 {id=WavePlayer_12, properties {soundWaveAssetPtr=/AmbientAudioContent/Waves/OneShots/Spot_Japanese_Ezo_Brown_Frog_Solo_Woodland_2}}
 {id=WavePlayer_13, properties {soundWaveAssetPtr=/AmbientAudioContent/Waves/OneShots/Spot_Japanese_Ezo_Brown_Frogs_Woodland_4}}
 {id=WavePlayer_14, properties {soundWaveAssetPtr=/AmbientAudioContent/Waves/OneShots/Spot_Japanese_Ezo_Brown_Frogs_Woodland_1}}
 {id=WavePlayer_15, properties {soundWaveAssetPtr=/AmbientAudioContent/Waves/OneShots/Spot_Japanese_Ezo_Brown_Frogs_Woodland_2}}
 {id=WavePlayer_16, properties {soundWaveAssetPtr=/AmbientAudioContent/Waves/OneShots/Spot_Japanese_Ezo_Brown_Frogs_Woodland_3}}
 {id=WavePlayer_17, properties {soundWaveAssetPtr=/AmbientAudioContent/Waves/OneShots/Spot_Japanese_Ezo_Brown_Frogs_Woodland_5}}
 {id=WavePlayer_18, properties {soundWaveAssetPtr=/AmbientAudioContent/Waves/OneShots/Spot_Japanese_Ezo_Brown_Frogs_Woodland_6}}
 {id=WavePlayer_19, properties {soundWaveAssetPtr=/AmbientAudioContent/Waves/OneShots/Spot_Japanese_Ezo_Brown_Frogs_Woodland_7}}
 {id=WavePlayer_20, properties {soundWaveAssetPtr=/AmbientAudioContent/Waves/OneShots/Spot_Frog_Song_Japanese_Ezo_Brown_Night_01}}
 {id=WavePlayer_21, properties {soundWaveAssetPtr=/AmbientAudioContent/Waves/OneShots/Spot_Frog_Song_Japanese_Ezo_Brown_Night_02}}
 {id=WavePlayer_22, properties {soundWaveAssetPtr=/AmbientAudioContent/Waves/OneShots/Spot_Frog_Song_Japanese_Ezo_Brown_Night_03}}
 {id=WavePlayer_35, properties {soundWaveAssetPtr=/AmbientAudioContent/Waves/OneShots/Spot_Frog_Chirp_Japanese_Swamp_01}}
 {id=WavePlayer_36, properties {soundWaveAssetPtr=/AmbientAudioContent/Waves/OneShots/Spot_Frog_Chirp_Japanese_Swamp_02}}
 {id=WavePlayer_37, properties {soundWaveAssetPtr=/AmbientAudioContent/Waves/OneShots/Spot_Frog_Chirp_Japanese_Swamp_03}}
 {id=WavePlayer_38, properties {soundWaveAssetPtr=/AmbientAudioContent/Waves/OneShots/Spot_Frog_Chirp_Japanese_Swamp_04}}
 {id=WavePlayer_39, properties {soundWaveAssetPtr=/AmbientAudioContent/Waves/OneShots/Spot_Frog_Croak_Japanese_Swamp_Evening_01}}
 {id=WavePlayer_40, properties {soundWaveAssetPtr=/AmbientAudioContent/Waves/OneShots/Spot_Frog_Croak_Japanese_Swamp_Evening_02}}
 {id=WavePlayer_41, properties {soundWaveAssetPtr=/AmbientAudioContent/Waves/OneShots/Spot_Frog_Japanese_Croaky_Small_05}}
 {id=WavePlayer_42, properties {soundWaveAssetPtr=/AmbientAudioContent/Waves/OneShots/Spot_Frog_Japanese_Croaky_Small_04}}
 {id=WavePlayer_43, properties {soundWaveAssetPtr=/AmbientAudioContent/Waves/OneShots/Spot_Frog_Japanese_Croaky_Small_03}}
 {id=WavePlayer_44, properties {soundWaveAssetPtr=/AmbientAudioContent/Waves/OneShots/Spot_Frog_Japanese_Croaky_Small_02}}
 {id=WavePlayer_45, properties {soundWaveAssetPtr=/AmbientAudioContent/Waves/OneShots/Spot_Frog_Japanese_Croaky_Small_01}}
 {id=WavePlayer_46, properties {soundWaveAssetPtr=/AmbientAudioContent/Waves/OneShots/Spot_Frog_Japanese_Fukui_02}}
 {id=WavePlayer_47, properties {soundWaveAssetPtr=/AmbientAudioContent/Waves/OneShots/Spot_Frog_Japanese_Fukui_01}}
 {id=WavePlayer_48, properties {soundWaveAssetPtr=/AmbientAudioContent/Waves/OneShots/Spot_Frog_Striped_Tree_Calls_01}}
 {id=WavePlayer_49, properties {soundWaveAssetPtr=/AmbientAudioContent/Waves/OneShots/Spot_Frog_Striped_Tree_Calls_02}}
 {id=WavePlayer_50, properties {soundWaveAssetPtr=/AmbientAudioContent/Waves/OneShots/Spot_Frog_Striped_Tree_Calls_03}}
 {id=WavePlayer_51, properties {soundWaveAssetPtr=/AmbientAudioContent/Waves/OneShots/Spot_Frog_Striped_Tree_Calls_04}}
 {id=WavePlayer_52, properties {soundWaveAssetPtr=/AmbientAudioContent/Waves/OneShots/Spot_Frog_Striped_Tree_Calls_05}}
 {id=WavePlayer_53, properties {soundWaveAssetPtr=/AmbientAudioContent/Waves/OneShots/Spot_Frog_Striped_Tree_Calls_06}}
 {id=WavePlayer_54, properties {soundWaveAssetPtr=/AmbientAudioContent/Waves/OneShots/Spot_Frog_Striped_Tree_Calls_07}}
 {id=WavePlayer_55, properties {soundWaveAssetPtr=/AmbientAudioContent/Waves/OneShots/Spot_Frog_Arctic_Toad_Croaks_04}}
 {id=WavePlayer_56, properties {soundWaveAssetPtr=/AmbientAudioContent/Waves/OneShots/Spot_Frog_Arctic_Toad_Croaks_01}}
 {id=WavePlayer_57, properties {soundWaveAssetPtr=/AmbientAudioContent/Waves/OneShots/Spot_Frog_Arctic_Toad_Croaks_02}}
 {id=WavePlayer_58, properties {soundWaveAssetPtr=/AmbientAudioContent/Waves/OneShots/Spot_Frog_Arctic_Toad_Croaks_03}}
 {id=WavePlayer_59, properties {soundWaveAssetPtr=&Spot_Frog_Japanese_Croaky_Small_05}}
 {id=WavePlayer_60, properties {soundWaveAssetPtr=&Spot_Frog_Japanese_Croaky_Small_04}}
 {id=WavePlayer_61, properties {soundWaveAssetPtr=&Spot_Frog_Japanese_Croaky_Small_03}}
 {id=WavePlayer_62, properties {soundWaveAssetPtr=&Spot_Frog_Japanese_Croaky_Small_02}}
 {id=WavePlayer_63, properties {soundWaveAssetPtr=&Spot_Frog_Japanese_Croaky_Small_01}}
 {id=WavePlayer_7, properties {soundWaveAssetPtr=/AmbientAudioContent/Waves/OneShots/Spot_Japanese_Ezo_Brown_Frog_Solo_Woodland_6}}
 {id=WavePlayer_8, properties {soundWaveAssetPtr=/AmbientAudioContent/Waves/OneShots/Spot_Japanese_Ezo_Brown_Frog_Solo_Woodland_1}}
 {id=WavePlayer_9, properties {soundWaveAssetPtr=/AmbientAudioContent/Waves/OneShots/Spot_Japanese_Ezo_Brown_Frog_Solo_Woodland_5}}
]
properties {
 attenuationSettings=&Fort_TileAmbient_OneShot_Bird_Att
 maxDistance=8005
 soundClassObject=&Fort_Ambient_OneShots_Outdoors
 volumeMultiplier=0.4
 concurrencyOverrides {maxCount=2, voiceStealReleaseTime=1}
 concurrencySet [/AmbientAudioBase/Concurrency/AmbientConc_2Max_Spot_Concurrency, null]
 preEffectBusSends [{maxSendDistance=8000, minSendDistance=3000, minSendLevel=0.4, customSendLevelCurve {editorCurveData {keys [<at 0 0.4>,<at 1 1>]}}}]
 soundSubmixSends [{
   maxSendDistance=8000
   minSendDistance=3000
   minSendLevel=0.4
   soundSubmix=&Fortnite_Reverb_Default
   customSendLevelCurve {editorCurveData {keys [<at 0 0.4>,<at 1 1>]}}
  }]
}
