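# Sound cue for outdoor ambient one-shot frog and insect calls, attenuated with the tile-ambient one-shot settings.
# A quality-level root sends its first two levels to a random pick of ten call waves and its last two to a pick of only the first two waves, trimming variety by quality.
# Each equally weighted random branch passes through its own Modulator for pitch and volume variation.
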SoundCue /AmbientAudioContent/SoundCues/OneShots/Spot_Frog_Insects_Call_Cue
root=QualityLevel_0
nodes [
 {id=Modulator_0, children [Random_1]}
 {id=Modulator_1, children [Random_2]}
 {id=QualityLevel_0, children [Modulator_0,Modulator_0,Modulator_1,Modulator_1]}
 {
  id=Random_1
  children [
   WavePlayer_0
   WavePlayer_5
   WavePlayer_6
   WavePlayer_7
   WavePlayer_9
   WavePlayer_10
   WavePlayer_18
   WavePlayer_19
   WavePlayer_20
   WavePlayer_21
  ]
  properties {weights [1,1,1,1,1,1,1,1,1,1]}
 }
 {id=Random_2, children [WavePlayer_16, WavePlayer_17], properties {weights [1,1]}}
 {id=WavePlayer_0, properties {soundWaveAssetPtr=/AmbientAudioContent/Waves/OneShots/Spot_Frog_Insects_Call_01}}
 {id=WavePlayer_10, properties {soundWaveAssetPtr=/AmbientAudioContent/Waves/OneShots/Spot_Frog_Insects_Call_06}}
 {id=WavePlayer_16, properties {soundWaveAssetPtr=&Spot_Frog_Insects_Call_01}}
 {id=WavePlayer_17, properties {soundWaveAssetPtr=/AmbientAudioContent/Waves/OneShots/Spot_Frog_Insects_Call_02}}
 {id=WavePlayer_18, properties {soundWaveAssetPtr=/AmbientAudioContent/Waves/OneShots/Spot_Frog_Insects_Call_07}}
 {id=WavePlayer_19, properties {soundWaveAssetPtr=/AmbientAudioContent/Waves/OneShots/Spot_Frog_Insects_Call_08}}
 {id=WavePlayer_20, properties {soundWaveAssetPtr=/AmbientAudioContent/Waves/OneShots/Spot_Frog_Insects_Call_09}}
 {id=WavePlayer_21, properties {soundWaveAssetPtr=/AmbientAudioContent/Waves/OneShots/Spot_Frog_Insects_Call_10}}
 {id=WavePlayer_5, properties {soundWaveAssetPtr=&Spot_Frog_Insects_Call_02}}
 {id=WavePlayer_6, properties {soundWaveAssetPtr=/AmbientAudioContent/Waves/OneShots/Spot_Frog_Insects_Call_03}}
 {id=WavePlayer_7, properties {soundWaveAssetPtr=/AmbientAudioContent/Waves/OneShots/Spot_Frog_Insects_Call_04}}
 {id=WavePlayer_9, properties {soundWaveAssetPtr=/AmbientAudioContent/Waves/OneShots/Spot_Frog_Insects_Call_05}}
]
properties {
 attenuationSettings=/Game/Sounds/Attenuation/Fort_TileAmbient_OneShot_Att
 maxDistance=8000
 soundClassObject=&Fort_Ambient_OneShots_Outdoors
 volumeMultiplier=0.501
 concurrencyOverrides {maxCount=2, voiceStealReleaseTime=1}
 concurrencySet [null]
 preEffectBusSends [{}]
}
