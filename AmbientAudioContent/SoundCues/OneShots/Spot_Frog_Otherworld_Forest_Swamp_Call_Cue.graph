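# One-shot ambient Sound Cue for otherworld forest swamp frog calls, played quietly in the outdoor ambience sound class.
# An equal-weight Random picks one of three waves, the third attenuated by a Mixer, then a Modulator varies pitch and volume.
# All quality levels share that chain.
# Distant-creature attenuation, a tight concurrency cap, and distance-based sends to an exterior source bus and two reverb submixes shape playback.
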SoundCue /AmbientAudioContent/SoundCues/OneShots/Spot_Frog_Otherworld_Forest_Swamp_Call_Cue
root=QualityLevel_0
nodes [
 {id=Mixer_0, children [WavePlayer_6], properties {inputVolume [0.5]}}
 {id=Modulator_0, children [Random_0], properties {pitchMax=1.1, pitchMin=0.8, volumeMax=1.2, volumeMin=1}}
 {id=QualityLevel_0, children [Modulator_0,Modulator_0,Modulator_0,Modulator_0]}
 {id=Random_0, children [WavePlayer_3,WavePlayer_5,Mixer_0], properties {bShouldExcludeFromBranchCulling=true, weights [1,1,1]}}
 {id=WavePlayer_3, properties {soundWaveAssetPtr=/AmbientAudioContent/Waves/OneShots/Spot_Frog_Otherworld_Forest_Swamp_Call_01}}
 {id=WavePlayer_5, properties {soundWaveAssetPtr=/AmbientAudioContent/Waves/OneShots/Spot_Frog_Otherworld_Forest_Swamp_Call_02}}
 {id=WavePlayer_6, properties {soundWaveAssetPtr=/AmbientAudioContent/Waves/OneShots/Spot_Frog_Otherworld_Forest_Swamp_Call_03}}
]
properties {
 attenuationSettings=/Game/Sounds/Attenuation/Fort_TileAmbient_OneShot_Distant_Creature_Att
 maxDistance=8500
 soundClassObject=&Fort_Ambient_OneShots_Outdoors
 volumeMultiplier=0.2
 busSends [{
   audioBus=null
   maxSendDistance=8000
   maxSendLevel=1.2
   minSendDistance=3000
   minSendLevel=0.4
   sendLevel=0
   soundSourceBus=&AmbientAudio_Exterior_SourceBus
   customSendLevelCurve {editorCurveData {keys [<at 0 0.4>,<at 1 1>]}}
  }]
 concurrencyOverrides {maxCount=2, voiceStealReleaseTime=1}
 concurrencySet [null]
 soundSubmixSends [
  {
   maxSendDistance=8000
   minSendDistance=3000
   minSendLevel=0.4
   soundSubmix=&Fortnite_Reverb_Distant
   customSendLevelCurve {editorCurveData {keys [<at 0 0>,<at 0.5 0>,<at 1 1>]}}
  }
  {minSendLevel=0.6, soundSubmix=&Fortnite_Reverb_Default}
 ]
}
