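# Sound Cue for striped tree frog call one-shots in outdoor ambient audio.
# A Random node picks among seven equally weighted call waves, then a Modulator varies pitch and volume.
# All four Quality Level inputs share that same chain, so playback is identical at every quality.
# It uses a bird one-shot attenuation asset, the outdoor one-shot sound class and a bus send to the exterior ambient source bus.
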SoundCue /AmbientAudioContent/SoundCues/OneShots/Spot_Frog_Striped_Tree_Calls_Cue
root=QualityLevel_0
nodes [
 {id=Modulator_0, children [Random_0], properties {pitchMax=1.1, pitchMin=0.8, volumeMax=1.2, volumeMin=1}}
 {id=QualityLevel_0, children [Modulator_0,Modulator_0,Modulator_0,Modulator_0]}
 {
  id=Random_0
  children [WavePlayer_14, WavePlayer_15, WavePlayer_16, WavePlayer_17, WavePlayer_18, WavePlayer_19, WavePlayer_20]
  properties {weights [1,1,1,1,1,1,1]}
 }
 {id=WavePlayer_14, properties {soundWaveAssetPtr=/AmbientAudioContent/Waves/OneShots/Spot_Frog_Striped_Tree_Calls_01}}
 {id=WavePlayer_15, properties {soundWaveAssetPtr=/AmbientAudioContent/Waves/OneShots/Spot_Frog_Striped_Tree_Calls_02}}
 {id=WavePlayer_16, properties {soundWaveAssetPtr=/AmbientAudioContent/Waves/OneShots/Spot_Frog_Striped_Tree_Calls_03}}
 {id=WavePlayer_17, properties {soundWaveAssetPtr=/AmbientAudioContent/Waves/OneShots/Spot_Frog_Striped_Tree_Calls_04}}
 {id=WavePlayer_18, properties {soundWaveAssetPtr=/AmbientAudioContent/Waves/OneShots/Spot_Frog_Striped_Tree_Calls_05}}
 {id=WavePlayer_19, properties {soundWaveAssetPtr=/AmbientAudioContent/Waves/OneShots/Spot_Frog_Striped_Tree_Calls_06}}
 {id=WavePlayer_20, properties {soundWaveAssetPtr=/AmbientAudioContent/Waves/OneShots/Spot_Frog_Striped_Tree_Calls_07}}
]
properties {
 attenuationSettings=&Fort_TileAmbient_OneShot_Bird_Att
 maxDistance=8005
 soundClassObject=&Fort_Ambient_OneShots_Outdoors
 volumeMultiplier=0.4
 busSends [{
   audioBus=null
   maxSendDistance=5000
   maxSendLevel=1.2
   minSendDistance=3000
   minSendLevel=0.2
   sendLevel=0
   soundSourceBus=&AmbientAudio_Exterior_SourceBus
  }]
 concurrencyOverrides {maxCount=2, voiceStealReleaseTime=1}
 concurrencySet [null]
 preEffectBusSends [{}]
}
