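# Ambient exterior SoundCue that plays randomized golden chime swell one-shots for a gold mountain spot, capped at one concurrent instance.
# A shared loop, random delay, envelope choice and pitch/volume modulator chain picks from twelve waves.
# A Mixer runs that chain through three attenuation branches: two HRTF cones and an occluded, low-pass-filtered one, all with listener focus and distant reverb sends.
# The root quality switch leaves the last two levels empty.
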SoundCue /AmbientAudioContent/SoundCues/OneShots/Spot_Gold_Mountain_Oneshots_Ext_Cue
root=QualityLevel_1
nodes [
 {
  id=Attenuation_0
  children [Looping_0]
  properties {
   bOverrideAttenuation=true
   attenuationOverrides {
    attenuationShape=Cone
    bEnableFocusInterpolation=true
    bEnableListenerFocus=true
    bEnableSubmixSends=true
    distanceAlgorithm=Logarithmic
    falloffDistance=6500
    focusAttackInterpSpeed=0.05
    focusAzimuth=15
    focusReleaseInterpSpeed=0.02
    nonFocusVolumeAttenuation=0.4
    spatializationAlgorithm=SPATIALIZATION_HRTF
    stereoSpread=100
    attenuationShapeExtents=<1000 85>
    submixSendSettings [{
      bEnableHPFCutoff=false
      bEnableLPFCutoff=false
      disableManualSendClamp=false
      hPFCutoff=20
      lPFCutoff=20000
      maxSendDistance=15000
      maxSendLevel=1
      minSendDistance=7500
      minSendLevel=0
      sendLevel=0.2
      sendLevelControlMethod=Linear
      soundSubmix=&Fortnite_Reverb_Distant
      customSendLevelCurve {editorCurveData {keys [<at 0 0.2>,<at 1 0.6>]}}
     }]
   }
  }
 }
 {
  id=Attenuation_2
  children [Looping_0]
  properties {
   bOverrideAttenuation=true
   attenuationOverrides {
    bAttenuateWithLPF=true
    bEnableFocusInterpolation=true
    bEnableListenerFocus=true
    bEnableOcclusion=true
    bEnableSubmixSends=true
    distanceAlgorithm=Logarithmic
    falloffDistance=3500
    focusAttackInterpSpeed=0.1
    focusAzimuth=15
    focusReleaseInterpSpeed=0.5
    lPFFrequencyAtMax=1000
    lPFRadiusMax=7500
    lPFRadiusMin=4000
    nonFocusAzimuth=48
    nonFocusVolumeAttenuation=0.2
    nonSpatializedRadiusEnd=3500
    nonSpatializedRadiusStart=7500
    occlusionInterpolationTime=2
    occlusionLowPassFilterFrequency=4000
    occlusionTraceChannel=ECC_GameTraceChannel11
    occlusionVolumeAttenuation=0.5
    stereoSpread=800
    attenuationShapeExtents=<4000 120>
    submixSendSettings [{
      bEnableHPFCutoff=false
      bEnableLPFCutoff=false
      disableManualSendClamp=false
      hPFCutoff=20
      lPFCutoff=20000
      maxSendDistance=7500
      maxSendLevel=0.6
      minSendDistance=3000
      minSendLevel=0.25
      sendLevel=0.2
      sendLevelControlMethod=Linear
      soundSubmix=&Fortnite_Reverb_Distant
      customSendLevelCurve {editorCurveData {keys [<at 0 0.2>,<at 1 0.6>]}}
     }]
   }
  }
 }
 {
  id=Attenuation_3
  children [Looping_0]
  properties {
   bOverrideAttenuation=true
   attenuationOverrides {
    attenuationShape=Cone
    bEnableFocusInterpolation=true
    bEnableListenerFocus=true
    bEnableSubmixSends=true
    distanceAlgorithm=Logarithmic
    falloffDistance=3000
    focusAttackInterpSpeed=0.05
    focusAzimuth=45
    focusReleaseInterpSpeed=0.05
    nonFocusVolumeAttenuation=0.2
    spatializationAlgorithm=SPATIALIZATION_HRTF
    attenuationShapeExtents=<1000 85>
    submixSendSettings [{
      bEnableHPFCutoff=false
      bEnableLPFCutoff=false
      disableManualSendClamp=false
      hPFCutoff=20
      lPFCutoff=20000
      maxSendDistance=15000
      maxSendLevel=1
      minSendDistance=7500
      minSendLevel=0
      sendLevel=0.2
      sendLevelControlMethod=Linear
      soundSubmix=&Fortnite_Reverb_Distant
      customSendLevelCurve {editorCurveData {keys [<at 0 0.2>,<at 1 0.6>]}}
     }]
   }
  }
 }
 {id=Delay_0, children [Random_0], properties {delayMax=12, delayMin=5}}
 {
  id=Enveloper_0
  children [QualityLevel_0]
  properties {
   loopEnd=5
   pitchCurve {editorCurveData {keys [{time=0}]}}
   volumeCurve {editorCurveData {keys [
      <at 0 1 RCIM_Cubic>
      {arriveTangent=0.0011, interpMode=RCIM_Cubic, leaveTangent=0.0011, tangentMode=RCTM_User, time=0.33, value=0.95}
      {arriveTangent=0.0746, interpMode=RCIM_Cubic, leaveTangent=0.0746, time=0.7048, value=1.05}
      <at 1 1>
     ]}}
  }
 }
 {
  id=Enveloper_1
  children [QualityLevel_0]
  properties {
   loopEnd=5
   pitchCurve {editorCurveData {keys [{time=0}]}}
   volumeCurve {editorCurveData {keys [
      <at 0 1 RCIM_Cubic>
      {arriveTangent=0.0011, interpMode=RCIM_Cubic, leaveTangent=0.0011, tangentMode=RCTM_User, time=0.33, value=1.05}
      {arriveTangent=-0.0039, interpMode=RCIM_Cubic, leaveTangent=-0.0039, tangentMode=RCTM_User, time=0.7048, value=0.95}
      <at 1 1>
     ]}}
  }
 }
 {id=Looping_0, children [Delay_0]}
 {id=Mixer_0, children [Attenuation_3, Attenuation_0, Attenuation_2], properties {inputVolume [0.9,0.9,0.9]}}
 {id=Modulator_1, children [Random_1], properties {pitchMax=1.1, volumeMax=1.1}}
 {id=QualityLevel_0, children [Modulator_1,Modulator_1,Modulator_1,Modulator_1]}
 {id=QualityLevel_1, children [Mixer_0,Mixer_0,null,null]}
 {id=Random_0, children [Enveloper_0,Enveloper_1], properties {weights [1,1]}}
 {
  id=Random_1
  children [
   WavePlayer_0
   WavePlayer_1
   WavePlayer_2
   WavePlayer_3
   WavePlayer_4
   WavePlayer_5
   WavePlayer_14
   WavePlayer_15
   WavePlayer_16
   WavePlayer_17
   WavePlayer_18
   WavePlayer_19
  ]
  properties {weights [1,1,1,1,1,1,1,1,1,1,1,1]}
 }
 {id=WavePlayer_0, properties {soundWaveAssetPtr=/AmbientAudioContent/Waves/OneShots/Spot_Golden_Chime_Swell_01}}
 {id=WavePlayer_1, properties {soundWaveAssetPtr=/AmbientAudioContent/Waves/OneShots/Spot_Golden_Chime_Swell_02}}
 {id=WavePlayer_14, properties {soundWaveAssetPtr=/AmbientAudioContent/Waves/OneShots/Spot_Golden_Chime_Swell_07}}
 {id=WavePlayer_15, properties {soundWaveAssetPtr=/AmbientAudioContent/Waves/OneShots/Spot_Golden_Chime_Swell_08}}
 {id=WavePlayer_16, properties {soundWaveAssetPtr=/AmbientAudioContent/Waves/OneShots/Spot_Golden_Chime_Swell_09}}
 {id=WavePlayer_17, properties {soundWaveAssetPtr=/AmbientAudioContent/Waves/OneShots/Spot_Golden_Chime_Swell_10}}
 {id=WavePlayer_18, properties {soundWaveAssetPtr=/AmbientAudioContent/Waves/OneShots/Spot_Golden_Chime_Swell_11}}
 {id=WavePlayer_19, properties {soundWaveAssetPtr=/AmbientAudioContent/Waves/OneShots/Spot_Golden_Chime_Swell_12}}
 {id=WavePlayer_2, properties {soundWaveAssetPtr=/AmbientAudioContent/Waves/OneShots/Spot_Golden_Chime_Swell_03}}
 {id=WavePlayer_3, properties {soundWaveAssetPtr=/AmbientAudioContent/Waves/OneShots/Spot_Golden_Chime_Swell_04}}
 {id=WavePlayer_4, properties {soundWaveAssetPtr=/AmbientAudioContent/Waves/OneShots/Spot_Golden_Chime_Swell_05}}
 {id=WavePlayer_5, properties {soundWaveAssetPtr=/AmbientAudioContent/Waves/OneShots/Spot_Golden_Chime_Swell_06}}
]
properties {
 bOverrideConcurrency=true
 maxDistance=7500
 soundClassObject=&Fort_Ambient_OneShots_Outdoors
 volumeMultiplier=0.4
 attenuationOverrides {
  attenuationShape=Cone
  bAttenuateWithLPF=true
  bEnableFocusInterpolation=true
  bEnableListenerFocus=true
  falloffDistance=5000
  focusAttackInterpSpeed=0.1
  focusReleaseInterpSpeed=0.01
  hPFFrequencyAtMax=750
  hPFFrequencyAtMin=750
  lPFFrequencyAtMax=1000
  lPFRadiusMax=15000
  lPFRadiusMin=7500
  nonFocusAzimuth=55
  nonFocusVolumeAttenuation=0
  attenuationShapeExtents=<10000 120>
 }
 concurrencyOverrides {maxCount=1, voiceStealReleaseTime=5}
 soundSubmixSends [{
   disableManualSendClamp=true
   maxSendDistance=15000
   maxSendLevel=8
   minSendDistance=7500
   minSendLevel=6
   sendLevel=8
   soundSubmix=&Fortnite_Reverb_Default
   customSendLevelCurve {editorCurveData {keys [<at 0 0.4>,<at 1 1>]}}
  }]
}
