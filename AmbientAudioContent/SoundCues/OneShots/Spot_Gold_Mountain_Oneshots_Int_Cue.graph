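# Ambient one-shot Sound Cue that plays golden chime swells for the gold mountain spot ambience in Fortnite's Athena ambient sound class.
# One of twelve waves is picked at random, modulated in pitch and volume, then shaped by one of two subtly different volume envelopes.
# The root quality switch plays sound only in its first two slots.
# Overrides set custom distance falloff with low-pass filtering, limited concurrency and two distance-scaled reverb sends.
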SoundCue /AmbientAudioContent/SoundCues/OneShots/Spot_Gold_Mountain_Oneshots_Int_Cue
root=QualityLevel_1
nodes [
 {
  id=Enveloper_0
  children [QualityLevel_0]
  properties {
   loopEnd=5
   pitchCurve {editorCurveData {keys [{time=0}]}}
   volumeCurve {editorCurveData {keys [
      <at 0 1 RCIM_Cubic>
      {arriveTangent=0.0011, interpMode=RCIM_Cubic, leaveTangent=0.0011, tangentMode=RCTM_User, time=0.33, value=0.95}
      {arriveTangent=0.0746, interpMode=RCIM_Cubic, leaveTangent=0.0746, time=0.7048, value=1.05}
      <at 1 1>
     ]}}
  }
 }
 {
  id=Enveloper_1
  children [QualityLevel_0]
  properties {
   loopEnd=5
   pitchCurve {editorCurveData {keys [{time=0}]}}
   volumeCurve {editorCurveData {keys [
      <at 0 1 RCIM_Cubic>
      {arriveTangent=0.0011, interpMode=RCIM_Cubic, leaveTangent=0.0011, tangentMode=RCTM_User, time=0.33, value=1.05}
      {arriveTangent=-0.0039, interpMode=RCIM_Cubic, leaveTangent=-0.0039, tangentMode=RCTM_User, time=0.7048, value=0.95}
      <at 1 1>
     ]}}
  }
 }
 {id=Modulator_1, children [Random_1], properties {pitchMax=1.1, pitchMin=1, volumeMax=1.1}}
 {id=QualityLevel_0, children [Modulator_1,Modulator_1,Modulator_1,Modulator_1]}
 {id=QualityLevel_1, children [Random_0,Random_0,null,null]}
 {id=Random_0, children [Enveloper_0,Enveloper_1], properties {weights [1,1]}}
 {
  id=Random_1
  children [
   WavePlayer_0
   WavePlayer_1
   WavePlayer_2
   WavePlayer_3
   WavePlayer_4
   WavePlayer_5
   WavePlayer_14
   WavePlayer_15
   WavePlayer_16
   WavePlayer_17
   WavePlayer_18
   WavePlayer_19
  ]
  properties {weights [1,1,1,1,1,1,1,1,1,1,1,1]}
 }
 {id=WavePlayer_0, properties {soundWaveAssetPtr=/AmbientAudioContent/Waves/OneShots/Spot_Golden_Chime_Swell_01}}
 {id=WavePlayer_1, properties {soundWaveAssetPtr=/AmbientAudioContent/Waves/OneShots/Spot_Golden_Chime_Swell_02}}
 {id=WavePlayer_14, properties {soundWaveAssetPtr=/AmbientAudioContent/Waves/OneShots/Spot_Golden_Chime_Swell_07}}
 {id=WavePlayer_15, properties {soundWaveAssetPtr=/AmbientAudioContent/Waves/OneShots/Spot_Golden_Chime_Swell_08}}
 {id=WavePlayer_16, properties {soundWaveAssetPtr=/AmbientAudioContent/Waves/OneShots/Spot_Golden_Chime_Swell_09}}
 {id=WavePlayer_17, properties {soundWaveAssetPtr=/AmbientAudioContent/Waves/OneShots/Spot_Golden_Chime_Swell_10}}
 {id=WavePlayer_18, properties {soundWaveAssetPtr=/AmbientAudioContent/Waves/OneShots/Spot_Golden_Chime_Swell_11}}
 {id=WavePlayer_19, properties {soundWaveAssetPtr=/AmbientAudioContent/Waves/OneShots/Spot_Golden_Chime_Swell_12}}
 {id=WavePlayer_2, properties {soundWaveAssetPtr=/AmbientAudioContent/Waves/OneShots/Spot_Golden_Chime_Swell_03}}
 {id=WavePlayer_3, properties {soundWaveAssetPtr=/AmbientAudioContent/Waves/OneShots/Spot_Golden_Chime_Swell_04}}
 {id=WavePlayer_4, properties {soundWaveAssetPtr=/AmbientAudioContent/Waves/OneShots/Spot_Golden_Chime_Swell_05}}
 {id=WavePlayer_5, properties {soundWaveAssetPtr=/AmbientAudioContent/Waves/OneShots/Spot_Golden_Chime_Swell_06}}
]
properties {
 attenuationSettings=/AmbientAudioBase/Attenuations/Fort_TileAmbient_OneShot_Insect_By_Att
 bOverrideAttenuation=true
 bOverrideConcurrency=true
 maxDistance=8000
 soundClassObject=&Fort_Ambient_Athena
 volumeMultiplier=0.3
 attenuationOverrides {
  bAttenuateWithLPF=true
  bEnableFocusInterpolation=true
  distanceAlgorithm=Custom
  falloffDistance=8000
  focusAttackInterpSpeed=0.1
  focusReleaseInterpSpeed=0.01
  hPFFrequencyAtMax=750
  hPFFrequencyAtMin=750
  lPFFrequencyAtMax=1000
  lPFRadiusMax=15000
  lPFRadiusMin=7500
  nonFocusAzimuth=55
  nonFocusVolumeAttenuation=0
  stereoSpread=1000
  attenuationShapeExtents=<0 120>
  customAttenuationCurve {editorCurveData {keys [
     {arriveTangent=-0.0182, interpMode=RCIM_Cubic, leaveTangent=-0.0182, tangentMode=RCTM_User, time=0, value=0.7}
     {arriveTangent=-1.9548, interpMode=RCIM_Cubic, leaveTangent=-1.9548, tangentMode=RCTM_User, time=0.5, value=0.7}
     <at 1 0>
    ]}}
 }
 concurrencyOverrides {maxCount=3, voiceStealReleaseTime=5}
 soundSubmixSends [
  {
   disableManualSendClamp=true
   maxSendDistance=15000
   maxSendLevel=8
   minSendDistance=7500
   minSendLevel=6
   sendLevel=8
   soundSubmix=&Fortnite_Reverb_Default
   customSendLevelCurve {editorCurveData {keys [<at 0 0.4>,<at 1 1>]}}
  }
  {
   maxSendDistance=6000
   maxSendLevel=0.6
   minSendDistance=3000
   minSendLevel=0.45
   soundSubmix=&Fortnite_Reverb_Distant
   customSendLevelCurve {editorCurveData {keys [<at 0 0>,<at 1 1>]}}
  }
 ]
}
