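# Sound cue for a spot one-shot evening raven call in grassland ambience, attenuated as an outdoor bird sound.
# The first two quality levels usually play one random call with pitch and volume modulation, or less often mix it with a quieter, lower-pitched second call after a short delay, like an answering bird.
# The other two quality levels skip the layering and pick from three of the four waves.
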SoundCue /AmbientAudioContent/SoundCues/OneShots/Spot_Grassland_Evening_Raven_Call_Cue
root=QualityLevel_0
nodes [
 {id=Delay_0, children [Modulator_0], properties {delayMax=2.5, delayMin=2}}
 {id=Mixer_0, children [Modulator_1,Delay_0], properties {inputVolume [1,0.5]}}
 {id=Modulator_0, children [Random_4], properties {pitchMax=0.9, pitchMin=0.8, volumeMax=0.5, volumeMin=0.4}}
 {id=Modulator_1, children [Random_2], properties {pitchMax=1.1, volumeMax=1, volumeMin=0.8}}
 {id=Modulator_2, children [Random_3], properties {pitchMax=1.1, volumeMax=1, volumeMin=0.8}}
 {id=QualityLevel_0, children [Random_5,Random_5,Modulator_2,Modulator_2]}
 {id=Random_2, children [WavePlayer_8, WavePlayer_9, WavePlayer_10, WavePlayer_11], properties {weights [1,1,1,1]}}
 {id=Random_3, children [WavePlayer_19, WavePlayer_20, WavePlayer_21], properties {weights [1,1,1]}}
 {id=Random_4, children [WavePlayer_12, WavePlayer_13, WavePlayer_14, WavePlayer_15], properties {weights [1,1,1,1]}}
 {id=Random_5, children [Modulator_1,Mixer_0], properties {weights [1,0.33]}}
 {id=WavePlayer_10, properties {soundWaveAssetPtr=/AmbientAudioContent/Waves/OneShots/Spot_Grassland_Evening_Raven_Call_03}}
 {id=WavePlayer_11, properties {soundWaveAssetPtr=/AmbientAudioContent/Waves/OneShots/Spot_Grassland_Evening_Raven_Call_02}}
 {id=WavePlayer_12, properties {soundWaveAssetPtr=/AmbientAudioContent/Waves/OneShots/Spot_Grassland_Evening_Raven_Call_01}}
 {id=WavePlayer_13, properties {soundWaveAssetPtr=/AmbientAudioContent/Waves/OneShots/Spot_Grassland_Evening_Raven_Call_04}}
 {id=WavePlayer_14, properties {soundWaveAssetPtr=&Spot_Grassland_Evening_Raven_Call_03}}
 {id=WavePlayer_15, properties {soundWaveAssetPtr=&Spot_Grassland_Evening_Raven_Call_02}}
 {id=WavePlayer_19, properties {soundWaveAssetPtr=&Spot_Grassland_Evening_Raven_Call_01}}
 {id=WavePlayer_20, properties {soundWaveAssetPtr=&Spot_Grassland_Evening_Raven_Call_02}}
 {id=WavePlayer_21, properties {soundWaveAssetPtr=&Spot_Grassland_Evening_Raven_Call_03}}
 {id=WavePlayer_8, properties {soundWaveAssetPtr=&Spot_Grassland_Evening_Raven_Call_01}}
 {id=WavePlayer_9, properties {soundWaveAssetPtr=&Spot_Grassland_Evening_Raven_Call_04}}
]
properties {
 attenuationSettings=&Fort_TileAmbient_OneShot_Bird_Att
 bEnableBusSends=false
 maxDistance=8005
 soundClassObject=&Fort_Ambient_OneShots_Outdoors
 volumeMultiplier=0.2
 preEffectBusSends [{}]
}
