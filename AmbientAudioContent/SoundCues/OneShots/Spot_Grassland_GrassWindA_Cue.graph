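# Sound cue for a grassland grass-wind ambient one-shot in the outdoor one-shots sound class.
# A Random node picks one of five equally weighted GrassWindA waves, and a Modulator at the root varies its pitch and volume.
# The cue plays at reduced volume with wind one-shot attenuation, and a concurrency override caps it at two simultaneous voices with a release time on voice stealing.
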SoundCue /AmbientAudioContent/SoundCues/OneShots/Spot_Grassland_GrassWindA_Cue
root=Modulator_0
nodes [
 {id=Modulator_0, children [Random_1], properties {pitchMax=1.2, pitchMin=0.8, volumeMin=0.8}}
 {id=Random_1, children [WavePlayer_0,WavePlayer_6,WavePlayer_7,WavePlayer_8,WavePlayer_9], properties {weights [1,1,1,1,1]}}
 {id=WavePlayer_0, properties {soundWaveAssetPtr=/AmbientAudioContent/Waves/OneShots/Spot_Grassland_01_GrassWindA_01}}
 {id=WavePlayer_6, properties {soundWaveAssetPtr=/AmbientAudioContent/Waves/OneShots/Spot_Grassland_01_GrassWindA_02}}
 {id=WavePlayer_7, properties {soundWaveAssetPtr=/AmbientAudioContent/Waves/OneShots/Spot_Grassland_01_GrassWindA_03}}
 {id=WavePlayer_8, properties {soundWaveAssetPtr=/AmbientAudioContent/Waves/OneShots/Spot_Grassland_01_GrassWindA_04}}
 {id=WavePlayer_9, properties {soundWaveAssetPtr=/AmbientAudioContent/Waves/OneShots/Spot_Grassland_01_GrassWindA_05}}
]
properties {
 attenuationSettings=/AmbientAudioBase/Attenuations/Fort_TileAmbient_OneShot_Wind_Att
 bOverrideConcurrency=true
 maxDistance=8000
 soundClassObject=&Fort_Ambient_OneShots_Outdoors
 volumeMultiplier=0.3
 concurrencyOverrides {maxCount=2, voiceStealReleaseTime=2}
 preEffectBusSends [{}]
}
