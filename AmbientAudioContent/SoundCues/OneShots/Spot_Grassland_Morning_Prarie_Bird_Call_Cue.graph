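# Outdoor one-shot ambient Sound Cue that plays morning grassland prairie bird calls with bird-specific attenuation.
# All paths draw from one random pool of five equally weighted waves, varied by pitch and volume modulators.
# The first two quality slots usually play a single call, sometimes mixing in a delayed, quieter, lower-pitched draw as a distant-sounding answer.
# The last two slots skip that layering and play one modulated call.
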SoundCue /AmbientAudioContent/SoundCues/OneShots/Spot_Grassland_Morning_Prarie_Bird_Call_Cue
root=QualityLevel_0
nodes [
 {id=Delay_0, children [Modulator_0], properties {delayMax=1.75, delayMin=1}}
 {id=Mixer_0, children [Modulator_1,Delay_0], properties {inputVolume [1,0.25]}}
 {id=Modulator_0, children [Random_0], properties {pitchMax=0.9, pitchMin=0.8, volumeMax=0.6, volumeMin=0.5}}
 {id=Modulator_1, children [Random_0], properties {pitchMax=1.1, pitchMin=1, volumeMax=1.1, volumeMin=0.9}}
 {id=Modulator_2, children [Random_0], properties {pitchMax=1.1, volumeMax=1.1, volumeMin=0.9}}
 {id=QualityLevel_0, children [Random_5,Random_5,Modulator_2,Modulator_2]}
 {id=Random_0, children [WavePlayer_0,WavePlayer_1,WavePlayer_2,WavePlayer_3,WavePlayer_4], properties {weights [1,1,1,1,1]}}
 {id=Random_5, children [Modulator_1,Mixer_0], properties {weights [1,0.33]}}
 {id=WavePlayer_0, properties {soundWaveAssetPtr=/AmbientAudioContent/Waves/OneShots/Spot_Grassland_Morning_Prarie_Bird_Call_01}}
 {id=WavePlayer_1, properties {soundWaveAssetPtr=/AmbientAudioContent/Waves/OneShots/Spot_Grassland_Morning_Prarie_Bird_Call_02}}
 {id=WavePlayer_2, properties {soundWaveAssetPtr=/AmbientAudioContent/Waves/OneShots/Spot_Grassland_Morning_Prarie_Bird_Call_03}}
 {id=WavePlayer_3, properties {soundWaveAssetPtr=/AmbientAudioContent/Waves/OneShots/Spot_Grassland_Morning_Prarie_Bird_Call_04}}
 {id=WavePlayer_4, properties {soundWaveAssetPtr=/AmbientAudioContent/Waves/OneShots/Spot_Grassland_Morning_Prarie_Bird_Call_05}}
]
properties {
 attenuationSettings=&Fort_TileAmbient_OneShot_Bird_Att
 bEnableBusSends=false
 maxDistance=8005
 soundClassObject=&Fort_Ambient_OneShots_Outdoors
 volumeMultiplier=0.251
 preEffectBusSends [{}]
}
