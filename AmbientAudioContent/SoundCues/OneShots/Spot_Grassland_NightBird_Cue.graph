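# Sound cue for one-shot night bird calls in the grassland biome ambience, on the outdoor ambient one-shots sound class.
# A Random node picks among six equally weighted bird waves and feeds a Modulator with lowered minimum pitch and volume for wide variation.
# All four Quality Level inputs share that Modulator, so every quality tier plays the same chain.
# A bird-specific attenuation asset handles distance falloff, with overall volume reduced.
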SoundCue /AmbientAudioContent/SoundCues/OneShots/Spot_Grassland_NightBird_Cue
root=QualityLevel_0
nodes [
 {id=Modulator_0, children [Random_1], properties {pitchMin=0.3, volumeMin=0.3}}
 {id=QualityLevel_0, children [Modulator_0,Modulator_0,Modulator_0,Modulator_0]}
 {
  id=Random_1
  children [WavePlayer_0, WavePlayer_7, WavePlayer_8, WavePlayer_9, WavePlayer_10, WavePlayer_11]
  properties {weights [1,1,1,1,1,1]}
 }
 {id=WavePlayer_0, properties {soundWaveAssetPtr=/AmbientAudioContent/Waves/OneShots/Spot_Biome_Grassland_01_NightBird_01}}
 {id=WavePlayer_10, properties {soundWaveAssetPtr=/AmbientAudioContent/Waves/OneShots/Spot_Biome_Grassland_01_NightBird_05}}
 {id=WavePlayer_11, properties {soundWaveAssetPtr=/AmbientAudioContent/Waves/OneShots/Spot_Biome_Grassland_01_NightBird_06}}
 {id=WavePlayer_7, properties {soundWaveAssetPtr=/AmbientAudioContent/Waves/OneShots/Spot_Biome_Grassland_01_NightBird_02}}
 {id=WavePlayer_8, properties {soundWaveAssetPtr=/AmbientAudioContent/Waves/OneShots/Spot_Biome_Grassland_01_NightBird_03}}
 {id=WavePlayer_9, properties {soundWaveAssetPtr=/AmbientAudioContent/Waves/OneShots/Spot_Biome_Grassland_01_NightBird_04}}
]
properties {
 attenuationSettings=&Fort_TileAmbient_OneShot_Bird_Att
 bEnableBusSends=false
 maxDistance=8005
 soundClassObject=&Fort_Ambient_OneShots_Outdoors
 volumeMultiplier=0.426
 preEffectBusSends [{}]
}
