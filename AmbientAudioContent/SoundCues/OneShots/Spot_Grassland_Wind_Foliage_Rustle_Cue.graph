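# Sound cue for an outdoor grassland ambient one-shot, layering grass wind with foliage rustle.
# A mixer plays two layers at once, each a random pick from five equally weighted waves with its own pitch and volume modulation, the grass wind layer varying pitch more widely.
# All quality levels share that mixer.
# The cue plays at reduced volume through a wind one-shot attenuation and an outdoor ambient one-shot sound class.
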SoundCue /AmbientAudioContent/SoundCues/OneShots/Spot_Grassland_Wind_Foliage_Rustle_Cue
root=QualityLevel_0
nodes [
 {id=Mixer_0, children [Modulator_0,Modulator_1], properties {inputVolume [1,1]}}
 {id=Modulator_0, children [Random_0], properties {pitchMax=1.2, pitchMin=0.8, volumeMin=0.8}}
 {id=Modulator_1, children [Random_3], properties {pitchMax=1, pitchMin=0.9, volumeMax=1, volumeMin=0.8}}
 {id=QualityLevel_0, children [Mixer_0,Mixer_0,Mixer_0,Mixer_0]}
 {id=Random_0, children [WavePlayer_0, WavePlayer_6, WavePlayer_7, WavePlayer_13, WavePlayer_14], properties {weights [1,1,1,1,1]}}
 {id=Random_3, children [WavePlayer_15, WavePlayer_16, WavePlayer_17, WavePlayer_18, WavePlayer_19], properties {weights [1,1,1,1,1]}}
 {id=WavePlayer_0, properties {soundWaveAssetPtr=/AmbientAudioContent/Waves/OneShots/Spot_Grassland_01_GrassWindA_05}}
 {id=WavePlayer_13, properties {soundWaveAssetPtr=/AmbientAudioContent/Waves/OneShots/Spot_Grassland_01_GrassWindA_02}}
 {id=WavePlayer_14, properties {soundWaveAssetPtr=/AmbientAudioContent/Waves/OneShots/Spot_Grassland_01_GrassWindA_01}}
 {id=WavePlayer_15, properties {soundWaveAssetPtr=/AmbientAudioContent/Waves/OneShots/Spot_Grassland_Wind_Foliage_Rustle_01}}
 {id=WavePlayer_16, properties {soundWaveAssetPtr=/AmbientAudioContent/Waves/OneShots/Spot_Grassland_Wind_Foliage_Rustle_02}}
 {id=WavePlayer_17, properties {soundWaveAssetPtr=/AmbientAudioContent/Waves/OneShots/Spot_Grassland_Wind_Foliage_Rustle_03}}
 {id=WavePlayer_18, properties {soundWaveAssetPtr=/AmbientAudioContent/Waves/OneShots/Spot_Grassland_Wind_Foliage_Rustle_04}}
 {id=WavePlayer_19, properties {soundWaveAssetPtr=/AmbientAudioContent/Waves/OneShots/Spot_Grassland_Wind_Foliage_Rustle_05}}
 {id=WavePlayer_6, properties {soundWaveAssetPtr=/AmbientAudioContent/Waves/OneShots/Spot_Grassland_01_GrassWindA_04}}
 {id=WavePlayer_7, properties {soundWaveAssetPtr=/AmbientAudioContent/Waves/OneShots/Spot_Grassland_01_GrassWindA_03}}
]
properties {
 attenuationSettings=/Game/Sounds/Attenuation/Fort_TileAmbient_OneShot_Wind_Att
 bEnableBusSends=false
 maxDistance=8000
 soundClassObject=&Fort_Ambient_OneShots_Outdoors
 volumeMultiplier=0.25
 preEffectBusSends [{}]
}
