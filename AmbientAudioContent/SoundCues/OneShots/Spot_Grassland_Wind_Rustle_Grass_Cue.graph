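# Outdoor ambient one-shot cue for grassland wind rustling through grass, played quietly with wind one-shot attenuation.
# A quality switch sends its first two levels to a Mixer that layers a random GrassWindA wave over a random Wind_Rustle wave.
# The last two levels play only one random pick from a smaller subset of the Wind_Rustle waves.
# Each random branch passes through its own Modulator for pitch and volume variation.
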SoundCue /AmbientAudioContent/SoundCues/OneShots/Spot_Grassland_Wind_Rustle_Grass_Cue
root=QualityLevel_0
nodes [
 {id=Mixer_0, children [Modulator_0,Modulator_1], properties {inputVolume [1,1]}}
 {id=Modulator_0, children [Random_3], properties {pitchMax=1.2, pitchMin=0.8, volumeMax=1, volumeMin=0.8}}
 {id=Modulator_1, children [Random_4], properties {pitchMax=1.1, pitchMin=1, volumeMax=1, volumeMin=0.8}}
 {id=Modulator_2, children [Random_5], properties {pitchMax=1.1, volumeMax=1.1, volumeMin=0.9}}
 {id=QualityLevel_0, children [Mixer_0,Mixer_0,Modulator_2,Modulator_2]}
 {id=Random_3, children [WavePlayer_20, WavePlayer_22, WavePlayer_23, WavePlayer_24, WavePlayer_25], properties {weights [1,1,1,1,1]}}
 {
  id=Random_4
  children [
   WavePlayer_26
   WavePlayer_27
   WavePlayer_28
   WavePlayer_29
   WavePlayer_30
   WavePlayer_31
   WavePlayer_32
   WavePlayer_33
   WavePlayer_34
   WavePlayer_35
  ]
  properties {weights [1,1,1,1,1,1,1,1,1,1]}
 }
 {
  id=Random_5
  children [WavePlayer_36, WavePlayer_37, WavePlayer_38, WavePlayer_39, WavePlayer_40, WavePlayer_41]
  properties {weights [1,1,1,1,1,1]}
 }
 {id=WavePlayer_20, properties {soundWaveAssetPtr=/AmbientAudioContent/Waves/OneShots/Spot_Grassland_01_GrassWindA_01}}
 {id=WavePlayer_22, properties {soundWaveAssetPtr=/AmbientAudioContent/Waves/OneShots/Spot_Grassland_01_GrassWindA_02}}
 {id=WavePlayer_23, properties {soundWaveAssetPtr=/AmbientAudioContent/Waves/OneShots/Spot_Grassland_01_GrassWindA_03}}
 {id=WavePlayer_24, properties {soundWaveAssetPtr=/AmbientAudioContent/Waves/OneShots/Spot_Grassland_01_GrassWindA_04}}
 {id=WavePlayer_25, properties {soundWaveAssetPtr=/AmbientAudioContent/Waves/OneShots/Spot_Grassland_01_GrassWindA_05}}
 {id=WavePlayer_26, properties {soundWaveAssetPtr=/AmbientAudioContent/Waves/OneShots/Spot_Grassland_Wind_Rustle_01}}
 {id=WavePlayer_27, properties {soundWaveAssetPtr=/AmbientAudioContent/Waves/OneShots/Spot_Grassland_Wind_Rustle_02}}
 {id=WavePlayer_28, properties {soundWaveAssetPtr=/AmbientAudioContent/Waves/OneShots/Spot_Grassland_Wind_Rustle_03}}
 {id=WavePlayer_29, properties {soundWaveAssetPtr=/AmbientAudioContent/Waves/OneShots/Spot_Grassland_Wind_Rustle_04}}
 {id=WavePlayer_30, properties {soundWaveAssetPtr=/AmbientAudioContent/Waves/OneShots/Spot_Grassland_Wind_Rustle_05}}
 {id=WavePlayer_31, properties {soundWaveAssetPtr=/AmbientAudioContent/Waves/OneShots/Spot_Grassland_Wind_Rustle_06}}
 {id=WavePlayer_32, properties {soundWaveAssetPtr=/AmbientAudioContent/Waves/OneShots/Spot_Grassland_Wind_Rustle_07}}
 {id=WavePlayer_33, properties {soundWaveAssetPtr=/AmbientAudioContent/Waves/OneShots/Spot_Grassland_Wind_Rustle_08}}
 {id=WavePlayer_34, properties {soundWaveAssetPtr=/AmbientAudioContent/Waves/OneShots/Spot_Grassland_Wind_Rustle_09}}
 {id=WavePlayer_35, properties {soundWaveAssetPtr=/AmbientAudioContent/Waves/OneShots/Spot_Grassland_Wind_Rustle_10}}
 {id=WavePlayer_36, properties {soundWaveAssetPtr=&Spot_Grassland_Wind_Rustle_03}}
 {id=WavePlayer_37, properties {soundWaveAssetPtr=&Spot_Grassland_Wind_Rustle_04}}
 {id=WavePlayer_38, properties {soundWaveAssetPtr=&Spot_Grassland_Wind_Rustle_05}}
 {id=WavePlayer_39, properties {soundWaveAssetPtr=&Spot_Grassland_Wind_Rustle_06}}
 {id=WavePlayer_40, properties {soundWaveAssetPtr=&Spot_Grassland_Wind_Rustle_07}}
 {id=WavePlayer_41, properties {soundWaveAssetPtr=&Spot_Grassland_Wind_Rustle_08}}
]
properties {
 attenuationSettings=/Game/Sounds/Attenuation/Fort_TileAmbient_OneShot_Wind_Att
 bEnableBusSends=false
 maxDistance=8000
 soundClassObject=&Fort_Ambient_OneShots_Outdoors
 volumeMultiplier=0.25
 preEffectBusSends [{}]
}
