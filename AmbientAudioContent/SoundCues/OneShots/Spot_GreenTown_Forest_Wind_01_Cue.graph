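# Sound cue for an outdoor ambient one-shot of forest wind in autumnal leaves, named for GreenTown.
# A quality-level switch sends its first two levels to a random pick among seventeen short wind waves and its last two to a reduced set of four.
# Each branch passes through a Modulator that varies pitch and volume.
# The cue uses a wider tile-ambient attenuation and a distance-scaled send to the default reverb submix.
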SoundCue /AmbientAudioContent/SoundCues/OneShots/Spot_GreenTown_Forest_Wind_01_Cue
root=QualityLevel_0
nodes [
 {id=Modulator_1, children [Random_1], properties {pitchMax=1.1, volumeMax=1.1, volumeMin=0.9}}
 {id=Modulator_2, children [Random_2], properties {pitchMax=1.1, volumeMax=1.1, volumeMin=0.9}}
 {id=QualityLevel_0, children [Modulator_1,Modulator_1,Modulator_2,Modulator_2]}
 {
  id=Random_1
  children [
   WavePlayer_10
   WavePlayer_11
   WavePlayer_12
   WavePlayer_13
   WavePlayer_14
   WavePlayer_15
   WavePlayer_16
   WavePlayer_17
   WavePlayer_18
   WavePlayer_19
   WavePlayer_20
   WavePlayer_21
   WavePlayer_33
   WavePlayer_34
   WavePlayer_35
   WavePlayer_36
   WavePlayer_37
  ]
  properties {weights [1,1,1,1,1,1,1,1,1,1,1,1,1,1,1,1,1]}
 }
 {id=Random_2, children [WavePlayer_38, WavePlayer_39, WavePlayer_40, WavePlayer_41], properties {weights [1,1,1,1]}}
 {id=WavePlayer_10, properties {soundWaveAssetPtr=/AmbientAudioContent/Waves/OneShots/Spot_Wind_Autumnal_Leaves_Short_01}}
 {id=WavePlayer_11, properties {soundWaveAssetPtr=/AmbientAudioContent/Waves/OneShots/Spot_Wind_Autumnal_Leaves_Short_02}}
 {id=WavePlayer_12, properties {soundWaveAssetPtr=/AmbientAudioContent/Waves/OneShots/Spot_Wind_Autumnal_Leaves_Short_03}}
 {id=WavePlayer_13, properties {soundWaveAssetPtr=/AmbientAudioContent/Waves/OneShots/Spot_Wind_Autumnal_Leaves_Short_04}}
 {id=WavePlayer_14, properties {soundWaveAssetPtr=/AmbientAudioContent/Waves/OneShots/Spot_Wind_Autumnal_Leaves_Short_05}}
 {id=WavePlayer_15, properties {soundWaveAssetPtr=/AmbientAudioContent/Waves/OneShots/Spot_Wind_Autumnal_Leaves_Short_06}}
 {id=WavePlayer_16, properties {soundWaveAssetPtr=/AmbientAudioContent/Waves/OneShots/Spot_Wind_Autumnal_Leaves_Short_07}}
 {id=WavePlayer_17, properties {soundWaveAssetPtr=/AmbientAudioContent/Waves/OneShots/Spot_Wind_Autumnal_Leaves_Short_08}}
 {id=WavePlayer_18, properties {soundWaveAssetPtr=/AmbientAudioContent/Waves/OneShots/Spot_Wind_Autumnal_Leaves_Short_09}}
 {id=WavePlayer_19, properties {soundWaveAssetPtr=/AmbientAudioContent/Waves/OneShots/Spot_Wind_Autumnal_Leaves_Short_10}}
 {id=WavePlayer_20, properties {soundWaveAssetPtr=/AmbientAudioContent/Waves/OneShots/Spot_Wind_Autumnal_Leaves_Short_11}}
 {id=WavePlayer_21, properties {soundWaveAssetPtr=/AmbientAudioContent/Waves/OneShots/Spot_Wind_Autumnal_Leaves_Short_12}}
 {id=WavePlayer_33, properties {soundWaveAssetPtr=/AmbientAudioContent/Waves/OneShots/Spot_Wind_Autumnal_Leaves_Short_13}}
 {id=WavePlayer_34, properties {soundWaveAssetPtr=/AmbientAudioContent/Waves/OneShots/Spot_Wind_Autumnal_Leaves_Short_14}}
 {id=WavePlayer_35, properties {soundWaveAssetPtr=/AmbientAudioContent/Waves/OneShots/Spot_Wind_Autumnal_Leaves_Short_15}}
 {id=WavePlayer_36, properties {soundWaveAssetPtr=/AmbientAudioContent/Waves/OneShots/Spot_Wind_Autumnal_Leaves_Short_16}}
 {id=WavePlayer_37, properties {soundWaveAssetPtr=/AmbientAudioContent/Waves/OneShots/Spot_Wind_Autumnal_Leaves_Short_17}}
 {id=WavePlayer_38, properties {soundWaveAssetPtr=&Spot_Wind_Autumnal_Leaves_Short_01}}
 {id=WavePlayer_39, properties {soundWaveAssetPtr=&Spot_Wind_Autumnal_Leaves_Short_05}}
 {id=WavePlayer_40, properties {soundWaveAssetPtr=&Spot_Wind_Autumnal_Leaves_Short_09}}
 {id=WavePlayer_41, properties {soundWaveAssetPtr=&Spot_Wind_Autumnal_Leaves_Short_13}}
]
properties {
 attenuationSettings=/Game/Sounds/Attenuation/Fort_TileAmbient_OneShot_Wider_Att
 bEnableBusSends=false
 maxDistance=8000
 soundClassObject=&Fort_Ambient_OneShots_Outdoors
 volumeMultiplier=0.251
 preEffectBusSends [{}]
 soundSubmixSends [{
   maxSendDistance=6000
   minSendDistance=2000
   minSendLevel=0
   soundSubmix=&Fortnite_Reverb_Default
   customSendLevelCurve {editorCurveData {keys [<at 0 0>,<at 1 0.6>]}}
  }]
}
