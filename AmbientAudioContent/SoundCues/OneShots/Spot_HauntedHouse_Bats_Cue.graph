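# Sound cue for one-shot bat spot sounds, named for a haunted house ambience.
# A Random node picks one of five equally weighted cave bat waves, and a Modulator at the root processes the pick.
# Playback routes through an interior-space ambient sound class and references a closer one-shot tile ambient attenuation asset.
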SoundCue /AmbientAudioContent/SoundCues/OneShots/Spot_HauntedHouse_Bats_Cue
root=Modulator_0
nodes [
 {id=Modulator_0, children [Random_1]}
 {id=Random_1, children [WavePlayer_5,WavePlayer_6,WavePlayer_7,WavePlayer_8,WavePlayer_9], properties {weights [1,1,1,1,1]}}
 {id=WavePlayer_5, properties {soundWaveAssetPtr=/AmbientAudioContent/Waves/OneShots/Spot_Cattus_Cave_Bat_05}}
 {id=WavePlayer_6, properties {soundWaveAssetPtr=/AmbientAudioContent/Waves/OneShots/Spot_Cattus_Cave_Bat_04}}
 {id=WavePlayer_7, properties {soundWaveAssetPtr=/AmbientAudioContent/Waves/OneShots/Spot_Cattus_Cave_Bat_03}}
 {id=WavePlayer_8, properties {soundWaveAssetPtr=/AmbientAudioContent/Waves/OneShots/Spot_Cattus_Cave_Bat_02}}
 {id=WavePlayer_9, properties {soundWaveAssetPtr=/AmbientAudioContent/Waves/OneShots/Spot_Cattus_Cave_Bat_01}}
]
properties {
 attenuationSettings=/Game/Sounds/Attenuation/Fort_TileAmbient_OneShot_Att_Closer
 maxDistance=8500
 soundClassObject=/Game/Sounds/FortSoundClassesAndModes/Fort_Ambient_InteriorSpace_Duck_TileMix_PSM
 volumeMultiplier=1
 attenuationOverrides {bAttenuate=false}
}
