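# One-shot ambient SoundCue for haunted house wolf sounds, using a wide tile-ambient attenuation and the ambient one-shots sound class.
# A Random node picks among five equally weighted wolf wave variations.
# All four Quality Level inputs share that Random node, so every quality tier gets the same content.
# The cue sends to the exterior ambient source bus and defines a distant reverb submix send.
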SoundCue /AmbientAudioContent/SoundCues/OneShots/Spot_HauntedHouse_Wolf_Cue
root=QualityLevel_0
nodes [
 {id=QualityLevel_0, children [Random_1,Random_1,Random_1,Random_1]}
 {id=Random_1, children [WavePlayer_3,WavePlayer_4,WavePlayer_5,WavePlayer_6,WavePlayer_7], properties {weights [1,1,1,1,1]}}
 {id=WavePlayer_3, properties {soundWaveAssetPtr=/AmbientAudioContent/Waves/OneShots/Spot_HauntedHouse_Wolf_01}}
 {id=WavePlayer_4, properties {soundWaveAssetPtr=/AmbientAudioContent/Waves/OneShots/Spot_HauntedHouse_Wolf_02}}
 {id=WavePlayer_5, properties {soundWaveAssetPtr=/AmbientAudioContent/Waves/OneShots/Spot_HauntedHouse_Wolf_03}}
 {id=WavePlayer_6, properties {soundWaveAssetPtr=/AmbientAudioContent/Waves/OneShots/Spot_HauntedHouse_Wolf_04}}
 {id=WavePlayer_7, properties {soundWaveAssetPtr=/AmbientAudioContent/Waves/OneShots/Spot_HauntedHouse_Wolf_05}}
]
properties {
 attenuationSettings=/Game/Sounds/Attenuation/Fort_TileAmbient_OneShot_Wider_Att
 maxDistance=8000
 soundClassObject=&Fort_Ambient_OneShots
 volumeMultiplier=0.702
 busSends [{audioBus=null, minSendLevel=0, sendLevel=1, soundSourceBus=&AmbientAudio_Exterior_SourceBus}]
 soundSubmixSends [{
   maxSendDistance=8000
   maxSendLevel=0.6
   minSendDistance=5000
   minSendLevel=0.2
   sendLevel=0
   soundSubmix=&Fortnite_Reverb_Distant
  }]
}
