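# Sound cue for one-shot metal drip ambience, named for Iceberg Isles, in the outdoor ambient sound class.
# A Random node picks one of twelve equally weighted shipping container interior drip waves and feeds a Modulator that varies each play, with a raised pitch ceiling.
# A shared positional one-shot attenuation asset spatializes it, and a concurrency override caps simultaneous drips at two.
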SoundCue /AmbientAudioContent/SoundCues/OneShots/Spot_IcebergIsles_Metal_Drip_Singles_Cue
root=Modulator_0
nodes [
 {id=Modulator_0, children [Random_0], properties {pitchMax=1.1}}
 {
  id=Random_0
  children [
   WavePlayer_0
   WavePlayer_1
   WavePlayer_2
   WavePlayer_3
   WavePlayer_4
   WavePlayer_5
   WavePlayer_6
   WavePlayer_7
   WavePlayer_8
   WavePlayer_9
   WavePlayer_10
   WavePlayer_11
  ]
  properties {weights [1,1,1,1,1,1,1,1,1,1,1,1]}
 }
 {id=WavePlayer_0, properties {soundWaveAssetPtr=/AmbientAudioContent/Waves/OneShots/Spot_Shipping_Container_Interior_Drip_01}}
 {id=WavePlayer_1, properties {soundWaveAssetPtr=/AmbientAudioContent/Waves/OneShots/Spot_Shipping_Container_Interior_Drip_02}}
 {id=WavePlayer_10, properties {soundWaveAssetPtr=/AmbientAudioContent/Waves/OneShots/Spot_Shipping_Container_Interior_Drip_11}}
 {id=WavePlayer_11, properties {soundWaveAssetPtr=/AmbientAudioContent/Waves/OneShots/Spot_Shipping_Container_Interior_Drip_12}}
 {id=WavePlayer_2, properties {soundWaveAssetPtr=/AmbientAudioContent/Waves/OneShots/Spot_Shipping_Container_Interior_Drip_03}}
 {id=WavePlayer_3, properties {soundWaveAssetPtr=/AmbientAudioContent/Waves/OneShots/Spot_Shipping_Container_Interior_Drip_04}}
 {id=WavePlayer_4, properties {soundWaveAssetPtr=/AmbientAudioContent/Waves/OneShots/Spot_Shipping_Container_Interior_Drip_05}}
 {id=WavePlayer_5, properties {soundWaveAssetPtr=/AmbientAudioContent/Waves/OneShots/Spot_Shipping_Container_Interior_Drip_06}}
 {id=WavePlayer_6, properties {soundWaveAssetPtr=/AmbientAudioContent/Waves/OneShots/Spot_Shipping_Container_Interior_Drip_07}}
 {id=WavePlayer_7, properties {soundWaveAssetPtr=/AmbientAudioContent/Waves/OneShots/Spot_Shipping_Container_Interior_Drip_08}}
 {id=WavePlayer_8, properties {soundWaveAssetPtr=/AmbientAudioContent/Waves/OneShots/Spot_Shipping_Container_Interior_Drip_09}}
 {id=WavePlayer_9, properties {soundWaveAssetPtr=/AmbientAudioContent/Waves/OneShots/Spot_Shipping_Container_Interior_Drip_10}}
]
properties {
 attenuationSettings=/Game/Sounds/Attenuation/Fort_TileAmbient_OneShot_MorePositional_Att
 bEnableBusSends=false
 bEnableSubmixSends=false
 bOverrideConcurrency=true
 maxDistance=8000
 soundClassObject=&Fort_Ambient_Outdoors_Athena
 volumeMultiplier=1.504
 attenuationOverrides {attenuationShape=Capsule, falloffDistance=200, stereoSpread=10000, attenuationShapeExtents=<236.449 200>}
 concurrencyOverrides {maxCount=2}
}
