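# Sound cue for a medium-running motor one-shot spot ambience in the Iceberg Isles area, using Access Point motor recordings.
# A Random node picks one of four equally weighted wave variations, then a Modulator randomizes its volume within a narrow range.
# Playback uses the Fort ambient Athena attenuation and indoor ambient sound class at a reduced cue volume.
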SoundCue /AmbientAudioContent/SoundCues/OneShots/Spot_IcebergIsles_MotorRunMed_Cue
root=Modulator_0
nodes [
 {id=Modulator_0, children [Random_1], properties {volumeMax=0.9, volumeMin=0.6}}
 {id=Random_1, children [WavePlayer_2,WavePlayer_5,WavePlayer_6,WavePlayer_7], properties {weights [1,1,1,1]}}
 {id=WavePlayer_2, properties {soundWaveAssetPtr=/AmbientAudioContent/Waves/OneShots/Spot_Access_Point_MotorRunMed_01}}
 {id=WavePlayer_5, properties {soundWaveAssetPtr=/AmbientAudioContent/Waves/OneShots/Spot_Access_Point_MotorRunMed_02}}
 {id=WavePlayer_6, properties {soundWaveAssetPtr=/AmbientAudioContent/Waves/OneShots/Spot_Access_Point_MotorRunMed_03}}
 {id=WavePlayer_7, properties {soundWaveAssetPtr=/AmbientAudioContent/Waves/OneShots/Spot_Access_Point_MotorRunMed_04}}
]
properties {
 attenuationSettings=&Fort_Ambient_Athena_Att
 maxDistance=2097152
 soundClassObject=&Fort_Ambient_Indoors_Athena
 volumeMultiplier=0.276
 preEffectBusSends [{}]
}
