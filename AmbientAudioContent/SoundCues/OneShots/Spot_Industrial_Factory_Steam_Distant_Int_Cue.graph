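# Sound Cue for a distant industrial factory steam one-shot, played as indoor ambient spot audio.
# A Random node picks one of five distant industrial waves with equal weight and feeds a Modulator that varies playback, with a lowered pitch minimum.
# Output is quiet, uses the indoor ambient one-shot sound class and a less-positional attenuation with falloff, spread and shape overrides, and has bus and submix sends disabled.
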SoundCue /AmbientAudioContent/SoundCues/OneShots/Spot_Industrial_Factory_Steam_Distant_Int_Cue
root=Modulator_0
nodes [
 {id=Modulator_0, children [Random_1], properties {pitchMin=0.9}}
 {id=Random_1, children [WavePlayer_1,WavePlayer_2,WavePlayer_3,WavePlayer_4,WavePlayer_5], properties {weights [1,1,1,1,1]}}
 {id=WavePlayer_1, properties {soundWaveAssetPtr=/AmbientAudioContent/Waves/OneShots/Spot_Amb_Industrial_Distant_01}}
 {id=WavePlayer_2, properties {soundWaveAssetPtr=/AmbientAudioContent/Waves/OneShots/Spot_Amb_Industrial_Distant_02}}
 {id=WavePlayer_3, properties {soundWaveAssetPtr=/AmbientAudioContent/Waves/OneShots/Spot_Amb_Industrial_Distant_06}}
 {id=WavePlayer_4, properties {soundWaveAssetPtr=/AmbientAudioContent/Waves/OneShots/Spot_Amb_Industrial_Distant_14}}
 {id=WavePlayer_5, properties {soundWaveAssetPtr=/AmbientAudioContent/Waves/OneShots/Spot_Amb_Industrial_Distant_16}}
]
properties {
 attenuationSettings=/Game/Sounds/Attenuation/Fort_TileAmbient_OneShot_LessPositional_Log_Att
 bEnableBusSends=false
 bEnableSubmixSends=false
 maxDistance=6200
 soundClassObject=/Game/Sounds/FortSoundClassesAndModes/Fort_Ambient_OneShots_Indoors
 volumeMultiplier=0.251
 attenuationOverrides {falloffDistance=6000, stereoSpread=10000, attenuationShapeExtents {x=2000}}
}
