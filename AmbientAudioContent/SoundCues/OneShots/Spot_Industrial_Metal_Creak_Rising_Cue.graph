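# Sound cue for an industrial ambient one-shot of rising metal creaks.
# A Random node picks one of seven equally weighted creak waves, and a Modulator with a lowered minimum pitch varies each play.
# It uses a distant metal-movement attenuation asset from the junkyard sounds and the outdoor ambient sound class, with bus and submix sends disabled.
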SoundCue /AmbientAudioContent/SoundCues/OneShots/Spot_Industrial_Metal_Creak_Rising_Cue
root=Modulator_0
nodes [
 {id=Modulator_0, children [Random_0], properties {pitchMin=0.9}}
 {
  id=Random_0
  children [WavePlayer_6, WavePlayer_7, WavePlayer_8, WavePlayer_9, WavePlayer_10, WavePlayer_11, WavePlayer_12]
  properties {weights [1,1,1,1,1,1,1]}
 }
 {id=WavePlayer_10, properties {soundWaveAssetPtr=/AmbientAudioContent/Waves/OneShots/Spot_Metal_Creak_Rising_04}}
 {id=WavePlayer_11, properties {soundWaveAssetPtr=/AmbientAudioContent/Waves/OneShots/Spot_Metal_Creak_Rising_03}}
 {id=WavePlayer_12, properties {soundWaveAssetPtr=/AmbientAudioContent/Waves/OneShots/Spot_Metal_Creak_Rising_02}}
 {id=WavePlayer_6, properties {soundWaveAssetPtr=/AmbientAudioContent/Waves/OneShots/Spot_Metal_Creak_Rising_01}}
 {id=WavePlayer_7, properties {soundWaveAssetPtr=/AmbientAudioContent/Waves/OneShots/Spot_Metal_Creak_Rising_07}}
 {id=WavePlayer_8, properties {soundWaveAssetPtr=/AmbientAudioContent/Waves/OneShots/Spot_Metal_Creak_Rising_06}}
 {id=WavePlayer_9, properties {soundWaveAssetPtr=/AmbientAudioContent/Waves/OneShots/Spot_Metal_Creak_Rising_05}}
]
properties {
 attenuationSettings=/Game/Sounds/Asteria/POI/JunkYard/Spots/Spot_Metal_Movements_Distant_Att
 bEnableBusSends=false
 bEnableSubmixSends=false
 maxDistance=7500
 soundClassObject=&Fort_Ambient_Outdoors_Athena
 volumeMultiplier=0.752
 attenuationOverrides {falloffDistance=6000, stereoSpread=10000, attenuationShapeExtents {x=2000}}
}
