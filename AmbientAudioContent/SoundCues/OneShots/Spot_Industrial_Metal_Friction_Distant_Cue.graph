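# Sound cue for a distant industrial metal friction one-shot, classed as outdoor ambient audio.
# A Random node picks among six equally weighted distant industrial waves, then a Modulator varies playback with a lowered minimum pitch.
# Output plays at reduced volume through a shared wide one-shot attenuation asset named for air absorption, with bus and submix sends disabled.
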SoundCue /AmbientAudioContent/SoundCues/OneShots/Spot_Industrial_Metal_Friction_Distant_Cue
root=Modulator_0
nodes [
 {id=Modulator_0, children [Random_1], properties {pitchMin=0.9}}
 {
  id=Random_1
  children [WavePlayer_6, WavePlayer_7, WavePlayer_8, WavePlayer_9, WavePlayer_10, WavePlayer_11]
  properties {weights [1,1,1,1,1,1]}
 }
 {id=WavePlayer_10, properties {soundWaveAssetPtr=/AmbientAudioContent/Waves/OneShots/Spot_Amb_Industrial_Distant_18}}
 {id=WavePlayer_11, properties {soundWaveAssetPtr=/AmbientAudioContent/Waves/OneShots/Spot_Amb_Industrial_Distant_17}}
 {id=WavePlayer_6, properties {soundWaveAssetPtr=/AmbientAudioContent/Waves/OneShots/Spot_Amb_Industrial_Distant_10}}
 {id=WavePlayer_7, properties {soundWaveAssetPtr=/AmbientAudioContent/Waves/OneShots/Spot_Amb_Industrial_Distant_11}}
 {id=WavePlayer_8, properties {soundWaveAssetPtr=/AmbientAudioContent/Waves/OneShots/Spot_Amb_Industrial_Distant_12}}
 {id=WavePlayer_9, properties {soundWaveAssetPtr=/AmbientAudioContent/Waves/OneShots/Spot_Amb_Industrial_Distant_13}}
]
properties {
 attenuationSettings=/Game/Sounds/Attenuation/Fort_TileAmbient_OneShot_Wider_AirAbsorbtion_Att
 bEnableBusSends=false
 bEnableSubmixSends=false
 maxDistance=8000
 soundClassObject=&Fort_Ambient_Outdoors_Athena
 volumeMultiplier=0.251
 attenuationOverrides {falloffDistance=6000, stereoSpread=10000, attenuationShapeExtents {x=2000}}
}
