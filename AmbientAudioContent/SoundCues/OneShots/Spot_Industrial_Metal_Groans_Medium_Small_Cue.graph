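# Sound Cue for one-shot small and medium industrial metal groans, classed as Athena outdoor ambience.
# An equally weighted Random node picks one of eleven groan waves, mostly small variants, and feeds a root Modulator that varies playback with a lowered pitch minimum.
# It references a distant metal-movement attenuation asset from the Asteria JunkYard POI, with bus and submix sends disabled.
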SoundCue /AmbientAudioContent/SoundCues/OneShots/Spot_Industrial_Metal_Groans_Medium_Small_Cue
root=Modulator_0
nodes [
 {id=Modulator_0, children [Random_0], properties {pitchMin=0.9}}
 {
  id=Random_0
  children [
   WavePlayer_10
   WavePlayer_11
   WavePlayer_12
   WavePlayer_13
   WavePlayer_15
   WavePlayer_16
   WavePlayer_17
   WavePlayer_18
   WavePlayer_19
   WavePlayer_20
   WavePlayer_21
  ]
  properties {weights [1,1,1,1,1,1,1,1,1,1,1]}
 }
 {id=WavePlayer_10, properties {soundWaveAssetPtr=/AmbientAudioContent/Waves/OneShots/Spot_Metal_Groan_Small_02}}
 {id=WavePlayer_11, properties {soundWaveAssetPtr=/AmbientAudioContent/Waves/OneShots/Spot_Metal_Groan_Small_08}}
 {id=WavePlayer_12, properties {soundWaveAssetPtr=/AmbientAudioContent/Waves/OneShots/Spot_Metal_Groan_Small_07}}
 {id=WavePlayer_13, properties {soundWaveAssetPtr=/AmbientAudioContent/Waves/OneShots/Spot_Metal_Groan_Small_06}}
 {id=WavePlayer_15, properties {soundWaveAssetPtr=/AmbientAudioContent/Waves/OneShots/Spot_Metal_Groan_Small_05}}
 {id=WavePlayer_16, properties {soundWaveAssetPtr=/AmbientAudioContent/Waves/OneShots/Spot_Metal_Groan_Small_04}}
 {id=WavePlayer_17, properties {soundWaveAssetPtr=/AmbientAudioContent/Waves/OneShots/Spot_Metal_Groan_Small_03}}
 {id=WavePlayer_18, properties {soundWaveAssetPtr=/AmbientAudioContent/Waves/OneShots/Spot_Metal_Groan_Small_01}}
 {id=WavePlayer_19, properties {soundWaveAssetPtr=/AmbientAudioContent/Waves/OneShots/Spot_Metal_Groan_Medium_01}}
 {id=WavePlayer_20, properties {soundWaveAssetPtr=/AmbientAudioContent/Waves/OneShots/Spot_Metal_Groan_Medium_02}}
 {id=WavePlayer_21, properties {soundWaveAssetPtr=/AmbientAudioContent/Waves/OneShots/Spot_Metal_Groan_Medium_03}}
]
properties {
 attenuationSettings=/Game/Sounds/Asteria/POI/JunkYard/Spots/Spot_Metal_Movements_Distant_Att
 bEnableBusSends=false
 bEnableSubmixSends=false
 maxDistance=7500
 soundClassObject=&Fort_Ambient_Outdoors_Athena
 volumeMultiplier=0.752
 attenuationOverrides {falloffDistance=6000, stereoSpread=10000, attenuationShapeExtents {x=2000}}
}
