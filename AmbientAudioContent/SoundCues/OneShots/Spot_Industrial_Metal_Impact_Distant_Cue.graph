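# One-shot ambient SoundCue for distant industrial metal impacts, routed to the Athena outdoor ambient sound class.
# A Random node picks among ten equally weighted distant industrial spot waves.
# A Modulator at the root varies the pick's pitch on each play, with a lowered minimum, to reduce repetition.
# The cue plays at reduced volume through a shared less-positional logarithmic one-shot attenuation asset, with bus and submix sends disabled.
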SoundCue /AmbientAudioContent/SoundCues/OneShots/Spot_Industrial_Metal_Impact_Distant_Cue
root=Modulator_0
nodes [
 {id=Modulator_0, children [Random_1], properties {pitchMin=0.9}}
 {
  id=Random_1
  children [
   WavePlayer_0
   WavePlayer_1
   WavePlayer_2
   WavePlayer_3
   WavePlayer_4
   WavePlayer_5
   WavePlayer_16
   WavePlayer_17
   WavePlayer_18
   WavePlayer_19
  ]
  properties {weights [1,1,1,1,1,1,1,1,1,1]}
 }
 {id=WavePlayer_0, properties {soundWaveAssetPtr=/AmbientAudioContent/Waves/OneShots/Spot_Amb_Industrial_Distant_21}}
 {id=WavePlayer_1, properties {soundWaveAssetPtr=/AmbientAudioContent/Waves/OneShots/Spot_Amb_Industrial_Distant_09}}
 {id=WavePlayer_16, properties {soundWaveAssetPtr=/AmbientAudioContent/Waves/OneShots/Spot_Amb_Industrial_Distant_08}}
 {id=WavePlayer_17, properties {soundWaveAssetPtr=/AmbientAudioContent/Waves/OneShots/Spot_Amb_Industrial_Distant_05}}
 {id=WavePlayer_18, properties {soundWaveAssetPtr=/AmbientAudioContent/Waves/OneShots/Spot_Amb_Industrial_Distant_04}}
 {id=WavePlayer_19, properties {soundWaveAssetPtr=/AmbientAudioContent/Waves/OneShots/Spot_Amb_Industrial_Distant_03}}
 {id=WavePlayer_2, properties {soundWaveAssetPtr=/AmbientAudioContent/Waves/OneShots/Spot_Amb_Industrial_Distant_15}}
 {id=WavePlayer_3, properties {soundWaveAssetPtr=/AmbientAudioContent/Waves/OneShots/Spot_Amb_Industrial_Distant_19}}
 {id=WavePlayer_4, properties {soundWaveAssetPtr=/AmbientAudioContent/Waves/OneShots/Spot_Amb_Industrial_Distant_20}}
 {id=WavePlayer_5, properties {soundWaveAssetPtr=/AmbientAudioContent/Waves/OneShots/Spot_Amb_Industrial_Distant_22}}
]
properties {
 attenuationSettings=/Game/Sounds/Attenuation/Fort_TileAmbient_OneShot_LessPositional_Log_Att
 bEnableBusSends=false
 bEnableSubmixSends=false
 maxDistance=6200
 soundClassObject=&Fort_Ambient_Outdoors_Athena
 volumeMultiplier=0.301
 attenuationOverrides {falloffDistance=6000, stereoSpread=10000, attenuationShapeExtents {x=2000}}
}
