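# Sound cue for outdoor ambient one-shot insect spot sounds, playing Japanese cricket, locust and chirp calls.
# A top-level Random picks one of five species groups, each a per-group volume Mixer over a Random of three or four wave variations.
# The pick passes through a Modulator, shared by all four QualityLevel branches.
# Insect fly-by attenuation, a two-voice-max concurrency set, low overall volume and a reverb submix send shape playback.
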SoundCue /AmbientAudioContent/SoundCues/OneShots/Spot_Insect_Japanese_Cricket_Chirps_Cue
root=QualityLevel_0
nodes [
 {id=Mixer_0, children [Random_5], properties {inputVolume [0.66]}}
 {id=Mixer_1, children [Random_4], properties {inputVolume [1.75]}}
 {id=Mixer_2, children [Random_0], properties {inputVolume [1.25]}}
 {id=Mixer_3, children [Random_3], properties {inputVolume [1.1]}}
 {id=Mixer_4, children [Random_6], properties {inputVolume [0.66]}}
 {id=Modulator_0, children [Random_1]}
 {id=QualityLevel_0, children [Modulator_0,Modulator_0,Modulator_0,Modulator_0]}
 {id=Random_0, children [WavePlayer_0,WavePlayer_1,WavePlayer_2], properties {weights [1,1,1]}}
 {id=Random_1, children [Mixer_2,Mixer_3,Mixer_1,Mixer_0,Mixer_4], properties {weights [1,1,1,1,1]}}
 {id=Random_3, children [WavePlayer_3,WavePlayer_4,WavePlayer_5,WavePlayer_6], properties {weights [1,1,1,1]}}
 {id=Random_4, children [WavePlayer_7, WavePlayer_20, WavePlayer_21], properties {weights [1,1,1]}}
 {id=Random_5, children [WavePlayer_22, WavePlayer_23, WavePlayer_24, WavePlayer_25], properties {weights [1,1,1,1]}}
 {id=Random_6, children [WavePlayer_26, WavePlayer_27, WavePlayer_28, WavePlayer_29], properties {weights [1,1,1,1]}}
 {id=WavePlayer_0, properties {soundWaveAssetPtr=/AmbientAudioContent/Waves/OneShots/Spot_Insect_Mitsukado_Cricket_01}}
 {id=WavePlayer_1, properties {soundWaveAssetPtr=/AmbientAudioContent/Waves/OneShots/Spot_Insect_Mitsukado_Cricket_03}}
 {id=WavePlayer_2, properties {soundWaveAssetPtr=/AmbientAudioContent/Waves/OneShots/Spot_Insect_Mitsukado_Cricket_02}}
 {id=WavePlayer_20, properties {soundWaveAssetPtr=/AmbientAudioContent/Waves/OneShots/Spot_Insect_Japanese_Cricket_Chirp_Kantan_Chirp_02}}
 {id=WavePlayer_21, properties {soundWaveAssetPtr=/AmbientAudioContent/Waves/OneShots/Spot_Insect_Japanese_Cricket_Chirp_Kantan_Chirp_03}}
 {id=WavePlayer_22, properties {soundWaveAssetPtr=/AmbientAudioContent/Waves/OneShots/Spot_Insect_Japanese_Chirp_Call_01}}
 {id=WavePlayer_23, properties {soundWaveAssetPtr=/AmbientAudioContent/Waves/OneShots/Spot_Insect_Japanese_Chirp_Call_04}}
 {id=WavePlayer_24, properties {soundWaveAssetPtr=/AmbientAudioContent/Waves/OneShots/Spot_Insect_Japanese_Chirp_Call_02}}
 {id=WavePlayer_25, properties {soundWaveAssetPtr=/AmbientAudioContent/Waves/OneShots/Spot_Insect_Japanese_Chirp_Call_03}}
 {id=WavePlayer_26, properties {soundWaveAssetPtr=/AmbientAudioContent/Waves/OneShots/Spot_Insect_Hillside_Chirp_Song_01}}
 {id=WavePlayer_27, properties {soundWaveAssetPtr=/AmbientAudioContent/Waves/OneShots/Spot_Insect_Hillside_Chirp_Song_04}}
 {id=WavePlayer_28, properties {soundWaveAssetPtr=/AmbientAudioContent/Waves/OneShots/Spot_Insect_Hillside_Chirp_Song_02}}
 {id=WavePlayer_29, properties {soundWaveAssetPtr=/AmbientAudioContent/Waves/OneShots/Spot_Insect_Hillside_Chirp_Song_03}}
 {id=WavePlayer_3, properties {soundWaveAssetPtr=/AmbientAudioContent/Waves/OneShots/Spot_Insect_Japanese_Swamp_Locust_04}}
 {id=WavePlayer_4, properties {soundWaveAssetPtr=/AmbientAudioContent/Waves/OneShots/Spot_Insect_Japanese_Swamp_Locust_03}}
 {id=WavePlayer_5, properties {soundWaveAssetPtr=/AmbientAudioContent/Waves/OneShots/Spot_Insect_Japanese_Swamp_Locust_02}}
 {id=WavePlayer_6, properties {soundWaveAssetPtr=/AmbientAudioContent/Waves/OneShots/Spot_Insect_Japanese_Swamp_Locust_01}}
 {id=WavePlayer_7, properties {soundWaveAssetPtr=/AmbientAudioContent/Waves/OneShots/Spot_Insect_Japanese_Cricket_Chirp_Kantan_Chirp_01}}
]
properties {
 attenuationSettings=/AmbientAudioBase/Attenuations/Fort_TileAmbient_OneShot_Insect_By_Att
 maxDistance=6000
 soundClassObject=&Fort_Ambient_OneShots_Outdoors
 volumeMultiplier=0.15
 concurrencySet [/AmbientAudioBase/Concurrency/AmbientConc_2Max_Spot_Concurrency]
 preEffectBusSends [{minSendLevel=1}]
 soundSubmixSends [{minSendLevel=0.4, soundSubmix=&Fortnite_Reverb_Default}]
}
