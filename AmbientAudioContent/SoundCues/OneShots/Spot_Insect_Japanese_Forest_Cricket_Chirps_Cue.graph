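# Sound cue for spot one-shot insect chirps of a Japanese forest, played as quiet, attenuated outdoor ambience.
# A random node picks one of three species groups (Enma Koorogi cricket, Mitsukado cricket, Sashigame hemiptera), each randomizing among three waves, with mixers boosting the latter two.
# The result passes through a modulator shared by all quality levels.
# A concurrency set limits overlap, and a distance-scaled send feeds the default reverb submix.
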SoundCue /AmbientAudioContent/SoundCues/OneShots/Spot_Insect_Japanese_Forest_Cricket_Chirps_Cue
root=QualityLevel_0
nodes [
 {id=Mixer_0, children [Random_8], properties {inputVolume [2]}}
 {id=Mixer_1, children [Random_2], properties {inputVolume [1.75]}}
 {id=Modulator_0, children [Random_7]}
 {id=QualityLevel_0, children [Modulator_0,Modulator_0,Modulator_0,Modulator_0]}
 {id=Random_1, children [WavePlayer_8, WavePlayer_9, WavePlayer_10], properties {weights [1,1,1]}}
 {id=Random_2, children [WavePlayer_11, WavePlayer_12, WavePlayer_13], properties {weights [1,1,1]}}
 {id=Random_7, children [Random_1,Mixer_1,Mixer_0], properties {weights [1,1,1]}}
 {id=Random_8, children [WavePlayer_14, WavePlayer_15, WavePlayer_16], properties {weights [1,1,1]}}
 {
  id=WavePlayer_10
  properties {soundWaveAssetPtr=/AmbientAudioContent/Waves/OneShots/Spot_Insect_Japanese_Forest_Night_Cricket_Enma_Koorogi_04}
 }
 {id=WavePlayer_11, properties {soundWaveAssetPtr=/AmbientAudioContent/Waves/OneShots/Spot_Insect_Mitsukado_Cricket_01}}
 {id=WavePlayer_12, properties {soundWaveAssetPtr=/AmbientAudioContent/Waves/OneShots/Spot_Insect_Mitsukado_Cricket_02}}
 {id=WavePlayer_13, properties {soundWaveAssetPtr=/AmbientAudioContent/Waves/OneShots/Spot_Insect_Mitsukado_Cricket_03}}
 {
  id=WavePlayer_14
  properties {soundWaveAssetPtr=/AmbientAudioContent/Waves/OneShots/Spot_Insect_Japanese_Kiiro_Forest_Sashigame_Hempitera_03}
 }
 {
  id=WavePlayer_15
  properties {soundWaveAssetPtr=/AmbientAudioContent/Waves/OneShots/Spot_Insect_Japanese_Kiiro_Forest_Sashigame_Hempitera_02}
 }
 {
  id=WavePlayer_16
  properties {soundWaveAssetPtr=/AmbientAudioContent/Waves/OneShots/Spot_Insect_Japanese_Kiiro_Forest_Sashigame_Hempitera_01}
 }
 {
  id=WavePlayer_8
  properties {soundWaveAssetPtr=/AmbientAudioContent/Waves/OneShots/Spot_Insect_Japanese_Forest_Night_Cricket_Enma_Koorogi_06}
 }
 {
  id=WavePlayer_9
  properties {soundWaveAssetPtr=/AmbientAudioContent/Waves/OneShots/Spot_Insect_Japanese_Forest_Night_Cricket_Enma_Koorogi_05}
 }
]
properties {
 attenuationSettings=/Game/Sounds/Attenuation/Fort_TileAmbient_OneShot_Insect_By_Att
 maxDistance=6000
 soundClassObject=&Fort_Ambient_OneShots_Outdoors
 volumeMultiplier=0.1
 concurrencySet [/AmbientAudioBase/Concurrency/AmbientConc_2Max_Spot_Concurrency]
 preEffectBusSends [{minSendLevel=1}]
 soundSubmixSends [{
   maxSendDistance=4000
   minSendDistance=1000
   minSendLevel=0.4
   soundSubmix=&Fortnite_Reverb_Default
   customSendLevelCurve {editorCurveData {keys [<at 0 0.4>,<at 1 1>]}}
  }]
}
