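# Outdoor ambient one-shot Sound Cue of long Japanese cicada calls, named for daytime use.
# A weighted Random node, with repeats allowed, picks between the Long_01 and Long_02 waves, favoring the first, and feeds a Modulator at the root.
# Playback uses reduced volume, the bird one-shot attenuation, the outdoor ambient one-shot sound class, and a distance-ranged send to the exterior ambient source bus.
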SoundCue /AmbientAudioContent/SoundCues/OneShots/Spot_Japanese_Cicadas_Long_02_Day_Cue
root=Modulator_0
nodes [
 {id=Modulator_0, children [Random_0]}
 {id=Random_0, children [WavePlayer_4,WavePlayer_5], properties {bRandomizeWithoutReplacement=false, weights [1,0.75]}}
 {id=WavePlayer_4, properties {soundWaveAssetPtr=/AmbientAudioContent/Waves/OneShots/Spot_Japanese_Cicadas_Long_01}}
 {id=WavePlayer_5, properties {soundWaveAssetPtr=/AmbientAudioContent/Waves/OneShots/Spot_Japanese_Cicadas_Long_02}}
]
properties {
 attenuationSettings=&Fort_TileAmbient_OneShot_Bird_Att
 maxDistance=8005
 soundClassObject=&Fort_Ambient_OneShots_Outdoors
 volumeMultiplier=0.6
 busSends [{
   audioBus=null
   maxSendDistance=6000
   minSendDistance=3000
   minSendLevel=0.6
   sendLevel=1
   soundSourceBus=&AmbientAudio_Exterior_SourceBus
  }]
}
