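# Sound cue for long daytime Japanese cicada one-shot spots in outdoor ambience, using bird one-shot attenuation and the outdoor one-shots sound class.
# A quality switch sends its first two levels to a weighted random pick, repeats allowed, of five waves, and its last two to the first wave alone.
# Each branch passes through a modulator, and the cue feeds an exterior ambience source bus with a distance-based send.
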SoundCue /AmbientAudioContent/SoundCues/OneShots/Spot_Japanese_Cicadas_Long_Day_Cue
root=QualityLevel_0
nodes [
 {id=Modulator_0, children [Random_0]}
 {id=Modulator_1, children [WavePlayer_1]}
 {id=QualityLevel_0, children [Modulator_0,Modulator_0,Modulator_1,Modulator_1]}
 {
  id=Random_0
  children [WavePlayer_0,WavePlayer_2,WavePlayer_3,WavePlayer_8,WavePlayer_9]
  properties {bRandomizeWithoutReplacement=false, weights [1,0.75,1,1,1]}
 }
 {id=WavePlayer_0, properties {soundWaveAssetPtr=/AmbientAudioContent/Waves/OneShots/Spot_Insect_Cicada_Day_Long_01}}
 {id=WavePlayer_1, properties {soundWaveAssetPtr=&Spot_Insect_Cicada_Day_Long_01}}
 {id=WavePlayer_2, properties {soundWaveAssetPtr=/AmbientAudioContent/Waves/OneShots/Spot_Insect_Cicada_Day_Long_02}}
 {id=WavePlayer_3, properties {soundWaveAssetPtr=/AmbientAudioContent/Waves/OneShots/Spot_Insect_Cicada_Day_Long_03}}
 {id=WavePlayer_8, properties {soundWaveAssetPtr=/AmbientAudioContent/Waves/OneShots/Spot_Insect_Cicada_Day_Long_04}}
 {id=WavePlayer_9, properties {soundWaveAssetPtr=/AmbientAudioContent/Waves/OneShots/Spot_Insect_Cicada_Day_Long_05}}
]
properties {
 attenuationSettings=&Fort_TileAmbient_OneShot_Bird_Att
 maxDistance=8005
 soundClassObject=&Fort_Ambient_OneShots_Outdoors
 volumeMultiplier=0.6
 busSends [{
   audioBus=null
   maxSendDistance=6000
   minSendDistance=3000
   minSendLevel=0.6
   sendLevel=1
   soundSourceBus=&AmbientAudio_Exterior_SourceBus
  }]
}
