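# Japanese night forest one-shot ambience cue that plays random, low-volume, mostly pitched-down calls from 32 bird and insect waves.
# The first two quality levels usually play one modulated call, occasionally mixing in a quieter, lower-pitched second call after a short random delay.
# The other two levels play only a single modulated call.
# It uses bird one-shot attenuation, with distance-based sends to the exterior ambience bus and reverb submix.
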SoundCue /AmbientAudioContent/SoundCues/OneShots/Spot_Japanese_Night_Forest_Low_Calls_Cue
root=QualityLevel_0
nodes [
 {id=Delay_0, children [Modulator_0], properties {delayMax=3, delayMin=2}}
 {id=Mixer_0, children [Modulator_1,Delay_0], properties {inputVolume [1,0.75]}}
 {id=Modulator_0, children [Random_1], properties {pitchMax=0.9, pitchMin=0.8, volumeMax=0.7, volumeMin=0.6}}
 {id=Modulator_1, children [Random_1], properties {pitchMax=1, pitchMin=0.7, volumeMax=1.1, volumeMin=0.9}}
 {id=Modulator_2, children [Random_1], properties {pitchMax=1.1, volumeMax=1.1, volumeMin=0.9}}
 {id=QualityLevel_0, children [Random_5,Random_5,Modulator_2,Modulator_2]}
 {id=Random_1, children [Random_2], properties {weights [1]}}
 {
  id=Random_2
  children [
   WavePlayer_0
   WavePlayer_8
   WavePlayer_9
   WavePlayer_13
   WavePlayer_16
   WavePlayer_17
   WavePlayer_19
   WavePlayer_20
   WavePlayer_21
   WavePlayer_22
   WavePlayer_26
   WavePlayer_27
   WavePlayer_28
   WavePlayer_29
   WavePlayer_30
   WavePlayer_31
   WavePlayer_34
   WavePlayer_35
   WavePlayer_37
   WavePlayer_38
   WavePlayer_39
   WavePlayer_40
   WavePlayer_41
   WavePlayer_42
   WavePlayer_43
   WavePlayer_48
   WavePlayer_49
   WavePlayer_50
   WavePlayer_52
   WavePlayer_53
   WavePlayer_54
   WavePlayer_55
  ]
  properties {weights [1,1,1,1,1,1,1,1,1,1,1,1,1,1,1,1,1,1,1,1,1,1,1,1,1,1,1,1,1,1,1,1]}
 }
 {id=Random_5, children [Modulator_1,Mixer_0], properties {weights [1,0.33]}}
 {id=WavePlayer_0, properties {soundWaveAssetPtr=/AmbientAudioContent/Waves/OneShots/Spot_Bird_Japanese_Akita_Chirp_01_NF}}
 {id=WavePlayer_13, properties {soundWaveAssetPtr=/AmbientAudioContent/Waves/OneShots/Spot_Bird_Japanese_Blakiston_Fish_Owl_01_NF}}
 {id=WavePlayer_16, properties {soundWaveAssetPtr=/AmbientAudioContent/Waves/OneShots/Spot_Bird_Japanese_Blue_White_Flycatcher_Call_01_NF}}
 {id=WavePlayer_17, properties {soundWaveAssetPtr=/AmbientAudioContent/Waves/OneShots/Spot_Bird_Japanese_Blue_White_Flycatcher_Call_02_NF}}
 {id=WavePlayer_19, properties {soundWaveAssetPtr=/AmbientAudioContent/Waves/OneShots/Spot_Bird_Japanese_Crane_01_NF}}
 {id=WavePlayer_20, properties {soundWaveAssetPtr=/AmbientAudioContent/Waves/OneShots/Spot_Bird_Japanese_Crane_02_NF}}
 {id=WavePlayer_21, properties {soundWaveAssetPtr=/AmbientAudioContent/Waves/OneShots/Spot_Bird_Japanese_Crane_03_NF}}
 {id=WavePlayer_22, properties {soundWaveAssetPtr=/AmbientAudioContent/Waves/OneShots/Spot_Bird_Japanese_Crane_04_NF}}
 {id=WavePlayer_26, properties {soundWaveAssetPtr=/AmbientAudioContent/Waves/OneShots/Spot_Bird_Japanese_Crow_Parkland_01_NF}}
 {id=WavePlayer_27, properties {soundWaveAssetPtr=/AmbientAudioContent/Waves/OneShots/Spot_Bird_Japanese_Crow_Parkland_02_NF}}
 {id=WavePlayer_28, properties {soundWaveAssetPtr=/AmbientAudioContent/Waves/OneShots/Spot_Bird_Japanese_Cuckoo_Call_01_NF}}
 {id=WavePlayer_29, properties {soundWaveAssetPtr=/AmbientAudioContent/Waves/OneShots/Spot_Bird_Japanese_Cuckoo_Call_02_NF}}
 {id=WavePlayer_30, properties {soundWaveAssetPtr=/AmbientAudioContent/Waves/OneShots/Spot_Bird_Japanese_Cuckoo_Forest_01_NF}}
 {id=WavePlayer_31, properties {soundWaveAssetPtr=/AmbientAudioContent/Waves/OneShots/Spot_Bird_Japanese_Cuckoo_Forest_02_NF}}
 {id=WavePlayer_34, properties {soundWaveAssetPtr=/AmbientAudioContent/Waves/OneShots/Spot_Bird_Japanese_Dove_Coo_01_NF}}
 {id=WavePlayer_35, properties {soundWaveAssetPtr=/AmbientAudioContent/Waves/OneShots/Spot_Bird_Japanese_Dove_Coo_02_NF}}
 {
  id=WavePlayer_37
  properties {soundWaveAssetPtr=/AmbientAudioContent/Waves/OneShots/Spot_Bird_Japanese_Forest_Evening_Kiiro_Sashigame_Hempitera_Insect_01_NF}
 }
 {
  id=WavePlayer_38
  properties {soundWaveAssetPtr=/AmbientAudioContent/Waves/OneShots/Spot_Bird_Japanese_Forest_Evening_Kiiro_Sashigame_Hempitera_Insect_02_NF}
 }
 {
  id=WavePlayer_39
  properties {soundWaveAssetPtr=/AmbientAudioContent/Waves/OneShots/Spot_Bird_Japanese_Forest_Evening_Kiiro_Sashigame_Hempitera_Insect_03_NF}
 }
 {id=WavePlayer_40, properties {soundWaveAssetPtr=/AmbientAudioContent/Waves/OneShots/Spot_Bird_Japanese_Forest_Willow_Ptarmigan_Call_01_NF}}
 {id=WavePlayer_41, properties {soundWaveAssetPtr=/AmbientAudioContent/Waves/OneShots/Spot_Bird_Japanese_Forest_Willow_Ptarmigan_Call_02_NF}}
 {id=WavePlayer_42, properties {soundWaveAssetPtr=/AmbientAudioContent/Waves/OneShots/Spot_Bird_Japanese_Hinode_Chirp_01_NF}}
 {id=WavePlayer_43, properties {soundWaveAssetPtr=/AmbientAudioContent/Waves/OneShots/Spot_Bird_Japanese_Hinode_Chirp_02_NF}}
 {id=WavePlayer_48, properties {soundWaveAssetPtr=/AmbientAudioContent/Waves/OneShots/Spot_Bird_Japanese_Inawashiro_Crow_01_NF}}
 {id=WavePlayer_49, properties {soundWaveAssetPtr=/AmbientAudioContent/Waves/OneShots/Spot_Bird_Japanese_Inawashiro_Crow_02_NF}}
 {id=WavePlayer_50, properties {soundWaveAssetPtr=/AmbientAudioContent/Waves/OneShots/Spot_Bird_Japanese_Inawashiro_Crow_03_NF}}
 {id=WavePlayer_52, properties {soundWaveAssetPtr=/AmbientAudioContent/Waves/OneShots/Spot_Bird_Japanese_Minobu_Screech_01_NF}}
 {id=WavePlayer_53, properties {soundWaveAssetPtr=/AmbientAudioContent/Waves/OneShots/Spot_Bird_Japanese_Minobu_Screech_02_NF}}
 {id=WavePlayer_54, properties {soundWaveAssetPtr=/AmbientAudioContent/Waves/OneShots/Spot_Bird_Japanese_Mizumoto_Crow_Caw_01_NF}}
 {id=WavePlayer_55, properties {soundWaveAssetPtr=/AmbientAudioContent/Waves/OneShots/Spot_Bird_Japanese_Mizumoto_Crow_Caw_02_NF}}
 {id=WavePlayer_8, properties {soundWaveAssetPtr=/AmbientAudioContent/Waves/OneShots/Spot_Bird_Japanese_Black_Kite_Call_01_NF-001}}
 {id=WavePlayer_9, properties {soundWaveAssetPtr=/AmbientAudioContent/Waves/OneShots/Spot_Bird_Japanese_Black_Kite_Call_02_NF-001}}
]
properties {
 attenuationSettings=/AmbientAudioBase/Attenuations/Fort_TileAmbient_OneShot_Bird_Att
 maxDistance=8005
 pitchMultiplier=0.9
 soundClassObject=&Fort_Ambient_OneShots_Outdoors
 volumeMultiplier=0.35
 busSends [{
   audioBus=null
   maxSendDistance=8000
   minSendDistance=3000
   minSendLevel=0.4
   sendLevel=1
   soundSourceBus=&AmbientAudio_Exterior_SourceBus
   customSendLevelCurve {editorCurveData {keys [<at 0 0.4>,<at 1 1>]}}
  }]
 soundSubmixSends [{
   maxSendDistance=8000
   minSendDistance=3000
   minSendLevel=0.4
   soundSubmix=&Fortnite_Reverb_Default
   customSendLevelCurve {editorCurveData {keys [<at 0 0.4>,<at 1 1>]}}
  }]
}
