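# Sound cue for one-shot forest tree rustle ambience, routed to the outdoor Athena ambient sound class with wide tile-ambient one-shot attenuation.
# A Random node picks one of twenty equally weighted rustle waves, then a Modulator varies volume widely and nudges pitch upward slightly.
# The root Quality Level node feeds that same Modulator into its first two slots and leaves the other two empty, so those levels have no sound branch.
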SoundCue /AmbientAudioContent/SoundCues/OneShots/Spot_Knights_Forest_Wind_Tree_Rustle_01_Cue
root=QualityLevel_0
nodes [
 {id=Modulator_1, children [Random_2], properties {pitchMax=1.1, pitchMin=1, volumeMax=1, volumeMin=0.25}}
 {id=QualityLevel_0, children [Modulator_1,Modulator_1,null,null]}
 {
  id=Random_2
  children [
   WavePlayer_14
   WavePlayer_15
   WavePlayer_16
   WavePlayer_17
   WavePlayer_18
   WavePlayer_19
   WavePlayer_20
   WavePlayer_21
   WavePlayer_22
   WavePlayer_23
   WavePlayer_24
   WavePlayer_25
   WavePlayer_26
   WavePlayer_27
   WavePlayer_28
   WavePlayer_29
   WavePlayer_30
   WavePlayer_31
   WavePlayer_32
   WavePlayer_33
  ]
  properties {weights [1,1,1,1,1,1,1,1,1,1,1,1,1,1,1,1,1,1,1,1]}
 }
 {id=WavePlayer_14, properties {soundWaveAssetPtr=/AmbientAudioContent/Waves/OneShots/Spot_Knights_Forest_Tree_Rustle_01}}
 {id=WavePlayer_15, properties {soundWaveAssetPtr=/AmbientAudioContent/Waves/OneShots/Spot_Knights_Forest_Tree_Rustle_02}}
 {id=WavePlayer_16, properties {soundWaveAssetPtr=/AmbientAudioContent/Waves/OneShots/Spot_Knights_Forest_Tree_Rustle_03}}
 {id=WavePlayer_17, properties {soundWaveAssetPtr=/AmbientAudioContent/Waves/OneShots/Spot_Knights_Forest_Tree_Rustle_04}}
 {id=WavePlayer_18, properties {soundWaveAssetPtr=/AmbientAudioContent/Waves/OneShots/Spot_Knights_Forest_Tree_Rustle_05}}
 {id=WavePlayer_19, properties {soundWaveAssetPtr=/AmbientAudioContent/Waves/OneShots/Spot_Knights_Forest_Tree_Rustle_06}}
 {id=WavePlayer_20, properties {soundWaveAssetPtr=/AmbientAudioContent/Waves/OneShots/Spot_Knights_Forest_Tree_Rustle_07}}
 {id=WavePlayer_21, properties {soundWaveAssetPtr=/AmbientAudioContent/Waves/OneShots/Spot_Knights_Forest_Tree_Rustle_08}}
 {id=WavePlayer_22, properties {soundWaveAssetPtr=/AmbientAudioContent/Waves/OneShots/Spot_Knights_Forest_Tree_Rustle_09}}
 {id=WavePlayer_23, properties {soundWaveAssetPtr=/AmbientAudioContent/Waves/OneShots/Spot_Knights_Forest_Tree_Rustle_10}}
 {id=WavePlayer_24, properties {soundWaveAssetPtr=/AmbientAudioContent/Waves/OneShots/Spot_Knights_Forest_Tree_Rustle_11}}
 {id=WavePlayer_25, properties {soundWaveAssetPtr=/AmbientAudioContent/Waves/OneShots/Spot_Knights_Forest_Tree_Rustle_12}}
 {id=WavePlayer_26, properties {soundWaveAssetPtr=/AmbientAudioContent/Waves/OneShots/Spot_Knights_Forest_Tree_Rustle_13}}
 {id=WavePlayer_27, properties {soundWaveAssetPtr=/AmbientAudioContent/Waves/OneShots/Spot_Knights_Forest_Tree_Rustle_14}}
 {id=WavePlayer_28, properties {soundWaveAssetPtr=/AmbientAudioContent/Waves/OneShots/Spot_Knights_Forest_Tree_Rustle_15}}
 {id=WavePlayer_29, properties {soundWaveAssetPtr=/AmbientAudioContent/Waves/OneShots/Spot_Knights_Forest_Tree_Rustle_16}}
 {id=WavePlayer_30, properties {soundWaveAssetPtr=/AmbientAudioContent/Waves/OneShots/Spot_Knights_Forest_Tree_Rustle_17}}
 {id=WavePlayer_31, properties {soundWaveAssetPtr=/AmbientAudioContent/Waves/OneShots/Spot_Knights_Forest_Tree_Rustle_18}}
 {id=WavePlayer_32, properties {soundWaveAssetPtr=/AmbientAudioContent/Waves/OneShots/Spot_Knights_Forest_Tree_Rustle_19}}
 {id=WavePlayer_33, properties {soundWaveAssetPtr=/AmbientAudioContent/Waves/OneShots/Spot_Knights_Forest_Tree_Rustle_20}}
]
properties {
 attenuationSettings=/AmbientAudioBase/Attenuations/Fort_TileAmbient_OneShot_Wider_Att
 bEnableBusSends=false
 maxDistance=11000
 soundClassObject=&Fort_Ambient_Outdoors_Athena
 volumeMultiplier=1
}
