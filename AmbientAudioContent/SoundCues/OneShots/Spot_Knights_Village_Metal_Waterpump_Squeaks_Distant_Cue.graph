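# Sound cue for distant one-shot metal water pump squeaks, named as a Knights Village ambient spot sound.
# A Random node picks among seven equally weighted squeak waves, feeding every branch of a Quality Level node, so all quality levels share the same set.
# A root Modulator sits above that.
# The cue uses the ambient one-shots sound class, borrows a bird one-shot attenuation, slightly lowers volume and disables submix sends.
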SoundCue /AmbientAudioContent/SoundCues/OneShots/Spot_Knights_Village_Metal_Waterpump_Squeaks_Distant_Cue
root=Modulator_0
nodes [
 {id=Modulator_0, children [QualityLevel_0]}
 {id=QualityLevel_0, children [Random_1,Random_1,Random_1,Random_1]}
 {
  id=Random_1
  children [WavePlayer_7, WavePlayer_8, WavePlayer_9, WavePlayer_10, WavePlayer_11, WavePlayer_12, WavePlayer_13]
  properties {weights [1,1,1,1,1,1,1]}
 }
 {id=WavePlayer_10, properties {soundWaveAssetPtr=/AmbientAudioContent/Waves/OneShots/Spot_Knights_Village_Metal_Waterpump_Squeak_04}}
 {id=WavePlayer_11, properties {soundWaveAssetPtr=/AmbientAudioContent/Waves/OneShots/Spot_Knights_Village_Metal_Waterpump_Squeak_05}}
 {id=WavePlayer_12, properties {soundWaveAssetPtr=/AmbientAudioContent/Waves/OneShots/Spot_Knights_Village_Metal_Waterpump_Squeak_06}}
 {id=WavePlayer_13, properties {soundWaveAssetPtr=/AmbientAudioContent/Waves/OneShots/Spot_Knights_Village_Metal_Waterpump_Squeak_07}}
 {id=WavePlayer_7, properties {soundWaveAssetPtr=/AmbientAudioContent/Waves/OneShots/Spot_Knights_Village_Metal_Waterpump_Squeak_01}}
 {id=WavePlayer_8, properties {soundWaveAssetPtr=/AmbientAudioContent/Waves/OneShots/Spot_Knights_Village_Metal_Waterpump_Squeak_02}}
 {id=WavePlayer_9, properties {soundWaveAssetPtr=/AmbientAudioContent/Waves/OneShots/Spot_Knights_Village_Metal_Waterpump_Squeak_03}}
]
properties {
 attenuationSettings=&Fort_TileAmbient_OneShot_Bird_Att
 bEnableSubmixSends=false
 maxDistance=8005
 soundClassObject=&Fort_Ambient_OneShots
 volumeMultiplier=0.752
 preEffectBusSends [{customSendLevelCurve {editorCurveData {keys [<at 0 0>,<at 1 1>]}}}]
 soundSubmixSends [{minSendLevel=0, sendLevel=0, soundSubmix=null}]
}
