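# Marina ambient spot cue that plays intermittent halyard-on-pole impacts as outdoor one-shots, using waves named for Chaparral.
# A loop repeats a random delay and volume modulation over an equal choice between silence and one of nineteen impact variations.
# The empty random branch makes the clanks irregular.
# Overridden attenuation disables spatialization while keeping distance falloff, and a distance-scaled send feeds the exterior ambient source bus.
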SoundCue /AmbientAudioContent/SoundCues/OneShots/Spot_Marina_Halyard_Pole_Impact_Cue
root=Looping_0
nodes [
 {id=Delay_0, children [Modulator_0], properties {delayMax=4, delayMin=0.2}}
 {id=Looping_0, children [Delay_0]}
 {id=Modulator_0, children [Random_0], properties {volumeMin=0.4}}
 {id=Random_0, children [Random_2,null], properties {weights [1,1]}}
 {
  id=Random_2
  children [
   WavePlayer_0
   WavePlayer_1
   WavePlayer_2
   WavePlayer_3
   WavePlayer_4
   WavePlayer_5
   WavePlayer_6
   WavePlayer_7
   WavePlayer_8
   WavePlayer_9
   WavePlayer_10
   WavePlayer_11
   WavePlayer_12
   WavePlayer_13
   WavePlayer_14
   WavePlayer_15
   WavePlayer_16
   WavePlayer_17
   WavePlayer_18
  ]
  properties {weights [1,1,1,1,1,1,1,1,1,1,1,1,1,1,1,1,1,1,1]}
 }
 {id=WavePlayer_0, properties {soundWaveAssetPtr=/AmbientAudioContent/Waves/OneShots/Spot_Chaparral_Halyard_Pole_Impact_01}}
 {id=WavePlayer_1, properties {soundWaveAssetPtr=/AmbientAudioContent/Waves/OneShots/Spot_Chaparral_Halyard_Pole_Impact_02}}
 {id=WavePlayer_10, properties {soundWaveAssetPtr=/AmbientAudioContent/Waves/OneShots/Spot_Chaparral_Halyard_Pole_Impact_11}}
 {id=WavePlayer_11, properties {soundWaveAssetPtr=/AmbientAudioContent/Waves/OneShots/Spot_Chaparral_Halyard_Pole_Impact_12}}
 {id=WavePlayer_12, properties {soundWaveAssetPtr=/AmbientAudioContent/Waves/OneShots/Spot_Chaparral_Halyard_Pole_Impact_13}}
 {id=WavePlayer_13, properties {soundWaveAssetPtr=/AmbientAudioContent/Waves/OneShots/Spot_Chaparral_Halyard_Pole_Impact_14}}
 {id=WavePlayer_14, properties {soundWaveAssetPtr=/AmbientAudioContent/Waves/OneShots/Spot_Chaparral_Halyard_Pole_Impact_15}}
 {id=WavePlayer_15, properties {soundWaveAssetPtr=/AmbientAudioContent/Waves/OneShots/Spot_Chaparral_Halyard_Pole_Impact_16}}
 {id=WavePlayer_16, properties {soundWaveAssetPtr=/AmbientAudioContent/Waves/OneShots/Spot_Chaparral_Halyard_Pole_Impact_17}}
 {id=WavePlayer_17, properties {soundWaveAssetPtr=/AmbientAudioContent/Waves/OneShots/Spot_Chaparral_Halyard_Pole_Impact_18}}
 {id=WavePlayer_18, properties {soundWaveAssetPtr=/AmbientAudioContent/Waves/OneShots/Spot_Chaparral_Halyard_Pole_Impact_19}}
 {id=WavePlayer_2, properties {soundWaveAssetPtr=/AmbientAudioContent/Waves/OneShots/Spot_Chaparral_Halyard_Pole_Impact_03}}
 {id=WavePlayer_3, properties {soundWaveAssetPtr=/AmbientAudioContent/Waves/OneShots/Spot_Chaparral_Halyard_Pole_Impact_04}}
 {id=WavePlayer_4, properties {soundWaveAssetPtr=/AmbientAudioContent/Waves/OneShots/Spot_Chaparral_Halyard_Pole_Impact_05}}
 {id=WavePlayer_5, properties {soundWaveAssetPtr=/AmbientAudioContent/Waves/OneShots/Spot_Chaparral_Halyard_Pole_Impact_06}}
 {id=WavePlayer_6, properties {soundWaveAssetPtr=/AmbientAudioContent/Waves/OneShots/Spot_Chaparral_Halyard_Pole_Impact_07}}
 {id=WavePlayer_7, properties {soundWaveAssetPtr=/AmbientAudioContent/Waves/OneShots/Spot_Chaparral_Halyard_Pole_Impact_08}}
 {id=WavePlayer_8, properties {soundWaveAssetPtr=/AmbientAudioContent/Waves/OneShots/Spot_Chaparral_Halyard_Pole_Impact_09}}
 {id=WavePlayer_9, properties {soundWaveAssetPtr=/AmbientAudioContent/Waves/OneShots/Spot_Chaparral_Halyard_Pole_Impact_10}}
]
properties {
 attenuationSettings=&Fort_TileAmbient_OneShot_Bird_Att
 bOverrideAttenuation=true
 maxDistance=5400
 soundClassObject=&Fort_Ambient_OneShots_Outdoors
 volumeMultiplier=0.501
 attenuationOverrides {bSpatialize=false, binauralRadius=0.3, falloffDistance=5000, nonSpatializedRadiusStart=0.95}
 busSends [{
   audioBus=null
   maxSendDistance=6000
   minSendDistance=3000
   minSendLevel=0.6
   sendLevel=1
   soundSourceBus=&AmbientAudio_Exterior_SourceBus
  }]
}
