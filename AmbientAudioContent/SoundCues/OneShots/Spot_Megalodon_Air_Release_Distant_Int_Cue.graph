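# Sound cue for a distant Megalodon air-release ambient one-shot, routed to the indoor ambient one-shot sound class.
# An equally weighted Random node picks one of ten stereo distant blimp air-release waves, and a Modulator above it varies the pitch.
# It references a less-positional logarithmic attenuation asset, carries wide stereo spread values, and overrides concurrency to cap simultaneous instances at two.
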SoundCue /AmbientAudioContent/SoundCues/OneShots/Spot_Megalodon_Air_Release_Distant_Int_Cue
root=Modulator_0
nodes [
 {id=Modulator_0, children [Random_0], properties {pitchMin=0.9}}
 {
  id=Random_0
  children [WavePlayer_0, WavePlayer_1, WavePlayer_2, WavePlayer_3, WavePlayer_4, WavePlayer_5, WavePlayer_6, WavePlayer_7, WavePlayer_18, WavePlayer_19]
  properties {weights [1,1,1,1,1,1,1,1,1,1]}
 }
 {id=WavePlayer_0, properties {soundWaveAssetPtr=/AmbientAudioContent/Waves/OneShots/Spot_Blimp_Air_Release_Distant_01_Stereo}}
 {id=WavePlayer_1, properties {soundWaveAssetPtr=/AmbientAudioContent/Waves/OneShots/Spot_Blimp_Air_Release_Distant_02_Stereo}}
 {id=WavePlayer_18, properties {soundWaveAssetPtr=/AmbientAudioContent/Waves/OneShots/Spot_Blimp_Air_Release_Distant_09_Stereo}}
 {id=WavePlayer_19, properties {soundWaveAssetPtr=/AmbientAudioContent/Waves/OneShots/Spot_Blimp_Air_Release_Distant_10_Stereo}}
 {id=WavePlayer_2, properties {soundWaveAssetPtr=/AmbientAudioContent/Waves/OneShots/Spot_Blimp_Air_Release_Distant_03_Stereo}}
 {id=WavePlayer_3, properties {soundWaveAssetPtr=/AmbientAudioContent/Waves/OneShots/Spot_Blimp_Air_Release_Distant_04_Stereo}}
 {id=WavePlayer_4, properties {soundWaveAssetPtr=/AmbientAudioContent/Waves/OneShots/Spot_Blimp_Air_Release_Distant_05_Stereo}}
 {id=WavePlayer_5, properties {soundWaveAssetPtr=/AmbientAudioContent/Waves/OneShots/Spot_Blimp_Air_Release_Distant_06_Stereo}}
 {id=WavePlayer_6, properties {soundWaveAssetPtr=/AmbientAudioContent/Waves/OneShots/Spot_Blimp_Air_Release_Distant_07_Stereo}}
 {id=WavePlayer_7, properties {soundWaveAssetPtr=/AmbientAudioContent/Waves/OneShots/Spot_Blimp_Air_Release_Distant_08_Stereo}}
]
properties {
 attenuationSettings=/Game/Sounds/Attenuation/Fort_TileAmbient_OneShot_LessPositional_Log_Att
 bEnableBusSends=false
 bEnableSubmixSends=false
 bOverrideConcurrency=true
 maxDistance=6200
 soundClassObject=/Game/Sounds/FortSoundClassesAndModes/Fort_Ambient_OneShots_Indoors
 volumeMultiplier=1.002
 attenuationOverrides {falloffDistance=6000, stereoSpread=10000, attenuationShapeExtents {x=2000}}
 concurrencyOverrides {maxCount=2}
}
