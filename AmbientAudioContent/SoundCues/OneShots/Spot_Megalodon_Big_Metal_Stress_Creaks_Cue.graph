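# Sound cue for an ambient one-shot of big, distant metal stress creaks and groans, classed as indoor ambience.
# A quality switch gives the first two levels a modulated random pick from twelve knock and stretch waves, and the last two an unmodulated pick from three knocks.
# Distant-creak attenuation, a two-voice concurrency override, a distance-based reverb submix send and an exterior ambient bus send shape playback.
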SoundCue /AmbientAudioContent/SoundCues/OneShots/Spot_Megalodon_Big_Metal_Stress_Creaks_Cue
root=QualityLevel_0
nodes [
 {id=Modulator_0, children [Random_2]}
 {id=QualityLevel_0, children [Modulator_0,Modulator_0,Random_3,Random_3]}
 {
  id=Random_2
  children [
   WavePlayer_11
   WavePlayer_12
   WavePlayer_17
   WavePlayer_18
   WavePlayer_19
   WavePlayer_20
   WavePlayer_21
   WavePlayer_22
   WavePlayer_23
   WavePlayer_24
   WavePlayer_25
   WavePlayer_26
  ]
  properties {weights [1,1,1,1,1,1,1,1,1,1,1,1]}
 }
 {id=Random_3, children [WavePlayer_27, WavePlayer_28, WavePlayer_29], properties {weights [1,1,1]}}
 {id=WavePlayer_11, properties {soundWaveAssetPtr=/AmbientAudioContent/Waves/OneShots/Spot_Metal_Stress_Wronk_Groan_Knock_01}}
 {id=WavePlayer_12, properties {soundWaveAssetPtr=/AmbientAudioContent/Waves/OneShots/Spot_Metal_Stress_Wronk_Groan_Knock_02}}
 {id=WavePlayer_17, properties {soundWaveAssetPtr=/AmbientAudioContent/Waves/OneShots/Spot_Metal_Stress_Wronk_Groan_Knock_03}}
 {id=WavePlayer_18, properties {soundWaveAssetPtr=/AmbientAudioContent/Waves/OneShots/Spot_Metal_Stress_Wronk_Groan_Knock_04}}
 {id=WavePlayer_19, properties {soundWaveAssetPtr=/AmbientAudioContent/Waves/OneShots/Spot_Metal_Stress_Wronk_Groan_Knock_05}}
 {id=WavePlayer_20, properties {soundWaveAssetPtr=/AmbientAudioContent/Waves/OneShots/Spot_Metal_Stress_Wronk_Groan_Knock_06}}
 {id=WavePlayer_21, properties {soundWaveAssetPtr=/AmbientAudioContent/Waves/OneShots/Spot_Metal_Stress_Wronk_Groan_Knock_07}}
 {id=WavePlayer_22, properties {soundWaveAssetPtr=/AmbientAudioContent/Waves/OneShots/Spot_Metal_Stress_Wronk_Groan_Knock_08}}
 {id=WavePlayer_23, properties {soundWaveAssetPtr=/AmbientAudioContent/Waves/OneShots/Spot_Metal_Stress_Wronk_Groan_Knock_09}}
 {id=WavePlayer_24, properties {soundWaveAssetPtr=/AmbientAudioContent/Waves/OneShots/Spot_Metal_Stress_Wronk_Groan_Stretch_02}}
 {id=WavePlayer_25, properties {soundWaveAssetPtr=/AmbientAudioContent/Waves/OneShots/Spot_Metal_Stress_Wronk_Groan_Stretch_03}}
 {id=WavePlayer_26, properties {soundWaveAssetPtr=/AmbientAudioContent/Waves/OneShots/Spot_Metal_Stress_Wronk_Groan_Stretch_04}}
 {id=WavePlayer_27, properties {soundWaveAssetPtr=&Spot_Metal_Stress_Wronk_Groan_Knock_05}}
 {id=WavePlayer_28, properties {soundWaveAssetPtr=&Spot_Metal_Stress_Wronk_Groan_Knock_01}}
 {id=WavePlayer_29, properties {soundWaveAssetPtr=&Spot_Metal_Stress_Wronk_Groan_Knock_06}}
]
properties {
 attenuationSettings=/Game/Sounds/Attenuation/Fort_TileAmbient_OneShot_Distant_Creak_Att
 bOverrideConcurrency=true
 maxDistance=8200
 soundClassObject=/Game/Sounds/FortSoundClassesAndModes/Fort_Ambient_OneShots_Indoors
 volumeMultiplier=0.551
 busSends [{audioBus=null, minSendLevel=1, sendLevel=1, soundSourceBus=&AmbientAudio_Exterior_SourceBus}]
 concurrencyOverrides {maxCount=2}
 soundSubmixSends [{
   maxSendDistance=8000
   minSendDistance=3000
   minSendLevel=0.6
   soundSubmix=&Fortnite_Reverb_Distant
   customSendLevelCurve {editorCurveData {keys [<at 0 0.6>,<at 1 1>]}}
  }]
}
