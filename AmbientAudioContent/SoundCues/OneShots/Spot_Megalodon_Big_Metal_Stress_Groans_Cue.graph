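# Sound cue for an ambient one-shot of huge metal bridge stress groans, named as a Megalodon spot sound.
# A random node picks among seven equally weighted wave variations, then a modulator slightly varies pitch.
# It uses positional tile-ambient attenuation and the ambient one-shot sound class, with concurrency overridden to two instances.
# Output also feeds an exterior ambient source bus and a distant reverb submix with a distance-scaled send.
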SoundCue /AmbientAudioContent/SoundCues/OneShots/Spot_Megalodon_Big_Metal_Stress_Groans_Cue
root=Modulator_0
nodes [
 {id=Modulator_0, children [Random_1], properties {pitchMax=1.1, pitchMin=0.9}}
 {
  id=Random_1
  children [WavePlayer_7, WavePlayer_8, WavePlayer_9, WavePlayer_10, WavePlayer_11, WavePlayer_12, WavePlayer_13]
  properties {weights [1,1,1,1,1,1,1]}
 }
 {id=WavePlayer_10, properties {soundWaveAssetPtr=/AmbientAudioContent/Waves/OneShots/Spot_Big_Metal_Bridge_Huge_Stress_Groan_04}}
 {id=WavePlayer_11, properties {soundWaveAssetPtr=/AmbientAudioContent/Waves/OneShots/Spot_Big_Metal_Bridge_Huge_Stress_Groan_03}}
 {id=WavePlayer_12, properties {soundWaveAssetPtr=/AmbientAudioContent/Waves/OneShots/Spot_Big_Metal_Bridge_Huge_Stress_Groan_02}}
 {id=WavePlayer_13, properties {soundWaveAssetPtr=/AmbientAudioContent/Waves/OneShots/Spot_Big_Metal_Bridge_Huge_Stress_Groan_01}}
 {id=WavePlayer_7, properties {soundWaveAssetPtr=/AmbientAudioContent/Waves/OneShots/Spot_Big_Metal_Bridge_Huge_Stress_Groan_07}}
 {id=WavePlayer_8, properties {soundWaveAssetPtr=/AmbientAudioContent/Waves/OneShots/Spot_Big_Metal_Bridge_Huge_Stress_Groan_06}}
 {id=WavePlayer_9, properties {soundWaveAssetPtr=/AmbientAudioContent/Waves/OneShots/Spot_Big_Metal_Bridge_Huge_Stress_Groan_05}}
]
properties {
 attenuationSettings=/Game/Sounds/Attenuation/Fort_TileAmbient_OneShot_MorePositional_Att
 bOverrideConcurrency=true
 maxDistance=8000
 soundClassObject=&Fort_Ambient_OneShots
 volumeMultiplier=0.551
 busSends [{audioBus=null, minSendLevel=1, sendLevel=1, soundSourceBus=&AmbientAudio_Exterior_SourceBus}]
 concurrencyOverrides {maxCount=2}
 soundSubmixSends [{
   maxSendDistance=8000
   minSendDistance=3000
   minSendLevel=0.6
   soundSubmix=&Fortnite_Reverb_Distant
   customSendLevelCurve {editorCurveData {keys [<at 0 0>,<at 1 1>]}}
  }]
}
