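# Ambient one-shot Sound Cue that plays distant metal "wronk" spot sounds, named for Megalodon but using Cave wave assets.
# An equal-weight Random node picks one of five waves and feeds a Modulator for pitch variation.
# A quality node routes this chain to its first two levels, leaving the other two empty.
# It uses distant-creak attenuation, a two-voice concurrency cap and a distance-based send to the distant reverb submix.
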SoundCue /AmbientAudioContent/SoundCues/OneShots/Spot_Megalodon_Distant_Metal_Wronk_Cue
root=QualityLevel_0
nodes [
 {id=Modulator_0, children [Random_1], properties {pitchMin=0.9}}
 {id=QualityLevel_0, children [Modulator_0,Modulator_0,null,null]}
 {id=Random_1, children [WavePlayer_3,WavePlayer_5,WavePlayer_6,WavePlayer_8,WavePlayer_9], properties {weights [1,1,1,1,1]}}
 {id=WavePlayer_3, properties {soundWaveAssetPtr=/AmbientAudioContent/Waves/OneShots/Spot_Cave_Distant_Metal_Wronk_1}}
 {id=WavePlayer_5, properties {soundWaveAssetPtr=/AmbientAudioContent/Waves/OneShots/Spot_Cave_Distant_Metal_Wronk_2}}
 {id=WavePlayer_6, properties {soundWaveAssetPtr=/AmbientAudioContent/Waves/OneShots/Spot_Cave_Distant_Metal_Wronk_3}}
 {id=WavePlayer_8, properties {soundWaveAssetPtr=/AmbientAudioContent/Waves/OneShots/Spot_Cave_Distant_Metal_Wronk_5}}
 {id=WavePlayer_9, properties {soundWaveAssetPtr=/AmbientAudioContent/Waves/OneShots/Spot_Cave_Distant_Metal_Wronk_8}}
]
properties {
 attenuationSettings=/Game/Sounds/Attenuation/Fort_TileAmbient_OneShot_Distant_Creak_Att
 bEnableBusSends=false
 bOverrideConcurrency=true
 maxDistance=8200
 soundClassObject=&Fort_Ambient_OneShots
 volumeMultiplier=0.2
 attenuationOverrides {falloffDistance=6000, stereoSpread=10000, attenuationShapeExtents {x=2000}}
 concurrencyOverrides {maxCount=2}
 soundSubmixSends [{
   maxSendDistance=8000
   minSendDistance=3000
   minSendLevel=0.6
   soundSubmix=&Fortnite_Reverb_Distant
   customSendLevelCurve {editorCurveData {keys [<at 0 0>,<at 1 1>]}}
  }]
}
