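# Sound cue for a distant ambient one-shot of heavy metal shuffling, classed as an indoor ambient one-shot.
# A random node picks one of four equally weighted waves, and a modulator varies its pitch and volume.
# The quality-level root plays this chain at the first two levels and leaves the other two empty.
# Distant-creak attenuation shapes falloff, with sends to an exterior ambient source bus and distance-scaled distant reverb.
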SoundCue /AmbientAudioContent/SoundCues/OneShots/Spot_Megalodon_Metal_Heavy_Shuffling_Cue
root=QualityLevel_0
nodes [
 {id=Modulator_0, children [Random_0], properties {pitchMin=0.9}}
 {id=QualityLevel_0, children [Modulator_0,Modulator_0,null,null]}
 {id=Random_0, children [WavePlayer_0,WavePlayer_1,WavePlayer_2,WavePlayer_3], properties {weights [1,1,1,1]}}
 {id=WavePlayer_0, properties {soundWaveAssetPtr=/AmbientAudioContent/Waves/OneShots/Spot_Metal_Heavy_Shuffle_01}}
 {id=WavePlayer_1, properties {soundWaveAssetPtr=/AmbientAudioContent/Waves/OneShots/Spot_Metal_Heavy_Shuffle_02}}
 {id=WavePlayer_2, properties {soundWaveAssetPtr=/AmbientAudioContent/Waves/OneShots/Spot_Metal_Heavy_Shuffle_03}}
 {id=WavePlayer_3, properties {soundWaveAssetPtr=/AmbientAudioContent/Waves/OneShots/Spot_Metal_Heavy_Shuffle_04}}
]
properties {
 attenuationSettings=/Game/Sounds/Attenuation/Fort_TileAmbient_OneShot_Distant_Creak_Att
 maxDistance=8200
 soundClassObject=/Game/Sounds/FortSoundClassesAndModes/Fort_Ambient_OneShots_Indoors
 volumeMultiplier=0.501
 attenuationOverrides {falloffDistance=6000, stereoSpread=10000, attenuationShapeExtents {x=2000}}
 busSends [{audioBus=null, minSendLevel=0.6, sendLevel=1, soundSourceBus=&AmbientAudio_Exterior_SourceBus}]
 soundSubmixSends [{maxSendDistance=8000, minSendDistance=3000, minSendLevel=0, soundSubmix=&Fortnite_Reverb_Distant}]
}
