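# Ambient spot Sound Cue that plays intermittent small metal creaks, groans and knocks for wind-stressed metal supports outdoors.
# A Quality Level node selects one of two looping branches, each a random delay before a random wave pick, feeding a root Modulator.
# The first two quality slots share a six-wave pool; the last two share a separate three-wave pool with a slightly wider delay range.
# An external attenuation asset handles spatial falloff.
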SoundCue /AmbientAudioContent/SoundCues/OneShots/Spot_MetalSupport_Wind_Vibrating_Metal_Creaks_Small_Cue
root=Modulator_0
nodes [
 {id=Delay_0, children [Random_3], properties {delayMax=7, delayMin=3}}
 {id=Delay_1, children [Random_2], properties {delayMax=8, delayMin=3}}
 {id=Looping_0, children [Delay_0]}
 {id=Looping_1, children [Delay_1]}
 {id=Modulator_0, children [QualityLevel_0]}
 {id=QualityLevel_0, children [Looping_0,Looping_0,Looping_1,Looping_1]}
 {id=Random_2, children [WavePlayer_11, WavePlayer_19, WavePlayer_20], properties {weights [1,1,1]}}
 {
  id=Random_3
  children [WavePlayer_12, WavePlayer_17, WavePlayer_18, WavePlayer_21, WavePlayer_22, WavePlayer_23]
  properties {weights [1,1,1,1,1,1]}
 }
 {id=WavePlayer_11, properties {soundWaveAssetPtr=/AmbientAudioContent/Waves/OneShots/Spot_Metal_Stress_Wronk_Groan_Knock_01}}
 {id=WavePlayer_12, properties {soundWaveAssetPtr=/AmbientAudioContent/Waves/OneShots/Spot_Metal_Stress_Wronk_Groan_Knock_02}}
 {id=WavePlayer_17, properties {soundWaveAssetPtr=/AmbientAudioContent/Waves/OneShots/Spot_Metal_Stress_Wronk_Groan_Knock_03}}
 {id=WavePlayer_18, properties {soundWaveAssetPtr=/AmbientAudioContent/Waves/OneShots/Spot_Metal_Stress_Wronk_Groan_Knock_04}}
 {id=WavePlayer_19, properties {soundWaveAssetPtr=/AmbientAudioContent/Waves/OneShots/Spot_Metal_Stress_Wronk_Groan_Knock_05}}
 {id=WavePlayer_20, properties {soundWaveAssetPtr=/AmbientAudioContent/Waves/OneShots/Spot_Metal_Stress_Wronk_Groan_Knock_06}}
 {id=WavePlayer_21, properties {soundWaveAssetPtr=/AmbientAudioContent/Waves/OneShots/Spot_Metal_Stress_Wronk_Groan_Knock_07}}
 {id=WavePlayer_22, properties {soundWaveAssetPtr=/AmbientAudioContent/Waves/OneShots/Spot_Metal_Stress_Wronk_Groan_Knock_08}}
 {id=WavePlayer_23, properties {soundWaveAssetPtr=/AmbientAudioContent/Waves/OneShots/Spot_Metal_Stress_Wronk_Groan_Knock_09}}
]
properties {
 attenuationSettings=/Game/Sounds/Artemis/POI/Blimps_Crashsites/Spots/Spot_Wind_Vibrating_Metal_Structure_Large_Linear_Att
 maxDistance=2140.1951
 soundClassObject=&Fort_Ambient_Outdoors_Athena
 volumeMultiplier=0.251
 preEffectBusSends [{}]
}
