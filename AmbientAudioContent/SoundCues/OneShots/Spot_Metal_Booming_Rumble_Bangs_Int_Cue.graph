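# Ambient one-shot sound cue for interior metal booming, rumbling bangs.
# A Random root picks with equal weight among nine wave players, each holding one numbered variant.
# It plays through the ambient one-shots sound class with tile-ambient attenuation at reduced volume, with bus and submix sends disabled.
# A concurrency override allows a single instance and blocks new triggers while one plays.
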SoundCue /AmbientAudioContent/SoundCues/OneShots/Spot_Metal_Booming_Rumble_Bangs_Int_Cue
root=Random_0
nodes [
 {
  id=Random_0
  children [WavePlayer_7, WavePlayer_8, WavePlayer_9, WavePlayer_10, WavePlayer_11, WavePlayer_12, WavePlayer_13, WavePlayer_16, WavePlayer_17]
  properties {weights [1,1,1,1,1,1,1,1,1]}
 }
 {id=WavePlayer_10, properties {soundWaveAssetPtr=/AmbientAudioContent/Waves/OneShots/Spot_Metal_Booming_Rumble_Bangs_Int_04}}
 {id=WavePlayer_11, properties {soundWaveAssetPtr=/AmbientAudioContent/Waves/OneShots/Spot_Metal_Booming_Rumble_Bangs_Int_05}}
 {id=WavePlayer_12, properties {soundWaveAssetPtr=/AmbientAudioContent/Waves/OneShots/Spot_Metal_Booming_Rumble_Bangs_Int_06}}
 {id=WavePlayer_13, properties {soundWaveAssetPtr=/AmbientAudioContent/Waves/OneShots/Spot_Metal_Booming_Rumble_Bangs_Int_07}}
 {id=WavePlayer_16, properties {soundWaveAssetPtr=/AmbientAudioContent/Waves/OneShots/Spot_Metal_Booming_Rumble_Bangs_Int_08}}
 {id=WavePlayer_17, properties {soundWaveAssetPtr=/AmbientAudioContent/Waves/OneShots/Spot_Metal_Booming_Rumble_Bangs_Int_09}}
 {id=WavePlayer_7, properties {soundWaveAssetPtr=/AmbientAudioContent/Waves/OneShots/Spot_Metal_Booming_Rumble_Bangs_Int_01}}
 {id=WavePlayer_8, properties {soundWaveAssetPtr=/AmbientAudioContent/Waves/OneShots/Spot_Metal_Booming_Rumble_Bangs_Int_02}}
 {id=WavePlayer_9, properties {soundWaveAssetPtr=/AmbientAudioContent/Waves/OneShots/Spot_Metal_Booming_Rumble_Bangs_Int_03}}
]
properties {
 attenuationSettings=/Game/Sounds/Attenuation/Fort_TileAmbient_OneShot_Att
 bEnableBusSends=false
 bEnableSubmixSends=false
 bOverrideConcurrency=true
 maxDistance=8000
 soundClassObject=&Fort_Ambient_OneShots
 volumeMultiplier=0.301
 concurrencyOverrides {maxCount=1, resolutionRule=PreventNew, voiceStealReleaseTime=5}
}
