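# Sound Cue for an indoor ambient one-shot of distant metal buckle drops.
# A Random node picks one of ten equally weighted stereo blimp recordings, and a root Modulator varies its pitch with a widened lower bound.
# It uses a shared distant-creak attenuation asset and the indoor ambient one-shots sound class, and sends to a distant reverb submix.
# An overridden concurrency setting caps it at two simultaneous instances.
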SoundCue /AmbientAudioContent/SoundCues/OneShots/Spot_Metal_Buckle_Drops_Distant_Int_Cue
root=Modulator_0
nodes [
 {id=Modulator_0, children [Random_0], properties {pitchMin=0.9}}
 {
  id=Random_0
  children [WavePlayer_0, WavePlayer_1, WavePlayer_2, WavePlayer_3, WavePlayer_4, WavePlayer_5, WavePlayer_6, WavePlayer_7, WavePlayer_18, WavePlayer_19]
  properties {weights [1,1,1,1,1,1,1,1,1,1]}
 }
 {id=WavePlayer_0, properties {soundWaveAssetPtr=/AmbientAudioContent/Waves/OneShots/Spot_Blimp_Metal_Buckle_Drops_Distant_04_Stereo}}
 {id=WavePlayer_1, properties {soundWaveAssetPtr=/AmbientAudioContent/Waves/OneShots/Spot_Blimp_Metal_Buckle_Drops_Distant_01_Stereo}}
 {id=WavePlayer_18, properties {soundWaveAssetPtr=/AmbientAudioContent/Waves/OneShots/Spot_Blimp_Metal_Buckle_Drops_Distant_09_Stereo}}
 {id=WavePlayer_19, properties {soundWaveAssetPtr=/AmbientAudioContent/Waves/OneShots/Spot_Blimp_Metal_Buckle_Drops_Distant_10_Stereo}}
 {id=WavePlayer_2, properties {soundWaveAssetPtr=/AmbientAudioContent/Waves/OneShots/Spot_Blimp_Metal_Buckle_Drops_Distant_02_Stereo}}
 {id=WavePlayer_3, properties {soundWaveAssetPtr=/AmbientAudioContent/Waves/OneShots/Spot_Blimp_Metal_Buckle_Drops_Distant_03_Stereo}}
 {id=WavePlayer_4, properties {soundWaveAssetPtr=/AmbientAudioContent/Waves/OneShots/Spot_Blimp_Metal_Buckle_Drops_Distant_05_Stereo}}
 {id=WavePlayer_5, properties {soundWaveAssetPtr=/AmbientAudioContent/Waves/OneShots/Spot_Blimp_Metal_Buckle_Drops_Distant_06_Stereo}}
 {id=WavePlayer_6, properties {soundWaveAssetPtr=/AmbientAudioContent/Waves/OneShots/Spot_Blimp_Metal_Buckle_Drops_Distant_07_Stereo}}
 {id=WavePlayer_7, properties {soundWaveAssetPtr=/AmbientAudioContent/Waves/OneShots/Spot_Blimp_Metal_Buckle_Drops_Distant_08_Stereo}}
]
properties {
 attenuationSettings=/Game/Sounds/Attenuation/Fort_TileAmbient_OneShot_Distant_Creak_Att
 bOverrideConcurrency=true
 maxDistance=8200
 soundClassObject=/Game/Sounds/FortSoundClassesAndModes/Fort_Ambient_OneShots_Indoors
 volumeMultiplier=0.652
 attenuationOverrides {falloffDistance=6000, stereoSpread=10000, attenuationShapeExtents {x=2000}}
 concurrencyOverrides {maxCount=2}
 soundSubmixSends [{minSendLevel=0.4, soundSubmix=&Fortnite_Reverb_Distant}]
}
