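# Looping ambient sound cue that plays intermittent corrugated metal rattle one-shots in the outdoor ambient sound class.
# A random node picks one of seven equally weighted waves, and a modulator varies its pitch and volume.
# All four quality-level inputs share that modulator, feeding a randomized delay inside the loop to space out the rattles.
# A metal-creak attenuation asset spatializes it, and a concurrency override tightly caps simultaneous instances.
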SoundCue /AmbientAudioContent/SoundCues/OneShots/Spot_Metal_Corrugated_Rattles_Cue
root=Looping_0
nodes [
 {id=Delay_0, children [QualityLevel_0], properties {delayMax=18, delayMin=0.8}}
 {id=Looping_0, children [Delay_0]}
 {id=Modulator_0, children [Random_2], properties {pitchMin=0.8, volumeMin=0.8}}
 {id=QualityLevel_0, children [Modulator_0,Modulator_0,Modulator_0,Modulator_0]}
 {
  id=Random_2
  children [WavePlayer_12, WavePlayer_17, WavePlayer_18, WavePlayer_19, WavePlayer_20, WavePlayer_21, WavePlayer_22]
  properties {weights [1,1,1,1,1,1,1]}
 }
 {id=WavePlayer_12, properties {soundWaveAssetPtr=/AmbientAudioContent/Waves/OneShots/Spot_Metal_Corrugated_Rattles_01}}
 {id=WavePlayer_17, properties {soundWaveAssetPtr=/AmbientAudioContent/Waves/OneShots/Spot_Metal_Corrugated_Rattles_02}}
 {id=WavePlayer_18, properties {soundWaveAssetPtr=/AmbientAudioContent/Waves/OneShots/Spot_Metal_Corrugated_Rattles_03}}
 {id=WavePlayer_19, properties {soundWaveAssetPtr=/AmbientAudioContent/Waves/OneShots/Spot_Metal_Corrugated_Rattles_04}}
 {id=WavePlayer_20, properties {soundWaveAssetPtr=/AmbientAudioContent/Waves/OneShots/Spot_Metal_Corrugated_Rattles_05}}
 {id=WavePlayer_21, properties {soundWaveAssetPtr=/AmbientAudioContent/Waves/OneShots/Spot_Metal_Corrugated_Rattles_06}}
 {id=WavePlayer_22, properties {soundWaveAssetPtr=/AmbientAudioContent/Waves/OneShots/Spot_Metal_Corrugated_Rattles_07}}
]
properties {
 attenuationSettings=/Game/Sounds/Attenuation/Fort_TileAmbient_OneShot_Metal_Creak_Actor_NoInner_Att
 bOverrideConcurrency=true
 maxDistance=4000
 soundClassObject=&Fort_Ambient_OneShots_Outdoors
 volumeMultiplier=0.301
 attenuationOverrides {falloffDistance=6000, stereoSpread=10000, attenuationShapeExtents {x=2000}}
 concurrencyOverrides {maxCount=2, voiceStealReleaseTime=2}
}
