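# Sound cue for outdoor ambient one-shot corrugated metal rattles, played as spot sounds.
# An equally weighted Random node picks one of seven rattle waves and feeds a Modulator that varies pitch and volume, shared by all four quality-level inputs.
# It overrides attenuation for long-range falloff, limits concurrent voices to two and sends to the exterior ambient source bus.
# A distance-based submix send is configured, but no submix is assigned.
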SoundCue /AmbientAudioContent/SoundCues/OneShots/Spot_Metal_Corrugated_Rattles_Oneshot_Cue
root=QualityLevel_0
nodes [
 {id=Modulator_0, children [Random_2], properties {pitchMin=0.8, volumeMin=0.8}}
 {id=QualityLevel_0, children [Modulator_0,Modulator_0,Modulator_0,Modulator_0]}
 {
  id=Random_2
  children [WavePlayer_12, WavePlayer_17, WavePlayer_18, WavePlayer_19, WavePlayer_20, WavePlayer_21, WavePlayer_22]
  properties {weights [1,1,1,1,1,1,1]}
 }
 {id=WavePlayer_12, properties {soundWaveAssetPtr=/AmbientAudioContent/Waves/OneShots/Spot_Metal_Corrugated_Rattles_01}}
 {id=WavePlayer_17, properties {soundWaveAssetPtr=/AmbientAudioContent/Waves/OneShots/Spot_Metal_Corrugated_Rattles_02}}
 {id=WavePlayer_18, properties {soundWaveAssetPtr=/AmbientAudioContent/Waves/OneShots/Spot_Metal_Corrugated_Rattles_03}}
 {id=WavePlayer_19, properties {soundWaveAssetPtr=/AmbientAudioContent/Waves/OneShots/Spot_Metal_Corrugated_Rattles_04}}
 {id=WavePlayer_20, properties {soundWaveAssetPtr=/AmbientAudioContent/Waves/OneShots/Spot_Metal_Corrugated_Rattles_05}}
 {id=WavePlayer_21, properties {soundWaveAssetPtr=/AmbientAudioContent/Waves/OneShots/Spot_Metal_Corrugated_Rattles_06}}
 {id=WavePlayer_22, properties {soundWaveAssetPtr=/AmbientAudioContent/Waves/OneShots/Spot_Metal_Corrugated_Rattles_07}}
]
properties {
 bOverrideAttenuation=true
 bOverrideConcurrency=true
 maxDistance=8400
 soundClassObject=&Fort_Ambient_OneShots_Outdoors
 volumeMultiplier=0.4
 attenuationOverrides {
  bEnableSubmixSends=true
  falloffDistance=8000
  stereoSpread=400
  submixSendSettings [{
    bEnableHPFCutoff=false
    bEnableLPFCutoff=false
    disableManualSendClamp=false
    hPFCutoff=20
    lPFCutoff=20000
    maxSendDistance=6000
    maxSendLevel=1
    minSendDistance=400
    minSendLevel=0
    sendLevel=0.2
    sendLevelControlMethod=Linear
    soundSubmix=null
   }]
 }
 busSends [{audioBus=null, minSendLevel=0, sendLevel=1, soundSourceBus=&AmbientAudio_Exterior_SourceBus}]
 concurrencyOverrides {maxCount=2, voiceStealReleaseTime=2}
}
