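# Sound cue for distant Japanese deer calls, an outdoor ambient one-shot going by its mountain spot name and sound class.
# A random node picks evenly among nine distress, sika stag, Yamanashi and fallow deer waves, then a modulator slightly varies volume and pitch.
# All quality levels share that one chain.
# The quiet cue uses distant-creature attenuation and a pre-effect bus send whose level follows a distance curve.
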SoundCue /AmbientAudioContent/SoundCues/OneShots/Spot_Mountain_Japanese_Deer_Calls_Distant_Cue
root=QualityLevel_0
nodes [
 {id=Modulator_1, children [Random_0], properties {pitchMax=1.1, volumeMax=1.1, volumeMin=0.9}}
 {id=QualityLevel_0, children [Modulator_1,Modulator_1,Modulator_1,Modulator_1]}
 {
  id=Random_0
  children [WavePlayer_0, WavePlayer_1, WavePlayer_3, WavePlayer_4, WavePlayer_5, WavePlayer_6, WavePlayer_8, WavePlayer_9, WavePlayer_10]
  properties {weights [1,1,1,1,1,1,1,1,1]}
 }
 {id=WavePlayer_0, properties {soundWaveAssetPtr=/AmbientAudioContent/Waves/OneShots/Spot_Animal_Deer_Distress_Call_02}}
 {id=WavePlayer_1, properties {soundWaveAssetPtr=/AmbientAudioContent/Waves/OneShots/Spot_Animal_Deer_Sika_Stag_Call_01}}
 {id=WavePlayer_10, properties {soundWaveAssetPtr=/AmbientAudioContent/Waves/OneShots/Spot_Animal_Deer_Sika_Stag_Call_03}}
 {id=WavePlayer_3, properties {soundWaveAssetPtr=/AmbientAudioContent/Waves/OneShots/Spot_Animal_Deer_Yamanashi_Call_01}}
 {id=WavePlayer_4, properties {soundWaveAssetPtr=/AmbientAudioContent/Waves/OneShots/Spot_Animal_Deer_Yamanashi_Call_02}}
 {id=WavePlayer_5, properties {soundWaveAssetPtr=/AmbientAudioContent/Waves/OneShots/Spot_Animal_Deer_Yamanashi_Call_03}}
 {id=WavePlayer_6, properties {soundWaveAssetPtr=/AmbientAudioContent/Waves/OneShots/Spot_Animal_Fallow_Deer_Call_02}}
 {id=WavePlayer_8, properties {soundWaveAssetPtr=/AmbientAudioContent/Waves/OneShots/Spot_Animal_Deer_Distress_Call_01}}
 {id=WavePlayer_9, properties {soundWaveAssetPtr=/AmbientAudioContent/Waves/OneShots/Spot_Animal_Deer_Distress_Call_03}}
]
properties {
 attenuationSettings=/Game/Sounds/Attenuation/Fort_TileAmbient_OneShot_Distant_Creature_Att
 maxDistance=8500
 soundClassObject=&Fort_Ambient_OneShots_Outdoors
 volumeMultiplier=0.25
 preEffectBusSends [{maxSendDistance=8000, minSendDistance=3000, minSendLevel=0.4, customSendLevelCurve {editorCurveData {keys [<at 0 0.4>,<at 1 1>]}}}]
}
