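# Sound cue for long, distant tree creak one-shots in outdoor mountain ambience, using a distant creak attenuation and the outdoor ambient one-shots sound class.
# A quality-level node sends its first two slots to an equal-weight random pick of sixteen creak waves and its last two to a pool of four of those waves.
# Each pool has its own wide pitch and volume modulator, and a narrower root modulator adds further variation.
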SoundCue /AmbientAudioContent/SoundCues/OneShots/Spot_Mountain_Tree_Creak_Long_Cue
root=Modulator_13
nodes [
 {id=Modulator_0, children [Random_1], properties {pitchMax=1, pitchMin=0.7, volumeMax=1.1, volumeMin=0.5}}
 {id=Modulator_1, children [Random_0], properties {pitchMax=1, pitchMin=0.7, volumeMax=1.1, volumeMin=0.5}}
 {id=Modulator_13, children [QualityLevel_0], properties {pitchMax=1.1, pitchMin=0.9, volumeMax=1.1, volumeMin=0.9}}
 {id=QualityLevel_0, children [Modulator_1,Modulator_1,Modulator_0,Modulator_0]}
 {
  id=Random_0
  children [
   WavePlayer_0
   WavePlayer_1
   WavePlayer_2
   WavePlayer_3
   WavePlayer_4
   WavePlayer_5
   WavePlayer_6
   WavePlayer_7
   WavePlayer_8
   WavePlayer_9
   WavePlayer_10
   WavePlayer_11
   WavePlayer_12
   WavePlayer_13
   WavePlayer_14
   WavePlayer_15
  ]
  properties {weights [1,1,1,1,1,1,1,1,1,1,1,1,1,1,1,1]}
 }
 {id=Random_1, children [WavePlayer_16, WavePlayer_17, WavePlayer_18, WavePlayer_19], properties {weights [1,1,1,1]}}
 {id=WavePlayer_0, properties {soundWaveAssetPtr=/AmbientAudioContent/Waves/OneShots/Spot_TreeCreakLong001}}
 {id=WavePlayer_1, properties {soundWaveAssetPtr=/AmbientAudioContent/Waves/OneShots/Spot_TreeCreakLong002}}
 {id=WavePlayer_10, properties {soundWaveAssetPtr=/AmbientAudioContent/Waves/OneShots/Spot_TreeCreakLong011}}
 {id=WavePlayer_11, properties {soundWaveAssetPtr=/AmbientAudioContent/Waves/OneShots/Spot_TreeCreakLong012}}
 {id=WavePlayer_12, properties {soundWaveAssetPtr=/AmbientAudioContent/Waves/OneShots/Spot_TreeCreakLong013}}
 {id=WavePlayer_13, properties {soundWaveAssetPtr=/AmbientAudioContent/Waves/OneShots/Spot_TreeCreakLong014}}
 {id=WavePlayer_14, properties {soundWaveAssetPtr=/AmbientAudioContent/Waves/OneShots/Spot_TreeCreakLong015}}
 {id=WavePlayer_15, properties {soundWaveAssetPtr=/AmbientAudioContent/Waves/OneShots/Spot_TreeCreakLong016}}
 {id=WavePlayer_16, properties {soundWaveAssetPtr=&Spot_TreeCreakLong015}}
 {id=WavePlayer_17, properties {soundWaveAssetPtr=/AmbientAudioContent/Waves/OneShots/Spot_TreeCreakLong005}}
 {id=WavePlayer_18, properties {soundWaveAssetPtr=&Spot_TreeCreakLong013}}
 {id=WavePlayer_19, properties {soundWaveAssetPtr=&Spot_TreeCreakLong011}}
 {id=WavePlayer_2, properties {soundWaveAssetPtr=/AmbientAudioContent/Waves/OneShots/Spot_TreeCreakLong003}}
 {id=WavePlayer_3, properties {soundWaveAssetPtr=/AmbientAudioContent/Waves/OneShots/Spot_TreeCreakLong004}}
 {id=WavePlayer_4, properties {soundWaveAssetPtr=&Spot_TreeCreakLong005}}
 {id=WavePlayer_5, properties {soundWaveAssetPtr=/AmbientAudioContent/Waves/OneShots/Spot_TreeCreakLong006}}
 {id=WavePlayer_6, properties {soundWaveAssetPtr=/AmbientAudioContent/Waves/OneShots/Spot_TreeCreakLong007}}
 {id=WavePlayer_7, properties {soundWaveAssetPtr=/AmbientAudioContent/Waves/OneShots/Spot_TreeCreakLong008}}
 {id=WavePlayer_8, properties {soundWaveAssetPtr=/AmbientAudioContent/Waves/OneShots/Spot_TreeCreakLong009}}
 {id=WavePlayer_9, properties {soundWaveAssetPtr=/AmbientAudioContent/Waves/OneShots/Spot_TreeCreakLong010}}
]
properties {
 attenuationSettings=/Game/Sounds/Attenuation/Fort_TileAmbient_OneShot_Distant_Creak_Att
 bEnableBusSends=false
 maxDistance=8200
 soundClassObject=&Fort_Ambient_OneShots_Outdoors
 volumeMultiplier=0.301
 preEffectBusSends [{}]
}
