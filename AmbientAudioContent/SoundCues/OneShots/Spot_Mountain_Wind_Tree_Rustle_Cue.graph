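# One-shot ambient Sound Cue for mountain wind rustling through trees, played at low volume under the outdoor ambient one-shot sound class.
# A Random node picks among eight equally weighted Primal Biome tree-rustle waves, then a Modulator varies volume and slightly raises pitch.
# The root Quality Level node routes all four of its inputs to that same Modulator chain.
# A tree-rustle attenuation asset shapes falloff, and a distance-based send feeds the exterior ambient source bus.
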SoundCue /AmbientAudioContent/SoundCues/OneShots/Spot_Mountain_Wind_Tree_Rustle_Cue
root=QualityLevel_0
nodes [
 {id=Modulator_1, children [Random_1], properties {pitchMax=1.1, pitchMin=1, volumeMax=1, volumeMin=0.25}}
 {id=QualityLevel_0, children [Modulator_1,Modulator_1,Modulator_1,Modulator_1]}
 {
  id=Random_1
  children [WavePlayer_8, WavePlayer_9, WavePlayer_10, WavePlayer_11, WavePlayer_12, WavePlayer_13, WavePlayer_14, WavePlayer_15]
  properties {weights [1,1,1,1,1,1,1,1]}
 }
 {id=WavePlayer_10, properties {soundWaveAssetPtr=/AmbientAudioContent/Waves/OneShots/Spot_Primal_Biome_Tree_Rustle_Wind_03}}
 {id=WavePlayer_11, properties {soundWaveAssetPtr=/AmbientAudioContent/Waves/OneShots/Spot_Primal_Biome_Tree_Rustle_Wind_04}}
 {id=WavePlayer_12, properties {soundWaveAssetPtr=/AmbientAudioContent/Waves/OneShots/Spot_Primal_Biome_Tree_Rustle_Wind_05}}
 {id=WavePlayer_13, properties {soundWaveAssetPtr=/AmbientAudioContent/Waves/OneShots/Spot_Primal_Biome_Tree_Rustle_Wind_06}}
 {id=WavePlayer_14, properties {soundWaveAssetPtr=/AmbientAudioContent/Waves/OneShots/Spot_Primal_Biome_Tree_Rustle_Wind_07}}
 {id=WavePlayer_15, properties {soundWaveAssetPtr=/AmbientAudioContent/Waves/OneShots/Spot_Primal_Biome_Tree_Rustle_Wind_08}}
 {id=WavePlayer_8, properties {soundWaveAssetPtr=/AmbientAudioContent/Waves/OneShots/Spot_Primal_Biome_Tree_Rustle_Wind_01}}
 {id=WavePlayer_9, properties {soundWaveAssetPtr=/AmbientAudioContent/Waves/OneShots/Spot_Primal_Biome_Tree_Rustle_Wind_02}}
]
properties {
 attenuationSettings=/Game/Sounds/Attenuation/Fort_TileAmbient_OneShot_Tree_Rustle_Att
 maxDistance=6850
 soundClassObject=&Fort_Ambient_OneShots_Outdoors
 volumeMultiplier=0.075
 busSends [{
   audioBus=null
   maxSendDistance=6000
   minSendDistance=3000
   minSendLevel=0.6
   sendLevel=1
   soundSourceBus=&AmbientAudio_Exterior_SourceBus
  }]
}
