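# Sound cue for one-shot mysterious chime spot sounds, assigned to the Athena outdoor ambient sound class.
# A Random node picks among 22 equally weighted rainforest mystery chime waves from the A and B sets.
# A root Modulator randomizes pitch and volume on each play to vary the chimes.
# The cue plays at reduced volume and references a shared magic-presence tile ambient attenuation asset.
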SoundCue /AmbientAudioContent/SoundCues/OneShots/Spot_Mysterious_Chimes_Cue
root=Modulator_0
nodes [
 {id=Modulator_0, children [Random_1], properties {pitchMin=0.65, volumeMin=0.8}}
 {
  id=Random_1
  children [
   WavePlayer_0
   WavePlayer_1
   WavePlayer_2
   WavePlayer_3
   WavePlayer_4
   WavePlayer_5
   WavePlayer_6
   WavePlayer_7
   WavePlayer_8
   WavePlayer_22
   WavePlayer_23
   WavePlayer_24
   WavePlayer_25
   WavePlayer_26
   WavePlayer_27
   WavePlayer_28
   WavePlayer_29
   WavePlayer_30
   WavePlayer_31
   WavePlayer_32
   WavePlayer_33
   WavePlayer_34
  ]
  properties {weights [1,1,1,1,1,1,1,1,1,1,1,1,1,1,1,1,1,1,1,1,1,1]}
 }
 {id=WavePlayer_0, properties {soundWaveAssetPtr=/AmbientAudioContent/Waves/OneShots/Spot_Rainforest_Mystery_Deconstructed_Chime_B_13}}
 {id=WavePlayer_1, properties {soundWaveAssetPtr=/AmbientAudioContent/Waves/OneShots/Spot_Rainforest_Mystery_Deconstructed_Chime_B_12}}
 {id=WavePlayer_2, properties {soundWaveAssetPtr=/AmbientAudioContent/Waves/OneShots/Spot_Rainforest_Mystery_Deconstructed_Chime_B_11}}
 {id=WavePlayer_22, properties {soundWaveAssetPtr=/AmbientAudioContent/Waves/OneShots/Spot_Rainforest_Mystery_Deconstructed_Chime_B_04}}
 {id=WavePlayer_23, properties {soundWaveAssetPtr=/AmbientAudioContent/Waves/OneShots/Spot_Rainforest_Mystery_Deconstructed_Chime_B_03}}
 {id=WavePlayer_24, properties {soundWaveAssetPtr=/AmbientAudioContent/Waves/OneShots/Spot_Rainforest_Mystery_Deconstructed_Chime_B_02}}
 {id=WavePlayer_25, properties {soundWaveAssetPtr=/AmbientAudioContent/Waves/OneShots/Spot_Rainforest_Mystery_Deconstructed_Chime_B_01}}
 {id=WavePlayer_26, properties {soundWaveAssetPtr=/AmbientAudioContent/Waves/OneShots/Spot_Rainforest_Mystery_Deconstructed_Chime_A_09}}
 {id=WavePlayer_27, properties {soundWaveAssetPtr=/AmbientAudioContent/Waves/OneShots/Spot_Rainforest_Mystery_Deconstructed_Chime_A_08}}
 {id=WavePlayer_28, properties {soundWaveAssetPtr=/AmbientAudioContent/Waves/OneShots/Spot_Rainforest_Mystery_Deconstructed_Chime_A_07}}
 {id=WavePlayer_29, properties {soundWaveAssetPtr=/AmbientAudioContent/Waves/OneShots/Spot_Rainforest_Mystery_Deconstructed_Chime_A_06}}
 {id=WavePlayer_3, properties {soundWaveAssetPtr=/AmbientAudioContent/Waves/OneShots/Spot_Rainforest_Mystery_Deconstructed_Chime_B_10}}
 {id=WavePlayer_30, properties {soundWaveAssetPtr=/AmbientAudioContent/Waves/OneShots/Spot_Rainforest_Mystery_Deconstructed_Chime_A_05}}
 {id=WavePlayer_31, properties {soundWaveAssetPtr=/AmbientAudioContent/Waves/OneShots/Spot_Rainforest_Mystery_Deconstructed_Chime_A_04}}
 {id=WavePlayer_32, properties {soundWaveAssetPtr=/AmbientAudioContent/Waves/OneShots/Spot_Rainforest_Mystery_Deconstructed_Chime_A_03}}
 {id=WavePlayer_33, properties {soundWaveAssetPtr=/AmbientAudioContent/Waves/OneShots/Spot_Rainforest_Mystery_Deconstructed_Chime_A_02}}
 {id=WavePlayer_34, properties {soundWaveAssetPtr=/AmbientAudioContent/Waves/OneShots/Spot_Rainforest_Mystery_Deconstructed_Chime_A_01}}
 {id=WavePlayer_4, properties {soundWaveAssetPtr=/AmbientAudioContent/Waves/OneShots/Spot_Rainforest_Mystery_Deconstructed_Chime_B_09}}
 {id=WavePlayer_5, properties {soundWaveAssetPtr=/AmbientAudioContent/Waves/OneShots/Spot_Rainforest_Mystery_Deconstructed_Chime_B_08}}
 {id=WavePlayer_6, properties {soundWaveAssetPtr=/AmbientAudioContent/Waves/OneShots/Spot_Rainforest_Mystery_Deconstructed_Chime_B_07}}
 {id=WavePlayer_7, properties {soundWaveAssetPtr=/AmbientAudioContent/Waves/OneShots/Spot_Rainforest_Mystery_Deconstructed_Chime_B_06}}
 {id=WavePlayer_8, properties {soundWaveAssetPtr=/AmbientAudioContent/Waves/OneShots/Spot_Rainforest_Mystery_Deconstructed_Chime_B_05}}
]
properties {
 attenuationSettings=/Game/Sounds/Attenuation/Fort_TileAmbient_Magic_Pressence_Att
 maxDistance=10500
 soundClassObject=&Fort_Ambient_Outdoors_Athena
 volumeMultiplier=0.25
 attenuationOverrides {bAttenuate=false}
}
