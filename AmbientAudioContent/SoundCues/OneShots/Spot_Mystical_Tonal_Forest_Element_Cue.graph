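# Sound cue for one-shot mystical tonal forest spot ambience, classed as outdoor ambient audio.
# An equal-weight Random node picks one of eight stereo Red Forest tone waves, and a Modulator adds very slight pitch variation.
# Playback is quiet, uses overridden attenuation with reverb send off, and feeds the exterior ambient source bus.
# A concurrency override allows one instance at a time and blocks new ones while it plays.
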SoundCue /AmbientAudioContent/SoundCues/OneShots/Spot_Mystical_Tonal_Forest_Element_Cue
root=Modulator_0
nodes [
 {id=Modulator_0, children [Random_1], properties {pitchMax=1.01, pitchMin=0.99}}
 {
  id=Random_1
  children [WavePlayer_10, WavePlayer_11, WavePlayer_12, WavePlayer_13, WavePlayer_14, WavePlayer_16, WavePlayer_17, WavePlayer_18]
  properties {weights [1,1,1,1,1,1,1,1]}
 }
 {id=WavePlayer_10, properties {soundWaveAssetPtr=/AmbientAudioContent/Waves/OneShots/Spot_Red_Forest_Mystical_Tone_Stereo_09}}
 {id=WavePlayer_11, properties {soundWaveAssetPtr=/AmbientAudioContent/Waves/OneShots/Spot_Red_Forest_Mystical_Tone_Stereo_08}}
 {id=WavePlayer_12, properties {soundWaveAssetPtr=/AmbientAudioContent/Waves/OneShots/Spot_Red_Forest_Mystical_Tone_Stereo_07}}
 {id=WavePlayer_13, properties {soundWaveAssetPtr=/AmbientAudioContent/Waves/OneShots/Spot_Red_Forest_Mystical_Tone_Stereo_06}}
 {id=WavePlayer_14, properties {soundWaveAssetPtr=/AmbientAudioContent/Waves/OneShots/Spot_Red_Forest_Mystical_Tone_Stereo_05}}
 {id=WavePlayer_16, properties {soundWaveAssetPtr=/AmbientAudioContent/Waves/OneShots/Spot_Red_Forest_Mystical_Tone_Stereo_03}}
 {id=WavePlayer_17, properties {soundWaveAssetPtr=/AmbientAudioContent/Waves/OneShots/Spot_Red_Forest_Mystical_Tone_Stereo_02}}
 {id=WavePlayer_18, properties {soundWaveAssetPtr=/AmbientAudioContent/Waves/OneShots/Spot_Red_Forest_Mystical_Tone_Stereo_01}}
]
properties {
 attenuationSettings=/Game/Sounds/Artemis/POI/Chrome_BugleBay/Spot/Fort_Oneshot_Wood_Creak_Ext_Att
 bOverrideAttenuation=true
 bOverrideConcurrency=true
 maxDistance=3600
 soundClassObject=&Fort_Ambient_Outdoors_Athena
 volumeMultiplier=0.3
 attenuationOverrides {bEnableReverbSend=false, stereoSpread=1024, attenuationShapeExtents {x=0}}
 busSends [{audioBus=null, minSendLevel=0.6, sendLevel=1, soundSourceBus=&AmbientAudio_Exterior_SourceBus}]
 concurrencyOverrides {maxCount=1, resolutionRule=PreventNew, voiceStealReleaseTime=5}
}
