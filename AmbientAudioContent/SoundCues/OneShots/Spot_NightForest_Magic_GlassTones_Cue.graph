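# Sound cue for magical glass-tone one-shots in a night forest, an outdoor ambient spot sound with shared tile-ambient attenuation.
# A quality-level root sends its first two slots to a random pick of six tone waves and its last two to a pick of only tones 01–03.
# Each pick passes through a pitch-varying modulator.
# The cue plays at reduced volume and lists an exterior ambient source bus send at zero level.
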SoundCue /AmbientAudioContent/SoundCues/OneShots/Spot_NightForest_Magic_GlassTones_Cue
root=QualityLevel_0
nodes [
 {id=Modulator_0, children [Random_1], properties {pitchMin=0.85}}
 {id=Modulator_1, children [Random_2], properties {pitchMin=0.85}}
 {id=QualityLevel_0, children [Modulator_0,Modulator_0,Modulator_1,Modulator_1]}
 {
  id=Random_1
  children [WavePlayer_5, WavePlayer_6, WavePlayer_7, WavePlayer_8, WavePlayer_9, WavePlayer_10]
  properties {weights [1,1,1,1,1,1]}
 }
 {id=Random_2, children [WavePlayer_11, WavePlayer_12, WavePlayer_13], properties {weights [1,1,1]}}
 {id=WavePlayer_10, properties {soundWaveAssetPtr=/AmbientAudioContent/Waves/OneShots/Spot_NightForest_Magic_GlassTone_01}}
 {id=WavePlayer_11, properties {soundWaveAssetPtr=/AmbientAudioContent/Waves/OneShots/Spot_NightForest_Magic_GlassTone_03}}
 {id=WavePlayer_12, properties {soundWaveAssetPtr=/AmbientAudioContent/Waves/OneShots/Spot_NightForest_Magic_GlassTone_02}}
 {id=WavePlayer_13, properties {soundWaveAssetPtr=&Spot_NightForest_Magic_GlassTone_01}}
 {id=WavePlayer_5, properties {soundWaveAssetPtr=/AmbientAudioContent/Waves/OneShots/Spot_NightForest_Magic_GlassTone_06}}
 {id=WavePlayer_6, properties {soundWaveAssetPtr=/AmbientAudioContent/Waves/OneShots/Spot_NightForest_Magic_GlassTone_05}}
 {id=WavePlayer_7, properties {soundWaveAssetPtr=/AmbientAudioContent/Waves/OneShots/Spot_NightForest_Magic_GlassTone_04}}
 {id=WavePlayer_8, properties {soundWaveAssetPtr=&Spot_NightForest_Magic_GlassTone_03}}
 {id=WavePlayer_9, properties {soundWaveAssetPtr=&Spot_NightForest_Magic_GlassTone_02}}
]
properties {
 attenuationSettings=/Game/Sounds/Attenuation/Fort_TileAmbient_OneShot_Att
 maxDistance=8000
 soundClassObject=&Fort_Ambient_Outdoors_Athena
 volumeMultiplier=0.22
 busSends [{audioBus=null, minSendLevel=0, sendLevel=0, soundSourceBus=&AmbientAudio_Exterior_SourceBus}]
}
