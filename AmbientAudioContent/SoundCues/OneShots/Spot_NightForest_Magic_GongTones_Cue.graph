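# Ambient one-shot sound cue for magical gong tones in a night forest, routed to the outdoor ambient sound class and exterior audio bus.
# A quality-level root feeds its first two inputs from a random pick of eight gong waves, and its last two from a reduced pool of three.
# Each pool passes through a modulator for pitch variation, and the attenuated cue plays at lowered volume.
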SoundCue /AmbientAudioContent/SoundCues/OneShots/Spot_NightForest_Magic_GongTones_Cue
root=QualityLevel_0
nodes [
 {id=Modulator_0, children [Random_0], properties {pitchMin=0.85}}
 {id=Modulator_1, children [Random_3], properties {pitchMin=0.85}}
 {id=QualityLevel_0, children [Modulator_0,Modulator_0,Modulator_1,Modulator_1]}
 {
  id=Random_0
  children [WavePlayer_0, WavePlayer_1, WavePlayer_2, WavePlayer_3, WavePlayer_4, WavePlayer_14, WavePlayer_15, WavePlayer_16]
  properties {weights [1,1,1,1,1,1,1,1]}
 }
 {id=Random_3, children [WavePlayer_17, WavePlayer_18, WavePlayer_19], properties {weights [1,1,1]}}
 {id=WavePlayer_0, properties {soundWaveAssetPtr=/AmbientAudioContent/Waves/OneShots/Spot_NightForest_Magic_GongTone_08}}
 {id=WavePlayer_1, properties {soundWaveAssetPtr=/AmbientAudioContent/Waves/OneShots/Spot_NightForest_Magic_GongTone_07}}
 {id=WavePlayer_14, properties {soundWaveAssetPtr=/AmbientAudioContent/Waves/OneShots/Spot_NightForest_Magic_GongTone_03}}
 {id=WavePlayer_15, properties {soundWaveAssetPtr=/AmbientAudioContent/Waves/OneShots/Spot_NightForest_Magic_GongTone_02}}
 {id=WavePlayer_16, properties {soundWaveAssetPtr=/AmbientAudioContent/Waves/OneShots/Spot_NightForest_Magic_GongTone_01}}
 {id=WavePlayer_17, properties {soundWaveAssetPtr=&Spot_NightForest_Magic_GongTone_03}}
 {id=WavePlayer_18, properties {soundWaveAssetPtr=&Spot_NightForest_Magic_GongTone_02}}
 {id=WavePlayer_19, properties {soundWaveAssetPtr=&Spot_NightForest_Magic_GongTone_01}}
 {id=WavePlayer_2, properties {soundWaveAssetPtr=/AmbientAudioContent/Waves/OneShots/Spot_NightForest_Magic_GongTone_06}}
 {id=WavePlayer_3, properties {soundWaveAssetPtr=/AmbientAudioContent/Waves/OneShots/Spot_NightForest_Magic_GongTone_05}}
 {id=WavePlayer_4, properties {soundWaveAssetPtr=/AmbientAudioContent/Waves/OneShots/Spot_NightForest_Magic_GongTone_04}}
]
properties {
 attenuationSettings=/Game/Sounds/Attenuation/Fort_TileAmbient_OneShot_Att
 maxDistance=8000
 soundClassObject=&Fort_Ambient_Outdoors_Athena
 volumeMultiplier=0.3
 busSends [{audioBus=null, minSendLevel=0, sendLevel=0, soundSourceBus=&AmbientAudio_Exterior_SourceBus}]
}
