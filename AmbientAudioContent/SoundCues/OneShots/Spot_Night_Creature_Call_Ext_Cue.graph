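# Sound cue for one-shot tropical night creature calls in outdoor ambience, using distant-creature attenuation.
# A quality-level node sends its first two levels to a random pick of seven call waves and its last two to a smaller pool of three.
# Each pool passes through its own modulator for pitch and volume variation.
# The cue also feeds a distant reverb submix, with bus sends disabled.
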SoundCue /AmbientAudioContent/SoundCues/OneShots/Spot_Night_Creature_Call_Ext_Cue
root=QualityLevel_0
nodes [
 {id=Modulator_1, children [Random_2], properties {pitchMax=1.1, pitchMin=0.8, volumeMax=1.1, volumeMin=0.9}}
 {id=Modulator_2, children [Random_3], properties {pitchMax=1.1, volumeMax=1.1, volumeMin=0.9}}
 {id=QualityLevel_0, children [Modulator_1,Modulator_1,Modulator_2,Modulator_2]}
 {
  id=Random_2
  children [WavePlayer_0,WavePlayer_1,WavePlayer_2,WavePlayer_3,WavePlayer_4,WavePlayer_5,WavePlayer_6]
  properties {weights [1,1,1,1,1,1,1]}
 }
 {id=Random_3, children [WavePlayer_7,WavePlayer_8,WavePlayer_9], properties {weights [1,1,1]}}
 {id=WavePlayer_0, properties {soundWaveAssetPtr=/AmbientAudioContent/Waves/OneShots/Spot_Tropical_Night_Creature_Call_01}}
 {id=WavePlayer_1, properties {soundWaveAssetPtr=/AmbientAudioContent/Waves/OneShots/Spot_Tropical_Night_Creature_Call_02}}
 {id=WavePlayer_2, properties {soundWaveAssetPtr=/AmbientAudioContent/Waves/OneShots/Spot_Tropical_Night_Creature_Call_03}}
 {id=WavePlayer_3, properties {soundWaveAssetPtr=/AmbientAudioContent/Waves/OneShots/Spot_Tropical_Night_Creature_Call_04}}
 {id=WavePlayer_4, properties {soundWaveAssetPtr=/AmbientAudioContent/Waves/OneShots/Spot_Tropical_Night_Creature_Call_05}}
 {id=WavePlayer_5, properties {soundWaveAssetPtr=/AmbientAudioContent/Waves/OneShots/Spot_Tropical_Night_Creature_Call_06}}
 {id=WavePlayer_6, properties {soundWaveAssetPtr=/AmbientAudioContent/Waves/OneShots/Spot_Tropical_Night_Creature_Call_07}}
 {id=WavePlayer_7, properties {soundWaveAssetPtr=&Spot_Tropical_Night_Creature_Call_01}}
 {id=WavePlayer_8, properties {soundWaveAssetPtr=&Spot_Tropical_Night_Creature_Call_02}}
 {id=WavePlayer_9, properties {soundWaveAssetPtr=&Spot_Tropical_Night_Creature_Call_04}}
]
properties {
 attenuationSettings=/AmbientAudioBase/Attenuations/Fort_OneShot_Distant_Creature_Att
 bEnableBusSends=false
 maxDistance=8500
 soundClassObject=&Fort_Ambient_OneShots_Outdoors
 volumeMultiplier=0.4493
 soundSubmixSends [{
   maxSendLevel=0.6
   minSendLevel=0.25
   soundSubmix=&Fortnite_Reverb_Distant
   customSendLevelCurve {editorCurveData {keys [<at 0 0>,<at 1 1>]}}
  }]
}
