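# Indoor ambient one-shot cue for positional debris creaks and shifts in the Nitro Wasteland setting, using the indoor ambient one-shot sound class.
# A quality switch gives its first two levels a random pick from thirteen waves and the last two only six.
# Each branch adds random pitch and volume modulation, then two nested looping envelopers shaping pitch and volume.
# Concurrency allows one instance and blocks new ones while it plays.
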SoundCue /AmbientAudioContent/SoundCues/OneShots/Spot_Nitro_Wasteland_Debris_Creaks_Indoors_Int_Cue
root=QualityLevel_0
nodes [
 {
  id=Enveloper_0
  children [Enveloper_3]
  properties {
   bLoop=true
   bLoopIndefinitely=true
   loopEnd=100
   pitchCurve {editorCurveData {keys [<at 0 0.8>]}}
   volumeCurve {editorCurveData {keys [<at 0 1 RCIM_Cubic>,<at 51.3938 0.9352 RCIM_Cubic>,<at 100 1 RCIM_Cubic>]}}
  }
 }
 {
  id=Enveloper_1
  children [Enveloper_2]
  properties {
   bLoop=true
   bLoopIndefinitely=true
   loopEnd=100
   pitchCurve {editorCurveData {keys [<at 0 0.8>]}}
   volumeCurve {editorCurveData {keys [<at 0 1 RCIM_Cubic>,<at 51.3938 0.9352 RCIM_Cubic>,<at 100 1 RCIM_Cubic>]}}
  }
 }
 {
  id=Enveloper_2
  children [Modulator_1]
  properties {
   bLoop=true
   bLoopIndefinitely=true
   loopEnd=30
   pitchCurve {editorCurveData {keys [{interpMode=RCIM_Cubic, time=0}, <at 15.9468 0.9181 RCIM_Cubic>, {interpMode=RCIM_Cubic, time=30}]}}
   volumeCurve {editorCurveData {keys [<at 0 1>]}}
  }
 }
 {
  id=Enveloper_3
  children [Modulator_0]
  properties {
   bLoop=true
   bLoopIndefinitely=true
   loopEnd=30
   pitchCurve {editorCurveData {keys [{interpMode=RCIM_Cubic, time=0}, <at 15.9468 0.9181 RCIM_Cubic>, {interpMode=RCIM_Cubic, time=30}]}}
   volumeCurve {editorCurveData {keys [<at 0 1>]}}
  }
 }
 {id=Modulator_0, children [Random_0], properties {pitchMax=1.2, pitchMin=0.7, volumeMax=1.1, volumeMin=0.9}}
 {id=Modulator_1, children [Random_2], properties {pitchMax=1.2, pitchMin=0.7, volumeMax=1.1, volumeMin=0.9}}
 {id=QualityLevel_0, children [Enveloper_0,Enveloper_0,Enveloper_1,Enveloper_1]}
 {
  id=Random_0
  children [
   WavePlayer_0
   WavePlayer_1
   WavePlayer_2
   WavePlayer_3
   WavePlayer_4
   WavePlayer_5
   WavePlayer_6
   WavePlayer_7
   WavePlayer_8
   WavePlayer_9
   WavePlayer_10
   WavePlayer_11
   WavePlayer_12
  ]
  properties {weights [1,1,1,1,1,1,1,1,1,1,1,1,1]}
 }
 {
  id=Random_2
  children [WavePlayer_13, WavePlayer_14, WavePlayer_15, WavePlayer_16, WavePlayer_17, WavePlayer_18]
  properties {weights [1,1,1,1,1,1]}
 }
 {id=WavePlayer_0, properties {soundWaveAssetPtr=/AmbientAudioContent/Waves/OneShots/Spot_Interior_Debris_Fall_Creaks_Shift_01}}
 {id=WavePlayer_1, properties {soundWaveAssetPtr=/AmbientAudioContent/Waves/OneShots/Spot_Interior_Debris_Fall_Creaks_Shift_02}}
 {id=WavePlayer_10, properties {soundWaveAssetPtr=/AmbientAudioContent/Waves/OneShots/Spot_Interior_Debris_Fall_Creaks_Shift_11}}
 {id=WavePlayer_11, properties {soundWaveAssetPtr=/AmbientAudioContent/Waves/OneShots/Spot_Interior_Debris_Fall_Creaks_Shift_12}}
 {id=WavePlayer_12, properties {soundWaveAssetPtr=/AmbientAudioContent/Waves/OneShots/Spot_Interior_Debris_Fall_Creaks_Shift_13}}
 {id=WavePlayer_13, properties {soundWaveAssetPtr=&Spot_Interior_Debris_Fall_Creaks_Shift_01}}
 {id=WavePlayer_14, properties {soundWaveAssetPtr=&Spot_Interior_Debris_Fall_Creaks_Shift_02}}
 {id=WavePlayer_15, properties {soundWaveAssetPtr=/AmbientAudioContent/Waves/OneShots/Spot_Interior_Debris_Fall_Creaks_Shift_03}}
 {id=WavePlayer_16, properties {soundWaveAssetPtr=/AmbientAudioContent/Waves/OneShots/Spot_Interior_Debris_Fall_Creaks_Shift_04}}
 {id=WavePlayer_17, properties {soundWaveAssetPtr=/AmbientAudioContent/Waves/OneShots/Spot_Interior_Debris_Fall_Creaks_Shift_05}}
 {id=WavePlayer_18, properties {soundWaveAssetPtr=/AmbientAudioContent/Waves/OneShots/Spot_Interior_Debris_Fall_Creaks_Shift_06}}
 {id=WavePlayer_2, properties {soundWaveAssetPtr=&Spot_Interior_Debris_Fall_Creaks_Shift_03}}
 {id=WavePlayer_3, properties {soundWaveAssetPtr=&Spot_Interior_Debris_Fall_Creaks_Shift_04}}
 {id=WavePlayer_4, properties {soundWaveAssetPtr=&Spot_Interior_Debris_Fall_Creaks_Shift_05}}
 {id=WavePlayer_5, properties {soundWaveAssetPtr=&Spot_Interior_Debris_Fall_Creaks_Shift_06}}
 {id=WavePlayer_6, properties {soundWaveAssetPtr=/AmbientAudioContent/Waves/OneShots/Spot_Interior_Debris_Fall_Creaks_Shift_07}}
 {id=WavePlayer_7, properties {soundWaveAssetPtr=/AmbientAudioContent/Waves/OneShots/Spot_Interior_Debris_Fall_Creaks_Shift_08}}
 {id=WavePlayer_8, properties {soundWaveAssetPtr=/AmbientAudioContent/Waves/OneShots/Spot_Interior_Debris_Fall_Creaks_Shift_09}}
 {id=WavePlayer_9, properties {soundWaveAssetPtr=/AmbientAudioContent/Waves/OneShots/Spot_Interior_Debris_Fall_Creaks_Shift_10}}
]
properties {
 attenuationSettings=/Game/Sounds/Attenuation/Fort_TileAmbient_OneShot_MorePositional_Att
 bOverrideConcurrency=true
 maxDistance=8000
 soundClassObject=/Game/Sounds/FortSoundClassesAndModes/Fort_Ambient_OneShots_Indoors
 volumeMultiplier=0.251
 concurrencyOverrides {maxCount=1, resolutionRule=PreventNew}
 preEffectBusSends [{}]
}
