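# Indoor ambient one-shot cue for stressed wooden creaks, routed to the indoor one-shot sound class with positional attenuation.
# An equal-weight random node picks from Nitro Wasteland interior and Primitive Structure creak waves, then a modulator randomizes pitch and volume.
# Two nested, indefinitely looping envelopers add pitch and slow volume shaping, feeding all quality levels identically.
# A concurrency override caps it at two instances and prevents new ones beyond that.
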SoundCue /AmbientAudioContent/SoundCues/OneShots/Spot_Nitro_Wasteland_Wooden_Stress_Creak_Int_Cue
root=QualityLevel_0
nodes [
 {
  id=Enveloper_0
  children [Enveloper_3]
  properties {
   bLoop=true
   bLoopIndefinitely=true
   loopEnd=100
   pitchCurve {editorCurveData {keys [<at 0 0.8>]}}
   volumeCurve {editorCurveData {keys [<at 0 1 RCIM_Cubic>,<at 51.3938 0.9352 RCIM_Cubic>,<at 100 1 RCIM_Cubic>]}}
  }
 }
 {
  id=Enveloper_3
  children [Modulator_0]
  properties {
   bLoop=true
   bLoopIndefinitely=true
   loopEnd=30
   pitchCurve {editorCurveData {keys [{interpMode=RCIM_Cubic, time=0}, <at 15.9468 0.9181 RCIM_Cubic>, {interpMode=RCIM_Cubic, time=30}]}}
   volumeCurve {editorCurveData {keys [<at 0 1>]}}
  }
 }
 {id=Modulator_0, children [Random_4], properties {pitchMax=1, pitchMin=0.5, volumeMax=1.1, volumeMin=0.9}}
 {id=QualityLevel_0, children [Enveloper_0,Enveloper_0,Enveloper_0,Enveloper_0]}
 {
  id=Random_4
  children [
   WavePlayer_23
   WavePlayer_24
   WavePlayer_25
   WavePlayer_26
   WavePlayer_27
   WavePlayer_28
   WavePlayer_29
   WavePlayer_30
   WavePlayer_31
   WavePlayer_42
   WavePlayer_43
   WavePlayer_44
   WavePlayer_45
   WavePlayer_46
   WavePlayer_47
   WavePlayer_48
   WavePlayer_49
   WavePlayer_50
   WavePlayer_51
   WavePlayer_52
   WavePlayer_53
   WavePlayer_54
   WavePlayer_55
  ]
  properties {weights [1,1,1,1,1,1,1,1,1,1,1,1,1,1,1,1,1,1,1,1,1,1,1]}
 }
 {id=WavePlayer_23, properties {soundWaveAssetPtr=/AmbientAudioContent/Waves/OneShots/Spot_Nitro_Wasteland_Wood_Creaks_Interior_01}}
 {id=WavePlayer_24, properties {soundWaveAssetPtr=/AmbientAudioContent/Waves/OneShots/Spot_Nitro_Wasteland_Wood_Creaks_Interior_02}}
 {id=WavePlayer_25, properties {soundWaveAssetPtr=/AmbientAudioContent/Waves/OneShots/Spot_Nitro_Wasteland_Wood_Creaks_Interior_03}}
 {id=WavePlayer_26, properties {soundWaveAssetPtr=/AmbientAudioContent/Waves/OneShots/Spot_Nitro_Wasteland_Wood_Creaks_Interior_04}}
 {id=WavePlayer_27, properties {soundWaveAssetPtr=/AmbientAudioContent/Waves/OneShots/Spot_Nitro_Wasteland_Wood_Creaks_Interior_05}}
 {id=WavePlayer_28, properties {soundWaveAssetPtr=/AmbientAudioContent/Waves/OneShots/Spot_Nitro_Wasteland_Wood_Creaks_Interior_06}}
 {id=WavePlayer_29, properties {soundWaveAssetPtr=/AmbientAudioContent/Waves/OneShots/Spot_Nitro_Wasteland_Wood_Creaks_Interior_07}}
 {id=WavePlayer_30, properties {soundWaveAssetPtr=/AmbientAudioContent/Waves/OneShots/Spot_Nitro_Wasteland_Wood_Creaks_Interior_08}}
 {id=WavePlayer_31, properties {soundWaveAssetPtr=/AmbientAudioContent/Waves/OneShots/Spot_Nitro_Wasteland_Wood_Creaks_Interior_09}}
 {id=WavePlayer_42, properties {soundWaveAssetPtr=/AmbientAudioContent/Waves/OneShots/Spot_Nitro_Wasteland_Wood_Creaks_Interior_10}}
 {id=WavePlayer_43, properties {soundWaveAssetPtr=/AmbientAudioContent/Waves/OneShots/Spot_Nitro_Wasteland_Wood_Creaks_Interior_11}}
 {id=WavePlayer_44, properties {soundWaveAssetPtr=/AmbientAudioContent/Waves/OneShots/Spot_Nitro_Wasteland_Wood_Creaks_Interior_12}}
 {id=WavePlayer_45, properties {soundWaveAssetPtr=/AmbientAudioContent/Waves/OneShots/Spot_Nitro_Wasteland_Wood_Creaks_Interior_13}}
 {id=WavePlayer_46, properties {soundWaveAssetPtr=/AmbientAudioContent/Waves/OneShots/Spot_Primitive_Structure_WoodCreaks_01}}
 {id=WavePlayer_47, properties {soundWaveAssetPtr=/AmbientAudioContent/Waves/OneShots/Spot_Primitive_Structure_WoodCreaks_02}}
 {id=WavePlayer_48, properties {soundWaveAssetPtr=/AmbientAudioContent/Waves/OneShots/Spot_Primitive_Structure_WoodCreaks_03}}
 {id=WavePlayer_49, properties {soundWaveAssetPtr=/AmbientAudioContent/Waves/OneShots/Spot_Primitive_Structure_WoodCreaks_04}}
 {id=WavePlayer_50, properties {soundWaveAssetPtr=/AmbientAudioContent/Waves/OneShots/Spot_Primitive_Structure_WoodCreaks_05}}
 {id=WavePlayer_51, properties {soundWaveAssetPtr=/AmbientAudioContent/Waves/OneShots/Spot_Primitive_Structure_WoodCreaks_06}}
 {id=WavePlayer_52, properties {soundWaveAssetPtr=/AmbientAudioContent/Waves/OneShots/Spot_Primitive_Structure_WoodCreaks_07}}
 {id=WavePlayer_53, properties {soundWaveAssetPtr=/AmbientAudioContent/Waves/OneShots/Spot_Primitive_Structure_WoodCreaks_08}}
 {id=WavePlayer_54, properties {soundWaveAssetPtr=/AmbientAudioContent/Waves/OneShots/Spot_Primitive_Structure_WoodCreaks_09}}
 {id=WavePlayer_55, properties {soundWaveAssetPtr=/AmbientAudioContent/Waves/OneShots/Spot_Primitive_Structure_WoodCreaks_10}}
]
properties {
 attenuationSettings=/Game/Sounds/Attenuation/Fort_TileAmbient_OneShot_MorePositional_Att
 bOverrideConcurrency=true
 maxDistance=8000
 soundClassObject=/Game/Sounds/FortSoundClassesAndModes/Fort_Ambient_OneShots_Indoors
 volumeMultiplier=0.501
 concurrencyOverrides {maxCount=2, resolutionRule=PreventNew}
 preEffectBusSends [{}]
}
