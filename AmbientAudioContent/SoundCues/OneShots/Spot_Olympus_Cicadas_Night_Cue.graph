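# One-shot SoundCue for nighttime cicada spot ambience in Olympus, using the shared bird one-shot attenuation and outdoor one-shot sound class.
# A Quality Level node gives the first two levels a weighted random pick, with replacement, between 10- and 7.5-second waves, favoring the longer.
# The last two levels play a single 5-second wave.
# A distance-scaled send feeds the exterior ambient source bus.
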SoundCue /AmbientAudioContent/SoundCues/OneShots/Spot_Olympus_Cicadas_Night_Cue
root=QualityLevel_0
nodes [
 {id=QualityLevel_0, children [Random_0,Random_0,WavePlayer_7,WavePlayer_7]}
 {id=Random_0, children [WavePlayer_4,WavePlayer_6], properties {bRandomizeWithoutReplacement=false, weights [1,0.75]}}
 {id=WavePlayer_4, properties {soundWaveAssetPtr=/AmbientAudioContent/Waves/OneShots/Spot_Olympus_Cicadas_01_Night_10000ms}}
 {id=WavePlayer_6, properties {soundWaveAssetPtr=/AmbientAudioContent/Waves/OneShots/Spot_Olympus_Cicadas_01_Night_7500ms}}
 {id=WavePlayer_7, properties {soundWaveAssetPtr=/AmbientAudioContent/Waves/OneShots/Spot_Olympus_Cicadas_01_Night_5000ms}}
]
properties {
 attenuationSettings=&Fort_TileAmbient_OneShot_Bird_Att
 maxDistance=8005
 soundClassObject=&Fort_Ambient_OneShots_Outdoors
 volumeMultiplier=0.501
 busSends [{
   audioBus=null
   maxSendDistance=6000
   minSendDistance=3000
   minSendLevel=0.6
   sendLevel=1
   soundSourceBus=&AmbientAudio_Exterior_SourceBus
  }]
}
