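# Sound cue for night-time desolate wind gust one-shots in the Olympus outdoor ambience, spatialized through a wide one-shot attenuation.
# A quality switch gives the first two levels an even random pick between a tonal Hades gust pool and a Tower night gust pool.
# The last two levels share a reduced pool of four Tower gusts with wider pitch and volume modulation.
# Each pool picks one wave at random and varies it through a modulator.
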SoundCue /AmbientAudioContent/SoundCues/OneShots/Spot_Olympus_Desolate_Wind_Oneshots_Night_Cue
root=QualityLevel_0
nodes [
 {id=Modulator_0, children [Random_1], properties {pitchMax=1.1, pitchMin=0.7, volumeMax=1.1, volumeMin=0.5}}
 {id=Modulator_1, children [Random_4], properties {pitchMax=1.2, pitchMin=1, volumeMax=1.1, volumeMin=0.9}}
 {id=Modulator_2, children [Random_2], properties {pitchMax=1.2, pitchMin=1, volumeMax=1.1, volumeMin=0.9}}
 {id=QualityLevel_0, children [Random_3,Random_3,Modulator_0,Modulator_0]}
 {id=Random_1, children [WavePlayer_22, WavePlayer_23, WavePlayer_24, WavePlayer_25], properties {weights [1,1,1,1]}}
 {
  id=Random_2
  children [
   WavePlayer_0
   WavePlayer_1
   WavePlayer_2
   WavePlayer_3
   WavePlayer_4
   WavePlayer_5
   WavePlayer_6
   WavePlayer_7
   WavePlayer_8
   WavePlayer_9
   WavePlayer_10
  ]
  properties {weights [1,1,1,1,1,1,1,1,1,1,1]}
 }
 {id=Random_3, children [Modulator_2,Modulator_1], properties {weights [1,1]}}
 {
  id=Random_4
  children [
   WavePlayer_26
   WavePlayer_27
   WavePlayer_28
   WavePlayer_29
   WavePlayer_30
   WavePlayer_31
   WavePlayer_32
   WavePlayer_33
   WavePlayer_45
   WavePlayer_46
   WavePlayer_47
  ]
  properties {weights [1,1,1,1,1,1,1,1,1,1,1]}
 }
 {id=WavePlayer_0, properties {soundWaveAssetPtr=/AmbientAudioContent/Waves/OneShots/Spot_Hades_Wind_Gust_Tonal_01}}
 {id=WavePlayer_1, properties {soundWaveAssetPtr=/AmbientAudioContent/Waves/OneShots/Spot_Hades_Wind_Gust_Tonal_02}}
 {id=WavePlayer_10, properties {soundWaveAssetPtr=/AmbientAudioContent/Waves/OneShots/Spot_Hades_Wind_Gust_Tonal_11}}
 {id=WavePlayer_2, properties {soundWaveAssetPtr=/AmbientAudioContent/Waves/OneShots/Spot_Hades_Wind_Gust_Tonal_03}}
 {id=WavePlayer_22, properties {soundWaveAssetPtr=/AmbientAudioContent/Waves/OneShots/Spot_Tower_AM_WindGust_Night_01}}
 {id=WavePlayer_23, properties {soundWaveAssetPtr=/AmbientAudioContent/Waves/OneShots/Spot_Tower_AM_WindGust_Night_11}}
 {id=WavePlayer_24, properties {soundWaveAssetPtr=/AmbientAudioContent/Waves/OneShots/Spot_Tower_AM_WindGust_Night_10}}
 {id=WavePlayer_25, properties {soundWaveAssetPtr=/AmbientAudioContent/Waves/OneShots/Spot_Tower_AM_WindGust_Night_02}}
 {id=WavePlayer_26, properties {soundWaveAssetPtr=&Spot_Tower_AM_WindGust_Night_01}}
 {id=WavePlayer_27, properties {soundWaveAssetPtr=&Spot_Tower_AM_WindGust_Night_02}}
 {id=WavePlayer_28, properties {soundWaveAssetPtr=/AmbientAudioContent/Waves/OneShots/Spot_Tower_AM_WindGust_Night_03}}
 {id=WavePlayer_29, properties {soundWaveAssetPtr=/AmbientAudioContent/Waves/OneShots/Spot_Tower_AM_WindGust_Night_04}}
 {id=WavePlayer_3, properties {soundWaveAssetPtr=/AmbientAudioContent/Waves/OneShots/Spot_Hades_Wind_Gust_Tonal_04}}
 {id=WavePlayer_30, properties {soundWaveAssetPtr=/AmbientAudioContent/Waves/OneShots/Spot_Tower_AM_WindGust_Night_05}}
 {id=WavePlayer_31, properties {soundWaveAssetPtr=/AmbientAudioContent/Waves/OneShots/Spot_Tower_AM_WindGust_Night_06}}
 {id=WavePlayer_32, properties {soundWaveAssetPtr=/AmbientAudioContent/Waves/OneShots/Spot_Tower_AM_WindGust_Night_07}}
 {id=WavePlayer_33, properties {soundWaveAssetPtr=/AmbientAudioContent/Waves/OneShots/Spot_Tower_AM_WindGust_Night_08}}
 {id=WavePlayer_4, properties {soundWaveAssetPtr=/AmbientAudioContent/Waves/OneShots/Spot_Hades_Wind_Gust_Tonal_05}}
 {id=WavePlayer_45, properties {soundWaveAssetPtr=/AmbientAudioContent/Waves/OneShots/Spot_Tower_AM_WindGust_Night_09}}
 {id=WavePlayer_46, properties {soundWaveAssetPtr=&Spot_Tower_AM_WindGust_Night_10}}
 {id=WavePlayer_47, properties {soundWaveAssetPtr=&Spot_Tower_AM_WindGust_Night_11}}
 {id=WavePlayer_5, properties {soundWaveAssetPtr=/AmbientAudioContent/Waves/OneShots/Spot_Hades_Wind_Gust_Tonal_06}}
 {id=WavePlayer_6, properties {soundWaveAssetPtr=/AmbientAudioContent/Waves/OneShots/Spot_Hades_Wind_Gust_Tonal_07}}
 {id=WavePlayer_7, properties {soundWaveAssetPtr=/AmbientAudioContent/Waves/OneShots/Spot_Hades_Wind_Gust_Tonal_08}}
 {id=WavePlayer_8, properties {soundWaveAssetPtr=/AmbientAudioContent/Waves/OneShots/Spot_Hades_Wind_Gust_Tonal_09}}
 {id=WavePlayer_9, properties {soundWaveAssetPtr=/AmbientAudioContent/Waves/OneShots/Spot_Hades_Wind_Gust_Tonal_10}}
]
properties {
 attenuationSettings=/Game/Sounds/Attenuation/Fort_TileAmbient_OneShot_Wider_Att
 maxDistance=8000
 soundClassObject=&Fort_Ambient_Outdoors_Athena
 volumeMultiplier=0.401
 preEffectBusSends [{}]
}
