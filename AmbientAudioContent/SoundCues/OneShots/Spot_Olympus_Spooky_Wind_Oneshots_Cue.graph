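# Sound cue for spooky wind gust one-shots, played as outdoor ambient spot sounds with tile-ambient one-shot attenuation.
# A quality-level root sends its first two slots to a five-player random pool of four gust waves, with gust 01 listed twice.
# Its last two slots use a smaller random pool of gusts 01 and 02.
# Each pool feeds its own modulator for slight pitch and volume variation.
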SoundCue /AmbientAudioContent/SoundCues/OneShots/Spot_Olympus_Spooky_Wind_Oneshots_Cue
root=QualityLevel_0
nodes [
 {id=Modulator_0, children [Random_3], properties {pitchMax=1.1, pitchMin=0.9, volumeMax=0.9, volumeMin=0.8}}
 {id=Modulator_1, children [Random_2], properties {pitchMax=1.2, pitchMin=0.9, volumeMax=1.1, volumeMin=0.9}}
 {id=QualityLevel_0, children [Modulator_1,Modulator_1,Modulator_0,Modulator_0]}
 {id=Random_2, children [WavePlayer_1,WavePlayer_2,WavePlayer_3,WavePlayer_4,WavePlayer_5], properties {weights [1,1,1,1,1]}}
 {id=Random_3, children [WavePlayer_6,WavePlayer_7], properties {weights [1,1]}}
 {id=WavePlayer_1, properties {soundWaveAssetPtr=/AmbientAudioContent/Waves/OneShots/Spot_grassland_evening_spookgust_01}}
 {id=WavePlayer_2, properties {soundWaveAssetPtr=/AmbientAudioContent/Waves/OneShots/Spot_grassland_evening_spookgust_04}}
 {id=WavePlayer_3, properties {soundWaveAssetPtr=/AmbientAudioContent/Waves/OneShots/Spot_grassland_evening_spookgust_03}}
 {id=WavePlayer_4, properties {soundWaveAssetPtr=/AmbientAudioContent/Waves/OneShots/Spot_grassland_evening_spookgust_02}}
 {id=WavePlayer_5, properties {soundWaveAssetPtr=&Spot_grassland_evening_spookgust_01}}
 {id=WavePlayer_6, properties {soundWaveAssetPtr=&Spot_grassland_evening_spookgust_02}}
 {id=WavePlayer_7, properties {soundWaveAssetPtr=&Spot_grassland_evening_spookgust_01}}
]
properties {
 attenuationSettings=/Game/Sounds/Attenuation/Fort_TileAmbient_OneShot_Att
 maxDistance=8000
 soundClassObject=&Fort_Ambient_Outdoors_Athena
 volumeMultiplier=0.31
 preEffectBusSends [{}]
}
