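# Sound cue for a heavy interior wind-debris ambient one-shot, reusing Canary jungle temple recordings on the indoor ambient sound class.
# A quality-level switch sends its first two slots to a random pick of eight waves and its last two to four of those waves.
# Each branch has its own pitch and volume modulator.
# The cue boosts pitch and volume, uses wide one-shot attenuation, and caps concurrency at two.
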SoundCue /AmbientAudioContent/SoundCues/OneShots/Spot_Olympus_Wind_Debris_Heavy_Int_Cue
root=QualityLevel_0
nodes [
 {id=Modulator_1, children [Random_1], properties {pitchMax=1.1, pitchMin=0.9, volumeMax=1.4, volumeMin=0.7}}
 {id=Modulator_2, children [Random_0], properties {pitchMax=1.1, volumeMax=1.1, volumeMin=0.1}}
 {id=QualityLevel_0, children [Modulator_1,Modulator_1,Modulator_2,Modulator_2]}
 {id=Random_0, children [WavePlayer_0,WavePlayer_1,WavePlayer_2,WavePlayer_3], properties {weights [1,1,1,1]}}
 {
  id=Random_1
  children [WavePlayer_4, WavePlayer_5, WavePlayer_6, WavePlayer_7, WavePlayer_8, WavePlayer_9, WavePlayer_10, WavePlayer_11]
  properties {weights [1,1,1,1,1,1,1,1]}
 }
 {id=WavePlayer_0, properties {soundWaveAssetPtr=/AmbientAudioContent/Waves/OneShots/Spot_Canary_Jungle_Temple_Int_SS_Wind_Debris_01}}
 {id=WavePlayer_1, properties {soundWaveAssetPtr=/AmbientAudioContent/Waves/OneShots/Spot_Canary_Jungle_Temple_Int_SS_Wind_Debris_02}}
 {id=WavePlayer_10, properties {soundWaveAssetPtr=/AmbientAudioContent/Waves/OneShots/Spot_Canary_Jungle_Temple_Int_SS_Wind_Debris_07}}
 {id=WavePlayer_11, properties {soundWaveAssetPtr=/AmbientAudioContent/Waves/OneShots/Spot_Canary_Jungle_Temple_Int_SS_Wind_Debris_08}}
 {id=WavePlayer_2, properties {soundWaveAssetPtr=/AmbientAudioContent/Waves/OneShots/Spot_Canary_Jungle_Temple_Int_SS_Wind_Debris_03}}
 {id=WavePlayer_3, properties {soundWaveAssetPtr=/AmbientAudioContent/Waves/OneShots/Spot_Canary_Jungle_Temple_Int_SS_Wind_Debris_04}}
 {id=WavePlayer_4, properties {soundWaveAssetPtr=&Spot_Canary_Jungle_Temple_Int_SS_Wind_Debris_01}}
 {id=WavePlayer_5, properties {soundWaveAssetPtr=&Spot_Canary_Jungle_Temple_Int_SS_Wind_Debris_02}}
 {id=WavePlayer_6, properties {soundWaveAssetPtr=&Spot_Canary_Jungle_Temple_Int_SS_Wind_Debris_03}}
 {id=WavePlayer_7, properties {soundWaveAssetPtr=&Spot_Canary_Jungle_Temple_Int_SS_Wind_Debris_04}}
 {id=WavePlayer_8, properties {soundWaveAssetPtr=/AmbientAudioContent/Waves/OneShots/Spot_Canary_Jungle_Temple_Int_SS_Wind_Debris_05}}
 {id=WavePlayer_9, properties {soundWaveAssetPtr=/AmbientAudioContent/Waves/OneShots/Spot_Canary_Jungle_Temple_Int_SS_Wind_Debris_06}}
]
properties {
 attenuationSettings=/Game/Sounds/Attenuation/Fort_TileAmbient_OneShot_Wider_Att
 bEnableBusSends=false
 bOverrideConcurrency=true
 maxDistance=8000
 pitchMultiplier=1.2
 soundClassObject=&Fort_Ambient_Indoors_Athena
 volumeMultiplier=2.005
 concurrencyOverrides {maxCount=2}
 preEffectBusSends [{}]
}
