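# Sound cue for an indoor one-shot ambient spot sound marking the queen bug boss's presence, built from bug queen tunnel recordings.
# A Random node picks among seven equally weighted wave variations.
# A Modulator applies slight random pitch and volume variation to each pick.
# It uses a wide tile-ambient one-shot attenuation, the indoor ambient sound class, and a concurrency override that caps simultaneous instances.
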SoundCue /AmbientAudioContent/SoundCues/OneShots/Spot_Queen_Boss_Bug_Presence_Int_Cue
root=Modulator_1
nodes [
 {id=Modulator_1, children [Random_1], properties {pitchMax=1.1, pitchMin=0.9, volumeMax=1.1, volumeMin=0.9}}
 {
  id=Random_1
  children [WavePlayer_8, WavePlayer_9, WavePlayer_10, WavePlayer_11, WavePlayer_12, WavePlayer_13, WavePlayer_14]
  properties {weights [1,1,1,1,1,1,1]}
 }
 {id=WavePlayer_10, properties {soundWaveAssetPtr=/AmbientAudioContent/Waves/OneShots/BugQueen/SC_Spot_BugQueen_Tunnel_Int_05}}
 {id=WavePlayer_11, properties {soundWaveAssetPtr=/AmbientAudioContent/Waves/OneShots/BugQueen/SC_Spot_BugQueen_Tunnel_Int_04}}
 {id=WavePlayer_12, properties {soundWaveAssetPtr=/AmbientAudioContent/Waves/OneShots/BugQueen/SC_Spot_BugQueen_Tunnel_Int_03}}
 {id=WavePlayer_13, properties {soundWaveAssetPtr=/AmbientAudioContent/Waves/OneShots/BugQueen/SC_Spot_BugQueen_Tunnel_Int_02}}
 {id=WavePlayer_14, properties {soundWaveAssetPtr=/AmbientAudioContent/Waves/OneShots/BugQueen/SC_Spot_BugQueen_Tunnel_Int_01}}
 {id=WavePlayer_8, properties {soundWaveAssetPtr=/AmbientAudioContent/Waves/OneShots/BugQueen/SC_Spot_BugQueen_Tunnel_Int_07}}
 {id=WavePlayer_9, properties {soundWaveAssetPtr=/AmbientAudioContent/Waves/OneShots/BugQueen/SC_Spot_BugQueen_Tunnel_Int_06}}
]
properties {
 attenuationSettings=/Game/Sounds/Attenuation/Fort_TileAmbient_OneShot_Wider_Att
 bEnableBusSends=false
 bOverrideConcurrency=true
 maxDistance=8000
 soundClassObject=&Fort_Ambient_Indoors_Athena
 volumeMultiplier=1
 concurrencyOverrides {maxCount=2}
 preEffectBusSends [{}]
}
