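# Sound cue for outdoor rainforest frog-call one-shots, named for daytime rain.
# All quality levels share one random node choosing equally among five nested random pools: two overlapping striped-tree-frog sets, mating calls, swamp frogs and frog-insect calls.
# The tree-frog and mating pools get pitch and volume modulation, while the other two pass through mixers that boost volume.
# Playback is quiet overall, uses bird one-shot attenuation and has a low concurrency limit.
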SoundCue /AmbientAudioContent/SoundCues/OneShots/Spot_Rainforest_Frog_Calls_Day_Rain_All_Cue
root=QualityLevel_0
nodes [
 {id=Mixer_0, children [Random_10], properties {inputVolume [2]}}
 {id=Mixer_1, children [Random_9], properties {inputVolume [2]}}
 {id=Modulator_2, children [Random_0], properties {pitchMax=1.1, pitchMin=0.8, volumeMax=1.2, volumeMin=1}}
 {id=Modulator_3, children [Random_7], properties {pitchMax=1.1, pitchMin=0.8, volumeMax=1.2, volumeMin=1}}
 {id=Modulator_4, children [Random_8], properties {pitchMax=1, pitchMin=0.8, volumeMax=1.2, volumeMin=1}}
 {id=QualityLevel_0, children [Random_2,Random_2,Random_2,Random_2]}
 {
  id=Random_0
  children [WavePlayer_0, WavePlayer_1, WavePlayer_2, WavePlayer_3, WavePlayer_48, WavePlayer_49, WavePlayer_50]
  properties {weights [1,1,1,1,1,1,1]}
 }
 {
  id=Random_10
  children [
   WavePlayer_78
   WavePlayer_79
   WavePlayer_80
   WavePlayer_81
   WavePlayer_82
   WavePlayer_83
   WavePlayer_84
   WavePlayer_85
   WavePlayer_86
   WavePlayer_87
  ]
  properties {weights [1,1,1,1,1,1,1,1,1,1]}
 }
 {
  id=Random_2
  children [Modulator_2,Modulator_3,Modulator_4,Mixer_1,Mixer_0]
  properties {bShouldExcludeFromBranchCulling=true, weights [1,1,1,1,1]}
 }
 {id=Random_7, children [WavePlayer_51, WavePlayer_52, WavePlayer_53, WavePlayer_54, WavePlayer_55], properties {weights [1,1,1,1,1]}}
 {
  id=Random_8
  children [
   WavePlayer_56
   WavePlayer_57
   WavePlayer_58
   WavePlayer_59
   WavePlayer_60
   WavePlayer_61
   WavePlayer_62
   WavePlayer_63
   WavePlayer_64
   WavePlayer_65
   WavePlayer_66
   WavePlayer_67
  ]
  properties {weights [1,1,1,1,1,1,1,1,1,1,1,1]}
 }
 {
  id=Random_9
  children [
   WavePlayer_68
   WavePlayer_69
   WavePlayer_70
   WavePlayer_71
   WavePlayer_72
   WavePlayer_73
   WavePlayer_74
   WavePlayer_75
   WavePlayer_76
   WavePlayer_77
  ]
  properties {weights [1,1,1,1,1,1,1,1,1,1]}
 }
 {id=WavePlayer_0, properties {soundWaveAssetPtr=/AmbientAudioContent/Waves/OneShots/Spot_Frog_Striped_Tree_Calls_01}}
 {id=WavePlayer_1, properties {soundWaveAssetPtr=/AmbientAudioContent/Waves/OneShots/Spot_Frog_Striped_Tree_Calls_02}}
 {id=WavePlayer_2, properties {soundWaveAssetPtr=/AmbientAudioContent/Waves/OneShots/Spot_Frog_Striped_Tree_Calls_03}}
 {id=WavePlayer_3, properties {soundWaveAssetPtr=/AmbientAudioContent/Waves/OneShots/Spot_Frog_Striped_Tree_Calls_04}}
 {id=WavePlayer_48, properties {soundWaveAssetPtr=/AmbientAudioContent/Waves/OneShots/Spot_Frog_Striped_Tree_Calls_05}}
 {id=WavePlayer_49, properties {soundWaveAssetPtr=/AmbientAudioContent/Waves/OneShots/Spot_Frog_Striped_Tree_Calls_06}}
 {id=WavePlayer_50, properties {soundWaveAssetPtr=/AmbientAudioContent/Waves/OneShots/Spot_Frog_Striped_Tree_Calls_07}}
 {id=WavePlayer_51, properties {soundWaveAssetPtr=&Spot_Frog_Striped_Tree_Calls_01}}
 {id=WavePlayer_52, properties {soundWaveAssetPtr=&Spot_Frog_Striped_Tree_Calls_02}}
 {id=WavePlayer_53, properties {soundWaveAssetPtr=&Spot_Frog_Striped_Tree_Calls_03}}
 {id=WavePlayer_54, properties {soundWaveAssetPtr=&Spot_Frog_Striped_Tree_Calls_04}}
 {id=WavePlayer_55, properties {soundWaveAssetPtr=&Spot_Frog_Striped_Tree_Calls_05}}
 {id=WavePlayer_56, properties {soundWaveAssetPtr=/AmbientAudioContent/Waves/OneShots/Spot_Frog_Mating_Call_01}}
 {id=WavePlayer_57, properties {soundWaveAssetPtr=/AmbientAudioContent/Waves/OneShots/Spot_Frog_Mating_Call_02}}
 {id=WavePlayer_58, properties {soundWaveAssetPtr=/AmbientAudioContent/Waves/OneShots/Spot_Frog_Mating_Call_03}}
 {id=WavePlayer_59, properties {soundWaveAssetPtr=/AmbientAudioContent/Waves/OneShots/Spot_Frog_Mating_Call_04}}
 {id=WavePlayer_60, properties {soundWaveAssetPtr=/AmbientAudioContent/Waves/OneShots/Spot_Frog_Mating_Call_05}}
 {id=WavePlayer_61, properties {soundWaveAssetPtr=/AmbientAudioContent/Waves/OneShots/Spot_Frog_Mating_Call_06}}
 {id=WavePlayer_62, properties {soundWaveAssetPtr=/AmbientAudioContent/Waves/OneShots/Spot_Frog_Mating_Call_Pitched_01}}
 {id=WavePlayer_63, properties {soundWaveAssetPtr=/AmbientAudioContent/Waves/OneShots/Spot_Frog_Mating_Call_Pitched_02}}
 {id=WavePlayer_64, properties {soundWaveAssetPtr=/AmbientAudioContent/Waves/OneShots/Spot_Frog_Mating_Call_Pitched_03}}
 {id=WavePlayer_65, properties {soundWaveAssetPtr=/AmbientAudioContent/Waves/OneShots/Spot_Frog_Mating_Call_Pitched_04}}
 {id=WavePlayer_66, properties {soundWaveAssetPtr=/AmbientAudioContent/Waves/OneShots/Spot_Frog_Mating_Call_Pitched_05}}
 {id=WavePlayer_67, properties {soundWaveAssetPtr=/AmbientAudioContent/Waves/OneShots/Spot_Frog_Mating_Call_Pitched_06}}
 {id=WavePlayer_68, properties {soundWaveAssetPtr=/AmbientAudioContent/Waves/OneShots/Spot_Biome_Swamp_01_Frog_01}}
 {id=WavePlayer_69, properties {soundWaveAssetPtr=/AmbientAudioContent/Waves/OneShots/Spot_Biome_Swamp_01_CricketFrog_01}}
 {id=WavePlayer_70, properties {soundWaveAssetPtr=/AmbientAudioContent/Waves/OneShots/Spot_Biome_Swamp_01_CricketFrog_02}}
 {id=WavePlayer_71, properties {soundWaveAssetPtr=/AmbientAudioContent/Waves/OneShots/Spot_Biome_Swamp_01_CricketFrog_03}}
 {id=WavePlayer_72, properties {soundWaveAssetPtr=/AmbientAudioContent/Waves/OneShots/Spot_Biome_Swamp_01_CricketFrog_04}}
 {id=WavePlayer_73, properties {soundWaveAssetPtr=/AmbientAudioContent/Waves/OneShots/Spot_Biome_Swamp_01_Frog_02}}
 {id=WavePlayer_74, properties {soundWaveAssetPtr=/AmbientAudioContent/Waves/OneShots/Spot_Biome_Swamp_01_Frog_03}}
 {id=WavePlayer_75, properties {soundWaveAssetPtr=/AmbientAudioContent/Waves/OneShots/Spot_Biome_Swamp_01_Frog_04}}
 {id=WavePlayer_76, properties {soundWaveAssetPtr=/AmbientAudioContent/Waves/OneShots/Spot_Biome_Swamp_01_Frog_05}}
 {id=WavePlayer_77, properties {soundWaveAssetPtr=/AmbientAudioContent/Waves/OneShots/Spot_Biome_Swamp_01_Frog_06}}
 {id=WavePlayer_78, properties {soundWaveAssetPtr=/AmbientAudioContent/Waves/OneShots/Spot_Frog_Insects_Call_01}}
 {id=WavePlayer_79, properties {soundWaveAssetPtr=/AmbientAudioContent/Waves/OneShots/Spot_Frog_Insects_Call_02}}
 {id=WavePlayer_80, properties {soundWaveAssetPtr=/AmbientAudioContent/Waves/OneShots/Spot_Frog_Insects_Call_03}}
 {id=WavePlayer_81, properties {soundWaveAssetPtr=/AmbientAudioContent/Waves/OneShots/Spot_Frog_Insects_Call_04}}
 {id=WavePlayer_82, properties {soundWaveAssetPtr=/AmbientAudioContent/Waves/OneShots/Spot_Frog_Insects_Call_05}}
 {id=WavePlayer_83, properties {soundWaveAssetPtr=/AmbientAudioContent/Waves/OneShots/Spot_Frog_Insects_Call_06}}
 {id=WavePlayer_84, properties {soundWaveAssetPtr=/AmbientAudioContent/Waves/OneShots/Spot_Frog_Insects_Call_07}}
 {id=WavePlayer_85, properties {soundWaveAssetPtr=/AmbientAudioContent/Waves/OneShots/Spot_Frog_Insects_Call_08}}
 {id=WavePlayer_86, properties {soundWaveAssetPtr=/AmbientAudioContent/Waves/OneShots/Spot_Frog_Insects_Call_09}}
 {id=WavePlayer_87, properties {soundWaveAssetPtr=/AmbientAudioContent/Waves/OneShots/Spot_Frog_Insects_Call_10}}
]
properties {
 attenuationSettings=&Fort_TileAmbient_OneShot_Bird_Att
 bEnableBusSends=false
 maxDistance=8005
 soundClassObject=&Fort_Ambient_OneShots_Outdoors
 volumeMultiplier=0.1
 concurrencyOverrides {maxCount=2, voiceStealReleaseTime=1}
 concurrencySet [null]
 preEffectBusSends [{}]
}
