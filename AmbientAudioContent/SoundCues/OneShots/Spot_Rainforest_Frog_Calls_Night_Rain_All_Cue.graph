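# Sound Cue for one-shot rainforest frog calls, named for rainy nights, playing through the outdoor ambient one-shot sound class with a shared bird one-shot attenuation.
# All quality levels share one equally weighted Random node that picks among five banks: striped tree frog, pitched mating call, swamp frog, bullfrog and frog-insect calls.
# Each bank randomizes its wave choice under its own Modulator, which varies pitch and volume and boosts gain heavily for some banks.
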SoundCue /AmbientAudioContent/SoundCues/OneShots/Spot_Rainforest_Frog_Calls_Night_Rain_All_Cue
root=QualityLevel_0
nodes [
 {id=Modulator_0, children [Random_8], properties {pitchMax=1, pitchMin=0.8, volumeMax=4.2, volumeMin=4}}
 {id=Modulator_1, children [Random_9], properties {pitchMax=1, pitchMin=0.8, volumeMax=2.7, volumeMin=2.5}}
 {id=Modulator_2, children [Random_4], properties {pitchMax=1.1, pitchMin=0.8, volumeMax=1, volumeMin=0.7}}
 {id=Modulator_4, children [Random_7], properties {pitchMax=1, pitchMin=0.8, volumeMax=1, volumeMin=0.9}}
 {id=Modulator_5, children [Random_11], properties {pitchMax=1, pitchMin=0.8, volumeMax=3.2, volumeMin=3}}
 {id=QualityLevel_0, children [Random_2,Random_2,Random_2,Random_2]}
 {
  id=Random_11
  children [
   WavePlayer_78
   WavePlayer_79
   WavePlayer_80
   WavePlayer_81
   WavePlayer_82
   WavePlayer_83
   WavePlayer_84
   WavePlayer_85
   WavePlayer_86
   WavePlayer_87
  ]
  properties {weights [1,1,1,1,1,1,1,1,1,1]}
 }
 {
  id=Random_2
  children [Modulator_2,Modulator_4,Modulator_0,Modulator_1,Modulator_5]
  properties {bShouldExcludeFromBranchCulling=true, weights [1,1,1,1,1]}
 }
 {
  id=Random_4
  children [WavePlayer_4, WavePlayer_5, WavePlayer_9, WavePlayer_10, WavePlayer_19, WavePlayer_20, WavePlayer_21]
  properties {weights [1,1,1,1,1,1,1]}
 }
 {
  id=Random_7
  children [WavePlayer_22, WavePlayer_23, WavePlayer_24, WavePlayer_25, WavePlayer_26, WavePlayer_27]
  properties {weights [1,1,1,1,1,1]}
 }
 {
  id=Random_8
  children [WavePlayer_52, WavePlayer_53, WavePlayer_54, WavePlayer_55, WavePlayer_56, WavePlayer_57]
  properties {weights [1,1,1,1,1,1]}
 }
 {id=Random_9, children [WavePlayer_58, WavePlayer_59, WavePlayer_60, WavePlayer_61], properties {weights [1,1,1,1]}}
 {id=WavePlayer_10, properties {soundWaveAssetPtr=/AmbientAudioContent/Waves/OneShots/Spot_Frog_Striped_Tree_Calls_04}}
 {id=WavePlayer_19, properties {soundWaveAssetPtr=/AmbientAudioContent/Waves/OneShots/Spot_Frog_Striped_Tree_Calls_05}}
 {id=WavePlayer_20, properties {soundWaveAssetPtr=/AmbientAudioContent/Waves/OneShots/Spot_Frog_Striped_Tree_Calls_06}}
 {id=WavePlayer_21, properties {soundWaveAssetPtr=/AmbientAudioContent/Waves/OneShots/Spot_Frog_Striped_Tree_Calls_07}}
 {id=WavePlayer_22, properties {soundWaveAssetPtr=/AmbientAudioContent/Waves/OneShots/Spot_Frog_Mating_Call_Pitched_01}}
 {id=WavePlayer_23, properties {soundWaveAssetPtr=/AmbientAudioContent/Waves/OneShots/Spot_Frog_Mating_Call_Pitched_02}}
 {id=WavePlayer_24, properties {soundWaveAssetPtr=/AmbientAudioContent/Waves/OneShots/Spot_Frog_Mating_Call_Pitched_03}}
 {id=WavePlayer_25, properties {soundWaveAssetPtr=/AmbientAudioContent/Waves/OneShots/Spot_Frog_Mating_Call_Pitched_04}}
 {id=WavePlayer_26, properties {soundWaveAssetPtr=/AmbientAudioContent/Waves/OneShots/Spot_Frog_Mating_Call_Pitched_05}}
 {id=WavePlayer_27, properties {soundWaveAssetPtr=/AmbientAudioContent/Waves/OneShots/Spot_Frog_Mating_Call_Pitched_06}}
 {id=WavePlayer_4, properties {soundWaveAssetPtr=/AmbientAudioContent/Waves/OneShots/Spot_Frog_Striped_Tree_Calls_01}}
 {id=WavePlayer_5, properties {soundWaveAssetPtr=/AmbientAudioContent/Waves/OneShots/Spot_Frog_Striped_Tree_Calls_02}}
 {id=WavePlayer_52, properties {soundWaveAssetPtr=/AmbientAudioContent/Waves/OneShots/Spot_Biome_Swamp_01_Frog_01}}
 {id=WavePlayer_53, properties {soundWaveAssetPtr=/AmbientAudioContent/Waves/OneShots/Spot_Biome_Swamp_01_Frog_02}}
 {id=WavePlayer_54, properties {soundWaveAssetPtr=/AmbientAudioContent/Waves/OneShots/Spot_Biome_Swamp_01_Frog_03}}
 {id=WavePlayer_55, properties {soundWaveAssetPtr=/AmbientAudioContent/Waves/OneShots/Spot_Biome_Swamp_01_Frog_04}}
 {id=WavePlayer_56, properties {soundWaveAssetPtr=/AmbientAudioContent/Waves/OneShots/Spot_Biome_Swamp_01_Frog_05}}
 {id=WavePlayer_57, properties {soundWaveAssetPtr=/AmbientAudioContent/Waves/OneShots/Spot_Biome_Swamp_01_Frog_06}}
 {id=WavePlayer_58, properties {soundWaveAssetPtr=/AmbientAudioContent/Waves/OneShots/Spot_Biome_Swamp_01_Bullfrog_04}}
 {id=WavePlayer_59, properties {soundWaveAssetPtr=/AmbientAudioContent/Waves/OneShots/Spot_Biome_Swamp_01_Bullfrog_03}}
 {id=WavePlayer_60, properties {soundWaveAssetPtr=/AmbientAudioContent/Waves/OneShots/Spot_Biome_Swamp_01_Bullfrog_02}}
 {id=WavePlayer_61, properties {soundWaveAssetPtr=/AmbientAudioContent/Waves/OneShots/Spot_Biome_Swamp_01_Bullfrog_01}}
 {id=WavePlayer_78, properties {soundWaveAssetPtr=/AmbientAudioContent/Waves/OneShots/Spot_Frog_Insects_Call_01}}
 {id=WavePlayer_79, properties {soundWaveAssetPtr=/AmbientAudioContent/Waves/OneShots/Spot_Frog_Insects_Call_02}}
 {id=WavePlayer_80, properties {soundWaveAssetPtr=/AmbientAudioContent/Waves/OneShots/Spot_Frog_Insects_Call_03}}
 {id=WavePlayer_81, properties {soundWaveAssetPtr=/AmbientAudioContent/Waves/OneShots/Spot_Frog_Insects_Call_04}}
 {id=WavePlayer_82, properties {soundWaveAssetPtr=/AmbientAudioContent/Waves/OneShots/Spot_Frog_Insects_Call_05}}
 {id=WavePlayer_83, properties {soundWaveAssetPtr=/AmbientAudioContent/Waves/OneShots/Spot_Frog_Insects_Call_06}}
 {id=WavePlayer_84, properties {soundWaveAssetPtr=/AmbientAudioContent/Waves/OneShots/Spot_Frog_Insects_Call_07}}
 {id=WavePlayer_85, properties {soundWaveAssetPtr=/AmbientAudioContent/Waves/OneShots/Spot_Frog_Insects_Call_08}}
 {id=WavePlayer_86, properties {soundWaveAssetPtr=/AmbientAudioContent/Waves/OneShots/Spot_Frog_Insects_Call_09}}
 {id=WavePlayer_87, properties {soundWaveAssetPtr=/AmbientAudioContent/Waves/OneShots/Spot_Frog_Insects_Call_10}}
 {id=WavePlayer_9, properties {soundWaveAssetPtr=/AmbientAudioContent/Waves/OneShots/Spot_Frog_Striped_Tree_Calls_03}}
]
properties {
 attenuationSettings=&Fort_TileAmbient_OneShot_Bird_Att
 bEnableBusSends=false
 maxDistance=8005
 soundClassObject=&Fort_Ambient_OneShots_Outdoors
 volumeMultiplier=0.501
 concurrencyOverrides {maxCount=2, voiceStealReleaseTime=1}
 concurrencySet [null]
 preEffectBusSends [{}]
}
