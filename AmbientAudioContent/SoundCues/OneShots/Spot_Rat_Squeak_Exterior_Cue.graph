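# Sound cue for exterior ambient rat squeak one-shots, classed as ambient one-shots and using a bird one-shot attenuation asset.
# A quality-level switch sends the first two levels to an equal-weight random pick of fourteen waves and the last two to a five-wave subset.
# Each branch adds slight pitch and volume modulation.
# The cue also has a distance-scaled send to the exterior ambient source bus.
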SoundCue /AmbientAudioContent/SoundCues/OneShots/Spot_Rat_Squeak_Exterior_Cue
root=QualityLevel_0
nodes [
 {id=Modulator_1, children [Random_0], properties {pitchMax=1.1, volumeMax=1.1, volumeMin=0.9}}
 {id=Modulator_2, children [Random_3], properties {pitchMax=1.1, volumeMax=1.1, volumeMin=0.9}}
 {id=QualityLevel_0, children [Modulator_1,Modulator_1,Modulator_2,Modulator_2]}
 {
  id=Random_0
  children [
   WavePlayer_5
   WavePlayer_6
   WavePlayer_7
   WavePlayer_8
   WavePlayer_9
   WavePlayer_10
   WavePlayer_11
   WavePlayer_12
   WavePlayer_25
   WavePlayer_26
   WavePlayer_27
   WavePlayer_28
   WavePlayer_29
   WavePlayer_30
  ]
  properties {weights [1,1,1,1,1,1,1,1,1,1,1,1,1,1]}
 }
 {id=Random_3, children [WavePlayer_22, WavePlayer_23, WavePlayer_24, WavePlayer_36, WavePlayer_37], properties {weights [1,1,1,1,1]}}
 {id=WavePlayer_10, properties {soundWaveAssetPtr=/AmbientAudioContent/Waves/OneShots/Spot_Rat_Squeak_06_Exterior}}
 {id=WavePlayer_11, properties {soundWaveAssetPtr=/AmbientAudioContent/Waves/OneShots/Spot_Rat_Squeak_07_Exterior}}
 {id=WavePlayer_12, properties {soundWaveAssetPtr=/AmbientAudioContent/Waves/OneShots/Spot_Rat_Squeak_08_Exterior}}
 {id=WavePlayer_22, properties {soundWaveAssetPtr=/AmbientAudioContent/Waves/OneShots/Spot_Rat_Squeak_09_Exterior}}
 {id=WavePlayer_23, properties {soundWaveAssetPtr=&Spot_Rat_Squeak_07_Exterior}}
 {id=WavePlayer_24, properties {soundWaveAssetPtr=/AmbientAudioContent/Waves/OneShots/Spot_Rat_Squeak_04_Exterior}}
 {id=WavePlayer_25, properties {soundWaveAssetPtr=&Spot_Rat_Squeak_09_Exterior}}
 {id=WavePlayer_26, properties {soundWaveAssetPtr=/AmbientAudioContent/Waves/OneShots/Spot_Rat_Squeak_10_Exterior}}
 {id=WavePlayer_27, properties {soundWaveAssetPtr=/AmbientAudioContent/Waves/OneShots/Spot_Rat_Squeak_11_Exterior}}
 {id=WavePlayer_28, properties {soundWaveAssetPtr=/AmbientAudioContent/Waves/OneShots/Spot_Rat_Squeak_12_Exterior}}
 {id=WavePlayer_29, properties {soundWaveAssetPtr=/AmbientAudioContent/Waves/OneShots/Spot_Rat_Squeak_13_Exterior}}
 {id=WavePlayer_30, properties {soundWaveAssetPtr=/AmbientAudioContent/Waves/OneShots/Spot_Rat_Squeak_14_Exterior}}
 {id=WavePlayer_36, properties {soundWaveAssetPtr=/AmbientAudioContent/Waves/OneShots/Spot_Rat_Squeak_02_Exterior}}
 {id=WavePlayer_37, properties {soundWaveAssetPtr=&Spot_Rat_Squeak_14_Exterior}}
 {id=WavePlayer_5, properties {soundWaveAssetPtr=/AmbientAudioContent/Waves/OneShots/Spot_Rat_Squeak_01_Exterior}}
 {id=WavePlayer_6, properties {soundWaveAssetPtr=&Spot_Rat_Squeak_02_Exterior}}
 {id=WavePlayer_7, properties {soundWaveAssetPtr=/AmbientAudioContent/Waves/OneShots/Spot_Rat_Squeak_03_Exterior}}
 {id=WavePlayer_8, properties {soundWaveAssetPtr=&Spot_Rat_Squeak_04_Exterior}}
 {id=WavePlayer_9, properties {soundWaveAssetPtr=/AmbientAudioContent/Waves/OneShots/Spot_Rat_Squeak_05_Exterior}}
]
properties {
 attenuationSettings=&Fort_TileAmbient_OneShot_Bird_Att
 maxDistance=8005
 soundClassObject=&Fort_Ambient_OneShots
 volumeMultiplier=0.451
 busSends [{
   audioBus=null
   maxSendDistance=4500
   minSendDistance=2000
   minSendLevel=0.2
   sendLevel=1
   soundSourceBus=&AmbientAudio_Exterior_SourceBus
  }]
}
