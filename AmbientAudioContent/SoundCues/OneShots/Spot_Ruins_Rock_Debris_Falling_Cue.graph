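# Sound cue for a ruins ambient one-shot of falling rock debris, built from cave debris recordings.
# A Random node picks among nine equally weighted wave variations and feeds a Modulator at the root for playback variation.
# It uses a shared ceiling-debris attenuation asset and the outdoor ambient sound class, at reduced volume with bus and submix sends disabled.
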SoundCue /AmbientAudioContent/SoundCues/OneShots/Spot_Ruins_Rock_Debris_Falling_Cue
root=Modulator_0
nodes [
 {id=Modulator_0, children [Random_0]}
 {
  id=Random_0
  children [WavePlayer_0, WavePlayer_1, WavePlayer_2, WavePlayer_3, WavePlayer_4, WavePlayer_11, WavePlayer_15, WavePlayer_16, WavePlayer_17]
  properties {weights [1,1,1,1,1,1,1,1,1]}
 }
 {id=WavePlayer_0, properties {soundWaveAssetPtr=/AmbientAudioContent/Waves/OneShots/Spot_Cave_Rock_Debris_Falling_01}}
 {id=WavePlayer_1, properties {soundWaveAssetPtr=/AmbientAudioContent/Waves/OneShots/Spot_Cave_Rock_Debris_Falling_02}}
 {id=WavePlayer_11, properties {soundWaveAssetPtr=/AmbientAudioContent/Waves/OneShots/Spot_Cave_Rock_Debris_Falling_06}}
 {id=WavePlayer_15, properties {soundWaveAssetPtr=/AmbientAudioContent/Waves/OneShots/Spot_Cave_Rock_Debris_Falling_07}}
 {id=WavePlayer_16, properties {soundWaveAssetPtr=/AmbientAudioContent/Waves/OneShots/Spot_Cave_Rock_Debris_Falling_08}}
 {id=WavePlayer_17, properties {soundWaveAssetPtr=/AmbientAudioContent/Waves/OneShots/Spot_Cave_Rock_Debris_Falling_09}}
 {id=WavePlayer_2, properties {soundWaveAssetPtr=/AmbientAudioContent/Waves/OneShots/Spot_Cave_Rock_Debris_Falling_03}}
 {id=WavePlayer_3, properties {soundWaveAssetPtr=/AmbientAudioContent/Waves/OneShots/Spot_Cave_Rock_Debris_Falling_04}}
 {id=WavePlayer_4, properties {soundWaveAssetPtr=/AmbientAudioContent/Waves/OneShots/Spot_Cave_Rock_Debris_Falling_05}}
]
properties {
 attenuationSettings=/Game/Sounds/Attenuation/Fort_TileAmbient_OneShot_Ceiling_Debris_Att
 bEnableBusSends=false
 bEnableSubmixSends=false
 maxDistance=6750
 soundClassObject=&Fort_Ambient_Outdoors_Athena
 volumeMultiplier=0.752
 attenuationOverrides {bAttenuate=false}
}
