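# Spotted outdoor ambience cue for a Sakura Springs mud geyser.
# The first two quality levels mix a geyser loop, a pitched-down mud blob loop and randomly delayed, repeating mud blob one-shots; the other two play only the pitched-down loop.
# The result passes through listener-focus attenuation with low-pass filtering, a random start delay and pitch modulation.
# Concurrency caps it at three instances, stopping the quietest.
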SoundCue /AmbientAudioContent/SoundCues/OneShots/Spot_SakuraSprings_Mud_Geyser_Amb_Cue
root=Modulator_3
nodes [
 {
  id=Attenuation_1
  children [QualityLevel_0]
  properties {
   bOverrideAttenuation=true
   attenuationOverrides {
    bAttenuateWithLPF=true
    bEnableFocusInterpolation=true
    bEnableListenerFocus=true
    bEnableReverbSend=false
    falloffDistance=1600
    focusReleaseInterpSpeed=0.5
    lPFFrequencyAtMax=2000
    lPFRadiusMax=3000
    lPFRadiusMin=500
    nonFocusVolumeAttenuation=0.25
    stereoSpread=512
    attenuationShapeExtents {x=200}
   }
  }
 }
 {id=Delay_0, children [Random_1], properties {delayMax=3, delayMin=0.1}}
 {id=Delay_1, children [Attenuation_1], properties {delayMax=3}}
 {id=Looping_0, children [Delay_0]}
 {id=Mixer_0, children [Modulator_0,Modulator_1,Looping_0], properties {inputVolume [1,0.4,1]}}
 {id=Mixer_1, children [Modulator_2], properties {inputVolume [0.6]}}
 {id=Modulator_0, children [WavePlayer_0], properties {volumeMax=1.1, volumeMin=0.8}}
 {id=Modulator_1, children [WavePlayer_1], properties {pitchMax=0.6, pitchMin=0.6}}
 {id=Modulator_2, children [WavePlayer_8], properties {pitchMax=0.6, pitchMin=0.6}}
 {id=Modulator_3, children [Delay_1], properties {pitchMin=0.8}}
 {id=QualityLevel_0, children [Mixer_0,Mixer_0,Mixer_1,Mixer_1]}
 {
  id=Random_1
  children [WavePlayer_9, WavePlayer_10, WavePlayer_11, WavePlayer_12, WavePlayer_13, WavePlayer_14]
  properties {weights [1,1,1,1,1,1]}
 }
 {id=WavePlayer_0, properties {bLooping=true, soundWaveAssetPtr=/AmbientAudioContent/Waves/SpottedLoops/Spot_Geyser_AmbientLoop_Close_01}}
 {id=WavePlayer_1, properties {bLooping=true, soundWaveAssetPtr=/AmbientAudioContent/Waves/SpottedLoops/Spot_MudBlob002_Loop}}
 {id=WavePlayer_10, properties {soundWaveAssetPtr=/AmbientAudioContent/Waves/OneShots/Spot_MudBlobA002}}
 {id=WavePlayer_11, properties {soundWaveAssetPtr=/AmbientAudioContent/Waves/OneShots/Spot_MudBlobA003}}
 {id=WavePlayer_12, properties {soundWaveAssetPtr=/AmbientAudioContent/Waves/OneShots/Spot_MudBlobA004}}
 {id=WavePlayer_13, properties {soundWaveAssetPtr=/AmbientAudioContent/Waves/OneShots/Spot_MudBlobA005}}
 {id=WavePlayer_14, properties {soundWaveAssetPtr=/AmbientAudioContent/Waves/OneShots/Spot_MudBlobA006}}
 {id=WavePlayer_8, properties {bLooping=true, soundWaveAssetPtr=&Spot_MudBlob002_Loop}}
 {id=WavePlayer_9, properties {soundWaveAssetPtr=/AmbientAudioContent/Waves/OneShots/Spot_MudBlobA001}}
]
properties {
 bOverrideConcurrency=true
 maxDistance=1800
 soundClassObject=&Fort_Ambient_Outdoors_Athena
 volumeMultiplier=2.005
 attenuationOverrides {attenuationShapeExtents {x=0}}
 concurrencyOverrides {maxCount=3, resolutionRule=StopQuietest}
}
