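# One-shot ambient crane bird call cue for Sakura Springs, using outdoor ambient sound class and bird one-shot attenuation.
# A mixer layers three random picks from the same seven crane waves: one immediate, one after a short delay, and a quieter one after a longer delay.
# The late layer has extra random volume that can fall to silence, and a root modulator varies overall pitch and volume, so each trigger sounds like a different group call.
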SoundCue /AmbientAudioContent/SoundCues/OneShots/Spot_Sakura_Springs_Crane_Cue
root=Modulator_1
nodes [
 {id=Delay_0, children [Random_3], properties {delayMax=0.8, delayMin=0.4}}
 {id=Delay_1, children [Modulator_2], properties {delayMax=1.3, delayMin=0.9}}
 {id=Mixer_2, children [Delay_0,Random_4,Delay_1], properties {inputVolume [1,1,0.3]}}
 {id=Modulator_1, children [Mixer_2], properties {pitchMax=1.1, pitchMin=0.9, volumeMax=1.2, volumeMin=0.8}}
 {id=Modulator_2, children [Random_5], properties {volumeMax=0.8, volumeMin=0}}
 {
  id=Random_3
  children [WavePlayer_0,WavePlayer_1,WavePlayer_2,WavePlayer_3,WavePlayer_4,WavePlayer_5,WavePlayer_6]
  properties {weights [1,1,1,1,1,1,1]}
 }
 {
  id=Random_4
  children [WavePlayer_7, WavePlayer_9, WavePlayer_15, WavePlayer_16, WavePlayer_17, WavePlayer_18, WavePlayer_19]
  properties {weights [1,1,1,1,1,1,1]}
 }
 {
  id=Random_5
  children [WavePlayer_20, WavePlayer_21, WavePlayer_22, WavePlayer_23, WavePlayer_39, WavePlayer_40, WavePlayer_41]
  properties {weights [1,1,1,1,1,1,1]}
 }
 {id=WavePlayer_0, properties {soundWaveAssetPtr=/AmbientAudioContent/Waves/OneShots/Spot_Bird_Crane_7}}
 {id=WavePlayer_1, properties {soundWaveAssetPtr=/AmbientAudioContent/Waves/OneShots/Spot_Bird_Crane_6}}
 {id=WavePlayer_15, properties {soundWaveAssetPtr=/AmbientAudioContent/Waves/OneShots/Spot_Bird_Crane_5}}
 {id=WavePlayer_16, properties {soundWaveAssetPtr=/AmbientAudioContent/Waves/OneShots/Spot_Bird_Crane_4}}
 {id=WavePlayer_17, properties {soundWaveAssetPtr=/AmbientAudioContent/Waves/OneShots/Spot_Bird_Crane_3}}
 {id=WavePlayer_18, properties {soundWaveAssetPtr=/AmbientAudioContent/Waves/OneShots/Spot_Bird_Crane_2}}
 {id=WavePlayer_19, properties {soundWaveAssetPtr=/AmbientAudioContent/Waves/OneShots/Spot_Bird_Crane_1}}
 {id=WavePlayer_2, properties {soundWaveAssetPtr=&Spot_Bird_Crane_5}}
 {id=WavePlayer_20, properties {soundWaveAssetPtr=&Spot_Bird_Crane_7}}
 {id=WavePlayer_21, properties {soundWaveAssetPtr=&Spot_Bird_Crane_6}}
 {id=WavePlayer_22, properties {soundWaveAssetPtr=&Spot_Bird_Crane_5}}
 {id=WavePlayer_23, properties {soundWaveAssetPtr=&Spot_Bird_Crane_4}}
 {id=WavePlayer_3, properties {soundWaveAssetPtr=&Spot_Bird_Crane_4}}
 {id=WavePlayer_39, properties {soundWaveAssetPtr=&Spot_Bird_Crane_3}}
 {id=WavePlayer_4, properties {soundWaveAssetPtr=&Spot_Bird_Crane_3}}
 {id=WavePlayer_40, properties {soundWaveAssetPtr=&Spot_Bird_Crane_2}}
 {id=WavePlayer_41, properties {soundWaveAssetPtr=&Spot_Bird_Crane_1}}
 {id=WavePlayer_5, properties {soundWaveAssetPtr=&Spot_Bird_Crane_2}}
 {id=WavePlayer_6, properties {soundWaveAssetPtr=&Spot_Bird_Crane_1}}
 {id=WavePlayer_7, properties {soundWaveAssetPtr=&Spot_Bird_Crane_7}}
 {id=WavePlayer_9, properties {soundWaveAssetPtr=&Spot_Bird_Crane_6}}
]
properties {
 attenuationSettings=&Fort_TileAmbient_OneShot_Bird_Att
 maxDistance=8005
 soundClassObject=&Fort_Ambient_Outdoors_Athena
 volumeMultiplier=0.226
 preEffectBusSends [{}]
}
