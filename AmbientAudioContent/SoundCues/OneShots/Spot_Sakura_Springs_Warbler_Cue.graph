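# Sound Cue for a one-shot Japanese warbler bird call, used as an outdoor ambient spot sound for Sakura Springs.
# A Random node picks one of six warbler waves with equal weight, feeding a Modulator that varies pitch and volume.
# All four Quality Level inputs share that same branch.
# Playback uses bird one-shot attenuation, the outdoor ambient sound class and a distance-scaled pre-effect bus send.
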SoundCue /AmbientAudioContent/SoundCues/OneShots/Spot_Sakura_Springs_Warbler_Cue
root=QualityLevel_0
nodes [
 {id=Modulator_0, children [Random_2], properties {pitchMax=1.1, pitchMin=0.9, volumeMax=1.2, volumeMin=0.5}}
 {id=QualityLevel_0, children [Modulator_0,Modulator_0,Modulator_0,Modulator_0]}
 {
  id=Random_2
  children [WavePlayer_9, WavePlayer_10, WavePlayer_11, WavePlayer_12, WavePlayer_13, WavePlayer_14]
  properties {weights [1,1,1,1,1,1]}
 }
 {id=WavePlayer_10, properties {soundWaveAssetPtr=/AmbientAudioContent/Waves/OneShots/Spot_Bird_JapaneseWarbler_5}}
 {id=WavePlayer_11, properties {soundWaveAssetPtr=/AmbientAudioContent/Waves/OneShots/Spot_Bird_JapaneseWarbler_4}}
 {id=WavePlayer_12, properties {soundWaveAssetPtr=/AmbientAudioContent/Waves/OneShots/Spot_Bird_JapaneseWarbler_3}}
 {id=WavePlayer_13, properties {soundWaveAssetPtr=/AmbientAudioContent/Waves/OneShots/Spot_Bird_JapaneseWarbler_2}}
 {id=WavePlayer_14, properties {soundWaveAssetPtr=/AmbientAudioContent/Waves/OneShots/Spot_Bird_JapaneseWarbler_1}}
 {id=WavePlayer_9, properties {soundWaveAssetPtr=/AmbientAudioContent/Waves/OneShots/Spot_Bird_JapaneseWarbler_6}}
]
properties {
 attenuationSettings=&Fort_TileAmbient_OneShot_Bird_Att
 maxDistance=8005
 soundClassObject=&Fort_Ambient_Outdoors_Athena
 volumeMultiplier=0.226
 preEffectBusSends [{maxSendDistance=8000, minSendDistance=3000, minSendLevel=0.4, customSendLevelCurve {editorCurveData {keys [<at 0 0.4>,<at 1 1>]}}}]
}
