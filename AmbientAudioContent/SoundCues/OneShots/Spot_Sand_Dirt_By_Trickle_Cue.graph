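# Sound cue for outdoor ambient one-shot spot sounds of trickling sand and dirt, with distance attenuation and a low concurrency cap.
# A quality switch sends its first two levels to a random pick between trickle and mountain sand banks of six waves, each with its own modulation.
# The last two levels use a cut-down pool of three mountain sand waves, with a wider downward pitch range for variation.
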SoundCue /AmbientAudioContent/SoundCues/OneShots/Spot_Sand_Dirt_By_Trickle_Cue
root=QualityLevel_0
nodes [
 {id=Modulator_0, children [Random_4], properties {volumeMax=1.4, volumeMin=1}}
 {id=Modulator_1, children [Random_5], properties {pitchMax=1.1, volumeMax=1.2, volumeMin=1}}
 {id=Modulator_2, children [Random_1], properties {pitchMax=1.1, pitchMin=0.5, volumeMax=1.2, volumeMin=1}}
 {id=QualityLevel_0, children [Random_2,Random_2,Modulator_2,Modulator_2]}
 {id=Random_1, children [WavePlayer_6,WavePlayer_7,WavePlayer_8], properties {weights [1,1,1]}}
 {id=Random_2, children [Modulator_0,Modulator_1], properties {weights [1,1]}}
 {
  id=Random_4
  children [WavePlayer_15, WavePlayer_16, WavePlayer_17, WavePlayer_18, WavePlayer_19, WavePlayer_20]
  properties {weights [1,1,1,1,1,1]}
 }
 {
  id=Random_5
  children [WavePlayer_21, WavePlayer_22, WavePlayer_23, WavePlayer_24, WavePlayer_25, WavePlayer_26]
  properties {weights [1,1,1,1,1,1]}
 }
 {id=WavePlayer_15, properties {soundWaveAssetPtr=/AmbientAudioContent/Waves/OneShots/Spot_Sand_Dirt_By_Trickle_01}}
 {id=WavePlayer_16, properties {soundWaveAssetPtr=/AmbientAudioContent/Waves/OneShots/Spot_Sand_Dirt_By_Trickle_02}}
 {id=WavePlayer_17, properties {soundWaveAssetPtr=/AmbientAudioContent/Waves/OneShots/Spot_Sand_Dirt_By_Trickle_03}}
 {id=WavePlayer_18, properties {soundWaveAssetPtr=/AmbientAudioContent/Waves/OneShots/Spot_Sand_Dirt_By_Trickle_04}}
 {id=WavePlayer_19, properties {soundWaveAssetPtr=/AmbientAudioContent/Waves/OneShots/Spot_Sand_Dirt_By_Trickle_05}}
 {id=WavePlayer_20, properties {soundWaveAssetPtr=/AmbientAudioContent/Waves/OneShots/Spot_Sand_Dirt_By_Trickle_06}}
 {id=WavePlayer_21, properties {soundWaveAssetPtr=/AmbientAudioContent/Waves/OneShots/Spot_Mountain_Sand_Bys_01}}
 {id=WavePlayer_22, properties {soundWaveAssetPtr=/AmbientAudioContent/Waves/OneShots/Spot_Mountain_Sand_Bys_02}}
 {id=WavePlayer_23, properties {soundWaveAssetPtr=/AmbientAudioContent/Waves/OneShots/Spot_Mountain_Sand_Bys_03}}
 {id=WavePlayer_24, properties {soundWaveAssetPtr=/AmbientAudioContent/Waves/OneShots/Spot_Mountain_Sand_Bys_04}}
 {id=WavePlayer_25, properties {soundWaveAssetPtr=/AmbientAudioContent/Waves/OneShots/Spot_Mountain_Sand_Bys_05}}
 {id=WavePlayer_26, properties {soundWaveAssetPtr=/AmbientAudioContent/Waves/OneShots/Spot_Mountain_Sand_Bys_06}}
 {id=WavePlayer_6, properties {soundWaveAssetPtr=&Spot_Mountain_Sand_Bys_03}}
 {id=WavePlayer_7, properties {soundWaveAssetPtr=&Spot_Mountain_Sand_Bys_06}}
 {id=WavePlayer_8, properties {soundWaveAssetPtr=&Spot_Mountain_Sand_Bys_01}}
]
properties {
 attenuationSettings=/Game/Sounds/Attenuation/Fort_TileAmbient_OneShot_Distant_Creak_Att
 bEnableBusSends=false
 bOverrideConcurrency=true
 maxDistance=8200
 soundClassObject=&Fort_Ambient_OneShots_Outdoors
 volumeMultiplier=1
 concurrencyOverrides {maxCount=2}
 preEffectBusSends [{}]
}
