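# Ambient one-shot SoundCue that plays light chain rattles, named as a spot sound for the Seven Workshop.
# A Random node picks among six equally weighted chain movement waves reused from the Tilted Towers set, then a Modulator varies each play with a lowered minimum pitch.
# It routes to the ambient one-shots sound class and references a less-positional, logarithmic tile-ambient attenuation asset.
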SoundCue /AmbientAudioContent/SoundCues/OneShots/Spot_SevenWorkshop_Chain_Rattles_Cue
root=Modulator_0
nodes [
 {id=Modulator_0, children [Random_0], properties {pitchMin=0.9}}
 {
  id=Random_0
  children [WavePlayer_0, WavePlayer_1, WavePlayer_2, WavePlayer_9, WavePlayer_10, WavePlayer_11]
  properties {weights [1,1,1,1,1,1]}
 }
 {id=WavePlayer_0, properties {soundWaveAssetPtr=/AmbientAudioContent/Waves/OneShots/Spot_Tilted_Towers_Chain_Movement_Light_01}}
 {id=WavePlayer_1, properties {soundWaveAssetPtr=/AmbientAudioContent/Waves/OneShots/Spot_Tilted_Towers_Chain_Movement_Light_02}}
 {id=WavePlayer_10, properties {soundWaveAssetPtr=/AmbientAudioContent/Waves/OneShots/Spot_Tilted_Towers_Chain_Movement_Light_05}}
 {id=WavePlayer_11, properties {soundWaveAssetPtr=/AmbientAudioContent/Waves/OneShots/Spot_Tilted_Towers_Chain_Movement_Light_06}}
 {id=WavePlayer_2, properties {soundWaveAssetPtr=/AmbientAudioContent/Waves/OneShots/Spot_Tilted_Towers_Chain_Movement_Light_03}}
 {id=WavePlayer_9, properties {soundWaveAssetPtr=/AmbientAudioContent/Waves/OneShots/Spot_Tilted_Towers_Chain_Movement_Light_04}}
]
properties {
 attenuationSettings=/Game/Sounds/Attenuation/Fort_TileAmbient_OneShot_LessPositional_Log_Att
 maxDistance=6200
 soundClassObject=&Fort_Ambient_OneShots
 volumeMultiplier=1.002
 attenuationOverrides {falloffDistance=6000, stereoSpread=10000, attenuationShapeExtents {x=2000}}
}
