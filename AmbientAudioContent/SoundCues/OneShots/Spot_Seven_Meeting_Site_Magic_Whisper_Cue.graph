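# Sound cue for an outdoor ambient one-shot magic whisper, named for the Seven meeting site spot.
# A Random node picks among eight equally weighted whisper waves (09 to 16) and feeds a Modulator that varies pitch and volume.
# All four quality-level inputs share that same branch, so every level sounds alike.
# It borrows a bird one-shot attenuation asset and plays through the outdoor ambient Athena sound class at reduced volume.
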SoundCue /AmbientAudioContent/SoundCues/OneShots/Spot_Seven_Meeting_Site_Magic_Whisper_Cue
root=QualityLevel_0
nodes [
 {id=Modulator_1, children [Random_0], properties {pitchMax=1.1, volumeMax=1.1, volumeMin=0.9}}
 {id=QualityLevel_0, children [Modulator_1,Modulator_1,Modulator_1,Modulator_1]}
 {
  id=Random_0
  children [WavePlayer_6, WavePlayer_7, WavePlayer_8, WavePlayer_9, WavePlayer_10, WavePlayer_11, WavePlayer_12, WavePlayer_13]
  properties {weights [1,1,1,1,1,1,1,1]}
 }
 {id=WavePlayer_10, properties {soundWaveAssetPtr=/AmbientAudioContent/Waves/OneShots/Spot_Seven_Meeting_Site_Magic_Whisper_13}}
 {id=WavePlayer_11, properties {soundWaveAssetPtr=/AmbientAudioContent/Waves/OneShots/Spot_Seven_Meeting_Site_Magic_Whisper_14}}
 {id=WavePlayer_12, properties {soundWaveAssetPtr=/AmbientAudioContent/Waves/OneShots/Spot_Seven_Meeting_Site_Magic_Whisper_15}}
 {id=WavePlayer_13, properties {soundWaveAssetPtr=/AmbientAudioContent/Waves/OneShots/Spot_Seven_Meeting_Site_Magic_Whisper_16}}
 {id=WavePlayer_6, properties {soundWaveAssetPtr=/AmbientAudioContent/Waves/OneShots/Spot_Seven_Meeting_Site_Magic_Whisper_09}}
 {id=WavePlayer_7, properties {soundWaveAssetPtr=/AmbientAudioContent/Waves/OneShots/Spot_Seven_Meeting_Site_Magic_Whisper_10}}
 {id=WavePlayer_8, properties {soundWaveAssetPtr=/AmbientAudioContent/Waves/OneShots/Spot_Seven_Meeting_Site_Magic_Whisper_11}}
 {id=WavePlayer_9, properties {soundWaveAssetPtr=/AmbientAudioContent/Waves/OneShots/Spot_Seven_Meeting_Site_Magic_Whisper_12}}
]
properties {
 attenuationSettings=&Fort_TileAmbient_OneShot_Bird_Att
 maxDistance=8005
 soundClassObject=&Fort_Ambient_Outdoors_Athena
 volumeMultiplier=0.752
 preEffectBusSends [{}]
}
