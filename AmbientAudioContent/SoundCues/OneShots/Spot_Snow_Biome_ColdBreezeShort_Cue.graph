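# Sound cue for short cold-breeze one-shots, an outdoor ambient spot sound for the snow biome.
# A Random node picks one of six winter breeze waves with equal weight and feeds a Modulator that varies pitch and volume on each play.
# The cue plays at reduced volume through the outdoor ambient one-shot sound class and a shared wind one-shot attenuation asset.
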SoundCue /AmbientAudioContent/SoundCues/OneShots/Spot_Snow_Biome_ColdBreezeShort_Cue
root=Modulator_1
nodes [
 {id=Modulator_1, children [Random_0], properties {pitchMax=1.2, volumeMax=1.1, volumeMin=0.75}}
 {
  id=Random_0
  children [WavePlayer_1, WavePlayer_2, WavePlayer_3, WavePlayer_4, WavePlayer_5, WavePlayer_10]
  properties {weights [1,1,1,1,1,1]}
 }
 {id=WavePlayer_1, properties {soundWaveAssetPtr=/AmbientAudioContent/Waves/OneShots/Spot_Winter_Oneshot_ColdBreezeShort_01}}
 {id=WavePlayer_10, properties {soundWaveAssetPtr=/AmbientAudioContent/Waves/OneShots/Spot_Winter_Oneshot_ColdBreezeShort_07}}
 {id=WavePlayer_2, properties {soundWaveAssetPtr=/AmbientAudioContent/Waves/OneShots/Spot_Winter_Oneshot_ColdBreezeShort_03}}
 {id=WavePlayer_3, properties {soundWaveAssetPtr=/AmbientAudioContent/Waves/OneShots/Spot_Winter_Oneshot_ColdBreezeShort_04}}
 {id=WavePlayer_4, properties {soundWaveAssetPtr=/AmbientAudioContent/Waves/OneShots/Spot_Winter_Oneshot_ColdBreezeShort_05}}
 {id=WavePlayer_5, properties {soundWaveAssetPtr=/AmbientAudioContent/Waves/OneShots/Spot_Winter_Oneshot_ColdBreezeShort_06}}
]
properties {
 attenuationSettings=/Game/Sounds/Attenuation/Fort_TileAmbient_OneShot_Wind_Att
 maxDistance=8000
 soundClassObject=&Fort_Ambient_OneShots_Outdoors
 volumeMultiplier=0.251
 preEffectBusSends [{}]
}
