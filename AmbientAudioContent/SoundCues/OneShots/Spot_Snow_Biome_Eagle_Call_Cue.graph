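# Sound cue for distant one-shot eagle calls in snow biome ambience, using an ambient one-shot class and distant-creature attenuation.
# Each call randomly picks one of five mountain eagle waves, with random pitch and volume modulation.
# The first two quality-level inputs randomly play one, two or three calls staggered by short random delays, favoring two.
# The last two inputs play only a single modulated call.
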SoundCue /AmbientAudioContent/SoundCues/OneShots/Spot_Snow_Biome_Eagle_Call_Cue
root=QualityLevel_0
nodes [
 {id=Delay_0, children [Modulator_2], properties {delayMax=0.4, delayMin=0.25}}
 {id=Delay_1, children [Modulator_0], properties {delayMax=0.8, delayMin=0.7}}
 {id=Mixer_0, children [Modulator_1,Delay_0,Delay_1], properties {inputVolume [1,1,1]}}
 {id=Mixer_1, children [Modulator_1,Delay_0], properties {inputVolume [1,1]}}
 {id=Modulator_0, children [Random_3], properties {pitchMax=1.1, pitchMin=0.9, volumeMax=1.1, volumeMin=0.9}}
 {id=Modulator_1, children [Random_3], properties {pitchMax=1.1, pitchMin=0.9, volumeMax=1.1, volumeMin=0.9}}
 {id=Modulator_2, children [Random_3], properties {pitchMax=1.2, pitchMin=1, volumeMax=1.1, volumeMin=0.9}}
 {id=Modulator_3, children [Random_3], properties {pitchMax=1.1, pitchMin=0.9, volumeMax=1.1, volumeMin=0.9}}
 {id=QualityLevel_0, children [Random_2,Random_2,Modulator_3,Modulator_3]}
 {id=Random_2, children [Modulator_1,Mixer_1,Mixer_0], properties {weights [0.33,1,0.1]}}
 {id=Random_3, children [WavePlayer_6, WavePlayer_7, WavePlayer_22, WavePlayer_23, WavePlayer_24], properties {weights [1,1,1,1,1]}}
 {id=WavePlayer_22, properties {soundWaveAssetPtr=/AmbientAudioContent/Waves/OneShots/Spot_Mountain_Eagle_03}}
 {id=WavePlayer_23, properties {soundWaveAssetPtr=/AmbientAudioContent/Waves/OneShots/Spot_Mountain_Eagle_04}}
 {id=WavePlayer_24, properties {soundWaveAssetPtr=/AmbientAudioContent/Waves/OneShots/Spot_Mountain_Eagle_05}}
 {id=WavePlayer_6, properties {soundWaveAssetPtr=/AmbientAudioContent/Waves/OneShots/Spot_Mountain_Eagle_01}}
 {id=WavePlayer_7, properties {soundWaveAssetPtr=/AmbientAudioContent/Waves/OneShots/Spot_Mountain_Eagle_02}}
]
properties {
 attenuationSettings=/Game/Sounds/Attenuation/Fort_TileAmbient_OneShot_Distant_Creature_Att
 bEnableBusSends=false
 maxDistance=8500
 soundClassObject=&Fort_Ambient_OneShots
 volumeMultiplier=0.351
 preEffectBusSends [{}]
}
