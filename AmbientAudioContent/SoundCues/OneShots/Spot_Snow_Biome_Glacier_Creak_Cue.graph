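# Sound cue for one-shot glacier creak spot ambience in the snow biome, classed as outdoor ambient one-shots.
# A Random node picks among seven equally weighted creak waves, feeding a Modulator that varies volume and slightly raises pitch.
# It uses a linear tile-ambient attenuation asset and sends to an exterior ambient source bus and a distant reverb submix.
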SoundCue /AmbientAudioContent/SoundCues/OneShots/Spot_Snow_Biome_Glacier_Creak_Cue
root=Modulator_0
nodes [
 {id=Modulator_0, children [Random_0], properties {pitchMax=1.1, pitchMin=1, volumeMin=0.6}}
 {
  id=Random_0
  children [WavePlayer_1, WavePlayer_2, WavePlayer_3, WavePlayer_4, WavePlayer_5, WavePlayer_10, WavePlayer_11]
  properties {weights [1,1,1,1,1,1,1]}
 }
 {id=WavePlayer_1, properties {soundWaveAssetPtr=/AmbientAudioContent/Waves/OneShots/Spot_Glacier_Creak_01}}
 {id=WavePlayer_10, properties {soundWaveAssetPtr=/AmbientAudioContent/Waves/OneShots/Spot_Glacier_Creak_06}}
 {id=WavePlayer_11, properties {soundWaveAssetPtr=/AmbientAudioContent/Waves/OneShots/Spot_Glacier_Creak_07}}
 {id=WavePlayer_2, properties {soundWaveAssetPtr=/AmbientAudioContent/Waves/OneShots/Spot_Glacier_Creak_02}}
 {id=WavePlayer_3, properties {soundWaveAssetPtr=/AmbientAudioContent/Waves/OneShots/Spot_Glacier_Creak_03}}
 {id=WavePlayer_4, properties {soundWaveAssetPtr=/AmbientAudioContent/Waves/OneShots/Spot_Glacier_Creak_04}}
 {id=WavePlayer_5, properties {soundWaveAssetPtr=/AmbientAudioContent/Waves/OneShots/Spot_Glacier_Creak_05}}
]
properties {
 attenuationSettings=/Game/Sounds/Attenuation/Fort_TileAmbient_OneShot_Linear_Att
 maxDistance=8000
 soundClassObject=&Fort_Ambient_OneShots_Outdoors
 volumeMultiplier=0.376
 busSends [{audioBus=null, minSendLevel=0.6, sendLevel=1, soundSourceBus=&AmbientAudio_Exterior_SourceBus}]
 soundSubmixSends [{minSendLevel=1, soundSubmix=&Fortnite_Reverb_Distant}]
}
