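# Ambient one-shot sound cue that plays a textured ice pass-by spot effect for the snow biome.
# A Random node picks evenly among five ice wave variations and feeds a Modulator.
# A Quality Level node at the root routes the first two quality slots to that Modulator and leaves the other two empty.
# The cue uses wind one-shot attenuation and the ambient one-shot sound class, and sends to the exterior ambient source bus.
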SoundCue /AmbientAudioContent/SoundCues/OneShots/Spot_Snow_Biome_Ice_Textured_By_Cue
root=QualityLevel_1
nodes [
 {id=Modulator_5, children [Random_0]}
 {id=QualityLevel_1, children [Modulator_5,Modulator_5,null,null]}
 {id=Random_0, children [WavePlayer_0,WavePlayer_1,WavePlayer_2,WavePlayer_3,WavePlayer_4], properties {weights [1,1,1,1,1]}}
 {id=WavePlayer_0, properties {soundWaveAssetPtr=/AmbientAudioContent/Waves/OneShots/Spot_RealityWarrior_Ice_Textured_By_01}}
 {id=WavePlayer_1, properties {soundWaveAssetPtr=/AmbientAudioContent/Waves/OneShots/Spot_RealityWarrior_Ice_Textured_By_02}}
 {id=WavePlayer_2, properties {soundWaveAssetPtr=/AmbientAudioContent/Waves/OneShots/Spot_RealityWarrior_Ice_Textured_By_03}}
 {id=WavePlayer_3, properties {soundWaveAssetPtr=/AmbientAudioContent/Waves/OneShots/Spot_RealityWarrior_Ice_Textured_By_04}}
 {id=WavePlayer_4, properties {soundWaveAssetPtr=/AmbientAudioContent/Waves/OneShots/Spot_RealityWarrior_Ice_Textured_By_06}}
]
properties {
 attenuationSettings=/Game/Sounds/Attenuation/Fort_TileAmbient_OneShot_Wind_Att
 maxDistance=8000
 soundClassObject=&Fort_Ambient_OneShots
 volumeMultiplier=0.401
 busSends [{audioBus=null, minSendLevel=0.5, sendLevel=1, soundSourceBus=&AmbientAudio_Exterior_SourceBus}]
}
