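# Sound Cue for a snow-crunch ambient one-shot in the snow biome, played as a spot sound.
# A Random node picks one of five equally weighted winter crunch waves and feeds a Modulator that slightly varies pitch.
# It plays at reduced volume in the outdoor ambient one-shots sound class, using a shared wind one-shot tile-ambient attenuation asset.
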SoundCue /AmbientAudioContent/SoundCues/OneShots/Spot_Snow_Biome_SnowCrunch_Cue
root=Modulator_0
nodes [
 {id=Modulator_0, children [Random_0], properties {pitchMax=1.1, pitchMin=0.9}}
 {id=Random_0, children [WavePlayer_1,WavePlayer_2,WavePlayer_3,WavePlayer_4,WavePlayer_5], properties {weights [1,1,1,1,1]}}
 {id=WavePlayer_1, properties {soundWaveAssetPtr=/AmbientAudioContent/Waves/OneShots/Spot_Winter_Oneshot_SnowCrunch_01}}
 {id=WavePlayer_2, properties {soundWaveAssetPtr=/AmbientAudioContent/Waves/OneShots/Spot_Winter_Oneshot_SnowCrunch_02}}
 {id=WavePlayer_3, properties {soundWaveAssetPtr=/AmbientAudioContent/Waves/OneShots/Spot_Winter_Oneshot_SnowCrunch_03}}
 {id=WavePlayer_4, properties {soundWaveAssetPtr=/AmbientAudioContent/Waves/OneShots/Spot_Winter_Oneshot_SnowCrunch_04}}
 {id=WavePlayer_5, properties {soundWaveAssetPtr=/AmbientAudioContent/Waves/OneShots/Spot_Winter_Oneshot_SnowCrunch_05}}
]
properties {
 attenuationSettings=/Game/Sounds/Attenuation/Fort_TileAmbient_OneShot_Wind_Att
 maxDistance=8000
 soundClassObject=&Fort_Ambient_OneShots_Outdoors
 volumeMultiplier=0.301
 preEffectBusSends [{}]
}
