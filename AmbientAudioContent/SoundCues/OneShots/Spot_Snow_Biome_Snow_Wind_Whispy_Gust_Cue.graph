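# One-shot ambient Sound Cue for wispy wind gusts in a snow biome, reusing desert gust whistle waves.
# A Random node picks among five equally weighted whistle waves and feeds a Modulator that boosts and slightly varies volume.
# A Quality Level root routes every level to that same Modulator.
# Playback uses the outdoor ambient one-shot sound class and shared tile-ambient attenuation, with a reduced cue volume multiplier.
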SoundCue /AmbientAudioContent/SoundCues/OneShots/Spot_Snow_Biome_Snow_Wind_Whispy_Gust_Cue
root=QualityLevel_0
nodes [
 {id=Modulator_0, children [Random_0], properties {volumeMax=2.2, volumeMin=2}}
 {id=QualityLevel_0, children [Modulator_0,Modulator_0,Modulator_0,Modulator_0]}
 {
  id=Random_0
  children [WavePlayer_0,WavePlayer_1,WavePlayer_2,WavePlayer_3,WavePlayer_4]
  properties {bShouldExcludeFromBranchCulling=true, weights [1,1,1,1,1]}
 }
 {id=WavePlayer_0, properties {soundWaveAssetPtr=/AmbientAudioContent/Waves/OneShots/Spot_Wind_DesertGust_Whistle14}}
 {id=WavePlayer_1, properties {soundWaveAssetPtr=/AmbientAudioContent/Waves/OneShots/Spot_Wind_DesertGust_Whistle15}}
 {id=WavePlayer_2, properties {soundWaveAssetPtr=/AmbientAudioContent/Waves/OneShots/Spot_Wind_DesertGust_Whistle16}}
 {id=WavePlayer_3, properties {soundWaveAssetPtr=/AmbientAudioContent/Waves/OneShots/Spot_Wind_DesertGust_Whistle18}}
 {id=WavePlayer_4, properties {soundWaveAssetPtr=/AmbientAudioContent/Waves/OneShots/Spot_Wind_DesertGust_Whistle26}}
]
properties {
 attenuationSettings=/Game/Sounds/Attenuation/Fort_TileAmbient_OneShot_Att
 bEnableBusSends=false
 maxDistance=8000
 soundClassObject=&Fort_Ambient_OneShots_Outdoors
 volumeMultiplier=0.501
 preEffectBusSends [{}]
}
