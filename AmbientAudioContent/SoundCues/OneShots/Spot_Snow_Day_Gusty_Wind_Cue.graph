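# One-shot outdoor ambient cue for gusty daytime snow wind, attenuated for spot playback.
# A quality switch sends its first two slots to a random pick of two mixes: grassland rustle with either quad alpine wind or six mountain gusts.
# The last two slots use a smaller mix of three mountain gusts and alpine "ST" wind.
# Each layer randomly picks a wave and varies its pitch and volume.
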SoundCue /AmbientAudioContent/SoundCues/OneShots/Spot_Snow_Day_Gusty_Wind_Cue
root=QualityLevel_1
nodes [
 {id=Mixer_1, children [Modulator_2,Modulator_0], properties {inputVolume [1,1]}}
 {id=Mixer_2, children [Modulator_3,Modulator_4], properties {inputVolume [1,1]}}
 {id=Mixer_3, children [Modulator_4,Modulator_1], properties {inputVolume [1,1]}}
 {id=Modulator_0, children [Random_9], properties {pitchMax=1.1, pitchMin=0.9, volumeMax=0.9, volumeMin=0.6}}
 {id=Modulator_1, children [Random_8], properties {pitchMax=1.1, volumeMax=1, volumeMin=0.6}}
 {id=Modulator_2, children [Random_1], properties {pitchMax=1.1, volumeMax=1, volumeMin=0.7}}
 {id=Modulator_3, children [Random_6], properties {pitchMax=1.1, pitchMin=0.9, volumeMax=1, volumeMin=0.6}}
 {id=Modulator_4, children [Random_7], properties {pitchMax=1.1, pitchMin=0.9, volumeMax=1, volumeMin=0.6}}
 {id=QualityLevel_1, children [Random_5,Random_5,Mixer_1,Mixer_1]}
 {id=Random_1, children [WavePlayer_6,WavePlayer_7,WavePlayer_8], properties {weights [1,1,1]}}
 {id=Random_5, children [Mixer_2,Mixer_3], properties {weights [1,1]}}
 {id=Random_6, children [WavePlayer_20, WavePlayer_27, WavePlayer_28, WavePlayer_29], properties {weights [1,1,1,1]}}
 {
  id=Random_7
  children [
   WavePlayer_30
   WavePlayer_31
   WavePlayer_32
   WavePlayer_33
   WavePlayer_34
   WavePlayer_35
   WavePlayer_36
   WavePlayer_37
   WavePlayer_38
   WavePlayer_39
  ]
  properties {weights [1,1,1,1,1,1,1,1,1,1]}
 }
 {
  id=Random_8
  children [WavePlayer_40, WavePlayer_41, WavePlayer_42, WavePlayer_43, WavePlayer_44, WavePlayer_45]
  properties {weights [1,1,1,1,1,1]}
 }
 {id=Random_9, children [WavePlayer_46, WavePlayer_47, WavePlayer_48, WavePlayer_49], properties {weights [1,1,1,1]}}
 {id=WavePlayer_20, properties {soundWaveAssetPtr=/AmbientAudioContent/Waves/OneShots/Spot_Wind_Alpine_Oneshot_Day_Quad_01}}
 {id=WavePlayer_27, properties {soundWaveAssetPtr=/AmbientAudioContent/Waves/OneShots/Spot_Wind_Alpine_Oneshot_Day_Quad_02}}
 {id=WavePlayer_28, properties {soundWaveAssetPtr=/AmbientAudioContent/Waves/OneShots/Spot_Wind_Alpine_Oneshot_Day_Quad_03}}
 {id=WavePlayer_29, properties {soundWaveAssetPtr=/AmbientAudioContent/Waves/OneShots/Spot_Wind_Alpine_Oneshot_Day_Quad_04}}
 {id=WavePlayer_30, properties {soundWaveAssetPtr=/AmbientAudioContent/Waves/OneShots/Spot_Grassland_Wind_Rustle_01}}
 {id=WavePlayer_31, properties {soundWaveAssetPtr=/AmbientAudioContent/Waves/OneShots/Spot_Grassland_Wind_Rustle_02}}
 {id=WavePlayer_32, properties {soundWaveAssetPtr=/AmbientAudioContent/Waves/OneShots/Spot_Grassland_Wind_Rustle_03}}
 {id=WavePlayer_33, properties {soundWaveAssetPtr=/AmbientAudioContent/Waves/OneShots/Spot_Grassland_Wind_Rustle_04}}
 {id=WavePlayer_34, properties {soundWaveAssetPtr=/AmbientAudioContent/Waves/OneShots/Spot_Grassland_Wind_Rustle_05}}
 {id=WavePlayer_35, properties {soundWaveAssetPtr=/AmbientAudioContent/Waves/OneShots/Spot_Grassland_Wind_Rustle_06}}
 {id=WavePlayer_36, properties {soundWaveAssetPtr=/AmbientAudioContent/Waves/OneShots/Spot_Grassland_Wind_Rustle_07}}
 {id=WavePlayer_37, properties {soundWaveAssetPtr=/AmbientAudioContent/Waves/OneShots/Spot_Grassland_Wind_Rustle_08}}
 {id=WavePlayer_38, properties {soundWaveAssetPtr=/AmbientAudioContent/Waves/OneShots/Spot_Grassland_Wind_Rustle_09}}
 {id=WavePlayer_39, properties {soundWaveAssetPtr=/AmbientAudioContent/Waves/OneShots/Spot_Grassland_Wind_Rustle_10}}
 {id=WavePlayer_40, properties {soundWaveAssetPtr=/AmbientAudioContent/Waves/OneShots/Spot_Mountain_Day_Gusty_Wind_01}}
 {id=WavePlayer_41, properties {soundWaveAssetPtr=/AmbientAudioContent/Waves/OneShots/Spot_Mountain_Day_Gusty_Wind_04}}
 {id=WavePlayer_42, properties {soundWaveAssetPtr=/AmbientAudioContent/Waves/OneShots/Spot_Mountain_Day_Gusty_Wind_06}}
 {id=WavePlayer_43, properties {soundWaveAssetPtr=/AmbientAudioContent/Waves/OneShots/Spot_Mountain_Day_Gusty_Wind_02}}
 {id=WavePlayer_44, properties {soundWaveAssetPtr=/AmbientAudioContent/Waves/OneShots/Spot_Mountain_Day_Gusty_Wind_03}}
 {id=WavePlayer_45, properties {soundWaveAssetPtr=/AmbientAudioContent/Waves/OneShots/Spot_Mountain_Day_Gusty_Wind_05}}
 {id=WavePlayer_46, properties {soundWaveAssetPtr=/AmbientAudioContent/Waves/OneShots/Spot_Wind_Alpine_Oneshot_Day_ST_01}}
 {id=WavePlayer_47, properties {soundWaveAssetPtr=/AmbientAudioContent/Waves/OneShots/Spot_Wind_Alpine_Oneshot_Day_ST_02}}
 {id=WavePlayer_48, properties {soundWaveAssetPtr=/AmbientAudioContent/Waves/OneShots/Spot_Wind_Alpine_Oneshot_Day_ST_03}}
 {id=WavePlayer_49, properties {soundWaveAssetPtr=/AmbientAudioContent/Waves/OneShots/Spot_Wind_Alpine_Oneshot_Day_ST_04}}
 {id=WavePlayer_6, properties {soundWaveAssetPtr=&Spot_Mountain_Day_Gusty_Wind_03}}
 {id=WavePlayer_7, properties {soundWaveAssetPtr=&Spot_Mountain_Day_Gusty_Wind_04}}
 {id=WavePlayer_8, properties {soundWaveAssetPtr=&Spot_Mountain_Day_Gusty_Wind_01}}
]
properties {
 attenuationSettings=/Game/Sounds/Attenuation/Fort_TileAmbient_OneShot_Wind_Att
 bEnableBusSends=false
 maxDistance=8000
 soundClassObject=&Fort_Ambient_OneShots_Outdoors
 volumeMultiplier=0.251
 preEffectBusSends [{}]
}
