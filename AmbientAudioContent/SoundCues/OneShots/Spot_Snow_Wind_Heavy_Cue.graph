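# Sound cue for outdoor ambient one-shots of heavy snowy wind, attenuated by a tile-ambient wind one-shot setting.
# A quality-level node routes the first two levels to five heavy and five gust waves, and the last two to three of each.
# Each branch randomly picks heavy or gust with equal weight, then a random wave, with slight pitch and volume modulation for variation.
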SoundCue /AmbientAudioContent/SoundCues/OneShots/Spot_Snow_Wind_Heavy_Cue
root=QualityLevel_1
nodes [
 {id=Modulator_0, children [Random_7], properties {pitchMax=1.1, pitchMin=0.9, volumeMax=1.1, volumeMin=0.9}}
 {id=Modulator_1, children [Random_5], properties {pitchMax=1.1, pitchMin=0.9, volumeMax=1.1, volumeMin=0.9}}
 {id=Modulator_3, children [Random_4], properties {pitchMax=1.1, pitchMin=0.9, volumeMax=1.1, volumeMin=0.9}}
 {id=Modulator_5, children [Random_1], properties {pitchMax=1.1, pitchMin=0.9, volumeMax=1.1, volumeMin=0.9}}
 {id=QualityLevel_1, children [Random_2,Random_2,Random_6,Random_6]}
 {id=Random_1, children [WavePlayer_5,WavePlayer_6,WavePlayer_7], properties {weights [1,1,1]}}
 {id=Random_2, children [Modulator_3,Modulator_0], properties {weights [1,1]}}
 {id=Random_4, children [WavePlayer_16, WavePlayer_17, WavePlayer_18, WavePlayer_19, WavePlayer_20], properties {weights [1,1,1,1,1]}}
 {id=Random_5, children [WavePlayer_13, WavePlayer_14, WavePlayer_15], properties {weights [1,1,1]}}
 {id=Random_6, children [Modulator_5,Modulator_1], properties {weights [1,1]}}
 {id=Random_7, children [WavePlayer_21, WavePlayer_22, WavePlayer_23, WavePlayer_24, WavePlayer_25], properties {weights [1,1,1,1,1]}}
 {id=WavePlayer_13, properties {soundWaveAssetPtr=/AmbientAudioContent/Waves/OneShots/Spot_Snow_Wind_Gust_01}}
 {id=WavePlayer_14, properties {soundWaveAssetPtr=/AmbientAudioContent/Waves/OneShots/Spot_Snow_Wind_Gust_02}}
 {id=WavePlayer_15, properties {soundWaveAssetPtr=/AmbientAudioContent/Waves/OneShots/Spot_Snow_Wind_Gust_03}}
 {id=WavePlayer_16, properties {soundWaveAssetPtr=/AmbientAudioContent/Waves/OneShots/Spot_Snow_Wind_Heavy_01}}
 {id=WavePlayer_17, properties {soundWaveAssetPtr=/AmbientAudioContent/Waves/OneShots/Spot_Snow_Wind_Heavy_02}}
 {id=WavePlayer_18, properties {soundWaveAssetPtr=/AmbientAudioContent/Waves/OneShots/Spot_Snow_Wind_Heavy_03}}
 {id=WavePlayer_19, properties {soundWaveAssetPtr=/AmbientAudioContent/Waves/OneShots/Spot_Snow_Wind_Heavy_04}}
 {id=WavePlayer_20, properties {soundWaveAssetPtr=/AmbientAudioContent/Waves/OneShots/Spot_Snow_Wind_Heavy_05}}
 {id=WavePlayer_21, properties {soundWaveAssetPtr=&Spot_Snow_Wind_Gust_01}}
 {id=WavePlayer_22, properties {soundWaveAssetPtr=&Spot_Snow_Wind_Gust_02}}
 {id=WavePlayer_23, properties {soundWaveAssetPtr=&Spot_Snow_Wind_Gust_03}}
 {id=WavePlayer_24, properties {soundWaveAssetPtr=/AmbientAudioContent/Waves/OneShots/Spot_Snow_Wind_Gust_04}}
 {id=WavePlayer_25, properties {soundWaveAssetPtr=/AmbientAudioContent/Waves/OneShots/Spot_Snow_Wind_Gust_05}}
 {id=WavePlayer_5, properties {soundWaveAssetPtr=&Spot_Snow_Wind_Heavy_03}}
 {id=WavePlayer_6, properties {soundWaveAssetPtr=&Spot_Snow_Wind_Heavy_04}}
 {id=WavePlayer_7, properties {soundWaveAssetPtr=&Spot_Snow_Wind_Heavy_05}}
]
properties {
 attenuationSettings=/Game/Sounds/Attenuation/Fort_TileAmbient_OneShot_Wind_Att
 bEnableBusSends=false
 maxDistance=8000
 soundClassObject=&Fort_Ambient_OneShots_Outdoors
 volumeMultiplier=0.251
 preEffectBusSends [{}]
}
