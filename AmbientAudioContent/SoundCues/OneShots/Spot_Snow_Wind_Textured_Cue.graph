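# Sound cue for textured snow wind one-shots, played as attenuated outdoor ambient spot sounds at reduced volume.
# A quality-level root sends its first two slots to a random pick from fourteen wave variations and its last two to a pick from a four-wave subset.
# Each random branch passes through a modulator that varies pitch and volume slightly, with a narrower volume range on the smaller set.
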SoundCue /AmbientAudioContent/SoundCues/OneShots/Spot_Snow_Wind_Textured_Cue
root=QualityLevel_1
nodes [
 {id=Modulator_3, children [Random_2], properties {pitchMax=1.1, pitchMin=0.9, volumeMax=1.2, volumeMin=0.9}}
 {id=Modulator_5, children [Random_1], properties {pitchMax=1.1, pitchMin=0.9, volumeMax=1.1, volumeMin=0.9}}
 {id=QualityLevel_1, children [Modulator_3,Modulator_3,Modulator_5,Modulator_5]}
 {id=Random_1, children [WavePlayer_18, WavePlayer_19, WavePlayer_20, WavePlayer_21], properties {weights [1,1,1,1]}}
 {
  id=Random_2
  children [
   WavePlayer_5
   WavePlayer_6
   WavePlayer_7
   WavePlayer_8
   WavePlayer_22
   WavePlayer_23
   WavePlayer_24
   WavePlayer_25
   WavePlayer_26
   WavePlayer_27
   WavePlayer_28
   WavePlayer_29
   WavePlayer_30
   WavePlayer_31
  ]
  properties {weights [1,1,1,1,1,1,1,1,1,1,1,1,1,1]}
 }
 {id=WavePlayer_18, properties {soundWaveAssetPtr=/AmbientAudioContent/Waves/OneShots/Spot_Snow_Wind_Textured_03}}
 {id=WavePlayer_19, properties {soundWaveAssetPtr=/AmbientAudioContent/Waves/OneShots/Spot_Snow_Wind_Textured_02}}
 {id=WavePlayer_20, properties {soundWaveAssetPtr=/AmbientAudioContent/Waves/OneShots/Spot_Snow_Wind_Textured_04}}
 {id=WavePlayer_21, properties {soundWaveAssetPtr=/AmbientAudioContent/Waves/OneShots/Spot_Snow_Wind_Textured_06}}
 {id=WavePlayer_22, properties {soundWaveAssetPtr=/AmbientAudioContent/Waves/OneShots/Spot_Snow_Wind_Textured_10}}
 {id=WavePlayer_23, properties {soundWaveAssetPtr=/AmbientAudioContent/Waves/OneShots/Spot_Snow_Wind_Textured_09}}
 {id=WavePlayer_24, properties {soundWaveAssetPtr=/AmbientAudioContent/Waves/OneShots/Spot_Snow_Wind_Textured_08}}
 {id=WavePlayer_25, properties {soundWaveAssetPtr=/AmbientAudioContent/Waves/OneShots/Spot_Snow_Wind_Textured_07}}
 {id=WavePlayer_26, properties {soundWaveAssetPtr=&Spot_Snow_Wind_Textured_06}}
 {id=WavePlayer_27, properties {soundWaveAssetPtr=/AmbientAudioContent/Waves/OneShots/Spot_Snow_Wind_Textured_05}}
 {id=WavePlayer_28, properties {soundWaveAssetPtr=&Spot_Snow_Wind_Textured_04}}
 {id=WavePlayer_29, properties {soundWaveAssetPtr=&Spot_Snow_Wind_Textured_03}}
 {id=WavePlayer_30, properties {soundWaveAssetPtr=&Spot_Snow_Wind_Textured_02}}
 {id=WavePlayer_31, properties {soundWaveAssetPtr=/AmbientAudioContent/Waves/OneShots/Spot_Snow_Wind_Textured_01}}
 {id=WavePlayer_5, properties {soundWaveAssetPtr=/AmbientAudioContent/Waves/OneShots/Spot_Snow_Wind_Textured_14}}
 {id=WavePlayer_6, properties {soundWaveAssetPtr=/AmbientAudioContent/Waves/OneShots/Spot_Snow_Wind_Textured_13}}
 {id=WavePlayer_7, properties {soundWaveAssetPtr=/AmbientAudioContent/Waves/OneShots/Spot_Snow_Wind_Textured_12}}
 {id=WavePlayer_8, properties {soundWaveAssetPtr=/AmbientAudioContent/Waves/OneShots/Spot_Snow_Wind_Textured_11}}
]
properties {
 attenuationSettings=/Game/Sounds/Attenuation/Fort_TileAmbient_OneShot_Wind_Att
 bEnableBusSends=false
 maxDistance=8000
 soundClassObject=&Fort_Ambient_OneShots_Outdoors
 volumeMultiplier=0.301
 preEffectBusSends [{}]
}
