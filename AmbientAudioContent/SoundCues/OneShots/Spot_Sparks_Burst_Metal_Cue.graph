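# Ambient one-shot Sound Cue for a burst of sparks on metal, in the Fort_Ambient_Athena sound class.
# A Random node picks one of twelve equally weighted wave variations, and a Modulator adds a random upward pitch shift on top of the cue's raised base pitch.
# A root Attenuation node applies a custom falloff curve, listener focus and a send to the default reverb submix, while a cue-level concurrency override caps simultaneous plays.
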SoundCue /AmbientAudioContent/SoundCues/OneShots/Spot_Sparks_Burst_Metal_Cue
root=Attenuation_2
nodes [
 {
  id=Attenuation_2
  children [Modulator_0]
  properties {
   bOverrideAttenuation=true
   attenuationOverrides {
    bEnableListenerFocus=true
    bEnableSubmixSends=true
    distanceAlgorithm=Custom
    falloffDistance=2000
    nonFocusVolumeAttenuation=0.1
    stereoSpread=3000
    attenuationShapeExtents {x=0}
    customAttenuationCurve {editorCurveData {keys [
       {arriveTangent=-0.0131, interpMode=RCIM_Cubic, leaveTangent=-0.0131, tangentMode=RCTM_User, time=0.2357, value=0.1409}
       <at 0.3121 0.9761>
       {arriveTangent=-5.1404, interpMode=RCIM_Cubic, leaveTangent=-5.1404, tangentMode=RCTM_User, time=0.6, value=0}
      ]}}
    submixSendSettings [{
      bEnableHPFCutoff=false
      bEnableLPFCutoff=false
      disableManualSendClamp=false
      hPFCutoff=20
      lPFCutoff=20000
      maxSendDistance=6000
      maxSendLevel=1
      minSendDistance=400
      minSendLevel=1
      sendLevel=0.4
      sendLevelControlMethod=Linear
      soundSubmix=&Fortnite_Reverb_Default
     }]
   }
  }
 }
 {id=Modulator_0, children [Random_1], properties {pitchMax=1.3, pitchMin=1}}
 {
  id=Random_1
  children [
   WavePlayer_8
   WavePlayer_9
   WavePlayer_10
   WavePlayer_11
   WavePlayer_12
   WavePlayer_13
   WavePlayer_14
   WavePlayer_15
   WavePlayer_20
   WavePlayer_21
   WavePlayer_22
   WavePlayer_23
  ]
  properties {weights [1,1,1,1,1,1,1,1,1,1,1,1]}
 }
 {id=WavePlayer_10, properties {soundWaveAssetPtr=/AmbientAudioContent/Waves/OneShots/Spot_Spark_Burst_Metal_Int_10}}
 {id=WavePlayer_11, properties {soundWaveAssetPtr=/AmbientAudioContent/Waves/OneShots/Spot_Spark_Burst_Metal_Int_09}}
 {id=WavePlayer_12, properties {soundWaveAssetPtr=/AmbientAudioContent/Waves/OneShots/Spot_Spark_Burst_Metal_Int_08}}
 {id=WavePlayer_13, properties {soundWaveAssetPtr=/AmbientAudioContent/Waves/OneShots/Spot_Spark_Burst_Metal_Int_07}}
 {id=WavePlayer_14, properties {soundWaveAssetPtr=/AmbientAudioContent/Waves/OneShots/Spot_Spark_Burst_Metal_Int_06}}
 {id=WavePlayer_15, properties {soundWaveAssetPtr=/AmbientAudioContent/Waves/OneShots/Spot_Spark_Burst_Metal_Int_05}}
 {id=WavePlayer_20, properties {soundWaveAssetPtr=/AmbientAudioContent/Waves/OneShots/Spot_Spark_Burst_Metal_Int_04}}
 {id=WavePlayer_21, properties {soundWaveAssetPtr=/AmbientAudioContent/Waves/OneShots/Spot_Spark_Burst_Metal_Int_03}}
 {id=WavePlayer_22, properties {soundWaveAssetPtr=/AmbientAudioContent/Waves/OneShots/Spot_Spark_Burst_Metal_Int_02}}
 {id=WavePlayer_23, properties {soundWaveAssetPtr=/AmbientAudioContent/Waves/OneShots/Spot_Spark_Burst_Metal_Int_01}}
 {id=WavePlayer_8, properties {soundWaveAssetPtr=/AmbientAudioContent/Waves/OneShots/Spot_Spark_Burst_Metal_Int_12}}
 {id=WavePlayer_9, properties {soundWaveAssetPtr=/AmbientAudioContent/Waves/OneShots/Spot_Spark_Burst_Metal_Int_11}}
]
properties {
 bEnableBusSends=false
 bOverrideAttenuation=true
 bOverrideConcurrency=true
 maxDistance=8000
 pitchMultiplier=2
 soundClassObject=&Fort_Ambient_Athena
 volumeMultiplier=0.5
 attenuationOverrides {falloffDistance=6000, stereoSpread=10000, attenuationShapeExtents {x=2000}}
 concurrencyOverrides {maxCount=4}
 soundSubmixSends [{minSendLevel=1, soundSubmix=&Fortnite_Reverb_Default}]
}
