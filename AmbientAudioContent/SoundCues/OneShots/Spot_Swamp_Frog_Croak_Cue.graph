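# Sound cue for a one-shot frog croak in swamp ambience, classed as an outdoor ambient one-shot.
# A quality level node routes all tiers to one random node that picks among six equally weighted frog waves.
# It borrows a bird one-shot attenuation, plays at reduced volume, and sends to a distant reverb submix.
# A concurrency override caps simultaneous voices at a low count with a timed release when stealing.
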SoundCue /AmbientAudioContent/SoundCues/OneShots/Spot_Swamp_Frog_Croak_Cue
root=QualityLevel_0
nodes [
 {id=QualityLevel_0, children [Random_0,Random_0,Random_0,Random_0]}
 {
  id=Random_0
  children [WavePlayer_1, WavePlayer_2, WavePlayer_3, WavePlayer_4, WavePlayer_8, WavePlayer_11]
  properties {weights [1,1,1,1,1,1]}
 }
 {id=WavePlayer_1, properties {soundWaveAssetPtr=/AmbientAudioContent/Waves/OneShots/Spot_Biome_Swamp_01_Frog_01}}
 {id=WavePlayer_11, properties {soundWaveAssetPtr=/AmbientAudioContent/Waves/OneShots/Spot_Biome_Swamp_01_Frog_06}}
 {id=WavePlayer_2, properties {soundWaveAssetPtr=/AmbientAudioContent/Waves/OneShots/Spot_Biome_Swamp_01_Frog_02}}
 {id=WavePlayer_3, properties {soundWaveAssetPtr=/AmbientAudioContent/Waves/OneShots/Spot_Biome_Swamp_01_Frog_03}}
 {id=WavePlayer_4, properties {soundWaveAssetPtr=/AmbientAudioContent/Waves/OneShots/Spot_Biome_Swamp_01_Frog_04}}
 {id=WavePlayer_8, properties {soundWaveAssetPtr=/AmbientAudioContent/Waves/OneShots/Spot_Biome_Swamp_01_Frog_05}}
]
properties {
 attenuationSettings=&Fort_TileAmbient_OneShot_Bird_Att
 bEnableBusSends=false
 bOverrideConcurrency=true
 maxDistance=8005
 soundClassObject=&Fort_Ambient_OneShots_Outdoors
 volumeMultiplier=0.501
 concurrencyOverrides {maxCount=2, voiceStealReleaseTime=1}
 concurrencySet [null]
 preEffectBusSends [{}]
 soundSubmixSends [{minSendLevel=0.2, sendLevel=0, soundSubmix=&Fortnite_Reverb_Distant}]
}
