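# Ambient one-shot cue for metal singing resonances at The Raft Floatmax spot, reusing Blimp Crashsite waves.
# A Random node picks among eight equally weighted variations, with the first boosted slightly through a Mixer.
# All quality levels share that Random, and a root Modulator varies pitch.
# The cue lowers overall pitch and volume, using tile-ambient one-shot attenuation and the outdoor ambient sound class.
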SoundCue /AmbientAudioContent/SoundCues/OneShots/Spot_TheRaft_Floatmax_Metal_Singing_Resonances_Cue
root=Modulator_0
nodes [
 {id=Mixer_1, children [WavePlayer_0], properties {inputVolume [1.2]}}
 {id=Modulator_0, children [QualityLevel_0], properties {pitchMax=1.1, pitchMin=0.8}}
 {id=QualityLevel_0, children [Random_0,Random_0,Random_0,Random_0]}
 {
  id=Random_0
  children [Mixer_1,WavePlayer_1,WavePlayer_2,WavePlayer_3,WavePlayer_4,WavePlayer_5,WavePlayer_6,WavePlayer_7]
  properties {weights [1,1,1,1,1,1,1,1]}
 }
 {id=WavePlayer_0, properties {soundWaveAssetPtr=/AmbientAudioContent/Waves/OneShots/Spot_Blimp_Crashsite_Metal_Singing_Resonances_01}}
 {id=WavePlayer_1, properties {soundWaveAssetPtr=/AmbientAudioContent/Waves/OneShots/Spot_Blimp_Crashsite_Metal_Singing_Resonances_02}}
 {id=WavePlayer_2, properties {soundWaveAssetPtr=/AmbientAudioContent/Waves/OneShots/Spot_Blimp_Crashsite_Metal_Singing_Resonances_03}}
 {id=WavePlayer_3, properties {soundWaveAssetPtr=/AmbientAudioContent/Waves/OneShots/Spot_Blimp_Crashsite_Metal_Singing_Resonances_04}}
 {id=WavePlayer_4, properties {soundWaveAssetPtr=/AmbientAudioContent/Waves/OneShots/Spot_Blimp_Crashsite_Metal_Singing_Resonances_05}}
 {id=WavePlayer_5, properties {soundWaveAssetPtr=/AmbientAudioContent/Waves/OneShots/Spot_Blimp_Crashsite_Metal_Singing_Resonances_06}}
 {id=WavePlayer_6, properties {soundWaveAssetPtr=/AmbientAudioContent/Waves/OneShots/Spot_Blimp_Crashsite_Metal_Singing_Resonances_07}}
 {id=WavePlayer_7, properties {soundWaveAssetPtr=/AmbientAudioContent/Waves/OneShots/Spot_Blimp_Crashsite_Metal_Singing_Resonances_08}}
]
properties {
 attenuationSettings=/Game/Sounds/Attenuation/Fort_TileAmbient_OneShot_Att
 maxDistance=8000
 pitchMultiplier=0.75
 soundClassObject=&Fort_Ambient_Outdoors_Athena
 volumeMultiplier=0.251
 preEffectBusSends [{}]
}
