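# Sound cue for distant metal wronk one-shots at The Raft/Floatmax, an outdoor ambient spot sound with wide one-shot attenuation.
# A quality-level node sends its first two levels to a mixer layering two randomized, modulated pools: UFO crashsite wronks and dark, distant reverberant metal-stress wronks.
# Its last two levels use a lighter branch with only three crashsite waves under the same pitch and volume variation as the crashsite layer.
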SoundCue /AmbientAudioContent/SoundCues/OneShots/Spot_TheRaft_Floatmax_Metal_Wronks_Distant_Cue
root=QualityLevel_0
nodes [
 {id=Mixer_11, children [Modulator_1,Modulator_3], properties {inputVolume [0.75,0.75]}}
 {id=Modulator_1, children [Random_2], properties {pitchMax=1.1, pitchMin=0.8, volumeMax=2.5, volumeMin=2}}
 {id=Modulator_2, children [Random_0], properties {pitchMax=1.1, pitchMin=0.8, volumeMax=2.5, volumeMin=2}}
 {id=Modulator_3, children [Random_4]}
 {id=QualityLevel_0, children [Mixer_11,Mixer_11,Modulator_2,Modulator_2]}
 {id=Random_0, children [WavePlayer_27, WavePlayer_28, WavePlayer_29], properties {weights [1,1,1]}}
 {
  id=Random_2
  children [WavePlayer_0, WavePlayer_1, WavePlayer_2, WavePlayer_3, WavePlayer_4, WavePlayer_7, WavePlayer_8, WavePlayer_9, WavePlayer_10, WavePlayer_11]
  properties {weights [1,1,1,1,1,1,1,1,1,1]}
 }
 {
  id=Random_4
  children [WavePlayer_17, WavePlayer_20, WavePlayer_21, WavePlayer_22, WavePlayer_23, WavePlayer_24]
  properties {weights [1,1,1,1,1,1]}
 }
 {id=WavePlayer_0, properties {soundWaveAssetPtr=/AmbientAudioContent/Waves/OneShots/Spot_UFO_Crashsite_Metal_Wronk_Spot_01}}
 {id=WavePlayer_1, properties {soundWaveAssetPtr=/AmbientAudioContent/Waves/OneShots/Spot_UFO_Crashsite_Metal_Wronk_Spot_02}}
 {id=WavePlayer_10, properties {soundWaveAssetPtr=/AmbientAudioContent/Waves/OneShots/Spot_UFO_Crashsite_Metal_Wronk_Spot_09}}
 {id=WavePlayer_11, properties {soundWaveAssetPtr=/AmbientAudioContent/Waves/OneShots/Spot_UFO_Crashsite_Metal_Wronk_Spot_10}}
 {id=WavePlayer_17, properties {soundWaveAssetPtr=/AmbientAudioContent/Waves/OneShots/Spot_Metal_Stress_Wronk_Dark_Dist_Verb_01}}
 {id=WavePlayer_2, properties {soundWaveAssetPtr=/AmbientAudioContent/Waves/OneShots/Spot_UFO_Crashsite_Metal_Wronk_Spot_03}}
 {id=WavePlayer_20, properties {soundWaveAssetPtr=/AmbientAudioContent/Waves/OneShots/Spot_Metal_Stress_Wronk_Dark_Dist_Verb_02}}
 {id=WavePlayer_21, properties {soundWaveAssetPtr=/AmbientAudioContent/Waves/OneShots/Spot_Metal_Stress_Wronk_Dark_Dist_Verb_03}}
 {id=WavePlayer_22, properties {soundWaveAssetPtr=/AmbientAudioContent/Waves/OneShots/Spot_Metal_Stress_Wronk_Dark_Dist_Verb_04}}
 {id=WavePlayer_23, properties {soundWaveAssetPtr=/AmbientAudioContent/Waves/OneShots/Spot_Metal_Stress_Wronk_Dark_Dist_Verb_05}}
 {id=WavePlayer_24, properties {soundWaveAssetPtr=/AmbientAudioContent/Waves/OneShots/Spot_Metal_Stress_Wronk_Dark_Dist_Verb_06}}
 {id=WavePlayer_27, properties {soundWaveAssetPtr=&Spot_UFO_Crashsite_Metal_Wronk_Spot_09}}
 {id=WavePlayer_28, properties {soundWaveAssetPtr=&Spot_UFO_Crashsite_Metal_Wronk_Spot_02}}
 {id=WavePlayer_29, properties {soundWaveAssetPtr=&Spot_UFO_Crashsite_Metal_Wronk_Spot_10}}
 {id=WavePlayer_3, properties {soundWaveAssetPtr=/AmbientAudioContent/Waves/OneShots/Spot_UFO_Crashsite_Metal_Wronk_Spot_04}}
 {id=WavePlayer_4, properties {soundWaveAssetPtr=/AmbientAudioContent/Waves/OneShots/Spot_UFO_Crashsite_Metal_Wronk_Spot_05}}
 {id=WavePlayer_7, properties {soundWaveAssetPtr=/AmbientAudioContent/Waves/OneShots/Spot_UFO_Crashsite_Metal_Wronk_Spot_06}}
 {id=WavePlayer_8, properties {soundWaveAssetPtr=/AmbientAudioContent/Waves/OneShots/Spot_UFO_Crashsite_Metal_Wronk_Spot_07}}
 {id=WavePlayer_9, properties {soundWaveAssetPtr=/AmbientAudioContent/Waves/OneShots/Spot_UFO_Crashsite_Metal_Wronk_Spot_08}}
]
properties {
 attenuationSettings=/AmbientAudioBase/Attenuations/Fort_TileAmbient_OneShot_Wider_Att
 maxDistance=11000
 soundClassObject=&Fort_Ambient_Outdoors_Athena
 volumeMultiplier=0.4
 preEffectBusSends [{}]
}
