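# Ambient one-shot sound cue for stone debris at TheRaft Floatmax, reusing the Fortress Interior large stone debris waves.
# A quality-level node sends its first two slots to a random pick of twelve variations and its last two to a pick of the first four.
# The selection passes through a modulator left at default settings.
# The cue plays at reduced volume in the indoor ambient one-shot class with less-positional log attenuation.
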SoundCue /AmbientAudioContent/SoundCues/OneShots/Spot_TheRaft_Floatmax_Stone_Debris_Cue
root=Modulator_0
nodes [
 {id=Modulator_0, children [QualityLevel_0]}
 {id=QualityLevel_0, children [Random_2,Random_2,Random_0,Random_0]}
 {id=Random_0, children [WavePlayer_0,WavePlayer_1,WavePlayer_2,WavePlayer_3], properties {weights [1,1,1,1]}}
 {
  id=Random_2
  children [
   WavePlayer_4
   WavePlayer_5
   WavePlayer_6
   WavePlayer_7
   WavePlayer_8
   WavePlayer_9
   WavePlayer_10
   WavePlayer_11
   WavePlayer_18
   WavePlayer_19
   WavePlayer_20
   WavePlayer_21
  ]
  properties {weights [1,1,1,1,1,1,1,1,1,1,1,1]}
 }
 {id=WavePlayer_0, properties {soundWaveAssetPtr=/AmbientAudioContent/Waves/OneShots/Spot_Fortress_Interior_Large_Stone_Debris_01}}
 {id=WavePlayer_1, properties {soundWaveAssetPtr=/AmbientAudioContent/Waves/OneShots/Spot_Fortress_Interior_Large_Stone_Debris_02}}
 {id=WavePlayer_10, properties {soundWaveAssetPtr=/AmbientAudioContent/Waves/OneShots/Spot_Fortress_Interior_Large_Stone_Debris_06}}
 {id=WavePlayer_11, properties {soundWaveAssetPtr=/AmbientAudioContent/Waves/OneShots/Spot_Fortress_Interior_Large_Stone_Debris_05}}
 {id=WavePlayer_18, properties {soundWaveAssetPtr=/AmbientAudioContent/Waves/OneShots/Spot_Fortress_Interior_Large_Stone_Debris_04}}
 {id=WavePlayer_19, properties {soundWaveAssetPtr=/AmbientAudioContent/Waves/OneShots/Spot_Fortress_Interior_Large_Stone_Debris_03}}
 {id=WavePlayer_2, properties {soundWaveAssetPtr=&Spot_Fortress_Interior_Large_Stone_Debris_03}}
 {id=WavePlayer_20, properties {soundWaveAssetPtr=&Spot_Fortress_Interior_Large_Stone_Debris_02}}
 {id=WavePlayer_21, properties {soundWaveAssetPtr=&Spot_Fortress_Interior_Large_Stone_Debris_01}}
 {id=WavePlayer_3, properties {soundWaveAssetPtr=&Spot_Fortress_Interior_Large_Stone_Debris_04}}
 {id=WavePlayer_4, properties {soundWaveAssetPtr=/AmbientAudioContent/Waves/OneShots/Spot_Fortress_Interior_Large_Stone_Debris_12}}
 {id=WavePlayer_5, properties {soundWaveAssetPtr=/AmbientAudioContent/Waves/OneShots/Spot_Fortress_Interior_Large_Stone_Debris_11}}
 {id=WavePlayer_6, properties {soundWaveAssetPtr=/AmbientAudioContent/Waves/OneShots/Spot_Fortress_Interior_Large_Stone_Debris_10}}
 {id=WavePlayer_7, properties {soundWaveAssetPtr=/AmbientAudioContent/Waves/OneShots/Spot_Fortress_Interior_Large_Stone_Debris_09}}
 {id=WavePlayer_8, properties {soundWaveAssetPtr=/AmbientAudioContent/Waves/OneShots/Spot_Fortress_Interior_Large_Stone_Debris_08}}
 {id=WavePlayer_9, properties {soundWaveAssetPtr=/AmbientAudioContent/Waves/OneShots/Spot_Fortress_Interior_Large_Stone_Debris_07}}
]
properties {
 attenuationSettings=/Game/Sounds/Attenuation/Fort_TileAmbient_OneShot_LessPositional_Log_Att
 maxDistance=6200
 soundClassObject=/Game/Sounds/FortSoundClassesAndModes/Fort_Ambient_OneShots_Indoors
 volumeMultiplier=0.376
}
