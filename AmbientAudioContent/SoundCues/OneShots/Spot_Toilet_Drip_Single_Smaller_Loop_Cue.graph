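# Sound cue that plays a single small toilet drip as a quiet indoor ambient one-shot.
# A Random node picks one of six equally weighted stereo drip waves, and a Modulator varies it with a slight upward pitch shift.
# Despite "Loop" in the name, the cue has no looping node.
# It uses a shared positional tile-ambient attenuation asset and overrides concurrency to cap simultaneous voices at a low count.
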SoundCue /AmbientAudioContent/SoundCues/OneShots/Spot_Toilet_Drip_Single_Smaller_Loop_Cue
root=Modulator_0
nodes [
 {id=Modulator_0, children [Random_1], properties {pitchMax=1.1, pitchMin=1}}
 {id=Random_1, children [WavePlayer_0,WavePlayer_1,WavePlayer_2,WavePlayer_3,WavePlayer_4,WavePlayer_5], properties {weights [1,1,1,1,1,1]}}
 {id=WavePlayer_0, properties {soundWaveAssetPtr=/AmbientAudioContent/Waves/OneShots/Spot_Toilet_Drips_18_Stereo}}
 {id=WavePlayer_1, properties {soundWaveAssetPtr=/AmbientAudioContent/Waves/OneShots/Spot_Toilet_Drips_17_Stereo}}
 {id=WavePlayer_2, properties {soundWaveAssetPtr=/AmbientAudioContent/Waves/OneShots/Spot_Toilet_Drips_16_Stereo}}
 {id=WavePlayer_3, properties {soundWaveAssetPtr=/AmbientAudioContent/Waves/OneShots/Spot_Toilet_Drips_15_Stereo}}
 {id=WavePlayer_4, properties {soundWaveAssetPtr=/AmbientAudioContent/Waves/OneShots/Spot_Toilet_Drips_14_Stereo}}
 {id=WavePlayer_5, properties {soundWaveAssetPtr=/AmbientAudioContent/Waves/OneShots/Spot_Toilet_Drips_13_Stereo}}
]
properties {
 attenuationSettings=/Game/Sounds/Attenuation/Fort_TileAmbient_OneShot_MorePositional_Att
 bEnableBusSends=false
 bOverrideConcurrency=true
 maxDistance=8000
 soundClassObject=/Game/Sounds/FortSoundClassesAndModes/Fort_Ambient_OneShots_Indoors
 volumeMultiplier=0.251
 attenuationOverrides {falloffDistance=6000, stereoSpread=10000, attenuationShapeExtents {x=2000}}
 concurrencyOverrides {maxCount=2, voiceStealReleaseTime=2}
}
