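# Sound cue for outdoor ambient one-shot spot sounds of a train yard, in the outdoor ambient one-shots sound class.
# An equally weighted Random node picks one of fifteen train station yard waves (variant 14 is absent) and feeds a Modulator at the root.
# Overridden attenuation adds occlusion with low-pass filtering, and a distance-based send feeds the ambient reflection source bus.
# Concurrency override caps it at two simultaneous voices.
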SoundCue /AmbientAudioContent/SoundCues/OneShots/Spot_TrainYard_Various_Cue
root=Modulator_0
nodes [
 {id=Modulator_0, children [Random_2]}
 {
  id=Random_2
  children [
   WavePlayer_25
   WavePlayer_26
   WavePlayer_28
   WavePlayer_29
   WavePlayer_30
   WavePlayer_31
   WavePlayer_32
   WavePlayer_33
   WavePlayer_34
   WavePlayer_35
   WavePlayer_36
   WavePlayer_37
   WavePlayer_38
   WavePlayer_39
   WavePlayer_40
  ]
  properties {weights [1,1,1,1,1,1,1,1,1,1,1,1,1,1,1]}
 }
 {id=WavePlayer_25, properties {soundWaveAssetPtr=/AmbientAudioContent/Waves/OneShots/Spot_TrainStationYard_Various_16}}
 {id=WavePlayer_26, properties {soundWaveAssetPtr=/AmbientAudioContent/Waves/OneShots/Spot_TrainStationYard_Various_15}}
 {id=WavePlayer_28, properties {soundWaveAssetPtr=/AmbientAudioContent/Waves/OneShots/Spot_TrainStationYard_Various_13}}
 {id=WavePlayer_29, properties {soundWaveAssetPtr=/AmbientAudioContent/Waves/OneShots/Spot_TrainStationYard_Various_12}}
 {id=WavePlayer_30, properties {soundWaveAssetPtr=/AmbientAudioContent/Waves/OneShots/Spot_TrainStationYard_Various_11}}
 {id=WavePlayer_31, properties {soundWaveAssetPtr=/AmbientAudioContent/Waves/OneShots/Spot_TrainStationYard_Various_10}}
 {id=WavePlayer_32, properties {soundWaveAssetPtr=/AmbientAudioContent/Waves/OneShots/Spot_TrainStationYard_Various_09}}
 {id=WavePlayer_33, properties {soundWaveAssetPtr=/AmbientAudioContent/Waves/OneShots/Spot_TrainStationYard_Various_08}}
 {id=WavePlayer_34, properties {soundWaveAssetPtr=/AmbientAudioContent/Waves/OneShots/Spot_TrainStationYard_Various_07}}
 {id=WavePlayer_35, properties {soundWaveAssetPtr=/AmbientAudioContent/Waves/OneShots/Spot_TrainStationYard_Various_06}}
 {id=WavePlayer_36, properties {soundWaveAssetPtr=/AmbientAudioContent/Waves/OneShots/Spot_TrainStationYard_Various_05}}
 {id=WavePlayer_37, properties {soundWaveAssetPtr=/AmbientAudioContent/Waves/OneShots/Spot_TrainStationYard_Various_04}}
 {id=WavePlayer_38, properties {soundWaveAssetPtr=/AmbientAudioContent/Waves/OneShots/Spot_TrainStationYard_Various_03}}
 {id=WavePlayer_39, properties {soundWaveAssetPtr=/AmbientAudioContent/Waves/OneShots/Spot_TrainStationYard_Various_02}}
 {id=WavePlayer_40, properties {soundWaveAssetPtr=/AmbientAudioContent/Waves/OneShots/Spot_TrainStationYard_Various_01}}
]
properties {
 attenuationSettings=/Game/Sounds/Attenuation/Fort_TileAmbient_OneShot_Att
 bEnableSubmixSends=false
 bOverrideAttenuation=true
 bOverrideConcurrency=true
 maxDistance=7000
 soundClassObject=&Fort_Ambient_OneShots_Outdoors
 volumeMultiplier=1
 attenuationOverrides {
  bEnableOcclusion=true
  falloffDistance=6000
  nonSpatializedRadiusEnd=2000
  nonSpatializedRadiusStart=3000
  occlusionInterpolationTime=2
  occlusionLowPassFilterFrequency=1000
  occlusionTraceChannel=ECC_GameTraceChannel11
  occlusionVolumeAttenuation=0.5
  stereoSpread=500
  attenuationShapeExtents {x=1000}
 }
 busSends [{
   audioBus=null
   maxSendDistance=5000
   minSendDistance=3000
   minSendLevel=0
   sendLevel=1
   soundSourceBus=/Game/Sounds/Apollo/Biomes/Reflections/AmbientAudio_Reflection_SourceBus
   customSendLevelCurve {editorCurveData {keys [<at 0 0>,<at 1 1>]}}
  }]
 concurrencyOverrides {maxCount=2, voiceStealReleaseTime=0.5}
}
