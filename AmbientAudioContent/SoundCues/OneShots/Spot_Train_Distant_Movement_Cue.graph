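# Spot ambient one-shot Sound Cue for distant train movement, classed as an outdoor ambient one-shot.
# A Random node picks among eight equally weighted train wave variations, and a Modulator at the root processes the result.
# Overridden attenuation sets a long falloff, distance-based low-pass filtering and focus-dependent volume.
# A bus send also feeds the exterior ambient audio source bus.
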SoundCue /AmbientAudioContent/SoundCues/OneShots/Spot_Train_Distant_Movement_Cue
root=Modulator_0
nodes [
 {id=Modulator_0, children [Random_0]}
 {
  id=Random_0
  children [WavePlayer_1,WavePlayer_2,WavePlayer_3,WavePlayer_4,WavePlayer_5,WavePlayer_6,WavePlayer_7,WavePlayer_8]
  properties {weights [1,1,1,1,1,1,1,1]}
 }
 {id=WavePlayer_1, properties {soundWaveAssetPtr=/AmbientAudioContent/Waves/OneShots/Train_Distant_Movement_01}}
 {id=WavePlayer_2, properties {soundWaveAssetPtr=/AmbientAudioContent/Waves/OneShots/Train_Distant_Movement_02}}
 {id=WavePlayer_3, properties {soundWaveAssetPtr=/AmbientAudioContent/Waves/OneShots/Train_Distant_Movement_03}}
 {id=WavePlayer_4, properties {soundWaveAssetPtr=/AmbientAudioContent/Waves/OneShots/Train_Distant_Movement_04}}
 {id=WavePlayer_5, properties {soundWaveAssetPtr=/AmbientAudioContent/Waves/OneShots/Train_Distant_Movement_05}}
 {id=WavePlayer_6, properties {soundWaveAssetPtr=/AmbientAudioContent/Waves/OneShots/Train_Distant_Movement_06}}
 {id=WavePlayer_7, properties {soundWaveAssetPtr=/AmbientAudioContent/Waves/OneShots/Train_Distant_Movement_07}}
 {id=WavePlayer_8, properties {soundWaveAssetPtr=/AmbientAudioContent/Waves/OneShots/Train_Distant_Movement_08}}
]
properties {
 bOverrideAttenuation=true
 maxDistance=7000
 soundClassObject=&Fort_Ambient_OneShots_Outdoors
 attenuationOverrides {
  falloffDistance=6000
  focusVolumeAttenuation=1.2
  lPFFrequencyAtMax=2000
  lPFRadiusMin=2000
  nonFocusVolumeAttenuation=0.4
  attenuationShapeExtents {x=1000}
 }
 busSends [{
   audioBus=null
   maxSendDistance=4000
   minSendLevel=0.6
   sendLevel=1
   soundSourceBus=&AmbientAudio_Exterior_SourceBus
  }]
}
