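# Sound cue for a spot train horn ambient one-shot, in the Fort_Ambient_OneShots class.
# Two random nodes independently pick a horn or whistle, one from close recordings and one from distant ones.
# Custom attenuation curves crossfade them by distance: close fades out by mid-range as distant swells, then fades.
# The mix passes through a modulator, with cue-level occlusion, low-pass filtering and an exterior ambience bus send.
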SoundCue /AmbientAudioContent/SoundCues/OneShots/Spot_Train_Horn_Cue
root=Modulator_0
nodes [
 {
  id=Attenuation_1
  children [Random_2]
  properties {
   bOverrideAttenuation=true
   attenuationOverrides {
    bEnableListenerFocus=true
    distanceAlgorithm=Custom
    falloffDistance=10000
    nonFocusVolumeAttenuation=0.75
    attenuationShapeExtents {x=0}
    customAttenuationCurve {editorCurveData {keys [<at 0.2 0 RCIM_Cubic>,<at 0.593 0.649 RCIM_Cubic>,<at 1 0>]}}
   }
  }
 }
 {
  id=Attenuation_2
  children [Random_1]
  properties {
   bOverrideAttenuation=true
   attenuationOverrides {
    bEnableListenerFocus=true
    distanceAlgorithm=Custom
    falloffDistance=10000
    nonFocusVolumeAttenuation=0.75
    attenuationShapeExtents {x=0}
    customAttenuationCurve {editorCurveData {keys [
       <at 0 1>
       {arriveTangent=-1.6863, interpMode=RCIM_Cubic, leaveTangent=-1.6863, time=0.2, value=1}
       {arriveTangent=-3.4272, interpMode=RCIM_Cubic, leaveTangent=-3.4272, tangentMode=RCTM_User, time=0.593, value=0}
      ]}}
   }
  }
 }
 {id=Mixer_0, children [Attenuation_1, Attenuation_2], properties {inputVolume [1,1]}}
 {id=Modulator_0, children [Mixer_0]}
 {id=Random_1, children [WavePlayer_9, WavePlayer_10], properties {weights [1,1]}}
 {id=Random_2, children [WavePlayer_0, WavePlayer_11], properties {weights [1,1]}}
 {id=WavePlayer_0, properties {soundWaveAssetPtr=/AmbientAudioContent/Waves/OneShots/Train_Horn_Distant_Japan}}
 {id=WavePlayer_10, properties {soundWaveAssetPtr=/AmbientAudioContent/Waves/OneShots/Train_Proto_Whistle}}
 {id=WavePlayer_11, properties {soundWaveAssetPtr=/AmbientAudioContent/Waves/OneShots/Train_Proto_Whistle_Distant}}
 {id=WavePlayer_9, properties {soundWaveAssetPtr=/AmbientAudioContent/Waves/OneShots/Train_Horn_Japan}}
]
properties {
 bOverrideAttenuation=true
 maxDistance=10000
 soundClassObject=&Fort_Ambient_OneShots
 attenuationOverrides {
  bEnableOcclusion=true
  falloffDistance=6000
  focusVolumeAttenuation=1.2
  lPFFrequencyAtMax=2000
  lPFRadiusMin=2000
  nonFocusVolumeAttenuation=0.4
  occlusionLowPassFilterFrequency=1000
  occlusionTraceChannel=ECC_GameTraceChannel11
  occlusionVolumeAttenuation=0.5
  stereoSpread=2000
  attenuationShapeExtents {x=1000}
 }
 busSends [{
   audioBus=null
   maxSendDistance=4000
   minSendLevel=0.6
   sendLevel=1
   soundSourceBus=&AmbientAudio_Exterior_SourceBus
  }]
}
