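# Looping ambient SoundCue for train station PA announcements in Kurai City (S33).
# After each randomized delay, a modulated random pick plays either a female voice line alone or a quieter jingle concatenated before one.
# Jingles and general or city-greeting lines come from random pools, with the voice branch shared by both paths.
# Overrides add occluded, focus-aware, low-pass attenuation and limited concurrency, with sends to exterior and reflection buses.
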SoundCue /AmbientAudioContent/SoundCues/OneShots/Spot_Train_Station_PA_S33_Kurai_City_Announce_Cue
root=Looping_0
nodes [
 {id=Concatenator_2, children [Mixer_4,Mixer_0], properties {inputVolume [0.9,1]}}
 {id=Delay_0, children [Modulator_0], properties {delayMax=35, delayMin=12}}
 {id=Looping_0, children [Delay_0]}
 {id=Mixer_0, children [Random_3], properties {inputVolume [0.75]}}
 {id=Mixer_4, children [Random_2], properties {inputVolume [0.2]}}
 {id=Modulator_0, children [Random_5]}
 {id=Random_2, children [WavePlayer_17, WavePlayer_18, WavePlayer_19, WavePlayer_20], properties {weights [1,1,1,1]}}
 {
  id=Random_3
  children [
   WavePlayer_21
   WavePlayer_22
   WavePlayer_23
   WavePlayer_24
   WavePlayer_25
   WavePlayer_26
   WavePlayer_27
   WavePlayer_28
   WavePlayer_30
   WavePlayer_32
   WavePlayer_33
   WavePlayer_34
   WavePlayer_35
   WavePlayer_36
   WavePlayer_37
  ]
  properties {weights [1,1,1,1,1,1,1,1,1,1,1,1,1,1,1]}
 }
 {id=Random_5, children [Mixer_0, Concatenator_2], properties {weights [1,1]}}
 {id=WavePlayer_17, properties {soundWaveAssetPtr=/AmbientAudioContent/Waves/OneShots/S33_Train_Station_Jingle_Small_Misc_B_01}}
 {id=WavePlayer_18, properties {soundWaveAssetPtr=/AmbientAudioContent/Waves/OneShots/S33_Train_Station_Jingle_Small_Shared_2sec_A}}
 {id=WavePlayer_19, properties {soundWaveAssetPtr=/AmbientAudioContent/Waves/OneShots/S33_Train_Station_Jingle_Small_Shared_2sec_B}}
 {id=WavePlayer_20, properties {soundWaveAssetPtr=/AmbientAudioContent/Waves/OneShots/S33_Train_Station_Jingle_Small_Shared_5sec_01}}
 {id=WavePlayer_21, properties {soundWaveAssetPtr=/AmbientAudioContent/Waves/OneShots/VO_Train_Station_PA_S33_Female_Small_General_01}}
 {id=WavePlayer_22, properties {soundWaveAssetPtr=/AmbientAudioContent/Waves/OneShots/VO_Train_Station_PA_S33_Female_Small_General_02}}
 {id=WavePlayer_23, properties {soundWaveAssetPtr=/AmbientAudioContent/Waves/OneShots/VO_Train_Station_PA_S33_Female_Small_General_03}}
 {id=WavePlayer_24, properties {soundWaveAssetPtr=/AmbientAudioContent/Waves/OneShots/VO_Train_Station_PA_S33_Female_Small_General_04}}
 {id=WavePlayer_25, properties {soundWaveAssetPtr=/AmbientAudioContent/Waves/OneShots/VO_Train_Station_PA_S33_Female_Small_General_05}}
 {id=WavePlayer_26, properties {soundWaveAssetPtr=/AmbientAudioContent/Waves/OneShots/VO_Train_Station_PA_S33_Female_Small_General_06}}
 {id=WavePlayer_27, properties {soundWaveAssetPtr=/AmbientAudioContent/Waves/OneShots/VO_Train_Station_PA_S33_Female_Small_General_07}}
 {id=WavePlayer_28, properties {soundWaveAssetPtr=/AmbientAudioContent/Waves/OneShots/VO_Train_Station_PA_S33_Female_Small_General_08}}
 {id=WavePlayer_30, properties {soundWaveAssetPtr=/AmbientAudioContent/Waves/OneShots/VO_Train_Station_PA_S33_Female_Small_General_09}}
 {id=WavePlayer_32, properties {soundWaveAssetPtr=/AmbientAudioContent/Waves/OneShots/VO_Train_Station_PA_S33_Female_Small_General_10}}
 {id=WavePlayer_33, properties {soundWaveAssetPtr=/AmbientAudioContent/Waves/OneShots/VO_Train_Station_PA_S33_Female_Small_General_11}}
 {id=WavePlayer_34, properties {soundWaveAssetPtr=/AmbientAudioContent/Waves/OneShots/VO_Train_Station_PA_S33_Female_Small_General_12}}
 {id=WavePlayer_35, properties {soundWaveAssetPtr=/AmbientAudioContent/Waves/OneShots/VO_Train_Station_PA_S33_Female_Small_General_13}}
 {id=WavePlayer_36, properties {soundWaveAssetPtr=/AmbientAudioContent/Waves/OneShots/VO_Train_Station_PA_S33_Female_Small_Greeting_City}}
 {id=WavePlayer_37, properties {soundWaveAssetPtr=/AmbientAudioContent/Waves/OneShots/VO_Train_Station_PA_S33_Female_Small_Greeting_City_Alt}}
]
properties {
 bEnableBaseSubmix=false
 bEnableSubmixSends=false
 bOverrideAttenuation=true
 bOverrideConcurrency=true
 maxDistance=8000
 soundClassObject=&Fort_Ambient_Athena
 volumeMultiplier=1
 attenuationOverrides {
  bAttenuateWithLPF=true
  bEnableListenerFocus=true
  bEnableOcclusion=true
  falloffDistance=6000
  hPFFrequencyAtMax=100
  lPFFrequencyAtMax=2000
  lPFRadiusMax=7000
  nonFocusVolumeAttenuation=0.75
  occlusionLowPassFilterFrequency=1000
  occlusionTraceChannel=ECC_GameTraceChannel11
  occlusionVolumeAttenuation=0.5
  stereoSpread=1000
  attenuationShapeExtents {x=2000}
 }
 busSends [{
   audioBus=null
   maxSendDistance=6000
   minSendDistance=1000
   minSendLevel=0.4
   sendLevel=1
   soundSourceBus=&AmbientAudio_Exterior_SourceBus
   customSendLevelCurve {editorCurveData {keys [<at 0 0.4>,<at 1 1>]}}
  }]
 concurrencyOverrides {maxCount=2}
 preEffectBusSends [{
   maxSendDistance=6000
   minSendDistance=2000
   sendLevel=0.4
   soundSourceBus=/Game/Sounds/Apollo/Biomes/Reflections/AmbientAudio_Reflection_SourceBus
   customSendLevelCurve {editorCurveData {keys [<at 0 0.4>,<at 1 1>]}}
  }]
}
