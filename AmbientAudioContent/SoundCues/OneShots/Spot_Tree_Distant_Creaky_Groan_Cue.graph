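# Sound cue for one-shot ambient distant creaky tree groans, classed as an outdoor ambient one-shot.
# A Random node picks among nine equally weighted groan wave variations and feeds a Modulator at the root, which is left at its default settings.
# The cue references a shared distant-creature attenuation asset, runs at reduced volume, and overrides concurrency to cap simultaneous voices at a low count with a voice-steal release.
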SoundCue /AmbientAudioContent/SoundCues/OneShots/Spot_Tree_Distant_Creaky_Groan_Cue
root=Modulator_0
nodes [
 {id=Modulator_0, children [Random_0]}
 {
  id=Random_0
  children [WavePlayer_0, WavePlayer_1, WavePlayer_2, WavePlayer_3, WavePlayer_4, WavePlayer_5, WavePlayer_6, WavePlayer_14, WavePlayer_15]
  properties {weights [1,1,1,1,1,1,1,1,1]}
 }
 {id=WavePlayer_0, properties {soundWaveAssetPtr=/AmbientAudioContent/Waves/OneShots/Spot_Tree_Distant_Creaky_Groan_01}}
 {id=WavePlayer_1, properties {soundWaveAssetPtr=/AmbientAudioContent/Waves/OneShots/Spot_Tree_Distant_Creaky_Groan_02}}
 {id=WavePlayer_14, properties {soundWaveAssetPtr=/AmbientAudioContent/Waves/OneShots/Spot_Tree_Distant_Creaky_Groan_08}}
 {id=WavePlayer_15, properties {soundWaveAssetPtr=/AmbientAudioContent/Waves/OneShots/Spot_Tree_Distant_Creaky_Groan_09}}
 {id=WavePlayer_2, properties {soundWaveAssetPtr=/AmbientAudioContent/Waves/OneShots/Spot_Tree_Distant_Creaky_Groan_03}}
 {id=WavePlayer_3, properties {soundWaveAssetPtr=/AmbientAudioContent/Waves/OneShots/Spot_Tree_Distant_Creaky_Groan_04}}
 {id=WavePlayer_4, properties {soundWaveAssetPtr=/AmbientAudioContent/Waves/OneShots/Spot_Tree_Distant_Creaky_Groan_05}}
 {id=WavePlayer_5, properties {soundWaveAssetPtr=/AmbientAudioContent/Waves/OneShots/Spot_Tree_Distant_Creaky_Groan_06}}
 {id=WavePlayer_6, properties {soundWaveAssetPtr=/AmbientAudioContent/Waves/OneShots/Spot_Tree_Distant_Creaky_Groan_07}}
]
properties {
 attenuationSettings=/AmbientAudioBase/Attenuations/Fort_OneShot_Distant_Creature_Att
 bOverrideConcurrency=true
 maxDistance=8500
 soundClassObject=&Fort_Ambient_OneShots_Outdoors
 volumeMultiplier=0.66
 attenuationOverrides {bAttenuate=false}
 concurrencyOverrides {maxCount=2, voiceStealReleaseTime=2}
}
