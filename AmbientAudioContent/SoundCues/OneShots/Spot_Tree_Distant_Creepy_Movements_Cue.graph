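# Sound cue for a distant, creepy tree-movement one-shot in outdoor ambience.
# A Random node with equal weights picks one of ten wave variations and feeds a Modulator at the root for pitch and volume variation.
# It uses a distant-creature attenuation asset and the outdoor ambient one-shots sound class.
# A concurrency override limits it to two simultaneous instances, with a release time for stolen voices.
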SoundCue /AmbientAudioContent/SoundCues/OneShots/Spot_Tree_Distant_Creepy_Movements_Cue
root=Modulator_0
nodes [
 {id=Modulator_0, children [Random_1]}
 {
  id=Random_1
  children [
   WavePlayer_7
   WavePlayer_8
   WavePlayer_9
   WavePlayer_10
   WavePlayer_11
   WavePlayer_12
   WavePlayer_13
   WavePlayer_16
   WavePlayer_17
   WavePlayer_18
  ]
  properties {weights [1,1,1,1,1,1,1,1,1,1]}
 }
 {id=WavePlayer_10, properties {soundWaveAssetPtr=/AmbientAudioContent/Waves/OneShots/Spot_Tree_Distant_Creepy_Movement_04}}
 {id=WavePlayer_11, properties {soundWaveAssetPtr=/AmbientAudioContent/Waves/OneShots/Spot_Tree_Distant_Creepy_Movement_05}}
 {id=WavePlayer_12, properties {soundWaveAssetPtr=/AmbientAudioContent/Waves/OneShots/Spot_Tree_Distant_Creepy_Movement_06}}
 {id=WavePlayer_13, properties {soundWaveAssetPtr=/AmbientAudioContent/Waves/OneShots/Spot_Tree_Distant_Creepy_Movement_07}}
 {id=WavePlayer_16, properties {soundWaveAssetPtr=/AmbientAudioContent/Waves/OneShots/Spot_Tree_Distant_Creepy_Movement_08}}
 {id=WavePlayer_17, properties {soundWaveAssetPtr=/AmbientAudioContent/Waves/OneShots/Spot_Tree_Distant_Creepy_Movement_09}}
 {id=WavePlayer_18, properties {soundWaveAssetPtr=/AmbientAudioContent/Waves/OneShots/Spot_Tree_Distant_Creepy_Movement_10}}
 {id=WavePlayer_7, properties {soundWaveAssetPtr=/AmbientAudioContent/Waves/OneShots/Spot_Tree_Distant_Creepy_Movement_01}}
 {id=WavePlayer_8, properties {soundWaveAssetPtr=/AmbientAudioContent/Waves/OneShots/Spot_Tree_Distant_Creepy_Movement_02}}
 {id=WavePlayer_9, properties {soundWaveAssetPtr=/AmbientAudioContent/Waves/OneShots/Spot_Tree_Distant_Creepy_Movement_03}}
]
properties {
 attenuationSettings=/AmbientAudioBase/Attenuations/Fort_OneShot_Distant_Creature_Att
 bOverrideConcurrency=true
 maxDistance=8500
 soundClassObject=&Fort_Ambient_OneShots_Outdoors
 volumeMultiplier=1
 attenuationOverrides {bAttenuate=false}
 concurrencyOverrides {maxCount=2, voiceStealReleaseTime=2}
}
