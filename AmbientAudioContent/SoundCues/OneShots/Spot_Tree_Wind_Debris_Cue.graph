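# Sound cue for a one-shot ambient spot sound of wind gusting through tree leaves.
# A Random node picks one of four leaf gust waves with equal weight, and a Modulator varies its pitch and volume.
# All four quality levels route to that same branch.
# The cue uses a wide one-shot attenuation and the ambient one-shots sound class, with a concurrency override capping simultaneous instances at two.
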SoundCue /AmbientAudioContent/SoundCues/OneShots/Spot_Tree_Wind_Debris_Cue
root=QualityLevel_0
nodes [
 {id=Modulator_1, children [Random_1], properties {pitchMax=1.1, volumeMax=1.1, volumeMin=0.9}}
 {id=QualityLevel_0, children [Modulator_1,Modulator_1,Modulator_1,Modulator_1]}
 {id=Random_1, children [WavePlayer_8, WavePlayer_9, WavePlayer_10, WavePlayer_11], properties {weights [1,1,1,1]}}
 {id=WavePlayer_10, properties {soundWaveAssetPtr=/AmbientAudioContent/Waves/OneShots/Spot_Tree_Wind_LeafGust_ST_03}}
 {id=WavePlayer_11, properties {soundWaveAssetPtr=/AmbientAudioContent/Waves/OneShots/Spot_Tree_Wind_LeafGust_ST_04}}
 {id=WavePlayer_8, properties {soundWaveAssetPtr=/AmbientAudioContent/Waves/OneShots/Spot_Tree_Wind_LeafGust_ST_01}}
 {id=WavePlayer_9, properties {soundWaveAssetPtr=/AmbientAudioContent/Waves/OneShots/Spot_Tree_Wind_LeafGust_ST_02}}
]
properties {
 attenuationSettings=/AmbientAudioBase/Attenuations/Fort_TileAmbient_OneShot_Wider_Att
 bEnableBusSends=false
 bOverrideConcurrency=true
 maxDistance=11000
 soundClassObject=&Fort_Ambient_OneShots
 volumeMultiplier=0.401
 concurrencyOverrides {maxCount=2}
 preEffectBusSends [{}]
}
