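# Sound cue for one-shot mosquito fly-by spot sounds in tropical daytime ambience, assigned to the outdoor ambient one-shots sound class.
# A Random node picks among eleven equally weighted fly-by waves and feeds a Modulator that varies pitch and volume.
# The root Quality Level node routes all four inputs to that same Modulator, so every quality level plays identically.
# An insect fly-by attenuation asset handles spatial falloff, and overall volume is reduced.
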SoundCue /AmbientAudioContent/SoundCues/OneShots/Spot_Tropical_Day_Mosquito_FlyBy_Cue
root=QualityLevel_0
nodes [
 {id=Modulator_1, children [Random_0], properties {pitchMax=1.1, volumeMax=1.1, volumeMin=0.9}}
 {id=QualityLevel_0, children [Modulator_1,Modulator_1,Modulator_1,Modulator_1]}
 {
  id=Random_0
  children [
   WavePlayer_0
   WavePlayer_1
   WavePlayer_2
   WavePlayer_3
   WavePlayer_4
   WavePlayer_5
   WavePlayer_6
   WavePlayer_7
   WavePlayer_16
   WavePlayer_17
   WavePlayer_18
  ]
  properties {weights [1,1,1,1,1,1,1,1,1,1,1]}
 }
 {id=WavePlayer_0, properties {soundWaveAssetPtr=/AmbientAudioContent/Waves/OneShots/Spot_Tropical_Day_Mosquito_FlyBy_01}}
 {id=WavePlayer_1, properties {soundWaveAssetPtr=/AmbientAudioContent/Waves/OneShots/Spot_Tropical_Day_Mosquito_FlyBy_02}}
 {id=WavePlayer_16, properties {soundWaveAssetPtr=/AmbientAudioContent/Waves/OneShots/Spot_Tropical_Day_Mosquito_FlyBy_11}}
 {id=WavePlayer_17, properties {soundWaveAssetPtr=/AmbientAudioContent/Waves/OneShots/Spot_Tropical_Day_Mosquito_FlyBy_12}}
 {id=WavePlayer_18, properties {soundWaveAssetPtr=/AmbientAudioContent/Waves/OneShots/Spot_Tropical_Day_Mosquito_FlyBy_13}}
 {id=WavePlayer_2, properties {soundWaveAssetPtr=/AmbientAudioContent/Waves/OneShots/Spot_Tropical_Day_Mosquito_FlyBy_03}}
 {id=WavePlayer_3, properties {soundWaveAssetPtr=/AmbientAudioContent/Waves/OneShots/Spot_Tropical_Day_Mosquito_FlyBy_04}}
 {id=WavePlayer_4, properties {soundWaveAssetPtr=/AmbientAudioContent/Waves/OneShots/Spot_Tropical_Day_Mosquito_FlyBy_05}}
 {id=WavePlayer_5, properties {soundWaveAssetPtr=/AmbientAudioContent/Waves/OneShots/Spot_Tropical_Day_Mosquito_FlyBy_06}}
 {id=WavePlayer_6, properties {soundWaveAssetPtr=/AmbientAudioContent/Waves/OneShots/Spot_Tropical_Day_Mosquito_FlyBy_07}}
 {id=WavePlayer_7, properties {soundWaveAssetPtr=/AmbientAudioContent/Waves/OneShots/Spot_Tropical_Day_Mosquito_FlyBy_08}}
]
properties {
 attenuationSettings=/Game/Sounds/Attenuation/Fort_TileAmbient_OneShot_Insect_By_Att
 bEnableBusSends=false
 maxDistance=6000
 soundClassObject=&Fort_Ambient_OneShots_Outdoors
 volumeMultiplier=0.351
 preEffectBusSends [{}]
}
