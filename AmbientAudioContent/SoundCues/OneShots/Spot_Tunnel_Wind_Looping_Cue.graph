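# Looping ambient one-shot cue that plays randomized wind gusts for tunnels, reusing chaparral gust waves at lowered pitch and volume in an indoor ambient sound class.
# A Looping node wraps a quality-level switch: the first two levels randomize across eighteen gusts, the last two across only four.
# Each pool has its own pitch and volume modulator for variation.
# Wind attenuation and an owner-limited concurrency override constrain playback.
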SoundCue /AmbientAudioContent/SoundCues/OneShots/Spot_Tunnel_Wind_Looping_Cue
root=Looping_0
nodes [
 {id=Looping_0, children [QualityLevel_1]}
 {id=Modulator_1, children [Random_0], properties {pitchMax=1.1, pitchMin=0.9, volumeMax=1, volumeMin=0.6}}
 {id=Modulator_2, children [Random_1], properties {pitchMax=1.1, volumeMax=1, volumeMin=0.7}}
 {id=QualityLevel_1, children [Modulator_1,Modulator_1,Modulator_2,Modulator_2]}
 {
  id=Random_0
  children [
   WavePlayer_0
   WavePlayer_1
   WavePlayer_2
   WavePlayer_3
   WavePlayer_4
   WavePlayer_5
   WavePlayer_6
   WavePlayer_7
   WavePlayer_8
   WavePlayer_9
   WavePlayer_10
   WavePlayer_11
   WavePlayer_12
   WavePlayer_13
   WavePlayer_14
   WavePlayer_15
   WavePlayer_16
   WavePlayer_17
  ]
  properties {weights [1,1,1,1,1,1,1,1,1,1,1,1,1,1,1,1,1,1]}
 }
 {id=Random_1, children [WavePlayer_44, WavePlayer_45, WavePlayer_46, WavePlayer_47], properties {weights [1,1,1,1]}}
 {id=WavePlayer_0, properties {soundWaveAssetPtr=/AmbientAudioContent/Waves/OneShots/Spot_Chaparral_Wind_Gust_18}}
 {id=WavePlayer_1, properties {soundWaveAssetPtr=/AmbientAudioContent/Waves/OneShots/Spot_Chaparral_Wind_Gust_17}}
 {id=WavePlayer_10, properties {soundWaveAssetPtr=/AmbientAudioContent/Waves/OneShots/Spot_Chaparral_Wind_Gust_08}}
 {id=WavePlayer_11, properties {soundWaveAssetPtr=/AmbientAudioContent/Waves/OneShots/Spot_Chaparral_Wind_Gust_07}}
 {id=WavePlayer_12, properties {soundWaveAssetPtr=/AmbientAudioContent/Waves/OneShots/Spot_Chaparral_Wind_Gust_06}}
 {id=WavePlayer_13, properties {soundWaveAssetPtr=/AmbientAudioContent/Waves/OneShots/Spot_Chaparral_Wind_Gust_05}}
 {id=WavePlayer_14, properties {soundWaveAssetPtr=/AmbientAudioContent/Waves/OneShots/Spot_Chaparral_Wind_Gust_04}}
 {id=WavePlayer_15, properties {soundWaveAssetPtr=/AmbientAudioContent/Waves/OneShots/Spot_Chaparral_Wind_Gust_03}}
 {id=WavePlayer_16, properties {soundWaveAssetPtr=/AmbientAudioContent/Waves/OneShots/Spot_Chaparral_Wind_Gust_02}}
 {id=WavePlayer_17, properties {soundWaveAssetPtr=/AmbientAudioContent/Waves/OneShots/Spot_Chaparral_Wind_Gust_01}}
 {id=WavePlayer_2, properties {soundWaveAssetPtr=/AmbientAudioContent/Waves/OneShots/Spot_Chaparral_Wind_Gust_16}}
 {id=WavePlayer_3, properties {soundWaveAssetPtr=/AmbientAudioContent/Waves/OneShots/Spot_Chaparral_Wind_Gust_15}}
 {id=WavePlayer_4, properties {soundWaveAssetPtr=/AmbientAudioContent/Waves/OneShots/Spot_Chaparral_Wind_Gust_14}}
 {id=WavePlayer_44, properties {soundWaveAssetPtr=&Spot_Chaparral_Wind_Gust_04}}
 {id=WavePlayer_45, properties {soundWaveAssetPtr=&Spot_Chaparral_Wind_Gust_03}}
 {id=WavePlayer_46, properties {soundWaveAssetPtr=&Spot_Chaparral_Wind_Gust_02}}
 {id=WavePlayer_47, properties {soundWaveAssetPtr=&Spot_Chaparral_Wind_Gust_01}}
 {id=WavePlayer_5, properties {soundWaveAssetPtr=/AmbientAudioContent/Waves/OneShots/Spot_Chaparral_Wind_Gust_13}}
 {id=WavePlayer_6, properties {soundWaveAssetPtr=/AmbientAudioContent/Waves/OneShots/Spot_Chaparral_Wind_Gust_12}}
 {id=WavePlayer_7, properties {soundWaveAssetPtr=/AmbientAudioContent/Waves/OneShots/Spot_Chaparral_Wind_Gust_11}}
 {id=WavePlayer_8, properties {soundWaveAssetPtr=/AmbientAudioContent/Waves/OneShots/Spot_Chaparral_Wind_Gust_10}}
 {id=WavePlayer_9, properties {soundWaveAssetPtr=/AmbientAudioContent/Waves/OneShots/Spot_Chaparral_Wind_Gust_09}}
]
properties {
 attenuationSettings=/Game/Sounds/Attenuation/Fort_TileAmbient_OneShot_Wind_Att
 bEnableBusSends=false
 bOverrideConcurrency=true
 maxDistance=8000
 pitchMultiplier=0.8
 soundClassObject=/Game/Sounds/FortSoundClassesAndModes/Fort_Ambient_OneShots_Indoors
 volumeMultiplier=0.251
 concurrencyOverrides {bLimitToOwner=true, maxCount=2}
 preEffectBusSends [{}]
}
