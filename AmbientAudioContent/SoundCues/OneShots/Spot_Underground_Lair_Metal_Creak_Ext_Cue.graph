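# Sound cue for a one-shot outdoor ambient metal creak, named for an underground lair exterior.
# A random node picks one of four equally weighted train-station creak waves.
# A modulator at the root randomly varies pitch, at or below normal, and volume on each play.
# It uses a distant-creak attenuation asset, the outdoor ambient one-shot sound class and a concurrency override that limits simultaneous plays.
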SoundCue /AmbientAudioContent/SoundCues/OneShots/Spot_Underground_Lair_Metal_Creak_Ext_Cue
root=Modulator_0
nodes [
 {id=Modulator_0, children [Random_1], properties {pitchMax=1, pitchMin=0.8, volumeMax=1, volumeMin=0.7}}
 {id=Random_1, children [WavePlayer_0,WavePlayer_1,WavePlayer_2,WavePlayer_3], properties {weights [1,1,1,1]}}
 {id=WavePlayer_0, properties {soundWaveAssetPtr=/AmbientAudioContent/Waves/OneShots/Spot_UndergroundTrainStation_Creak_04}}
 {id=WavePlayer_1, properties {soundWaveAssetPtr=/AmbientAudioContent/Waves/OneShots/Spot_UndergroundTrainStation_Creak_03}}
 {id=WavePlayer_2, properties {soundWaveAssetPtr=/AmbientAudioContent/Waves/OneShots/Spot_UndergroundTrainStation_Creak_02}}
 {id=WavePlayer_3, properties {soundWaveAssetPtr=/AmbientAudioContent/Waves/OneShots/Spot_UndergroundTrainStation_Creak_01}}
]
properties {
 attenuationSettings=/Game/Sounds/Attenuation/Fort_TileAmbient_OneShot_Distant_Creak_Att
 bOverrideConcurrency=true
 maxDistance=8200
 soundClassObject=&Fort_Ambient_OneShots_Outdoors
 volumeMultiplier=0.501
 concurrencyOverrides {maxCount=2}
}
