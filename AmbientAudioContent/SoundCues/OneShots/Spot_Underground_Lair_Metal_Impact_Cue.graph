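# Sound cue for a distant indoor ambient metal impact one-shot in an underground lair, built from underground train station impact waves.
# A random node picks among three equally weighted wave players, two sharing the second wave, so that impact plays twice as often as the first.
# A modulator varies pitch downward and volume slightly on each play.
# Distant-creak attenuation shapes falloff, and a concurrency override limits simultaneous instances.
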SoundCue /AmbientAudioContent/SoundCues/OneShots/Spot_Underground_Lair_Metal_Impact_Cue
root=Modulator_0
nodes [
 {id=Modulator_0, children [Random_0], properties {pitchMax=1, pitchMin=0.8, volumeMax=1, volumeMin=0.7}}
 {id=Random_0, children [WavePlayer_4,WavePlayer_5,WavePlayer_6], properties {weights [1,1,1]}}
 {id=WavePlayer_4, properties {soundWaveAssetPtr=/AmbientAudioContent/Waves/OneShots/Spot_UndergroundTrainStation_Impact_01}}
 {id=WavePlayer_5, properties {soundWaveAssetPtr=/AmbientAudioContent/Waves/OneShots/Spot_UndergroundTrainStation_Impact_02}}
 {id=WavePlayer_6, properties {soundWaveAssetPtr=&Spot_UndergroundTrainStation_Impact_02}}
]
properties {
 attenuationSettings=/Game/Sounds/Attenuation/Fort_TileAmbient_OneShot_Distant_Creak_Att
 bEnableSubmixSends=false
 bOverrideConcurrency=true
 maxDistance=8200
 soundClassObject=/Game/Sounds/FortSoundClassesAndModes/Fort_Ambient_OneShots_Indoors
 volumeMultiplier=0.501
 concurrencyOverrides {maxCount=2}
}
