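# Sound cue for quiet, distant one-shot metal knocks, an ambient spot sound for an underground lair.
# A random node picks among five equally weighted knock waves, feeding an enveloper with time-based pitch and volume curves.
# A root modulator adds random downward pitch and volume variation.
# The cue uses distant-creak attenuation and the indoor ambient one-shot sound class, with a send to the distant reverb submix.
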SoundCue /AmbientAudioContent/SoundCues/OneShots/Spot_Underground_Lair_Metal_Knocks_Cue
root=Modulator_0
nodes [
 {
  id=Enveloper_0
  children [Random_1]
  properties {pitchCurve {editorCurveData {keys [{time=0}, <at 10 0.8>]}}, volumeCurve {editorCurveData {keys [<at 0 1>,<at 2 1>,<at 10 0.5>]}}}
 }
 {id=Modulator_0, children [Enveloper_0], properties {pitchMax=1, pitchMin=0.8, volumeMax=1, volumeMin=0.7}}
 {id=Random_1, children [WavePlayer_0,WavePlayer_1,WavePlayer_2,WavePlayer_3,WavePlayer_8], properties {weights [1,1,1,1,1]}}
 {id=WavePlayer_0, properties {soundWaveAssetPtr=/AmbientAudioContent/Waves/OneShots/Spot_Metal_Knocks_01}}
 {id=WavePlayer_1, properties {soundWaveAssetPtr=/AmbientAudioContent/Waves/OneShots/Spot_Metal_Knocks_02}}
 {id=WavePlayer_2, properties {soundWaveAssetPtr=/AmbientAudioContent/Waves/OneShots/Spot_Metal_Knocks_03}}
 {id=WavePlayer_3, properties {soundWaveAssetPtr=/AmbientAudioContent/Waves/OneShots/Spot_Metal_Knocks_04}}
 {id=WavePlayer_8, properties {soundWaveAssetPtr=/AmbientAudioContent/Waves/OneShots/Spot_Metal_Knocks_05}}
]
properties {
 attenuationSettings=/Game/Sounds/Attenuation/Fort_TileAmbient_OneShot_Distant_Creak_Att
 bEnableBusSends=false
 maxDistance=8200
 soundClassObject=/Game/Sounds/FortSoundClassesAndModes/Fort_Ambient_OneShots_Indoors
 volumeMultiplier=0.2
 soundSubmixSends [{minSendLevel=0.6, soundSubmix=&Fortnite_Reverb_Distant}]
}
