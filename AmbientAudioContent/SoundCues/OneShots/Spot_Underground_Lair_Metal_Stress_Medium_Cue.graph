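# Sound cue for a medium metal-stress ambient one-shot in an underground lair, built from underground train station recordings.
# A Random node picks among thirteen equally weighted waves, and a Modulator varies pitch slightly downward and volume slightly below full.
# It plays very quietly in the indoor ambient one-shot sound class, with a distant-creak attenuation and a send to a distant reverb submix.
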SoundCue /AmbientAudioContent/SoundCues/OneShots/Spot_Underground_Lair_Metal_Stress_Medium_Cue
root=Modulator_0
nodes [
 {id=Modulator_0, children [Random_1], properties {pitchMax=1, pitchMin=0.8, volumeMax=1, volumeMin=0.7}}
 {
  id=Random_1
  children [
   WavePlayer_26
   WavePlayer_27
   WavePlayer_28
   WavePlayer_29
   WavePlayer_30
   WavePlayer_31
   WavePlayer_32
   WavePlayer_33
   WavePlayer_34
   WavePlayer_35
   WavePlayer_36
   WavePlayer_37
   WavePlayer_38
  ]
  properties {weights [1,1,1,1,1,1,1,1,1,1,1,1,1]}
 }
 {id=WavePlayer_26, properties {soundWaveAssetPtr=/AmbientAudioContent/Waves/OneShots/Spot_UndergroundTrainStation_MetalStress_01}}
 {id=WavePlayer_27, properties {soundWaveAssetPtr=/AmbientAudioContent/Waves/OneShots/Spot_UndergroundTrainStation_MetalStress_02}}
 {id=WavePlayer_28, properties {soundWaveAssetPtr=/AmbientAudioContent/Waves/OneShots/Spot_UndergroundTrainStation_MetalStress_03}}
 {id=WavePlayer_29, properties {soundWaveAssetPtr=/AmbientAudioContent/Waves/OneShots/Spot_UndergroundTrainStation_MetalStress_04}}
 {id=WavePlayer_30, properties {soundWaveAssetPtr=/AmbientAudioContent/Waves/OneShots/Spot_UndergroundTrainStation_MetalStress_05}}
 {id=WavePlayer_31, properties {soundWaveAssetPtr=/AmbientAudioContent/Waves/OneShots/Spot_UndergroundTrainStation_MetalStress_06}}
 {id=WavePlayer_32, properties {soundWaveAssetPtr=/AmbientAudioContent/Waves/OneShots/Spot_UndergroundTrainStation_MetalStress_07}}
 {id=WavePlayer_33, properties {soundWaveAssetPtr=/AmbientAudioContent/Waves/OneShots/Spot_UndergroundTrainStation_MetalStress_08}}
 {id=WavePlayer_34, properties {soundWaveAssetPtr=/AmbientAudioContent/Waves/OneShots/Spot_UndergroundTrainStation_MetalStress_09}}
 {id=WavePlayer_35, properties {soundWaveAssetPtr=/AmbientAudioContent/Waves/OneShots/Spot_UndergroundTrainStation_MetalStress_10}}
 {id=WavePlayer_36, properties {soundWaveAssetPtr=/AmbientAudioContent/Waves/OneShots/Spot_UndergroundTrainStation_MetalStress_11}}
 {id=WavePlayer_37, properties {soundWaveAssetPtr=/AmbientAudioContent/Waves/OneShots/Spot_UndergroundTrainStation_MetalStress_12}}
 {id=WavePlayer_38, properties {soundWaveAssetPtr=/AmbientAudioContent/Waves/OneShots/Spot_UndergroundTrainStation_MetalStress_13}}
]
properties {
 attenuationSettings=/Game/Sounds/Attenuation/Fort_TileAmbient_OneShot_Distant_Creak_Att
 bEnableBusSends=false
 maxDistance=8200
 soundClassObject=/Game/Sounds/FortSoundClassesAndModes/Fort_Ambient_OneShots_Indoors
 volumeMultiplier=0.1
 soundSubmixSends [{minSendLevel=0.6, soundSubmix=&Fortnite_Reverb_Distant}]
}
