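# Sound cue for a quiet ambient one-shot of distant rail rumble, named for an underground lair and classed as an indoor ambient one-shot.
# A Random node picks among five equally weighted underground train station rumble waves.
# An Enveloper applies pitch and volume curves, lowering volume over the later part, then a Modulator adds slight random volume variation.
# Distant creak attenuation and a distant reverb submix send place the sound in space.
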SoundCue /AmbientAudioContent/SoundCues/OneShots/Spot_Underground_Lair_Rail_Rumble_Cue
root=Modulator_0
nodes [
 {
  id=Enveloper_0
  children [Random_0]
  properties {pitchCurve {editorCurveData {keys [{time=0}, <at 10 0.95>]}}, volumeCurve {editorCurveData {keys [<at 0 1>,<at 2 1>,<at 10 0.5>]}}}
 }
 {id=Modulator_0, children [Enveloper_0], properties {pitchMax=1, pitchMin=1, volumeMax=1, volumeMin=0.8}}
 {id=Random_0, children [WavePlayer_4,WavePlayer_5,WavePlayer_6,WavePlayer_7,WavePlayer_9], properties {weights [1,1,1,1,1]}}
 {id=WavePlayer_4, properties {soundWaveAssetPtr=/AmbientAudioContent/Waves/OneShots/Spot_UndergroundTrainStation_RailRumble_01}}
 {id=WavePlayer_5, properties {soundWaveAssetPtr=/AmbientAudioContent/Waves/OneShots/Spot_UndergroundTrainStation_RailRumble_02}}
 {id=WavePlayer_6, properties {soundWaveAssetPtr=/AmbientAudioContent/Waves/OneShots/Spot_UndergroundTrainStation_RailRumble_03}}
 {id=WavePlayer_7, properties {soundWaveAssetPtr=/AmbientAudioContent/Waves/OneShots/Spot_UndergroundTrainStation_RailRumble_04}}
 {id=WavePlayer_9, properties {soundWaveAssetPtr=/AmbientAudioContent/Waves/OneShots/Spot_UndergroundTrainStation_RailRumble_05}}
]
properties {
 attenuationSettings=/Game/Sounds/Attenuation/Fort_TileAmbient_OneShot_Distant_Creak_Att
 bEnableBusSends=false
 maxDistance=8200
 soundClassObject=/Game/Sounds/FortSoundClassesAndModes/Fort_Ambient_OneShots_Indoors
 volumeMultiplier=0.15
 soundSubmixSends [{minSendLevel=0.6, soundSubmix=&Fortnite_Reverb_Distant}]
}
