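# Sound cue for daytime desolate Underworld wind one-shots, played as attenuated outdoor ambient spot sounds with at most two concurrent instances.
# A quality-level switch sends its first two levels to an even random choice between Hades tonal gust and Alpine quad wind pools.
# The last two levels use only a smaller pool of Alpine ST wind waves.
# Each pool picks a wave at random under slight pitch and volume modulation.
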SoundCue /AmbientAudioContent/SoundCues/OneShots/Spot_Underworld_Desolate_Wind_Oneshots_Day_Cue
root=QualityLevel_0
nodes [
 {id=Modulator_0, children [Random_6], properties {pitchMax=1.1, pitchMin=0.9, volumeMax=0.9, volumeMin=0.8}}
 {id=Modulator_2, children [Random_4], properties {pitchMax=1.1, pitchMin=0.9, volumeMax=0.9, volumeMin=0.7}}
 {id=Modulator_3, children [Random_5], properties {pitchMax=1.1, pitchMin=0.9, volumeMax=0.9, volumeMin=0.7}}
 {id=QualityLevel_0, children [Random_3,Random_3,Modulator_0,Modulator_0]}
 {id=Random_3, children [Modulator_2,Modulator_3], properties {weights [1,1]}}
 {
  id=Random_4
  children [
   WavePlayer_19
   WavePlayer_20
   WavePlayer_21
   WavePlayer_22
   WavePlayer_23
   WavePlayer_24
   WavePlayer_25
   WavePlayer_26
   WavePlayer_27
   WavePlayer_28
   WavePlayer_29
  ]
  properties {weights [1,1,1,1,1,1,1,1,1,1,1]}
 }
 {id=Random_5, children [WavePlayer_30, WavePlayer_31, WavePlayer_32, WavePlayer_33], properties {weights [1,1,1,1]}}
 {id=Random_6, children [WavePlayer_34, WavePlayer_35, WavePlayer_36, WavePlayer_37], properties {weights [1,1,1,1]}}
 {id=WavePlayer_19, properties {soundWaveAssetPtr=/AmbientAudioContent/Waves/OneShots/Spot_Hades_Wind_Gust_Tonal_01}}
 {id=WavePlayer_20, properties {soundWaveAssetPtr=/AmbientAudioContent/Waves/OneShots/Spot_Hades_Wind_Gust_Tonal_02}}
 {id=WavePlayer_21, properties {soundWaveAssetPtr=/AmbientAudioContent/Waves/OneShots/Spot_Hades_Wind_Gust_Tonal_03}}
 {id=WavePlayer_22, properties {soundWaveAssetPtr=/AmbientAudioContent/Waves/OneShots/Spot_Hades_Wind_Gust_Tonal_04}}
 {id=WavePlayer_23, properties {soundWaveAssetPtr=/AmbientAudioContent/Waves/OneShots/Spot_Hades_Wind_Gust_Tonal_05}}
 {id=WavePlayer_24, properties {soundWaveAssetPtr=/AmbientAudioContent/Waves/OneShots/Spot_Hades_Wind_Gust_Tonal_06}}
 {id=WavePlayer_25, properties {soundWaveAssetPtr=/AmbientAudioContent/Waves/OneShots/Spot_Hades_Wind_Gust_Tonal_07}}
 {id=WavePlayer_26, properties {soundWaveAssetPtr=/AmbientAudioContent/Waves/OneShots/Spot_Hades_Wind_Gust_Tonal_08}}
 {id=WavePlayer_27, properties {soundWaveAssetPtr=/AmbientAudioContent/Waves/OneShots/Spot_Hades_Wind_Gust_Tonal_09}}
 {id=WavePlayer_28, properties {soundWaveAssetPtr=/AmbientAudioContent/Waves/OneShots/Spot_Hades_Wind_Gust_Tonal_10}}
 {id=WavePlayer_29, properties {soundWaveAssetPtr=/AmbientAudioContent/Waves/OneShots/Spot_Hades_Wind_Gust_Tonal_11}}
 {id=WavePlayer_30, properties {soundWaveAssetPtr=/AmbientAudioContent/Waves/OneShots/Spot_Wind_Alpine_Oneshot_Day_Quad_01}}
 {id=WavePlayer_31, properties {soundWaveAssetPtr=/AmbientAudioContent/Waves/OneShots/Spot_Wind_Alpine_Oneshot_Day_Quad_02}}
 {id=WavePlayer_32, properties {soundWaveAssetPtr=/AmbientAudioContent/Waves/OneShots/Spot_Wind_Alpine_Oneshot_Day_Quad_03}}
 {id=WavePlayer_33, properties {soundWaveAssetPtr=/AmbientAudioContent/Waves/OneShots/Spot_Wind_Alpine_Oneshot_Day_Quad_04}}
 {id=WavePlayer_34, properties {soundWaveAssetPtr=/AmbientAudioContent/Waves/OneShots/Spot_Wind_Alpine_Oneshot_Day_ST_01}}
 {id=WavePlayer_35, properties {soundWaveAssetPtr=/AmbientAudioContent/Waves/OneShots/Spot_Wind_Alpine_Oneshot_Day_ST_02}}
 {id=WavePlayer_36, properties {soundWaveAssetPtr=/AmbientAudioContent/Waves/OneShots/Spot_Wind_Alpine_Oneshot_Day_ST_03}}
 {id=WavePlayer_37, properties {soundWaveAssetPtr=/AmbientAudioContent/Waves/OneShots/Spot_Wind_Alpine_Oneshot_Day_ST_04}}
]
properties {
 attenuationSettings=/AmbientAudioBase/Attenuations/Fort_TileAmbient_OneShot_Wind_Att
 bOverrideConcurrency=true
 maxDistance=8000
 soundClassObject=&Fort_Ambient_OneShots_Outdoors
 volumeMultiplier=0.33
 concurrencyOverrides {maxCount=2}
 preEffectBusSends [{}]
}
