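# Ambient one-shot SoundCue for ghostly voices in the Underworld, using distant-creature attenuation and the indoor one-shot sound class.
# A weighted random picks spooky night sounds under a pitch-drop envelope, crypt ghost voices, or modulated lava-ghost sounds, each randomizing among its own waves.
# The Quality Level node feeds that random only into its first two slots, leaving the others empty.
# Concurrency is capped at two voices, and output also sends to the exterior ambient source bus.
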SoundCue /AmbientAudioContent/SoundCues/OneShots/Spot_Underworld_Ghostly_Voices_Cue
root=QualityLevel_0
nodes [
 {
  id=Enveloper_0
  children [Mixer_1]
  properties {pitchCurve {editorCurveData {keys [{time=0}, {time=2}, <at 10 0.85>]}}, volumeCurve {editorCurveData {keys [<at 0 1>]}}}
 }
 {id=Mixer_1, children [Random_4], properties {inputVolume [0.75]}}
 {id=Mixer_2, children [WavePlayer_32], properties {inputVolume [2]}}
 {id=Mixer_4, children [Random_0], properties {inputVolume [1]}}
 {id=Mixer_5, children [Random_1], properties {inputVolume [0.75]}}
 {id=Mixer_8, children [WavePlayer_29], properties {inputVolume [0.75]}}
 {id=Modulator_0, children [Mixer_5], properties {volumeMax=1.2}}
 {id=QualityLevel_0, children [Random_2,Random_2,null,null]}
 {id=Random_0, children [WavePlayer_0,WavePlayer_1,WavePlayer_2,WavePlayer_3,WavePlayer_4,WavePlayer_5], properties {weights [1,1,1,1,1,1]}}
 {
  id=Random_1
  children [
   WavePlayer_6
   WavePlayer_7
   WavePlayer_8
   WavePlayer_9
   WavePlayer_10
   WavePlayer_11
   WavePlayer_12
   WavePlayer_13
   WavePlayer_14
   WavePlayer_15
   WavePlayer_16
  ]
  properties {weights [1,1,1,1,1,1,1,1,1,1,1]}
 }
 {id=Random_2, children [Enveloper_0,Mixer_4,Modulator_0], properties {weights [1,0.5,0.8]}}
 {id=Random_4, children [WavePlayer_28, Mixer_8, WavePlayer_30, WavePlayer_31, Mixer_2], properties {weights [1,1,1,1,1]}}
 {id=WavePlayer_0, properties {soundWaveAssetPtr=/AmbientAudioContent/Waves/OneShots/Spot_Starlight_Crypt_Ghost_Voice_01}}
 {id=WavePlayer_1, properties {soundWaveAssetPtr=/AmbientAudioContent/Waves/OneShots/Spot_Starlight_Crypt_Ghost_Voice_02}}
 {id=WavePlayer_10, properties {soundWaveAssetPtr=/AmbientAudioContent/Waves/OneShots/Spot_Starlight_S2_Amb_Lava_Ghost_05}}
 {id=WavePlayer_11, properties {soundWaveAssetPtr=/AmbientAudioContent/Waves/OneShots/Spot_Starlight_S2_Amb_Lava_Ghost_06}}
 {id=WavePlayer_12, properties {soundWaveAssetPtr=/AmbientAudioContent/Waves/OneShots/Spot_Starlight_S2_Amb_Lava_Ghost_07}}
 {id=WavePlayer_13, properties {soundWaveAssetPtr=/AmbientAudioContent/Waves/OneShots/Spot_Starlight_S2_Amb_Lava_Ghost_08}}
 {id=WavePlayer_14, properties {soundWaveAssetPtr=/AmbientAudioContent/Waves/OneShots/Spot_Starlight_S2_Amb_Lava_Ghost_09}}
 {id=WavePlayer_15, properties {soundWaveAssetPtr=/AmbientAudioContent/Waves/OneShots/Spot_Starlight_S2_Amb_Lava_Ghost_10}}
 {id=WavePlayer_16, properties {soundWaveAssetPtr=/AmbientAudioContent/Waves/OneShots/Spot_Starlight_S2_Amb_Lava_Ghost_11}}
 {id=WavePlayer_2, properties {soundWaveAssetPtr=/AmbientAudioContent/Waves/OneShots/Spot_Starlight_Crypt_Ghost_Voice_03}}
 {id=WavePlayer_28, properties {soundWaveAssetPtr=/AmbientAudioContent/Waves/OneShots/Spot_Starlight_Spooky_Rand_Night_02}}
 {id=WavePlayer_29, properties {soundWaveAssetPtr=/AmbientAudioContent/Waves/OneShots/Spot_Starlight_Spooky_Rand_Night_04}}
 {id=WavePlayer_3, properties {soundWaveAssetPtr=/AmbientAudioContent/Waves/OneShots/Spot_Starlight_Crypt_Ghost_Voice_04}}
 {id=WavePlayer_30, properties {soundWaveAssetPtr=/AmbientAudioContent/Waves/OneShots/Spot_Starlight_Spooky_Rand_Night_05}}
 {id=WavePlayer_31, properties {soundWaveAssetPtr=/AmbientAudioContent/Waves/OneShots/Spot_Starlight_Spooky_Rand_Night_06}}
 {id=WavePlayer_32, properties {soundWaveAssetPtr=/AmbientAudioContent/Waves/OneShots/Spot_Starlight_Spooky_Rand_Night_03}}
 {id=WavePlayer_4, properties {soundWaveAssetPtr=/AmbientAudioContent/Waves/OneShots/Spot_Starlight_Crypt_Ghost_Voice_05}}
 {id=WavePlayer_5, properties {soundWaveAssetPtr=/AmbientAudioContent/Waves/OneShots/Spot_Starlight_Crypt_Ghost_Voice_06}}
 {id=WavePlayer_6, properties {soundWaveAssetPtr=/AmbientAudioContent/Waves/OneShots/Spot_Starlight_S2_Amb_Lava_Ghost_01}}
 {id=WavePlayer_7, properties {soundWaveAssetPtr=/AmbientAudioContent/Waves/OneShots/Spot_Starlight_S2_Amb_Lava_Ghost_02}}
 {id=WavePlayer_8, properties {soundWaveAssetPtr=/AmbientAudioContent/Waves/OneShots/Spot_Starlight_S2_Amb_Lava_Ghost_03}}
 {id=WavePlayer_9, properties {soundWaveAssetPtr=/AmbientAudioContent/Waves/OneShots/Spot_Starlight_S2_Amb_Lava_Ghost_04}}
]
properties {
 attenuationSettings=/Game/Sounds/Attenuation/Fort_TileAmbient_OneShot_Distant_Creature_Att
 bOverrideConcurrency=true
 maxDistance=8500
 soundClassObject=/Game/Sounds/FortSoundClassesAndModes/Fort_Ambient_OneShots_Indoors
 volumeMultiplier=0.501
 busSends [{audioBus=null, minSendLevel=1, sendLevel=1, soundSourceBus=&AmbientAudio_Exterior_SourceBus}]
 concurrencyOverrides {maxCount=2, voiceStealReleaseTime=2}
}
