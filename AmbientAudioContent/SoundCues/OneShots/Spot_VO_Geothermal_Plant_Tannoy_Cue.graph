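# Sound cue for one-shot geothermal plant tannoy voice announcements in outdoor ambience, silent on the last two of four quality levels.
# It randomly picks either cave announcements pitched up, with one empty slot for silence, or quieter geothermal announcements with varied volume and lowered pitch.
# Overridden attenuation and a two-voice limit control playback, with sends to a distance-scaled reflection source bus and a distant reverb submix.
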SoundCue /AmbientAudioContent/SoundCues/OneShots/Spot_VO_Geothermal_Plant_Tannoy_Cue
root=QualityLevel_0
nodes [
 {id=Mixer_2, children [Modulator_3], properties {inputVolume [0.8]}}
 {id=Modulator_0, children [Random_1], properties {pitchMax=1.2, pitchMin=1.1}}
 {id=Modulator_3, children [Random_4], properties {pitchMax=1, pitchMin=0.8, volumeMin=0.8}}
 {id=QualityLevel_0, children [Random_5,Random_5,null,null]}
 {
  id=Random_1
  children [
   WavePlayer_13
   WavePlayer_14
   WavePlayer_15
   WavePlayer_16
   WavePlayer_17
   WavePlayer_18
   WavePlayer_19
   WavePlayer_20
   WavePlayer_21
   WavePlayer_22
   WavePlayer_23
   WavePlayer_24
   null
  ]
  properties {weights [1,1,1,1,1,1,1,1,1,1,1,1,1]}
 }
 {
  id=Random_4
  children [WavePlayer_62, WavePlayer_63, WavePlayer_64, WavePlayer_65, WavePlayer_66, WavePlayer_67, WavePlayer_68, WavePlayer_69, WavePlayer_70]
  properties {weights [1,1,1,1,1,1,1,1,1]}
 }
 {id=Random_5, children [Modulator_0,Mixer_2], properties {weights [1,1]}}
 {id=WavePlayer_13, properties {soundWaveAssetPtr=/AmbientAudioContent/Waves/OneShots/Spot_Cave_Tannoy_Announcement_01}}
 {id=WavePlayer_14, properties {soundWaveAssetPtr=/AmbientAudioContent/Waves/OneShots/Spot_Cave_Tannoy_Announcement_02}}
 {id=WavePlayer_15, properties {soundWaveAssetPtr=/AmbientAudioContent/Waves/OneShots/Spot_Cave_Tannoy_Announcement_03}}
 {id=WavePlayer_16, properties {soundWaveAssetPtr=/AmbientAudioContent/Waves/OneShots/Spot_Cave_Tannoy_Announcement_04}}
 {id=WavePlayer_17, properties {soundWaveAssetPtr=/AmbientAudioContent/Waves/OneShots/Spot_Cave_Tannoy_Announcement_05}}
 {id=WavePlayer_18, properties {soundWaveAssetPtr=/AmbientAudioContent/Waves/OneShots/Spot_Cave_Tannoy_Announcement_06}}
 {id=WavePlayer_19, properties {soundWaveAssetPtr=/AmbientAudioContent/Waves/OneShots/Spot_Cave_Tannoy_Announcement_07}}
 {id=WavePlayer_20, properties {soundWaveAssetPtr=/AmbientAudioContent/Waves/OneShots/Spot_Cave_Tannoy_Announcement_08}}
 {id=WavePlayer_21, properties {soundWaveAssetPtr=/AmbientAudioContent/Waves/OneShots/Spot_Cave_Tannoy_Announcement_09}}
 {id=WavePlayer_22, properties {soundWaveAssetPtr=/AmbientAudioContent/Waves/OneShots/Spot_Cave_Tannoy_Announcement_10}}
 {id=WavePlayer_23, properties {soundWaveAssetPtr=/AmbientAudioContent/Waves/OneShots/Spot_Cave_Tannoy_Announcement_11}}
 {id=WavePlayer_24, properties {soundWaveAssetPtr=/AmbientAudioContent/Waves/OneShots/Spot_Cave_Tannoy_Announcement_12}}
 {id=WavePlayer_62, properties {soundWaveAssetPtr=/AmbientAudioContent/Waves/OneShots/Spot_VO_Geothermal_Tannoy_Annouce_01}}
 {id=WavePlayer_63, properties {soundWaveAssetPtr=/AmbientAudioContent/Waves/OneShots/Spot_VO_Geothermal_Tannoy_Annouce_02}}
 {id=WavePlayer_64, properties {soundWaveAssetPtr=/AmbientAudioContent/Waves/OneShots/Spot_VO_Geothermal_Tannoy_Annouce_03}}
 {id=WavePlayer_65, properties {soundWaveAssetPtr=/AmbientAudioContent/Waves/OneShots/Spot_VO_Geothermal_Tannoy_Annouce_04}}
 {id=WavePlayer_66, properties {soundWaveAssetPtr=/AmbientAudioContent/Waves/OneShots/Spot_VO_Geothermal_Tannoy_Annouce_05}}
 {id=WavePlayer_67, properties {soundWaveAssetPtr=/AmbientAudioContent/Waves/OneShots/Spot_VO_Geothermal_Tannoy_Annouce_06}}
 {id=WavePlayer_68, properties {soundWaveAssetPtr=/AmbientAudioContent/Waves/OneShots/Spot_VO_Geothermal_Tannoy_Annouce_07}}
 {id=WavePlayer_69, properties {soundWaveAssetPtr=/AmbientAudioContent/Waves/OneShots/Spot_VO_Geothermal_Tannoy_Annouce_08}}
 {id=WavePlayer_70, properties {soundWaveAssetPtr=/AmbientAudioContent/Waves/OneShots/Spot_VO_Geothermal_Tannoy_Annouce_09}}
]
properties {
 bOverrideAttenuation=true
 bOverrideConcurrency=true
 maxDistance=6000
 soundClassObject=&Fort_Ambient_Outdoors_Athena
 volumeMultiplier=1
 attenuationOverrides {falloffDistance=5000, stereoSpread=500, attenuationShapeExtents {x=1000}}
 busSends [{
   audioBus=null
   maxSendDistance=4000
   minSendDistance=1000
   minSendLevel=0.4
   sendLevel=1
   soundSourceBus=/Game/Sounds/Apollo/Biomes/Reflections/AmbientAudio_Reflection_SourceBus
   customSendLevelCurve {editorCurveData {keys [
      {arriveTangent=0.8438, interpMode=RCIM_Cubic, leaveTangent=0.8438, tangentMode=RCTM_User, time=0, value=0.4}
      {arriveTangent=-0.0305, interpMode=RCIM_Cubic, tangentMode=RCTM_Break, time=1, value=1}
     ]}}
  }]
 concurrencyOverrides {maxCount=2, voiceStealReleaseTime=0.5}
 soundSubmixSends [{minSendLevel=0.2, sendLevel=0.2, soundSubmix=&Fortnite_Reverb_Distant}]
}
