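# Sound cue for a small exterior waterflow spot ambience, built from looping waves despite its OneShots folder.
# A mixer blends close and distant cascade layers, each with its own box attenuation, and a random pick of waterfall roar or river loop.
# Only the close cascade enables occlusion.
# Pitch modulators vary each layer, then an enveloper and outer focus-aware attenuation shape the mix, with raised cue pitch and volume and capped concurrency.
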SoundCue /AmbientAudioContent/SoundCues/OneShots/Spot_Waterflow_Small_Top_Exterior_Cue
root=Attenuation_1
nodes [
 {
  id=Attenuation_0
  children [Modulator_2]
  properties {
   bOverrideAttenuation=true
   attenuationOverrides {
    attenuationShape=Box
    bEnableListenerFocus=true
    bEnableOcclusion=true
    falloffDistance=500
    nonFocusVolumeAttenuation=0.75
    nonSpatializedRadiusEnd=100
    occlusionInterpolationTime=5
    occlusionTraceChannel=ECC_GameTraceChannel11
    occlusionVolumeAttenuation=0.2
    attenuationShapeExtents=<300 100 400>
   }
  }
 }
 {
  id=Attenuation_1
  children [Enveloper_0]
  properties {
   bOverrideAttenuation=true
   attenuationOverrides {
    bAttenuateWithLPF=true
    bEnableFocusInterpolation=true
    bEnableListenerFocus=true
    distanceAlgorithm=Logarithmic
    falloffDistance=1500
    focusAttackInterpSpeed=0.2
    focusVolumeAttenuation=1.2
    lPFFrequencyAtMax=1000
    lPFRadiusMax=5000
    lPFRadiusMin=2000
    nonFocusAzimuth=75
    nonFocusVolumeAttenuation=0.8
    occlusionInterpolationTime=2
    occlusionLowPassFilterFrequency=1000
    occlusionTraceChannel=ECC_GameTraceChannel11
    occlusionVolumeAttenuation=0.2
    stereoSpread=250
    attenuationShapeExtents {x=750}
    customAttenuationCurve {editorCurveData {keys [
       <at 0 0>
       {arriveTangent=0.028, interpMode=RCIM_Cubic, leaveTangent=0.028, tangentMode=RCTM_User, time=0.2209, value=1}
       {arriveTangent=-0.2554, interpMode=RCIM_Cubic, leaveTangent=-0.2554, tangentMode=RCTM_User, time=0.4, value=0.97}
       <at 1 0>
      ]}}
   }
  }
 }
 {
  id=Attenuation_2
  children [Modulator_3]
  properties {
   bOverrideAttenuation=true
   attenuationOverrides {
    attenuationShape=Box
    bAttenuateWithLPF=true
    bEnableListenerFocus=true
    distanceAlgorithm=NaturalSound
    falloffDistance=2000
    lPFFrequencyAtMax=5000
    lPFFrequencyAtMin=22000
    lPFRadiusMax=3000
    lPFRadiusMin=750
    nonFocusVolumeAttenuation=0.75
    occlusionInterpolationTime=5
    occlusionTraceChannel=ECC_GameTraceChannel11
    occlusionVolumeAttenuation=0.2
    stereoSpread=100
    attenuationShapeExtents=<300 100 400>
   }
  }
 }
 {
  id=Enveloper_0
  children [Mixer_0]
  properties {
   pitchCurve {editorCurveData {keys [{time=0}]}}
   volumeCurve {editorCurveData {keys [<at 0 1 RCIM_Cubic>,<at 2.1763 0.8818 RCIM_Cubic>,<at 5 1>]}}
  }
 }
 {id=Mixer_0, children [Attenuation_0, Attenuation_2, Modulator_0], properties {inputVolume [0.6,1,1]}}
 {id=Modulator_0, children [Random_0], properties {pitchMax=1.1}}
 {id=Modulator_2, children [WavePlayer_0], properties {pitchMax=1.1}}
 {id=Modulator_3, children [WavePlayer_2], properties {pitchMax=1.1}}
 {id=Random_0, children [WavePlayer_7,WavePlayer_1], properties {weights [1,1]}}
 {id=WavePlayer_0, properties {bLooping=true, soundWaveAssetPtr=/AmbientAudioContent/Waves/Water/Spot_Waterfall_Infinitiy_Pool_Cascade_Close}}
 {id=WavePlayer_1, properties {bLooping=true, soundWaveAssetPtr=/AmbientAudioContent/Waves/Water/Spot_Water_River_Close_Loop_03}}
 {id=WavePlayer_2, properties {bLooping=true, soundWaveAssetPtr=/AmbientAudioContent/Waves/Water/Spot_Waterfall_Infinitiy_Pool_Cascade_Dist}}
 {id=WavePlayer_7, properties {bLooping=true, soundWaveAssetPtr=/AmbientAudioContent/Waves/Water/Spot_RealityFalls_Waterfall_Roar_Close}}
]
properties {
 bOverrideAttenuation=true
 bOverrideConcurrency=true
 maxDistance=2400
 pitchMultiplier=1.3
 soundClassObject=/Game/Sounds/FortSoundClassesAndModes/Fort_Water_Ambient
 volumeMultiplier=1.6
 attenuationOverrides {
  bEnableListenerFocus=true
  distanceAlgorithm=Logarithmic
  falloffDistance=1500
  focusAzimuth=45
  focusVolumeAttenuation=1.2
  lPFFrequencyAtMax=1000
  nonFocusVolumeAttenuation=0.9
  nonSpatializedRadiusStart=700
  stereoSpread=1000
  attenuationShapeExtents {x=200}
 }
 concurrencyOverrides {maxCount=6, voiceStealReleaseTime=5}
 concurrencySet [/Game/Sounds/Concurrency_Rules/Waterfall_StopFarthestOldest_Conc]
}
