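# Sound cue that plays a random distant wetland creature one-shot for outdoor ambience: deer, wapiti, bullfrog, cricket frog or frog.
# An equally weighted random node picks a species branch, whose modulator varies playback before a random wave variation, except the single wapiti wave.
# The frog branch gets a strong volume boost.
# All four quality levels share the same tree, which uses distant-creature attenuation and a distance-ranged pre-effect bus send.
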SoundCue /AmbientAudioContent/SoundCues/OneShots/Spot_Wetland_Creatures_All_Cue
root=QualityLevel_0
nodes [
 {id=Modulator_0, children [Random_3], properties {volumeMax=2.5, volumeMin=2}}
 {id=Modulator_4, children [Random_0]}
 {id=Modulator_5, children [WavePlayer_5]}
 {id=Modulator_6, children [Random_1]}
 {id=Modulator_7, children [Random_2], properties {volumeMin=0.7}}
 {id=QualityLevel_0, children [Random_9,Random_9,Random_9,Random_9]}
 {id=Random_0, children [WavePlayer_2,WavePlayer_3], properties {weights [1,1]}}
 {id=Random_1, children [WavePlayer_4, WavePlayer_20, WavePlayer_21, WavePlayer_22], properties {weights [1,1,1,1]}}
 {id=Random_2, children [WavePlayer_23, WavePlayer_24, WavePlayer_25, WavePlayer_26], properties {weights [1,1,1,1]}}
 {
  id=Random_3
  children [WavePlayer_27, WavePlayer_28, WavePlayer_29, WavePlayer_30, WavePlayer_31, WavePlayer_32]
  properties {weights [1,1,1,1,1,1]}
 }
 {id=Random_9, children [Modulator_4,Modulator_5,Modulator_6,Modulator_7,Modulator_0], properties {weights [1,1,1,1,1]}}
 {id=WavePlayer_2, properties {soundWaveAssetPtr=/AmbientAudioContent/Waves/OneShots/Spot_Wetland_Animal_Deer_01}}
 {id=WavePlayer_20, properties {soundWaveAssetPtr=/AmbientAudioContent/Waves/OneShots/Spot_Wetland_Bullfrog_02}}
 {id=WavePlayer_21, properties {soundWaveAssetPtr=/AmbientAudioContent/Waves/OneShots/Spot_Wetland_Bullfrog_03}}
 {id=WavePlayer_22, properties {soundWaveAssetPtr=/AmbientAudioContent/Waves/OneShots/Spot_Wetland_Bullfrog_04}}
 {id=WavePlayer_23, properties {soundWaveAssetPtr=/AmbientAudioContent/Waves/OneShots/Spot_Wetland_CricketFrog_01}}
 {id=WavePlayer_24, properties {soundWaveAssetPtr=/AmbientAudioContent/Waves/OneShots/Spot_Wetland_CricketFrog_02}}
 {id=WavePlayer_25, properties {soundWaveAssetPtr=/AmbientAudioContent/Waves/OneShots/Spot_Wetland_CricketFrog_03}}
 {id=WavePlayer_26, properties {soundWaveAssetPtr=/AmbientAudioContent/Waves/OneShots/Spot_Wetland_CricketFrog_04}}
 {id=WavePlayer_27, properties {soundWaveAssetPtr=/AmbientAudioContent/Waves/OneShots/Spot_Wetland_Frog_01}}
 {id=WavePlayer_28, properties {soundWaveAssetPtr=/AmbientAudioContent/Waves/OneShots/Spot_Wetland_Frog_02}}
 {id=WavePlayer_29, properties {soundWaveAssetPtr=/AmbientAudioContent/Waves/OneShots/Spot_Wetland_Frog_03}}
 {id=WavePlayer_3, properties {soundWaveAssetPtr=/AmbientAudioContent/Waves/OneShots/Spot_Wetland_Animal_Deer_02}}
 {id=WavePlayer_30, properties {soundWaveAssetPtr=/AmbientAudioContent/Waves/OneShots/Spot_Wetland_Frog_04}}
 {id=WavePlayer_31, properties {soundWaveAssetPtr=/AmbientAudioContent/Waves/OneShots/Spot_Wetland_Frog_05}}
 {id=WavePlayer_32, properties {soundWaveAssetPtr=/AmbientAudioContent/Waves/OneShots/Spot_Wetland_Frog_06}}
 {id=WavePlayer_4, properties {soundWaveAssetPtr=/AmbientAudioContent/Waves/OneShots/Spot_Wetland_Bullfrog_01}}
 {id=WavePlayer_5, properties {soundWaveAssetPtr=/AmbientAudioContent/Waves/OneShots/Spot_Wetland_Animal_Wapiti_01}}
]
properties {
 attenuationSettings=/Game/Sounds/Attenuation/Fort_TileAmbient_OneShot_Distant_Creature_Att
 bEnableSubmixSends=false
 maxDistance=8500
 soundClassObject=&Fort_Ambient_Outdoors_Athena
 soundSubmixObject=&Fortnite_Reverb_Distant
 volumeMultiplier=0.4
 preEffectBusSends [{maxSendDistance=8000, minSendDistance=3000, minSendLevel=0.6}]
}
