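# Sound cue for distant creature one-shots in wetland ambience, classed as outdoor ambient one-shots with a distant-creature attenuation asset.
# A weighted random node usually picks one of three wetland animal calls, lowered through a mixer, and less often one of seven Cattus cave creature groans.
# Each branch randomizes pitch, and the groans also vary volume.
# Distance-scaled sends feed an exterior ambient source bus and a distant reverb submix.
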SoundCue /AmbientAudioContent/SoundCues/OneShots/Spot_Wetland_Distant_Creature_Groans_Cue
root=Modulator_0
nodes [
 {id=Mixer_0, children [Modulator_3], properties {inputVolume [0.4]}}
 {id=Modulator_0, children [Random_2]}
 {id=Modulator_1, children [Random_4], properties {pitchMax=1.1, pitchMin=0.75, volumeMax=1, volumeMin=0.8}}
 {id=Modulator_3, children [Random_1], properties {pitchMax=1.1, pitchMin=0.75, volumeMax=1, volumeMin=1}}
 {id=Random_1, children [WavePlayer_10, WavePlayer_11, WavePlayer_12], properties {weights [1,1,1]}}
 {id=Random_2, children [Mixer_0,Modulator_1], properties {weights [1,0.25]}}
 {
  id=Random_4
  children [WavePlayer_13, WavePlayer_14, WavePlayer_15, WavePlayer_16, WavePlayer_17, WavePlayer_18, WavePlayer_19]
  properties {weights [1,1,1,1,1,1,1]}
 }
 {id=WavePlayer_10, properties {soundWaveAssetPtr=/AmbientAudioContent/Waves/OneShots/Spot_Wetland_Animal_OneShots_01}}
 {id=WavePlayer_11, properties {soundWaveAssetPtr=/AmbientAudioContent/Waves/OneShots/Spot_Wetland_Animal_OneShots_02}}
 {id=WavePlayer_12, properties {soundWaveAssetPtr=/AmbientAudioContent/Waves/OneShots/Spot_Wetland_Animal_OneShots_03}}
 {id=WavePlayer_13, properties {soundWaveAssetPtr=/AmbientAudioContent/Waves/OneShots/Spot_Cattus_Cave_Distant_Creature_Groan_01}}
 {id=WavePlayer_14, properties {soundWaveAssetPtr=/AmbientAudioContent/Waves/OneShots/Spot_Cattus_Cave_Distant_Creature_Groan_02}}
 {id=WavePlayer_15, properties {soundWaveAssetPtr=/AmbientAudioContent/Waves/OneShots/Spot_Cattus_Cave_Distant_Creature_Groan_03}}
 {id=WavePlayer_16, properties {soundWaveAssetPtr=/AmbientAudioContent/Waves/OneShots/Spot_Cattus_Cave_Distant_Creature_Groan_04}}
 {id=WavePlayer_17, properties {soundWaveAssetPtr=/AmbientAudioContent/Waves/OneShots/Spot_Cattus_Cave_Distant_Creature_Groan_05}}
 {id=WavePlayer_18, properties {soundWaveAssetPtr=/AmbientAudioContent/Waves/OneShots/Spot_Cattus_Cave_Distant_Creature_Groan_06}}
 {id=WavePlayer_19, properties {soundWaveAssetPtr=/AmbientAudioContent/Waves/OneShots/Spot_Cattus_Cave_Distant_Creature_Groan_07}}
]
properties {
 attenuationSettings=/Game/Sounds/Attenuation/Fort_TileAmbient_OneShot_Distant_Creature_Att
 maxDistance=8500
 soundClassObject=&Fort_Ambient_OneShots_Outdoors
 volumeMultiplier=0.5
 attenuationOverrides {bAttenuate=false}
 busSends [{
   audioBus=null
   maxSendDistance=8000
   minSendDistance=3000
   minSendLevel=0.6
   sendLevel=1
   soundSourceBus=&AmbientAudio_Exterior_SourceBus
   customSendLevelCurve {externalCurve=/Game/Sounds/SubMixes/Reverb/Fortnite_Reverb_Distant_Send}
  }]
 soundSubmixSends [{
   maxSendDistance=8000
   maxSendLevel=0.4
   minSendDistance=3000
   minSendLevel=0.25
   soundSubmix=&Fortnite_Reverb_Distant
   customSendLevelCurve {editorCurveData {keys [<at 0 0>,<at 1 1>]}}
  }]
}
