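# Sound cue for outdoor ambient one-shot wetland insect and fly-by spot sounds.
# A Random node picks among nine equally weighted waves, with the first boosted through a Mixer, then feeds a Modulator for pitch variation.
# All four quality levels share that same chain.
# It uses insect fly-by attenuation and the outdoor one-shot sound class at reduced volume; an exterior bus send is configured but bus sends are disabled.
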SoundCue /AmbientAudioContent/SoundCues/OneShots/Spot_Wetland_Fly_Cue
root=QualityLevel_0
nodes [
 {id=Mixer_0, children [WavePlayer_4], properties {inputVolume [2]}}
 {id=Modulator_1, children [Random_0], properties {pitchMax=1.1}}
 {id=QualityLevel_0, children [Modulator_1,Modulator_1,Modulator_1,Modulator_1]}
 {
  id=Random_0
  children [Mixer_0, WavePlayer_5, WavePlayer_6, WavePlayer_7, WavePlayer_13, WavePlayer_14, WavePlayer_15, WavePlayer_16, WavePlayer_17]
  properties {weights [1,1,1,1,1,1,1,1,1]}
 }
 {id=WavePlayer_13, properties {soundWaveAssetPtr=/AmbientAudioContent/Waves/OneShots/Spot_Wetland_Insects_05}}
 {id=WavePlayer_14, properties {soundWaveAssetPtr=/AmbientAudioContent/Waves/OneShots/Spot_Wetland_Insects_06}}
 {id=WavePlayer_15, properties {soundWaveAssetPtr=/AmbientAudioContent/Waves/OneShots/Spot_Wetland_Insects_Flyby_01}}
 {id=WavePlayer_16, properties {soundWaveAssetPtr=/AmbientAudioContent/Waves/OneShots/Spot_Wetland_Insects_Flyby_02}}
 {id=WavePlayer_17, properties {soundWaveAssetPtr=/AmbientAudioContent/Waves/OneShots/Spot_Wetland_Insects_Flyby_03}}
 {id=WavePlayer_4, properties {soundWaveAssetPtr=/AmbientAudioContent/Waves/OneShots/Spot_Wetland_Insects_01}}
 {id=WavePlayer_5, properties {soundWaveAssetPtr=/AmbientAudioContent/Waves/OneShots/Spot_Wetland_Insects_02}}
 {id=WavePlayer_6, properties {soundWaveAssetPtr=/AmbientAudioContent/Waves/OneShots/Spot_Wetland_Insects_03}}
 {id=WavePlayer_7, properties {soundWaveAssetPtr=/AmbientAudioContent/Waves/OneShots/Spot_Wetland_Insects_04}}
]
properties {
 attenuationSettings=/Game/Sounds/Attenuation/Fort_TileAmbient_OneShot_Insect_By_Att
 bEnableBusSends=false
 maxDistance=6000
 soundClassObject=&Fort_Ambient_OneShots_Outdoors
 volumeMultiplier=0.4
 busSends [{
   audioBus=null
   maxSendDistance=6000
   minSendDistance=3000
   minSendLevel=0.6
   sendLevel=1
   soundSourceBus=&AmbientAudio_Exterior_SourceBus
  }]
}
