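# Sound cue for distant one-shot wetland rumble ambience in the outdoor Athena ambient sound class.
# A Random node picks one of six equally weighted rumble waves, and a Modulator slightly varies its volume.
# It uses a distant-creature attenuation asset and overrides concurrency to limit simultaneous instances.
# Distance-based sends feed an exterior ambient source bus and the default reverb submix.
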SoundCue /AmbientAudioContent/SoundCues/OneShots/Spot_Wetland_Rumbles_Cue
root=Modulator_0
nodes [
 {id=Modulator_0, children [Random_0], properties {volumeMax=1, volumeMin=0.8}}
 {
  id=Random_0
  children [WavePlayer_6, WavePlayer_7, WavePlayer_8, WavePlayer_9, WavePlayer_10, WavePlayer_11]
  properties {weights [1,1,1,1,1,1]}
 }
 {id=WavePlayer_10, properties {soundWaveAssetPtr=/AmbientAudioContent/Waves/OneShots/Spot_Wetland_Rumble_02}}
 {id=WavePlayer_11, properties {soundWaveAssetPtr=/AmbientAudioContent/Waves/OneShots/Spot_Wetland_Rumble_01}}
 {id=WavePlayer_6, properties {soundWaveAssetPtr=/AmbientAudioContent/Waves/OneShots/Spot_Wetland_Rumble_06}}
 {id=WavePlayer_7, properties {soundWaveAssetPtr=/AmbientAudioContent/Waves/OneShots/Spot_Wetland_Rumble_05}}
 {id=WavePlayer_8, properties {soundWaveAssetPtr=/AmbientAudioContent/Waves/OneShots/Spot_Wetland_Rumble_04}}
 {id=WavePlayer_9, properties {soundWaveAssetPtr=/AmbientAudioContent/Waves/OneShots/Spot_Wetland_Rumble_03}}
]
properties {
 attenuationSettings=/Game/Sounds/Attenuation/Fort_TileAmbient_OneShot_Distant_Creature_Att
 bOverrideConcurrency=true
 maxDistance=8500
 soundClassObject=&Fort_Ambient_Outdoors_Athena
 volumeMultiplier=0.5
 busSends [{
   audioBus=null
   maxSendDistance=8000
   minSendDistance=3000
   minSendLevel=0.4
   sendLevel=1
   soundSourceBus=&AmbientAudio_Exterior_SourceBus
  }]
 concurrencyOverrides {maxCount=2}
 soundSubmixSends [{
   maxSendDistance=8000
   maxSendLevel=0.6
   minSendDistance=3000
   minSendLevel=0.4
   soundSubmix=&Fortnite_Reverb_Default
   customSendLevelCurve {editorCurveData {keys [<at 0 0.4>,<at 1 1>]}}
  }]
}
