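# Sound cue for spot wetland wind one-shots in the outdoor ambient sound class.
# A Random node picks among five equally weighted wind waves and feeds a Modulator that varies pitch and volume.
# All four inputs of the root quality-level node share that same chain, so quality levels sound identical.
# It uses a one-shot wind attenuation asset, a distance-based pre-effect bus send, and a concurrency override capping simultaneous voices.
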SoundCue /AmbientAudioContent/SoundCues/OneShots/Spot_Wetland_Wind_Cue
root=QualityLevel_0
nodes [
 {id=Modulator_0, children [Random_1], properties {pitchMax=1.4, pitchMin=0.9, volumeMax=1.2}}
 {id=QualityLevel_0, children [Modulator_0,Modulator_0,Modulator_0,Modulator_0]}
 {id=Random_1, children [WavePlayer_5,WavePlayer_6,WavePlayer_7,WavePlayer_8,WavePlayer_9], properties {weights [1,1,1,1,1]}}
 {id=WavePlayer_5, properties {soundWaveAssetPtr=/AmbientAudioContent/Waves/OneShots/Spot_Wetland_Wind_01}}
 {id=WavePlayer_6, properties {soundWaveAssetPtr=/AmbientAudioContent/Waves/OneShots/Spot_Wetland_Wind_02}}
 {id=WavePlayer_7, properties {soundWaveAssetPtr=/AmbientAudioContent/Waves/OneShots/Spot_Wetland_Wind_03}}
 {id=WavePlayer_8, properties {soundWaveAssetPtr=/AmbientAudioContent/Waves/OneShots/Spot_Wetland_Wind_04}}
 {id=WavePlayer_9, properties {soundWaveAssetPtr=/AmbientAudioContent/Waves/OneShots/Spot_Wetland_Wind_05}}
]
properties {
 attenuationSettings=/Game/Sounds/Attenuation/Fort_TileAmbient_OneShot_Wind_Att
 bOverrideConcurrency=true
 maxDistance=8000
 soundClassObject=&Fort_Ambient_OneShots_Outdoors
 volumeMultiplier=0.3
 concurrencyOverrides {maxCount=2, voiceStealReleaseTime=1}
 preEffectBusSends [{maxSendDistance=6000, minSendDistance=3000, minSendLevel=0.6}]
}
