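# Ambient one-shot cue that plays a random whispering, panning voice spot sound with spatial attenuation and a send to the exterior ambient audio bus.
# A quality-level root sends the first two levels to a full random pool and the last two to a smaller subset excluded from branch culling.
# Each pool feeds a modulator for slight pitch and volume variation.
# A concurrency override allows one instance and blocks new ones.
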SoundCue /AmbientAudioContent/SoundCues/OneShots/Spot_Whispering_Panning_Voices_Cue
root=QualityLevel_0
nodes [
 {id=Modulator_2, children [Random_2], properties {pitchMax=1.2, pitchMin=0.8, volumeMax=1.1, volumeMin=0.9}}
 {id=Modulator_3, children [Random_1], properties {pitchMax=1.2, pitchMin=0.8, volumeMax=1.1, volumeMin=0.9}}
 {id=QualityLevel_0, children [Modulator_2,Modulator_2,Modulator_3,Modulator_3]}
 {
  id=Random_1
  children [WavePlayer_22, WavePlayer_23, WavePlayer_24, WavePlayer_25, WavePlayer_66, WavePlayer_74]
  properties {bShouldExcludeFromBranchCulling=true, weights [1,1,1,1,1,1]}
 }
 {
  id=Random_2
  children [
   WavePlayer_26
   WavePlayer_27
   WavePlayer_28
   WavePlayer_29
   WavePlayer_30
   WavePlayer_31
   WavePlayer_32
   WavePlayer_33
   WavePlayer_34
   WavePlayer_35
   WavePlayer_36
   WavePlayer_37
   WavePlayer_38
   WavePlayer_39
   WavePlayer_40
   WavePlayer_41
   WavePlayer_42
   WavePlayer_43
   WavePlayer_44
   WavePlayer_45
   WavePlayer_46
   WavePlayer_47
  ]
  properties {weights [1,1,1,1,1,1,1,1,1,1,1,1,1,1,1,1,1,1,1,1,1,1]}
 }
 {id=WavePlayer_22, properties {soundWaveAssetPtr=/AmbientAudioContent/Waves/OneShots/Spot_Whispering_Panning_Voices_22}}
 {id=WavePlayer_23, properties {soundWaveAssetPtr=/AmbientAudioContent/Waves/OneShots/Spot_Whispering_Panning_Voices_20}}
 {id=WavePlayer_24, properties {soundWaveAssetPtr=/AmbientAudioContent/Waves/OneShots/Spot_Whispering_Panning_Voices_21}}
 {id=WavePlayer_25, properties {soundWaveAssetPtr=/AmbientAudioContent/Waves/OneShots/Spot_Whispering_Panning_Voices_01}}
 {id=WavePlayer_26, properties {soundWaveAssetPtr=&Spot_Whispering_Panning_Voices_01}}
 {id=WavePlayer_27, properties {soundWaveAssetPtr=/AmbientAudioContent/Waves/OneShots/Spot_Whispering_Panning_Voices_02}}
 {id=WavePlayer_28, properties {soundWaveAssetPtr=/AmbientAudioContent/Waves/OneShots/Spot_Whispering_Panning_Voices_03}}
 {id=WavePlayer_29, properties {soundWaveAssetPtr=/AmbientAudioContent/Waves/OneShots/Spot_Whispering_Panning_Voices_04}}
 {id=WavePlayer_30, properties {soundWaveAssetPtr=/AmbientAudioContent/Waves/OneShots/Spot_Whispering_Panning_Voices_05}}
 {id=WavePlayer_31, properties {soundWaveAssetPtr=/AmbientAudioContent/Waves/OneShots/Spot_Whispering_Panning_Voices_06}}
 {id=WavePlayer_32, properties {soundWaveAssetPtr=/AmbientAudioContent/Waves/OneShots/Spot_Whispering_Panning_Voices_07}}
 {id=WavePlayer_33, properties {soundWaveAssetPtr=/AmbientAudioContent/Waves/OneShots/Spot_Whispering_Panning_Voices_08}}
 {id=WavePlayer_34, properties {soundWaveAssetPtr=/AmbientAudioContent/Waves/OneShots/Spot_Whispering_Panning_Voices_09}}
 {id=WavePlayer_35, properties {soundWaveAssetPtr=/AmbientAudioContent/Waves/OneShots/Spot_Whispering_Panning_Voices_10}}
 {id=WavePlayer_36, properties {soundWaveAssetPtr=/AmbientAudioContent/Waves/OneShots/Spot_Whispering_Panning_Voices_11}}
 {id=WavePlayer_37, properties {soundWaveAssetPtr=/AmbientAudioContent/Waves/OneShots/Spot_Whispering_Panning_Voices_12}}
 {id=WavePlayer_38, properties {soundWaveAssetPtr=/AmbientAudioContent/Waves/OneShots/Spot_Whispering_Panning_Voices_13}}
 {id=WavePlayer_39, properties {soundWaveAssetPtr=/AmbientAudioContent/Waves/OneShots/Spot_Whispering_Panning_Voices_14}}
 {id=WavePlayer_40, properties {soundWaveAssetPtr=/AmbientAudioContent/Waves/OneShots/Spot_Whispering_Panning_Voices_15}}
 {id=WavePlayer_41, properties {soundWaveAssetPtr=/AmbientAudioContent/Waves/OneShots/Spot_Whispering_Panning_Voices_16}}
 {id=WavePlayer_42, properties {soundWaveAssetPtr=/AmbientAudioContent/Waves/OneShots/Spot_Whispering_Panning_Voices_17}}
 {id=WavePlayer_43, properties {soundWaveAssetPtr=/AmbientAudioContent/Waves/OneShots/Spot_Whispering_Panning_Voices_18}}
 {id=WavePlayer_44, properties {soundWaveAssetPtr=/AmbientAudioContent/Waves/OneShots/Spot_Whispering_Panning_Voices_19}}
 {id=WavePlayer_45, properties {soundWaveAssetPtr=&Spot_Whispering_Panning_Voices_20}}
 {id=WavePlayer_46, properties {soundWaveAssetPtr=&Spot_Whispering_Panning_Voices_21}}
 {id=WavePlayer_47, properties {soundWaveAssetPtr=&Spot_Whispering_Panning_Voices_22}}
 {id=WavePlayer_66, properties {soundWaveAssetPtr=&Spot_Whispering_Panning_Voices_02}}
 {id=WavePlayer_74, properties {soundWaveAssetPtr=&Spot_Whispering_Panning_Voices_03}}
]
properties {
 attenuationSettings=/Game/Sounds/Attenuation/Fort_TileAmbient_OneShot_Wider_Att
 bOverrideConcurrency=true
 maxDistance=8000
 soundClassObject=&Fort_Ambient_OneShots
 volumeMultiplier=0.301
 busSends [{audioBus=null, minSendLevel=1, sendLevel=1, soundSourceBus=&AmbientAudio_Exterior_SourceBus}]
 concurrencyOverrides {maxCount=1, resolutionRule=PreventNew, voiceStealReleaseTime=2}
}
