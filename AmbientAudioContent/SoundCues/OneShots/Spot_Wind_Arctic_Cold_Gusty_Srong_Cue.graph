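# Sound cue for one-shot spot ambience of strong, gusty, cold arctic wind, in the Athena ambient sound class with a tile-ambient wind attenuation.
# A Random node picks among twelve equally weighted wind waves, and a Modulator varies pitch and volume on each play.
# The root Quality Level node routes all four of its inputs to that same Modulator, so the listed quality levels sound identical.
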SoundCue /AmbientAudioContent/SoundCues/OneShots/Spot_Wind_Arctic_Cold_Gusty_Srong_Cue
root=QualityLevel_1
nodes [
 {id=Modulator_3, children [Random_0], properties {pitchMax=1.1, pitchMin=0.9, volumeMax=1, volumeMin=0.6}}
 {id=QualityLevel_1, children [Modulator_3,Modulator_3,Modulator_3,Modulator_3]}
 {
  id=Random_0
  children [
   WavePlayer_0
   WavePlayer_1
   WavePlayer_2
   WavePlayer_3
   WavePlayer_4
   WavePlayer_5
   WavePlayer_6
   WavePlayer_7
   WavePlayer_8
   WavePlayer_9
   WavePlayer_10
   WavePlayer_11
  ]
  properties {weights [1,1,1,1,1,1,1,1,1,1,1,1]}
 }
 {id=WavePlayer_0, properties {soundWaveAssetPtr=/AmbientAudioContent/Waves/OneShots/Spot_Wind_Arctic_Cold_Gusty_Strong_01}}
 {id=WavePlayer_1, properties {soundWaveAssetPtr=/AmbientAudioContent/Waves/OneShots/Spot_Wind_Arctic_Cold_Gusty_Strong_02}}
 {id=WavePlayer_10, properties {soundWaveAssetPtr=/AmbientAudioContent/Waves/OneShots/Spot_Wind_Arctic_Cold_Gusty_Strong_11}}
 {id=WavePlayer_11, properties {soundWaveAssetPtr=/AmbientAudioContent/Waves/OneShots/Spot_Wind_Arctic_Cold_Gusty_Strong_12}}
 {id=WavePlayer_2, properties {soundWaveAssetPtr=/AmbientAudioContent/Waves/OneShots/Spot_Wind_Arctic_Cold_Gusty_Strong_03}}
 {id=WavePlayer_3, properties {soundWaveAssetPtr=/AmbientAudioContent/Waves/OneShots/Spot_Wind_Arctic_Cold_Gusty_Strong_04}}
 {id=WavePlayer_4, properties {soundWaveAssetPtr=/AmbientAudioContent/Waves/OneShots/Spot_Wind_Arctic_Cold_Gusty_Strong_05}}
 {id=WavePlayer_5, properties {soundWaveAssetPtr=/AmbientAudioContent/Waves/OneShots/Spot_Wind_Arctic_Cold_Gusty_Strong_06}}
 {id=WavePlayer_6, properties {soundWaveAssetPtr=/AmbientAudioContent/Waves/OneShots/Spot_Wind_Arctic_Cold_Gusty_Strong_07}}
 {id=WavePlayer_7, properties {soundWaveAssetPtr=/AmbientAudioContent/Waves/OneShots/Spot_Wind_Arctic_Cold_Gusty_Strong_08}}
 {id=WavePlayer_8, properties {soundWaveAssetPtr=/AmbientAudioContent/Waves/OneShots/Spot_Wind_Arctic_Cold_Gusty_Strong_09}}
 {id=WavePlayer_9, properties {soundWaveAssetPtr=/AmbientAudioContent/Waves/OneShots/Spot_Wind_Arctic_Cold_Gusty_Strong_10}}
]
properties {
 attenuationSettings=/Game/Sounds/Attenuation/Fort_TileAmbient_OneShot_Wind_Att
 bEnableBusSends=false
 maxDistance=8000
 soundClassObject=&Fort_Ambient_Athena
 volumeMultiplier=0.301
 preEffectBusSends [{}]
}
