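# Sound cue for outdoor ambient one-shots of wind rustling autumn tree leaves, using a wide one-shot attenuation.
# A quality-level node sends its first two slots to a full variation pool, randomly choosing between regular and short leaf sets, then a wave within.
# Its last two slots use a small mixed subset of those waves, excluded from branch culling.
# Each branch has its own modulator randomizing pitch and volume.
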SoundCue /AmbientAudioContent/SoundCues/OneShots/Spot_Wind_Autumnal_Leaves_Tree_Rustle_Cue
root=QualityLevel_0
nodes [
 {id=Modulator_1, children [Random_4], properties {pitchMax=1.1, pitchMin=1, volumeMax=1, volumeMin=0.25}}
 {id=Modulator_2, children [Random_5], properties {pitchMax=1.1, volumeMax=1.1, volumeMin=0.1}}
 {id=QualityLevel_0, children [Modulator_1,Modulator_1,Modulator_2,Modulator_2]}
 {
  id=Random_2
  children [
   WavePlayer_24
   WavePlayer_25
   WavePlayer_26
   WavePlayer_27
   WavePlayer_28
   WavePlayer_29
   WavePlayer_30
   WavePlayer_31
   WavePlayer_32
   WavePlayer_33
   WavePlayer_34
   WavePlayer_35
   WavePlayer_46
   WavePlayer_47
   WavePlayer_48
   WavePlayer_49
   WavePlayer_50
  ]
  properties {weights [1,1,1,1,1,1,1,1,1,1,1,1,1,1,1,1,1]}
 }
 {
  id=Random_3
  children [
   WavePlayer_51
   WavePlayer_58
   WavePlayer_59
   WavePlayer_60
   WavePlayer_61
   WavePlayer_62
   WavePlayer_63
   WavePlayer_64
   WavePlayer_65
   WavePlayer_66
   WavePlayer_67
   WavePlayer_68
   WavePlayer_69
   WavePlayer_70
   WavePlayer_71
   WavePlayer_72
   WavePlayer_73
  ]
  properties {weights [1,1,1,1,1,1,1,1,1,1,1,1,1,1,1,1,1]}
 }
 {id=Random_4, children [Random_2,Random_3], properties {weights [1,1]}}
 {
  id=Random_5
  children [WavePlayer_52, WavePlayer_53, WavePlayer_54, WavePlayer_55, WavePlayer_56, WavePlayer_57]
  properties {bShouldExcludeFromBranchCulling=true, weights [1,1,1,1,1,1]}
 }
 {id=WavePlayer_24, properties {soundWaveAssetPtr=/AmbientAudioContent/Waves/OneShots/Spot_Wind_Autumnal_Leaves_17}}
 {id=WavePlayer_25, properties {soundWaveAssetPtr=/AmbientAudioContent/Waves/OneShots/Spot_Wind_Autumnal_Leaves_16}}
 {id=WavePlayer_26, properties {soundWaveAssetPtr=/AmbientAudioContent/Waves/OneShots/Spot_Wind_Autumnal_Leaves_15}}
 {id=WavePlayer_27, properties {soundWaveAssetPtr=/AmbientAudioContent/Waves/OneShots/Spot_Wind_Autumnal_Leaves_14}}
 {id=WavePlayer_28, properties {soundWaveAssetPtr=/AmbientAudioContent/Waves/OneShots/Spot_Wind_Autumnal_Leaves_13}}
 {id=WavePlayer_29, properties {soundWaveAssetPtr=/AmbientAudioContent/Waves/OneShots/Spot_Wind_Autumnal_Leaves_12}}
 {id=WavePlayer_30, properties {soundWaveAssetPtr=/AmbientAudioContent/Waves/OneShots/Spot_Wind_Autumnal_Leaves_11}}
 {id=WavePlayer_31, properties {soundWaveAssetPtr=/AmbientAudioContent/Waves/OneShots/Spot_Wind_Autumnal_Leaves_10}}
 {id=WavePlayer_32, properties {soundWaveAssetPtr=/AmbientAudioContent/Waves/OneShots/Spot_Wind_Autumnal_Leaves_09}}
 {id=WavePlayer_33, properties {soundWaveAssetPtr=/AmbientAudioContent/Waves/OneShots/Spot_Wind_Autumnal_Leaves_08}}
 {id=WavePlayer_34, properties {soundWaveAssetPtr=/AmbientAudioContent/Waves/OneShots/Spot_Wind_Autumnal_Leaves_07}}
 {id=WavePlayer_35, properties {soundWaveAssetPtr=/AmbientAudioContent/Waves/OneShots/Spot_Wind_Autumnal_Leaves_06}}
 {id=WavePlayer_46, properties {soundWaveAssetPtr=/AmbientAudioContent/Waves/OneShots/Spot_Wind_Autumnal_Leaves_05}}
 {id=WavePlayer_47, properties {soundWaveAssetPtr=/AmbientAudioContent/Waves/OneShots/Spot_Wind_Autumnal_Leaves_04}}
 {id=WavePlayer_48, properties {soundWaveAssetPtr=/AmbientAudioContent/Waves/OneShots/Spot_Wind_Autumnal_Leaves_03}}
 {id=WavePlayer_49, properties {soundWaveAssetPtr=/AmbientAudioContent/Waves/OneShots/Spot_Wind_Autumnal_Leaves_02}}
 {id=WavePlayer_50, properties {soundWaveAssetPtr=/AmbientAudioContent/Waves/OneShots/Spot_Wind_Autumnal_Leaves_01}}
 {id=WavePlayer_51, properties {soundWaveAssetPtr=/AmbientAudioContent/Waves/OneShots/Spot_Wind_Autumnal_Leaves_Short_17}}
 {id=WavePlayer_52, properties {soundWaveAssetPtr=/AmbientAudioContent/Waves/OneShots/Spot_Wind_Autumnal_Leaves_Short_11}}
 {id=WavePlayer_53, properties {soundWaveAssetPtr=/AmbientAudioContent/Waves/OneShots/Spot_Wind_Autumnal_Leaves_Short_13}}
 {id=WavePlayer_54, properties {soundWaveAssetPtr=/AmbientAudioContent/Waves/OneShots/Spot_Wind_Autumnal_Leaves_Short_06}}
 {id=WavePlayer_55, properties {soundWaveAssetPtr=&Spot_Wind_Autumnal_Leaves_03}}
 {id=WavePlayer_56, properties {soundWaveAssetPtr=&Spot_Wind_Autumnal_Leaves_05}}
 {id=WavePlayer_57, properties {soundWaveAssetPtr=&Spot_Wind_Autumnal_Leaves_08}}
 {id=WavePlayer_58, properties {soundWaveAssetPtr=/AmbientAudioContent/Waves/OneShots/Spot_Wind_Autumnal_Leaves_Short_16}}
 {id=WavePlayer_59, properties {soundWaveAssetPtr=/AmbientAudioContent/Waves/OneShots/Spot_Wind_Autumnal_Leaves_Short_15}}
 {id=WavePlayer_60, properties {soundWaveAssetPtr=/AmbientAudioContent/Waves/OneShots/Spot_Wind_Autumnal_Leaves_Short_14}}
 {id=WavePlayer_61, properties {soundWaveAssetPtr=&Spot_Wind_Autumnal_Leaves_Short_13}}
 {id=WavePlayer_62, properties {soundWaveAssetPtr=/AmbientAudioContent/Waves/OneShots/Spot_Wind_Autumnal_Leaves_Short_12}}
 {id=WavePlayer_63, properties {soundWaveAssetPtr=&Spot_Wind_Autumnal_Leaves_Short_11}}
 {id=WavePlayer_64, properties {soundWaveAssetPtr=/AmbientAudioContent/Waves/OneShots/Spot_Wind_Autumnal_Leaves_Short_10}}
 {id=WavePlayer_65, properties {soundWaveAssetPtr=/AmbientAudioContent/Waves/OneShots/Spot_Wind_Autumnal_Leaves_Short_09}}
 {id=WavePlayer_66, properties {soundWaveAssetPtr=/AmbientAudioContent/Waves/OneShots/Spot_Wind_Autumnal_Leaves_Short_08}}
 {id=WavePlayer_67, properties {soundWaveAssetPtr=/AmbientAudioContent/Waves/OneShots/Spot_Wind_Autumnal_Leaves_Short_07}}
 {id=WavePlayer_68, properties {soundWaveAssetPtr=&Spot_Wind_Autumnal_Leaves_Short_06}}
 {id=WavePlayer_69, properties {soundWaveAssetPtr=/AmbientAudioContent/Waves/OneShots/Spot_Wind_Autumnal_Leaves_Short_05}}
 {id=WavePlayer_70, properties {soundWaveAssetPtr=/AmbientAudioContent/Waves/OneShots/Spot_Wind_Autumnal_Leaves_Short_04}}
 {id=WavePlayer_71, properties {soundWaveAssetPtr=/AmbientAudioContent/Waves/OneShots/Spot_Wind_Autumnal_Leaves_Short_03}}
 {id=WavePlayer_72, properties {soundWaveAssetPtr=/AmbientAudioContent/Waves/OneShots/Spot_Wind_Autumnal_Leaves_Short_02}}
 {id=WavePlayer_73, properties {soundWaveAssetPtr=/AmbientAudioContent/Waves/OneShots/Spot_Wind_Autumnal_Leaves_Short_01}}
]
properties {
 attenuationSettings=/Game/Sounds/Attenuation/Fort_TileAmbient_OneShot_Wider_Att
 bEnableBusSends=false
 maxDistance=8000
 soundClassObject=&Fort_Ambient_OneShots_Outdoors
 volumeMultiplier=0.501
 preEffectBusSends [{}]
}
