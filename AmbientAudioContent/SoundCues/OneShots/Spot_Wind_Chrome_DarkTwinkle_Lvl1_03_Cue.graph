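# Sound cue for a dark-twinkle chrome wind ambient one-shot, routed to the Fortnite ambient one-shot sound class with wind one-shot attenuation.
# A Random node picks among ten equally weighted wave variations.
# All four quality-level inputs share that same Random node, so every level draws on the full set.
# A root Modulator adds slight variation on top, and the cue plays at reduced overall volume.
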SoundCue /AmbientAudioContent/SoundCues/OneShots/Spot_Wind_Chrome_DarkTwinkle_Lvl1_03_Cue
root=Modulator_0
nodes [
 {id=Modulator_0, children [QualityLevel_0], properties {volumeMax=1.1, volumeMin=1}}
 {id=QualityLevel_0, children [Random_0,Random_0,Random_0,Random_0]}
 {
  id=Random_0
  children [WavePlayer_0,WavePlayer_1,WavePlayer_2,WavePlayer_3,WavePlayer_4,WavePlayer_5,WavePlayer_6,WavePlayer_7,WavePlayer_8,WavePlayer_9]
  properties {weights [1,1,1,1,1,1,1,1,1,1]}
 }
 {id=WavePlayer_0, properties {soundWaveAssetPtr=/AmbientAudioContent/Waves/OneShots/Spot_Wind_Chrome_DarkTwinkle_Lvl1_03_01}}
 {id=WavePlayer_1, properties {soundWaveAssetPtr=/AmbientAudioContent/Waves/OneShots/Spot_Wind_Chrome_DarkTwinkle_Lvl1_03_02}}
 {id=WavePlayer_2, properties {soundWaveAssetPtr=/AmbientAudioContent/Waves/OneShots/Spot_Wind_Chrome_DarkTwinkle_Lvl1_03_03}}
 {id=WavePlayer_3, properties {soundWaveAssetPtr=/AmbientAudioContent/Waves/OneShots/Spot_Wind_Chrome_DarkTwinkle_Lvl1_03_04}}
 {id=WavePlayer_4, properties {soundWaveAssetPtr=/AmbientAudioContent/Waves/OneShots/Spot_Wind_Chrome_DarkTwinkle_Lvl1_03_05}}
 {id=WavePlayer_5, properties {soundWaveAssetPtr=/AmbientAudioContent/Waves/OneShots/Spot_Wind_Chrome_DarkTwinkle_Lvl1_03_06}}
 {id=WavePlayer_6, properties {soundWaveAssetPtr=/AmbientAudioContent/Waves/OneShots/Spot_Wind_Chrome_DarkTwinkle_Lvl1_03_07}}
 {id=WavePlayer_7, properties {soundWaveAssetPtr=/AmbientAudioContent/Waves/OneShots/Spot_Wind_Chrome_DarkTwinkle_Lvl1_03_08}}
 {id=WavePlayer_8, properties {soundWaveAssetPtr=/AmbientAudioContent/Waves/OneShots/Spot_Wind_Chrome_DarkTwinkle_Lvl1_03_09}}
 {id=WavePlayer_9, properties {soundWaveAssetPtr=/AmbientAudioContent/Waves/OneShots/Spot_Wind_Chrome_DarkTwinkle_Lvl1_03_10}}
]
properties {
 attenuationSettings=/Game/Sounds/Attenuation/Fort_TileAmbient_OneShot_Wind_Att
 bEnableBusSends=false
 maxDistance=8000
 soundClassObject=&Fort_Ambient_OneShots
 volumeMultiplier=0.401
 preEffectBusSends [{}]
}
