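# Sound cue for a one-shot ambient spot wind effect, the howly chrome shimmer at level 1, variant 02.
# A Random node picks evenly among ten wave variations, and all four inputs of the Quality Level node share it, so every quality tier behaves the same.
# A Modulator at the root varies volume slightly.
# The cue uses the ambient one-shot sound class and a wider tile-ambient attenuation, with bus sends disabled.
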SoundCue /AmbientAudioContent/SoundCues/OneShots/Spot_Wind_Chrome_HowlyShimmer_Lvl1_02_Cue
root=Modulator_0
nodes [
 {id=Modulator_0, children [QualityLevel_0], properties {volumeMin=0.9}}
 {id=QualityLevel_0, children [Random_0,Random_0,Random_0,Random_0]}
 {
  id=Random_0
  children [WavePlayer_0,WavePlayer_1,WavePlayer_2,WavePlayer_3,WavePlayer_4,WavePlayer_5,WavePlayer_6,WavePlayer_7,WavePlayer_8,WavePlayer_9]
  properties {weights [1,1,1,1,1,1,1,1,1,1]}
 }
 {id=WavePlayer_0, properties {soundWaveAssetPtr=/AmbientAudioContent/Waves/OneShots/Spot_Wind_Chrome_HowlyShimmer_Lvl1_02_01}}
 {id=WavePlayer_1, properties {soundWaveAssetPtr=/AmbientAudioContent/Waves/OneShots/Spot_Wind_Chrome_HowlyShimmer_Lvl1_02_02}}
 {id=WavePlayer_2, properties {soundWaveAssetPtr=/AmbientAudioContent/Waves/OneShots/Spot_Wind_Chrome_HowlyShimmer_Lvl1_02_03}}
 {id=WavePlayer_3, properties {soundWaveAssetPtr=/AmbientAudioContent/Waves/OneShots/Spot_Wind_Chrome_HowlyShimmer_Lvl1_02_04}}
 {id=WavePlayer_4, properties {soundWaveAssetPtr=/AmbientAudioContent/Waves/OneShots/Spot_Wind_Chrome_HowlyShimmer_Lvl1_02_05}}
 {id=WavePlayer_5, properties {soundWaveAssetPtr=/AmbientAudioContent/Waves/OneShots/Spot_Wind_Chrome_HowlyShimmer_Lvl1_02_06}}
 {id=WavePlayer_6, properties {soundWaveAssetPtr=/AmbientAudioContent/Waves/OneShots/Spot_Wind_Chrome_HowlyShimmer_Lvl1_02_07}}
 {id=WavePlayer_7, properties {soundWaveAssetPtr=/AmbientAudioContent/Waves/OneShots/Spot_Wind_Chrome_HowlyShimmer_Lvl1_02_08}}
 {id=WavePlayer_8, properties {soundWaveAssetPtr=/AmbientAudioContent/Waves/OneShots/Spot_Wind_Chrome_HowlyShimmer_Lvl1_02_09}}
 {id=WavePlayer_9, properties {soundWaveAssetPtr=/AmbientAudioContent/Waves/OneShots/Spot_Wind_Chrome_HowlyShimmer_Lvl1_02_10}}
]
properties {
 attenuationSettings=/Game/Sounds/Attenuation/Fort_TileAmbient_OneShot_Wider_Att
 bEnableBusSends=false
 maxDistance=8000
 soundClassObject=&Fort_Ambient_OneShots
 volumeMultiplier=0.501
 preEffectBusSends [{}]
}
