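# Sound cue for one-shot dirty wind gust ambience, using cave gust waves under the indoor ambient one-shot sound class with tile ambient attenuation.
# A quality-level root sends its first two slots to a random pick of eleven gusts and its last two to a random pick of the first four.
# Each equally weighted pool feeds its own modulator, which varies pitch and boosts volume, offset by a low cue volume multiplier.
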SoundCue /AmbientAudioContent/SoundCues/OneShots/Spot_Wind_Dirty_Gust_Int_Cue
root=QualityLevel_0
nodes [
 {id=Modulator_0, children [Random_2], properties {pitchMax=1, pitchMin=0.8, volumeMax=2.1, volumeMin=2}}
 {id=Modulator_2, children [Random_0], properties {pitchMax=1.1, volumeMax=2.2, volumeMin=2}}
 {id=QualityLevel_0, children [Modulator_0,Modulator_0,Modulator_2,Modulator_2]}
 {id=Random_0, children [WavePlayer_0,WavePlayer_1,WavePlayer_2,WavePlayer_3], properties {weights [1,1,1,1]}}
 {
  id=Random_2
  children [
   WavePlayer_4
   WavePlayer_16
   WavePlayer_17
   WavePlayer_18
   WavePlayer_19
   WavePlayer_20
   WavePlayer_21
   WavePlayer_22
   WavePlayer_23
   WavePlayer_24
   WavePlayer_25
  ]
  properties {weights [1,1,1,1,1,1,1,1,1,1,1]}
 }
 {id=WavePlayer_0, properties {soundWaveAssetPtr=/AmbientAudioContent/Waves/OneShots/Spot_Cave_Dirty_Wind_Gust_01}}
 {id=WavePlayer_1, properties {soundWaveAssetPtr=/AmbientAudioContent/Waves/OneShots/Spot_Cave_Dirty_Wind_Gust_02}}
 {id=WavePlayer_16, properties {soundWaveAssetPtr=&Spot_Cave_Dirty_Wind_Gust_02}}
 {id=WavePlayer_17, properties {soundWaveAssetPtr=/AmbientAudioContent/Waves/OneShots/Spot_Cave_Dirty_Wind_Gust_03}}
 {id=WavePlayer_18, properties {soundWaveAssetPtr=/AmbientAudioContent/Waves/OneShots/Spot_Cave_Dirty_Wind_Gust_04}}
 {id=WavePlayer_19, properties {soundWaveAssetPtr=/AmbientAudioContent/Waves/OneShots/Spot_Cave_Dirty_Wind_Gust_05}}
 {id=WavePlayer_2, properties {soundWaveAssetPtr=&Spot_Cave_Dirty_Wind_Gust_03}}
 {id=WavePlayer_20, properties {soundWaveAssetPtr=/AmbientAudioContent/Waves/OneShots/Spot_Cave_Dirty_Wind_Gust_06}}
 {id=WavePlayer_21, properties {soundWaveAssetPtr=/AmbientAudioContent/Waves/OneShots/Spot_Cave_Dirty_Wind_Gust_07}}
 {id=WavePlayer_22, properties {soundWaveAssetPtr=/AmbientAudioContent/Waves/OneShots/Spot_Cave_Dirty_Wind_Gust_08}}
 {id=WavePlayer_23, properties {soundWaveAssetPtr=/AmbientAudioContent/Waves/OneShots/Spot_Cave_Dirty_Wind_Gust_09}}
 {id=WavePlayer_24, properties {soundWaveAssetPtr=/AmbientAudioContent/Waves/OneShots/Spot_Cave_Dirty_Wind_Gust_10}}
 {id=WavePlayer_25, properties {soundWaveAssetPtr=/AmbientAudioContent/Waves/OneShots/Spot_Cave_Dirty_Wind_Gust_11}}
 {id=WavePlayer_3, properties {soundWaveAssetPtr=&Spot_Cave_Dirty_Wind_Gust_04}}
 {id=WavePlayer_4, properties {soundWaveAssetPtr=&Spot_Cave_Dirty_Wind_Gust_01}}
]
properties {
 attenuationSettings=/Game/Sounds/Attenuation/Fort_TileAmbient_OneShot_Att
 maxDistance=8000
 soundClassObject=/Game/Sounds/FortSoundClassesAndModes/Fort_Ambient_OneShots_Indoors
 volumeMultiplier=0.251
 preEffectBusSends [{}]
}
