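# Sound cue for outdoor ambient wind and sand gust one-shots in the Eastern Lowlands, attenuated by distance.
# A quality node sends its first two levels to a larger wave pool and its last two to a smaller subset.
# Each pool uses nested weighted randoms favoring desert gust and debris waves about three to one over tower sandy-wind waves.
# A modulator on each pool randomizes pitch and volume.
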SoundCue /AmbientAudioContent/SoundCues/OneShots/Spot_Wind_Eastern_Lowlands_Wind_Sand_Cue
root=QualityLevel_0
nodes [
 {id=Modulator_1, children [Random_2], properties {pitchMax=1.1, pitchMin=1, volumeMax=1.1, volumeMin=0.9}}
 {id=Modulator_2, children [Random_3], properties {pitchMax=1.1, volumeMax=1.1, volumeMin=0.9}}
 {id=QualityLevel_0, children [Modulator_1,Modulator_1,Modulator_2,Modulator_2]}
 {id=Random_0, children [WavePlayer_0,WavePlayer_1,WavePlayer_2,WavePlayer_3,WavePlayer_4], properties {weights [1,1,1,1,1]}}
 {id=Random_1, children [WavePlayer_5, WavePlayer_6, WavePlayer_7, WavePlayer_13, WavePlayer_14], properties {weights [1,1,1,1,1]}}
 {id=Random_10, children [WavePlayer_43, WavePlayer_44, WavePlayer_45], properties {weights [1,1,1]}}
 {id=Random_14, children [Random_0,Random_1], properties {weights [1,1]}}
 {id=Random_2, children [Random_14,Random_7], properties {weights [1,0.33]}}
 {id=Random_3, children [Random_10,Random_8], properties {weights [1,0.33]}}
 {
  id=Random_7
  children [WavePlayer_27, WavePlayer_28, WavePlayer_29, WavePlayer_30, WavePlayer_39, WavePlayer_40, WavePlayer_41, WavePlayer_42]
  properties {weights [1,1,1,1,1,1,1,1]}
 }
 {id=Random_8, children [WavePlayer_46, WavePlayer_47, WavePlayer_48], properties {weights [1,1,1]}}
 {id=WavePlayer_0, properties {soundWaveAssetPtr=/AmbientAudioContent/Waves/OneShots/Spot_Desert_Wind_Sand_Gust_01}}
 {id=WavePlayer_1, properties {soundWaveAssetPtr=/AmbientAudioContent/Waves/OneShots/Spot_Desert_Wind_Sand_Gust_02}}
 {id=WavePlayer_13, properties {soundWaveAssetPtr=/AmbientAudioContent/Waves/OneShots/Spot_Desert_WindGust_Med_Lt_Debris_04}}
 {id=WavePlayer_14, properties {soundWaveAssetPtr=/AmbientAudioContent/Waves/OneShots/Spot_Desert_WindGust_Med_Lt_Debris_05}}
 {id=WavePlayer_2, properties {soundWaveAssetPtr=/AmbientAudioContent/Waves/OneShots/Spot_Desert_Wind_Sand_Gust_03}}
 {id=WavePlayer_27, properties {soundWaveAssetPtr=/AmbientAudioContent/Waves/OneShots/Spot_Tower_AM_Sandy_Wind_Oneshot_12}}
 {id=WavePlayer_28, properties {soundWaveAssetPtr=/AmbientAudioContent/Waves/OneShots/Spot_Tower_AM_Sandy_Wind_Oneshot_07}}
 {id=WavePlayer_29, properties {soundWaveAssetPtr=/AmbientAudioContent/Waves/OneShots/Spot_Tower_AM_Sandy_Wind_Oneshot_06}}
 {id=WavePlayer_3, properties {soundWaveAssetPtr=/AmbientAudioContent/Waves/OneShots/Spot_Desert_Wind_Sand_Gust_04}}
 {id=WavePlayer_30, properties {soundWaveAssetPtr=/AmbientAudioContent/Waves/OneShots/Spot_Tower_AM_Sandy_Wind_Oneshot_05}}
 {id=WavePlayer_39, properties {soundWaveAssetPtr=/AmbientAudioContent/Waves/OneShots/Spot_Tower_AM_Sandy_Wind_Oneshot_04}}
 {id=WavePlayer_4, properties {soundWaveAssetPtr=/AmbientAudioContent/Waves/OneShots/Spot_Desert_Wind_Sand_Gust_05}}
 {id=WavePlayer_40, properties {soundWaveAssetPtr=/AmbientAudioContent/Waves/OneShots/Spot_Tower_AM_Sandy_Wind_Oneshot_03}}
 {id=WavePlayer_41, properties {soundWaveAssetPtr=/AmbientAudioContent/Waves/OneShots/Spot_Tower_AM_Sandy_Wind_Oneshot_02}}
 {id=WavePlayer_42, properties {soundWaveAssetPtr=/AmbientAudioContent/Waves/OneShots/Spot_Tower_AM_Sandy_Wind_Oneshot_01}}
 {id=WavePlayer_43, properties {soundWaveAssetPtr=/AmbientAudioContent/Waves/OneShots/Spot_Desert_WindGust_Med_Lt_Debris_01}}
 {id=WavePlayer_44, properties {soundWaveAssetPtr=/AmbientAudioContent/Waves/OneShots/Spot_Desert_WindGust_Med_Lt_Debris_02}}
 {id=WavePlayer_45, properties {soundWaveAssetPtr=/AmbientAudioContent/Waves/OneShots/Spot_Desert_WindGust_Med_Lt_Debris_03}}
 {id=WavePlayer_46, properties {soundWaveAssetPtr=&Spot_Tower_AM_Sandy_Wind_Oneshot_01}}
 {id=WavePlayer_47, properties {soundWaveAssetPtr=&Spot_Tower_AM_Sandy_Wind_Oneshot_12}}
 {id=WavePlayer_48, properties {soundWaveAssetPtr=&Spot_Tower_AM_Sandy_Wind_Oneshot_02}}
 {id=WavePlayer_5, properties {soundWaveAssetPtr=&Spot_Desert_WindGust_Med_Lt_Debris_01}}
 {id=WavePlayer_6, properties {soundWaveAssetPtr=&Spot_Desert_WindGust_Med_Lt_Debris_02}}
 {id=WavePlayer_7, properties {soundWaveAssetPtr=&Spot_Desert_WindGust_Med_Lt_Debris_03}}
]
properties {
 attenuationSettings=/Game/Sounds/Attenuation/Fort_TileAmbient_OneShot_Wind_Att
 bEnableBusSends=false
 maxDistance=8000
 soundClassObject=&Fort_Ambient_OneShots_Outdoors
 volumeMultiplier=0.501
 preEffectBusSends [{}]
}
