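# Sound cue for one-shot sandy night wind gusts in the Eastern Lowlands, on the outdoor ambient one-shot class with wind attenuation.
# A weighted random picks one of three pools: most often sand gusts or light-debris gusts (split evenly), less often tower sandy wind, rarely heavy sandy gusts.
# The pick passes through a modulator that slightly varies pitch and volume.
# All four quality-level inputs share that same modulator, so every level sounds the same.
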SoundCue /AmbientAudioContent/SoundCues/OneShots/Spot_Wind_Eastern_Lowlands_Wind_Sand_Night_Cue
root=QualityLevel_0
nodes [
 {id=Modulator_1, children [Random_2], properties {pitchMax=1.1, pitchMin=1, volumeMax=1.1, volumeMin=0.9}}
 {id=QualityLevel_0, children [Modulator_1,Modulator_1,Modulator_1,Modulator_1]}
 {id=Random_0, children [WavePlayer_4, WavePlayer_5, WavePlayer_6, WavePlayer_7, WavePlayer_13], properties {weights [1,1,1,1,1]}}
 {id=Random_1, children [WavePlayer_14, WavePlayer_15, WavePlayer_16, WavePlayer_17, WavePlayer_18], properties {weights [1,1,1,1,1]}}
 {
  id=Random_11
  children [
   WavePlayer_45
   WavePlayer_46
   WavePlayer_47
   WavePlayer_48
   WavePlayer_49
   WavePlayer_50
   WavePlayer_51
   WavePlayer_52
   WavePlayer_53
   WavePlayer_54
  ]
  properties {weights [1,1,1,1,1,1,1,1,1,1]}
 }
 {id=Random_14, children [Random_0,Random_1], properties {weights [1,1]}}
 {id=Random_2, children [Random_14,Random_7,Random_11], properties {weights [1,0.33,0.1]}}
 {
  id=Random_7
  children [WavePlayer_19, WavePlayer_21, WavePlayer_39, WavePlayer_40, WavePlayer_41, WavePlayer_42, WavePlayer_43, WavePlayer_44]
  properties {weights [1,1,1,1,1,1,1,1]}
 }
 {id=WavePlayer_13, properties {soundWaveAssetPtr=/AmbientAudioContent/Waves/OneShots/Spot_Desert_Wind_Sand_Gust_01}}
 {id=WavePlayer_14, properties {soundWaveAssetPtr=/AmbientAudioContent/Waves/OneShots/Spot_Desert_WindGust_Med_Lt_Debris_01}}
 {id=WavePlayer_15, properties {soundWaveAssetPtr=/AmbientAudioContent/Waves/OneShots/Spot_Desert_WindGust_Med_Lt_Debris_02}}
 {id=WavePlayer_16, properties {soundWaveAssetPtr=/AmbientAudioContent/Waves/OneShots/Spot_Desert_WindGust_Med_Lt_Debris_03}}
 {id=WavePlayer_17, properties {soundWaveAssetPtr=/AmbientAudioContent/Waves/OneShots/Spot_Desert_WindGust_Med_Lt_Debris_04}}
 {id=WavePlayer_18, properties {soundWaveAssetPtr=/AmbientAudioContent/Waves/OneShots/Spot_Desert_WindGust_Med_Lt_Debris_05}}
 {id=WavePlayer_19, properties {soundWaveAssetPtr=/AmbientAudioContent/Waves/OneShots/Spot_Tower_AM_Sandy_Wind_Oneshot_01}}
 {id=WavePlayer_21, properties {soundWaveAssetPtr=/AmbientAudioContent/Waves/OneShots/Spot_Tower_AM_Sandy_Wind_Oneshot_02}}
 {id=WavePlayer_39, properties {soundWaveAssetPtr=/AmbientAudioContent/Waves/OneShots/Spot_Tower_AM_Sandy_Wind_Oneshot_03}}
 {id=WavePlayer_4, properties {soundWaveAssetPtr=/AmbientAudioContent/Waves/OneShots/Spot_Desert_Wind_Sand_Gust_05}}
 {id=WavePlayer_40, properties {soundWaveAssetPtr=/AmbientAudioContent/Waves/OneShots/Spot_Tower_AM_Sandy_Wind_Oneshot_04}}
 {id=WavePlayer_41, properties {soundWaveAssetPtr=/AmbientAudioContent/Waves/OneShots/Spot_Tower_AM_Sandy_Wind_Oneshot_05}}
 {id=WavePlayer_42, properties {soundWaveAssetPtr=/AmbientAudioContent/Waves/OneShots/Spot_Tower_AM_Sandy_Wind_Oneshot_06}}
 {id=WavePlayer_43, properties {soundWaveAssetPtr=/AmbientAudioContent/Waves/OneShots/Spot_Tower_AM_Sandy_Wind_Oneshot_07}}
 {id=WavePlayer_44, properties {soundWaveAssetPtr=/AmbientAudioContent/Waves/OneShots/Spot_Tower_AM_Sandy_Wind_Oneshot_12}}
 {id=WavePlayer_45, properties {soundWaveAssetPtr=/AmbientAudioContent/Waves/OneShots/Spot_Desert_WindGust_Med_hv_Sandy_01}}
 {id=WavePlayer_46, properties {soundWaveAssetPtr=/AmbientAudioContent/Waves/OneShots/Spot_Desert_WindGust_Med_hv_Sandy_02}}
 {id=WavePlayer_47, properties {soundWaveAssetPtr=/AmbientAudioContent/Waves/OneShots/Spot_Desert_WindGust_Med_hv_Sandy_03}}
 {id=WavePlayer_48, properties {soundWaveAssetPtr=/AmbientAudioContent/Waves/OneShots/Spot_Desert_WindGust_Med_hv_Sandy_04}}
 {id=WavePlayer_49, properties {soundWaveAssetPtr=/AmbientAudioContent/Waves/OneShots/Spot_Desert_WindGust_Med_hv_Sandy_05}}
 {id=WavePlayer_5, properties {soundWaveAssetPtr=/AmbientAudioContent/Waves/OneShots/Spot_Desert_Wind_Sand_Gust_04}}
 {id=WavePlayer_50, properties {soundWaveAssetPtr=/AmbientAudioContent/Waves/OneShots/Spot_Desert_WindGust_Med_hv_Sandy_06}}
 {id=WavePlayer_51, properties {soundWaveAssetPtr=/AmbientAudioContent/Waves/OneShots/Spot_Desert_WindGust_Med_hv_Sandy_07}}
 {id=WavePlayer_52, properties {soundWaveAssetPtr=/AmbientAudioContent/Waves/OneShots/Spot_Desert_WindGust_Med_hv_Sandy_08}}
 {id=WavePlayer_53, properties {soundWaveAssetPtr=/AmbientAudioContent/Waves/OneShots/Spot_Desert_WindGust_Med_hv_Sandy_09}}
 {id=WavePlayer_54, properties {soundWaveAssetPtr=/AmbientAudioContent/Waves/OneShots/Spot_Desert_WindGust_Med_hv_Sandy_10}}
 {id=WavePlayer_6, properties {soundWaveAssetPtr=/AmbientAudioContent/Waves/OneShots/Spot_Desert_Wind_Sand_Gust_03}}
 {id=WavePlayer_7, properties {soundWaveAssetPtr=/AmbientAudioContent/Waves/OneShots/Spot_Desert_Wind_Sand_Gust_02}}
]
properties {
 attenuationSettings=/Game/Sounds/Attenuation/Fort_TileAmbient_OneShot_Wind_Att
 bEnableBusSends=false
 maxDistance=8000
 soundClassObject=&Fort_Ambient_OneShots_Outdoors
 volumeMultiplier=0.551
 preEffectBusSends [{}]
}
