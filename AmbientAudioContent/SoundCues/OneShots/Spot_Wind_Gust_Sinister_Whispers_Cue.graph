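# Ambient one-shot cue for sinister wind gusts with whispers, using wide attenuation and a two-instance concurrency cap.
# The first two quality levels mix a random gust tone, from either a pitch- and volume-modulated pool or a smaller unmodulated one, with randomly delayed, volume-modulated panning whisper voices.
# The other two quality levels play only a modulated six-wave gust subset, excluded from branch culling.
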SoundCue /AmbientAudioContent/SoundCues/OneShots/Spot_Wind_Gust_Sinister_Whispers_Cue
root=QualityLevel_0
nodes [
 {id=Delay_0, children [Modulator_1], properties {delayMax=2, delayMin=0.5}}
 {id=Mixer_0, children [Random_3,Delay_0], properties {inputVolume [1.2,1]}}
 {id=Modulator_0, children [Random_6], properties {pitchMax=1.1, pitchMin=0.7, volumeMax=1.1, volumeMin=0.5}}
 {id=Modulator_1, children [Random_5], properties {volumeMax=1.1, volumeMin=0.9}}
 {id=Modulator_2, children [Random_1], properties {pitchMax=1.2, pitchMin=0.8, volumeMax=1.1, volumeMin=0.9}}
 {id=QualityLevel_0, children [Mixer_0,Mixer_0,Modulator_0,Modulator_0]}
 {
  id=Random_1
  children [
   WavePlayer_22
   WavePlayer_23
   WavePlayer_24
   WavePlayer_25
   WavePlayer_47
   WavePlayer_48
   WavePlayer_49
   WavePlayer_50
   WavePlayer_51
   WavePlayer_52
   WavePlayer_53
   WavePlayer_54
   WavePlayer_55
   WavePlayer_56
   WavePlayer_57
   WavePlayer_58
   WavePlayer_59
   WavePlayer_60
   WavePlayer_61
   WavePlayer_62
   WavePlayer_63
   WavePlayer_64
   WavePlayer_65
   WavePlayer_66
   WavePlayer_67
   WavePlayer_74
   WavePlayer_75
   WavePlayer_76
   WavePlayer_77
   WavePlayer_78
   WavePlayer_79
   WavePlayer_80
  ]
  properties {weights [1,1,1,1,1,1,1,1,1,1,1,1,1,1,1,1,1,1,1,1,1,1,1,1,1,1,1,1,1,1,1,1]}
 }
 {
  id=Random_2
  children [
   WavePlayer_81
   WavePlayer_82
   WavePlayer_83
   WavePlayer_84
   WavePlayer_85
   WavePlayer_86
   WavePlayer_87
   WavePlayer_88
   WavePlayer_89
   WavePlayer_90
  ]
  properties {weights [1,1,1,1,1,1,1,1,1,1]}
 }
 {id=Random_3, children [Modulator_2,Random_2], properties {weights [1,1]}}
 {
  id=Random_5
  children [
   WavePlayer_91
   WavePlayer_92
   WavePlayer_93
   WavePlayer_94
   WavePlayer_95
   WavePlayer_96
   WavePlayer_97
   WavePlayer_98
   WavePlayer_99
   WavePlayer_100
   WavePlayer_101
   WavePlayer_102
   WavePlayer_103
   WavePlayer_104
   WavePlayer_105
   WavePlayer_106
   WavePlayer_107
   WavePlayer_108
   WavePlayer_109
   WavePlayer_110
   WavePlayer_111
   WavePlayer_112
  ]
  properties {weights [1,1,1,1,1,1,1,1,1,1,1,1,1,1,1,1,1,1,1,1,1,1]}
 }
 {
  id=Random_6
  children [WavePlayer_71, WavePlayer_72, WavePlayer_73, WavePlayer_68, WavePlayer_69, WavePlayer_70]
  properties {bShouldExcludeFromBranchCulling=true, weights [1,1,1,1,1,1]}
 }
 {id=WavePlayer_100, properties {soundWaveAssetPtr=/AmbientAudioContent/Waves/OneShots/Spot_Whispering_Panning_Voices_10}}
 {id=WavePlayer_101, properties {soundWaveAssetPtr=/AmbientAudioContent/Waves/OneShots/Spot_Whispering_Panning_Voices_11}}
 {id=WavePlayer_102, properties {soundWaveAssetPtr=/AmbientAudioContent/Waves/OneShots/Spot_Whispering_Panning_Voices_12}}
 {id=WavePlayer_103, properties {soundWaveAssetPtr=/AmbientAudioContent/Waves/OneShots/Spot_Whispering_Panning_Voices_13}}
 {id=WavePlayer_104, properties {soundWaveAssetPtr=/AmbientAudioContent/Waves/OneShots/Spot_Whispering_Panning_Voices_14}}
 {id=WavePlayer_105, properties {soundWaveAssetPtr=/AmbientAudioContent/Waves/OneShots/Spot_Whispering_Panning_Voices_15}}
 {id=WavePlayer_106, properties {soundWaveAssetPtr=/AmbientAudioContent/Waves/OneShots/Spot_Whispering_Panning_Voices_16}}
 {id=WavePlayer_107, properties {soundWaveAssetPtr=/AmbientAudioContent/Waves/OneShots/Spot_Whispering_Panning_Voices_17}}
 {id=WavePlayer_108, properties {soundWaveAssetPtr=/AmbientAudioContent/Waves/OneShots/Spot_Whispering_Panning_Voices_18}}
 {id=WavePlayer_109, properties {soundWaveAssetPtr=/AmbientAudioContent/Waves/OneShots/Spot_Whispering_Panning_Voices_19}}
 {id=WavePlayer_110, properties {soundWaveAssetPtr=/AmbientAudioContent/Waves/OneShots/Spot_Whispering_Panning_Voices_20}}
 {id=WavePlayer_111, properties {soundWaveAssetPtr=/AmbientAudioContent/Waves/OneShots/Spot_Whispering_Panning_Voices_21}}
 {id=WavePlayer_112, properties {soundWaveAssetPtr=/AmbientAudioContent/Waves/OneShots/Spot_Whispering_Panning_Voices_22}}
 {id=WavePlayer_22, properties {soundWaveAssetPtr=/AmbientAudioContent/Waves/OneShots/Spot_Wind_Gust_Sinister_Whisper_Tone_01}}
 {id=WavePlayer_23, properties {soundWaveAssetPtr=/AmbientAudioContent/Waves/OneShots/Spot_Wind_Gust_Sinister_Whisper_Tone_02}}
 {id=WavePlayer_24, properties {soundWaveAssetPtr=/AmbientAudioContent/Waves/OneShots/Spot_Wind_Gust_Sinister_Whisper_Tone_03}}
 {id=WavePlayer_25, properties {soundWaveAssetPtr=/AmbientAudioContent/Waves/OneShots/Spot_Wind_Gust_Sinister_Whisper_Tone_04}}
 {id=WavePlayer_47, properties {soundWaveAssetPtr=/AmbientAudioContent/Waves/OneShots/Spot_Wind_Gust_Sinister_Whisper_Tone_05}}
 {id=WavePlayer_48, properties {soundWaveAssetPtr=/AmbientAudioContent/Waves/OneShots/Spot_Wind_Gust_Sinister_Whisper_Tone_06}}
 {id=WavePlayer_49, properties {soundWaveAssetPtr=/AmbientAudioContent/Waves/OneShots/Spot_Wind_Gust_Sinister_Whisper_Tone_07}}
 {id=WavePlayer_50, properties {soundWaveAssetPtr=/AmbientAudioContent/Waves/OneShots/Spot_Wind_Gust_Sinister_Whisper_Tone_08}}
 {id=WavePlayer_51, properties {soundWaveAssetPtr=/AmbientAudioContent/Waves/OneShots/Spot_Wind_Gust_Sinister_Whisper_Tone_09}}
 {id=WavePlayer_52, properties {soundWaveAssetPtr=/AmbientAudioContent/Waves/OneShots/Spot_Wind_Gust_Sinister_Whisper_Tone_10}}
 {id=WavePlayer_53, properties {soundWaveAssetPtr=/AmbientAudioContent/Waves/OneShots/Spot_Wind_Gust_Sinister_Whisper_Tone_11}}
 {id=WavePlayer_54, properties {soundWaveAssetPtr=/AmbientAudioContent/Waves/OneShots/Spot_Wind_Gust_Sinister_Whisper_Tone_12}}
 {id=WavePlayer_55, properties {soundWaveAssetPtr=/AmbientAudioContent/Waves/OneShots/Spot_Wind_Gust_Sinister_Whisper_Tone_13}}
 {id=WavePlayer_56, properties {soundWaveAssetPtr=/AmbientAudioContent/Waves/OneShots/Spot_Wind_Gust_Sinister_Whisper_Tone_14}}
 {id=WavePlayer_57, properties {soundWaveAssetPtr=/AmbientAudioContent/Waves/OneShots/Spot_Wind_Gust_Sinister_Whisper_Tone_15}}
 {id=WavePlayer_58, properties {soundWaveAssetPtr=/AmbientAudioContent/Waves/OneShots/Spot_Wind_Gust_Sinister_Whisper_Tone_16}}
 {id=WavePlayer_59, properties {soundWaveAssetPtr=/AmbientAudioContent/Waves/OneShots/Spot_Wind_Gust_Sinister_Whisper_Tone_17}}
 {id=WavePlayer_60, properties {soundWaveAssetPtr=/AmbientAudioContent/Waves/OneShots/Spot_Wind_Gust_Sinister_Whisper_Tone_18}}
 {id=WavePlayer_61, properties {soundWaveAssetPtr=/AmbientAudioContent/Waves/OneShots/Spot_Wind_Gust_Sinister_Whisper_Tone_19}}
 {id=WavePlayer_62, properties {soundWaveAssetPtr=/AmbientAudioContent/Waves/OneShots/Spot_Wind_Gust_Sinister_Whisper_Tone_20}}
 {id=WavePlayer_63, properties {soundWaveAssetPtr=/AmbientAudioContent/Waves/OneShots/Spot_Wind_Gust_Sinister_Whisper_Tone_21}}
 {id=WavePlayer_64, properties {soundWaveAssetPtr=/AmbientAudioContent/Waves/OneShots/Spot_Wind_Gust_Sinister_Whisper_Tone_22}}
 {id=WavePlayer_65, properties {soundWaveAssetPtr=/AmbientAudioContent/Waves/OneShots/Spot_Wind_Gust_Sinister_Whisper_Tone_23}}
 {id=WavePlayer_66, properties {soundWaveAssetPtr=/AmbientAudioContent/Waves/OneShots/Spot_Wind_Gust_Sinister_Whisper_Tone_24}}
 {id=WavePlayer_67, properties {soundWaveAssetPtr=/AmbientAudioContent/Waves/OneShots/Spot_Wind_Gust_Sinister_Whisper_Tone_25}}
 {id=WavePlayer_68, properties {soundWaveAssetPtr=/AmbientAudioContent/Waves/OneShots/Spot_Wind_Gust_Sinister_Whisper_Tone_40}}
 {id=WavePlayer_69, properties {soundWaveAssetPtr=/AmbientAudioContent/Waves/OneShots/Spot_Wind_Gust_Sinister_Whisper_Tone_41}}
 {id=WavePlayer_70, properties {soundWaveAssetPtr=/AmbientAudioContent/Waves/OneShots/Spot_Wind_Gust_Sinister_Whisper_Tone_42}}
 {id=WavePlayer_71, properties {soundWaveAssetPtr=&Spot_Wind_Gust_Sinister_Whisper_Tone_19}}
 {id=WavePlayer_72, properties {soundWaveAssetPtr=&Spot_Wind_Gust_Sinister_Whisper_Tone_20}}
 {id=WavePlayer_73, properties {soundWaveAssetPtr=&Spot_Wind_Gust_Sinister_Whisper_Tone_21}}
 {id=WavePlayer_74, properties {soundWaveAssetPtr=/AmbientAudioContent/Waves/OneShots/Spot_Wind_Gust_Sinister_Whisper_Tone_26}}
 {id=WavePlayer_75, properties {soundWaveAssetPtr=/AmbientAudioContent/Waves/OneShots/Spot_Wind_Gust_Sinister_Whisper_Tone_27}}
 {id=WavePlayer_76, properties {soundWaveAssetPtr=/AmbientAudioContent/Waves/OneShots/Spot_Wind_Gust_Sinister_Whisper_Tone_28}}
 {id=WavePlayer_77, properties {soundWaveAssetPtr=/AmbientAudioContent/Waves/OneShots/Spot_Wind_Gust_Sinister_Whisper_Tone_29}}
 {id=WavePlayer_78, properties {soundWaveAssetPtr=/AmbientAudioContent/Waves/OneShots/Spot_Wind_Gust_Sinister_Whisper_Tone_30}}
 {id=WavePlayer_79, properties {soundWaveAssetPtr=/AmbientAudioContent/Waves/OneShots/Spot_Wind_Gust_Sinister_Whisper_Tone_31}}
 {id=WavePlayer_80, properties {soundWaveAssetPtr=/AmbientAudioContent/Waves/OneShots/Spot_Wind_Gust_Sinister_Whisper_Tone_32}}
 {id=WavePlayer_81, properties {soundWaveAssetPtr=/AmbientAudioContent/Waves/OneShots/Spot_Wind_Gust_Sinister_Whisper_Tone_33}}
 {id=WavePlayer_82, properties {soundWaveAssetPtr=/AmbientAudioContent/Waves/OneShots/Spot_Wind_Gust_Sinister_Whisper_Tone_34}}
 {id=WavePlayer_83, properties {soundWaveAssetPtr=/AmbientAudioContent/Waves/OneShots/Spot_Wind_Gust_Sinister_Whisper_Tone_35}}
 {id=WavePlayer_84, properties {soundWaveAssetPtr=/AmbientAudioContent/Waves/OneShots/Spot_Wind_Gust_Sinister_Whisper_Tone_36}}
 {id=WavePlayer_85, properties {soundWaveAssetPtr=/AmbientAudioContent/Waves/OneShots/Spot_Wind_Gust_Sinister_Whisper_Tone_37}}
 {id=WavePlayer_86, properties {soundWaveAssetPtr=/AmbientAudioContent/Waves/OneShots/Spot_Wind_Gust_Sinister_Whisper_Tone_38}}
 {id=WavePlayer_87, properties {soundWaveAssetPtr=/AmbientAudioContent/Waves/OneShots/Spot_Wind_Gust_Sinister_Whisper_Tone_39}}
 {id=WavePlayer_88, properties {soundWaveAssetPtr=&Spot_Wind_Gust_Sinister_Whisper_Tone_40}}
 {id=WavePlayer_89, properties {soundWaveAssetPtr=&Spot_Wind_Gust_Sinister_Whisper_Tone_41}}
 {id=WavePlayer_90, properties {soundWaveAssetPtr=&Spot_Wind_Gust_Sinister_Whisper_Tone_42}}
 {id=WavePlayer_91, properties {soundWaveAssetPtr=/AmbientAudioContent/Waves/OneShots/Spot_Whispering_Panning_Voices_01}}
 {id=WavePlayer_92, properties {soundWaveAssetPtr=/AmbientAudioContent/Waves/OneShots/Spot_Whispering_Panning_Voices_02}}
 {id=WavePlayer_93, properties {soundWaveAssetPtr=/AmbientAudioContent/Waves/OneShots/Spot_Whispering_Panning_Voices_03}}
 {id=WavePlayer_94, properties {soundWaveAssetPtr=/AmbientAudioContent/Waves/OneShots/Spot_Whispering_Panning_Voices_04}}
 {id=WavePlayer_95, properties {soundWaveAssetPtr=/AmbientAudioContent/Waves/OneShots/Spot_Whispering_Panning_Voices_05}}
 {id=WavePlayer_96, properties {soundWaveAssetPtr=/AmbientAudioContent/Waves/OneShots/Spot_Whispering_Panning_Voices_06}}
 {id=WavePlayer_97, properties {soundWaveAssetPtr=/AmbientAudioContent/Waves/OneShots/Spot_Whispering_Panning_Voices_07}}
 {id=WavePlayer_98, properties {soundWaveAssetPtr=/AmbientAudioContent/Waves/OneShots/Spot_Whispering_Panning_Voices_08}}
 {id=WavePlayer_99, properties {soundWaveAssetPtr=/AmbientAudioContent/Waves/OneShots/Spot_Whispering_Panning_Voices_09}}
]
properties {
 attenuationSettings=/Game/Sounds/Attenuation/Fort_TileAmbient_OneShot_Wider_Att
 bOverrideConcurrency=true
 maxDistance=8000
 soundClassObject=&Fort_Ambient_OneShots
 volumeMultiplier=0.45
 concurrencyOverrides {maxCount=2}
 preEffectBusSends [{}]
}
