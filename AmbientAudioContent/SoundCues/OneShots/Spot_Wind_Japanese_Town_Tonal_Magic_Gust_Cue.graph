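# Sound cue for one-shot tonal, magical wind gusts, named as a spot ambient for a Japanese town and classed as outdoor ambient.
# A random node picks one of five equally weighted gust waves, then a modulator varies pitch and volume slightly.
# All four quality-level inputs share that same chain.
# It plays at reduced volume with wide one-shot attenuation and a distance-based send to the default reverb submix that increases farther away.
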SoundCue /AmbientAudioContent/SoundCues/OneShots/Spot_Wind_Japanese_Town_Tonal_Magic_Gust_Cue
root=QualityLevel_0
nodes [
 {id=Modulator_1, children [Random_0], properties {pitchMax=1.1, volumeMax=1.1, volumeMin=0.9}}
 {id=QualityLevel_0, children [Modulator_1,Modulator_1,Modulator_1,Modulator_1]}
 {id=Random_0, children [WavePlayer_0,WavePlayer_1,WavePlayer_2,WavePlayer_3,WavePlayer_4], properties {weights [1,1,1,1,1]}}
 {id=WavePlayer_0, properties {soundWaveAssetPtr=/AmbientAudioContent/Waves/OneShots/Spot_Town_Wind_Gust_01_Magic}}
 {id=WavePlayer_1, properties {soundWaveAssetPtr=/AmbientAudioContent/Waves/OneShots/Spot_Town_Wind_Gust_02_Magic}}
 {id=WavePlayer_2, properties {soundWaveAssetPtr=/AmbientAudioContent/Waves/OneShots/Spot_Town_Wind_Gust_03_Magic}}
 {id=WavePlayer_3, properties {soundWaveAssetPtr=/AmbientAudioContent/Waves/OneShots/Spot_Town_Wind_Gust_04_Magic}}
 {id=WavePlayer_4, properties {soundWaveAssetPtr=/AmbientAudioContent/Waves/OneShots/Spot_Town_Wind_Gust_05_Magic}}
]
properties {
 attenuationSettings=/Game/Sounds/Attenuation/Fort_TileAmbient_OneShot_Wider_Att
 bEnableBusSends=false
 maxDistance=8000
 soundClassObject=&Fort_Ambient_OneShots_Outdoors
 volumeMultiplier=0.251
 preEffectBusSends [{}]
 soundSubmixSends [{
   maxSendDistance=6000
   minSendDistance=2000
   minSendLevel=0
   soundSubmix=&Fortnite_Reverb_Default
   customSendLevelCurve {editorCurveData {keys [<at 0 0>,<at 1 0.6>]}}
  }]
}
