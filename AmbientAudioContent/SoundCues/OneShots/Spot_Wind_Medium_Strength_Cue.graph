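# Sound cue for a medium-strength sandy desert wind gust, used as an outdoor ambient one-shot.
# A Random node picks among ten equally weighted gust recordings.
# A Modulator above it varies pitch, skewed toward lower values, and slightly varies volume on each play.
# The cue plays at reduced volume through a wider tile-ambient one-shot attenuation asset and the outdoor ambient one-shots sound class.
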SoundCue /AmbientAudioContent/SoundCues/OneShots/Spot_Wind_Medium_Strength_Cue
root=Modulator_1
nodes [
 {id=Modulator_1, children [Random_1], properties {pitchMax=1.1, pitchMin=0.7, volumeMax=1.1, volumeMin=0.9}}
 {
  id=Random_1
  children [WavePlayer_0, WavePlayer_2, WavePlayer_3, WavePlayer_4, WavePlayer_5, WavePlayer_6, WavePlayer_7, WavePlayer_8, WavePlayer_9, WavePlayer_10]
  properties {weights [1,1,1,1,1,1,1,1,1,1]}
 }
 {id=WavePlayer_0, properties {soundWaveAssetPtr=/AmbientAudioContent/Waves/OneShots/Spot_Desert_WindGust_Med_hv_Sandy_10}}
 {id=WavePlayer_10, properties {soundWaveAssetPtr=/AmbientAudioContent/Waves/OneShots/Spot_Desert_WindGust_Med_hv_Sandy_01}}
 {id=WavePlayer_2, properties {soundWaveAssetPtr=/AmbientAudioContent/Waves/OneShots/Spot_Desert_WindGust_Med_hv_Sandy_09}}
 {id=WavePlayer_3, properties {soundWaveAssetPtr=/AmbientAudioContent/Waves/OneShots/Spot_Desert_WindGust_Med_hv_Sandy_08}}
 {id=WavePlayer_4, properties {soundWaveAssetPtr=/AmbientAudioContent/Waves/OneShots/Spot_Desert_WindGust_Med_hv_Sandy_07}}
 {id=WavePlayer_5, properties {soundWaveAssetPtr=/AmbientAudioContent/Waves/OneShots/Spot_Desert_WindGust_Med_hv_Sandy_06}}
 {id=WavePlayer_6, properties {soundWaveAssetPtr=/AmbientAudioContent/Waves/OneShots/Spot_Desert_WindGust_Med_hv_Sandy_05}}
 {id=WavePlayer_7, properties {soundWaveAssetPtr=/AmbientAudioContent/Waves/OneShots/Spot_Desert_WindGust_Med_hv_Sandy_04}}
 {id=WavePlayer_8, properties {soundWaveAssetPtr=/AmbientAudioContent/Waves/OneShots/Spot_Desert_WindGust_Med_hv_Sandy_03}}
 {id=WavePlayer_9, properties {soundWaveAssetPtr=/AmbientAudioContent/Waves/OneShots/Spot_Desert_WindGust_Med_hv_Sandy_02}}
]
properties {
 attenuationSettings=/Game/Sounds/Attenuation/Fort_TileAmbient_OneShot_Wider_Att
 maxDistance=8000
 soundClassObject=&Fort_Ambient_OneShots_Outdoors
 volumeMultiplier=0.301
 preEffectBusSends [{}]
}
