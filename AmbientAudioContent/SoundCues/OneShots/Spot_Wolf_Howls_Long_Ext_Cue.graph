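# Sound cue for outdoor ambient one-shot long wolf howls, using a wide tile-ambient attenuation and an outdoor one-shot sound class.
# A random node picks among three haunted-house wolf waves and one mountain howl, each balanced by its own mixer level, with the haunted-house waves pitched down.
# All quality levels share that random node, and a root modulator varies pitch and volume.
# Distance-based sends feed an exterior source bus and distant reverb submix.
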SoundCue /AmbientAudioContent/SoundCues/OneShots/Spot_Wolf_Howls_Long_Ext_Cue
root=Modulator_0
nodes [
 {id=Mixer_0, children [Modulator_1], properties {inputVolume [0.2]}}
 {id=Mixer_1, children [WavePlayer_14], properties {inputVolume [1.2]}}
 {id=Mixer_2, children [Modulator_2], properties {inputVolume [0.3]}}
 {id=Mixer_3, children [Modulator_3], properties {inputVolume [0.33]}}
 {id=Modulator_0, children [QualityLevel_0], properties {pitchMax=1.1, pitchMin=0.8, volumeMax=1.1}}
 {id=Modulator_1, children [WavePlayer_3], properties {pitchMax=0.9, pitchMin=0.9}}
 {id=Modulator_2, children [WavePlayer_4], properties {pitchMax=0.9, pitchMin=0.9}}
 {id=Modulator_3, children [WavePlayer_7], properties {pitchMax=0.9, pitchMin=0.9}}
 {id=QualityLevel_0, children [Random_1,Random_1,Random_1,Random_1]}
 {id=Random_1, children [Mixer_0,Mixer_2,Mixer_3,Mixer_1], properties {weights [1,1,1,1]}}
 {id=WavePlayer_14, properties {soundWaveAssetPtr=/AmbientAudioContent/Waves/OneShots/Spot_Mountain_Wolf_Howls_06}}
 {id=WavePlayer_3, properties {soundWaveAssetPtr=/AmbientAudioContent/Waves/OneShots/Spot_HauntedHouse_Wolf_01}}
 {id=WavePlayer_4, properties {soundWaveAssetPtr=/AmbientAudioContent/Waves/OneShots/Spot_HauntedHouse_Wolf_02}}
 {id=WavePlayer_7, properties {soundWaveAssetPtr=/AmbientAudioContent/Waves/OneShots/Spot_HauntedHouse_Wolf_05}}
]
properties {
 attenuationSettings=/Game/Sounds/Attenuation/Fort_TileAmbient_OneShot_Wider_Att
 maxDistance=8000
 soundClassObject=&Fort_Ambient_OneShots_Outdoors
 volumeMultiplier=0.702
 busSends [{
   audioBus=null
   maxSendDistance=10000
   minSendDistance=3000
   minSendLevel=0
   sendLevel=1
   soundSourceBus=&AmbientAudio_Exterior_SourceBus
   customSendLevelCurve {editorCurveData {keys [<at 0 0.4>,<at 1 1>]}}
  }]
 soundSubmixSends [{
   maxSendDistance=10000
   maxSendLevel=0.6
   minSendDistance=3000
   minSendLevel=0.2
   sendLevel=0
   soundSubmix=&Fortnite_Reverb_Distant
   customSendLevelCurve {editorCurveData {keys [<at 0 0.4>,<at 1 1>]}}
  }]
}
