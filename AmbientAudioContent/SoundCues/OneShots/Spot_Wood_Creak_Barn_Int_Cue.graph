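# One-shot ambient Sound Cue for interior barn wood creaks, built from wooden bridge corridor creak recordings.
# A Random node picks with equal weight among nine creak wave players, and a Modulator at the root varies the result.
# It plays at reduced volume through a distant-creak attenuation asset and the indoor ambient one-shots sound class.
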SoundCue /AmbientAudioContent/SoundCues/OneShots/Spot_Wood_Creak_Barn_Int_Cue
root=Modulator_0
nodes [
 {id=Modulator_0, children [Random_2]}
 {
  id=Random_2
  children [WavePlayer_13, WavePlayer_14, WavePlayer_15, WavePlayer_16, WavePlayer_17, WavePlayer_18, WavePlayer_19, WavePlayer_20, WavePlayer_21]
  properties {weights [1,1,1,1,1,1,1,1,1]}
 }
 {id=WavePlayer_13, properties {soundWaveAssetPtr=/AmbientAudioContent/Waves/OneShots/Spot_Bridge_Wooden_Corridor_Creaks_01}}
 {id=WavePlayer_14, properties {soundWaveAssetPtr=/AmbientAudioContent/Waves/OneShots/Spot_Bridge_Wooden_Corridor_Creaks_02}}
 {id=WavePlayer_15, properties {soundWaveAssetPtr=/AmbientAudioContent/Waves/OneShots/Spot_Bridge_Wooden_Corridor_Creaks_03}}
 {id=WavePlayer_16, properties {soundWaveAssetPtr=/AmbientAudioContent/Waves/OneShots/Spot_Bridge_Wooden_Corridor_Creaks_04}}
 {id=WavePlayer_17, properties {soundWaveAssetPtr=/AmbientAudioContent/Waves/OneShots/Spot_Bridge_Wooden_Corridor_Creaks_05}}
 {id=WavePlayer_18, properties {soundWaveAssetPtr=/AmbientAudioContent/Waves/OneShots/Spot_Bridge_Wooden_Corridor_Creaks_06}}
 {id=WavePlayer_19, properties {soundWaveAssetPtr=/AmbientAudioContent/Waves/OneShots/Spot_Bridge_Wooden_Corridor_Creaks_07}}
 {id=WavePlayer_20, properties {soundWaveAssetPtr=/AmbientAudioContent/Waves/OneShots/Spot_Bridge_Wooden_Corridor_Creaks_08}}
 {id=WavePlayer_21, properties {soundWaveAssetPtr=/AmbientAudioContent/Waves/OneShots/Spot_Bridge_Wooden_Corridor_Creaks_09}}
]
properties {
 attenuationSettings=/Game/Sounds/Attenuation/Fort_TileAmbient_OneShot_Distant_Creak_Att
 maxDistance=8200
 soundClassObject=/Game/Sounds/FortSoundClassesAndModes/Fort_Ambient_OneShots_Indoors
 volumeMultiplier=0.501
 preEffectBusSends [{}]
}
