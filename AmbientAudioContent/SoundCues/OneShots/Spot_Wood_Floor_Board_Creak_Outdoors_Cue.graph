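# Sound cue for one-shot wooden floorboard creaks, named for outdoor ambient spots but using an interior floor cone attenuation asset.
# A Random node picks among thirteen equally weighted creak waves, feeding a Modulator that varies pitch and volume.
# It plays in the ambient one-shot sound class at reduced volume, with a concurrency override capping simultaneous voices.
# Distance-based sends with custom curves feed default and distant reverb submixes.
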SoundCue /AmbientAudioContent/SoundCues/OneShots/Spot_Wood_Floor_Board_Creak_Outdoors_Cue
root=Modulator_0
nodes [
 {id=Modulator_0, children [Random_0], properties {pitchMax=1.1, pitchMin=0.8, volumeMin=0.85}}
 {
  id=Random_0
  children [
   WavePlayer_0
   WavePlayer_1
   WavePlayer_2
   WavePlayer_3
   WavePlayer_4
   WavePlayer_5
   WavePlayer_6
   WavePlayer_7
   WavePlayer_21
   WavePlayer_22
   WavePlayer_23
   WavePlayer_24
   WavePlayer_25
  ]
  properties {weights [1,1,1,1,1,1,1,1,1,1,1,1,1]}
 }
 {id=WavePlayer_0, properties {soundWaveAssetPtr=/AmbientAudioContent/Waves/OneShots/Spot_Wood_Floor_Board_Creak_14}}
 {id=WavePlayer_1, properties {soundWaveAssetPtr=/AmbientAudioContent/Waves/OneShots/Spot_Wood_Floor_Board_Creak_12}}
 {id=WavePlayer_2, properties {soundWaveAssetPtr=/AmbientAudioContent/Waves/OneShots/Spot_Wood_Floor_Board_Creak_11}}
 {id=WavePlayer_21, properties {soundWaveAssetPtr=/AmbientAudioContent/Waves/OneShots/Spot_Wood_Floor_Board_Creak_05}}
 {id=WavePlayer_22, properties {soundWaveAssetPtr=/AmbientAudioContent/Waves/OneShots/Spot_Wood_Floor_Board_Creak_04}}
 {id=WavePlayer_23, properties {soundWaveAssetPtr=/AmbientAudioContent/Waves/OneShots/Spot_Wood_Floor_Board_Creak_03}}
 {id=WavePlayer_24, properties {soundWaveAssetPtr=/AmbientAudioContent/Waves/OneShots/Spot_Wood_Floor_Board_Creak_02}}
 {id=WavePlayer_25, properties {soundWaveAssetPtr=/AmbientAudioContent/Waves/OneShots/Spot_Wood_Floor_Board_Creak_01}}
 {id=WavePlayer_3, properties {soundWaveAssetPtr=/AmbientAudioContent/Waves/OneShots/Spot_Wood_Floor_Board_Creak_10}}
 {id=WavePlayer_4, properties {soundWaveAssetPtr=/AmbientAudioContent/Waves/OneShots/Spot_Wood_Floor_Board_Creak_09}}
 {id=WavePlayer_5, properties {soundWaveAssetPtr=/AmbientAudioContent/Waves/OneShots/Spot_Wood_Floor_Board_Creak_08}}
 {id=WavePlayer_6, properties {soundWaveAssetPtr=/AmbientAudioContent/Waves/OneShots/Spot_Wood_Floor_Board_Creak_07}}
 {id=WavePlayer_7, properties {soundWaveAssetPtr=/AmbientAudioContent/Waves/OneShots/Spot_Wood_Floor_Board_Creak_06}}
]
properties {
 attenuationSettings=/AmbientAudioBase/Attenuations/Fort_Oneshot_Wood_Creak_Int_Floor_Cone_Att
 bOverrideConcurrency=true
 maxDistance=1074
 soundClassObject=&Fort_Ambient_OneShots
 volumeMultiplier=0.501
 concurrencyOverrides {maxCount=2, voiceStealReleaseTime=2}
 soundSubmixSends [
  {
   maxSendDistance=8500
   minSendDistance=200
   minSendLevel=0.6
   soundSubmix=&Fortnite_Reverb_Default
   customSendLevelCurve {editorCurveData {keys [<at 0 0.25>,<at 1 1>]}}
  }
  {
   maxSendDistance=8500
   maxSendLevel=0.6
   minSendDistance=200
   minSendLevel=0.25
   soundSubmix=&Fortnite_Reverb_Distant
   customSendLevelCurve {editorCurveData {keys [
      <at 0 0>
      {arriveTangent=1, interpMode=RCIM_Cubic, leaveTangent=1, time=0.66, value=0.33}
      {arriveTangent=2.1868, interpMode=RCIM_Cubic, leaveTangent=2.1868, tangentMode=RCTM_User, time=1, value=1}
     ]}}
  }
 ]
}
