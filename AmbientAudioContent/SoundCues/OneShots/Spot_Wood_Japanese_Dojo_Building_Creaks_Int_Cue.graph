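# Sound Cue for one-shot interior wooden building creaks in a Japanese dojo, used as spot ambience in the indoor ambient sound class.
# An equally weighted Random node picks one of nine creak waves, feeding all four Quality Level inputs, then a Modulator adds slight pitch variation.
# It references a shared distant-creak attenuation asset and a two-voice concurrency set, and sends to the default reverb submix.
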SoundCue /AmbientAudioContent/SoundCues/OneShots/Spot_Wood_Japanese_Dojo_Building_Creaks_Int_Cue
root=Modulator_0
nodes [
 {id=Modulator_0, children [QualityLevel_0], properties {pitchMax=1.1, pitchMin=0.9}}
 {id=QualityLevel_0, children [Random_0,Random_0,Random_0,Random_0]}
 {
  id=Random_0
  children [WavePlayer_0,WavePlayer_1,WavePlayer_2,WavePlayer_3,WavePlayer_4,WavePlayer_5,WavePlayer_6,WavePlayer_7,WavePlayer_8]
  properties {weights [1,1,1,1,1,1,1,1,1]}
 }
 {id=WavePlayer_0, properties {soundWaveAssetPtr=/AmbientAudioContent/Waves/OneShots/Spot_Japanese_Wood_Building_Creak_01}}
 {id=WavePlayer_1, properties {soundWaveAssetPtr=/AmbientAudioContent/Waves/OneShots/Spot_Japanese_Wood_Building_Creak_08}}
 {id=WavePlayer_2, properties {soundWaveAssetPtr=/AmbientAudioContent/Waves/OneShots/Spot_Japanese_Wood_Building_Creak_02}}
 {id=WavePlayer_3, properties {soundWaveAssetPtr=/AmbientAudioContent/Waves/OneShots/Spot_Japanese_Wood_Building_Creak_03}}
 {id=WavePlayer_4, properties {soundWaveAssetPtr=/AmbientAudioContent/Waves/OneShots/Spot_Japanese_Wood_Building_Creak_04}}
 {id=WavePlayer_5, properties {soundWaveAssetPtr=/AmbientAudioContent/Waves/OneShots/Spot_Japanese_Wood_Building_Creak_05}}
 {id=WavePlayer_6, properties {soundWaveAssetPtr=/AmbientAudioContent/Waves/OneShots/Spot_Japanese_Wood_Building_Creak_06}}
 {id=WavePlayer_7, properties {soundWaveAssetPtr=/AmbientAudioContent/Waves/OneShots/Spot_Japanese_Wood_Building_Creak_07}}
 {id=WavePlayer_8, properties {soundWaveAssetPtr=/AmbientAudioContent/Waves/OneShots/Spot_Japanese_Wood_Building_Creak_09}}
]
properties {
 attenuationSettings=/Game/Sounds/Attenuation/Fort_TileAmbient_OneShot_Distant_Creak_Att
 maxDistance=8200
 soundClassObject=/Game/Sounds/FortSoundClassesAndModes/Fort_Ambient_OneShots_Indoors
 volumeMultiplier=0.7
 attenuationOverrides {
  attenuationShape=Capsule
  bAttenuate=false
  bEnableFocusInterpolation=true
  bEnableListenerFocus=true
  distanceAlgorithm=Custom
  falloffDistance=3760
  focusAzimuth=25
  focusReleaseInterpSpeed=0.4
  nonFocusAzimuth=55
  nonFocusVolumeAttenuation=0.4
  nonSpatializedRadiusEnd=1500
  nonSpatializedRadiusStart=2000
  stereoSpread=100
  attenuationShapeExtents {x=2350}
  customAttenuationCurve {editorCurveData {keys [
     <at 0.2 0>
     <at 0.35 1>
     {arriveTangent=-0.8696, interpMode=RCIM_Cubic, leaveTangent=-0.8696, time=0.8, value=1}
     {arriveTangent=-0.0046, interpMode=RCIM_Cubic, leaveTangent=-0.0046, tangentMode=RCTM_User, time=1.5, value=0}
    ]}}
 }
 concurrencyOverrides {maxCount=2, voiceStealReleaseTime=2}
 concurrencySet [/AmbientAudioBase/Concurrency/AmbientConc_2Max_Spot_Concurrency]
 soundSubmixSends [{minSendLevel=0.6, soundSubmix=&Fortnite_Reverb_Default}]
}
