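# Sound cue that plays one-shot indoor wooden creaks as an ambient spot sound, using attenuation and an indoor ambient sound class.
# A top modulator feeds a quality switch: the first two levels pick from eleven creaks with replacement, the last two from three without replacement.
# Each branch adds its own random pitch and volume variation.
# A concurrency override caps playback at two instances.
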SoundCue /AmbientAudioContent/SoundCues/OneShots/Spot_Wooden_Creaks_Int_Cue
root=Modulator_0
nodes [
 {id=Modulator_0, children [QualityLevel_0], properties {pitchMin=0.85, volumeMax=1.2, volumeMin=1}}
 {id=Modulator_1, children [Random_0], properties {pitchMax=1.1, pitchMin=0.8, volumeMax=0.8, volumeMin=0.3}}
 {id=Modulator_2, children [Random_1], properties {pitchMax=1.1, pitchMin=0.8, volumeMax=1, volumeMin=0.3}}
 {id=QualityLevel_0, children [Modulator_1,Modulator_1,Modulator_2,Modulator_2]}
 {
  id=Random_0
  children [
   WavePlayer_2
   WavePlayer_3
   WavePlayer_4
   WavePlayer_5
   WavePlayer_6
   WavePlayer_7
   WavePlayer_8
   WavePlayer_9
   WavePlayer_10
   WavePlayer_11
   WavePlayer_12
  ]
  properties {bRandomizeWithoutReplacement=false, weights [0.5,0.5,0.5,0.5,0.5,0.5,0.5,0.5,0.5,0.5,0.5]}
 }
 {id=Random_1, children [WavePlayer_0, WavePlayer_1, WavePlayer_20], properties {weights [1,1,1]}}
 {id=WavePlayer_0, properties {soundWaveAssetPtr=/Game/Sounds/Fort_Ambient_NonLoop/Wood/wood_creak_medium_01}}
 {id=WavePlayer_1, properties {soundWaveAssetPtr=/Game/Sounds/Fort_Ambient_NonLoop/Wood/wood_creak_medium_03}}
 {id=WavePlayer_10, properties {soundWaveAssetPtr=/Game/Sounds/Fort_Ambient_NonLoop/Wood/wood_creak_medium_06}}
 {id=WavePlayer_11, properties {soundWaveAssetPtr=/Game/Sounds/Fort_Ambient_NonLoop/Wood/wood_creak_medium_07}}
 {id=WavePlayer_12, properties {soundWaveAssetPtr=/Game/Sounds/Fort_Ambient_NonLoop/Wood/wood_creak_medium_08}}
 {id=WavePlayer_2, properties {soundWaveAssetPtr=/Game/Sounds/Fort_Ambient_NonLoop/Wood/wood_creak_medium_09}}
 {id=WavePlayer_20, properties {soundWaveAssetPtr=&wood_creak_medium_06}}
 {id=WavePlayer_3, properties {soundWaveAssetPtr=/Game/Sounds/Fort_Ambient_NonLoop/Wood/wood_creak_medium_10}}
 {id=WavePlayer_4, properties {soundWaveAssetPtr=/Game/Sounds/Fort_Ambient_NonLoop/Wood/wood_creak_medium_11}}
 {id=WavePlayer_5, properties {soundWaveAssetPtr=&wood_creak_medium_01}}
 {id=WavePlayer_6, properties {soundWaveAssetPtr=/Game/Sounds/Fort_Ambient_NonLoop/Wood/wood_creak_medium_02}}
 {id=WavePlayer_7, properties {soundWaveAssetPtr=&wood_creak_medium_03}}
 {id=WavePlayer_8, properties {soundWaveAssetPtr=/Game/Sounds/Fort_Ambient_NonLoop/Wood/wood_creak_medium_04}}
 {id=WavePlayer_9, properties {soundWaveAssetPtr=/Game/Sounds/Fort_Ambient_NonLoop/Wood/wood_creak_medium_05}}
]
properties {
 attenuationSettings=/Game/Sounds/Artemis/POI/Chrome_BugleBay/Spot/Fort_Oneshot_Wood_Creak_Int_Att
 bOverrideConcurrency=true
 maxDistance=850
 soundClassObject=/Game/Sounds/FortSoundClassesAndModes/Fort_Ambient_OneShots_Indoors
 volumeMultiplier=1
 concurrencyOverrides {maxCount=2, voiceStealReleaseTime=2}
 preEffectBusSends [{}]
}
