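# Sound cue for one-shot wooden stress creaks, used as spot ambient audio.
# A root Random node picks among ten equally weighted wave players, each holding a different primitive-structure wood creak recording.
# Cue-level multipliers lower pitch and volume, while an exterior wood-creak attenuation asset handles spatial falloff.
# Output is assigned to the indoor ambient Athena sound class.
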SoundCue /AmbientAudioContent/SoundCues/OneShots/Spot_Wooden_Stress_Creak_Oneshot_Cue
root=Random_0
nodes [
 {
  id=Random_0
  children [WavePlayer_0,WavePlayer_1,WavePlayer_2,WavePlayer_3,WavePlayer_4,WavePlayer_5,WavePlayer_6,WavePlayer_7,WavePlayer_8,WavePlayer_9]
  properties {weights [1,1,1,1,1,1,1,1,1,1]}
 }
 {id=WavePlayer_0, properties {soundWaveAssetPtr=/AmbientAudioContent/Waves/OneShots/Spot_Primitive_Structure_WoodCreaks_01}}
 {id=WavePlayer_1, properties {soundWaveAssetPtr=/AmbientAudioContent/Waves/OneShots/Spot_Primitive_Structure_WoodCreaks_02}}
 {id=WavePlayer_2, properties {soundWaveAssetPtr=/AmbientAudioContent/Waves/OneShots/Spot_Primitive_Structure_WoodCreaks_03}}
 {id=WavePlayer_3, properties {soundWaveAssetPtr=/AmbientAudioContent/Waves/OneShots/Spot_Primitive_Structure_WoodCreaks_04}}
 {id=WavePlayer_4, properties {soundWaveAssetPtr=/AmbientAudioContent/Waves/OneShots/Spot_Primitive_Structure_WoodCreaks_05}}
 {id=WavePlayer_5, properties {soundWaveAssetPtr=/AmbientAudioContent/Waves/OneShots/Spot_Primitive_Structure_WoodCreaks_06}}
 {id=WavePlayer_6, properties {soundWaveAssetPtr=/AmbientAudioContent/Waves/OneShots/Spot_Primitive_Structure_WoodCreaks_07}}
 {id=WavePlayer_7, properties {soundWaveAssetPtr=/AmbientAudioContent/Waves/OneShots/Spot_Primitive_Structure_WoodCreaks_08}}
 {id=WavePlayer_8, properties {soundWaveAssetPtr=/AmbientAudioContent/Waves/OneShots/Spot_Primitive_Structure_WoodCreaks_09}}
 {id=WavePlayer_9, properties {soundWaveAssetPtr=/AmbientAudioContent/Waves/OneShots/Spot_Primitive_Structure_WoodCreaks_10}}
]
properties {
 attenuationSettings=/Game/Sounds/Artemis/POI/Chrome_BugleBay/Spot/Fort_Oneshot_Wood_Creak_Ext_Att
 maxDistance=2250
 pitchMultiplier=0.7
 soundClassObject=&Fort_Ambient_Indoors_Athena
 volumeMultiplier=0.7
 preEffectBusSends [{}]
}
