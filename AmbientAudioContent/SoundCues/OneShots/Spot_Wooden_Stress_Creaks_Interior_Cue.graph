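# Sound cue for intermittent wooden stress creaks, played as an indoor ambient spot one-shot.
# An endless loop wraps a pitch-varying modulator and a random delay of several seconds, then picks one of five equally weighted creak waves.
# All four quality-level branches share the same random node.
# It uses interior creak attenuation, the indoor ambient sound class, reduced volume and a concurrency override capping simultaneous instances at two.
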SoundCue /AmbientAudioContent/SoundCues/OneShots/Spot_Wooden_Stress_Creaks_Interior_Cue
root=Looping_0
nodes [
 {id=Delay_0, children [QualityLevel_0], properties {delayMax=12, delayMin=3}}
 {id=Looping_0, children [Modulator_0]}
 {id=Modulator_0, children [Delay_0], properties {pitchMin=0.85}}
 {id=QualityLevel_0, children [Random_0,Random_0,Random_0,Random_0]}
 {id=Random_0, children [WavePlayer_3,WavePlayer_4,WavePlayer_5,WavePlayer_8,WavePlayer_9], properties {weights [1,1,1,1,1]}}
 {id=WavePlayer_3, properties {soundWaveAssetPtr=/AmbientAudioContent/Waves/OneShots/Spot_Wood_Creak_Rnd_01}}
 {id=WavePlayer_4, properties {soundWaveAssetPtr=/AmbientAudioContent/Waves/OneShots/Spot_Wood_Creak_Rnd_02}}
 {id=WavePlayer_5, properties {soundWaveAssetPtr=/AmbientAudioContent/Waves/OneShots/Spot_Wood_Creak_Rnd_03}}
 {id=WavePlayer_8, properties {soundWaveAssetPtr=/AmbientAudioContent/Waves/OneShots/Spot_Wood_Creak_Rnd_04}}
 {id=WavePlayer_9, properties {soundWaveAssetPtr=/AmbientAudioContent/Waves/OneShots/Spot_Wood_Creak_Rnd_05}}
]
properties {
 attenuationSettings=/Game/Sounds/Artemis/POI/Chrome_BugleBay/Spot/Fort_Oneshot_Wood_Creak_Int_Att
 bEnableBusSends=false
 bOverrideConcurrency=true
 maxDistance=850
 soundClassObject=&Fort_Ambient_Indoors_Athena
 volumeMultiplier=0.4
 concurrencyOverrides {maxCount=2}
 soundSubmixSends [{minSendLevel=0, sendLevel=0, soundSubmix=null}]
}
