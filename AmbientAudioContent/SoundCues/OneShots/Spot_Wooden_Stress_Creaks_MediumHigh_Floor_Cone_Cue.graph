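# Looping ambient one-shot cue that plays wooden stress creaks at random intervals, using an interior floor cone attenuation asset.
# A Looping node repeats a random Delay into a Quality Level switch that scales variety from eleven creaks to three to one shared single wave.
# Each branch randomizes volume and pitch, with the multi-wave branches pitched up.
# Concurrency allows two instances and blocks new ones beyond that.
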SoundCue /AmbientAudioContent/SoundCues/OneShots/Spot_Wooden_Stress_Creaks_MediumHigh_Floor_Cone_Cue
root=Looping_0
nodes [
 {id=Delay_0, children [QualityLevel_0], properties {delayMax=18, delayMin=3}}
 {id=Looping_0, children [Delay_0]}
 {id=Modulator_0, children [WavePlayer_29], properties {pitchMax=1.1, pitchMin=0.8, volumeMax=1, volumeMin=0.3}}
 {id=Modulator_1, children [Random_3], properties {pitchMax=1.6, pitchMin=1.4, volumeMax=1, volumeMin=0.3}}
 {id=Modulator_2, children [Random_2], properties {pitchMax=1.6, pitchMin=1.4, volumeMax=1, volumeMin=0.3}}
 {id=QualityLevel_0, children [Modulator_1,Modulator_2,Modulator_0,Modulator_0]}
 {id=Random_2, children [WavePlayer_10, WavePlayer_15, WavePlayer_1], properties {weights [1,1,1]}}
 {
  id=Random_3
  children [
   WavePlayer_18
   WavePlayer_19
   WavePlayer_20
   WavePlayer_21
   WavePlayer_22
   WavePlayer_23
   WavePlayer_24
   WavePlayer_25
   WavePlayer_26
   WavePlayer_27
   WavePlayer_28
  ]
  properties {weights [1,1,1,1,1,1,1,1,1,1,1]}
 }
 {id=WavePlayer_1, properties {soundWaveAssetPtr=/AmbientAudioContent/Waves/OneShots/Spot_wood_creak_medium_03}}
 {id=WavePlayer_10, properties {soundWaveAssetPtr=/AmbientAudioContent/Waves/OneShots/Spot_wood_creak_medium_01}}
 {id=WavePlayer_15, properties {soundWaveAssetPtr=/AmbientAudioContent/Waves/OneShots/Spot_wood_creak_medium_06}}
 {id=WavePlayer_18, properties {soundWaveAssetPtr=/AmbientAudioContent/Waves/OneShots/Spot_wood_creak_medium_11}}
 {id=WavePlayer_19, properties {soundWaveAssetPtr=/AmbientAudioContent/Waves/OneShots/Spot_wood_creak_medium_10}}
 {id=WavePlayer_20, properties {soundWaveAssetPtr=/AmbientAudioContent/Waves/OneShots/Spot_wood_creak_medium_09}}
 {id=WavePlayer_21, properties {soundWaveAssetPtr=/AmbientAudioContent/Waves/OneShots/Spot_wood_creak_medium_08}}
 {id=WavePlayer_22, properties {soundWaveAssetPtr=/AmbientAudioContent/Waves/OneShots/Spot_wood_creak_medium_07}}
 {id=WavePlayer_23, properties {soundWaveAssetPtr=&Spot_wood_creak_medium_06}}
 {id=WavePlayer_24, properties {soundWaveAssetPtr=/AmbientAudioContent/Waves/OneShots/Spot_wood_creak_medium_05}}
 {id=WavePlayer_25, properties {soundWaveAssetPtr=/AmbientAudioContent/Waves/OneShots/Spot_wood_creak_medium_04}}
 {id=WavePlayer_26, properties {soundWaveAssetPtr=&Spot_wood_creak_medium_03}}
 {id=WavePlayer_27, properties {soundWaveAssetPtr=/AmbientAudioContent/Waves/OneShots/Spot_wood_creak_medium_02}}
 {id=WavePlayer_28, properties {soundWaveAssetPtr=&Spot_wood_creak_medium_01}}
 {id=WavePlayer_29, properties {soundWaveAssetPtr=&Spot_wood_creak_medium_01}}
]
properties {
 attenuationSettings=/Game/Sounds/Artemis/POI/Chrome_BugleBay/Spot/Fort_Oneshot_Wood_Creak_Int_Floor_Cone_Att
 bOverrideConcurrency=true
 maxDistance=1074
 soundClassObject=&Fort_Doors
 volumeMultiplier=1
 attenuationOverrides {falloffDistance=1000, attenuationShapeExtents {x=50}}
 concurrencyOverrides {maxCount=2, resolutionRule=PreventNew}
}
