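# Sound Cue for distant mechanical-movement ambient one-shots, routed to the indoor ambient one-shot sound class.
# A Random node picks among eight equally weighted stereo workshop waves, then a Modulator varies pitch at constant volume.
# Overrides set logarithmic falloff with distance-based filtering and cap playback at two instances, stopping the farthest first.
# A custom curve raises the distant reverb submix send only over the outer half of its range.
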SoundCue /AmbientAudioContent/SoundCues/OneShots/Spots_Mechanical_Movements_Distant_Cue
root=Modulator_0
nodes [
 {id=Modulator_0, children [Random_0], properties {pitchMin=0.8, volumeMax=1, volumeMin=1}}
 {
  id=Random_0
  children [WavePlayer_8, WavePlayer_9, WavePlayer_10, WavePlayer_11, WavePlayer_12, WavePlayer_13, WavePlayer_14, WavePlayer_15]
  properties {weights [1,1,1,1,1,1,1,1]}
 }
 {id=WavePlayer_10, properties {soundWaveAssetPtr=/AmbientAudioContent/Waves/OneShots/Spot_SevenWorkshop_Mechanical_Movements_01_Stereo}}
 {id=WavePlayer_11, properties {soundWaveAssetPtr=/AmbientAudioContent/Waves/OneShots/Spot_SevenWorkshop_Mechanical_Movements_08_Stereo}}
 {id=WavePlayer_12, properties {soundWaveAssetPtr=/AmbientAudioContent/Waves/OneShots/Spot_SevenWorkshop_Mechanical_Movements_06_Stereo}}
 {id=WavePlayer_13, properties {soundWaveAssetPtr=/AmbientAudioContent/Waves/OneShots/Spot_SevenWorkshop_Mechanical_Movements_05_Stereo}}
 {id=WavePlayer_14, properties {soundWaveAssetPtr=/AmbientAudioContent/Waves/OneShots/Spot_SevenWorkshop_Mechanical_Movements_04_Stereo}}
 {id=WavePlayer_15, properties {soundWaveAssetPtr=/AmbientAudioContent/Waves/OneShots/Spot_SevenWorkshop_Mechanical_Movements_02_Stereo}}
 {id=WavePlayer_8, properties {soundWaveAssetPtr=/AmbientAudioContent/Waves/OneShots/Spot_SevenWorkshop_Mechanical_Movements_07_Stereo}}
 {id=WavePlayer_9, properties {soundWaveAssetPtr=/AmbientAudioContent/Waves/OneShots/Spot_SevenWorkshop_Mechanical_Movements_03_Stereo}}
]
properties {
 bOverrideAttenuation=true
 bOverrideConcurrency=true
 maxDistance=8000
 soundClassObject=/Game/Sounds/FortSoundClassesAndModes/Fort_Ambient_OneShots_Indoors
 volumeMultiplier=1
 attenuationOverrides {
  bAttenuateWithLPF=true
  bEnableLogFrequencyScaling=true
  distanceAlgorithm=Logarithmic
  falloffDistance=8000
  hPFFrequencyAtMax=300
  lPFFrequencyAtMax=500
  lPFRadiusMax=8000
  attenuationShapeExtents {x=0}
 }
 concurrencyOverrides {maxCount=2, resolutionRule=StopFarthestThenPreventNew}
 soundSubmixSends [{
   maxSendDistance=8000
   minSendLevel=0
   sendLevelControlMethod=CustomCurve
   soundSubmix=&Fortnite_Reverb_Distant
   customSendLevelCurve {editorCurveData {keys [<at 0 0>,<at 0.5 0>,<at 1 0.6>]}}
  }]
}
