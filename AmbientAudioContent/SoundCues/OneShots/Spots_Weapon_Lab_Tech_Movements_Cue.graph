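# Sound cue for one-shot indoor ambient spots of tech movement in a weapon lab.
# A Random node picks one of nine equally weighted SevenWorkshop stereo waves, feeding a Modulator that varies pitch while holding volume fixed.
# The cue lowers base pitch, boosts volume, uses a less-positional ambient attenuation asset, and overrides concurrency to cap simultaneous plays at two.
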SoundCue /AmbientAudioContent/SoundCues/OneShots/Spots_Weapon_Lab_Tech_Movements_Cue
root=Modulator_0
nodes [
 {id=Modulator_0, children [Random_1], properties {pitchMin=0.8, volumeMax=1, volumeMin=1}}
 {
  id=Random_1
  children [WavePlayer_0,WavePlayer_1,WavePlayer_2,WavePlayer_3,WavePlayer_4,WavePlayer_5,WavePlayer_6,WavePlayer_7,WavePlayer_8]
  properties {weights [1,1,1,1,1,1,1,1,1]}
 }
 {id=WavePlayer_0, properties {soundWaveAssetPtr=/AmbientAudioContent/Waves/OneShots/Spot_SevenWorkshop_Tech_Movements_01_Stereo}}
 {id=WavePlayer_1, properties {soundWaveAssetPtr=/AmbientAudioContent/Waves/OneShots/Spot_SevenWorkshop_Tech_Movements_02_Stereo}}
 {id=WavePlayer_2, properties {soundWaveAssetPtr=/AmbientAudioContent/Waves/OneShots/Spot_SevenWorkshop_Tech_Movements_03_Stereo}}
 {id=WavePlayer_3, properties {soundWaveAssetPtr=/AmbientAudioContent/Waves/OneShots/Spot_SevenWorkshop_Tech_Movements_04_Stereo}}
 {id=WavePlayer_4, properties {soundWaveAssetPtr=/AmbientAudioContent/Waves/OneShots/Spot_SevenWorkshop_Tech_Movements_05_Stereo}}
 {id=WavePlayer_5, properties {soundWaveAssetPtr=/AmbientAudioContent/Waves/OneShots/Spot_SevenWorkshop_Tech_Movements_06_Stereo}}
 {id=WavePlayer_6, properties {soundWaveAssetPtr=/AmbientAudioContent/Waves/OneShots/Spot_SevenWorkshop_Tech_Movements_07_Stereo}}
 {id=WavePlayer_7, properties {soundWaveAssetPtr=/AmbientAudioContent/Waves/OneShots/Spot_SevenWorkshop_Tech_Movements_08_Stereo}}
 {id=WavePlayer_8, properties {soundWaveAssetPtr=/AmbientAudioContent/Waves/OneShots/Spot_SevenWorkshop_Tech_Movements_09_Stereo}}
]
properties {
 attenuationSettings=/Game/Sounds/Attenuation/Fort_TileAmbient_OneShot_LessPositional_Log_Att
 bOverrideConcurrency=true
 maxDistance=6200
 pitchMultiplier=0.8
 soundClassObject=/Game/Sounds/FortSoundClassesAndModes/Fort_Ambient_OneShots_Indoors
 volumeMultiplier=1.8
 attenuationOverrides {falloffDistance=6000, stereoSpread=10000, attenuationShapeExtents {x=2000}}
 concurrencyOverrides {maxCount=2}
}
